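# Editor utility widget for the Stage Manager editor: a window, collapsed by default, for adding or changing sequence pool, sequence and event data assets.
# Toggles offer selecting an existing asset, via a searchable combo box with a preview toggle, or creating a new one.
# Variables hold MojoStage pool, sequence and event assets, plus property and tab IDs.
# The OnAssetChanged dispatcher carries an object, string and new-file flag; the EventGraph is empty.
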EditorUtilityWidgetBlueprint /PilgrimQuickplay/StageManager/NeverCook/EditorUtilities/StageManagerEditor/SparksStageManagerChangeDataWindow
graphs [
 EventGraph {links [], nodes []}
 OnAssetChanged {nodes [OnAssetChanged!() -> (then, ReturnObject:Object Reference, ReturnString:String, IsNewFile:bool=~) {is=title K2Node_FunctionEntry}]}
]
variables [
 CurrentEventSelected {type=string, default=""}
 EventsEnumToStringMap {type="map<string>", default {}}
 FxManager_Event StepData {type=MojoFxManagerEvent_DataAsset_C, default=null}
 IsMidiText {type=boolean}
 MojoStage_SequenceData {type=MojoFxManagerSequence_DataAsset_C, default=null}
 MojoStage_Sequences PoolData {type=BP_MojoStageSequence_DataAsset_C, default=null}
 OnAssetChanged {default=()}
 PropertyName {type=string, default=MojoStage_SequenceData}
 TabID {type=string, default=MojoStage_SequenceData}
 TextsAvaliable {type="array<string>", default []}
]
widgets {
 info {rootWidgetClass=&Overlay, widgetCount=64}
 widgets [
  NewSequenceEventWindow {widgetClassPath=&Overlay, bIsVariable=true, widget=NewSequenceEventWindow, properties {visibility=Collapsed}}
  SizeBox_2 {
   widgetClassPath=&SizeBox
   parent=NewSequenceEventWindow
   widget=SizeBox_2
   properties {bOverride_MinDesiredWidth=true, minDesiredWidth=600}
   slot {class=OverlaySlot, properties {horizontalAlignment=HAlign_Fill, verticalAlignment=VAlign_Fill}}
  }
  Overlay {widgetClassPath=&Overlay, parent=SizeBox_2, widget=Overlay, slot {class=SizeBoxSlot}}
  Image_7 {
   widgetClassPath=&Image
   parent=Overlay
   widget=Image_7
   properties {
    brush {
     bIsDynamicallyLoaded=false
     drawAs=Image
     imageType=NoImage
     mirroring=NoMirror
     resourceName=None
     resourceObject=null
     tiling=NoTile
     imageSize=<32 32>
     margin=<ltrb 0 0 0 0>
     outlineSettings {
      bUseBrushTransparency=false
      roundingType=HalfHeightRadius
      width=0
      color {colorUseRule=UseColor_Specified, specifiedColor=<rgba 0 0 0 0>}
      cornerRadii=<0 0 0 0>
     }
     tintColor {colorUseRule=UseColor_Specified, specifiedColor=<rgba 0 0 0 0.232>}
     uVRegion {bIsValid=false, max=<0 0>, min=<0 0>}
    }
    colorAndOpacity=<rgba 0.026 0.026 0.026 1>
    renderTransform {angle=0, scale=<1.005 1.01>, shear=<0 0>, translation=<0 0.5>}
   }
   slot {class=OverlaySlot, properties {horizontalAlignment=HAlign_Fill, verticalAlignment=VAlign_Fill}}
  }
  Image_8 {
   widgetClassPath=&Image
   parent=Overlay
   widget=Image_8
   properties {colorAndOpacity=<rgba 0.026 0.026 0.026 1>}
   slot {class=OverlaySlot, properties {horizontalAlignment=HAlign_Fill, verticalAlignment=VAlign_Fill}}
  }
  Border {
   widgetClassPath=&Border
   parent=Overlay
   widget=Border
   properties {clipping=ClipToBounds, horizontalAlignment=HAlign_Left, brushColor=<rgba 0.063 0.1406 0.0994 1>, padding=<ltrb 10 2 4 2>}
   slot {class=OverlaySlot, properties {horizontalAlignment=HAlign_Fill}}
  }
  EventsTitleText {
   widgetClassPath=&TextBlock
   bIsVariable=true
   parent=Border
   widget=EventsTitleText
   properties {
    text="Add or Change Sequence Pool, Sequences, Events Window"
    colorAndOpacity {colorUseRule=UseColor_Specified, specifiedColor=<rgba 0.6562 0.6562 0.6562 1>}
    font {
     bForceMonospaced=false
     bMaterialIsStencil=false
     fontMaterial=null
     fontObject=&HeadingNow
     letterSpacing=0
     monospacedWidth=1
     size=11.25
     skewAmount=0
     typefaceFontName=75 Medium
     outlineSettings {
      bApplyOutlineToDropShadows=false
      bMiteredCorners=false
      bSeparateFillAlpha=false
      outlineMaterial=null
      outlineSize=0
      outlineColor=<rgba 0 0 0 1>
     }
    }
   }
   slot {class=BorderSlot, properties {horizontalAlignment=HAlign_Left, padding=<ltrb 10 2 4 2>}}
  }
  SizeBox_5 {
   widgetClassPath=&SizeBox
   parent=Overlay
   widget=SizeBox_5
   properties {bOverride_MinDesiredWidth=true, minDesiredWidth=670}
   slot {class=OverlaySlot, properties {horizontalAlignment=HAlign_Fill, verticalAlignment=VAlign_Fill}}
  }
  NewSequenceEventsInsideGroup {
   widgetClassPath=&Overlay
   bIsVariable=true
   parent=SizeBox_5
   widget=NewSequenceEventsInsideGroup
   properties {clipping=ClipToBounds, visibility=Visible}
   slot {class=SizeBoxSlot}
  }
  VerticalBox_0 {
   widgetClassPath=&VerticalBox
   parent=NewSequenceEventsInsideGroup
   widget=VerticalBox_0
   slot {class=OverlaySlot, properties {horizontalAlignment=HAlign_Fill, verticalAlignment=VAlign_Fill, padding=<ltrb 10 35 10 25>}}
  }
  Border_2 {
   widgetClassPath=&Border
   parent=VerticalBox_0
   widget=Border_2
   properties {
    background {
     bIsDynamicallyLoaded=false
     drawAs=Image
     imageType=NoImage
     mirroring=NoMirror
     resourceName=None
     resourceObject=null
     tiling=NoTile
     imageSize=<32 32>
     margin=<ltrb 0 0 0 0>
     outlineSettings {
      bUseBrushTransparency=false
      roundingType=HalfHeightRadius
      width=0
      color {colorUseRule=UseColor_Specified, specifiedColor=<rgba 0 0 0 0>}
      cornerRadii=<0 0 0 0>
     }
     tintColor {colorUseRule=UseColor_Specified, specifiedColor=<rgba 0.0729 0.0729 0.0729 1>}
     uVRegion {bIsValid=false, max=<0 0>, min=<0 0>}
    }
    padding=<ltrb 4 5 4 5>
   }
   slot {class=VerticalBoxSlot, properties {verticalAlignment=VAlign_Center, padding=<ltrb -10 0 -10 2>}}
  }
  HorizontalBox_10 {
   widgetClassPath=&HorizontalBox
   parent=Border_2
   widget=HorizontalBox_10
   slot {class=BorderSlot, properties {padding=<ltrb 4 5 4 5>}}
  }
  SelectAndChangeToggle {
   widgetClassPath=&EditorUtilityCheckBox
   bIsVariable=true
   parent=HorizontalBox_10
   widget=SelectAndChangeToggle
   properties {
    checkedState=Checked
    toolTipText=Shows the MIDI Text Events
    widgetStyle {
     checkBoxType=ToggleButton
     backgroundHoveredImage {
      bIsDynamicallyLoaded=false
      drawAs=NoDrawType
      imageType=NoImage
      mirroring=NoMirror
      resourceName=None
      resourceObject=null
      tiling=NoTile
      imageSize=<18 18>
      margin=<ltrb 0 0 0 0>
      outlineSettings {
       bUseBrushTransparency=false
       roundingType=FixedRadius
       width=1
       color {colorUseRule=UseColor_ColorTable, specifiedColor=<rgba 0 0 0 0>}
       cornerRadii=<3 3 3 3>
      }
      tintColor {colorUseRule=UseColor_Specified, specifiedColor=<rgba 0.0048 0.0048 0.0048 1>}
      uVRegion {bIsValid=false, max=<0 0>, min=<0 0>}
     }
     backgroundImage {
      bIsDynamicallyLoaded=false
      drawAs=NoDrawType
      imageType=NoImage
      mirroring=NoMirror
      resourceName=None
      resourceObject=null
      tiling=NoTile
      imageSize=<18 18>
      margin=<ltrb 0 0 0 0>
      outlineSettings {
       bUseBrushTransparency=false
       roundingType=FixedRadius
       width=1
       color {colorUseRule=UseColor_ColorTable, specifiedColor=<rgba 0 0 0 0>}
       cornerRadii=<3 3 3 3>
      }
      tintColor {colorUseRule=UseColor_Specified, specifiedColor=<rgba 0.0048 0.0048 0.0048 1>}
      uVRegion {bIsValid=false, max=<0 0>, min=<0 0>}
     }
     backgroundPressedImage {
      bIsDynamicallyLoaded=false
      drawAs=NoDrawType
      imageType=NoImage
      mirroring=NoMirror
      resourceName=None
      resourceObject=null
      tiling=NoTile
      imageSize=<18 18>
      margin=<ltrb 0 0 0 0>
      outlineSettings {
       bUseBrushTransparency=false
       roundingType=FixedRadius
       width=1
       color {colorUseRule=UseColor_ColorTable, specifiedColor=<rgba 0 0 0 0>}
       cornerRadii=<3 3 3 3>
      }
      tintColor {colorUseRule=UseColor_Specified, specifiedColor=<rgba 0.0048 0.0048 0.0048 1>}
      uVRegion {bIsValid=false, max=<0 0>, min=<0 0>}
     }
     borderBackgroundColor {colorUseRule=UseColor_Specified, specifiedColor=<rgba 1 1 1 1>}
     checkedForeground {colorUseRule=UseColor_Specified, specifiedColor=<rgba 1 1 1 1>}
     checkedHoveredForeground {colorUseRule=UseColor_Specified, specifiedColor=<rgba 1 1 1 1>}
     checkedHoveredImage {
      bIsDynamicallyLoaded=false
      drawAs=RoundedBox
      imageType=NoImage
      mirroring=NoMirror
      resourceName=None
      resourceObject=null
      tiling=NoTile
      imageSize=<32 32>
      margin=<ltrb 0 0 0 0>
      outlineSettings {
       bUseBrushTransparency=false
       roundingType=FixedRadius
       width=0.5
       color {colorUseRule=UseColor_Specified, specifiedColor=<rgba 0 0 0 0>}
       cornerRadii=<5 5 5 1>
      }
      tintColor {colorUseRule=UseColor_Specified, specifiedColor=<rgba 0.2734 0.4375 0.2816 1>}
      uVRegion {bIsValid=false, max=<0 0>, min=<0 0>}
     }
     checkedImage {
      bIsDynamicallyLoaded=false
      drawAs=RoundedBox
      imageType=NoImage
      mirroring=NoMirror
      resourceName=None
      resourceObject=null
      tiling=NoTile
      imageSize=<32 32>
      margin=<ltrb 0.5 0.5 0.5 0.5>
      outlineSettings {
       bUseBrushTransparency=false
       roundingType=FixedRadius
       width=0.5
       color {colorUseRule=UseColor_Specified, specifiedColor=<rgba 0 0 0 0>}
       cornerRadii=<5 5 5 1>
      }
      tintColor {colorUseRule=UseColor_Specified, specifiedColor=<rgba 0.0355 0.4375 0.0554 1>}
      uVRegion {bIsValid=false, max=<0 0>, min=<0 0>}
     }
     checkedPressedForeground {colorUseRule=UseColor_Specified, specifiedColor=<rgba 1 1 1 1>}
     checkedPressedImage {
      bIsDynamicallyLoaded=false
      drawAs=RoundedBox
      imageType=NoImage
      mirroring=NoMirror
      resourceName=None
      resourceObject=null
      tiling=NoTile
      imageSize=<32 32>
      margin=<ltrb 0 0 0 0>
      outlineSettings {
       bUseBrushTransparency=false
       roundingType=FixedRadius
       width=0.5
       color {colorUseRule=UseColor_Specified, specifiedColor=<rgba 0 0 0 0>}
       cornerRadii=<5 5 5 1>
      }
      tintColor {colorUseRule=UseColor_Specified, specifiedColor=<rgba 0.4248 0.9062 0.4486 1>}
      uVRegion {bIsValid=false, max=<0 0>, min=<0 0>}
     }
     checkedSlateSound {resourceObject=null}
     foregroundColor {colorUseRule=UseColor_Specified, specifiedColor=<rgba 1 1 1 1>}
     hoveredForeground {colorUseRule=UseColor_Specified, specifiedColor=<rgba 1 1 1 1>}
     hoveredSlateSound {resourceObject=null}
     padding=<ltrb 2 0 0 0>
     pressedForeground {colorUseRule=UseColor_Specified, specifiedColor=<rgba 1 1 1 1>}
     uncheckedHoveredImage {
      bIsDynamicallyLoaded=false
      drawAs=RoundedBox
      imageType=NoImage
      mirroring=NoMirror
      resourceName=None
      resourceObject=null
      tiling=NoTile
      imageSize=<32 60.845>
      margin=<ltrb 0 0 0 0>
      outlineSettings {
       bUseBrushTransparency=false
       roundingType=FixedRadius
       width=0.5
       color {colorUseRule=UseColor_Specified, specifiedColor=<rgba 0 0 0 0>}
       cornerRadii=<5 5 5 1>
      }
      tintColor {colorUseRule=UseColor_Specified, specifiedColor=<rgba 0.724 0.724 0.724 1>}
      uVRegion {bIsValid=false, max=<0 0>, min=<0 0>}
     }
     uncheckedImage {
      bIsDynamicallyLoaded=false
      drawAs=RoundedBox
      imageType=NoImage
      mirroring=NoMirror
      resourceName=None
      resourceObject=null
      tiling=NoTile
      imageSize=<32 32>
      margin=<ltrb 0 0 0 0>
      outlineSettings {
       bUseBrushTransparency=false
       roundingType=FixedRadius
       width=0.5
       color {colorUseRule=UseColor_Specified, specifiedColor=<rgba 0 0 0 0>}
       cornerRadii=<5 5 5 1>
      }
      tintColor {colorUseRule=UseColor_Specified, specifiedColor=<rgba 0.2448 0.2448 0.2448 1>}
      uVRegion {bIsValid=false, max=<0 0>, min=<0 0>}
     }
     uncheckedPressedImage {
      bIsDynamicallyLoaded=false
      drawAs=RoundedBox
      imageType=NoImage
      mirroring=NoMirror
      resourceName=None
      resourceObject=null
      tiling=NoTile
      imageSize=<32 32>
      margin=<ltrb 0 0 0 0>
      outlineSettings {
       bUseBrushTransparency=false
       roundingType=FixedRadius
       width=0.5
       color {colorUseRule=UseColor_Specified, specifiedColor=<rgba 0 0 0 0>}
       cornerRadii=<5 5 5 1>
      }
      tintColor {colorUseRule=UseColor_Specified, specifiedColor=<rgba 0.8594 0.8494 0.698 1>}
      uVRegion {bIsValid=false, max=<0 0>, min=<0 0>}
     }
     uncheckedSlateSound {resourceObject=null}
     undeterminedForeground {colorUseRule=UseColor_Specified, specifiedColor=<rgba 1 1 1 1>}
     undeterminedHoveredImage {
      bIsDynamicallyLoaded=false
      drawAs=Image
      imageType=Vector
      mirroring=NoMirror
      resourceName=../../../Engine/Content/Slate/Starship/CoreWidgets/CheckBox/indeterminate.svg
      resourceObject=null
      tiling=NoTile
      imageSize=<16 16>
      margin=<ltrb 0 0 0 0>
      outlineSettings {
       bUseBrushTransparency=false
       roundingType=HalfHeightRadius
       width=0
       color {colorUseRule=UseColor_Specified, specifiedColor=<rgba 0 0 0 0>}
       cornerRadii=<0 0 0 0>
      }
      tintColor {colorUseRule=UseColor_Specified, specifiedColor=<rgba 0.0044 0.2384 1 1>}
      uVRegion {bIsValid=false, max=<0 0>, min=<0 0>}
     }
     undeterminedImage {
      bIsDynamicallyLoaded=false
      drawAs=Image
      imageType=Vector
      mirroring=NoMirror
      resourceName=../../../Engine/Content/Slate/Starship/CoreWidgets/CheckBox/indeterminate.svg
      resourceObject=null
      tiling=NoTile
      imageSize=<16 16>
      margin=<ltrb 0 0 0 0>
      outlineSettings {
       bUseBrushTransparency=false
       roundingType=HalfHeightRadius
       width=0
       color {colorUseRule=UseColor_Specified, specifiedColor=<rgba 0 0 0 0>}
       cornerRadii=<0 0 0 0>
      }
      tintColor {colorUseRule=UseColor_Specified, specifiedColor=<rgba 0 0.162 0.7454 1>}
      uVRegion {bIsValid=false, max=<0 0>, min=<0 0>}
     }
     undeterminedPressedImage {
      bIsDynamicallyLoaded=false
      drawAs=Image
      imageType=Vector
      mirroring=NoMirror
      resourceName=../../../Engine/Content/Slate/Starship/CoreWidgets/CheckBox/indeterminate.svg
      resourceObject=null
      tiling=NoTile
      imageSize=<16 16>
      margin=<ltrb 0 0 0 0>
      outlineSettings {
       bUseBrushTransparency=false
       roundingType=HalfHeightRadius
       width=0
       color {colorUseRule=UseColor_Specified, specifiedColor=<rgba 0 0 0 0>}
       cornerRadii=<0 0 0 0>
      }
      tintColor {colorUseRule=UseColor_Specified, specifiedColor=<rgba 0 0.0802 0.3515 1>}
      uVRegion {bIsValid=false, max=<0 0>, min=<0 0>}
     }
    }
   }
   slot {class=HorizontalBoxSlot, properties {padding=<ltrb 0 0 2 0>}}
  }
  TextBlock_6 {
   widgetClassPath=&TextBlock
   parent=SelectAndChangeToggle
   widget=TextBlock_6
   properties {
    justification=Center
    text=Select Existent Data Asset
    font {
     bForceMonospaced=false
     bMaterialIsStencil=false
     fontMaterial=null
     fontObject=&HeadingNow
     letterSpacing=0
     monospacedWidth=1
     size=8.25
     skewAmount=0
     typefaceFontName=75 Medium
     outlineSettings {
      bApplyOutlineToDropShadows=false
      bMiteredCorners=false
      bSeparateFillAlpha=false
      outlineMaterial=null
      outlineSize=0
      outlineColor=<rgba 0 0 0 1>
     }
    }
    margin=<ltrb 7 7 7 7>
   }
   slot {class=PanelSlot}
  }
  AddAndCreateToggle {
   widgetClassPath=&EditorUtilityCheckBox
   bIsVariable=true
   parent=HorizontalBox_10
   widget=AddAndCreateToggle
   properties {
    toolTipText=Shows the MIDI Text Events
    widgetStyle {
     checkBoxType=ToggleButton
     backgroundHoveredImage {
      bIsDynamicallyLoaded=false
      drawAs=NoDrawType
      imageType=NoImage
      mirroring=NoMirror
      resourceName=None
      resourceObject=null
      tiling=NoTile
      imageSize=<18 18>
      margin=<ltrb 0 0 0 0>
      outlineSettings {
       bUseBrushTransparency=false
       roundingType=FixedRadius
       width=1
       color {colorUseRule=UseColor_ColorTable, specifiedColor=<rgba 0 0 0 0>}
       cornerRadii=<3 3 3 3>
      }
      tintColor {colorUseRule=UseColor_Specified, specifiedColor=<rgba 0.0048 0.0048 0.0048 1>}
      uVRegion {bIsValid=false, max=<0 0>, min=<0 0>}
     }
     backgroundImage {
      bIsDynamicallyLoaded=false
      drawAs=NoDrawType
      imageType=NoImage
      mirroring=NoMirror
      resourceName=None
      resourceObject=null
      tiling=NoTile
      imageSize=<18 18>
      margin=<ltrb 0 0 0 0>
      outlineSettings {
       bUseBrushTransparency=false
       roundingType=FixedRadius
       width=1
       color {colorUseRule=UseColor_ColorTable, specifiedColor=<rgba 0 0 0 0>}
       cornerRadii=<3 3 3 3>
      }
      tintColor {colorUseRule=UseColor_Specified, specifiedColor=<rgba 0.0048 0.0048 0.0048 1>}
      uVRegion {bIsValid=false, max=<0 0>, min=<0 0>}
     }
     backgroundPressedImage {
      bIsDynamicallyLoaded=false
      drawAs=NoDrawType
      imageType=NoImage
      mirroring=NoMirror
      resourceName=None
      resourceObject=null
      tiling=NoTile
      imageSize=<18 18>
      margin=<ltrb 0 0 0 0>
      outlineSettings {
       bUseBrushTransparency=false
       roundingType=FixedRadius
       width=1
       color {colorUseRule=UseColor_ColorTable, specifiedColor=<rgba 0 0 0 0>}
       cornerRadii=<3 3 3 3>
      }
      tintColor {colorUseRule=UseColor_Specified, specifiedColor=<rgba 0.0048 0.0048 0.0048 1>}
      uVRegion {bIsValid=false, max=<0 0>, min=<0 0>}
     }
     borderBackgroundColor {colorUseRule=UseColor_Specified, specifiedColor=<rgba 1 1 1 1>}
     checkedForeground {colorUseRule=UseColor_Specified, specifiedColor=<rgba 1 1 1 1>}
     checkedHoveredForeground {colorUseRule=UseColor_Specified, specifiedColor=<rgba 1 1 1 1>}
     checkedHoveredImage {
      bIsDynamicallyLoaded=false
      drawAs=RoundedBox
      imageType=NoImage
      mirroring=NoMirror
      resourceName=None
      resourceObject=null
      tiling=NoTile
      imageSize=<32 32>
      margin=<ltrb 0 0 0 0>
      outlineSettings {
       bUseBrushTransparency=false
       roundingType=FixedRadius
       width=0.5
       color {colorUseRule=UseColor_Specified, specifiedColor=<rgba 0 0 0 0>}
       cornerRadii=<5 5 5 1>
      }
      tintColor {colorUseRule=UseColor_Specified, specifiedColor=<rgba 0.2734 0.4375 0.2816 1>}
      uVRegion {bIsValid=false, max=<0 0>, min=<0 0>}
     }
     checkedImage {
      bIsDynamicallyLoaded=false
      drawAs=RoundedBox
      imageType=NoImage
      mirroring=NoMirror
      resourceName=None
      resourceObject=null
      tiling=NoTile
      imageSize=<32 32>
      margin=<ltrb 0.5 0.5 0.5 0.5>
      outlineSettings {
       bUseBrushTransparency=false
       roundingType=FixedRadius
       width=0.5
       color {colorUseRule=UseColor_Specified, specifiedColor=<rgba 0 0 0 0>}
       cornerRadii=<5 5 5 1>
      }
      tintColor {colorUseRule=UseColor_Specified, specifiedColor=<rgba 0.0355 0.4375 0.0554 1>}
      uVRegion {bIsValid=false, max=<0 0>, min=<0 0>}
     }
     checkedPressedForeground {colorUseRule=UseColor_Specified, specifiedColor=<rgba 1 1 1 1>}
     checkedPressedImage {
      bIsDynamicallyLoaded=false
      drawAs=RoundedBox
      imageType=NoImage
      mirroring=NoMirror
      resourceName=None
      resourceObject=null
      tiling=NoTile
      imageSize=<32 32>
      margin=<ltrb 0 0 0 0>
      outlineSettings {
       bUseBrushTransparency=false
       roundingType=FixedRadius
       width=0.5
       color {colorUseRule=UseColor_Specified, specifiedColor=<rgba 0 0 0 0>}
       cornerRadii=<5 5 5 1>
      }
      tintColor {colorUseRule=UseColor_Specified, specifiedColor=<rgba 0.4248 0.9062 0.4486 1>}
      uVRegion {bIsValid=false, max=<0 0>, min=<0 0>}
     }
     checkedSlateSound {resourceObject=null}
     foregroundColor {colorUseRule=UseColor_Specified, specifiedColor=<rgba 1 1 1 1>}
     hoveredForeground {colorUseRule=UseColor_Specified, specifiedColor=<rgba 1 1 1 1>}
     hoveredSlateSound {resourceObject=null}
     padding=<ltrb 2 0 0 0>
     pressedForeground {colorUseRule=UseColor_Specified, specifiedColor=<rgba 1 1 1 1>}
     uncheckedHoveredImage {
      bIsDynamicallyLoaded=false
      drawAs=RoundedBox
      imageType=NoImage
      mirroring=NoMirror
      resourceName=None
      resourceObject=null
      tiling=NoTile
      imageSize=<32 60.845>
      margin=<ltrb 0 0 0 0>
      outlineSettings {
       bUseBrushTransparency=false
       roundingType=FixedRadius
       width=0.5
       color {colorUseRule=UseColor_Specified, specifiedColor=<rgba 0 0 0 0>}
       cornerRadii=<5 5 5 1>
      }
      tintColor {colorUseRule=UseColor_Specified, specifiedColor=<rgba 0.724 0.724 0.724 1>}
      uVRegion {bIsValid=false, max=<0 0>, min=<0 0>}
     }
     uncheckedImage {
      bIsDynamicallyLoaded=false
      drawAs=RoundedBox
      imageType=NoImage
      mirroring=NoMirror
      resourceName=None
      resourceObject=null
      tiling=NoTile
      imageSize=<32 32>
      margin=<ltrb 0 0 0 0>
      outlineSettings {
       bUseBrushTransparency=false
       roundingType=FixedRadius
       width=0.5
       color {colorUseRule=UseColor_Specified, specifiedColor=<rgba 0 0 0 0>}
       cornerRadii=<5 5 5 1>
      }
      tintColor {colorUseRule=UseColor_Specified, specifiedColor=<rgba 0.2448 0.2448 0.2448 1>}
      uVRegion {bIsValid=false, max=<0 0>, min=<0 0>}
     }
     uncheckedPressedImage {
      bIsDynamicallyLoaded=false
      drawAs=RoundedBox
      imageType=NoImage
      mirroring=NoMirror
      resourceName=None
      resourceObject=null
      tiling=NoTile
      imageSize=<32 32>
      margin=<ltrb 0 0 0 0>
      outlineSettings {
       bUseBrushTransparency=false
       roundingType=FixedRadius
       width=0.5
       color {colorUseRule=UseColor_Specified, specifiedColor=<rgba 0 0 0 0>}
       cornerRadii=<5 5 5 1>
      }
      tintColor {colorUseRule=UseColor_Specified, specifiedColor=<rgba 0.8594 0.8494 0.698 1>}
      uVRegion {bIsValid=false, max=<0 0>, min=<0 0>}
     }
     uncheckedSlateSound {resourceObject=null}
     undeterminedForeground {colorUseRule=UseColor_Specified, specifiedColor=<rgba 1 1 1 1>}
     undeterminedHoveredImage {
      bIsDynamicallyLoaded=false
      drawAs=Image
      imageType=Vector
      mirroring=NoMirror
      resourceName=../../../Engine/Content/Slate/Starship/CoreWidgets/CheckBox/indeterminate.svg
      resourceObject=null
      tiling=NoTile
      imageSize=<16 16>
      margin=<ltrb 0 0 0 0>
      outlineSettings {
       bUseBrushTransparency=false
       roundingType=HalfHeightRadius
       width=0
       color {colorUseRule=UseColor_Specified, specifiedColor=<rgba 0 0 0 0>}
       cornerRadii=<0 0 0 0>
      }
      tintColor {colorUseRule=UseColor_Specified, specifiedColor=<rgba 0.0044 0.2384 1 1>}
      uVRegion {bIsValid=false, max=<0 0>, min=<0 0>}
     }
     undeterminedImage {
      bIsDynamicallyLoaded=false
      drawAs=Image
      imageType=Vector
      mirroring=NoMirror
      resourceName=../../../Engine/Content/Slate/Starship/CoreWidgets/CheckBox/indeterminate.svg
      resourceObject=null
      tiling=NoTile
      imageSize=<16 16>
      margin=<ltrb 0 0 0 0>
      outlineSettings {
       bUseBrushTransparency=false
       roundingType=HalfHeightRadius
       width=0
       color {colorUseRule=UseColor_Specified, specifiedColor=<rgba 0 0 0 0>}
       cornerRadii=<0 0 0 0>
      }
      tintColor {colorUseRule=UseColor_Specified, specifiedColor=<rgba 0 0.162 0.7454 1>}
      uVRegion {bIsValid=false, max=<0 0>, min=<0 0>}
     }
     undeterminedPressedImage {
      bIsDynamicallyLoaded=false
      drawAs=Image
      imageType=Vector
      mirroring=NoMirror
      resourceName=../../../Engine/Content/Slate/Starship/CoreWidgets/CheckBox/indeterminate.svg
      resourceObject=null
      tiling=NoTile
      imageSize=<16 16>
      margin=<ltrb 0 0 0 0>
      outlineSettings {
       bUseBrushTransparency=false
       roundingType=HalfHeightRadius
       width=0
       color {colorUseRule=UseColor_Specified, specifiedColor=<rgba 0 0 0 0>}
       cornerRadii=<0 0 0 0>
      }
      tintColor {colorUseRule=UseColor_Specified, specifiedColor=<rgba 0 0.0802 0.3515 1>}
      uVRegion {bIsValid=false, max=<0 0>, min=<0 0>}
     }
    }
   }
   slot {class=HorizontalBoxSlot}
  }
  TextBlock_10 {
   widgetClassPath=&TextBlock
   parent=AddAndCreateToggle
   widget=TextBlock_10
   properties {
    justification=Center
    text=Create New Data Asset
    font {
     bForceMonospaced=false
     bMaterialIsStencil=false
     fontMaterial=null
     fontObject=&HeadingNow
     letterSpacing=0
     monospacedWidth=1
     size=8.25
     skewAmount=0
     typefaceFontName=75 Medium
     outlineSettings {
      bApplyOutlineToDropShadows=false
      bMiteredCorners=false
      bSeparateFillAlpha=false
      outlineMaterial=null
      outlineSize=0
      outlineColor=<rgba 0 0 0 1>
     }
    }
    margin=<ltrb 7 7 7 7>
   }
   slot {class=PanelSlot}
  }
  SelectChangeGroup {
   widgetClassPath=&VerticalBox
   bIsVariable=true
   parent=VerticalBox_0
   widget=SelectChangeGroup
   slot {class=VerticalBoxSlot, properties {verticalAlignment=VAlign_Top, size {sizeRule=Automatic, value=0.5}}}
  }
  Border_3 {
   widgetClassPath=&Border
   parent=SelectChangeGroup
   widget=Border_3
   properties {background {
     bIsDynamicallyLoaded=false
     drawAs=Image
     imageType=NoImage
     mirroring=NoMirror
     resourceName=None
     resourceObject=null
     tiling=NoTile
     imageSize=<32 32>
     margin=<ltrb 0 0 0 0>
     outlineSettings {
      bUseBrushTransparency=false
      roundingType=HalfHeightRadius
      width=0
      color {colorUseRule=UseColor_Specified, specifiedColor=<rgba 0 0 0 0>}
      cornerRadii=<0 0 0 0>
     }
     tintColor {colorUseRule=UseColor_Specified, specifiedColor=<rgba 0.0729 0.0729 0.0729 1>}
     uVRegion {bIsValid=false, max=<0 0>, min=<0 0>}
    }}
   slot {class=VerticalBoxSlot, properties {padding=<ltrb -10 0 -10 0>}}
  }
  HorizontalBox_11 {widgetClassPath=&HorizontalBox, parent=Border_3, widget=HorizontalBox_11, slot {class=BorderSlot}}
  SizeBox_0 {
   widgetClassPath=&SizeBox
   parent=HorizontalBox_11
   widget=SizeBox_0
   properties {bOverride_MinDesiredHeight=true, bOverride_MinDesiredWidth=true, maxDesiredWidth=14.4662, minDesiredHeight=20, minDesiredWidth=20}
   slot {class=HorizontalBoxSlot, properties {verticalAlignment=VAlign_Center, padding=<ltrb 0 0 5 0>}}
  }
  ToggleViewDa {
   widgetClassPath=&EditorUtilityCheckBox
   bIsVariable=true
   parent=SizeBox_0
   widget=ToggleViewDa
   properties {
    toolTipText=Preview Data
    widgetStyle {
     checkBoxType=ToggleButton
     backgroundHoveredImage {
      bIsDynamicallyLoaded=false
      drawAs=RoundedBox
      imageType=NoImage
      mirroring=NoMirror
      resourceName=None
      resourceObject=null
      tiling=NoTile
      imageSize=<18 18>
      margin=<ltrb 0 0 0 0>
      outlineSettings {
       bUseBrushTransparency=false
       roundingType=FixedRadius
       width=1
       color {colorUseRule=UseColor_ColorTable, specifiedColor=<rgba 0 0 0 0>}
       cornerRadii=<3 3 3 3>
      }
      tintColor {colorUseRule=UseColor_Foreground, specifiedColor=<rgba 0.0048 0.0048 0.0048 1>}
      uVRegion {bIsValid=false, max=<0 0>, min=<0 0>}
     }
     backgroundImage {
      bIsDynamicallyLoaded=false
      drawAs=RoundedBox
      imageType=NoImage
      mirroring=NoMirror
      resourceName=None
      resourceObject=null
      tiling=NoTile
      imageSize=<18 18>
      margin=<ltrb 0 0 0 0>
      outlineSettings {
       bUseBrushTransparency=false
       roundingType=FixedRadius
       width=1
       color {colorUseRule=UseColor_ColorTable, specifiedColor=<rgba 0 0 0 0>}
       cornerRadii=<3 3 3 3>
      }
      tintColor {colorUseRule=UseColor_Foreground, specifiedColor=<rgba 0.0048 0.0048 0.0048 1>}
      uVRegion {bIsValid=false, max=<0 0>, min=<0 0>}
     }
     backgroundPressedImage {
      bIsDynamicallyLoaded=false
      drawAs=RoundedBox
      imageType=NoImage
      mirroring=NoMirror
      resourceName=None
      resourceObject=null
      tiling=NoTile
      imageSize=<18 18>
      margin=<ltrb 0 0 0 0>
      outlineSettings {
       bUseBrushTransparency=false
       roundingType=FixedRadius
       width=1
       color {colorUseRule=UseColor_ColorTable, specifiedColor=<rgba 0 0 0 0>}
       cornerRadii=<3 3 3 3>
      }
      tintColor {colorUseRule=UseColor_Specified, specifiedColor=<rgba 0.0048 0.0048 0.0048 1>}
      uVRegion {bIsValid=false, max=<0 0>, min=<0 0>}
     }
     borderBackgroundColor {colorUseRule=UseColor_Specified, specifiedColor=<rgba 1 1 1 1>}
     checkedForeground {colorUseRule=UseColor_Specified, specifiedColor=<rgba 1 1 1 1>}
     checkedHoveredForeground {colorUseRule=UseColor_Specified, specifiedColor=<rgba 1 1 1 1>}
     checkedHoveredImage {
      bIsDynamicallyLoaded=false
      drawAs=Image
      imageType=NoImage
      mirroring=NoMirror
      resourceName=None
      resourceObject=/Game/Athena/UI/InGame/Creative/PropertyEditor/Icon_Picker/Textures_V2/64x64/T_UI_IconLibrary_Eye_64
      tiling=NoTile
      imageSize=<64 64>
      margin=<ltrb 0 0 0 0>
      outlineSettings {
       bUseBrushTransparency=false
       roundingType=HalfHeightRadius
       width=0
       color {colorUseRule=UseColor_Specified, specifiedColor=<rgba 0 0 0 0>}
       cornerRadii=<0 0 0 0>
      }
      tintColor {colorUseRule=UseColor_Specified, specifiedColor=<rgba 1 1 1 1>}
      uVRegion {bIsValid=false, max=<0 0>, min=<0 0>}
     }
     checkedImage {
      bIsDynamicallyLoaded=false
      drawAs=Image
      imageType=NoImage
      mirroring=NoMirror
      resourceName=None
      resourceObject=&T_UI_IconLibrary_Eye_64
      tiling=NoTile
      imageSize=<15 15>
      margin=<ltrb 0 0 0 0>
      outlineSettings {
       bUseBrushTransparency=false
       roundingType=HalfHeightRadius
       width=0
       color {colorUseRule=UseColor_Specified, specifiedColor=<rgba 0 0 0 0>}
       cornerRadii=<0 0 0 0>
      }
      tintColor {colorUseRule=UseColor_Specified, specifiedColor=<rgba 1 1 1 1>}
      uVRegion {bIsValid=false, max=<0 0>, min=<0 0>}
     }
     checkedPressedForeground {colorUseRule=UseColor_Specified, specifiedColor=<rgba 1 1 1 1>}
     checkedPressedImage {
      bIsDynamicallyLoaded=false
      drawAs=Image
      imageType=NoImage
      mirroring=NoMirror
      resourceName=None
      resourceObject=&T_UI_IconLibrary_Eye_64
      tiling=NoTile
      imageSize=<15 15>
      margin=<ltrb 0 0 0 0>
      outlineSettings {
       bUseBrushTransparency=false
       roundingType=HalfHeightRadius
       width=0
       color {colorUseRule=UseColor_Specified, specifiedColor=<rgba 0 0 0 0>}
       cornerRadii=<0 0 0 0>
      }
      tintColor {colorUseRule=UseColor_Specified, specifiedColor=<rgba 1 1 1 1>}
      uVRegion {bIsValid=false, max=<0 0>, min=<0 0>}
     }
     checkedSlateSound {resourceObject=null}
     foregroundColor {colorUseRule=UseColor_Specified, specifiedColor=<rgba 1 1 1 1>}
     hoveredForeground {colorUseRule=UseColor_Specified, specifiedColor=<rgba 1 1 1 1>}
     hoveredSlateSound {resourceObject=null}
     padding=<ltrb 2 0 0 0>
     pressedForeground {colorUseRule=UseColor_Specified, specifiedColor=<rgba 1 1 1 1>}
     uncheckedHoveredImage {
      bIsDynamicallyLoaded=false
      drawAs=RoundedBox
      imageType=Vector
      mirroring=NoMirror
      resourceName=../../../Engine/Content/Slate/Starship/CoreWidgets/CheckBox/indeterminate.svg
      resourceObject=null
      tiling=NoTile
      imageSize=<21.6344 15.3298>
      margin=<ltrb 0 0 0 0>
      outlineSettings {
       bUseBrushTransparency=false
       roundingType=HalfHeightRadius
       width=0
       color {colorUseRule=UseColor_Specified, specifiedColor=<rgba 0 0 0 0>}
       cornerRadii=<0 0 0 0>
      }
      tintColor {colorUseRule=UseColor_Specified, specifiedColor=<rgba 0.9378 1 0.6936 1>}
      uVRegion {bIsValid=false, max=<0 0>, min=<0 0>}
     }
     uncheckedImage {
      bIsDynamicallyLoaded=false
      drawAs=RoundedBox
      imageType=Vector
      mirroring=NoMirror
      resourceName=../../../Engine/Content/Slate/Starship/CoreWidgets/CheckBox/indeterminate.svg
      resourceObject=null
      tiling=NoTile
      imageSize=<21.6344 15.3298>
      margin=<ltrb 0 0 0 0>
      outlineSettings {
       bUseBrushTransparency=false
       roundingType=HalfHeightRadius
       width=0
       color {colorUseRule=UseColor_Specified, specifiedColor=<rgba 0 0 0 0>}
       cornerRadii=<0 0 0 0>
      }
      tintColor {colorUseRule=UseColor_Specified, specifiedColor=<rgba 1 1 1 1>}
      uVRegion {bIsValid=false, max=<0 0>, min=<0 0>}
     }
     uncheckedPressedImage {
      bIsDynamicallyLoaded=false
      drawAs=RoundedBox
      imageType=Vector
      mirroring=NoMirror
      resourceName=../../../Engine/Content/Slate/Starship/CoreWidgets/CheckBox/indeterminate.svg
      resourceObject=null
      tiling=NoTile
      imageSize=<21.6344 15.3298>
      margin=<ltrb 0 0 0 0>
      outlineSettings {
       bUseBrushTransparency=false
       roundingType=HalfHeightRadius
       width=0
       color {colorUseRule=UseColor_Specified, specifiedColor=<rgba 0 0 0 0>}
       cornerRadii=<0 0 0 0>
      }
      tintColor {colorUseRule=UseColor_Specified, specifiedColor=<rgba 0.5417 0.5417 0.5417 1>}
      uVRegion {bIsValid=false, max=<0 0>, min=<0 0>}
     }
     uncheckedSlateSound {resourceObject=null}
     undeterminedForeground {colorUseRule=UseColor_Specified, specifiedColor=<rgba 1 1 1 1>}
     undeterminedHoveredImage {
      bIsDynamicallyLoaded=false
      drawAs=Image
      imageType=Vector
      mirroring=NoMirror
      resourceName=../../../Engine/Content/Slate/Starship/CoreWidgets/CheckBox/indeterminate.svg
      resourceObject=null
      tiling=NoTile
      imageSize=<16 16>
      margin=<ltrb 0 0 0 0>
      outlineSettings {
       bUseBrushTransparency=false
       roundingType=HalfHeightRadius
       width=0
       color {colorUseRule=UseColor_Specified, specifiedColor=<rgba 0 0 0 0>}
       cornerRadii=<0 0 0 0>
      }
      tintColor {colorUseRule=UseColor_Specified, specifiedColor=<rgba 0.0044 0.2384 1 1>}
      uVRegion {bIsValid=false, max=<0 0>, min=<0 0>}
     }
     undeterminedImage {
      bIsDynamicallyLoaded=false
      drawAs=Image
      imageType=Vector
      mirroring=NoMirror
      resourceName=../../../Engine/Content/Slate/Starship/CoreWidgets/CheckBox/indeterminate.svg
      resourceObject=null
      tiling=NoTile
      imageSize=<16 16>
      margin=<ltrb 0 0 0 0>
      outlineSettings {
       bUseBrushTransparency=false
       roundingType=HalfHeightRadius
       width=0
       color {colorUseRule=UseColor_Specified, specifiedColor=<rgba 0 0 0 0>}
       cornerRadii=<0 0 0 0>
      }
      tintColor {colorUseRule=UseColor_Specified, specifiedColor=<rgba 0 0.162 0.7454 1>}
      uVRegion {bIsValid=false, max=<0 0>, min=<0 0>}
     }
     undeterminedPressedImage {
      bIsDynamicallyLoaded=false
      drawAs=Image
      imageType=Vector
      mirroring=NoMirror
      resourceName=../../../Engine/Content/Slate/Starship/CoreWidgets/CheckBox/indeterminate.svg
      resourceObject=null
      tiling=NoTile
      imageSize=<16 16>
      margin=<ltrb 0 0 0 0>
      outlineSettings {
       bUseBrushTransparency=false
       roundingType=HalfHeightRadius
       width=0
       color {colorUseRule=UseColor_Specified, specifiedColor=<rgba 0 0 0 0>}
       cornerRadii=<0 0 0 0>
      }
      tintColor {colorUseRule=UseColor_Specified, specifiedColor=<rgba 0 0.0802 0.3515 1>}
      uVRegion {bIsValid=false, max=<0 0>, min=<0 0>}
     }
    }
   }
   slot {class=SizeBoxSlot}
  }
  NewSequenceEventCellName {
   widgetClassPath=&TextBlock
   bIsVariable=true
   parent=HorizontalBox_11
   widget=NewSequenceEventCellName
   properties {
    text="Index [0]"
    font {
     bForceMonospaced=false
     bMaterialIsStencil=false
     fontMaterial=null
     fontObject=&HeadingNow
     letterSpacing=0
     monospacedWidth=1
     size=7.5
     skewAmount=0
     typefaceFontName=75 Medium
     outlineSettings {
      bApplyOutlineToDropShadows=false
      bMiteredCorners=false
      bSeparateFillAlpha=false
      outlineMaterial=null
      outlineSize=0
      outlineColor=<rgba 0 0 0 1>
     }
    }
   }
   slot {class=HorizontalBoxSlot, properties {horizontalAlignment=HAlign_Left, verticalAlignment=VAlign_Center, padding=<ltrb 0 0 5 0>}}
  }
  DaSlotProperty {
   widgetClassPath=&SinglePropertyView
   parent=HorizontalBox_11
   widget=DaSlotProperty
   properties {nameOverride=" ", propertyName=DaSlot, visibility=Collapsed}
   slot {class=HorizontalBoxSlot, properties {verticalAlignment=VAlign_Center}}
  }
  SizeBox_1 {
   widgetClassPath=&SizeBox
   parent=HorizontalBox_11
   widget=SizeBox_1
   properties {bOverride_MaxDesiredWidth=true, bOverride_MinDesiredWidth=true, maxDesiredWidth=350, minDesiredWidth=350}
   slot {class=HorizontalBoxSlot, properties {horizontalAlignment=HAlign_Left, verticalAlignment=VAlign_Center}}
  }
  DAFilteredComboBox {
   widgetClassPath=/Script/FortniteEditor.SearchableStringComboBox
   bIsVariable=true
   parent=SizeBox_1
   widget=DAFilteredComboBox
   properties {
    enableGamepadNavigationMode=false
    selectedOption=aaaaaaaaaaaaaaaaaaaaaaaaa
    defaultOptions [aaaaaaaaaaaaaaaaaaaaaaaaa]
    font {
     bForceMonospaced=false
     bMaterialIsStencil=false
     fontMaterial=null
     fontObject=&HeadingNow
     letterSpacing=0
     monospacedWidth=1
     size=7.5
     skewAmount=0
     typefaceFontName=75 Medium
     outlineSettings {
      bApplyOutlineToDropShadows=false
      bMiteredCorners=false
      bSeparateFillAlpha=false
      outlineMaterial=null
      outlineSize=0
      outlineColor=<rgba 0 0 0 1>
     }
    }
    foregroundColor {colorUseRule=UseColor_Specified, specifiedColor=<rgba 0.5271 0.5271 0.5271 1>}
    itemStyle {
     bUseParentRowBrush=false
     activeBrush {
      bIsDynamicallyLoaded=false
      drawAs=Image
      imageType=NoImage
      mirroring=NoMirror
      resourceName=None
      resourceObject=null
      tiling=NoTile
      imageSize=<0 0>
      margin=<ltrb 0 0 0 0>
      outlineSettings {
       bUseBrushTransparency=false
       roundingType=HalfHeightRadius
       width=0
       color {colorUseRule=UseColor_Specified, specifiedColor=<rgba 0 0 0 0>}
       cornerRadii=<0 0 0 0>
      }
      tintColor {colorUseRule=UseColor_Specified, specifiedColor=<rgba 0 0.162 0.7454 1>}
      uVRegion {bIsValid=false, max=<0 0>, min=<0 0>}
     }
     activeHighlightedBrush {
      bIsDynamicallyLoaded=false
      drawAs=Image
      imageType=NoImage
      mirroring=NoMirror
      resourceName=None
      resourceObject=null
      tiling=NoTile
      imageSize=<0 0>
      margin=<ltrb 0 0 0 0>
      outlineSettings {
       bUseBrushTransparency=false
       roundingType=HalfHeightRadius
       width=0
       color {colorUseRule=UseColor_Specified, specifiedColor=<rgba 0 0 0 0>}
       cornerRadii=<0 0 0 0>
      }
      tintColor {colorUseRule=UseColor_Specified, specifiedColor=<rgba 0.0252 0.0319 0.0423 1>}
      uVRegion {bIsValid=false, max=<0 0>, min=<0 0>}
     }
     activeHoveredBrush {
      bIsDynamicallyLoaded=false
      drawAs=Image
      imageType=NoImage
      mirroring=NoMirror
      resourceName=None
      resourceObject=null
      tiling=NoTile
      imageSize=<0 0>
      margin=<ltrb 0 0 0 0>
      outlineSettings {
       bUseBrushTransparency=false
       roundingType=HalfHeightRadius
       width=0
       color {colorUseRule=UseColor_Specified, specifiedColor=<rgba 0 0 0 0>}
       cornerRadii=<0 0 0 0>
      }
      tintColor {colorUseRule=UseColor_Specified, specifiedColor=<rgba 0 0.162 0.7454 1>}
      uVRegion {bIsValid=false, max=<0 0>, min=<0 0>}
     }
     dropIndicator_Above {
      bIsDynamicallyLoaded=false
      drawAs=Box
      imageType=FullColor
      mirroring=NoMirror
      resourceName=../../../Engine/Content/Slate/Common/DropZoneIndicator_Above.png
      resourceObject=null
      tiling=NoTile
      imageSize=<0 0>
      margin=<ltrb 0.625 0.625 0 0>
      outlineSettings {
       bUseBrushTransparency=false
       roundingType=HalfHeightRadius
       width=0
       color {colorUseRule=UseColor_Specified, specifiedColor=<rgba 0 0 0 0>}
       cornerRadii=<0 0 0 0>
      }
      tintColor {colorUseRule=UseColor_Specified, specifiedColor=<rgba 0.0091 0.3515 0.9647 1>}
      uVRegion {bIsValid=false, max=<0 0>, min=<0 0>}
     }
     dropIndicator_Below {
      bIsDynamicallyLoaded=false
      drawAs=Box
      imageType=FullColor
      mirroring=NoMirror
      resourceName=../../../Engine/Content/Slate/Common/DropZoneIndicator_Below.png
      resourceObject=null
      tiling=NoTile
      imageSize=<0 0>
      margin=<ltrb 0.625 0 0 0.625>
      outlineSettings {
       bUseBrushTransparency=false
       roundingType=HalfHeightRadius
       width=0
       color {colorUseRule=UseColor_Specified, specifiedColor=<rgba 0 0 0 0>}
       cornerRadii=<0 0 0 0>
      }
      tintColor {colorUseRule=UseColor_Specified, specifiedColor=<rgba 0.0091 0.3515 0.9647 1>}
      uVRegion {bIsValid=false, max=<0 0>, min=<0 0>}
     }
     dropIndicator_Onto {
      bIsDynamicallyLoaded=false
      drawAs=Box
      imageType=FullColor
      mirroring=NoMirror
      resourceName=../../../Engine/Content/Slate/Common/DropZoneIndicator_Onto.png
      resourceObject=null
      tiling=NoTile
      imageSize=<0 0>
      margin=<ltrb 0.25 0.25 0.25 0.25>
      outlineSettings {
       bUseBrushTransparency=false
       roundingType=HalfHeightRadius
       width=0
       color {colorUseRule=UseColor_Specified, specifiedColor=<rgba 0 0 0 0>}
       cornerRadii=<0 0 0 0>
      }
      tintColor {colorUseRule=UseColor_Specified, specifiedColor=<rgba 0.0091 0.3515 0.9647 1>}
      uVRegion {bIsValid=false, max=<0 0>, min=<0 0>}
     }
     evenRowBackgroundBrush {
      bIsDynamicallyLoaded=false
      drawAs=Image
      imageType=NoImage
      mirroring=NoMirror
      resourceName=None
      resourceObject=null
      tiling=NoTile
      imageSize=<0 0>
      margin=<ltrb 0 0 0 0>
      outlineSettings {
       bUseBrushTransparency=false
       roundingType=HalfHeightRadius
       width=0
       color {colorUseRule=UseColor_Specified, specifiedColor=<rgba 0 0 0 0>}
       cornerRadii=<0 0 0 0>
      }
      tintColor {colorUseRule=UseColor_Specified, specifiedColor=<rgba 0.0103 0.0103 0.0103 1>}
      uVRegion {bIsValid=false, max=<0 0>, min=<0 0>}
     }
     evenRowBackgroundHoveredBrush {
      bIsDynamicallyLoaded=false
      drawAs=Image
      imageType=NoImage
      mirroring=NoMirror
      resourceName=None
      resourceObject=null
      tiling=NoTile
      imageSize=<0 0>
      margin=<ltrb 0 0 0 0>
      outlineSettings {
       bUseBrushTransparency=false
       roundingType=HalfHeightRadius
       width=0
       color {colorUseRule=UseColor_Specified, specifiedColor=<rgba 0 0 0 0>}
       cornerRadii=<0 0 0 0>
      }
      tintColor {colorUseRule=UseColor_Specified, specifiedColor=<rgba 0.0176 0.0176 0.0176 1>}
      uVRegion {bIsValid=false, max=<0 0>, min=<0 0>}
     }
     inactiveBrush {
      bIsDynamicallyLoaded=false
      drawAs=Image
      imageType=NoImage
      mirroring=NoMirror
      resourceName=None
      resourceObject=null
      tiling=NoTile
      imageSize=<0 0>
      margin=<ltrb 0 0 0 0>
      outlineSettings {
       bUseBrushTransparency=false
       roundingType=HalfHeightRadius
       width=0
       color {colorUseRule=UseColor_Specified, specifiedColor=<rgba 0 0 0 0>}
       cornerRadii=<0 0 0 0>
      }
      tintColor {colorUseRule=UseColor_Specified, specifiedColor=<rgba 0.0513 0.0953 0.159 1>}
      uVRegion {bIsValid=false, max=<0 0>, min=<0 0>}
     }
     inactiveHighlightedBrush {
      bIsDynamicallyLoaded=false
      drawAs=Image
      imageType=NoImage
      mirroring=NoMirror
      resourceName=None
      resourceObject=null
      tiling=NoTile
      imageSize=<0 0>
      margin=<ltrb 0 0 0 0>
      outlineSettings {
       bUseBrushTransparency=false
       roundingType=HalfHeightRadius
       width=0
       color {colorUseRule=UseColor_Specified, specifiedColor=<rgba 0 0 0 0>}
       cornerRadii=<0 0 0 0>
      }
      tintColor {colorUseRule=UseColor_Specified, specifiedColor=<rgba 0.0252 0.0319 0.0423 1>}
      uVRegion {bIsValid=false, max=<0 0>, min=<0 0>}
     }
     inactiveHoveredBrush {
      bIsDynamicallyLoaded=false
      drawAs=Image
      imageType=NoImage
      mirroring=NoMirror
      resourceName=None
      resourceObject=null
      tiling=NoTile
      imageSize=<0 0>
      margin=<ltrb 0 0 0 0>
      outlineSettings {
       bUseBrushTransparency=false
       roundingType=HalfHeightRadius
       width=0
       color {colorUseRule=UseColor_Specified, specifiedColor=<rgba 0 0 0 0>}
       cornerRadii=<0 0 0 0>
      }
      tintColor {colorUseRule=UseColor_Specified, specifiedColor=<rgba 0.0513 0.0953 0.159 1>}
      uVRegion {bIsValid=false, max=<0 0>, min=<0 0>}
     }
     oddRowBackgroundBrush {
      bIsDynamicallyLoaded=false
      drawAs=Image
      imageType=NoImage
      mirroring=NoMirror
      resourceName=None
      resourceObject=null
      tiling=NoTile
      imageSize=<0 0>
      margin=<ltrb 0 0 0 0>
      outlineSettings {
       bUseBrushTransparency=false
       roundingType=HalfHeightRadius
       width=0
       color {colorUseRule=UseColor_Specified, specifiedColor=<rgba 0 0 0 0>}
       cornerRadii=<0 0 0 0>
      }
      tintColor {colorUseRule=UseColor_Specified, specifiedColor=<rgba 0.0103 0.0103 0.0103 1>}
      uVRegion {bIsValid=false, max=<0 0>, min=<0 0>}
     }
     oddRowBackgroundHoveredBrush {
      bIsDynamicallyLoaded=false
      drawAs=Image
      imageType=NoImage
      mirroring=NoMirror
      resourceName=None
      resourceObject=null
      tiling=NoTile
      imageSize=<0 0>
      margin=<ltrb 0 0 0 0>
      outlineSettings {
       bUseBrushTransparency=false
       roundingType=HalfHeightRadius
       width=0
       color {colorUseRule=UseColor_Specified, specifiedColor=<rgba 0 0 0 0>}
       cornerRadii=<0 0 0 0>
      }
      tintColor {colorUseRule=UseColor_Specified, specifiedColor=<rgba 0.0176 0.0176 0.0176 1>}
      uVRegion {bIsValid=false, max=<0 0>, min=<0 0>}
     }
     parentRowBackgroundBrush {
      bIsDynamicallyLoaded=false
      drawAs=Image
      imageType=NoImage
      mirroring=NoMirror
      resourceName=None
      resourceObject=null
      tiling=NoTile
      imageSize=<32 32>
      margin=<ltrb 0 0 0 0>
      outlineSettings {
       bUseBrushTransparency=false
       roundingType=HalfHeightRadius
       width=0
       color {colorUseRule=UseColor_Specified, specifiedColor=<rgba 0 0 0 0>}
       cornerRadii=<0 0 0 0>
      }
      tintColor {colorUseRule=UseColor_Specified, specifiedColor=<rgba 1 1 1 1>}
      uVRegion {bIsValid=false, max=<0 0>, min=<0 0>}
     }
     parentRowBackgroundHoveredBrush {
      bIsDynamicallyLoaded=false
      drawAs=Image
      imageType=NoImage
      mirroring=NoMirror
      resourceName=None
      resourceObject=null
      tiling=NoTile
      imageSize=<32 32>
      margin=<ltrb 0 0 0 0>
      outlineSettings {
       bUseBrushTransparency=false
       roundingType=HalfHeightRadius
       width=0
       color {colorUseRule=UseColor_Specified, specifiedColor=<rgba 0 0 0 0>}
       cornerRadii=<0 0 0 0>
      }
      tintColor {colorUseRule=UseColor_Specified, specifiedColor=<rgba 1 1 1 1>}
      uVRegion {bIsValid=false, max=<0 0>, min=<0 0>}
     }
     selectedTextColor {colorUseRule=UseColor_Specified, specifiedColor=<rgba 1 1 1 1>}
     selectorFocusedBrush {
      bIsDynamicallyLoaded=false
      drawAs=Border
      imageType=FullColor
      mirroring=NoMirror
      resourceName=../../../Engine/Content/Slate/Common/Selector.png
      resourceObject=null
      tiling=Both
      imageSize=<0 0>
      margin=<ltrb 0.25 0.25 0.25 0.25>
      outlineSettings {
       bUseBrushTransparency=false
       roundingType=HalfHeightRadius
       width=0
       color {colorUseRule=UseColor_Specified, specifiedColor=<rgba 0 0 0 0>}
       cornerRadii=<0 0 0 0>
      }
      tintColor {colorUseRule=UseColor_Specified, specifiedColor=<rgba 0.701 0.225 0.003 1>}
      uVRegion {bIsValid=false, max=<0 0>, min=<0 0>}
     }
     textColor {colorUseRule=UseColor_Specified, specifiedColor=<rgba 0.5271 0.5271 0.5271 1>}
    }
    widgetStyle {
     comboButtonStyle {
      downArrowAlign=VAlign_Center
      buttonStyle {
       clickedSlateSound {resourceObject=null}
       disabled {
        bIsDynamicallyLoaded=false
        drawAs=RoundedBox
        imageType=NoImage
        mirroring=NoMirror
        resourceName=None
        resourceObject=null
        tiling=NoTile
        imageSize=<0 0>
        margin=<ltrb 0 0 0 0>
        outlineSettings {
         bUseBrushTransparency=false
         roundingType=FixedRadius
         width=1
         color {colorUseRule=UseColor_ColorTable, specifiedColor=<rgba 0 0 0 0>}
         cornerRadii=<4 4 4 4>
        }
        tintColor {colorUseRule=UseColor_Specified, specifiedColor=<rgba 0.0395 0.0395 0.0395 1>}
        uVRegion {bIsValid=false, max=<0 0>, min=<0 0>}
       }
       disabledForeground {colorUseRule=UseColor_ColorTable, specifiedColor=<rgba 0 0 0 0>}
       hovered {
        bIsDynamicallyLoaded=false
        drawAs=RoundedBox
        imageType=NoImage
        mirroring=NoMirror
        resourceName=None
        resourceObject=null
        tiling=NoTile
        imageSize=<0 0>
        margin=<ltrb 0 0 0 0>
        outlineSettings {
         bUseBrushTransparency=false
         roundingType=FixedRadius
         width=1
         color {colorUseRule=UseColor_ColorTable, specifiedColor=<rgba 0 0 0 0>}
         cornerRadii=<4 4 4 4>
        }
        tintColor {colorUseRule=UseColor_Specified, specifiedColor=<rgba 0.0048 0.0048 0.0048 1>}
        uVRegion {bIsValid=false, max=<0 0>, min=<0 0>}
       }
       hoveredForeground {colorUseRule=UseColor_ColorTable, specifiedColor=<rgba 0 0 0 0>}
       hoveredSlateSound {resourceObject=null}
       normal {
        bIsDynamicallyLoaded=false
        drawAs=RoundedBox
        imageType=NoImage
        mirroring=NoMirror
        resourceName=None
        resourceObject=null
        tiling=NoTile
        imageSize=<0 0>
        margin=<ltrb 0 0 0 0>
        outlineSettings {
         bUseBrushTransparency=false
         roundingType=FixedRadius
         width=1
         color {colorUseRule=UseColor_ColorTable, specifiedColor=<rgba 0 0 0 0>}
         cornerRadii=<4 4 4 4>
        }
        tintColor {colorUseRule=UseColor_Specified, specifiedColor=<rgba 0.0048 0.0048 0.0048 1>}
        uVRegion {bIsValid=false, max=<0 0>, min=<0 0>}
       }
       normalForeground {colorUseRule=UseColor_ColorTable, specifiedColor=<rgba 0 0 0 0>}
       normalPadding=<ltrb 8 1 8 1>
       pressed {
        bIsDynamicallyLoaded=false
        drawAs=RoundedBox
        imageType=NoImage
        mirroring=NoMirror
        resourceName=None
        resourceObject=null
        tiling=NoTile
        imageSize=<0 0>
        margin=<ltrb 0 0 0 0>
        outlineSettings {
         bUseBrushTransparency=false
         roundingType=FixedRadius
         width=1
         color {colorUseRule=UseColor_ColorTable, specifiedColor=<rgba 0 0 0 0>}
         cornerRadii=<4 4 4 4>
        }
        tintColor {colorUseRule=UseColor_Specified, specifiedColor=<rgba 0.0048 0.0048 0.0048 1>}
        uVRegion {bIsValid=false, max=<0 0>, min=<0 0>}
       }
       pressedForeground {colorUseRule=UseColor_ColorTable, specifiedColor=<rgba 0 0 0 0>}
       pressedPadding=<ltrb 8 2 8 0>
       pressedSlateSound {resourceObject=null}
      }
      contentPadding=<ltrb 0 0 0 0>
      downArrowImage {
       bIsDynamicallyLoaded=false
       drawAs=Image
       imageType=Vector
       mirroring=NoMirror
       resourceName=../../../Engine/Content/Slate/Starship/CoreWidgets/ComboBox/wide-chevron-down.svg
       resourceObject=null
       tiling=NoTile
       imageSize=<16 16>
       margin=<ltrb 0 0 0 0>
       outlineSettings {
        bUseBrushTransparency=false
        roundingType=HalfHeightRadius
        width=0
        color {colorUseRule=UseColor_Specified, specifiedColor=<rgba 0 0 0 0>}
        cornerRadii=<0 0 0 0>
       }
       tintColor {colorUseRule=UseColor_Specified, specifiedColor=<rgba 1 1 1 1>}
       uVRegion {bIsValid=false, max=<0 0>, min=<0 0>}
      }
      downArrowPadding=<ltrb 2 2 2 2>
      menuBorderBrush {
       bIsDynamicallyLoaded=false
       drawAs=RoundedBox
       imageType=NoImage
       mirroring=NoMirror
       resourceName=None
       resourceObject=null
       tiling=NoTile
       imageSize=<0 0>
       margin=<ltrb 0 0 0 0>
       outlineSettings {
        bUseBrushTransparency=false
        roundingType=FixedRadius
        width=1
        color {colorUseRule=UseColor_ColorTable, specifiedColor=<rgba 0 0 0 0>}
        cornerRadii=<0 0 0 0>
       }
       tintColor {colorUseRule=UseColor_Specified, specifiedColor=<rgba 0.0395 0.0395 0.0395 1>}
       uVRegion {bIsValid=false, max=<0 0>, min=<0 0>}
      }
      menuBorderPadding=<ltrb 0 0 0 0>
      shadowColorAndOpacity=<rgba 0 0 0 1>
      shadowOffset=<0 0>
     }
     contentPadding=<ltrb 0 0 0 0>
     menuRowPadding=<ltrb 8 3 8 3>
     pressedSlateSound {resourceObject=null}
     selectionChangeSlateSound {resourceObject=null}
    }
   }
   slot {class=SizeBoxSlot}
  }
  OpenEditorForEverything {
   widgetClassPath=&EditorUtilityButton
   parent=HorizontalBox_11
   widget=OpenEditorForEverything
   properties {
    toolTipText=Check
    widgetStyle {
     clickedSlateSound {resourceObject=null}
     disabled {
      bIsDynamicallyLoaded=false
      drawAs=RoundedBox
      imageType=NoImage
      mirroring=NoMirror
      resourceName=None
      resourceObject=/Game/UI/Foundation/Textures/Buttons/T-Button-Square
      tiling=NoTile
      imageSize=<32 32>
      margin=<ltrb 0.5 0.5 0.5 0.5>
      outlineSettings {
       bUseBrushTransparency=false
       roundingType=FixedRadius
       width=0
       color {colorUseRule=UseColor_ColorTable, specifiedColor=<rgba 0 0 0 0>}
       cornerRadii=<4 4 4 4>
      }
      tintColor {colorUseRule=UseColor_Specified, specifiedColor=<rgba 0.0395 0.0395 0.0395 1>}
      uVRegion {bIsValid=false, max=<0 0>, min=<0 0>}
     }
     disabledForeground {colorUseRule=UseColor_ColorTable, specifiedColor=<rgba 0 0 0 0>}
     hovered {
      bIsDynamicallyLoaded=false
      drawAs=RoundedBox
      imageType=NoImage
      mirroring=NoMirror
      resourceName=None
      resourceObject=&T-Button-Square
      tiling=NoTile
      imageSize=<32 32>
      margin=<ltrb 0.5 0.5 0.5 0.5>
      outlineSettings {
       bUseBrushTransparency=false
       roundingType=FixedRadius
       width=0
       color {colorUseRule=UseColor_ColorTable, specifiedColor=<rgba 0 0 0 0>}
       cornerRadii=<4 4 4 4>
      }
      tintColor {colorUseRule=UseColor_Specified, specifiedColor=<rgba 0 0.3021 0.0241 1>}
      uVRegion {bIsValid=false, max=<0 0>, min=<0 0>}
     }
     hoveredForeground {colorUseRule=UseColor_ColorTable, specifiedColor=<rgba 0 0 0 0>}
     hoveredSlateSound {resourceObject=null}
     normal {
      bIsDynamicallyLoaded=false
      drawAs=RoundedBox
      imageType=NoImage
      mirroring=NoMirror
      resourceName=None
      resourceObject=&T-Button-Square
      tiling=NoTile
      imageSize=<32 32>
      margin=<ltrb 0.5 0.5 0.5 0.5>
      outlineSettings {
       bUseBrushTransparency=false
       roundingType=FixedRadius
       width=0
       color {colorUseRule=UseColor_ColorTable, specifiedColor=<rgba 0 0 0 0>}
       cornerRadii=<4 4 4 4>
      }
      tintColor {colorUseRule=UseColor_Specified, specifiedColor=<rgba 0.0071 0.7594 0 1>}
      uVRegion {bIsValid=false, max=<0 0>, min=<0 0>}
     }
     normalForeground {colorUseRule=UseColor_ColorTable, specifiedColor=<rgba 0 0 0 0>}
     normalPadding=<ltrb 12 1.5 12 1.5>
     pressed {
      bIsDynamicallyLoaded=false
      drawAs=RoundedBox
      imageType=NoImage
      mirroring=NoMirror
      resourceName=None
      resourceObject=&T-Button-Square
      tiling=NoTile
      imageSize=<32 32>
      margin=<ltrb 0.5 0.5 0.5 0.5>
      outlineSettings {
       bUseBrushTransparency=false
       roundingType=FixedRadius
       width=0
       color {colorUseRule=UseColor_ColorTable, specifiedColor=<rgba 0 0 0 0>}
       cornerRadii=<4 4 4 4>
      }
      tintColor {colorUseRule=UseColor_Specified, specifiedColor=<rgba 0.3119 0.7917 0.2715 1>}
      uVRegion {bIsValid=false, max=<0 0>, min=<0 0>}
     }
     pressedForeground {colorUseRule=UseColor_ColorTable, specifiedColor=<rgba 0 0 0 0>}
     pressedPadding=<ltrb 12 2.5 12 0.5>
     pressedSlateSound {resourceObject=null}
    }
   }
   slot {class=HorizontalBoxSlot, properties {verticalAlignment=VAlign_Center, padding=<ltrb 5 5 5 5>}}
  }
  ExternalIcon_3 {
   widgetClassPath=&Image
   parent=OpenEditorForEverything
   widget=ExternalIcon_3
   properties {brush {
     bIsDynamicallyLoaded=false
     drawAs=Image
     imageType=NoImage
     mirroring=NoMirror
     resourceName=None
     resourceObject=/FortUILibrary/Foundation/Buttons/Materials/M_UI_Icon_ExternalLink
     tiling=NoTile
     imageSize=<20 20>
     margin=<ltrb 0 0 0 0>
     outlineSettings {
      bUseBrushTransparency=false
      roundingType=HalfHeightRadius
      width=0
      color {colorUseRule=UseColor_Specified, specifiedColor=<rgba 0 0 0 0>}
      cornerRadii=<0 0 0 0>
     }
     tintColor {colorUseRule=UseColor_Specified, specifiedColor=<rgba 1 1 1 1>}
     uVRegion {bIsValid=false, max=<0 0>, min=<0 0>}
    }}
   slot {class=ButtonSlot, properties {padding=<ltrb 0 0 0 0>}}
  }
  NewSequenceEventChangeButton {
   widgetClassPath=&EditorUtilityButton
   parent=HorizontalBox_11
   widget=NewSequenceEventChangeButton
   properties {widgetStyle {
     clickedSlateSound {resourceObject=null}
     disabled {
      bIsDynamicallyLoaded=false
      drawAs=RoundedBox
      imageType=NoImage
      mirroring=NoMirror
      resourceName=None
      resourceObject=null
      tiling=NoTile
      imageSize=<32 32>
      margin=<ltrb 0.5 0.5 0.5 0.5>
      outlineSettings {
       bUseBrushTransparency=false
       roundingType=FixedRadius
       width=0
       color {colorUseRule=UseColor_ColorTable, specifiedColor=<rgba 0 0 0 0>}
       cornerRadii=<4 4 4 4>
      }
      tintColor {colorUseRule=UseColor_Specified, specifiedColor=<rgba 0.0395 0.0395 0.0395 1>}
      uVRegion {bIsValid=false, max=<0 0>, min=<0 0>}
     }
     disabledForeground {colorUseRule=UseColor_ColorTable, specifiedColor=<rgba 0 0 0 0>}
     hovered {
      bIsDynamicallyLoaded=false
      drawAs=RoundedBox
      imageType=NoImage
      mirroring=NoMirror
      resourceName=None
      resourceObject=null
      tiling=NoTile
      imageSize=<32 32>
      margin=<ltrb 0 0 0 0>
      outlineSettings {
       bUseBrushTransparency=false
       roundingType=FixedRadius
       width=0
       color {colorUseRule=UseColor_ColorTable, specifiedColor=<rgba 0 0 0 0>}
       cornerRadii=<4 4 4 4>
      }
      tintColor {colorUseRule=UseColor_Specified, specifiedColor=<rgba 0.1126 0.7031 0 1>}
      uVRegion {bIsValid=false, max=<0 0>, min=<0 0>}
     }
     hoveredForeground {colorUseRule=UseColor_ColorTable, specifiedColor=<rgba 0 0 0 0>}
     hoveredSlateSound {resourceObject=null}
     normal {
      bIsDynamicallyLoaded=false
      drawAs=RoundedBox
      imageType=NoImage
      mirroring=NoMirror
      resourceName=None
      resourceObject=null
      tiling=NoTile
      imageSize=<32 32>
      margin=<ltrb 0.5 0.5 0.5 0.5>
      outlineSettings {
       bUseBrushTransparency=false
       roundingType=FixedRadius
       width=0
       color {colorUseRule=UseColor_ColorTable, specifiedColor=<rgba 0 0 0 0>}
       cornerRadii=<4 4 4 4>
      }
      tintColor {colorUseRule=UseColor_Specified, specifiedColor=<rgba 0.0771 0.6167 0.1289 1>}
      uVRegion {bIsValid=false, max=<0 0>, min=<0 0>}
     }
     normalForeground {colorUseRule=UseColor_ColorTable, specifiedColor=<rgba 0 0 0 0>}
     normalPadding=<ltrb 12 1.5 12 1.5>
     pressed {
      bIsDynamicallyLoaded=false
      drawAs=RoundedBox
      imageType=NoImage
      mirroring=NoMirror
      resourceName=None
      resourceObject=null
      tiling=NoTile
      imageSize=<32 32>
      margin=<ltrb 0.5 0.5 0.5 0.5>
      outlineSettings {
       bUseBrushTransparency=false
       roundingType=FixedRadius
       width=0
       color {colorUseRule=UseColor_ColorTable, specifiedColor=<rgba 0 0 0 0>}
       cornerRadii=<4 4 4 4>
      }
      tintColor {colorUseRule=UseColor_Specified, specifiedColor=<rgba 0.1806 0.4583 0.1572 1>}
      uVRegion {bIsValid=false, max=<0 0>, min=<0 0>}
     }
     pressedForeground {colorUseRule=UseColor_ColorTable, specifiedColor=<rgba 0 0 0 0>}
     pressedPadding=<ltrb 12 2.5 12 0.5>
     pressedSlateSound {resourceObject=null}
    }}
   slot {class=HorizontalBoxSlot, properties {verticalAlignment=VAlign_Center, padding=<ltrb 5 0 0 0>}}
  }
  HorizontalBox_12 {
   widgetClassPath=&HorizontalBox
   parent=NewSequenceEventChangeButton
   widget=HorizontalBox_12
   slot {class=ButtonSlot, properties {horizontalAlignment=HAlign_Left}}
  }
  Image_13 {
   widgetClassPath=&Image
   parent=HorizontalBox_12
   widget=Image_13
   properties {brush {
     bIsDynamicallyLoaded=false
     drawAs=Image
     imageType=NoImage
     mirroring=NoMirror
     resourceName=None
     resourceObject=/PilgrimQuickplay/StageManager/NeverCook/EditorUtilities/StageManagerEditor/Textures/MI_SparksEditorUI_PlusMinus_NoBorder
     tiling=NoTile
     imageSize=<15 15>
     margin=<ltrb 0 0 0 0>
     outlineSettings {
      bUseBrushTransparency=false
      roundingType=HalfHeightRadius
      width=0
      color {colorUseRule=UseColor_Specified, specifiedColor=<rgba 0 0 0 0>}
      cornerRadii=<0 0 0 0>
     }
     tintColor {colorUseRule=UseColor_Specified, specifiedColor=<rgba 1 1 1 1>}
     uVRegion {bIsValid=false, max=<0 0>, min=<0 0>}
    }}
   slot {class=HorizontalBoxSlot, properties {horizontalAlignment=HAlign_Center, verticalAlignment=VAlign_Center, padding=<ltrb 0 0 3 0>}}
  }
  TextBlock_8 {
   widgetClassPath=&TextBlock
   parent=HorizontalBox_12
   widget=TextBlock_8
   properties {
    text=Choose Selected
    font {
     bForceMonospaced=false
     bMaterialIsStencil=false
     fontMaterial=null
     fontObject=&HeadingNow
     letterSpacing=0
     monospacedWidth=1
     size=7.5
     skewAmount=0
     typefaceFontName=75 Medium
     outlineSettings {
      bApplyOutlineToDropShadows=false
      bMiteredCorners=false
      bSeparateFillAlpha=false
      outlineMaterial=null
      outlineSize=0
      outlineColor=<rgba 0 0 0 1>
     }
    }
   }
   slot {class=HorizontalBoxSlot, properties {verticalAlignment=VAlign_Center}}
  }
  EventPreviewerGroup {
   widgetClassPath=&SizeBox
   bIsVariable=true
   parent=SelectChangeGroup
   widget=EventPreviewerGroup
   properties {bOverride_MaxDesiredHeight=true, maxDesiredHeight=300, visibility=Collapsed}
   slot {class=VerticalBoxSlot, properties {verticalAlignment=VAlign_Top, padding=<ltrb 0 10 0 5>, size {sizeRule=Fill, value=1}}}
  }
  Border_1 {
   widgetClassPath=&Border
   parent=EventPreviewerGroup
   widget=Border_1
   properties {brushColor=<rgba 0.0267 0.0312 0.0267 1>, contentColorAndOpacity=<rgba 0.6417 1 0.8525 1>, padding=<ltrb 4 2 4 3>}
   slot {class=SizeBoxSlot}
  }
  EditorUtilityScrollBox_0 {
   widgetClassPath=&EditorUtilityScrollBox
   parent=Border_1
   widget=EditorUtilityScrollBox_0
   slot {class=BorderSlot, properties {padding=<ltrb 4 2 4 3>}}
  }
  EventDetailView {
   widgetClassPath=&DetailsView
   parent=EditorUtilityScrollBox_0
   widget=EventDetailView
   properties {bIsEnabled=false, bShowScrollBar=false}
   slot {class=ScrollBoxSlot}
  }
  Border_6 {
   widgetClassPath=&Border
   parent=SelectChangeGroup
   widget=Border_6
   properties {
    background {
     bIsDynamicallyLoaded=false
     drawAs=Image
     imageType=NoImage
     mirroring=NoMirror
     resourceName=None
     resourceObject=null
     tiling=NoTile
     imageSize=<32 32>
     margin=<ltrb 0 0 0 0>
     outlineSettings {
      bUseBrushTransparency=false
      roundingType=HalfHeightRadius
      width=0
      color {colorUseRule=UseColor_Specified, specifiedColor=<rgba 0 0 0 0>}
      cornerRadii=<0 0 0 0>
     }
     tintColor {colorUseRule=UseColor_Specified, specifiedColor=<rgba 0.0729 0.0729 0.0729 1>}
     uVRegion {bIsValid=false, max=<0 0>, min=<0 0>}
    }
    padding=<ltrb 10 2 4 2>
   }
   slot {class=VerticalBoxSlot, properties {padding=<ltrb -10 0 -10 0>}}
  }
  HorizontalBox_16 {
   widgetClassPath=&HorizontalBox
   parent=Border_6
   widget=HorizontalBox_16
   slot {class=BorderSlot, properties {padding=<ltrb 10 2 4 2>}}
  }
  SizeBox {
   widgetClassPath=&SizeBox
   parent=HorizontalBox_16
   widget=SizeBox
   properties {bOverride_MaxDesiredWidth=true, bOverride_MinDesiredWidth=true, maxDesiredWidth=450, minDesiredWidth=450}
   slot {class=HorizontalBoxSlot}
  }
  RenameAssetTextBox {
   widgetClassPath=&EditorUtilityEditableTextBox
   bIsVariable=true
   parent=SizeBox
   widget=RenameAssetTextBox
   properties {
    hintText=Choose the new name.
    toolTipText=Choose a new asset name.
    widgetStyle {
     backgroundColor {colorUseRule=UseColor_Specified, specifiedColor=<rgba 1 1 1 1>}
     backgroundImageFocused {
      bIsDynamicallyLoaded=false
      drawAs=RoundedBox
      imageType=NoImage
      mirroring=NoMirror
      resourceName=None
      resourceObject=null
      tiling=NoTile
      imageSize=<0 0>
      margin=<ltrb 0 0 0 0>
      outlineSettings {
       bUseBrushTransparency=false
       roundingType=FixedRadius
       width=1
       color {colorUseRule=UseColor_ColorTable, specifiedColor=<rgba 0 0 0 0>}
       cornerRadii=<4 4 4 4>
      }
      tintColor {colorUseRule=UseColor_Specified, specifiedColor=<rgba 0.0048 0.0048 0.0048 1>}
      uVRegion {bIsValid=false, max=<0 0>, min=<0 0>}
     }
     backgroundImageHovered {
      bIsDynamicallyLoaded=false
      drawAs=RoundedBox
      imageType=NoImage
      mirroring=NoMirror
      resourceName=None
      resourceObject=null
      tiling=NoTile
      imageSize=<0 0>
      margin=<ltrb 0 0 0 0>
      outlineSettings {
       bUseBrushTransparency=false
       roundingType=FixedRadius
       width=1
       color {colorUseRule=UseColor_ColorTable, specifiedColor=<rgba 0 0 0 0>}
       cornerRadii=<4 4 4 4>
      }
      tintColor {colorUseRule=UseColor_Specified, specifiedColor=<rgba 0.0048 0.0048 0.0048 1>}
      uVRegion {bIsValid=false, max=<0 0>, min=<0 0>}
     }
     backgroundImageNormal {
      bIsDynamicallyLoaded=false
      drawAs=RoundedBox
      imageType=NoImage
      mirroring=NoMirror
      resourceName=None
      resourceObject=null
      tiling=NoTile
      imageSize=<0 0>
      margin=<ltrb 0 0 0 0>
      outlineSettings {
       bUseBrushTransparency=false
       roundingType=FixedRadius
       width=1
       color {colorUseRule=UseColor_ColorTable, specifiedColor=<rgba 0 0 0 0>}
       cornerRadii=<4 4 4 4>
      }
      tintColor {colorUseRule=UseColor_Specified, specifiedColor=<rgba 0.0048 0.0048 0.0048 1>}
      uVRegion {bIsValid=false, max=<0 0>, min=<0 0>}
     }
     backgroundImageReadOnly {
      bIsDynamicallyLoaded=false
      drawAs=RoundedBox
      imageType=NoImage
      mirroring=NoMirror
      resourceName=None
      resourceObject=null
      tiling=NoTile
      imageSize=<0 0>
      margin=<ltrb 0 0 0 0>
      outlineSettings {
       bUseBrushTransparency=false
       roundingType=FixedRadius
       width=1
       color {colorUseRule=UseColor_ColorTable, specifiedColor=<rgba 0 0 0 0>}
       cornerRadii=<4 4 4 4>
      }
      tintColor {colorUseRule=UseColor_Specified, specifiedColor=<rgba 0.0284 0.0284 0.0284 1>}
      uVRegion {bIsValid=false, max=<0 0>, min=<0 0>}
     }
     focusedForegroundColor {colorUseRule=UseColor_Specified, specifiedColor=<rgba 1 1 1 1>}
     foregroundColor {colorUseRule=UseColor_Specified, specifiedColor=<rgba 0.5271 0.5271 0.5271 1>}
     hScrollBarPadding=<ltrb 0 0 0 0>
     padding=<ltrb 8 3 8 4>
     readOnlyForegroundColor {colorUseRule=UseColor_Foreground, specifiedColor=<rgba 1 0 1 1>}
     scrollBarStyle {
      thickness=8
      draggedThumbImage {
       bIsDynamicallyLoaded=false
       drawAs=RoundedBox
       imageType=NoImage
       mirroring=NoMirror
       resourceName=None
       resourceObject=null
       tiling=NoTile
       imageSize=<0 0>
       margin=<ltrb 0 0 0 0>
       outlineSettings {
        bUseBrushTransparency=false
        roundingType=FixedRadius
        width=0
        color {colorUseRule=UseColor_Specified, specifiedColor=<rgba 0 0 0 0>}
        cornerRadii=<4 4 4 4>
       }
       tintColor {colorUseRule=UseColor_Specified, specifiedColor=<rgba 0.2159 0.2159 0.2159 1>}
       uVRegion {bIsValid=false, max=<0 0>, min=<0 0>}
      }
      horizontalBackgroundImage {
       bIsDynamicallyLoaded=false
       drawAs=NoDrawType
       imageType=NoImage
       mirroring=NoMirror
       resourceName=None
       resourceObject=null
       tiling=NoTile
       imageSize=<0 0>
       margin=<ltrb 0 0 0 0>
       outlineSettings {
        bUseBrushTransparency=false
        roundingType=HalfHeightRadius
        width=0
        color {colorUseRule=UseColor_Specified, specifiedColor=<rgba 0 0 0 0>}
        cornerRadii=<0 0 0 0>
       }
       tintColor {colorUseRule=UseColor_Specified, specifiedColor=<rgba 1 1 1 1>}
       uVRegion {bIsValid=false, max=<0 0>, min=<0 0>}
      }
      horizontalBottomSlotImage {
       bIsDynamicallyLoaded=false
       drawAs=NoDrawType
       imageType=NoImage
       mirroring=NoMirror
       resourceName=None
       resourceObject=null
       tiling=NoTile
       imageSize=<0 0>
       margin=<ltrb 0 0 0 0>
       outlineSettings {
        bUseBrushTransparency=false
        roundingType=HalfHeightRadius
        width=0
        color {colorUseRule=UseColor_Specified, specifiedColor=<rgba 0 0 0 0>}
        cornerRadii=<0 0 0 0>
       }
       tintColor {colorUseRule=UseColor_Specified, specifiedColor=<rgba 1 1 1 1>}
       uVRegion {bIsValid=false, max=<0 0>, min=<0 0>}
      }
      horizontalTopSlotImage {
       bIsDynamicallyLoaded=false
       drawAs=NoDrawType
       imageType=NoImage
       mirroring=NoMirror
       resourceName=None
       resourceObject=null
       tiling=NoTile
       imageSize=<0 0>
       margin=<ltrb 0 0 0 0>
       outlineSettings {
        bUseBrushTransparency=false
        roundingType=HalfHeightRadius
        width=0
        color {colorUseRule=UseColor_Specified, specifiedColor=<rgba 0 0 0 0>}
        cornerRadii=<0 0 0 0>
       }
       tintColor {colorUseRule=UseColor_Specified, specifiedColor=<rgba 1 1 1 1>}
       uVRegion {bIsValid=false, max=<0 0>, min=<0 0>}
      }
      hoveredThumbImage {
       bIsDynamicallyLoaded=false
       drawAs=RoundedBox
       imageType=NoImage
       mirroring=NoMirror
       resourceName=None
       resourceObject=null
       tiling=NoTile
       imageSize=<0 0>
       margin=<ltrb 0 0 0 0>
       outlineSettings {
        bUseBrushTransparency=false
        roundingType=FixedRadius
        width=0
        color {colorUseRule=UseColor_Specified, specifiedColor=<rgba 0 0 0 0>}
        cornerRadii=<4 4 4 4>
       }
       tintColor {colorUseRule=UseColor_Specified, specifiedColor=<rgba 0.2159 0.2159 0.2159 1>}
       uVRegion {bIsValid=false, max=<0 0>, min=<0 0>}
      }
      normalThumbImage {
       bIsDynamicallyLoaded=false
       drawAs=RoundedBox
       imageType=NoImage
       mirroring=NoMirror
       resourceName=None
       resourceObject=null
       tiling=NoTile
       imageSize=<0 0>
       margin=<ltrb 0 0 0 0>
       outlineSettings {
        bUseBrushTransparency=false
        roundingType=FixedRadius
        width=0
        color {colorUseRule=UseColor_Specified, specifiedColor=<rgba 0 0 0 0>}
        cornerRadii=<4 4 4 4>
       }
       tintColor {colorUseRule=UseColor_Specified, specifiedColor=<rgba 0.0953 0.0953 0.0953 1>}
       uVRegion {bIsValid=false, max=<0 0>, min=<0 0>}
      }
      verticalBackgroundImage {
       bIsDynamicallyLoaded=false
       drawAs=NoDrawType
       imageType=NoImage
       mirroring=NoMirror
       resourceName=None
       resourceObject=null
       tiling=NoTile
       imageSize=<0 0>
       margin=<ltrb 0 0 0 0>
       outlineSettings {
        bUseBrushTransparency=false
        roundingType=HalfHeightRadius
        width=0
        color {colorUseRule=UseColor_Specified, specifiedColor=<rgba 0 0 0 0>}
        cornerRadii=<0 0 0 0>
       }
       tintColor {colorUseRule=UseColor_Specified, specifiedColor=<rgba 1 1 1 1>}
       uVRegion {bIsValid=false, max=<0 0>, min=<0 0>}
      }
      verticalBottomSlotImage {
       bIsDynamicallyLoaded=false
       drawAs=NoDrawType
       imageType=NoImage
       mirroring=NoMirror
       resourceName=None
       resourceObject=null
       tiling=NoTile
       imageSize=<0 0>
       margin=<ltrb 0 0 0 0>
       outlineSettings {
        bUseBrushTransparency=false
        roundingType=HalfHeightRadius
        width=0
        color {colorUseRule=UseColor_Specified, specifiedColor=<rgba 0 0 0 0>}
        cornerRadii=<0 0 0 0>
       }
       tintColor {colorUseRule=UseColor_Specified, specifiedColor=<rgba 1 1 1 1>}
       uVRegion {bIsValid=false, max=<0 0>, min=<0 0>}
      }
      verticalTopSlotImage {
       bIsDynamicallyLoaded=false
       drawAs=NoDrawType
       imageType=NoImage
       mirroring=NoMirror
       resourceName=None
       resourceObject=null
       tiling=NoTile
       imageSize=<0 0>
       margin=<ltrb 0 0 0 0>
       outlineSettings {
        bUseBrushTransparency=false
        roundingType=HalfHeightRadius
        width=0
        color {colorUseRule=UseColor_Specified, specifiedColor=<rgba 0 0 0 0>}
        cornerRadii=<0 0 0 0>
       }
       tintColor {colorUseRule=UseColor_Specified, specifiedColor=<rgba 1 1 1 1>}
       uVRegion {bIsValid=false, max=<0 0>, min=<0 0>}
      }
     }
     textStyle {
      overflowPolicy=Clip
      transformPolicy=None
      colorAndOpacity {colorUseRule=UseColor_Foreground, specifiedColor=<rgba 1 0 1 1>}
      font {
       bForceMonospaced=false
       bMaterialIsStencil=false
       fontMaterial=null
       fontObject=&HeadingNow
       letterSpacing=0
       monospacedWidth=1
       size=7.5
       skewAmount=0
       typefaceFontName=75 Medium
       outlineSettings {
        bApplyOutlineToDropShadows=false
        bMiteredCorners=false
        bSeparateFillAlpha=false
        outlineMaterial=null
        outlineSize=0
        outlineColor=<rgba 0 0 0 1>
       }
      }
      highlightColor {colorUseRule=UseColor_ColorTable, specifiedColor=<rgba 0 0 0 0>}
      highlightShape {
       bIsDynamicallyLoaded=false
       drawAs=Image
       imageType=NoImage
       mirroring=NoMirror
       resourceName=None
       resourceObject=null
       tiling=NoTile
       imageSize=<0 0>
       margin=<ltrb 0 0 0 0>
       outlineSettings {
        bUseBrushTransparency=false
        roundingType=HalfHeightRadius
        width=0
        color {colorUseRule=UseColor_Specified, specifiedColor=<rgba 0 0 0 0>}
        cornerRadii=<0 0 0 0>
       }
       tintColor {colorUseRule=UseColor_ColorTable, specifiedColor=<rgba 0 0 0 0>}
       uVRegion {bIsValid=false, max=<0 0>, min=<0 0>}
      }
      selectedBackgroundColor {colorUseRule=UseColor_ColorTable, specifiedColor=<rgba 0 0 0 0>}
      shadowColorAndOpacity=<rgba 0 0 0 1>
      shadowOffset=<0 0>
      strikeBrush {
       bIsDynamicallyLoaded=false
       drawAs=Image
       imageType=NoImage
       mirroring=NoMirror
       resourceName=None
       resourceObject=null
       tiling=NoTile
       imageSize=<32 32>
       margin=<ltrb 0 0 0 0>
       outlineSettings {
        bUseBrushTransparency=false
        roundingType=HalfHeightRadius
        width=0
        color {colorUseRule=UseColor_Specified, specifiedColor=<rgba 0 0 0 0>}
        cornerRadii=<0 0 0 0>
       }
       tintColor {colorUseRule=UseColor_Specified, specifiedColor=<rgba 1 1 1 1>}
       uVRegion {bIsValid=false, max=<0 0>, min=<0 0>}
      }
      underlineBrush {
       bIsDynamicallyLoaded=false
       drawAs=Image
       imageType=NoImage
       mirroring=NoMirror
       resourceName=None
       resourceObject=null
       tiling=NoTile
       imageSize=<32 32>
       margin=<ltrb 0 0 0 0>
       outlineSettings {
        bUseBrushTransparency=false
        roundingType=HalfHeightRadius
        width=0
        color {colorUseRule=UseColor_Specified, specifiedColor=<rgba 0 0 0 0>}
        cornerRadii=<0 0 0 0>
       }
       tintColor {colorUseRule=UseColor_Specified, specifiedColor=<rgba 1 1 1 1>}
       uVRegion {bIsValid=false, max=<0 0>, min=<0 0>}
      }
     }
     vScrollBarPadding=<ltrb 0 0 0 0>
    }
   }
   slot {class=SizeBoxSlot}
  }
  RenameEventButton {
   widgetClassPath=&EditorUtilityButton
   parent=HorizontalBox_16
   widget=RenameEventButton
   properties {widgetStyle {
     clickedSlateSound {resourceObject=null}
     disabled {
      bIsDynamicallyLoaded=false
      drawAs=RoundedBox
      imageType=NoImage
      mirroring=NoMirror
      resourceName=None
      resourceObject=null
      tiling=NoTile
      imageSize=<32 32>
      margin=<ltrb 0.5 0.5 0.5 0.5>
      outlineSettings {
       bUseBrushTransparency=false
       roundingType=FixedRadius
       width=0
       color {colorUseRule=UseColor_ColorTable, specifiedColor=<rgba 0 0 0 0>}
       cornerRadii=<4 4 4 4>
      }
      tintColor {colorUseRule=UseColor_Specified, specifiedColor=<rgba 0.0395 0.0395 0.0395 1>}
      uVRegion {bIsValid=false, max=<0 0>, min=<0 0>}
     }
     disabledForeground {colorUseRule=UseColor_ColorTable, specifiedColor=<rgba 0 0 0 0>}
     hovered {
      bIsDynamicallyLoaded=false
      drawAs=RoundedBox
      imageType=NoImage
      mirroring=NoMirror
      resourceName=None
      resourceObject=null
      tiling=NoTile
      imageSize=<32 32>
      margin=<ltrb 0 0 0 0>
      outlineSettings {
       bUseBrushTransparency=false
       roundingType=FixedRadius
       width=0
       color {colorUseRule=UseColor_ColorTable, specifiedColor=<rgba 0 0 0 0>}
       cornerRadii=<4 4 4 4>
      }
      tintColor {colorUseRule=UseColor_Specified, specifiedColor=<rgba 0.8542 0.7778 0.0801 1>}
      uVRegion {bIsValid=false, max=<0 0>, min=<0 0>}
     }
     hoveredForeground {colorUseRule=UseColor_ColorTable, specifiedColor=<rgba 0 0 0 0>}
     hoveredSlateSound {resourceObject=null}
     normal {
      bIsDynamicallyLoaded=false
      drawAs=RoundedBox
      imageType=NoImage
      mirroring=NoMirror
      resourceName=None
      resourceObject=null
      tiling=NoTile
      imageSize=<32 32>
      margin=<ltrb 0.5 0.5 0.5 0.5>
      outlineSettings {
       bUseBrushTransparency=false
       roundingType=FixedRadius
       width=0
       color {colorUseRule=UseColor_ColorTable, specifiedColor=<rgba 0 0 0 0>}
       cornerRadii=<4 4 4 4>
      }
      tintColor {colorUseRule=UseColor_Specified, specifiedColor=<rgba 0.7031 0.6338 0 1>}
      uVRegion {bIsValid=false, max=<0 0>, min=<0 0>}
     }
     normalForeground {colorUseRule=UseColor_ColorTable, specifiedColor=<rgba 0 0 0 0>}
     normalPadding=<ltrb 12 1.5 12 1.5>
     pressed {
      bIsDynamicallyLoaded=false
      drawAs=RoundedBox
      imageType=NoImage
      mirroring=NoMirror
      resourceName=None
      resourceObject=null
      tiling=NoTile
      imageSize=<32 32>
      margin=<ltrb 0.5 0.5 0.5 0.5>
      outlineSettings {
       bUseBrushTransparency=false
       roundingType=FixedRadius
       width=0
       color {colorUseRule=UseColor_ColorTable, specifiedColor=<rgba 0 0 0 0>}
       cornerRadii=<4 4 4 4>
      }
      tintColor {colorUseRule=UseColor_Specified, specifiedColor=<rgba 1 0.2833 0 1>}
      uVRegion {bIsValid=false, max=<0 0>, min=<0 0>}
     }
     pressedForeground {colorUseRule=UseColor_ColorTable, specifiedColor=<rgba 0 0 0 0>}
     pressedPadding=<ltrb 12 2.5 12 0.5>
     pressedSlateSound {resourceObject=null}
    }}
   slot {class=HorizontalBoxSlot, properties {verticalAlignment=VAlign_Center, padding=<ltrb 5 0 5 0>}}
  }
  HorizontalBox_15 {
   widgetClassPath=&HorizontalBox
   parent=RenameEventButton
   widget=HorizontalBox_15
   slot {class=ButtonSlot, properties {horizontalAlignment=HAlign_Left}}
  }
  Image_10 {
   widgetClassPath=&Image
   parent=HorizontalBox_15
   widget=Image_10
   properties {brush {
     bIsDynamicallyLoaded=false
     drawAs=Image
     imageType=NoImage
     mirroring=NoMirror
     resourceName=None
     resourceObject=&MI_SparksEditorUI_PlusMinus_NoBorder
     tiling=NoTile
     imageSize=<15 15>
     margin=<ltrb 0 0 0 0>
     outlineSettings {
      bUseBrushTransparency=false
      roundingType=HalfHeightRadius
      width=0
      color {colorUseRule=UseColor_Specified, specifiedColor=<rgba 0 0 0 0>}
      cornerRadii=<0 0 0 0>
     }
     tintColor {colorUseRule=UseColor_Specified, specifiedColor=<rgba 1 1 1 1>}
     uVRegion {bIsValid=false, max=<0 0>, min=<0 0>}
    }}
   slot {class=HorizontalBoxSlot, properties {horizontalAlignment=HAlign_Center, verticalAlignment=VAlign_Center, padding=<ltrb 0 0 3 0>}}
  }
  TextBlock_11 {
   widgetClassPath=&TextBlock
   parent=HorizontalBox_15
   widget=TextBlock_11
   properties {
    text=Rename Event
    font {
     bForceMonospaced=false
     bMaterialIsStencil=false
     fontMaterial=null
     fontObject=&HeadingNow
     letterSpacing=0
     monospacedWidth=1
     size=7.5
     skewAmount=0
     typefaceFontName=75 Medium
     outlineSettings {
      bApplyOutlineToDropShadows=false
      bMiteredCorners=false
      bSeparateFillAlpha=false
      outlineMaterial=null
      outlineSize=0
      outlineColor=<rgba 0 0 0 1>
     }
    }
   }
   slot {class=HorizontalBoxSlot, properties {verticalAlignment=VAlign_Center}}
  }
  CreateNewDAGroup {
   widgetClassPath=&VerticalBox
   bIsVariable=true
   parent=VerticalBox_0
   widget=CreateNewDAGroup
   properties {visibility=Collapsed}
   slot {class=VerticalBoxSlot}
  }
  Border_4 {
   widgetClassPath=&Border
   parent=CreateNewDAGroup
   widget=Border_4
   properties {
    background {
     bIsDynamicallyLoaded=false
     drawAs=Image
     imageType=NoImage
     mirroring=NoMirror
     resourceName=None
     resourceObject=null
     tiling=NoTile
     imageSize=<32 32>
     margin=<ltrb 0 0 0 0>
     outlineSettings {
      bUseBrushTransparency=false
      roundingType=HalfHeightRadius
      width=0
      color {colorUseRule=UseColor_Specified, specifiedColor=<rgba 0 0 0 0>}
      cornerRadii=<0 0 0 0>
     }
     tintColor {colorUseRule=UseColor_Specified, specifiedColor=<rgba 0.0729 0.0729 0.0729 1>}
     uVRegion {bIsValid=false, max=<0 0>, min=<0 0>}
    }
    padding=<ltrb 10 2 4 2>
   }
   slot {class=VerticalBoxSlot, properties {padding=<ltrb -10 0 -10 0>}}
  }
  HorizontalBox_13 {
   widgetClassPath=&HorizontalBox
   parent=Border_4
   widget=HorizontalBox_13
   slot {class=BorderSlot, properties {padding=<ltrb 10 2 4 2>}}
  }
  SizeBox_3 {
   widgetClassPath=&SizeBox
   parent=HorizontalBox_13
   widget=SizeBox_3
   properties {bOverride_MaxDesiredWidth=true, bOverride_MinDesiredWidth=true, maxDesiredWidth=240, minDesiredWidth=240}
   slot {class=HorizontalBoxSlot}
  }
  NewFxManagerNameTextBox {
   widgetClassPath=&EditorUtilityEditableTextBox
   bIsVariable=true
   parent=SizeBox_3
   widget=NewFxManagerNameTextBox
   properties {
    hintText=Insert Type if you want
    toolTipText=You can choose a different type than the default preview one if you want.
    widgetStyle {
     backgroundColor {colorUseRule=UseColor_Specified, specifiedColor=<rgba 1 1 1 1>}
     backgroundImageFocused {
      bIsDynamicallyLoaded=false
      drawAs=RoundedBox
      imageType=NoImage
      mirroring=NoMirror
      resourceName=None
      resourceObject=null
      tiling=NoTile
      imageSize=<0 0>
      margin=<ltrb 0 0 0 0>
      outlineSettings {
       bUseBrushTransparency=false
       roundingType=FixedRadius
       width=1
       color {colorUseRule=UseColor_ColorTable, specifiedColor=<rgba 0 0 0 0>}
       cornerRadii=<4 4 4 4>
      }
      tintColor {colorUseRule=UseColor_Specified, specifiedColor=<rgba 0.0048 0.0048 0.0048 1>}
      uVRegion {bIsValid=false, max=<0 0>, min=<0 0>}
     }
     backgroundImageHovered {
      bIsDynamicallyLoaded=false
      drawAs=RoundedBox
      imageType=NoImage
      mirroring=NoMirror
      resourceName=None
      resourceObject=null
      tiling=NoTile
      imageSize=<0 0>
      margin=<ltrb 0 0 0 0>
      outlineSettings {
       bUseBrushTransparency=false
       roundingType=FixedRadius
       width=1
       color {colorUseRule=UseColor_ColorTable, specifiedColor=<rgba 0 0 0 0>}
       cornerRadii=<4 4 4 4>
      }
      tintColor {colorUseRule=UseColor_Specified, specifiedColor=<rgba 0.0048 0.0048 0.0048 1>}
      uVRegion {bIsValid=false, max=<0 0>, min=<0 0>}
     }
     backgroundImageNormal {
      bIsDynamicallyLoaded=false
      drawAs=RoundedBox
      imageType=NoImage
      mirroring=NoMirror
      resourceName=None
      resourceObject=null
      tiling=NoTile
      imageSize=<0 0>
      margin=<ltrb 0 0 0 0>
      outlineSettings {
       bUseBrushTransparency=false
       roundingType=FixedRadius
       width=1
       color {colorUseRule=UseColor_ColorTable, specifiedColor=<rgba 0 0 0 0>}
       cornerRadii=<4 4 4 4>
      }
      tintColor {colorUseRule=UseColor_Specified, specifiedColor=<rgba 0.0048 0.0048 0.0048 1>}
      uVRegion {bIsValid=false, max=<0 0>, min=<0 0>}
     }
     backgroundImageReadOnly {
      bIsDynamicallyLoaded=false
      drawAs=RoundedBox
      imageType=NoImage
      mirroring=NoMirror
      resourceName=None
      resourceObject=null
      tiling=NoTile
      imageSize=<0 0>
      margin=<ltrb 0 0 0 0>
      outlineSettings {
       bUseBrushTransparency=false
       roundingType=FixedRadius
       width=1
       color {colorUseRule=UseColor_ColorTable, specifiedColor=<rgba 0 0 0 0>}
       cornerRadii=<4 4 4 4>
      }
      tintColor {colorUseRule=UseColor_Specified, specifiedColor=<rgba 0.0284 0.0284 0.0284 1>}
      uVRegion {bIsValid=false, max=<0 0>, min=<0 0>}
     }
     focusedForegroundColor {colorUseRule=UseColor_Specified, specifiedColor=<rgba 1 1 1 1>}
     foregroundColor {colorUseRule=UseColor_Specified, specifiedColor=<rgba 0.5271 0.5271 0.5271 1>}
     hScrollBarPadding=<ltrb 0 0 0 0>
     padding=<ltrb 8 3 8 4>
     readOnlyForegroundColor {colorUseRule=UseColor_Foreground, specifiedColor=<rgba 1 0 1 1>}
     scrollBarStyle {
      thickness=8
      draggedThumbImage {
       bIsDynamicallyLoaded=false
       drawAs=RoundedBox
       imageType=NoImage
       mirroring=NoMirror
       resourceName=None
       resourceObject=null
       tiling=NoTile
       imageSize=<0 0>
       margin=<ltrb 0 0 0 0>
       outlineSettings {
        bUseBrushTransparency=false
        roundingType=FixedRadius
        width=0
        color {colorUseRule=UseColor_Specified, specifiedColor=<rgba 0 0 0 0>}
        cornerRadii=<4 4 4 4>
       }
       tintColor {colorUseRule=UseColor_Specified, specifiedColor=<rgba 0.2159 0.2159 0.2159 1>}
       uVRegion {bIsValid=false, max=<0 0>, min=<0 0>}
      }
      horizontalBackgroundImage {
       bIsDynamicallyLoaded=false
       drawAs=NoDrawType
       imageType=NoImage
       mirroring=NoMirror
       resourceName=None
       resourceObject=null
       tiling=NoTile
       imageSize=<0 0>
       margin=<ltrb 0 0 0 0>
       outlineSettings {
        bUseBrushTransparency=false
        roundingType=HalfHeightRadius
        width=0
        color {colorUseRule=UseColor_Specified, specifiedColor=<rgba 0 0 0 0>}
        cornerRadii=<0 0 0 0>
       }
       tintColor {colorUseRule=UseColor_Specified, specifiedColor=<rgba 1 1 1 1>}
       uVRegion {bIsValid=false, max=<0 0>, min=<0 0>}
      }
      horizontalBottomSlotImage {
       bIsDynamicallyLoaded=false
       drawAs=NoDrawType
       imageType=NoImage
       mirroring=NoMirror
       resourceName=None
       resourceObject=null
       tiling=NoTile
       imageSize=<0 0>
       margin=<ltrb 0 0 0 0>
       outlineSettings {
        bUseBrushTransparency=false
        roundingType=HalfHeightRadius
        width=0
        color {colorUseRule=UseColor_Specified, specifiedColor=<rgba 0 0 0 0>}
        cornerRadii=<0 0 0 0>
       }
       tintColor {colorUseRule=UseColor_Specified, specifiedColor=<rgba 1 1 1 1>}
       uVRegion {bIsValid=false, max=<0 0>, min=<0 0>}
      }
      horizontalTopSlotImage {
       bIsDynamicallyLoaded=false
       drawAs=NoDrawType
       imageType=NoImage
       mirroring=NoMirror
       resourceName=None
       resourceObject=null
       tiling=NoTile
       imageSize=<0 0>
       margin=<ltrb 0 0 0 0>
       outlineSettings {
        bUseBrushTransparency=false
        roundingType=HalfHeightRadius
        width=0
        color {colorUseRule=UseColor_Specified, specifiedColor=<rgba 0 0 0 0>}
        cornerRadii=<0 0 0 0>
       }
       tintColor {colorUseRule=UseColor_Specified, specifiedColor=<rgba 1 1 1 1>}
       uVRegion {bIsValid=false, max=<0 0>, min=<0 0>}
      }
      hoveredThumbImage {
       bIsDynamicallyLoaded=false
       drawAs=RoundedBox
       imageType=NoImage
       mirroring=NoMirror
       resourceName=None
       resourceObject=null
       tiling=NoTile
       imageSize=<0 0>
       margin=<ltrb 0 0 0 0>
       outlineSettings {
        bUseBrushTransparency=false
        roundingType=FixedRadius
        width=0
        color {colorUseRule=UseColor_Specified, specifiedColor=<rgba 0 0 0 0>}
        cornerRadii=<4 4 4 4>
       }
       tintColor {colorUseRule=UseColor_Specified, specifiedColor=<rgba 0.2159 0.2159 0.2159 1>}
       uVRegion {bIsValid=false, max=<0 0>, min=<0 0>}
      }
      normalThumbImage {
       bIsDynamicallyLoaded=false
       drawAs=RoundedBox
       imageType=NoImage
       mirroring=NoMirror
       resourceName=None
       resourceObject=null
       tiling=NoTile
       imageSize=<0 0>
       margin=<ltrb 0 0 0 0>
       outlineSettings {
        bUseBrushTransparency=false
        roundingType=FixedRadius
        width=0
        color {colorUseRule=UseColor_Specified, specifiedColor=<rgba 0 0 0 0>}
        cornerRadii=<4 4 4 4>
       }
       tintColor {colorUseRule=UseColor_Specified, specifiedColor=<rgba 0.0953 0.0953 0.0953 1>}
       uVRegion {bIsValid=false, max=<0 0>, min=<0 0>}
      }
      verticalBackgroundImage {
       bIsDynamicallyLoaded=false
       drawAs=NoDrawType
       imageType=NoImage
       mirroring=NoMirror
       resourceName=None
       resourceObject=null
       tiling=NoTile
       imageSize=<0 0>
       margin=<ltrb 0 0 0 0>
       outlineSettings {
        bUseBrushTransparency=false
        roundingType=HalfHeightRadius
        width=0
        color {colorUseRule=UseColor_Specified, specifiedColor=<rgba 0 0 0 0>}
        cornerRadii=<0 0 0 0>
       }
       tintColor {colorUseRule=UseColor_Specified, specifiedColor=<rgba 1 1 1 1>}
       uVRegion {bIsValid=false, max=<0 0>, min=<0 0>}
      }
      verticalBottomSlotImage {
       bIsDynamicallyLoaded=false
       drawAs=NoDrawType
       imageType=NoImage
       mirroring=NoMirror
       resourceName=None
       resourceObject=null
       tiling=NoTile
       imageSize=<0 0>
       margin=<ltrb 0 0 0 0>
       outlineSettings {
        bUseBrushTransparency=false
        roundingType=HalfHeightRadius
        width=0
        color {colorUseRule=UseColor_Specified, specifiedColor=<rgba 0 0 0 0>}
        cornerRadii=<0 0 0 0>
       }
       tintColor {colorUseRule=UseColor_Specified, specifiedColor=<rgba 1 1 1 1>}
       uVRegion {bIsValid=false, max=<0 0>, min=<0 0>}
      }
      verticalTopSlotImage {
       bIsDynamicallyLoaded=false
       drawAs=NoDrawType
       imageType=NoImage
       mirroring=NoMirror
       resourceName=None
       resourceObject=null
       tiling=NoTile
       imageSize=<0 0>
       margin=<ltrb 0 0 0 0>
       outlineSettings {
        bUseBrushTransparency=false
        roundingType=HalfHeightRadius
        width=0
        color {colorUseRule=UseColor_Specified, specifiedColor=<rgba 0 0 0 0>}
        cornerRadii=<0 0 0 0>
       }
       tintColor {colorUseRule=UseColor_Specified, specifiedColor=<rgba 1 1 1 1>}
       uVRegion {bIsValid=false, max=<0 0>, min=<0 0>}
      }
     }
     textStyle {
      overflowPolicy=Clip
      transformPolicy=None
      colorAndOpacity {colorUseRule=UseColor_Foreground, specifiedColor=<rgba 1 0 1 1>}
      font {
       bForceMonospaced=false
       bMaterialIsStencil=false
       fontMaterial=null
       fontObject=&HeadingNow
       letterSpacing=0
       monospacedWidth=1
       size=11.25
       skewAmount=0
       typefaceFontName=76 Bold
       outlineSettings {
        bApplyOutlineToDropShadows=false
        bMiteredCorners=false
        bSeparateFillAlpha=false
        outlineMaterial=null
        outlineSize=0
        outlineColor=<rgba 0 0 0 1>
       }
      }
      highlightColor {colorUseRule=UseColor_ColorTable, specifiedColor=<rgba 0 0 0 0>}
      highlightShape {
       bIsDynamicallyLoaded=false
       drawAs=Image
       imageType=NoImage
       mirroring=NoMirror
       resourceName=None
       resourceObject=null
       tiling=NoTile
       imageSize=<0 0>
       margin=<ltrb 0 0 0 0>
       outlineSettings {
        bUseBrushTransparency=false
        roundingType=HalfHeightRadius
        width=0
        color {colorUseRule=UseColor_Specified, specifiedColor=<rgba 0 0 0 0>}
        cornerRadii=<0 0 0 0>
       }
       tintColor {colorUseRule=UseColor_ColorTable, specifiedColor=<rgba 0 0 0 0>}
       uVRegion {bIsValid=false, max=<0 0>, min=<0 0>}
      }
      selectedBackgroundColor {colorUseRule=UseColor_ColorTable, specifiedColor=<rgba 0 0 0 0>}
      shadowColorAndOpacity=<rgba 0 0 0 1>
      shadowOffset=<0 0>
      strikeBrush {
       bIsDynamicallyLoaded=false
       drawAs=Image
       imageType=NoImage
       mirroring=NoMirror
       resourceName=None
       resourceObject=null
       tiling=NoTile
       imageSize=<32 32>
       margin=<ltrb 0 0 0 0>
       outlineSettings {
        bUseBrushTransparency=false
        roundingType=HalfHeightRadius
        width=0
        color {colorUseRule=UseColor_Specified, specifiedColor=<rgba 0 0 0 0>}
        cornerRadii=<0 0 0 0>
       }
       tintColor {colorUseRule=UseColor_Specified, specifiedColor=<rgba 1 1 1 1>}
       uVRegion {bIsValid=false, max=<0 0>, min=<0 0>}
      }
      underlineBrush {
       bIsDynamicallyLoaded=false
       drawAs=Image
       imageType=NoImage
       mirroring=NoMirror
       resourceName=None
       resourceObject=null
       tiling=NoTile
       imageSize=<32 32>
       margin=<ltrb 0 0 0 0>
       outlineSettings {
        bUseBrushTransparency=false
        roundingType=HalfHeightRadius
        width=0
        color {colorUseRule=UseColor_Specified, specifiedColor=<rgba 0 0 0 0>}
        cornerRadii=<0 0 0 0>
       }
       tintColor {colorUseRule=UseColor_Specified, specifiedColor=<rgba 1 1 1 1>}
       uVRegion {bIsValid=false, max=<0 0>, min=<0 0>}
      }
     }
     vScrollBarPadding=<ltrb 0 0 0 0>
    }
   }
   slot {class=SizeBoxSlot}
  }
  NewSequenceEventAddButton {
   widgetClassPath=&EditorUtilityButton
   parent=HorizontalBox_13
   widget=NewSequenceEventAddButton
   properties {widgetStyle {
     clickedSlateSound {resourceObject=null}
     disabled {
      bIsDynamicallyLoaded=false
      drawAs=RoundedBox
      imageType=NoImage
      mirroring=NoMirror
      resourceName=None
      resourceObject=null
      tiling=NoTile
      imageSize=<32 32>
      margin=<ltrb 0.5 0.5 0.5 0.5>
      outlineSettings {
       bUseBrushTransparency=false
       roundingType=FixedRadius
       width=0
       color {colorUseRule=UseColor_ColorTable, specifiedColor=<rgba 0 0 0 0>}
       cornerRadii=<4 4 4 4>
      }
      tintColor {colorUseRule=UseColor_Specified, specifiedColor=<rgba 0.0395 0.0395 0.0395 1>}
      uVRegion {bIsValid=false, max=<0 0>, min=<0 0>}
     }
     disabledForeground {colorUseRule=UseColor_ColorTable, specifiedColor=<rgba 0 0 0 0>}
     hovered {
      bIsDynamicallyLoaded=false
      drawAs=RoundedBox
      imageType=NoImage
      mirroring=NoMirror
      resourceName=None
      resourceObject=null
      tiling=NoTile
      imageSize=<32 32>
      margin=<ltrb 0 0 0 0>
      outlineSettings {
       bUseBrushTransparency=false
       roundingType=FixedRadius
       width=0
       color {colorUseRule=UseColor_ColorTable, specifiedColor=<rgba 0 0 0 0>}
       cornerRadii=<4 4 4 4>
      }
      tintColor {colorUseRule=UseColor_Specified, specifiedColor=<rgba 0.1126 0.7031 0 1>}
      uVRegion {bIsValid=false, max=<0 0>, min=<0 0>}
     }
     hoveredForeground {colorUseRule=UseColor_ColorTable, specifiedColor=<rgba 0 0 0 0>}
     hoveredSlateSound {resourceObject=null}
     normal {
      bIsDynamicallyLoaded=false
      drawAs=RoundedBox
      imageType=NoImage
      mirroring=NoMirror
      resourceName=None
      resourceObject=null
      tiling=NoTile
      imageSize=<32 32>
      margin=<ltrb 0.5 0.5 0.5 0.5>
      outlineSettings {
       bUseBrushTransparency=false
       roundingType=FixedRadius
       width=0
       color {colorUseRule=UseColor_ColorTable, specifiedColor=<rgba 0 0 0 0>}
       cornerRadii=<4 4 4 4>
      }
      tintColor {colorUseRule=UseColor_Specified, specifiedColor=<rgba 0.0771 0.6167 0.1289 1>}
      uVRegion {bIsValid=false, max=<0 0>, min=<0 0>}
     }
     normalForeground {colorUseRule=UseColor_ColorTable, specifiedColor=<rgba 0 0 0 0>}
     normalPadding=<ltrb 12 1.5 12 1.5>
     pressed {
      bIsDynamicallyLoaded=false
      drawAs=RoundedBox
      imageType=NoImage
      mirroring=NoMirror
      resourceName=None
      resourceObject=null
      tiling=NoTile
      imageSize=<32 32>
      margin=<ltrb 0.5 0.5 0.5 0.5>
      outlineSettings {
       bUseBrushTransparency=false
       roundingType=FixedRadius
       width=0
       color {colorUseRule=UseColor_ColorTable, specifiedColor=<rgba 0 0 0 0>}
       cornerRadii=<4 4 4 4>
      }
      tintColor {colorUseRule=UseColor_Specified, specifiedColor=<rgba 0.1806 0.4583 0.1572 1>}
      uVRegion {bIsValid=false, max=<0 0>, min=<0 0>}
     }
     pressedForeground {colorUseRule=UseColor_ColorTable, specifiedColor=<rgba 0 0 0 0>}
     pressedPadding=<ltrb 12 2.5 12 0.5>
     pressedSlateSound {resourceObject=null}
    }}
   slot {class=HorizontalBoxSlot, properties {verticalAlignment=VAlign_Center, padding=<ltrb 10 0 0 0>}}
  }
  HorizontalBox_14 {
   widgetClassPath=&HorizontalBox
   parent=NewSequenceEventAddButton
   widget=HorizontalBox_14
   slot {class=ButtonSlot, properties {horizontalAlignment=HAlign_Left}}
  }
  Image_14 {
   widgetClassPath=&Image
   parent=HorizontalBox_14
   widget=Image_14
   properties {brush {
     bIsDynamicallyLoaded=false
     drawAs=Image
     imageType=NoImage
     mirroring=NoMirror
     resourceName=None
     resourceObject=&MI_SparksEditorUI_PlusMinus_NoBorder
     tiling=NoTile
     imageSize=<15 15>
     margin=<ltrb 0 0 0 0>
     outlineSettings {
      bUseBrushTransparency=false
      roundingType=HalfHeightRadius
      width=0
      color {colorUseRule=UseColor_Specified, specifiedColor=<rgba 0 0 0 0>}
      cornerRadii=<0 0 0 0>
     }
     tintColor {colorUseRule=UseColor_Specified, specifiedColor=<rgba 1 1 1 1>}
     uVRegion {bIsValid=false, max=<0 0>, min=<0 0>}
    }}
   slot {class=HorizontalBoxSlot, properties {horizontalAlignment=HAlign_Center, verticalAlignment=VAlign_Center, padding=<ltrb 0 0 3 0>}}
  }
  TextBlock_9 {
   widgetClassPath=&TextBlock
   parent=HorizontalBox_14
   widget=TextBlock_9
   properties {
    text=Add and Create
    font {
     bForceMonospaced=false
     bMaterialIsStencil=false
     fontMaterial=null
     fontObject=&HeadingNow
     letterSpacing=0
     monospacedWidth=1
     size=7.5
     skewAmount=0
     typefaceFontName=75 Medium
     outlineSettings {
      bApplyOutlineToDropShadows=false
      bMiteredCorners=false
      bSeparateFillAlpha=false
      outlineMaterial=null
      outlineSize=0
      outlineColor=<rgba 0 0 0 1>
     }
    }
   }
   slot {class=HorizontalBoxSlot, properties {verticalAlignment=VAlign_Center}}
  }
  CreateChangeButton {
   widgetClassPath=&EditorUtilityButton
   parent=HorizontalBox_13
   widget=CreateChangeButton
   properties {widgetStyle {
     clickedSlateSound {resourceObject=null}
     disabled {
      bIsDynamicallyLoaded=false
      drawAs=RoundedBox
      imageType=NoImage
      mirroring=NoMirror
      resourceName=None
      resourceObject=null
      tiling=NoTile
      imageSize=<32 32>
      margin=<ltrb 0.5 0.5 0.5 0.5>
      outlineSettings {
       bUseBrushTransparency=false
       roundingType=FixedRadius
       width=0
       color {colorUseRule=UseColor_ColorTable, specifiedColor=<rgba 0 0 0 0>}
       cornerRadii=<4 4 4 4>
      }
      tintColor {colorUseRule=UseColor_Specified, specifiedColor=<rgba 0.0395 0.0395 0.0395 1>}
      uVRegion {bIsValid=false, max=<0 0>, min=<0 0>}
     }
     disabledForeground {colorUseRule=UseColor_ColorTable, specifiedColor=<rgba 0 0 0 0>}
     hovered {
      bIsDynamicallyLoaded=false
      drawAs=RoundedBox
      imageType=NoImage
      mirroring=NoMirror
      resourceName=None
      resourceObject=null
      tiling=NoTile
      imageSize=<32 32>
      margin=<ltrb 0 0 0 0>
      outlineSettings {
       bUseBrushTransparency=false
       roundingType=FixedRadius
       width=0
       color {colorUseRule=UseColor_ColorTable, specifiedColor=<rgba 0 0 0 0>}
       cornerRadii=<4 4 4 4>
      }
      tintColor {colorUseRule=UseColor_Specified, specifiedColor=<rgba 0.5437 0.7344 0.1989 1>}
      uVRegion {bIsValid=false, max=<0 0>, min=<0 0>}
     }
     hoveredForeground {colorUseRule=UseColor_ColorTable, specifiedColor=<rgba 0 0 0 0>}
     hoveredSlateSound {resourceObject=null}
     normal {
      bIsDynamicallyLoaded=false
      drawAs=RoundedBox
      imageType=NoImage
      mirroring=NoMirror
      resourceName=None
      resourceObject=null
      tiling=NoTile
      imageSize=<32 32>
      margin=<ltrb 0.5 0.5 0.5 0.5>
      outlineSettings {
       bUseBrushTransparency=false
       roundingType=FixedRadius
       width=0
       color {colorUseRule=UseColor_ColorTable, specifiedColor=<rgba 0 0 0 0>}
       cornerRadii=<4 4 4 4>
      }
      tintColor {colorUseRule=UseColor_Specified, specifiedColor=<rgba 0.3885 0.5521 0.0926 1>}
      uVRegion {bIsValid=false, max=<0 0>, min=<0 0>}
     }
     normalForeground {colorUseRule=UseColor_ColorTable, specifiedColor=<rgba 0 0 0 0>}
     normalPadding=<ltrb 12 1.5 12 1.5>
     pressed {
      bIsDynamicallyLoaded=false
      drawAs=RoundedBox
      imageType=NoImage
      mirroring=NoMirror
      resourceName=None
      resourceObject=null
      tiling=NoTile
      imageSize=<32 32>
      margin=<ltrb 0.5 0.5 0.5 0.5>
      outlineSettings {
       bUseBrushTransparency=false
       roundingType=FixedRadius
       width=0
       color {colorUseRule=UseColor_ColorTable, specifiedColor=<rgba 0 0 0 0>}
       cornerRadii=<4 4 4 4>
      }
      tintColor {colorUseRule=UseColor_Specified, specifiedColor=<rgba 0.1806 0.4583 0.1572 1>}
      uVRegion {bIsValid=false, max=<0 0>, min=<0 0>}
     }
     pressedForeground {colorUseRule=UseColor_ColorTable, specifiedColor=<rgba 0 0 0 0>}
     pressedPadding=<ltrb 12 2.5 12 0.5>
     pressedSlateSound {resourceObject=null}
    }}
   slot {class=HorizontalBoxSlot, properties {verticalAlignment=VAlign_Center, padding=<ltrb 10 0 0 0>}}
  }
  HorizontalBox_9 {
   widgetClassPath=&HorizontalBox
   parent=CreateChangeButton
   widget=HorizontalBox_9
   slot {class=ButtonSlot, properties {horizontalAlignment=HAlign_Left}}
  }
  Image_9 {
   widgetClassPath=&Image
   parent=HorizontalBox_9
   widget=Image_9
   properties {brush {
     bIsDynamicallyLoaded=false
     drawAs=Image
     imageType=NoImage
     mirroring=NoMirror
     resourceName=None
     resourceObject=&MI_SparksEditorUI_PlusMinus_NoBorder
     tiling=NoTile
     imageSize=<15 15>
     margin=<ltrb 0 0 0 0>
     outlineSettings {
      bUseBrushTransparency=false
      roundingType=HalfHeightRadius
      width=0
      color {colorUseRule=UseColor_Specified, specifiedColor=<rgba 0 0 0 0>}
      cornerRadii=<0 0 0 0>
     }
     tintColor {colorUseRule=UseColor_Specified, specifiedColor=<rgba 1 1 1 1>}
     uVRegion {bIsValid=false, max=<0 0>, min=<0 0>}
    }}
   slot {class=HorizontalBoxSlot, properties {horizontalAlignment=HAlign_Center, verticalAlignment=VAlign_Center, padding=<ltrb 0 0 3 0>}}
  }
  TextBlock_5 {
   widgetClassPath=&TextBlock
   parent=HorizontalBox_9
   widget=TextBlock_5
   properties {
    text=Create and Change
    font {
     bForceMonospaced=false
     bMaterialIsStencil=false
     fontMaterial=null
     fontObject=&HeadingNow
     letterSpacing=0
     monospacedWidth=1
     size=7.5
     skewAmount=0
     typefaceFontName=75 Medium
     outlineSettings {
      bApplyOutlineToDropShadows=false
      bMiteredCorners=false
      bSeparateFillAlpha=false
      outlineMaterial=null
      outlineSize=0
      outlineColor=<rgba 0 0 0 1>
     }
    }
   }
   slot {class=HorizontalBoxSlot, properties {verticalAlignment=VAlign_Center}}
  }
  Border_5 {
   widgetClassPath=&Border
   parent=VerticalBox_0
   widget=Border_5
   properties {
    background {
     bIsDynamicallyLoaded=false
     drawAs=Image
     imageType=NoImage
     mirroring=NoMirror
     resourceName=None
     resourceObject=null
     tiling=NoTile
     imageSize=<32 32>
     margin=<ltrb 0 0 0 0>
     outlineSettings {
      bUseBrushTransparency=false
      roundingType=HalfHeightRadius
      width=0
      color {colorUseRule=UseColor_Specified, specifiedColor=<rgba 0 0 0 0>}
      cornerRadii=<0 0 0 0>
     }
     tintColor {colorUseRule=UseColor_Specified, specifiedColor=<rgba 0.0729 0.0729 0.0729 1>}
     uVRegion {bIsValid=false, max=<0 0>, min=<0 0>}
    }
    padding=<ltrb 10 2 4 2>
   }
   slot {class=VerticalBoxSlot, properties {padding=<ltrb -10 4 -10 5>}}
  }
  HorizontalBox_8 {
   widgetClassPath=&HorizontalBox
   parent=Border_5
   widget=HorizontalBox_8
   slot {class=BorderSlot, properties {padding=<ltrb 10 2 4 2>}}
  }
  DANamePreviewTitle {
   widgetClassPath=&TextBlock
   parent=HorizontalBox_8
   widget=DANamePreviewTitle
   properties {
    text="Data Asset Name Preview: "
    font {
     bForceMonospaced=false
     bMaterialIsStencil=false
     fontMaterial=null
     fontObject=&HeadingNow
     letterSpacing=0
     monospacedWidth=1
     size=9.75
     skewAmount=0
     typefaceFontName=75 Medium
     outlineSettings {
      bApplyOutlineToDropShadows=false
      bMiteredCorners=false
      bSeparateFillAlpha=false
      outlineMaterial=null
      outlineSize=0
      outlineColor=<rgba 0 0 0 1>
     }
    }
   }
   slot {class=HorizontalBoxSlot}
  }
  DANamePreviewText {
   widgetClassPath=&TextBlock
   bIsVariable=true
   parent=HorizontalBox_8
   widget=DANamePreviewText
   properties {
    colorAndOpacity {colorUseRule=UseColor_Specified, specifiedColor=<rgba 0.4062 0.4062 0.4062 1>}
    font {
     bForceMonospaced=false
     bMaterialIsStencil=false
     fontMaterial=null
     fontObject=&HeadingNow
     letterSpacing=0
     monospacedWidth=1
     size=9.75
     skewAmount=0
     typefaceFontName=74 Regular
     outlineSettings {
      bApplyOutlineToDropShadows=false
      bMiteredCorners=false
      bSeparateFillAlpha=false
      outlineMaterial=null
      outlineSize=0
      outlineColor=<rgba 0 0 0 1>
     }
    }
   }
   slot {class=HorizontalBoxSlot}
  }
  bg {
   widgetClassPath=&Image
   parent=NewSequenceEventsInsideGroup
   widget=bg
   properties {visibility=Collapsed}
   slot {class=OverlaySlot, properties {horizontalAlignment=HAlign_Fill, verticalAlignment=VAlign_Fill}}
  }
  Nosequenceeventstext {
   widgetClassPath=&TextBlock
   bIsVariable=true
   parent=Overlay
   widget=Nosequenceeventstext
   properties {
    justification=Center
    text=All avaliable events are in use.
    visibility=Collapsed
    font {
     bForceMonospaced=false
     bMaterialIsStencil=false
     fontMaterial=null
     fontObject=&Roboto
     letterSpacing=0
     monospacedWidth=1
     size=15
     skewAmount=0
     typefaceFontName=Bold
     outlineSettings {
      bApplyOutlineToDropShadows=false
      bMiteredCorners=false
      bSeparateFillAlpha=false
      outlineMaterial=null
      outlineSize=0
      outlineColor=<rgba 0 0 0 1>
     }
    }
   }
   slot {class=OverlaySlot, properties {horizontalAlignment=HAlign_Center, verticalAlignment=VAlign_Center}}
  }
  CloseNewSequenceEventWindowButton {
   widgetClassPath=&EditorUtilityButton
   parent=Overlay
   widget=CloseNewSequenceEventWindowButton
   properties {
    clipping=ClipToBounds
    widgetStyle {
     clickedSlateSound {resourceObject=null}
     disabled {
      bIsDynamicallyLoaded=false
      drawAs=RoundedBox
      imageType=NoImage
      mirroring=NoMirror
      resourceName=None
      resourceObject=null
      tiling=NoTile
      imageSize=<20 20>
      margin=<ltrb 0 0 0 0>
      outlineSettings {
       bUseBrushTransparency=false
       roundingType=FixedRadius
       width=0
       color {colorUseRule=UseColor_ColorTable, specifiedColor=<rgba 0 0 0 0>}
       cornerRadii=<4 4 4 4>
      }
      tintColor {colorUseRule=UseColor_Specified, specifiedColor=<rgba 0.0395 0.0395 0.0395 1>}
      uVRegion {bIsValid=false, max=<0 0>, min=<0 0>}
     }
     disabledForeground {colorUseRule=UseColor_ColorTable, specifiedColor=<rgba 0 0 0 0>}
     hovered {
      bIsDynamicallyLoaded=false
      drawAs=RoundedBox
      imageType=NoImage
      mirroring=NoMirror
      resourceName=None
      resourceObject=null
      tiling=NoTile
      imageSize=<32 32>
      margin=<ltrb 0 0 0 0>
      outlineSettings {
       bUseBrushTransparency=false
       roundingType=FixedRadius
       width=0
       color {colorUseRule=UseColor_ColorTable, specifiedColor=<rgba 0 0 0 0>}
       cornerRadii=<4 4 4 4>
      }
      tintColor {colorUseRule=UseColor_Specified, specifiedColor=<rgba 1 0.1252 0.253 1>}
      uVRegion {bIsValid=false, max=<0 0>, min=<0 0>}
     }
     hoveredForeground {colorUseRule=UseColor_ColorTable, specifiedColor=<rgba 0 0 0 0>}
     hoveredSlateSound {resourceObject=null}
     normal {
      bIsDynamicallyLoaded=false
      drawAs=RoundedBox
      imageType=NoImage
      mirroring=NoMirror
      resourceName=None
      resourceObject=null
      tiling=NoTile
      imageSize=<20 20>
      margin=<ltrb 0 0 0 0>
      outlineSettings {
       bUseBrushTransparency=false
       roundingType=FixedRadius
       width=0
       color {colorUseRule=UseColor_ColorTable, specifiedColor=<rgba 0 0 0 0>}
       cornerRadii=<4 4 4 4>
      }
      tintColor {colorUseRule=UseColor_Specified, specifiedColor=<rgba 0.775 0 0.0304 1>}
      uVRegion {bIsValid=false, max=<0 0>, min=<0 0>}
     }
     normalForeground {colorUseRule=UseColor_ColorTable, specifiedColor=<rgba 0 0 0 0>}
     normalPadding=<ltrb 12 1.5 12 1.5>
     pressed {
      bIsDynamicallyLoaded=false
      drawAs=RoundedBox
      imageType=NoImage
      mirroring=NoMirror
      resourceName=None
      resourceObject=null
      tiling=NoTile
      imageSize=<20 20>
      margin=<ltrb 0 0 0 0>
      outlineSettings {
       bUseBrushTransparency=false
       roundingType=FixedRadius
       width=0
       color {colorUseRule=UseColor_ColorTable, specifiedColor=<rgba 0 0 0 0>}
       cornerRadii=<4 4 4 4>
      }
      tintColor {colorUseRule=UseColor_Specified, specifiedColor=<rgba 0.2396 0 0 1>}
      uVRegion {bIsValid=false, max=<0 0>, min=<0 0>}
     }
     pressedForeground {colorUseRule=UseColor_ColorTable, specifiedColor=<rgba 0 0 0 0>}
     pressedPadding=<ltrb 12 2.5 12 0.5>
     pressedSlateSound {resourceObject=null}
    }
   }
   slot {class=OverlaySlot, properties {horizontalAlignment=HAlign_Right, padding=<ltrb 10 3.5 10 0>}}
  }
  TextBlock_7 {
   widgetClassPath=&TextBlock
   parent=CloseNewSequenceEventWindowButton
   widget=TextBlock_7
   properties {
    text=Close
    font {
     bForceMonospaced=false
     bMaterialIsStencil=false
     fontMaterial=null
     fontObject=&HeadingNow
     letterSpacing=0
     monospacedWidth=1
     size=7.5
     skewAmount=0
     typefaceFontName=75 Medium
     outlineSettings {
      bApplyOutlineToDropShadows=false
      bMiteredCorners=false
      bSeparateFillAlpha=false
      outlineMaterial=null
      outlineSize=0
      outlineColor=<rgba 0 0 0 1>
     }
    }
   }
   slot {class=ButtonSlot, properties {horizontalAlignment=HAlign_Fill, padding=<ltrb 0 0 0 0>}}
  }
 ]
}
