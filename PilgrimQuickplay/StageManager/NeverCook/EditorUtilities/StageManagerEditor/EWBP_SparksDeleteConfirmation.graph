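# Editor utility popup in the Sparks Stage Manager editor that asks whether to remove a data asset's entry or also delete the asset.
# OpenWindow stores and names the object, hiding "Remove Entry Only" for MIDI sequence and mood assets.
# "Delete Entry + Asset" requires Yes/No confirmation.
# The widget deletes nothing itself: OnClosed broadcasts the choice, object and a property name derived by a class-cast chain, or empty values on X.
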EditorUtilityWidgetBlueprint /PilgrimQuickplay/StageManager/NeverCook/EditorUtilities/StageManagerEditor/EWBP_SparksDeleteConfirmation
graphs [
 EventGraph {nodes [
   "On Pressed (CloseNewSequenceEventWindowButton)"() {is=title K2Node_ComponentBoundEvent}
   > SetVisibility(InVisibility=Collapsed)
   > Call On Closed!(execute:Exec, self:EWBP Sparks Delete Confirmation Object Reference, IsDeleted:bool, PropertyName:Name, ObjToDelete:Object Reference) -> (then:Exec) {is=title K2Node_CallDelegate}
   "On Pressed (NewSequenceEventAddButton)"() {is=title K2Node_ComponentBoundEvent}
   > SetVisibility(InVisibility=Collapsed)
   > Call On Closed(PropertyName=GetPropertyName!() -> (Property Name:Name=None), ObjToDelete=Get Object() -> (Object:Object Reference))
   "On Pressed (CreateChangeButton)"() {is=title K2Node_ComponentBoundEvent}
   > SetVisibility(self=Get ConfirmationGroup() -> (ConfirmationGroup:Vertical Box Object Reference), InVisibility=Visible)
   > SetVisibility(self=Get DeleteButtonsGroup() -> (DeleteButtonsGroup:Vertical Box Object Reference), InVisibility=Collapsed)
   "On Pressed (CreateChangeButton_1)"() {is=title K2Node_ComponentBoundEvent}
   > SetVisibility(self=Get ConfirmationGroup(), InVisibility=Collapsed)
   > SetVisibility(self=Get DeleteButtonsGroup(), InVisibility=Visible)
   "On Pressed (NewSequenceEventAddButton_1)"() {is=title K2Node_ComponentBoundEvent}
   > SetVisibility(InVisibility=Collapsed)
   > Call On Closed(IsDeleted=true, PropertyName=GetPropertyName() -> (Property Name=None), ObjToDelete=Get Object())
  ]}
 GetPropertyName {nodes [
   GetPropertyName() {is=title K2Node_FunctionEntry}
   > n1 = Get!(execute:Exec) -> (then:Exec, else:Exec, Object:Object Reference)
   > n2 = Cast To MojoFxManagerSequence_DataAsset!(execute, Object=@n1) -> (then, CastFailed, AsMojo Fx Manager Sequence Data Asset:Mojo Fx Manager Sequence Data Asset Object Reference) {is=title K2Node_DynamicCast}
   > Return Node!(execute, Property Name:Name=MojoStage_SequenceData)
   n3 = Cast To BP_MojoStageSequence_DataAsset!(execute=@n2.CastFailed, Object=@n1) -> (then, CastFailed, AsBP Mojo Stage Sequence Data Asset:BP Mojo Stage Sequence Data Asset Object Reference) {is=title K2Node_DynamicCast}
   > Return Node(Property Name=MojoStage_Sequences PoolData)
   n4 = Cast To MojoFxManagerEvent_DataAsset!(execute=@n3.CastFailed, Object=@n1) -> (then, CastFailed, AsMojo Fx Manager Event Data Asset:Mojo Fx Manager Event Data Asset Object Reference) {is=title K2Node_DynamicCast}
   > Return Node(Property Name=FxManager_Event StepData)
   n5 = Cast To BP_MojoStageMapMidiSequence_DataAsset!(execute=@n4.CastFailed, Object=@n1) -> (then, CastFailed, AsBP Mojo Stage Map Midi Sequence Data Asset:BP Mojo Stage Map Midi Sequence Data Asset Object Reference) {is=title K2Node_DynamicCast}
   > Return Node(Property Name=MojoStage_MIDISequenceData)
   n6 = Cast To BP_MojoStageMood_DataAsset!(execute=@n5.CastFailed, Object=@n1) -> (then, CastFailed, AsBP Mojo Stage Mood Data Asset:BP Mojo Stage Mood Data Asset Object Reference) {is=title K2Node_DynamicCast}
   > Return Node(Property Name=MojoStage_Mood)
   Return Node(execute=@n6.CastFailed, Property Name=None)
  ]}
 OnClosed {nodes [OnClosed!() -> (then, IsDeleted:bool=~, PropertyName:Name, ObjToDelete:Object Reference) {is=title K2Node_FunctionEntry}]}
 OpenWindow {nodes [
   n1 = OpenWindow!() -> (then, Object:Object Reference) {is=title K2Node_FunctionEntry}
   > n2 = Set Object(Object:Object Reference=@n1) -> (Output_Get:Object Reference)
   > "SetText (Text)"(self=Get DANamePreviewText() -> (DANamePreviewText:Text Object Reference), InText="To Text (String)"(InString=GetDisplayName(Object=@n2)))
   > SetVisibility(InVisibility=SelfHitTestInvisible)
   > SetVisibility(self=Get ConfirmationGroup(), InVisibility=Collapsed)
   > SetVisibility(self=Get DeleteButtonsGroup(), InVisibility=Visible)
   n3 = GetPropertyName() -> (Property Name=None) {is=title K2Node_CallFunction}
   n4 = "Equal (Name)"(A=@n3, B=MojoStage_MIDISequenceData)
   n5 = "Equal (Name)"(A=@n3, B=MojoStage_Mood)
   > SetVisibility(
    self=Get NewSequenceEventAddButton() -> (NewSequenceEventAddButton:Editor Utility Button Object Reference)
    InVisibility=Select(Option 0=Visible, Option 1=Collapsed, Index=OR Boolean(A=@n5, B=@n4))
   )
  ]}
]
variables [
 CurrentEventSelected {type=string, default=""}
 Object {type=Object, default=null}
 OnClosed {default=()}
 Property Name {type=string, default=None}
]
widgets {
 info {rootWidgetClass=&Overlay, widgetCount=42}
 widgets [
  NewSequenceEventWindow {widgetClassPath=&Overlay, bIsVariable=true, widget=NewSequenceEventWindow}
  Image_40 {
   widgetClassPath=&Image
   parent=NewSequenceEventWindow
   widget=Image_40
   properties {colorAndOpacity=<rgba 0 0 0 0.757>}
   slot {class=OverlaySlot, properties {horizontalAlignment=HAlign_Fill, verticalAlignment=VAlign_Fill, padding=<ltrb -6000 -6000 -6000 -6000>}}
  }
  SizeBox_2 {
   widgetClassPath=&SizeBox
   parent=NewSequenceEventWindow
   widget=SizeBox_2
   properties {bOverride_MinDesiredWidth=true, minDesiredWidth=600}
   slot {class=OverlaySlot, properties {horizontalAlignment=HAlign_Fill, verticalAlignment=VAlign_Fill}}
  }
  Overlay {widgetClassPath=&Overlay, parent=SizeBox_2, widget=Overlay, slot {class=SizeBoxSlot}}
  Image_7 {
   widgetClassPath=&Image
   parent=Overlay
   widget=Image_7
   properties {
    brush {
     bIsDynamicallyLoaded=false
     drawAs=Image
     imageType=NoImage
     mirroring=NoMirror
     resourceName=None
     resourceObject=null
     tiling=NoTile
     imageSize=<32 32>
     margin=<ltrb 0 0 0 0>
     outlineSettings {
      bUseBrushTransparency=false
      roundingType=HalfHeightRadius
      width=0
      color {colorUseRule=UseColor_Specified, specifiedColor=<rgba 0 0 0 0>}
      cornerRadii=<0 0 0 0>
     }
     tintColor {colorUseRule=UseColor_Specified, specifiedColor=<rgba 0 0 0 0.232>}
     uVRegion {bIsValid=false, max=<0 0>, min=<0 0>}
    }
    colorAndOpacity=<rgba 0.026 0.026 0.026 1>
    renderTransform {angle=0, scale=<1.005 1.01>, shear=<0 0>, translation=<0 0.5>}
   }
   slot {class=OverlaySlot, properties {horizontalAlignment=HAlign_Fill, verticalAlignment=VAlign_Fill}}
  }
  Image_8 {
   widgetClassPath=&Image
   parent=Overlay
   widget=Image_8
   properties {colorAndOpacity=<rgba 0.026 0.026 0.026 1>}
   slot {class=OverlaySlot, properties {horizontalAlignment=HAlign_Fill, verticalAlignment=VAlign_Fill}}
  }
  Border {
   widgetClassPath=&Border
   parent=Overlay
   widget=Border
   properties {clipping=ClipToBounds, horizontalAlignment=HAlign_Left, brushColor=<rgba 0.063 0.1406 0.0994 1>, padding=<ltrb 10 2 4 2>}
   slot {class=OverlaySlot, properties {horizontalAlignment=HAlign_Fill}}
  }
  EventsTitleText {
   widgetClassPath=&TextBlock
   bIsVariable=true
   parent=Border
   widget=EventsTitleText
   properties {
    text=Delete or remove entry
    colorAndOpacity {colorUseRule=UseColor_Specified, specifiedColor=<rgba 0.6562 0.6562 0.6562 1>}
    font {
     bForceMonospaced=false
     bMaterialIsStencil=false
     fontMaterial=null
     fontObject=&HeadingNow
     letterSpacing=0
     monospacedWidth=1
     size=11.25
     skewAmount=0
     typefaceFontName=75 Medium
     outlineSettings {
      bApplyOutlineToDropShadows=false
      bMiteredCorners=false
      bSeparateFillAlpha=false
      outlineMaterial=null
      outlineSize=0
      outlineColor=<rgba 0 0 0 1>
     }
    }
   }
   slot {class=BorderSlot, properties {horizontalAlignment=HAlign_Left, padding=<ltrb 10 2 4 2>}}
  }
  SizeBox_5 {
   widgetClassPath=&SizeBox
   parent=Overlay
   widget=SizeBox_5
   properties {bOverride_MinDesiredWidth=true, minDesiredWidth=670}
   slot {class=OverlaySlot, properties {horizontalAlignment=HAlign_Fill, verticalAlignment=VAlign_Fill}}
  }
  NewSequenceEventsInsideGroup {
   widgetClassPath=&Overlay
   bIsVariable=true
   parent=SizeBox_5
   widget=NewSequenceEventsInsideGroup
   properties {clipping=ClipToBounds, visibility=Visible}
   slot {class=SizeBoxSlot}
  }
  VerticalBox_0 {
   widgetClassPath=&VerticalBox
   parent=NewSequenceEventsInsideGroup
   widget=VerticalBox_0
   slot {class=OverlaySlot, properties {horizontalAlignment=HAlign_Fill, verticalAlignment=VAlign_Fill, padding=<ltrb 10 35 10 25>}}
  }
  Border_5 {
   widgetClassPath=&Border
   parent=VerticalBox_0
   widget=Border_5
   properties {
    background {
     bIsDynamicallyLoaded=false
     drawAs=Image
     imageType=NoImage
     mirroring=NoMirror
     resourceName=None
     resourceObject=null
     tiling=NoTile
     imageSize=<32 32>
     margin=<ltrb 0 0 0 0>
     outlineSettings {
      bUseBrushTransparency=false
      roundingType=HalfHeightRadius
      width=0
      color {colorUseRule=UseColor_Specified, specifiedColor=<rgba 0 0 0 0>}
      cornerRadii=<0 0 0 0>
     }
     tintColor {colorUseRule=UseColor_Specified, specifiedColor=<rgba 0.0729 0.0729 0.0729 1>}
     uVRegion {bIsValid=false, max=<0 0>, min=<0 0>}
    }
    padding=<ltrb 10 2 4 2>
   }
   slot {class=VerticalBoxSlot, properties {padding=<ltrb -10 4 -10 5>}}
  }
  HorizontalBox_8 {
   widgetClassPath=&HorizontalBox
   parent=Border_5
   widget=HorizontalBox_8
   slot {class=BorderSlot, properties {padding=<ltrb 10 2 4 2>}}
  }
  DANamePreviewTitle {
   widgetClassPath=&TextBlock
   parent=HorizontalBox_8
   widget=DANamePreviewTitle
   properties {
    text="Data Asset Name: "
    font {
     bForceMonospaced=false
     bMaterialIsStencil=false
     fontMaterial=null
     fontObject=&HeadingNow
     letterSpacing=0
     monospacedWidth=1
     size=9.75
     skewAmount=0
     typefaceFontName=75 Medium
     outlineSettings {
      bApplyOutlineToDropShadows=false
      bMiteredCorners=false
      bSeparateFillAlpha=false
      outlineMaterial=null
      outlineSize=0
      outlineColor=<rgba 0 0 0 1>
     }
    }
   }
   slot {class=HorizontalBoxSlot}
  }
  DANamePreviewText {
   widgetClassPath=&TextBlock
   bIsVariable=true
   parent=HorizontalBox_8
   widget=DANamePreviewText
   properties {
    colorAndOpacity {colorUseRule=UseColor_Specified, specifiedColor=<rgba 0.4062 0.4062 0.4062 1>}
    font {
     bForceMonospaced=false
     bMaterialIsStencil=false
     fontMaterial=null
     fontObject=&HeadingNow
     letterSpacing=0
     monospacedWidth=1
     size=9.75
     skewAmount=0
     typefaceFontName=74 Regular
     outlineSettings {
      bApplyOutlineToDropShadows=false
      bMiteredCorners=false
      bSeparateFillAlpha=false
      outlineMaterial=null
      outlineSize=0
      outlineColor=<rgba 0 0 0 1>
     }
    }
   }
   slot {class=HorizontalBoxSlot}
  }
  DeleteButtonsGroup {widgetClassPath=&VerticalBox, bIsVariable=true, parent=VerticalBox_0, widget=DeleteButtonsGroup, slot {class=VerticalBoxSlot}}
  Border_4 {
   widgetClassPath=&Border
   parent=DeleteButtonsGroup
   widget=Border_4
   properties {
    horizontalAlignment=HAlign_Center
    background {
     bIsDynamicallyLoaded=false
     drawAs=Image
     imageType=NoImage
     mirroring=NoMirror
     resourceName=None
     resourceObject=null
     tiling=NoTile
     imageSize=<32 32>
     margin=<ltrb 0 0 0 0>
     outlineSettings {
      bUseBrushTransparency=false
      roundingType=HalfHeightRadius
      width=0
      color {colorUseRule=UseColor_Specified, specifiedColor=<rgba 0 0 0 0>}
      cornerRadii=<0 0 0 0>
     }
     tintColor {colorUseRule=UseColor_Specified, specifiedColor=<rgba 0.0729 0.0729 0.0729 1>}
     uVRegion {bIsValid=false, max=<0 0>, min=<0 0>}
    }
    padding=<ltrb 10 2 4 2>
   }
   slot {class=VerticalBoxSlot, properties {padding=<ltrb -10 0 -10 0>}}
  }
  HorizontalBox_13 {
   widgetClassPath=&HorizontalBox
   parent=Border_4
   widget=HorizontalBox_13
   slot {class=BorderSlot, properties {horizontalAlignment=HAlign_Center, padding=<ltrb 10 2 4 2>}}
  }
  NewSequenceEventAddButton {
   widgetClassPath=&EditorUtilityButton
   parent=HorizontalBox_13
   widget=NewSequenceEventAddButton
   properties {widgetStyle {
     clickedSlateSound {resourceObject=null}
     disabled {
      bIsDynamicallyLoaded=false
      drawAs=RoundedBox
      imageType=NoImage
      mirroring=NoMirror
      resourceName=None
      resourceObject=null
      tiling=NoTile
      imageSize=<20 20>
      margin=<ltrb 0 0 0 0>
      outlineSettings {
       bUseBrushTransparency=false
       roundingType=FixedRadius
       width=0
       color {colorUseRule=UseColor_ColorTable, specifiedColor=<rgba 0 0 0 0>}
       cornerRadii=<4 4 4 4>
      }
      tintColor {colorUseRule=UseColor_Specified, specifiedColor=<rgba 0.0395 0.0395 0.0395 1>}
      uVRegion {bIsValid=false, max=<0 0>, min=<0 0>}
     }
     disabledForeground {colorUseRule=UseColor_ColorTable, specifiedColor=<rgba 0 0 0 0>}
     hovered {
      bIsDynamicallyLoaded=false
      drawAs=RoundedBox
      imageType=NoImage
      mirroring=NoMirror
      resourceName=None
      resourceObject=null
      tiling=NoTile
      imageSize=<32 32>
      margin=<ltrb 0 0 0 0>
      outlineSettings {
       bUseBrushTransparency=false
       roundingType=FixedRadius
       width=0
       color {colorUseRule=UseColor_ColorTable, specifiedColor=<rgba 0 0 0 0>}
       cornerRadii=<4 4 4 4>
      }
      tintColor {colorUseRule=UseColor_Specified, specifiedColor=<rgba 1 0.189 0.0729 1>}
      uVRegion {bIsValid=false, max=<0 0>, min=<0 0>}
     }
     hoveredForeground {colorUseRule=UseColor_ColorTable, specifiedColor=<rgba 0 0 0 0>}
     hoveredSlateSound {resourceObject=null}
     normal {
      bIsDynamicallyLoaded=false
      drawAs=RoundedBox
      imageType=NoImage
      mirroring=NoMirror
      resourceName=None
      resourceObject=null
      tiling=NoTile
      imageSize=<20 20>
      margin=<ltrb 0 0 0 0>
      outlineSettings {
       bUseBrushTransparency=false
       roundingType=FixedRadius
       width=0
       color {colorUseRule=UseColor_ColorTable, specifiedColor=<rgba 0 0 0 0>}
       cornerRadii=<4 4 4 4>
      }
      tintColor {colorUseRule=UseColor_Specified, specifiedColor=<rgba 0.8125 0.1329 0.0356 1>}
      uVRegion {bIsValid=false, max=<0 0>, min=<0 0>}
     }
     normalForeground {colorUseRule=UseColor_ColorTable, specifiedColor=<rgba 0 0 0 0>}
     normalPadding=<ltrb 12 1.5 12 1.5>
     pressed {
      bIsDynamicallyLoaded=false
      drawAs=RoundedBox
      imageType=NoImage
      mirroring=NoMirror
      resourceName=None
      resourceObject=null
      tiling=NoTile
      imageSize=<20 20>
      margin=<ltrb 0 0 0 0>
      outlineSettings {
       bUseBrushTransparency=false
       roundingType=FixedRadius
       width=0
       color {colorUseRule=UseColor_ColorTable, specifiedColor=<rgba 0 0 0 0>}
       cornerRadii=<4 4 4 4>
      }
      tintColor {colorUseRule=UseColor_Specified, specifiedColor=<rgba 0.2396 0 0 1>}
      uVRegion {bIsValid=false, max=<0 0>, min=<0 0>}
     }
     pressedForeground {colorUseRule=UseColor_ColorTable, specifiedColor=<rgba 0 0 0 0>}
     pressedPadding=<ltrb 12 2.5 12 0.5>
     pressedSlateSound {resourceObject=null}
    }}
   slot {class=HorizontalBoxSlot, properties {verticalAlignment=VAlign_Center, padding=<ltrb 10 5 0 5>}}
  }
  HorizontalBox_14 {
   widgetClassPath=&HorizontalBox
   parent=NewSequenceEventAddButton
   widget=HorizontalBox_14
   slot {class=ButtonSlot, properties {horizontalAlignment=HAlign_Left}}
  }
  Image_14 {
   widgetClassPath=&Image
   parent=HorizontalBox_14
   widget=Image_14
   properties {brush {
     bIsDynamicallyLoaded=false
     drawAs=Image
     imageType=FullColor
     mirroring=NoMirror
     resourceName=None
     resourceObject=/PilgrimQuickplay/StageManager/NeverCook/EditorUtilities/StageManagerEditor/Textures/T_SparksEditorUI_TrashCan_64
     tiling=NoTile
     imageSize=<20 20>
     margin=<ltrb 0 0 0 0>
     outlineSettings {
      bUseBrushTransparency=false
      roundingType=HalfHeightRadius
      width=0
      color {colorUseRule=UseColor_Specified, specifiedColor=<rgba 0 0 0 0>}
      cornerRadii=<0 0 0 0>
     }
     tintColor {colorUseRule=UseColor_Specified, specifiedColor=<rgba 1 1 1 1>}
     uVRegion {bIsValid=false, max=<0 0>, min=<0 0>}
    }}
   slot {class=HorizontalBoxSlot, properties {horizontalAlignment=HAlign_Center, verticalAlignment=VAlign_Center, padding=<ltrb 0 0 3 0>}}
  }
  TextBlock_9 {
   widgetClassPath=&TextBlock
   parent=HorizontalBox_14
   widget=TextBlock_9
   properties {
    text=Remove Entry Only
    font {
     bForceMonospaced=false
     bMaterialIsStencil=false
     fontMaterial=null
     fontObject=&HeadingNow
     letterSpacing=0
     monospacedWidth=1
     size=7.5
     skewAmount=0
     typefaceFontName=75 Medium
     outlineSettings {
      bApplyOutlineToDropShadows=false
      bMiteredCorners=false
      bSeparateFillAlpha=false
      outlineMaterial=null
      outlineSize=0
      outlineColor=<rgba 0 0 0 1>
     }
    }
   }
   slot {class=HorizontalBoxSlot, properties {verticalAlignment=VAlign_Center}}
  }
  CreateChangeButton {
   widgetClassPath=&EditorUtilityButton
   parent=HorizontalBox_13
   widget=CreateChangeButton
   properties {widgetStyle {
     clickedSlateSound {resourceObject=null}
     disabled {
      bIsDynamicallyLoaded=false
      drawAs=RoundedBox
      imageType=NoImage
      mirroring=NoMirror
      resourceName=None
      resourceObject=null
      tiling=NoTile
      imageSize=<20 20>
      margin=<ltrb 0 0 0 0>
      outlineSettings {
       bUseBrushTransparency=false
       roundingType=FixedRadius
       width=0
       color {colorUseRule=UseColor_ColorTable, specifiedColor=<rgba 0 0 0 0>}
       cornerRadii=<4 4 4 4>
      }
      tintColor {colorUseRule=UseColor_Specified, specifiedColor=<rgba 0.0395 0.0395 0.0395 1>}
      uVRegion {bIsValid=false, max=<0 0>, min=<0 0>}
     }
     disabledForeground {colorUseRule=UseColor_ColorTable, specifiedColor=<rgba 0 0 0 0>}
     hovered {
      bIsDynamicallyLoaded=false
      drawAs=RoundedBox
      imageType=NoImage
      mirroring=NoMirror
      resourceName=None
      resourceObject=null
      tiling=NoTile
      imageSize=<32 32>
      margin=<ltrb 0 0 0 0>
      outlineSettings {
       bUseBrushTransparency=false
       roundingType=FixedRadius
       width=0
       color {colorUseRule=UseColor_ColorTable, specifiedColor=<rgba 0 0 0 0>}
       cornerRadii=<4 4 4 4>
      }
      tintColor {colorUseRule=UseColor_Specified, specifiedColor=<rgba 1 0.1252 0.253 1>}
      uVRegion {bIsValid=false, max=<0 0>, min=<0 0>}
     }
     hoveredForeground {colorUseRule=UseColor_ColorTable, specifiedColor=<rgba 0 0 0 0>}
     hoveredSlateSound {resourceObject=null}
     normal {
      bIsDynamicallyLoaded=false
      drawAs=RoundedBox
      imageType=NoImage
      mirroring=NoMirror
      resourceName=None
      resourceObject=null
      tiling=NoTile
      imageSize=<20 20>
      margin=<ltrb 0 0 0 0>
      outlineSettings {
       bUseBrushTransparency=false
       roundingType=FixedRadius
       width=0
       color {colorUseRule=UseColor_ColorTable, specifiedColor=<rgba 0 0 0 0>}
       cornerRadii=<4 4 4 4>
      }
      tintColor {colorUseRule=UseColor_Specified, specifiedColor=<rgba 0.775 0 0.0304 1>}
      uVRegion {bIsValid=false, max=<0 0>, min=<0 0>}
     }
     normalForeground {colorUseRule=UseColor_ColorTable, specifiedColor=<rgba 0 0 0 0>}
     normalPadding=<ltrb 12 1.5 12 1.5>
     pressed {
      bIsDynamicallyLoaded=false
      drawAs=RoundedBox
      imageType=NoImage
      mirroring=NoMirror
      resourceName=None
      resourceObject=null
      tiling=NoTile
      imageSize=<20 20>
      margin=<ltrb 0 0 0 0>
      outlineSettings {
       bUseBrushTransparency=false
       roundingType=FixedRadius
       width=0
       color {colorUseRule=UseColor_ColorTable, specifiedColor=<rgba 0 0 0 0>}
       cornerRadii=<4 4 4 4>
      }
      tintColor {colorUseRule=UseColor_Specified, specifiedColor=<rgba 0.2396 0 0 1>}
      uVRegion {bIsValid=false, max=<0 0>, min=<0 0>}
     }
     pressedForeground {colorUseRule=UseColor_ColorTable, specifiedColor=<rgba 0 0 0 0>}
     pressedPadding=<ltrb 12 2.5 12 0.5>
     pressedSlateSound {resourceObject=null}
    }}
   slot {class=HorizontalBoxSlot, properties {verticalAlignment=VAlign_Center, padding=<ltrb 10 0 0 0>}}
  }
  HorizontalBox_9 {
   widgetClassPath=&HorizontalBox
   parent=CreateChangeButton
   widget=HorizontalBox_9
   slot {class=ButtonSlot, properties {horizontalAlignment=HAlign_Left}}
  }
  Image_9 {
   widgetClassPath=&Image
   parent=HorizontalBox_9
   widget=Image_9
   properties {brush {
     bIsDynamicallyLoaded=false
     drawAs=Image
     imageType=FullColor
     mirroring=NoMirror
     resourceName=None
     resourceObject=&T_SparksEditorUI_TrashCan_64
     tiling=NoTile
     imageSize=<20 20>
     margin=<ltrb 0 0 0 0>
     outlineSettings {
      bUseBrushTransparency=false
      roundingType=HalfHeightRadius
      width=0
      color {colorUseRule=UseColor_Specified, specifiedColor=<rgba 0 0 0 0>}
      cornerRadii=<0 0 0 0>
     }
     tintColor {colorUseRule=UseColor_Specified, specifiedColor=<rgba 1 1 1 1>}
     uVRegion {bIsValid=false, max=<0 0>, min=<0 0>}
    }}
   slot {class=HorizontalBoxSlot, properties {horizontalAlignment=HAlign_Center, verticalAlignment=VAlign_Center, padding=<ltrb 0 0 3 0>}}
  }
  TextBlock_5 {
   widgetClassPath=&TextBlock
   parent=HorizontalBox_9
   widget=TextBlock_5
   properties {
    text=Delete Entry + Asset
    font {
     bForceMonospaced=false
     bMaterialIsStencil=false
     fontMaterial=null
     fontObject=&HeadingNow
     letterSpacing=0
     monospacedWidth=1
     size=7.5
     skewAmount=0
     typefaceFontName=75 Medium
     outlineSettings {
      bApplyOutlineToDropShadows=false
      bMiteredCorners=false
      bSeparateFillAlpha=false
      outlineMaterial=null
      outlineSize=0
      outlineColor=<rgba 0 0 0 1>
     }
    }
   }
   slot {class=HorizontalBoxSlot, properties {verticalAlignment=VAlign_Center}}
  }
  ConfirmationGroup {
   widgetClassPath=&VerticalBox
   bIsVariable=true
   parent=VerticalBox_0
   widget=ConfirmationGroup
   properties {visibility=Collapsed}
   slot {class=VerticalBoxSlot}
  }
  DANamePreviewTitle_1 {
   widgetClassPath=&TextBlock
   parent=ConfirmationGroup
   widget=DANamePreviewTitle_1
   properties {
    text=Are you sure? You can undo reverting the change.
    font {
     bForceMonospaced=false
     bMaterialIsStencil=false
     fontMaterial=null
     fontObject=&HeadingNow
     letterSpacing=0
     monospacedWidth=1
     size=9.75
     skewAmount=0
     typefaceFontName=75 Medium
     outlineSettings {
      bApplyOutlineToDropShadows=false
      bMiteredCorners=false
      bSeparateFillAlpha=false
      outlineMaterial=null
      outlineSize=0
      outlineColor=<rgba 0 0 0 1>
     }
    }
   }
   slot {class=VerticalBoxSlot, properties {horizontalAlignment=HAlign_Center, padding=<ltrb 0 5 0 5>}}
  }
  Border_1 {
   widgetClassPath=&Border
   parent=ConfirmationGroup
   widget=Border_1
   properties {
    horizontalAlignment=HAlign_Center
    background {
     bIsDynamicallyLoaded=false
     drawAs=Image
     imageType=NoImage
     mirroring=NoMirror
     resourceName=None
     resourceObject=null
     tiling=NoTile
     imageSize=<32 32>
     margin=<ltrb 0 0 0 0>
     outlineSettings {
      bUseBrushTransparency=false
      roundingType=HalfHeightRadius
      width=0
      color {colorUseRule=UseColor_Specified, specifiedColor=<rgba 0 0 0 0>}
      cornerRadii=<0 0 0 0>
     }
     tintColor {colorUseRule=UseColor_Specified, specifiedColor=<rgba 0.0729 0.0729 0.0729 1>}
     uVRegion {bIsValid=false, max=<0 0>, min=<0 0>}
    }
    padding=<ltrb 10 2 4 2>
   }
   slot {class=VerticalBoxSlot, properties {padding=<ltrb -10 0 -10 0>}}
  }
  HorizontalBox {
   widgetClassPath=&HorizontalBox
   parent=Border_1
   widget=HorizontalBox
   slot {class=BorderSlot, properties {horizontalAlignment=HAlign_Center, padding=<ltrb 10 2 4 2>}}
  }
  NewSequenceEventAddButton_1 {
   widgetClassPath=&EditorUtilityButton
   parent=HorizontalBox
   widget=NewSequenceEventAddButton_1
   properties {widgetStyle {
     clickedSlateSound {resourceObject=null}
     disabled {
      bIsDynamicallyLoaded=false
      drawAs=RoundedBox
      imageType=NoImage
      mirroring=NoMirror
      resourceName=None
      resourceObject=null
      tiling=NoTile
      imageSize=<20 20>
      margin=<ltrb 0 0 0 0>
      outlineSettings {
       bUseBrushTransparency=false
       roundingType=FixedRadius
       width=0
       color {colorUseRule=UseColor_ColorTable, specifiedColor=<rgba 0 0 0 0>}
       cornerRadii=<4 4 4 4>
      }
      tintColor {colorUseRule=UseColor_Specified, specifiedColor=<rgba 0.0395 0.0395 0.0395 1>}
      uVRegion {bIsValid=false, max=<0 0>, min=<0 0>}
     }
     disabledForeground {colorUseRule=UseColor_ColorTable, specifiedColor=<rgba 0 0 0 0>}
     hovered {
      bIsDynamicallyLoaded=false
      drawAs=RoundedBox
      imageType=NoImage
      mirroring=NoMirror
      resourceName=None
      resourceObject=null
      tiling=NoTile
      imageSize=<32 32>
      margin=<ltrb 0 0 0 0>
      outlineSettings {
       bUseBrushTransparency=false
       roundingType=FixedRadius
       width=0
       color {colorUseRule=UseColor_ColorTable, specifiedColor=<rgba 0 0 0 0>}
       cornerRadii=<4 4 4 4>
      }
      tintColor {colorUseRule=UseColor_Specified, specifiedColor=<rgba 1 0.1252 0.253 1>}
      uVRegion {bIsValid=false, max=<0 0>, min=<0 0>}
     }
     hoveredForeground {colorUseRule=UseColor_ColorTable, specifiedColor=<rgba 0 0 0 0>}
     hoveredSlateSound {resourceObject=null}
     normal {
      bIsDynamicallyLoaded=false
      drawAs=RoundedBox
      imageType=NoImage
      mirroring=NoMirror
      resourceName=None
      resourceObject=null
      tiling=NoTile
      imageSize=<20 20>
      margin=<ltrb 0 0 0 0>
      outlineSettings {
       bUseBrushTransparency=false
       roundingType=FixedRadius
       width=0
       color {colorUseRule=UseColor_ColorTable, specifiedColor=<rgba 0 0 0 0>}
       cornerRadii=<4 4 4 4>
      }
      tintColor {colorUseRule=UseColor_Specified, specifiedColor=<rgba 0.775 0 0.0304 1>}
      uVRegion {bIsValid=false, max=<0 0>, min=<0 0>}
     }
     normalForeground {colorUseRule=UseColor_ColorTable, specifiedColor=<rgba 0 0 0 0>}
     normalPadding=<ltrb 12 1.5 12 1.5>
     pressed {
      bIsDynamicallyLoaded=false
      drawAs=RoundedBox
      imageType=NoImage
      mirroring=NoMirror
      resourceName=None
      resourceObject=null
      tiling=NoTile
      imageSize=<20 20>
      margin=<ltrb 0 0 0 0>
      outlineSettings {
       bUseBrushTransparency=false
       roundingType=FixedRadius
       width=0
       color {colorUseRule=UseColor_ColorTable, specifiedColor=<rgba 0 0 0 0>}
       cornerRadii=<4 4 4 4>
      }
      tintColor {colorUseRule=UseColor_Specified, specifiedColor=<rgba 0.2396 0 0 1>}
      uVRegion {bIsValid=false, max=<0 0>, min=<0 0>}
     }
     pressedForeground {colorUseRule=UseColor_ColorTable, specifiedColor=<rgba 0 0 0 0>}
     pressedPadding=<ltrb 12 2.5 12 0.5>
     pressedSlateSound {resourceObject=null}
    }}
   slot {class=HorizontalBoxSlot, properties {verticalAlignment=VAlign_Center, padding=<ltrb 10 5 0 5>}}
  }
  HorizontalBox_1 {
   widgetClassPath=&HorizontalBox
   parent=NewSequenceEventAddButton_1
   widget=HorizontalBox_1
   slot {class=ButtonSlot, properties {horizontalAlignment=HAlign_Left}}
  }
  Image {
   widgetClassPath=&Image
   parent=HorizontalBox_1
   widget=Image
   properties {brush {
     bIsDynamicallyLoaded=false
     drawAs=Image
     imageType=FullColor
     mirroring=NoMirror
     resourceName=None
     resourceObject=&T_SparksEditorUI_TrashCan_64
     tiling=NoTile
     imageSize=<20 20>
     margin=<ltrb 0 0 0 0>
     outlineSettings {
      bUseBrushTransparency=false
      roundingType=HalfHeightRadius
      width=0
      color {colorUseRule=UseColor_Specified, specifiedColor=<rgba 0 0 0 0>}
      cornerRadii=<0 0 0 0>
     }
     tintColor {colorUseRule=UseColor_Specified, specifiedColor=<rgba 1 1 1 1>}
     uVRegion {bIsValid=false, max=<0 0>, min=<0 0>}
    }}
   slot {class=HorizontalBoxSlot, properties {horizontalAlignment=HAlign_Center, verticalAlignment=VAlign_Center, padding=<ltrb 0 0 3 0>}}
  }
  TextBlock {
   widgetClassPath=&TextBlock
   parent=HorizontalBox_1
   widget=TextBlock
   properties {
    text=Yes
    font {
     bForceMonospaced=false
     bMaterialIsStencil=false
     fontMaterial=null
     fontObject=&HeadingNow
     letterSpacing=0
     monospacedWidth=1
     size=7.5
     skewAmount=0
     typefaceFontName=75 Medium
     outlineSettings {
      bApplyOutlineToDropShadows=false
      bMiteredCorners=false
      bSeparateFillAlpha=false
      outlineMaterial=null
      outlineSize=0
      outlineColor=<rgba 0 0 0 1>
     }
    }
   }
   slot {class=HorizontalBoxSlot, properties {verticalAlignment=VAlign_Center}}
  }
  CreateChangeButton_1 {
   widgetClassPath=&EditorUtilityButton
   parent=HorizontalBox
   widget=CreateChangeButton_1
   properties {widgetStyle {
     clickedSlateSound {resourceObject=null}
     disabled {
      bIsDynamicallyLoaded=false
      drawAs=RoundedBox
      imageType=NoImage
      mirroring=NoMirror
      resourceName=None
      resourceObject=null
      tiling=NoTile
      imageSize=<0 0>
      margin=<ltrb 0 0 0 0>
      outlineSettings {
       bUseBrushTransparency=false
       roundingType=FixedRadius
       width=0
       color {colorUseRule=UseColor_ColorTable, specifiedColor=<rgba 0 0 0 0>}
       cornerRadii=<4 4 4 4>
      }
      tintColor {colorUseRule=UseColor_Specified, specifiedColor=<rgba 0.0395 0.0395 0.0395 1>}
      uVRegion {bIsValid=false, max=<0 0>, min=<0 0>}
     }
     disabledForeground {colorUseRule=UseColor_ColorTable, specifiedColor=<rgba 0 0 0 0>}
     hovered {
      bIsDynamicallyLoaded=false
      drawAs=RoundedBox
      imageType=NoImage
      mirroring=NoMirror
      resourceName=None
      resourceObject=null
      tiling=NoTile
      imageSize=<32 32>
      margin=<ltrb 0 0 0 0>
      outlineSettings {
       bUseBrushTransparency=false
       roundingType=FixedRadius
       width=0
       color {colorUseRule=UseColor_ColorTable, specifiedColor=<rgba 0 0 0 0>}
       cornerRadii=<4 4 4 4>
      }
      tintColor {colorUseRule=UseColor_Specified, specifiedColor=<rgba 0.4271 0.4271 0.4271 1>}
      uVRegion {bIsValid=false, max=<0 0>, min=<0 0>}
     }
     hoveredForeground {colorUseRule=UseColor_ColorTable, specifiedColor=<rgba 0 0 0 0>}
     hoveredSlateSound {resourceObject=null}
     normal {
      bIsDynamicallyLoaded=false
      drawAs=RoundedBox
      imageType=NoImage
      mirroring=NoMirror
      resourceName=None
      resourceObject=null
      tiling=NoTile
      imageSize=<32 32>
      margin=<ltrb 0 0 0 0>
      outlineSettings {
       bUseBrushTransparency=false
       roundingType=FixedRadius
       width=0
       color {colorUseRule=UseColor_ColorTable, specifiedColor=<rgba 0 0 0 0>}
       cornerRadii=<4 4 4 4>
      }
      tintColor {colorUseRule=UseColor_Specified, specifiedColor=<rgba 0.2396 0.2396 0.2396 1>}
      uVRegion {bIsValid=false, max=<0 0>, min=<0 0>}
     }
     normalForeground {colorUseRule=UseColor_ColorTable, specifiedColor=<rgba 0 0 0 0>}
     normalPadding=<ltrb 12 1.5 12 1.5>
     pressed {
      bIsDynamicallyLoaded=false
      drawAs=RoundedBox
      imageType=NoImage
      mirroring=NoMirror
      resourceName=None
      resourceObject=null
      tiling=NoTile
      imageSize=<32 32>
      margin=<ltrb 0 0 0 0>
      outlineSettings {
       bUseBrushTransparency=false
       roundingType=FixedRadius
       width=0
       color {colorUseRule=UseColor_ColorTable, specifiedColor=<rgba 0 0 0 0>}
       cornerRadii=<4 4 4 4>
      }
      tintColor {colorUseRule=UseColor_Specified, specifiedColor=<rgba 0.0104 0.0104 0.0104 1>}
      uVRegion {bIsValid=false, max=<0 0>, min=<0 0>}
     }
     pressedForeground {colorUseRule=UseColor_ColorTable, specifiedColor=<rgba 0 0 0 0>}
     pressedPadding=<ltrb 12 2.5 12 0.5>
     pressedSlateSound {resourceObject=null}
    }}
   slot {class=HorizontalBoxSlot, properties {verticalAlignment=VAlign_Center, padding=<ltrb 10 0 0 0>, size {sizeRule=Fill, value=1}}}
  }
  HorizontalBox_2 {
   widgetClassPath=&HorizontalBox
   parent=CreateChangeButton_1
   widget=HorizontalBox_2
   slot {class=ButtonSlot, properties {horizontalAlignment=HAlign_Left}}
  }
  Image_1 {
   widgetClassPath=&Image
   parent=HorizontalBox_2
   widget=Image_1
   properties {
    brush {
     bIsDynamicallyLoaded=false
     drawAs=Image
     imageType=FullColor
     mirroring=NoMirror
     resourceName=None
     resourceObject=&T_SparksEditorUI_TrashCan_64
     tiling=NoTile
     imageSize=<0 20>
     margin=<ltrb 0 0 0 0>
     outlineSettings {
      bUseBrushTransparency=false
      roundingType=HalfHeightRadius
      width=0
      color {colorUseRule=UseColor_Specified, specifiedColor=<rgba 0 0 0 0>}
      cornerRadii=<0 0 0 0>
     }
     tintColor {colorUseRule=UseColor_Specified, specifiedColor=<rgba 1 1 1 1>}
     uVRegion {bIsValid=false, max=<0 0>, min=<0 0>}
    }
    colorAndOpacity=<rgba 1 1 1 0>
   }
   slot {class=HorizontalBoxSlot, properties {horizontalAlignment=HAlign_Center, verticalAlignment=VAlign_Center, padding=<ltrb 0 0 3 0>}}
  }
  TextBlock_1 {
   widgetClassPath=&TextBlock
   parent=HorizontalBox_2
   widget=TextBlock_1
   properties {
    text=No
    font {
     bForceMonospaced=false
     bMaterialIsStencil=false
     fontMaterial=null
     fontObject=&HeadingNow
     letterSpacing=0
     monospacedWidth=1
     size=7.5
     skewAmount=0
     typefaceFontName=75 Medium
     outlineSettings {
      bApplyOutlineToDropShadows=false
      bMiteredCorners=false
      bSeparateFillAlpha=false
      outlineMaterial=null
      outlineSize=0
      outlineColor=<rgba 0 0 0 1>
     }
    }
   }
   slot {class=HorizontalBoxSlot, properties {verticalAlignment=VAlign_Center}}
  }
  bg {
   widgetClassPath=&Image
   parent=NewSequenceEventsInsideGroup
   widget=bg
   properties {visibility=Collapsed}
   slot {class=OverlaySlot, properties {horizontalAlignment=HAlign_Fill, verticalAlignment=VAlign_Fill}}
  }
  Nosequenceeventstext {
   widgetClassPath=&TextBlock
   bIsVariable=true
   parent=Overlay
   widget=Nosequenceeventstext
   properties {
    justification=Center
    text=All avaliable events are in use.
    visibility=Collapsed
    font {
     bForceMonospaced=false
     bMaterialIsStencil=false
     fontMaterial=null
     fontObject=&Roboto
     letterSpacing=0
     monospacedWidth=1
     size=15
     skewAmount=0
     typefaceFontName=Bold
     outlineSettings {
      bApplyOutlineToDropShadows=false
      bMiteredCorners=false
      bSeparateFillAlpha=false
      outlineMaterial=null
      outlineSize=0
      outlineColor=<rgba 0 0 0 1>
     }
    }
   }
   slot {class=OverlaySlot, properties {horizontalAlignment=HAlign_Center, verticalAlignment=VAlign_Center}}
  }
  CloseNewSequenceEventWindowButton {
   widgetClassPath=&EditorUtilityButton
   parent=Overlay
   widget=CloseNewSequenceEventWindowButton
   properties {
    clipping=ClipToBounds
    widgetStyle {
     clickedSlateSound {resourceObject=null}
     disabled {
      bIsDynamicallyLoaded=false
      drawAs=RoundedBox
      imageType=NoImage
      mirroring=NoMirror
      resourceName=None
      resourceObject=null
      tiling=NoTile
      imageSize=<20 20>
      margin=<ltrb 0 0 0 0>
      outlineSettings {
       bUseBrushTransparency=false
       roundingType=FixedRadius
       width=0
       color {colorUseRule=UseColor_ColorTable, specifiedColor=<rgba 0 0 0 0>}
       cornerRadii=<4 4 4 4>
      }
      tintColor {colorUseRule=UseColor_Specified, specifiedColor=<rgba 0.0395 0.0395 0.0395 1>}
      uVRegion {bIsValid=false, max=<0 0>, min=<0 0>}
     }
     disabledForeground {colorUseRule=UseColor_ColorTable, specifiedColor=<rgba 0 0 0 0>}
     hovered {
      bIsDynamicallyLoaded=false
      drawAs=RoundedBox
      imageType=NoImage
      mirroring=NoMirror
      resourceName=None
      resourceObject=null
      tiling=NoTile
      imageSize=<32 32>
      margin=<ltrb 0 0 0 0>
      outlineSettings {
       bUseBrushTransparency=false
       roundingType=FixedRadius
       width=0
       color {colorUseRule=UseColor_ColorTable, specifiedColor=<rgba 0 0 0 0>}
       cornerRadii=<4 4 4 4>
      }
      tintColor {colorUseRule=UseColor_Specified, specifiedColor=<rgba 1 0.1252 0.253 1>}
      uVRegion {bIsValid=false, max=<0 0>, min=<0 0>}
     }
     hoveredForeground {colorUseRule=UseColor_ColorTable, specifiedColor=<rgba 0 0 0 0>}
     hoveredSlateSound {resourceObject=null}
     normal {
      bIsDynamicallyLoaded=false
      drawAs=RoundedBox
      imageType=NoImage
      mirroring=NoMirror
      resourceName=None
      resourceObject=null
      tiling=NoTile
      imageSize=<20 20>
      margin=<ltrb 0 0 0 0>
      outlineSettings {
       bUseBrushTransparency=false
       roundingType=FixedRadius
       width=0
       color {colorUseRule=UseColor_ColorTable, specifiedColor=<rgba 0 0 0 0>}
       cornerRadii=<4 4 4 4>
      }
      tintColor {colorUseRule=UseColor_Specified, specifiedColor=<rgba 0.775 0 0.0304 1>}
      uVRegion {bIsValid=false, max=<0 0>, min=<0 0>}
     }
     normalForeground {colorUseRule=UseColor_ColorTable, specifiedColor=<rgba 0 0 0 0>}
     normalPadding=<ltrb 12 1.5 12 1.5>
     pressed {
      bIsDynamicallyLoaded=false
      drawAs=RoundedBox
      imageType=NoImage
      mirroring=NoMirror
      resourceName=None
      resourceObject=null
      tiling=NoTile
      imageSize=<20 20>
      margin=<ltrb 0 0 0 0>
      outlineSettings {
       bUseBrushTransparency=false
       roundingType=FixedRadius
       width=0
       color {colorUseRule=UseColor_ColorTable, specifiedColor=<rgba 0 0 0 0>}
       cornerRadii=<4 4 4 4>
      }
      tintColor {colorUseRule=UseColor_Specified, specifiedColor=<rgba 0.2396 0 0 1>}
      uVRegion {bIsValid=false, max=<0 0>, min=<0 0>}
     }
     pressedForeground {colorUseRule=UseColor_ColorTable, specifiedColor=<rgba 0 0 0 0>}
     pressedPadding=<ltrb 12 2.5 12 0.5>
     pressedSlateSound {resourceObject=null}
    }
   }
   slot {class=OverlaySlot, properties {horizontalAlignment=HAlign_Right, padding=<ltrb 10 3.5 10 0>}}
  }
  TextBlock_7 {
   widgetClassPath=&TextBlock
   parent=CloseNewSequenceEventWindowButton
   widget=TextBlock_7
   properties {
    text=X
    font {
     bForceMonospaced=false
     bMaterialIsStencil=false
     fontMaterial=null
     fontObject=&HeadingNow
     letterSpacing=0
     monospacedWidth=1
     size=7.5
     skewAmount=0
     typefaceFontName=86 Bold
     outlineSettings {
      bApplyOutlineToDropShadows=false
      bMiteredCorners=false
      bSeparateFillAlpha=false
      outlineMaterial=null
      outlineSize=0
      outlineColor=<rgba 0 0 0 1>
     }
    }
   }
   slot {class=ButtonSlot, properties {horizontalAlignment=HAlign_Fill, padding=<ltrb 0 0 0 0>}}
  }
 ]
}
