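# Editor-only popup submenu for the Sparks Stage Manager timeline, offering Change, Edit and Remove for a selected sequence or event.
# Change opens the editor's add/change window, then OnAssetChanged adds or swaps the returned asset into the pool or step.
# Edit opens the FX event data window, or the default asset editor for other types.
# Remove uses a confirmation popup, removes the pool, step or event entry, optionally deleting the asset.
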EditorUtilityWidgetBlueprint /PilgrimQuickplay/StageManager/NeverCook/EditorUtilities/StageManagerEditor/EWBP_SparksTimelineSubmenu
graphs [
 ChangeButtonsVisibility {nodes [
   n1 = ChangeButtonsVisibility!() -> (then, ChangeButton:ESlateVisibility Enum=Visible, EditButton:ESlateVisibility Enum=Visible, RemoveButton:ESlateVisibility Enum=Visible) {is=title K2Node_FunctionEntry}
   > SetVisibility(self=Get ChangeButton() -> (ChangeButton:Editor Utility Button Object Reference), InVisibility=@n1)
   > SetVisibility(self=Get EditButton() -> (EditButton:Editor Utility Button Object Reference), InVisibility=@n1.EditButton)
   > SetVisibility(self=Get RemoveButton() -> (RemoveButton:Editor Utility Button Object Reference), InVisibility=@n1.RemoveButton)
  ]}
 EventGraph {nodes [
   "On Pressed (ChangeButton)"() {is=title K2Node_ComponentBoundEvent}
   > Set IsLooping(IsLooping:bool=false) -> (Output_Get:bool)
   > n1 = Sequence!(execute) -> (then_0, then_1, then_2:Exec, then_3:Exec)
   > n2 = Get!(execute:Exec) -> (then:Exec, else:Exec, SequenceReference:Mojo Fx Manager Sequence Data Asset Object Reference)
   n3 = Get SparksTimelineEditor() -> (SparksTimelineEditor:Sparks Stage Manager Timeline Editor Object Reference)
   > n4 = Set MojoStage_SequenceData!(
    execute
    MojoStage_SequenceData:Mojo Fx Manager Sequence Data Asset Object Reference=@n2
    self:Sparks Stage Manager Timeline Editor Object Reference=@n3
   ) -> (then, Output_Get:Mojo Fx Manager Sequence Data Asset Object Reference)
   > Set CurrentEventSelected!(execute, CurrentEventSelected:String=GetDisplayName(Object=@n4), self:Sparks Stage Manager Timeline Editor Object Reference=@n3) -> (then, Output_Get:String)
   n5 = Get!(execute=@n1.then_1) -> (then, else, EventReference:Object Reference)
   > n6 = Cast To MojoFxManagerEvent_DataAsset!(execute, Object=@n5) -> (then, CastFailed, AsMojo Fx Manager Event Data Asset:Mojo Fx Manager Event Data Asset Object Reference) {is=title K2Node_DynamicCast}
   n7 = Get SparksTimelineEditor()
   > n8 = Set FxManager_Event StepData!(
    execute
    FxManager_Event StepData:Mojo Fx Manager Event Data Asset Object Reference=@n6
    self:Sparks Stage Manager Timeline Editor Object Reference=@n7
   ) -> (then, Output_Get:Mojo Fx Manager Event Data Asset Object Reference)
   > Set CurrentEventSelected(CurrentEventSelected=GetDisplayName(Object=@n8), self=@n7)
   > Set IsLooping(IsLooping=Get LoopingEvent!(self:Stage Show Event Object Reference=@n8) -> (LoopingEvent:bool))
   n9 = Get!(execute=@n1.then_2) -> (then, else, EventBarReference:EWBP Sparks Sequence Pool Timeline Event Bar Object Reference)
   > Set CurrentEventBarReference!(
    execute
    CurrentEventBarReference:EWBP Sparks Sequence Pool Timeline Event Bar Object Reference=@n9
    self:Sparks Stage Manager Timeline Editor Object Reference=Get SparksTimelineEditor()
   ) -> (then, Output_Get:EWBP Sparks Sequence Pool Timeline Event Bar Object Reference)
   n10 = Get SequencePoolData() -> (SequencePoolData:Stage Show Sequence Pool Object Reference)
   OpenAddChangeWindow!(
    execute=@n1.then_3
    self:Sparks Stage Manager Timeline Editor Object Reference=Get SparksTimelineEditor()
    WindowType:Name=Select(Option 0:Name=FxManager_Event StepData, Option 1:Name=MojoStage_SequenceData, Index=IsValid(Object=@n10)) -> (ReturnValue:Name)
    IsJustCreate:bool=false
    HideSelectSection:bool=false
    IsLoopingEvent:bool=Get IsLooping() -> (IsLooping:bool)
    KeepSelectChange:bool=true
   ) -> (then)
   > Bind Event to On Asset Changed(self:Sparks Stage Manager Timeline Editor Object Reference=Get SparksTimelineEditor(), Delegate=Create Event()) {is=title K2Node_AddDelegate}
   > SetVisibility(InVisibility=Collapsed)
   > Call Update Visibility Status!(execute:Exec, self:EWBP Sparks Timeline Submenu Object Reference, IsVisible:bool) -> (then:Exec) {is=title K2Node_CallDelegate}
   "On Pressed (EditButton)"() {is=title K2Node_ComponentBoundEvent}
   > Get!(execute) -> (then, else, EventReference:Object Reference)
   > n11 = Sequence!(execute) -> (then_0, then_1)
   > n12 = Cast To MojoFxManagerEvent_DataAsset(Object=Get EventReference() -> (EventReference:Object Reference))
   > n13 = SpawnAndRegisterTabAndGetID!(
    execute
    self:Editor Utility Subsystem Object Reference=Get EditorUtilitySubsystem()
    InBlueprint:Editor Utility Widget Blueprint Object Reference=/PilgrimQuickplay/StageManager/NeverCook/EditorUtilities/StageManagerEditor/FxManagerEventWindow/SparksFxManagerEventDataWindow
   ) -> (then, NewTabID:Name=None, ReturnValue:Editor Utility Widget Object Reference)
   > n14 = Cast To SparksFxManagerEventDataWindow!(execute, Object=@n13.ReturnValue) -> (then, CastFailed, AsSparks Fx Manager Event Data Window:Sparks Fx Manager Event Data Window Object Reference) {is=title K2Node_DynamicCast}
   > Init!(
    execute
    self:Sparks Fx Manager Event Data Window Object Reference=@n14
    EventDataAsset:Mojo Fx Manager Event Data Asset Object Reference=@n12
   ) -> (then)
   > n15 = Sequence()
   > n16 = Get!(execute) -> (then, else, EventBarReference:EWBP Sparks Sequence Pool Timeline Event Bar Object Reference)
   > SetEventBarReference!(
    execute
    self:Sparks Fx Manager Event Data Window Object Reference=@n14
    EventBarReference:"EWBP Sparks Sequence Pool Timeline Event Bar Object Reference (by ref)"=@n16
   ) -> (then)
   n17 = Get!(execute=@n15.then_1) -> (then, else, SparksTimelineEditor:Sparks Stage Manager Timeline Editor Object Reference)
   > SetTimelineEditorReference!(
    execute
    self:Sparks Fx Manager Event Data Window Object Reference=@n14
    SparksTimelineEditor:Sparks Stage Manager Timeline Editor Object Reference=@n17
   ) -> (then)
   OpenEditorForAssets!(
    execute=@n12.CastFailed
    self:Asset Editor Subsystem Object Reference=Get AssetEditorSubsystem()
    Assets:Array of Object References=Make Array("[0]":Object Reference=Get EventReference()) -> (Array:Array of Object References)
    OpenedMethod:EAssetTypeActivationOpenedMethod Enum=Edit
   ) -> (then, ReturnValue:bool)
   SetVisibility(execute=@n11.then_1, InVisibility=Collapsed)
   > Call Update Visibility Status()
   "On Pressed (RemoveButton)"() {is=title K2Node_ComponentBoundEvent}
   > UpdateToggles!(
    execute
    self:Sparks Stage Manager Timeline Editor Object Reference=Get SparksTimelineEditor()
    Toggle:EWBP Sparks Sequence Pool Timeline Event Bar Object Reference
    ForceCloseAll:bool=true
    DontCallEvent:bool=false
   ) -> (then)
   > n18 = Get!(execute) -> (then, else, SequencePoolData:Stage Show Sequence Pool Object Reference)
   > n19 = OpenWindow!(
    execute
    self:EWBP Sparks Delete Confirmation Object Reference=Get DeleteConfirmationPopup!(self:Sparks Stage Manager Timeline Editor Object Reference=Get SparksTimelineEditor()) -> (DeleteConfirmationPopup:EWBP Sparks Delete Confirmation Object Reference)
    Object:Object Reference="Get (a copy)"(
     Array:Array of Stage Show Sequence Object References=Get Sequences!(self:Stage Show Sequence Pool Object Reference=@n18) -> (Sequences:Array of Stage Show Sequence Object References)
     Dimension 1=Get SequenceIndex() -> (SequenceIndex:int=0)
    ) -> (Output:Stage Show Sequence Object Reference)
   ) -> (then)
   Bind Event to On Closed(
    execute=[@n23.then, @n19.then]
    self:EWBP Sparks Delete Confirmation Object Reference=Get DeleteConfirmationPopup(self=Get SparksTimelineEditor())
    Delegate=Create Event()
   ) {is=title K2Node_AddDelegate}
   n20 = Branch(execute=@n18.else, Condition=NOT Boolean(A=Get Is Header() -> (Is Header:bool)))
   > n21 = Get!(execute) -> (then, else, SequenceReference:Mojo Fx Manager Sequence Data Asset Object Reference)
   n22 = "Get (a ref)"(
    Array:Array of Stage Manager Sequence Steps Structures=Get SequenceSteps!(self:Stage Show Sequence Object Reference=@n21) -> (SequenceSteps:Array of Stage Manager Sequence Steps Structures)
    Dimension 1=Get StepIndex() -> (StepIndex:int=0)
   ) -> (Output:"Stage Manager Sequence Steps Structure (by ref)")
   > n23 = OpenWindow(
    self=Get DeleteConfirmationPopup(self=Get SparksTimelineEditor())
    Object="Get (a copy)"(
     Array:Array of Stage Show Event Object References=Break Stage Manager Sequence Steps!(StageManagerSequenceSteps:"Stage Manager Sequence Steps Structure (by ref)"=@n22) -> (StepEvents:Array of Stage Show Event Object References) {is=title K2Node_BreakStruct}
     Dimension 1=Get EventIndex() -> (EventIndex:int=0)
    ) -> (Output:Stage Show Event Object Reference)
   )
   n24 = Get(execute=@n29.else)
   n25 = Get SequenceSteps(self=@n24)
   Remove Index!(
    execute=[@n20.else, @n33.else]
    TargetArray:Array of Stage Manager Sequence Steps Structures=@n25
    IndexToRemove:int=Get StepIndex() -> (StepIndex=0)
   ) -> (then)
   > Set SequenceSteps!(execute, SequenceSteps:Array of Stage Manager Sequence Steps Structures=@n25, self:Stage Show Sequence Object Reference=@n24) -> (then, Output_Get:Array of Stage Manager Sequence Steps Structures)
   > TransactObject(Object=@n24)
   > Call On Deleted!(execute:Exec, self:EWBP Sparks Timeline Submenu Object Reference) -> (then:Exec) {is=title K2Node_CallDelegate}
   n26 = OnClosed!() -> (OutputDelegate, then, IsDeleted:bool=~, PropertyName:Name, ObjToDelete:Object Reference) {is=title K2Node_CustomEvent}
   > n27 = Is Valid(InputObject=@n26.ObjToDelete)
   > Branch(Condition=OR Boolean(A="Equal (Name)"(A=@n26.PropertyName, B=MojoStage_SequenceData), B="Equal (Name)"(A=@n26.PropertyName, B=FxManager_Event StepData)))
   > n28 = Sequence!(execute) -> (then_0, then_1, then_2:Exec)
   > n29 = Get!(execute) -> (then, else, SequencePoolData:Stage Show Sequence Pool Object Reference)
   Remove Index(
    execute=[@n29.then, @n24.else]
    TargetArray:Array of Stage Show Sequence Object References=Get Sequences(self=Get SequencePoolData())
    IndexToRemove=Get SequenceIndex() -> (SequenceIndex=0)
   )
   > TransactObject(Object=Get SequencePoolData())
   Branch(execute=@n28.then_1, Condition=@n26.IsDeleted)
   > DeleteLoadedAsset!(
    execute
    self:Editor Asset Subsystem Object Reference=Get EditorAssetSubsystem() -> (ReturnValue:Editor Asset Subsystem Object Reference) {is=title K2Node_GetEditorSubsystem}
    AssetToDelete:Object Reference=@n26.ObjToDelete
   ) -> (then, ReturnValue:bool)
   Branch(execute=@n28.then_2, Condition=NOT Boolean(A=Get Is Header()))
   > Unbind Event from On Closed(self:EWBP Sparks Delete Confirmation Object Reference=Get DeleteConfirmationPopup(self=Get SparksTimelineEditor()), Delegate=Create Event()) {is=title K2Node_RemoveDelegate}
   > Call On Deleted()
   Unbind Event from On Closed(execute=@n27.Is Not Valid, self=Get DeleteConfirmationPopup(self=Get SparksTimelineEditor()), Delegate=Create Event())
   TransactObject(execute=@n31.then, Object=@n24)
   n30 = Break Stage Manager Sequence Steps(StageManagerSequenceSteps="Get (a copy)"(Array:Array of Stage Manager Sequence Steps Structures=Get SequenceSteps(self=@n24), Dimension 1=Get StepIndex() -> (StepIndex=0)) -> (Output:Stage Manager Sequence Steps Structure))
   n31 = Remove Index(execute=@n33.then, TargetArray:Array of Stage Show Event Object References=@n30, IndexToRemove=Get EventIndex() -> (EventIndex=0))
   n32 = NOT Boolean(A=Get Is Header())
   n33 = Branch(execute=@n24.then, Condition=@n32)
  ]}
 ForceInjectTimelineEditorReference {nodes [
   n1 = ForceInjectTimelineEditorReference!() -> (then, SparksTimelineEditor:"Sparks Stage Manager Timeline Editor Object Reference (by ref)") {is=title K2Node_FunctionEntry}
   > Set SparksTimelineEditor(SparksTimelineEditor:Sparks Stage Manager Timeline Editor Object Reference=@n1) -> (Output_Get:Sparks Stage Manager Timeline Editor Object Reference)
  ]}
 OnAssetChanged {nodes [
   n1 = OnAssetChanged!() -> (then, ReturnObject:Object Reference, IsNewAsset:bool=~) {is=title K2Node_FunctionEntry}
   > Set ReturnedObj(ReturnedObj:Object Reference=@n1) -> (Output_Get:Object Reference)
   > Set LocalNewAsset(LocalNewAsset:bool=@n1.IsNewAsset) -> (Output_Get:bool)
   > n2 = Sequence()
   > n3 = Cast To MojoFxManagerSequence_DataAsset!(execute, Object=Get ReturnedObj() -> (ReturnedObj:Object Reference)) -> (then, CastFailed, AsMojo Fx Manager Sequence Data Asset:Mojo Fx Manager Sequence Data Asset Object Reference) {is=title K2Node_DynamicCast}
   > n4 = Get()
   > n5 = Branch(Condition=Get LocalNewAsset() -> (LocalNewAsset:bool))
   n6 = Get Sequences(self=@n4)
   > n7 = Add(TargetArray:Array of Stage Show Sequence Object References=@n6, NewItem:"Stage Show Sequence Object Reference (by ref)"=@n3) {is=title K2Node_CallArrayFunction}
   TransactObject(execute=[@n9.then, @n7.then], Object=Get SequencePoolData())
   n8 = Add(execute=@n5.else, TargetArray=@n6, NewItem=@n3)
   > Swap Array Elements!(
    execute
    TargetArray:Array of Stage Show Sequence Object References=@n6
    FirstIndex:int=Get SequenceIndex() -> (SequenceIndex=0)
    SecondIndex:int=@n8
   ) -> (then) {is=title K2Node_CallArrayFunction}
   > n9 = Remove Index(
    TargetArray:Array of Stage Show Sequence Object References=@n6
    IndexToRemove=int - int(A=Length(TargetArray:Array of Stage Show Sequence Object References=@n6))
   )
   n10 = Cast To MojoFxManagerEvent_DataAsset(execute=@n2.then_1, Object=Get ReturnedObj())
   > n11 = Get!(execute) -> (then, else, SequenceReference:Mojo Fx Manager Sequence Data Asset Object Reference)
   > n12 = Branch(Condition=Get LocalNewAsset())
   n13 = Break Stage Manager Sequence Steps(StageManagerSequenceSteps="Get (a copy)"(Array=Get SequenceSteps(self=@n11), Dimension 1=Get StepIndex() -> (StepIndex=0)))
   > Add(TargetArray:Array of Stage Show Event Object References=@n13, NewItem:"Stage Show Event Object Reference (by ref)"=@n10)
   n14 = Add(execute=@n12.else, TargetArray=@n13, NewItem=@n10)
   > Swap Array Elements(TargetArray:Array of Stage Show Event Object References=@n13, FirstIndex=Get EventIndex() -> (EventIndex=0), SecondIndex=@n14)
   > Remove Index(TargetArray:Array of Stage Show Event Object References=@n13, IndexToRemove=@n14)
   > TransactObject(Object=Get SequenceReference() -> (SequenceReference:Mojo Fx Manager Sequence Data Asset Object Reference))
   Unbind Event from On Asset Changed(execute=@n2.then_2, self:Sparks Stage Manager Timeline Editor Object Reference=Get SparksTimelineEditor(), Delegate=Create Event()) {is=title K2Node_RemoveDelegate}
   > Call On Deleted()
  ]}
 OnDeleted {nodes [OnDeleted() {is=title K2Node_FunctionEntry}]}
 UpdateSubmenuData {nodes [
   n1 = UpdateSubmenuData!() -> (
    then
    SequencePoolData:"Stage Show Sequence Pool Object Reference (by ref)"
    SequenceReference:"Mojo Fx Manager Sequence Data Asset Object Reference (by ref)"
    EventReference:"Object Reference (by ref)"
    SparksTimelineEditor:"Sparks Stage Manager Timeline Editor Object Reference (by ref)"
    EventBarReference:EWBP Sparks Sequence Pool Timeline Event Bar Object Reference
    SequenceIndex:int
    StepIndex:int
    EventIndex:int
    IsHeader:bool=~
   ) {is=title K2Node_FunctionEntry}
   > n2 = Sequence!(execute) -> (then_0, then_1, then_2, then_3:Exec, then_4:Exec, then_5:Exec, then_6:Exec, then_7:Exec, then_8:Exec)
   > Set SequencePoolData(SequencePoolData:Stage Show Sequence Pool Object Reference=@n1) -> (Output_Get:Stage Show Sequence Pool Object Reference)
   Set SequenceReference(execute=@n2.then_1, SequenceReference:Mojo Fx Manager Sequence Data Asset Object Reference=@n1.SequenceReference) -> (Output_Get:Mojo Fx Manager Sequence Data Asset Object Reference)
   Set EventReference(execute=@n2.then_2, EventReference:Object Reference=@n1.EventReference) -> (Output_Get:Object Reference)
   Set SparksTimelineEditor(execute=@n2.then_3, SparksTimelineEditor=@n1.SparksTimelineEditor)
   Set SequenceIndex(execute=@n2.then_4, SequenceIndex:int=@n1.SequenceIndex) -> (Output_Get:int=0)
   Set EventBarReference(execute=@n2.then_5, EventBarReference:EWBP Sparks Sequence Pool Timeline Event Bar Object Reference=@n1.EventBarReference) -> (Output_Get:EWBP Sparks Sequence Pool Timeline Event Bar Object Reference)
   Set StepIndex(execute=@n2.then_6, StepIndex:int=@n1.StepIndex) -> (Output_Get:int=0)
   Set EventIndex(execute=@n2.then_7, EventIndex:int=@n1.EventIndex) -> (Output_Get:int=0)
   Set Is Header(execute=@n2.then_8, Is Header:bool=@n1.IsHeader) -> (Output_Get:bool)
  ]}
 UpdateVisibilityStatus {nodes [UpdateVisibilityStatus!() -> (then, IsVisible:bool=~) {is=title K2Node_FunctionEntry}]}
]
variables [
 EventBarReference {type=EWBP_SparksSequencePool_TimelineEventBar_C, default=null}
 EventIndex {type=integer}
 EventReference {type=Object, default=null}
 Is Header {type=boolean}
 IsClosed {type=Object, default=null}
 IsLooping {type=boolean}
 OnDeleted {default=()}
 SequenceIndex {type=integer}
 SequencePoolData {type=StageShowSequencePool, default=null}
 SequenceReference {type=MojoFxManagerSequence_DataAsset_C, default=null}
 SparksTimelineEditor {type=SparksStageManagerTimelineEditor_C, default=null}
 StepIndex {type=integer}
 UpdateVisibilityStatus {default=()}
]
widgets {
 info {rootWidgetClass=&Overlay, widgetCount=15}
 widgets [
  Overlay_0 {widgetClassPath=&Overlay, widget=Overlay_0}
  bg {
   widgetClassPath=&Image
   parent=Overlay_0
   widget=bg
   properties {
    brush {
     bIsDynamicallyLoaded=false
     drawAs=RoundedBox
     imageType=NoImage
     mirroring=NoMirror
     resourceName=None
     resourceObject=null
     tiling=NoTile
     imageSize=<32 32>
     margin=<ltrb 0 0 0 0>
     outlineSettings {
      bUseBrushTransparency=false
      roundingType=FixedRadius
      width=0
      color {colorUseRule=UseColor_Specified, specifiedColor=<rgba 0 0 0 0>}
      cornerRadii=<2 2 0 1>
     }
     tintColor {colorUseRule=UseColor_Specified, specifiedColor=<rgba 1 1 1 1>}
     uVRegion {bIsValid=false, max=<0 0>, min=<0 0>}
    }
    colorAndOpacity=<rgba 0.1458 0.1458 0.1458 1>
   }
   slot {class=OverlaySlot, properties {horizontalAlignment=HAlign_Fill, verticalAlignment=VAlign_Fill}}
  }
  VerticalBox_31 {
   widgetClassPath=&VerticalBox
   parent=Overlay_0
   widget=VerticalBox_31
   slot {class=OverlaySlot, properties {horizontalAlignment=HAlign_Fill, verticalAlignment=VAlign_Fill, padding=<ltrb 4 4 4 4>}}
  }
  ChangeButton {
   widgetClassPath=&EditorUtilityButton
   parent=VerticalBox_31
   widget=ChangeButton
   properties {widgetStyle {
     clickedSlateSound {resourceObject=null}
     disabled {
      bIsDynamicallyLoaded=false
      drawAs=RoundedBox
      imageType=NoImage
      mirroring=NoMirror
      resourceName=None
      resourceObject=null
      tiling=NoTile
      imageSize=<32 32>
      margin=<ltrb 0.5 0.5 0.5 0.5>
      outlineSettings {
       bUseBrushTransparency=false
       roundingType=FixedRadius
       width=0
       color {colorUseRule=UseColor_ColorTable, specifiedColor=<rgba 0 0 0 0>}
       cornerRadii=<4 4 4 4>
      }
      tintColor {colorUseRule=UseColor_Specified, specifiedColor=<rgba 0.0395 0.0395 0.0395 1>}
      uVRegion {bIsValid=false, max=<0 0>, min=<0 0>}
     }
     disabledForeground {colorUseRule=UseColor_ColorTable, specifiedColor=<rgba 0 0 0 0>}
     hovered {
      bIsDynamicallyLoaded=false
      drawAs=RoundedBox
      imageType=NoImage
      mirroring=NoMirror
      resourceName=None
      resourceObject=null
      tiling=NoTile
      imageSize=<32 32>
      margin=<ltrb 0 0 0 0>
      outlineSettings {
       bUseBrushTransparency=false
       roundingType=FixedRadius
       width=0
       color {colorUseRule=UseColor_ColorTable, specifiedColor=<rgba 0 0 0 0>}
       cornerRadii=<4 4 4 4>
      }
      tintColor {colorUseRule=UseColor_Specified, specifiedColor=<rgba 0.1126 0.7031 0 1>}
      uVRegion {bIsValid=false, max=<0 0>, min=<0 0>}
     }
     hoveredForeground {colorUseRule=UseColor_ColorTable, specifiedColor=<rgba 0 0 0 0>}
     hoveredSlateSound {resourceObject=null}
     normal {
      bIsDynamicallyLoaded=false
      drawAs=RoundedBox
      imageType=NoImage
      mirroring=NoMirror
      resourceName=None
      resourceObject=null
      tiling=NoTile
      imageSize=<32 32>
      margin=<ltrb 0.5 0.5 0.5 0.5>
      outlineSettings {
       bUseBrushTransparency=false
       roundingType=FixedRadius
       width=0
       color {colorUseRule=UseColor_ColorTable, specifiedColor=<rgba 0 0 0 0>}
       cornerRadii=<4 4 4 4>
      }
      tintColor {colorUseRule=UseColor_Specified, specifiedColor=<rgba 0.0771 0.6167 0.1289 1>}
      uVRegion {bIsValid=false, max=<0 0>, min=<0 0>}
     }
     normalForeground {colorUseRule=UseColor_ColorTable, specifiedColor=<rgba 0 0 0 0>}
     normalPadding=<ltrb 12 1.5 12 1.5>
     pressed {
      bIsDynamicallyLoaded=false
      drawAs=RoundedBox
      imageType=NoImage
      mirroring=NoMirror
      resourceName=None
      resourceObject=null
      tiling=NoTile
      imageSize=<32 32>
      margin=<ltrb 0.5 0.5 0.5 0.5>
      outlineSettings {
       bUseBrushTransparency=false
       roundingType=FixedRadius
       width=0
       color {colorUseRule=UseColor_ColorTable, specifiedColor=<rgba 0 0 0 0>}
       cornerRadii=<4 4 4 4>
      }
      tintColor {colorUseRule=UseColor_Specified, specifiedColor=<rgba 0.1806 0.4583 0.1572 1>}
      uVRegion {bIsValid=false, max=<0 0>, min=<0 0>}
     }
     pressedForeground {colorUseRule=UseColor_ColorTable, specifiedColor=<rgba 0 0 0 0>}
     pressedPadding=<ltrb 12 2.5 12 0.5>
     pressedSlateSound {resourceObject=null}
    }}
   slot {class=VerticalBoxSlot}
  }
  HorizontalBox_0 {
   widgetClassPath=&HorizontalBox
   parent=ChangeButton
   widget=HorizontalBox_0
   slot {class=ButtonSlot, properties {horizontalAlignment=HAlign_Left}}
  }
  Image_1 {
   widgetClassPath=&Image
   parent=HorizontalBox_0
   widget=Image_1
   properties {brush {
     bIsDynamicallyLoaded=false
     drawAs=Image
     imageType=NoImage
     mirroring=NoMirror
     resourceName=None
     resourceObject=/FortUILibrary/Foundation/Buttons/Materials/M_UI_Icon_ExternalLink
     tiling=NoTile
     imageSize=<15 15>
     margin=<ltrb 0 0 0 0>
     outlineSettings {
      bUseBrushTransparency=false
      roundingType=HalfHeightRadius
      width=0
      color {colorUseRule=UseColor_Specified, specifiedColor=<rgba 0 0 0 0>}
      cornerRadii=<0 0 0 0>
     }
     tintColor {colorUseRule=UseColor_Specified, specifiedColor=<rgba 1 1 1 1>}
     uVRegion {bIsValid=false, max=<0 0>, min=<0 0>}
    }}
   slot {class=HorizontalBoxSlot, properties {horizontalAlignment=HAlign_Center, verticalAlignment=VAlign_Center, padding=<ltrb 0 0 3 0>}}
  }
  TextBlock_25 {
   widgetClassPath=&TextBlock
   parent=HorizontalBox_0
   widget=TextBlock_25
   properties {
    text=Change
    font {
     bForceMonospaced=false
     bMaterialIsStencil=false
     fontMaterial=null
     fontObject=&HeadingNow
     letterSpacing=0
     monospacedWidth=1
     size=7.5
     skewAmount=0
     typefaceFontName=75 Medium
     outlineSettings {
      bApplyOutlineToDropShadows=false
      bMiteredCorners=false
      bSeparateFillAlpha=false
      outlineMaterial=null
      outlineSize=0
      outlineColor=<rgba 0 0 0 1>
     }
    }
   }
   slot {class=HorizontalBoxSlot, properties {verticalAlignment=VAlign_Center}}
  }
  EditButton {
   widgetClassPath=&EditorUtilityButton
   parent=VerticalBox_31
   widget=EditButton
   properties {widgetStyle {
     clickedSlateSound {resourceObject=null}
     disabled {
      bIsDynamicallyLoaded=false
      drawAs=RoundedBox
      imageType=NoImage
      mirroring=NoMirror
      resourceName=None
      resourceObject=/Game/UI/Foundation/Textures/Buttons/T-Button-Square
      tiling=NoTile
      imageSize=<32 32>
      margin=<ltrb 0.5 0.5 0.5 0.5>
      outlineSettings {
       bUseBrushTransparency=false
       roundingType=FixedRadius
       width=0
       color {colorUseRule=UseColor_ColorTable, specifiedColor=<rgba 0 0 0 0>}
       cornerRadii=<4 4 4 4>
      }
      tintColor {colorUseRule=UseColor_Specified, specifiedColor=<rgba 0.0395 0.0395 0.0395 1>}
      uVRegion {bIsValid=false, max=<0 0>, min=<0 0>}
     }
     disabledForeground {colorUseRule=UseColor_ColorTable, specifiedColor=<rgba 0 0 0 0>}
     hovered {
      bIsDynamicallyLoaded=false
      drawAs=RoundedBox
      imageType=NoImage
      mirroring=NoMirror
      resourceName=None
      resourceObject=&T-Button-Square
      tiling=NoTile
      imageSize=<32 32>
      margin=<ltrb 0.5 0.5 0.5 0.5>
      outlineSettings {
       bUseBrushTransparency=false
       roundingType=FixedRadius
       width=0
       color {colorUseRule=UseColor_ColorTable, specifiedColor=<rgba 0 0 0 0>}
       cornerRadii=<4 4 4 4>
      }
      tintColor {colorUseRule=UseColor_Specified, specifiedColor=<rgba 0.7344 0.4651 0 1>}
      uVRegion {bIsValid=false, max=<0 0>, min=<0 0>}
     }
     hoveredForeground {colorUseRule=UseColor_ColorTable, specifiedColor=<rgba 0 0 0 0>}
     hoveredSlateSound {resourceObject=null}
     normal {
      bIsDynamicallyLoaded=false
      drawAs=RoundedBox
      imageType=NoImage
      mirroring=NoMirror
      resourceName=None
      resourceObject=&T-Button-Square
      tiling=NoTile
      imageSize=<32 32>
      margin=<ltrb 0.5 0.5 0.5 0.5>
      outlineSettings {
       bUseBrushTransparency=false
       roundingType=FixedRadius
       width=0
       color {colorUseRule=UseColor_ColorTable, specifiedColor=<rgba 0 0 0 0>}
       cornerRadii=<4 4 4 4>
      }
      tintColor {colorUseRule=UseColor_Specified, specifiedColor=<rgba 1 0.6333 0 1>}
      uVRegion {bIsValid=false, max=<0 0>, min=<0 0>}
     }
     normalForeground {colorUseRule=UseColor_ColorTable, specifiedColor=<rgba 0 0 0 0>}
     normalPadding=<ltrb 12 1.5 12 1.5>
     pressed {
      bIsDynamicallyLoaded=false
      drawAs=RoundedBox
      imageType=NoImage
      mirroring=NoMirror
      resourceName=None
      resourceObject=&T-Button-Square
      tiling=NoTile
      imageSize=<32 32>
      margin=<ltrb 0.5 0.5 0.5 0.5>
      outlineSettings {
       bUseBrushTransparency=false
       roundingType=FixedRadius
       width=0
       color {colorUseRule=UseColor_ColorTable, specifiedColor=<rgba 0 0 0 0>}
       cornerRadii=<4 4 4 4>
      }
      tintColor {colorUseRule=UseColor_Specified, specifiedColor=<rgba 0.6458 0.409 0 1>}
      uVRegion {bIsValid=false, max=<0 0>, min=<0 0>}
     }
     pressedForeground {colorUseRule=UseColor_ColorTable, specifiedColor=<rgba 0 0 0 0>}
     pressedPadding=<ltrb 12 2.5 12 0.5>
     pressedSlateSound {resourceObject=null}
    }}
   slot {class=VerticalBoxSlot}
  }
  HorizontalBox_2 {
   widgetClassPath=&HorizontalBox
   parent=EditButton
   widget=HorizontalBox_2
   slot {class=ButtonSlot, properties {horizontalAlignment=HAlign_Left}}
  }
  ExternalIcon_1 {
   widgetClassPath=&Image
   parent=HorizontalBox_2
   widget=ExternalIcon_1
   properties {brush {
     bIsDynamicallyLoaded=false
     drawAs=Image
     imageType=NoImage
     mirroring=NoMirror
     resourceName=None
     resourceObject=&M_UI_Icon_ExternalLink
     tiling=NoTile
     imageSize=<15 15>
     margin=<ltrb 0 0 0 0>
     outlineSettings {
      bUseBrushTransparency=false
      roundingType=HalfHeightRadius
      width=0
      color {colorUseRule=UseColor_Specified, specifiedColor=<rgba 0 0 0 0>}
      cornerRadii=<0 0 0 0>
     }
     tintColor {colorUseRule=UseColor_Specified, specifiedColor=<rgba 1 1 1 1>}
     uVRegion {bIsValid=false, max=<0 0>, min=<0 0>}
    }}
   slot {class=HorizontalBoxSlot, properties {horizontalAlignment=HAlign_Center, verticalAlignment=VAlign_Center, padding=<ltrb 0 0 3 0>}}
  }
  TextBlock_1 {
   widgetClassPath=&TextBlock
   parent=HorizontalBox_2
   widget=TextBlock_1
   properties {
    text=Edit
    font {
     bForceMonospaced=false
     bMaterialIsStencil=false
     fontMaterial=null
     fontObject=&HeadingNow
     letterSpacing=0
     monospacedWidth=1
     size=7.5
     skewAmount=0
     typefaceFontName=75 Medium
     outlineSettings {
      bApplyOutlineToDropShadows=false
      bMiteredCorners=false
      bSeparateFillAlpha=false
      outlineMaterial=null
      outlineSize=0
      outlineColor=<rgba 0 0 0 1>
     }
    }
   }
   slot {class=HorizontalBoxSlot, properties {verticalAlignment=VAlign_Center}}
  }
  RemoveButton {
   widgetClassPath=&EditorUtilityButton
   parent=VerticalBox_31
   widget=RemoveButton
   properties {widgetStyle {
     clickedSlateSound {resourceObject=null}
     disabled {
      bIsDynamicallyLoaded=false
      drawAs=RoundedBox
      imageType=NoImage
      mirroring=NoMirror
      resourceName=None
      resourceObject=null
      tiling=NoTile
      imageSize=<20 20>
      margin=<ltrb 0 0 0 0>
      outlineSettings {
       bUseBrushTransparency=false
       roundingType=FixedRadius
       width=0
       color {colorUseRule=UseColor_ColorTable, specifiedColor=<rgba 0 0 0 0>}
       cornerRadii=<4 4 4 4>
      }
      tintColor {colorUseRule=UseColor_Specified, specifiedColor=<rgba 0.0395 0.0395 0.0395 1>}
      uVRegion {bIsValid=false, max=<0 0>, min=<0 0>}
     }
     disabledForeground {colorUseRule=UseColor_ColorTable, specifiedColor=<rgba 0 0 0 0>}
     hovered {
      bIsDynamicallyLoaded=false
      drawAs=RoundedBox
      imageType=NoImage
      mirroring=NoMirror
      resourceName=None
      resourceObject=null
      tiling=NoTile
      imageSize=<32 32>
      margin=<ltrb 0 0 0 0>
      outlineSettings {
       bUseBrushTransparency=false
       roundingType=FixedRadius
       width=0
       color {colorUseRule=UseColor_ColorTable, specifiedColor=<rgba 0 0 0 0>}
       cornerRadii=<4 4 4 4>
      }
      tintColor {colorUseRule=UseColor_Specified, specifiedColor=<rgba 0.5521 0 0.0807 1>}
      uVRegion {bIsValid=false, max=<0 0>, min=<0 0>}
     }
     hoveredForeground {colorUseRule=UseColor_ColorTable, specifiedColor=<rgba 0 0 0 0>}
     hoveredSlateSound {resourceObject=null}
     normal {
      bIsDynamicallyLoaded=false
      drawAs=RoundedBox
      imageType=NoImage
      mirroring=NoMirror
      resourceName=None
      resourceObject=null
      tiling=NoTile
      imageSize=<20 20>
      margin=<ltrb 0 0 0 0>
      outlineSettings {
       bUseBrushTransparency=false
       roundingType=FixedRadius
       width=0
       color {colorUseRule=UseColor_ColorTable, specifiedColor=<rgba 0 0 0 0>}
       cornerRadii=<4 4 4 4>
      }
      tintColor {colorUseRule=UseColor_Specified, specifiedColor=<rgba 0.775 0 0.0304 1>}
      uVRegion {bIsValid=false, max=<0 0>, min=<0 0>}
     }
     normalForeground {colorUseRule=UseColor_ColorTable, specifiedColor=<rgba 0 0 0 0>}
     normalPadding=<ltrb 12 1.5 12 1.5>
     pressed {
      bIsDynamicallyLoaded=false
      drawAs=RoundedBox
      imageType=NoImage
      mirroring=NoMirror
      resourceName=None
      resourceObject=null
      tiling=NoTile
      imageSize=<20 20>
      margin=<ltrb 0 0 0 0>
      outlineSettings {
       bUseBrushTransparency=false
       roundingType=FixedRadius
       width=0
       color {colorUseRule=UseColor_ColorTable, specifiedColor=<rgba 0 0 0 0>}
       cornerRadii=<4 4 4 4>
      }
      tintColor {colorUseRule=UseColor_Specified, specifiedColor=<rgba 0.2396 0 0 1>}
      uVRegion {bIsValid=false, max=<0 0>, min=<0 0>}
     }
     pressedForeground {colorUseRule=UseColor_ColorTable, specifiedColor=<rgba 0 0 0 0>}
     pressedPadding=<ltrb 12 2.5 12 0.5>
     pressedSlateSound {resourceObject=null}
    }}
   slot {class=VerticalBoxSlot}
  }
  HorizontalBox_1 {
   widgetClassPath=&HorizontalBox
   parent=RemoveButton
   widget=HorizontalBox_1
   slot {class=ButtonSlot, properties {horizontalAlignment=HAlign_Left}}
  }
  Image_2 {
   widgetClassPath=&Image
   parent=HorizontalBox_1
   widget=Image_2
   properties {brush {
     bIsDynamicallyLoaded=false
     drawAs=Image
     imageType=NoImage
     mirroring=NoMirror
     resourceName=None
     resourceObject=/Game/Athena/UI/InGame/Creative/PropertyEditor/Icon_Picker/Textures_V2/64x64/T_UI_IconLibrary_TrashCan_64
     tiling=NoTile
     imageSize=<15 15>
     margin=<ltrb 0 0 0 0>
     outlineSettings {
      bUseBrushTransparency=false
      roundingType=HalfHeightRadius
      width=0
      color {colorUseRule=UseColor_Specified, specifiedColor=<rgba 0 0 0 0>}
      cornerRadii=<0 0 0 0>
     }
     tintColor {colorUseRule=UseColor_Specified, specifiedColor=<rgba 1 1 1 1>}
     uVRegion {bIsValid=false, max=<0 0>, min=<0 0>}
    }}
   slot {class=HorizontalBoxSlot, properties {horizontalAlignment=HAlign_Center, verticalAlignment=VAlign_Center, padding=<ltrb 0 0 3 0>}}
  }
  TextBlock {
   widgetClassPath=&TextBlock
   parent=HorizontalBox_1
   widget=TextBlock
   properties {
    text=Remove
    font {
     bForceMonospaced=false
     bMaterialIsStencil=false
     fontMaterial=null
     fontObject=&HeadingNow
     letterSpacing=0
     monospacedWidth=1
     size=7.5
     skewAmount=0
     typefaceFontName=75 Medium
     outlineSettings {
      bApplyOutlineToDropShadows=false
      bMiteredCorners=false
      bSeparateFillAlpha=false
      outlineMaterial=null
      outlineSize=0
      outlineColor=<rgba 0 0 0 1>
     }
    }
   }
   slot {class=HorizontalBoxSlot, properties {verticalAlignment=VAlign_Center}}
  }
 ]
}
