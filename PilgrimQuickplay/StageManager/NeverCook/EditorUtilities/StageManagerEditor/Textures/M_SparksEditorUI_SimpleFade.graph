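# M_SparksEditorUI_SimpleFade is a translucent UI material that builds a linear fade, stored with the Stage Manager editor utility textures.
# Static switches IsVertical and Invert pick the V or U texture coordinate, optionally flipped.
# The chosen gradient is divided by a constant 0.5, doubling its slope so it reaches one at the midpoint.
# White constant and vertex color nodes are also present, with no connections shown.
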Material /PilgrimQuickplay/StageManager/NeverCook/EditorUtilities/StageManagerEditor/Textures/M_SparksEditorUI_SimpleFade
graph {
 properties {blendMode=BLEND_Translucent, materialDomain=MD_UI, pixelDepthOffsetMode=PDOM_Legacy, usedShadingModels=Default Lit}
 nodes [
  Constant3Vector() {constant=<rgba 1 1 1 0>}
  n1 = TextureCoordinate()
  n2 = ComponentMask(None=@n1) {g=true}
  n3 = OneMinus(None=@n2)
  n4 = ComponentMask(None=@n1) {r=true}
  n5 = OneMinus(None=@n4)
  Divide(A=StaticSwitchParameter(
    True=StaticSwitchParameter(True=@n3, False=@n2) {parameterName=Invert}
    False=StaticSwitchParameter(True=@n5, False=@n4) {parameterName=Invert}
   ) {parameterName=IsVertical}) {constB=0.5}
  VertexColor()
 ]
}
