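# Editor utility widget that draws one stage show event as a bar in the Sparks Stage Manager timeline editor.
# UpdateBar and PreConstruct size the delay and event segments from DelayValue and EventSize, setting color, labels and icon.
# It registers with the timeline editor, routes clicks to UpdateToggles, and feeds an embedded submenu whose deletion refreshes the editor.
# Duplicate target FX manager IDs show a yellow warning icon on non-looping events.
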EditorUtilityWidgetBlueprint /PilgrimQuickplay/StageManager/NeverCook/EditorUtilities/StageManagerEditor/EWBP_SparksSequencePool_TimelineEventBar
graphs [
 CheckForDuplicatedManagers {nodes [
   CheckForDuplicatedManagers() {is=title K2Node_FunctionEntry}
   > Clear(TargetArray:Array of Names=Get FilteredUniqueValues() -> (FilteredUniqueValues:Array of Names))
   > n1 = For Each Loop(Array:Array of Names=Get TargetFxManagerIds!(self:Stage Show Event Object Reference=Get EventReference() -> (EventReference:Stage Show Event Object Reference)) -> (TargetFxManagerIds:Array of Names)) -> (Array Element:Name)
   n2 = Get FilteredUniqueValues()
   n3 = Get TargetFxManagerIds(self=Get EventReference())
   MarkAsWarning!(
    execute=@n1.Completed
    self
    Value:bool="Not Equal (Integer)"(A=Length(TargetArray:Array of Names=@n2), B=Length(TargetArray=@n3))
    WarningContext:Text="INVTEXT(\"event\")"
   ) -> (then)
   Add Unique(execute=@n1.LoopBody, TargetArray:Array of Names=Get FilteredUniqueValues(), NewItem:"Name (by ref)"=@n1)
  ]}
 EventGraph {nodes [
   n1 = "On Hovered (EditorUtilityButton_35)"() {is=title K2Node_ComponentBoundEvent}
   n2 = Get EditorUtilityButton_35() -> (EditorUtilityButton_35:Editor Utility Button Object Reference)
   n3 = Break Slate Brush!(SlateBrush:"Slate Brush Structure (by ref)"=Break Button Style!(ButtonStyle:"Button Style Structure (by ref)"=Get WidgetStyle!(self:Button Object Reference=@n2) -> (WidgetStyle:Button Style Structure)) -> (Hovered:Slate Brush Structure, ClickedSlateSound:Slate Sound Structure) {is=title K2Node_BreakStruct}) -> (TintColor:Slate Color Structure) {is=title K2Node_BreakStruct}
   SetBrushTintColor!(
    execute=[@n1.then, @n9.then]
    self:Image Object Reference=[Get Degradee_Right() -> (Degradee_Right:Image Object Reference), Get Degradee_Left() -> (Degradee_Left:Image Object Reference)]
    TintColor:Slate Color Structure=@n3
   ) -> (then)
   "On Pressed (EditorUtilityButton_35)"() {is=title K2Node_ComponentBoundEvent}
   n4 = Get EditorUtilityButton_35()
   n5 = Break Slate Brush(SlateBrush=Break Button Style!(ButtonStyle=Get WidgetStyle(self=@n4)) -> (Pressed:Slate Brush Structure, ClickedSlateSound))
   > SetBrushTintColor(self=[Get Degradee_Right(), Get Degradee_Left()], TintColor=@n5)
   "On Pressed (EditorUtilityButton)"() {is=title K2Node_ComponentBoundEvent}
   > SetBrushTintColor(
    self=[Get Degradee_Right_1() -> (Degradee_Right_1:Image Object Reference), Get Degradee_Left_1() -> (Degradee_Left_1:Image Object Reference)]
    TintColor=@n5
   )
   n6 = "On Clicked (EditorUtilityButton)"() {is=title K2Node_ComponentBoundEvent}
   UpdateToggles!(
    execute=[@n6.then, @n10.then]
    self:Sparks Stage Manager Timeline Editor Object Reference=Get SparksTimelineEditor() -> (SparksTimelineEditor:Sparks Stage Manager Timeline Editor Object Reference)
    Toggle:EWBP Sparks Sequence Pool Timeline Event Bar Object Reference=Self-Reference()
    ForceCloseAll:bool=false
    DontCallEvent:bool=false
   ) -> (then)
   "On Unhovered (EditorUtilityButton_35)"() {is=title K2Node_ComponentBoundEvent}
   n7 = Get EditorUtilityButton_35()
   n8 = Break Slate Brush(SlateBrush=Break Button Style!(ButtonStyle=Get WidgetStyle(self=@n7)) -> (Normal:Slate Brush Structure, ClickedSlateSound))
   > SetBrushTintColor(self=[Get Degradee_Right(), Get Degradee_Left()], TintColor=@n8)
   n9 = "On Released (EditorUtilityButton_35)"() {is=title K2Node_ComponentBoundEvent}
   n10 = "On Clicked (EditorUtilityButton_35)"() {is=title K2Node_ComponentBoundEvent}
   n11 = "Update Visibility Status (EWBP_SparksTimelineSubmenu)"!() -> (OutputDelegate, then, IsVisible:bool) {is=title K2Node_ComponentBoundEvent}
   > Set IsSubmenuOpen(IsSubmenuOpen:bool=@n11.IsVisible) -> (Output_Get:bool)
   "On Deleted (EWBP_SparksTimelineSubmenu)"() {is=title K2Node_ComponentBoundEvent}
   > ForceRefresh(self:Sparks Stage Manager Timeline Editor Object Reference=Get SparksTimelineEditor())
   n12 = "On Hovered (EditorUtilityButton)"() {is=title K2Node_ComponentBoundEvent}
   SetBrushTintColor(execute=[@n12.then, @n13.then], self=[Get Degradee_Right_1(), Get Degradee_Left_1()], TintColor=@n3)
   n13 = "On Released (EditorUtilityButton)"() {is=title K2Node_ComponentBoundEvent}
   "On Unhovered (EditorUtilityButton)"() {is=title K2Node_ComponentBoundEvent}
   > SetBrushTintColor(self=[Get Degradee_Right_1(), Get Degradee_Left_1()], TintColor=@n8)
   n14 = UpdateBar() {is=title K2Node_CustomEvent}
   n15 = Event PreConstruct()
   n16 = Branch(execute=[@n14.then, @n15.then], Condition=NOT Boolean(A=@n15.IsDesignTime))
   SetContentColorAndOpacity!(
    execute=[@n24.then, @n16.then]
    self:Border Object Reference=Get ButtonBorder() -> (ButtonBorder:Border Object Reference)
    InContentColorAndOpacity:Linear Color Structure=Get BarColor() -> (BarColor:Linear Color Structure)
   ) -> (then)
   > SetMinDesiredWidth!(
    execute
    self:Size Box Object Reference=Get ButtonSizeBox() -> (ButtonSizeBox:Size Box Object Reference)
    InMinDesiredWidth:float=float * float(A=Get EventSize() -> (EventSize:float), B=100) {is=title K2Node_PromotableOperator}
   ) -> (then)
   n17 = Get DelayValue() -> (DelayValue:float)
   n18 = "float > float"(A=@n17)
   > SetWidthOverride(
    self=Get DelayBar() -> (DelayBar:Size Box Object Reference)
    InWidthOverride=float + float(A=float * float(A=@n17, B=100), B=Select(Option 0:double=0, Option 1:double=15, Index=@n18) -> (ReturnValue:double))
   )
   n19 = Format Text!(
    Format="LOCTEXT(\"{seconds}s\")"
    seconds:Text="To Text (Float)"(Value=@n17, MinimumIntegralDigits=1, MaximumIntegralDigits=324, MinimumFractionalDigits=1, MaximumFractionalDigits=3)
   ) -> (Result)
   n20 = Get DelayText() -> (DelayText:Text Object Reference)
   > "SetText (Text)"(self=@n20, InText=@n19)
   > SetToolTipText!(
    execute
    self:Widget Object Reference=[@n20, Get EditorUtilityButton() -> (EditorUtilityButton:Editor Utility Button Object Reference)]
    InToolTipText:"Text (by ref)"=Format Text(Format="LOCTEXT(\"Delay:  {seconds}\")", seconds=@n19)
   ) -> (then)
   n21 = Break Button Style!(ButtonStyle=Get WidgetStyle(self=Get EditorUtilityButton_35())) -> (Normal, Disabled:Slate Brush Structure, ClickedSlateSound)
   n22 = Self-Reference()
   > SetBrushTintColor(
    self=[Get Degradee_Right(), Get Degradee_Left()]
    TintColor=Select(
     Option 0:Slate Color Structure=Break Slate Brush(SlateBrush=@n21.Disabled)
     Option 1:Slate Color Structure=Break Slate Brush(SlateBrush=@n21)
     Index=GetIsEnabled!(self:Widget Object Reference=@n22) -> (ReturnValue:bool)
    ) -> (ReturnValue:Slate Color Structure)
   )
   > SetBrushFromTexture!(
    execute
    self:Image Object Reference=Get Icon() -> (Icon:Image Object Reference)
    Texture:Texture 2D Object Reference=Get SideIcon() -> (SideIcon:Texture 2D Object Reference)
    bMatchSize:bool=false
   ) -> (then)
   > SetVisibility(
    self=Get Icon()
    InVisibility=Select(Option 0:ESlateVisibility Enum=Collapsed, Option 1:ESlateVisibility Enum=SelfHitTestInvisible, Index=Get ShowIcon() -> (ShowIcon:bool)) -> (ReturnValue:ESlateVisibility Enum)
   )
   n23 = Get ButtonSizeBox()
   > SetMaxDesiredWidth!(
    execute
    self:Size Box Object Reference=Get EventTextNameSizeBox() -> (EventTextNameSizeBox:Size Box Object Reference)
    InMaxDesiredWidth:float=float - float(A=Get MinDesiredWidth!(self:Size Box Object Reference=@n23) -> (MinDesiredWidth:float), B=10)
   ) -> (then)
   n24 = "SetText (Text)"(execute=@n16.else, self=Get EventBarName() -> (EventBarName:Text Object Reference), InText=Get EventName() -> (EventName:Text))
   Event Construct()
   > SetToolTipText(self=[Get EventBarName(), Get EditorUtilityButton_35()], InToolTipText=Get EventName())
   > UpdateSubmenuData!(
    execute
    self:EWBP Sparks Timeline Submenu Object Reference=Get EWBP_SparksTimelineSubmenu() -> (EWBP_SparksTimelineSubmenu:EWBP Sparks Timeline Submenu Object Reference)
    SequencePoolData:"Stage Show Sequence Pool Object Reference (by ref)"
    SequenceReference:"Mojo Fx Manager Sequence Data Asset Object Reference (by ref)"=Get SequenceReference() -> (SequenceReference:Mojo Fx Manager Sequence Data Asset Object Reference)
    EventReference:"Object Reference (by ref)"=Get EventReference()
    SparksTimelineEditor:"Sparks Stage Manager Timeline Editor Object Reference (by ref)"=Get SparksTimelineEditor()
    EventBarReference:EWBP Sparks Sequence Pool Timeline Event Bar Object Reference=Self-Reference()
    SequenceIndex:int=0
    StepIndex:int=Get StepIndex() -> (StepIndex:int=0)
    EventIndex:int=Get EventIndex() -> (EventIndex:int=0)
    IsHeader:bool=false
   ) -> (then)
   > Add Unique(
    TargetArray:Array of EWBP Sparks Sequence Pool Timeline Event Bar Object References=Get AllEvents!(self:Sparks Stage Manager Timeline Editor Object Reference=Get SparksTimelineEditor()) -> (AllEvents:Array of EWBP Sparks Sequence Pool Timeline Event Bar Object References)
    NewItem:"EWBP Sparks Sequence Pool Timeline Event Bar Object Reference (by ref)"=Self-Reference()
   )
   > "SetText (Text)"(self=Get EventBarName(), InText=Get EventName())
   n25 = Get!(execute:Exec) -> (then:Exec, else:Exec, EventReference:Stage Show Event Object Reference)
   > OpenEditorForAssets!(
    execute
    self:Asset Editor Subsystem Object Reference=Get AssetEditorSubsystem()
    Assets:Array of Object References=Make Array("[0]":Object Reference=@n25) -> (Array:Array of Object References)
    OpenedMethod:EAssetTypeActivationOpenedMethod Enum=Edit
   ) -> (then, ReturnValue:bool)
   n26 = Self-Reference()
   Call Open Submenu!(
    execute:Exec
    self:EWBP Sparks Sequence Pool Timeline Event Bar Object Reference
    PosRef:Widget Object Reference=@n26
    NewParam:Vector 2D Structure=GetAbsolutePosition!(Geometry:"Geometry Structure (by ref)"=GetCachedGeometry!(self:Widget Object Reference=@n26) -> (ReturnValue:"Geometry Structure (by ref)")) -> (ReturnValue:Vector 2D Structure)
   ) -> (then:Exec) {is=title K2Node_CallDelegate}
   Call On Deleted!(execute:Exec, self:EWBP Sparks Sequence Pool Timeline Event Bar Object Reference) -> (then:Exec) {is=title K2Node_CallDelegate}
  ]}
 GetBarData {nodes [
   GetBarData() {is=title K2Node_FunctionEntry}
   > Return Node!(
    execute
    StepIndex:int=Get StepIndex() -> (StepIndex=0)
    EventIndex:int=Get EventIndex() -> (EventIndex=0)
    SequenceReference:Mojo Fx Manager Sequence Data Asset Object Reference=Get SequenceReference()
    EventReference:Stage Show Event Object Reference=Get EventReference()
   )
  ]}
 HideName {nodes [
   n1 = HideName!() -> (then, InVisibility:ESlateVisibility Enum=Visible) {is=title K2Node_FunctionEntry}
   > SetVisibility(self=[Get DownArrowSizeBox() -> (DownArrowSizeBox:Size Box Object Reference), Get EventTextNameSizeBox()], InVisibility=@n1)
   > SetVisibility(self=Get Icon(), InVisibility=Select(Option 0=Collapsed, Option 1=@n1, Index=Get ShowIcon()))
  ]}
 MarkAsWarning {nodes [
   n1 = MarkAsWarning!() -> (then, Value:bool=~, WarningContext:Text="INVTEXT(\"step\")") {is=title K2Node_FunctionEntry}
   n2 = Get EventReference()
   > Branch(Condition=NOT Boolean(A=Get LoopingEvent!(self:Stage Show Event Object Reference=@n2) -> (LoopingEvent:bool)))
   > SetVisibility(self=Get Icon(), InVisibility=Select(Option 0=Collapsed, Option 1=Visible, Index=@n1))
   > SetColorAndOpacity(self:Image Object Reference=Get Icon(), InColorAndOpacity:Linear Color Structure=(R=1,G=0.9014,B=0,A=1))
   > SetToolTipText(
    self=Get Icon()
    InToolTipText=Format Text!(Format="INVTEXT(\"Duplicated entry found please review this {context} to avoid possible show issues.\")", context:Text=@n1.WarningContext) -> (Result)
   )
  ]}
 OnDeleted {nodes [OnDeleted() {is=title K2Node_FunctionEntry}]}
 OpenSubmenu {nodes [OpenSubmenu!() -> (then, PosRef:Widget Object Reference, NewParam:Vector 2D Structure) {is=title K2Node_FunctionEntry}]}
 UpdateBarData {nodes [
   n1 = UpdateBarData!() -> (
    then
    StepIndex:int
    EventIndex:int
    SequenceReference:"Mojo Fx Manager Sequence Data Asset Object Reference (by ref)"
    EventReference:"Mojo Fx Manager Event Data Asset Object Reference (by ref)"
    IsHeader:bool=~
   ) {is=title K2Node_FunctionEntry}
   > n2 = Sequence!(execute) -> (then_0, then_1, then_2:Exec, then_3:Exec, then_4:Exec)
   > Set StepIndex(StepIndex:int=@n1) -> (Output_Get:int=0)
   Set EventIndex(execute=@n2.then_1, EventIndex:int=@n1.EventIndex) -> (Output_Get:int=0)
   Set SequenceReference(execute=@n2.then_2, SequenceReference:Mojo Fx Manager Sequence Data Asset Object Reference=@n1.SequenceReference) -> (Output_Get:Mojo Fx Manager Sequence Data Asset Object Reference)
   Set EventReference(execute=@n2.then_3, EventReference:Stage Show Event Object Reference=@n1.EventReference) -> (Output_Get:Stage Show Event Object Reference)
   UpdateSubmenuData(
    execute=@n2.then_4
    self=Get EWBP_SparksTimelineSubmenu()
    SequenceReference=Get SequenceReference()
    EventReference=Get EventReference()
    SparksTimelineEditor=Get SparksTimelineEditor()
    SequenceIndex=0
    StepIndex=Get StepIndex() -> (StepIndex=0)
    EventIndex=Get EventIndex() -> (EventIndex=0)
    IsHeader=@n1.IsHeader
   )
  ]}
 UpdateCellName {nodes [
   n1 = UpdateCellName!() -> (then, NewName:Text) {is=title K2Node_FunctionEntry}
   > n2 = Set EventName(EventName:Text=@n1) -> (Output_Get:Text)
   > "SetText (Text)"(self=Get EventBarName(), InText=@n2)
   > SetToolTipText(self=[Get EventBarName(), Get EditorUtilityButton_35()], InToolTipText=Get EventName())
  ]}
 UpdateSubmenu {nodes [
   n1 = UpdateSubmenu!() -> (then, isEnable:bool=~) {is=title K2Node_FunctionEntry}
   > n2 = Branch(Condition=@n1)
   > n3 = Set IsSubmenuOpen(IsSubmenuOpen=NOT Boolean(A=Get IsSubmenuOpen() -> (IsSubmenuOpen:bool)))
   > SetVisibility(self=Get EWBP_SparksTimelineSubmenu(), InVisibility=Select(Option 0=Collapsed, Option 1=Visible, Index=@n3))
   Set IsSubmenuOpen(execute=@n2.else, IsSubmenuOpen=false)
   > SetVisibility(self=Get EWBP_SparksTimelineSubmenu(), InVisibility=Collapsed)
  ]}
]
variables [
 BarColor {type=LinearColor, default=<rgba 0.7305 0.7084 1 1>}
 DelayValue {type=number}
 EventIndex {type=integer}
 EventName {type=string, default=Event Name}
 EventReference {type=StageShowEvent, default=null}
 EventSize {type=number, default=1}
 IsSubmenuOpen {type=boolean}
 OnDeleted {default=()}
 OpenSubmenu {default=()}
 SequenceReference {type=MojoFxManagerSequence_DataAsset_C, default=null}
 ShowIcon {type=boolean}
 SideIcon {
  type=Texture2D
  default=/PilgrimQuickplay/StageManager/NeverCook/EditorUtilities/StageManagerEditor/Textures/T_SparksEditorUI_IconLibrary_Infinity
 }
 SparksTimelineEditor {type=SparksStageManagerTimelineEditor_C, default=null}
 StepIndex {type=integer}
]
widgets {
 info {rootWidgetClass=&Overlay, widgetCount=26}
 widgets [
  Overlay_1 {widgetClassPath=&Overlay, widget=Overlay_1}
  HorizontalBox_0 {
   widgetClassPath=&HorizontalBox
   parent=Overlay_1
   widget=HorizontalBox_0
   slot {class=OverlaySlot, properties {horizontalAlignment=HAlign_Fill}}
  }
  DelayBar {
   widgetClassPath=&SizeBox
   bIsVariable=true
   parent=HorizontalBox_0
   widget=DelayBar
   properties {bOverride_WidthOverride=true, maxDesiredWidth=110.6064, minDesiredWidth=100}
   slot {class=HorizontalBoxSlot}
  }
  ButtonBorder_1 {
   widgetClassPath=&Border
   bIsVariable=true
   parent=DelayBar
   widget=ButtonBorder_1
   properties {
    background {
     bIsDynamicallyLoaded=false
     drawAs=NoDrawType
     imageType=NoImage
     mirroring=NoMirror
     resourceName=None
     resourceObject=null
     tiling=NoTile
     imageSize=<32 32>
     margin=<ltrb 0 0 0 0>
     outlineSettings {
      bUseBrushTransparency=false
      roundingType=HalfHeightRadius
      width=0
      color {colorUseRule=UseColor_Specified, specifiedColor=<rgba 0 0 0 0>}
      cornerRadii=<0 0 0 0>
     }
     tintColor {colorUseRule=UseColor_Specified, specifiedColor=<rgba 1 1 1 1>}
     uVRegion {bIsValid=false, max=<0 0>, min=<0 0>}
    }
    contentColorAndOpacity=<rgba 1 0.6362 0.1389 1>
    padding=<ltrb 0 0 1 1>
   }
   slot {class=SizeBoxSlot}
  }
  Overlay {widgetClassPath=&Overlay, parent=ButtonBorder_1, widget=Overlay, slot {class=BorderSlot, properties {padding=<ltrb 0 0 1 1>}}}
  EditorUtilityButton {
   widgetClassPath=&EditorUtilityButton
   parent=Overlay
   widget=EditorUtilityButton
   properties {
    clipping=ClipToBounds
    widgetStyle {
     clickedSlateSound {resourceObject=null}
     disabled {
      bIsDynamicallyLoaded=false
      drawAs=Box
      imageType=NoImage
      mirroring=NoMirror
      resourceName=None
      resourceObject=null
      tiling=NoTile
      imageSize=<0 0>
      margin=<ltrb 0 0 0 0>
      outlineSettings {
       bUseBrushTransparency=false
       roundingType=FixedRadius
       width=0
       color {colorUseRule=UseColor_ColorTable, specifiedColor=<rgba 0 0 0 0>}
       cornerRadii=<4 4 4 4>
      }
      tintColor {colorUseRule=UseColor_Specified, specifiedColor=<rgba 0.274 0.343 0.401 1>}
      uVRegion {bIsValid=false, max=<0 0>, min=<0 0>}
     }
     disabledForeground {colorUseRule=UseColor_ColorTable, specifiedColor=<rgba 0 0 0 0>}
     hovered {
      bIsDynamicallyLoaded=false
      drawAs=RoundedBox
      imageType=NoImage
      mirroring=NoMirror
      resourceName=None
      resourceObject=null
      tiling=NoTile
      imageSize=<32 32>
      margin=<ltrb 0 0 0 0>
      outlineSettings {
       bUseBrushTransparency=false
       roundingType=FixedRadius
       width=0
       color {colorUseRule=UseColor_ColorTable, specifiedColor=<rgba 0 0 0 0>}
       cornerRadii=<4 4 4 4>
      }
      tintColor {colorUseRule=UseColor_Specified, specifiedColor=<rgba 0.5783 0.8542 0.8542 1>}
      uVRegion {bIsValid=false, max=<0 0>, min=<0 0>}
     }
     hoveredForeground {colorUseRule=UseColor_ColorTable, specifiedColor=<rgba 0 0 0 0>}
     hoveredSlateSound {resourceObject=null}
     normal {
      bIsDynamicallyLoaded=false
      drawAs=Box
      imageType=NoImage
      mirroring=NoMirror
      resourceName=None
      resourceObject=null
      tiling=NoTile
      imageSize=<32 32>
      margin=<ltrb 0 0 0 0>
      outlineSettings {
       bUseBrushTransparency=false
       roundingType=FixedRadius
       width=0
       color {colorUseRule=UseColor_ColorTable, specifiedColor=<rgba 0 0 0 0>}
       cornerRadii=<4 4 4 4>
      }
      tintColor {colorUseRule=UseColor_Specified, specifiedColor=<rgba 0.4397 0.4397 0.4397 1>}
      uVRegion {bIsValid=false, max=<0 0>, min=<0 0>}
     }
     normalForeground {colorUseRule=UseColor_ColorTable, specifiedColor=<rgba 0 0 0 0>}
     normalPadding=<ltrb 5 1.5 5 1.5>
     pressed {
      bIsDynamicallyLoaded=false
      drawAs=RoundedBox
      imageType=NoImage
      mirroring=NoMirror
      resourceName=None
      resourceObject=null
      tiling=NoTile
      imageSize=<32 32>
      margin=<ltrb 0 0 0 0>
      outlineSettings {
       bUseBrushTransparency=false
       roundingType=FixedRadius
       width=0
       color {colorUseRule=UseColor_ColorTable, specifiedColor=<rgba 0 0 0 0>}
       cornerRadii=<4 4 4 4>
      }
      tintColor {colorUseRule=UseColor_Specified, specifiedColor=<rgba 0.0284 0.0284 0.0284 1>}
      uVRegion {bIsValid=false, max=<0 0>, min=<0 0>}
     }
     pressedForeground {colorUseRule=UseColor_ColorTable, specifiedColor=<rgba 0 0 0 0>}
     pressedPadding=<ltrb 6 2.5 6 0.5>
     pressedSlateSound {resourceObject=null}
    }
   }
   slot {class=OverlaySlot, properties {horizontalAlignment=HAlign_Fill, verticalAlignment=VAlign_Fill}}
  }
  HorizontalBox {
   widgetClassPath=&HorizontalBox
   parent=EditorUtilityButton
   widget=HorizontalBox
   slot {class=ButtonSlot, properties {horizontalAlignment=HAlign_Fill, padding=<ltrb 0 0 0 0>}}
  }
  EventTextNameSizeBox_1 {
   widgetClassPath=&SizeBox
   bIsVariable=true
   parent=HorizontalBox
   widget=EventTextNameSizeBox_1
   properties {bOverride_MaxDesiredWidth=true, bOverride_MinDesiredWidth=true, maxDesiredWidth=120, minDesiredWidth=120}
   slot {class=HorizontalBoxSlot, properties {horizontalAlignment=HAlign_Center, verticalAlignment=VAlign_Center, size {sizeRule=Fill, value=1}}}
  }
  DelayText {
   widgetClassPath=&TextBlock
   bIsVariable=true
   parent=EventTextNameSizeBox_1
   widget=DelayText
   properties {
    clipping=ClipToBounds
    justification=Center
    text=0.1s
    visibility=SelfHitTestInvisible
    font {
     bForceMonospaced=false
     bMaterialIsStencil=false
     fontMaterial=null
     fontObject=&HeadingNow
     letterSpacing=0
     monospacedWidth=1
     size=6.75
     skewAmount=0
     typefaceFontName=75 Medium
     outlineSettings {
      bApplyOutlineToDropShadows=false
      bMiteredCorners=false
      bSeparateFillAlpha=false
      outlineMaterial=null
      outlineSize=0
      outlineColor=<rgba 0 0 0 1>
     }
    }
   }
   slot {class=SizeBoxSlot}
  }
  Degradee_Right_1 {
   widgetClassPath=&Image
   bIsVariable=true
   parent=Overlay
   widget=Degradee_Right_1
   properties {
    visibility=SelfHitTestInvisible
    brush {
     bIsDynamicallyLoaded=false
     drawAs=Image
     imageType=NoImage
     mirroring=NoMirror
     resourceName=None
     resourceObject=/PilgrimQuickplay/StageManager/NeverCook/EditorUtilities/StageManagerEditor/Textures/M_SparksEditorUI_SimpleFade
     tiling=NoTile
     imageSize=<8 1>
     margin=<ltrb 0 0 0 0>
     outlineSettings {
      bUseBrushTransparency=false
      roundingType=HalfHeightRadius
      width=0
      color {colorUseRule=UseColor_Specified, specifiedColor=<rgba 0 0 0 0>}
      cornerRadii=<0 0 0 0>
     }
     tintColor {colorUseRule=UseColor_Specified, specifiedColor=<rgba 0.4397 0.4397 0.4397 1>}
     uVRegion {bIsValid=false, max=<0 0>, min=<0 0>}
    }
   }
   slot {class=OverlaySlot, properties {horizontalAlignment=HAlign_Right, verticalAlignment=VAlign_Fill}}
  }
  Degradee_Left_1 {
   widgetClassPath=&Image
   bIsVariable=true
   parent=Overlay
   widget=Degradee_Left_1
   properties {
    visibility=SelfHitTestInvisible
    brush {
     bIsDynamicallyLoaded=false
     drawAs=Image
     imageType=NoImage
     mirroring=NoMirror
     resourceName=None
     resourceObject=&M_SparksEditorUI_SimpleFade
     tiling=NoTile
     imageSize=<6 1>
     margin=<ltrb 0 0 0 0>
     outlineSettings {
      bUseBrushTransparency=false
      roundingType=HalfHeightRadius
      width=0
      color {colorUseRule=UseColor_Specified, specifiedColor=<rgba 0 0 0 0>}
      cornerRadii=<0 0 0 0>
     }
     tintColor {colorUseRule=UseColor_Specified, specifiedColor=<rgba 0.4397 0.4397 0.4397 1>}
     uVRegion {bIsValid=false, max=<0 0>, min=<0 0>}
    }
    renderTransform {angle=0, scale=<-1 1>, shear=<0 0>, translation=<0 0>}
   }
   slot {class=OverlaySlot, properties {verticalAlignment=VAlign_Fill}}
  }
  ButtonSizeBox {
   widgetClassPath=&SizeBox
   bIsVariable=true
   parent=HorizontalBox_0
   widget=ButtonSizeBox
   properties {bOverride_MinDesiredWidth=true, maxDesiredWidth=110.6064, minDesiredWidth=100}
   slot {class=HorizontalBoxSlot, properties {size {sizeRule=Fill, value=1}}}
  }
  ButtonBorder {
   widgetClassPath=&Border
   bIsVariable=true
   parent=ButtonSizeBox
   widget=ButtonBorder
   properties {
    background {
     bIsDynamicallyLoaded=false
     drawAs=NoDrawType
     imageType=NoImage
     mirroring=NoMirror
     resourceName=None
     resourceObject=null
     tiling=NoTile
     imageSize=<32 32>
     margin=<ltrb 0 0 0 0>
     outlineSettings {
      bUseBrushTransparency=false
      roundingType=HalfHeightRadius
      width=0
      color {colorUseRule=UseColor_Specified, specifiedColor=<rgba 0 0 0 0>}
      cornerRadii=<0 0 0 0>
     }
     tintColor {colorUseRule=UseColor_Specified, specifiedColor=<rgba 1 1 1 1>}
     uVRegion {bIsValid=false, max=<0 0>, min=<0 0>}
    }
    contentColorAndOpacity=<rgba 0.7305 0.7115 1 1>
    padding=<ltrb 0 0 1 1>
   }
   slot {class=SizeBoxSlot}
  }
  Overlay_0 {widgetClassPath=&Overlay, parent=ButtonBorder, widget=Overlay_0, slot {class=BorderSlot, properties {padding=<ltrb 0 0 1 1>}}}
  EditorUtilityButton_35 {
   widgetClassPath=&EditorUtilityButton
   parent=Overlay_0
   widget=EditorUtilityButton_35
   properties {
    clipping=ClipToBounds
    widgetStyle {
     clickedSlateSound {resourceObject=null}
     disabled {
      bIsDynamicallyLoaded=false
      drawAs=Box
      imageType=NoImage
      mirroring=NoMirror
      resourceName=None
      resourceObject=null
      tiling=NoTile
      imageSize=<0 0>
      margin=<ltrb 0 0 0 0>
      outlineSettings {
       bUseBrushTransparency=false
       roundingType=FixedRadius
       width=0
       color {colorUseRule=UseColor_ColorTable, specifiedColor=<rgba 0 0 0 0>}
       cornerRadii=<4 4 4 4>
      }
      tintColor {colorUseRule=UseColor_Specified, specifiedColor=<rgba 0.274 0.343 0.401 1>}
      uVRegion {bIsValid=false, max=<0 0>, min=<0 0>}
     }
     disabledForeground {colorUseRule=UseColor_ColorTable, specifiedColor=<rgba 0 0 0 0>}
     hovered {
      bIsDynamicallyLoaded=false
      drawAs=RoundedBox
      imageType=NoImage
      mirroring=NoMirror
      resourceName=None
      resourceObject=null
      tiling=NoTile
      imageSize=<32 32>
      margin=<ltrb 0 0 0 0>
      outlineSettings {
       bUseBrushTransparency=false
       roundingType=FixedRadius
       width=0
       color {colorUseRule=UseColor_ColorTable, specifiedColor=<rgba 0 0 0 0>}
       cornerRadii=<4 4 4 4>
      }
      tintColor {colorUseRule=UseColor_Specified, specifiedColor=<rgba 0.5783 0.8542 0.8542 1>}
      uVRegion {bIsValid=false, max=<0 0>, min=<0 0>}
     }
     hoveredForeground {colorUseRule=UseColor_ColorTable, specifiedColor=<rgba 0 0 0 0>}
     hoveredSlateSound {resourceObject=null}
     normal {
      bIsDynamicallyLoaded=false
      drawAs=Box
      imageType=NoImage
      mirroring=NoMirror
      resourceName=None
      resourceObject=null
      tiling=NoTile
      imageSize=<32 32>
      margin=<ltrb 0 0 0 0>
      outlineSettings {
       bUseBrushTransparency=false
       roundingType=FixedRadius
       width=0
       color {colorUseRule=UseColor_ColorTable, specifiedColor=<rgba 0 0 0 0>}
       cornerRadii=<4 4 4 4>
      }
      tintColor {colorUseRule=UseColor_Specified, specifiedColor=<rgba 0.4397 0.4397 0.4397 1>}
      uVRegion {bIsValid=false, max=<0 0>, min=<0 0>}
     }
     normalForeground {colorUseRule=UseColor_ColorTable, specifiedColor=<rgba 0 0 0 0>}
     normalPadding=<ltrb 5 1.5 5 1.5>
     pressed {
      bIsDynamicallyLoaded=false
      drawAs=RoundedBox
      imageType=NoImage
      mirroring=NoMirror
      resourceName=None
      resourceObject=null
      tiling=NoTile
      imageSize=<32 32>
      margin=<ltrb 0 0 0 0>
      outlineSettings {
       bUseBrushTransparency=false
       roundingType=FixedRadius
       width=0
       color {colorUseRule=UseColor_ColorTable, specifiedColor=<rgba 0 0 0 0>}
       cornerRadii=<4 4 4 4>
      }
      tintColor {colorUseRule=UseColor_Specified, specifiedColor=<rgba 0.0284 0.0284 0.0284 1>}
      uVRegion {bIsValid=false, max=<0 0>, min=<0 0>}
     }
     pressedForeground {colorUseRule=UseColor_ColorTable, specifiedColor=<rgba 0 0 0 0>}
     pressedPadding=<ltrb 6 2.5 6 0.5>
     pressedSlateSound {resourceObject=null}
    }
   }
   slot {class=OverlaySlot, properties {horizontalAlignment=HAlign_Fill, verticalAlignment=VAlign_Fill}}
  }
  HorizontalBox_1 {
   widgetClassPath=&HorizontalBox
   parent=EditorUtilityButton_35
   widget=HorizontalBox_1
   slot {class=ButtonSlot, properties {horizontalAlignment=HAlign_Fill, padding=<ltrb 0 0 0 0>}}
  }
  Icon {
   widgetClassPath=&Image
   bIsVariable=true
   parent=HorizontalBox_1
   widget=Icon
   properties {
    visibility=Collapsed
    brush {
     bIsDynamicallyLoaded=false
     drawAs=Image
     imageType=NoImage
     mirroring=NoMirror
     resourceName=None
     resourceObject=&T_SparksEditorUI_IconLibrary_Infinity
     tiling=NoTile
     imageSize=<15 15>
     margin=<ltrb 0 0 0 0>
     outlineSettings {
      bUseBrushTransparency=false
      roundingType=HalfHeightRadius
      width=0
      color {colorUseRule=UseColor_Specified, specifiedColor=<rgba 0 0 0 0>}
      cornerRadii=<0 0 0 0>
     }
     tintColor {colorUseRule=UseColor_Specified, specifiedColor=<rgba 0.7916 1 0.6832 1>}
     uVRegion {bIsValid=false, max=<0 0>, min=<0 0>}
    }
   }
   slot {class=HorizontalBoxSlot, properties {horizontalAlignment=HAlign_Left, verticalAlignment=VAlign_Center, padding=<ltrb 0 0 2 0>}}
  }
  EventTextNameSizeBox {
   widgetClassPath=&SizeBox
   bIsVariable=true
   parent=HorizontalBox_1
   widget=EventTextNameSizeBox
   properties {bOverride_MaxDesiredWidth=true, bOverride_MinDesiredWidth=true, maxDesiredWidth=120, minDesiredWidth=120}
   slot {class=HorizontalBoxSlot, properties {horizontalAlignment=HAlign_Left, verticalAlignment=VAlign_Center, size {sizeRule=Fill, value=1}}}
  }
  EventBarName {
   widgetClassPath=&TextBlock
   bIsVariable=true
   parent=EventTextNameSizeBox
   widget=EventBarName
   properties {
    clipping=ClipToBounds
    text=Event Name
    visibility=SelfHitTestInvisible
    font {
     bForceMonospaced=false
     bMaterialIsStencil=false
     fontMaterial=null
     fontObject=&HeadingNow
     letterSpacing=0
     monospacedWidth=1
     size=8.25
     skewAmount=0
     typefaceFontName=75 Medium
     outlineSettings {
      bApplyOutlineToDropShadows=false
      bMiteredCorners=false
      bSeparateFillAlpha=false
      outlineMaterial=null
      outlineSize=0
      outlineColor=<rgba 0 0 0 1>
     }
    }
   }
   slot {class=SizeBoxSlot}
  }
  DownArrowSizeBox {
   widgetClassPath=&SizeBox
   bIsVariable=true
   parent=HorizontalBox_1
   widget=DownArrowSizeBox
   properties {bOverride_MaxDesiredWidth=true, maxDesiredHeight=18.3163, maxDesiredWidth=19.3278, visibility=HitTestInvisible}
   slot {class=HorizontalBoxSlot, properties {horizontalAlignment=HAlign_Left, verticalAlignment=VAlign_Center}}
  }
  Overlay_3 {widgetClassPath=&Overlay, parent=DownArrowSizeBox, widget=Overlay_3, slot {class=SizeBoxSlot}}
  ArrowDown {
   widgetClassPath=&Image
   bIsVariable=true
   parent=Overlay_3
   widget=ArrowDown
   properties {
    visibility=SelfHitTestInvisible
    brush {
     bIsDynamicallyLoaded=false
     drawAs=Image
     imageType=NoImage
     mirroring=NoMirror
     resourceName=None
     resourceObject=/PilgrimQuickplay/StageManager/NeverCook/EditorUtilities/StageManagerEditor/Textures/T_EditorDownArrow
     tiling=NoTile
     imageSize=<15 15>
     margin=<ltrb 0 0 0 0>
     outlineSettings {
      bUseBrushTransparency=false
      roundingType=HalfHeightRadius
      width=0
      color {colorUseRule=UseColor_Specified, specifiedColor=<rgba 0 0 0 0>}
      cornerRadii=<0 0 0 0>
     }
     tintColor {colorUseRule=UseColor_Specified, specifiedColor=<rgba 1 1 1 1>}
     uVRegion {bIsValid=false, max=<0 0>, min=<0 0>}
    }
   }
   slot {class=OverlaySlot}
  }
  ArrowUp {
   widgetClassPath=&Image
   bIsVariable=true
   parent=Overlay_3
   widget=ArrowUp
   properties {
    visibility=Collapsed
    brush {
     bIsDynamicallyLoaded=false
     drawAs=Image
     imageType=NoImage
     mirroring=NoMirror
     resourceName=None
     resourceObject=&T_EditorDownArrow
     tiling=NoTile
     imageSize=<15 15>
     margin=<ltrb 0 0 0 0>
     outlineSettings {
      bUseBrushTransparency=false
      roundingType=HalfHeightRadius
      width=0
      color {colorUseRule=UseColor_Specified, specifiedColor=<rgba 0 0 0 0>}
      cornerRadii=<0 0 0 0>
     }
     tintColor {colorUseRule=UseColor_Specified, specifiedColor=<rgba 1 1 1 1>}
     uVRegion {bIsValid=false, max=<0 0>, min=<0 0>}
    }
    renderTransform {angle=0, scale=<1 -1>, shear=<0 0>, translation=<0 0>}
   }
   slot {class=OverlaySlot}
  }
  Degradee_Right {
   widgetClassPath=&Image
   bIsVariable=true
   parent=Overlay_0
   widget=Degradee_Right
   properties {
    visibility=SelfHitTestInvisible
    brush {
     bIsDynamicallyLoaded=false
     drawAs=Image
     imageType=NoImage
     mirroring=NoMirror
     resourceName=None
     resourceObject=&M_SparksEditorUI_SimpleFade
     tiling=NoTile
     imageSize=<8 1>
     margin=<ltrb 0 0 0 0>
     outlineSettings {
      bUseBrushTransparency=false
      roundingType=HalfHeightRadius
      width=0
      color {colorUseRule=UseColor_Specified, specifiedColor=<rgba 0 0 0 0>}
      cornerRadii=<0 0 0 0>
     }
     tintColor {colorUseRule=UseColor_Specified, specifiedColor=<rgba 0.4397 0.4397 0.4397 1>}
     uVRegion {bIsValid=false, max=<0 0>, min=<0 0>}
    }
   }
   slot {class=OverlaySlot, properties {horizontalAlignment=HAlign_Right, verticalAlignment=VAlign_Fill}}
  }
  Degradee_Left {
   widgetClassPath=&Image
   bIsVariable=true
   parent=Overlay_0
   widget=Degradee_Left
   properties {
    visibility=SelfHitTestInvisible
    brush {
     bIsDynamicallyLoaded=false
     drawAs=Image
     imageType=NoImage
     mirroring=NoMirror
     resourceName=None
     resourceObject=&M_SparksEditorUI_SimpleFade
     tiling=NoTile
     imageSize=<6 1>
     margin=<ltrb 0 0 0 0>
     outlineSettings {
      bUseBrushTransparency=false
      roundingType=HalfHeightRadius
      width=0
      color {colorUseRule=UseColor_Specified, specifiedColor=<rgba 0 0 0 0>}
      cornerRadii=<0 0 0 0>
     }
     tintColor {colorUseRule=UseColor_Specified, specifiedColor=<rgba 0.4397 0.4397 0.4397 1>}
     uVRegion {bIsValid=false, max=<0 0>, min=<0 0>}
    }
    renderTransform {angle=0, scale=<-1 1>, shear=<0 0>, translation=<0 0>}
   }
   slot {class=OverlaySlot, properties {verticalAlignment=VAlign_Fill}}
  }
  EWBP_SparksTimelineSubmenu {
   widgetClassPath=&/EWBP_SparksTimelineSubmenu
   bIsVariable=true
   parent=Overlay_1
   widget=EWBP_SparksTimelineSubmenu
   properties {visibility=Collapsed, renderTransform {angle=0, scale=<1 1>, shear=<0 0>, translation=<0 18>}}
   slot {class=OverlaySlot, properties {horizontalAlignment=HAlign_Fill, padding=<ltrb 0 0 5 -55>}}
  }
 ]
}
