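# Editor utility widget showing one step of a Mojo Fx Manager sequence in the Sparks Stage Manager timeline editor.
# It rebuilds event bars from the step's events into regular and looping columns, then flags duplicated events with a warning.
# New event buttons open the editor's add window and append the chosen asset to the step via temporarily bound delegates.
# The last step also shows a button that appends a new step.
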EditorUtilityWidgetBlueprint /PilgrimQuickplay/StageManager/NeverCook/EditorUtilities/StageManagerEditor/EWBP_SparksSequencePool_TimelineEventGroup
graphs [
 ChangeEventsVisibility {nodes [
   n1 = ChangeEventsVisibility!() -> (then, InVisibility:ESlateVisibility Enum=Visible) {is=title K2Node_FunctionEntry}
   > SetVisibility(
    self=[
     Get NewLoopEventButton() -> (NewLoopEventButton:Editor Utility Button Object Reference)
     Get NewEventButton() -> (NewEventButton:Editor Utility Button Object Reference)
    ]
    InVisibility=@n1
   )
   n2 = Get EventsCellsVerticalBox() -> (EventsCellsVerticalBox:Vertical Box Object Reference)
   > SetVisibility(
    self=Get NoeventsTextEvents() -> (NoeventsTextEvents:Text Object Reference)
    InVisibility=Select(Option 0=@n1, Option 1=Collapsed, Index=HasAnyChildren!(self:Panel Widget Object Reference=@n2) -> (ReturnValue:bool))
   )
   n3 = Get LoopingEventsCellsVerticalBox() -> (LoopingEventsCellsVerticalBox:Vertical Box Object Reference)
   > SetVisibility(
    self=Get NoEventsTextLooping() -> (NoEventsTextLooping:Text Object Reference)
    InVisibility=Select(Option 0=@n1, Option 1=Collapsed, Index=HasAnyChildren(self=@n3))
   )
   > n4 = For Each Loop(Array:Array of Widget Object References=GetAllChildren(self=Get EventsCellsVerticalBox())) -> (Array Element:Widget Object Reference)
   n5 = For Each Loop(Exec=@n4.Completed, Array=GetAllChildren(self=Get LoopingEventsCellsVerticalBox()))
   > n6 = Cast To EWBP_SparksSequencePool_TimelineEventBar!(execute, Object=@n5) -> (then, CastFailed, AsEWBP Sparks Sequence Pool Timeline Event Bar:EWBP Sparks Sequence Pool Timeline Event Bar Object Reference) {is=title K2Node_DynamicCast}
   > HideName!(execute, self:EWBP Sparks Sequence Pool Timeline Event Bar Object Reference=@n6, InVisibility:ESlateVisibility Enum=@n1) -> (then)
   n7 = Cast To EWBP_SparksSequencePool_TimelineEventBar(execute=@n4.LoopBody, Object=@n4)
   > HideName(self=@n7, InVisibility=@n1)
  ]}
 CheckForDuplicated {nodes [
   CheckForDuplicated() {is=title K2Node_FunctionEntry}
   > n1 = For Each Loop(Array:Array of Stage Show Event Object References=Get!() -> (StepData_StepEvents:Array of Stage Show Event Object References)) -> (Array Element:Stage Show Event Object Reference)
   > Clear(TargetArray:Array of EWBP Sparks Sequence Pool Timeline Event Bar Object References=Get DuplicatedEvents() -> (DuplicatedEvents:Array of EWBP Sparks Sequence Pool Timeline Event Bar Object References))
   > n2 = For Each Loop with Break!(
    Exec:Exec
    Array:Array of EWBP Sparks Sequence Pool Timeline Event Bar Object References=Get EventBarMap() -> (EventBarMap:Array of EWBP Sparks Sequence Pool Timeline Event Bar Object References)
    Break:Exec
   ) -> (LoopBody:Exec, Array Element:EWBP Sparks Sequence Pool Timeline Event Bar Object Reference, Array Index:int, Completed:Exec)
   n3 = For Each Loop(Exec=@n2.Completed, Array:Array of EWBP Sparks Sequence Pool Timeline Event Bar Object References=Get DuplicatedEvents()) -> (Array Element:EWBP Sparks Sequence Pool Timeline Event Bar Object Reference)
   Set Amount(execute=@n3.Completed, Amount:int=0) -> (Output_Get:int=0)
   n4 = "integer > integer"(A=Get Amount() -> (Amount:int=0), B=1)
   Branch(execute=@n3.LoopBody, Condition=@n4)
   > MarkAsWarning!(execute, self:EWBP Sparks Sequence Pool Timeline Event Bar Object Reference=@n3, Value:bool=@n4, WarningContext:Text="INVTEXT(\"step\")") -> (then)
   CheckForDuplicatedManagers(execute=@n2.LoopBody, self:EWBP Sparks Sequence Pool Timeline Event Bar Object Reference=@n2)
   > n5 = Branch(Condition="Equal (Object)"(
     A=Get EventReference!(self:EWBP Sparks Sequence Pool Timeline Event Bar Object Reference=@n2) -> (EventReference:Stage Show Event Object Reference)
     B=@n1
    ))
   > Increment Int(Value=Get Amount() -> (Amount=0))
   > Add(
    TargetArray:Array of EWBP Sparks Sequence Pool Timeline Event Bar Object References=Get DuplicatedEvents()
    NewItem:"EWBP Sparks Sequence Pool Timeline Event Bar Object Reference (by ref)"=@n2
   ) {is=title K2Node_CallArrayFunction}
   Set IsDuplicated(execute=@n5.else, IsDuplicated:bool="integer > integer"(A=Get Amount() -> (Amount=0), B=1)) -> (Output_Get:bool)
   Get IsDuplicated() -> (IsDuplicated:bool)
  ]}
 EventGraph {nodes [
   "On Pressed (NewStepButton)"() {is=title K2Node_ComponentBoundEvent}
   > Call On Pressed!(execute:Exec, self:EWBP Sparks Sequence Pool Timeline Event Group Object Reference) -> (then:Exec) {is=title K2Node_CallDelegate}
   n1 = Get SequenceReference() -> (SequenceReference:Mojo Fx Manager Sequence Data Asset Object Reference)
   > Add(
    TargetArray:Array of Stage Manager Sequence Steps Structures=Get SequenceSteps!(self:Stage Show Sequence Object Reference=@n1) -> (SequenceSteps:Array of Stage Manager Sequence Steps Structures)
    NewItem:"Stage Manager Sequence Steps Structure (by ref)"
   )
   > TransactObject(Object=@n1)
   > ForceRefresh(self:Sparks Stage Manager Timeline Editor Object Reference=Get SparksTimelineEditor() -> (SparksTimelineEditor:Sparks Stage Manager Timeline Editor Object Reference))
   "On Pressed (EditorUtilityButton_174)"() {is=title K2Node_ComponentBoundEvent}
   > UpdateToggles!(
    execute
    self:Sparks Stage Manager Timeline Editor Object Reference=Get SparksTimelineEditor()
    Toggle:EWBP Sparks Sequence Pool Timeline Event Bar Object Reference
    ForceCloseAll:bool=true
    DontCallEvent:bool=false
   ) -> (then)
   "On Pressed (NewEventButton)"() {is=title K2Node_ComponentBoundEvent}
   > UpdateToggles(self=Get SparksTimelineEditor(), ForceCloseAll=true, DontCallEvent=false)
   > n2 = OpenAddChangeWindow!(
    execute
    self:Sparks Stage Manager Timeline Editor Object Reference=Get SparksTimelineEditor()
    WindowType:Name=FxManager_Event StepData
    IsJustCreate:bool=true
    HideSelectSection:bool=false
    IsLoopingEvent:bool=false
    KeepSelectChange:bool=false
   ) -> (then)
   n3 = Create Event()
   n4 = Bind Event to On Asset Changed(execute=[@n2.then, @n6.then], self:Sparks Stage Manager Timeline Editor Object Reference=Get SparksTimelineEditor(), Delegate=@n3) {is=title K2Node_AddDelegate}
   n5 = OnPopupClose() {is=title K2Node_CustomEvent}
   Bind Event to On Popup Closed(execute=@n4.then, self:Sparks Stage Manager Timeline Editor Object Reference=Get SparksTimelineEditor(), Delegate=@n5) {is=title K2Node_AddDelegate}
   "On Pressed (NewLoopEventButton)"() {is=title K2Node_ComponentBoundEvent}
   > UpdateToggles(self=Get SparksTimelineEditor(), ForceCloseAll=true, DontCallEvent=false)
   > n6 = OpenAddChangeWindow(
    self=Get SparksTimelineEditor()
    WindowType=FxManager_Event StepData
    IsJustCreate=true
    HideSelectSection=false
    IsLoopingEvent=true
    KeepSelectChange=false
   )
   n7 = OnAssetUpdate!() -> (OutputDelegate, then, ReturnObj:Object Reference, IsNewFile:bool=~) {is=title K2Node_CustomEvent}
   > n8 = Cast To MojoFxManagerEvent_DataAsset!(execute, Object=@n7.ReturnObj) -> (then, CastFailed, AsMojo Fx Manager Event Data Asset:Mojo Fx Manager Event Data Asset Object Reference) {is=title K2Node_DynamicCast}
   n9 = Get StepIndex() -> (StepIndex:int=0)
   n10 = Get SequenceSteps(self=Get SequenceReference())
   > Add(
    TargetArray:Array of Stage Show Event Object References=Break Stage Manager Sequence Steps!(StageManagerSequenceSteps:"Stage Manager Sequence Steps Structure (by ref)"="Get (a copy)"(Array:Array of Stage Manager Sequence Steps Structures=@n10, Dimension 1=@n9) -> (Output:Stage Manager Sequence Steps Structure)) -> (StepEvents:Array of Stage Show Event Object References) {is=title K2Node_BreakStruct}
    NewItem:"Stage Show Event Object Reference (by ref)"=@n8
   )
   > TransactObject(Object=Get SequenceReference())
   > n11 = ForceRefresh(self=Get SparksTimelineEditor())
   Unbind Event from On Asset Changed(
    execute=[@n11.then, @n5.then, @n8.CastFailed]
    self:Sparks Stage Manager Timeline Editor Object Reference=Get SparksTimelineEditor()
    Delegate=@n3
   ) {is=title K2Node_RemoveDelegate}
   > Unbind Event from On Popup Closed(self:Sparks Stage Manager Timeline Editor Object Reference=Get SparksTimelineEditor(), Delegate=@n5) {is=title K2Node_RemoveDelegate}
   n12 = Cast To EWBP_SparksSequencePool_TimelineEventBar(Object=GetChildAt!(self:Panel Widget Object Reference=Get EventsCellsVerticalBox(), Index:int=0) -> (ReturnValue:Widget Object Reference))
   > Set EventSize!(execute, EventSize:float=Get EventSize() -> (EventSize:float), self:EWBP Sparks Sequence Pool Timeline Event Bar Object Reference=@n12) -> (then, Output_Get:float)
   > Set DelayValue!(execute, DelayValue:float=Get DelayValue() -> (DelayValue:float), self:EWBP Sparks Sequence Pool Timeline Event Bar Object Reference=@n12) -> (then, Output_Get:float)
   > UpdateBar(self:EWBP Sparks Sequence Pool Timeline Event Bar Object Reference=@n12)
   Event PreConstruct()
   n13 = Get StepCell() -> (StepCell:EWBP Sparks Sequence Pool Timeline Event Bar Object Reference)
   n14 = Get SparksTimelineEditor()
   > Set SparksTimelineEditor!(
    execute
    SparksTimelineEditor:Sparks Stage Manager Timeline Editor Object Reference=@n14
    self:EWBP Sparks Sequence Pool Timeline Event Bar Object Reference=@n13
   ) -> (then, Output_Get:Sparks Stage Manager Timeline Editor Object Reference)
   > Add Unique(
    TargetArray:Array of EWBP Sparks Sequence Pool Timeline Event Bar Object References=Get AllEvents!(self:Sparks Stage Manager Timeline Editor Object Reference=@n14) -> (AllEvents:Array of EWBP Sparks Sequence Pool Timeline Event Bar Object References)
    NewItem:"EWBP Sparks Sequence Pool Timeline Event Bar Object Reference (by ref)"=@n13
   )
   > SetVisibility(
    self=Get NewStepArea() -> (NewStepArea:Overlay Object Reference)
    InVisibility=Select(Option 0=Collapsed, Option 1=SelfHitTestInvisible, Index=Get LastElement() -> (LastElement:bool))
   )
   > FillWithEvents()
   > CheckForDuplicated() {is=title K2Node_CallFunction}
   Event Construct()
   > ChangeButtonsVisibility!(
    execute
    self:EWBP Sparks Timeline Submenu Object Reference=Get EWBP_SparksTimelineSubmenu!(self:EWBP Sparks Sequence Pool Timeline Event Bar Object Reference=Get StepCell()) -> (EWBP_SparksTimelineSubmenu:EWBP Sparks Timeline Submenu Object Reference)
    ChangeButton:ESlateVisibility Enum=Collapsed
    EditButton:ESlateVisibility Enum=Collapsed
    RemoveButton:ESlateVisibility Enum=Visible
   ) -> (then)
   > UpdateBarData!(
    execute
    self:EWBP Sparks Sequence Pool Timeline Event Bar Object Reference=Get StepCell()
    StepIndex:int=Get StepIndex() -> (StepIndex=0)
    EventIndex:int=0
    SequenceReference:"Mojo Fx Manager Sequence Data Asset Object Reference (by ref)"=Get SequenceReference()
    EventReference:"Mojo Fx Manager Event Data Asset Object Reference (by ref)"
    IsHeader:bool=true
   ) -> (then)
  ]}
 FillWithEvents {nodes [
   FillWithEvents() {is=title K2Node_FunctionEntry}
   > Clear(TargetArray=Get EventBarMap())
   > UpdateCellName!(
    execute
    self:EWBP Sparks Sequence Pool Timeline Event Bar Object Reference=Get StepCell()
    NewName:Text=Format Text!(Format="INVTEXT(\"Step {index}\")", index:int=Get StepIndex() -> (StepIndex=0)) -> (Result)
   ) -> (then)
   > ClearChildren(self=[Get LoopingEventsCellsVerticalBox(), Get EventsCellsVerticalBox()])
   > n1 = For Each Loop(Array:Array of Stage Show Event Object References=Get()) -> (Array Element:Stage Show Event Object Reference)
   n2 = Sequence(execute=@n1.Completed)
   n3 = HasAnyChildren(self=Get EventsCellsVerticalBox())
   > SetVisibility(self=Get EventsCellsVerticalBox(), InVisibility=Select(Option 0=Collapsed, Option 1=SelfHitTestInvisible, Index=@n3))
   > SetVisibility(self=Get NoeventsTextEvents(), InVisibility=Select(Option 0=SelfHitTestInvisible, Option 1=Collapsed, Index=@n3))
   n4 = HasAnyChildren(self=Get LoopingEventsCellsVerticalBox())
   SetVisibility(
    execute=@n2.then_1
    self=Get LoopingEventsCellsVerticalBox()
    InVisibility=Select(Option 0=Collapsed, Option 1=SelfHitTestInvisible, Index=@n4)
   )
   > SetVisibility(self=Get NoEventsTextLooping(), InVisibility=Select(Option 0=SelfHitTestInvisible, Option 1=Collapsed, Index=@n4))
   n5 = Branch(execute=@n1.LoopBody, Condition=Get LoopingEvent!(self:Stage Show Event Object Reference=@n1) -> (LoopingEvent:bool))
   n6 = GetDisplayName(Object=@n1)
   n7 = TextIsEmpty(InText="To Text (String)"(InString=@n6))
   n8 = Select(
    Option 0:Text="To Text (String)"(InString=Split(SourceString=@n6, InStr=Pilgrim_StageFX_Event_, SearchDir=FromStart).RightS)
    Option 1:Text="INVTEXT(\"Empty Event Slot\")"
    Index=@n7
   ) -> (ReturnValue:Text)
   n9 = Get StartDelay!(self:Stage Show Event Object Reference=@n1) -> (StartDelay:double)
   > n10 = Create Widget!(
    execute:Exec
    WorldContextObject:Object Reference
    Class:User Widget Class Reference=&/EWBP_SparksSequencePool_TimelineEventBar.EWBP_SparksSequencePool_TimelineEventBar_C
    OwningPlayer:Player Controller Object Reference
    BarColor:Linear Color Structure=(R=0.125,G=1,B=0.209,A=1)
    EventSize:float=0.84
    EventName:Text=@n8
    DelayValue:float=@n9
    SideIcon:Texture 2D Object Reference=/PilgrimQuickplay/StageManager/NeverCook/EditorUtilities/StageManagerEditor/Textures/T_SparksEditorUI_IconLibrary_Infinity
    ShowIcon:bool=true
    EventReference:Stage Show Event Object Reference=@n1
    SparksTimelineEditor:Sparks Stage Manager Timeline Editor Object Reference=Get SparksTimelineEditor()
    EventIndex:int=@n1.Array Index
    StepIndex:int=Get StepIndex() -> (StepIndex=0)
    SequenceReference:Mojo Fx Manager Sequence Data Asset Object Reference=Get SequenceReference()
   ) -> (then:Exec, ReturnValue:EWBP Sparks Sequence Pool Timeline Event Bar Object Reference)
   > Add(
    TargetArray:Array of EWBP Sparks Sequence Pool Timeline Event Bar Object References=Get EventBarMap()
    NewItem:"EWBP Sparks Sequence Pool Timeline Event Bar Object Reference (by ref)"=@n10
   )
   > AddChild(self=Get LoopingEventsCellsVerticalBox(), Content=@n10)
   > SetHorizontalAlignment(self:Vertical Box Slot Object Reference=SlotAsVerticalBoxSlot(Widget=@n10), InHorizontalAlignment=HAlign_Left)
   n11 = Create Widget(
    execute=@n5.else
    Class=&EWBP_SparksSequencePool_TimelineEventBar_C
    BarColor=(R=0.7305,G=0.7084,B=1,A=1)
    EventSize=Select(Option 0:float=Get NoteFrequency!(self:Stage Show Event Object Reference=@n1) -> (NoteFrequency:double), Option 1:float=0.5, Index=@n7) -> (ReturnValue:float)
    EventName=@n8
    DelayValue=@n9
    SideIcon=/PilgrimQuickplay/StageManager/NeverCook/EditorUtilities/StageManagerEditor/Textures/T_SparksEditorUI_IconLibrary_Warning_64
    ShowIcon=False
    EventReference=@n1
    SparksTimelineEditor=Get SparksTimelineEditor()
    EventIndex=@n1.Array Index
    StepIndex=Get StepIndex() -> (StepIndex=0)
    SequenceReference=Get SequenceReference()
   )
   > Add(TargetArray=Get EventBarMap(), NewItem=@n11)
   > AddChild(self=Get EventsCellsVerticalBox(), Content=@n11)
   > SetHorizontalAlignment(self=SlotAsVerticalBoxSlot(Widget=@n11), InHorizontalAlignment=HAlign_Left)
  ]}
 OnPressed {nodes [OnPressed() {is=title K2Node_FunctionEntry}]}
]
variables [
 BarColor {type=LinearColor, default=<rgba 0.7305 0.7084 1 1>}
 DelayValue {type=number}
 EventBarMap {type="array<EWBP_SparksSequencePool_TimelineEventBar_C>", default []}
 EventName {type=string, default=Event Name}
 EventSize {type=number, default=1}
 LastElement {type=boolean}
 OnPressed {default=()}
 SequenceReference {type=MojoFxManagerSequence_DataAsset_C, default=null}
 SparksTimelineEditor {type=SparksStageManagerTimelineEditor_C, default=null}
 StepData {type=StageManagerSequenceSteps, default {stepEvents []}}
 StepIndex {type=integer}
]
widgets {
 info {rootWidgetClass=&Overlay, widgetCount=39}
 widgets [
  Overlay_5 {widgetClassPath=&Overlay, widget=Overlay_5}
  EditorUtilityButton_174 {
   widgetClassPath=&EditorUtilityButton
   parent=Overlay_5
   widget=EditorUtilityButton_174
   properties {
    backgroundColor=<rgba 1 1 1 0>
    colorAndOpacity=<rgba 1 1 1 0>
    widgetStyle {
     clickedSlateSound {resourceObject=null}
     disabled {
      bIsDynamicallyLoaded=false
      drawAs=NoDrawType
      imageType=NoImage
      mirroring=NoMirror
      resourceName=None
      resourceObject=null
      tiling=NoTile
      imageSize=<0 0>
      margin=<ltrb 0 0 0 0>
      outlineSettings {
       bUseBrushTransparency=false
       roundingType=FixedRadius
       width=1
       color {colorUseRule=UseColor_ColorTable, specifiedColor=<rgba 0 0 0 0>}
       cornerRadii=<4 4 4 4>
      }
      tintColor {colorUseRule=UseColor_Specified, specifiedColor=<rgba 0.0395 0.0395 0.0395 1>}
      uVRegion {bIsValid=false, max=<0 0>, min=<0 0>}
     }
     disabledForeground {colorUseRule=UseColor_ColorTable, specifiedColor=<rgba 0 0 0 0>}
     hovered {
      bIsDynamicallyLoaded=false
      drawAs=NoDrawType
      imageType=NoImage
      mirroring=NoMirror
      resourceName=None
      resourceObject=null
      tiling=NoTile
      imageSize=<32 32>
      margin=<ltrb 0 0 0 0>
      outlineSettings {
       bUseBrushTransparency=false
       roundingType=FixedRadius
       width=1
       color {colorUseRule=UseColor_ColorTable, specifiedColor=<rgba 0 0 0 0>}
       cornerRadii=<4 4 4 4>
      }
      tintColor {colorUseRule=UseColor_Specified, specifiedColor=<rgba 0.0953 0.0953 0.0953 1>}
      uVRegion {bIsValid=false, max=<0 0>, min=<0 0>}
     }
     hoveredForeground {colorUseRule=UseColor_ColorTable, specifiedColor=<rgba 0 0 0 0>}
     hoveredSlateSound {resourceObject=null}
     normal {
      bIsDynamicallyLoaded=false
      drawAs=NoDrawType
      imageType=NoImage
      mirroring=NoMirror
      resourceName=None
      resourceObject=null
      tiling=NoTile
      imageSize=<32 32>
      margin=<ltrb 0 0 0 0>
      outlineSettings {
       bUseBrushTransparency=false
       roundingType=FixedRadius
       width=1
       color {colorUseRule=UseColor_ColorTable, specifiedColor=<rgba 0 0 0 0>}
       cornerRadii=<4 4 4 4>
      }
      tintColor {colorUseRule=UseColor_Specified, specifiedColor=<rgba 0.0395 0.0395 0.0395 1>}
      uVRegion {bIsValid=false, max=<0 0>, min=<0 0>}
     }
     normalForeground {colorUseRule=UseColor_ColorTable, specifiedColor=<rgba 0 0 0 0>}
     normalPadding=<ltrb 12 1.5 12 1.5>
     pressed {
      bIsDynamicallyLoaded=false
      drawAs=NoDrawType
      imageType=NoImage
      mirroring=NoMirror
      resourceName=None
      resourceObject=null
      tiling=NoTile
      imageSize=<32 32>
      margin=<ltrb 0 0 0 0>
      outlineSettings {
       bUseBrushTransparency=false
       roundingType=FixedRadius
       width=1
       color {colorUseRule=UseColor_ColorTable, specifiedColor=<rgba 0 0 0 0>}
       cornerRadii=<4 4 4 4>
      }
      tintColor {colorUseRule=UseColor_Specified, specifiedColor=<rgba 0.0284 0.0284 0.0284 1>}
      uVRegion {bIsValid=false, max=<0 0>, min=<0 0>}
     }
     pressedForeground {colorUseRule=UseColor_ColorTable, specifiedColor=<rgba 0 0 0 0>}
     pressedPadding=<ltrb 12 2.5 12 0.5>
     pressedSlateSound {resourceObject=null}
    }
   }
   slot {class=OverlaySlot, properties {horizontalAlignment=HAlign_Fill, verticalAlignment=VAlign_Fill}}
  }
  Border_0 {
   widgetClassPath=&Border
   parent=Overlay_5
   widget=Border_0
   properties {
    visibility=SelfHitTestInvisible
    background {
     bIsDynamicallyLoaded=false
     drawAs=NoDrawType
     imageType=NoImage
     mirroring=NoMirror
     resourceName=None
     resourceObject=null
     tiling=NoTile
     imageSize=<32 32>
     margin=<ltrb 0 0 0 0>
     outlineSettings {
      bUseBrushTransparency=false
      roundingType=HalfHeightRadius
      width=0
      color {colorUseRule=UseColor_Specified, specifiedColor=<rgba 0 0 0 0>}
      cornerRadii=<0 0 0 0>
     }
     tintColor {colorUseRule=UseColor_Specified, specifiedColor=<rgba 1 1 1 1>}
     uVRegion {bIsValid=false, max=<0 0>, min=<0 0>}
    }
    brushColor=<rgba 0.1198 0.1198 0.1198 1>
    padding=<ltrb 0 0 2 0>
   }
   slot {class=OverlaySlot, properties {horizontalAlignment=HAlign_Fill, verticalAlignment=VAlign_Fill}}
  }
  GeneralSizeBox {
   widgetClassPath=&SizeBox
   bIsVariable=true
   parent=Border_0
   widget=GeneralSizeBox
   properties {bOverride_MinDesiredWidth=true}
   slot {class=BorderSlot, properties {padding=<ltrb 0 0 2 0>}}
  }
  HorizontalBox_1 {widgetClassPath=&HorizontalBox, parent=GeneralSizeBox, widget=HorizontalBox_1, slot {class=SizeBoxSlot}}
  VerticalBox_0 {
   widgetClassPath=&VerticalBox
   parent=HorizontalBox_1
   widget=VerticalBox_0
   slot {class=HorizontalBoxSlot, properties {size {sizeRule=Fill, value=1}}}
  }
  EventsSizeBox {
   widgetClassPath=&SizeBox
   bIsVariable=true
   parent=VerticalBox_0
   widget=EventsSizeBox
   properties {bOverride_MinDesiredHeight=true, minDesiredWidth=472.4348}
   slot {class=VerticalBoxSlot, properties {padding=<ltrb 0 2 4 0>, size {sizeRule=Fill, value=1}}}
  }
  Overlay_1 {widgetClassPath=&Overlay, parent=EventsSizeBox, widget=Overlay_1, slot {class=SizeBoxSlot}}
  Image_210 {
   widgetClassPath=&Image
   parent=Overlay_1
   widget=Image_210
   properties {
    visibility=SelfHitTestInvisible
    brush {
     bIsDynamicallyLoaded=false
     drawAs=Image
     imageType=NoImage
     mirroring=NoMirror
     resourceName=None
     resourceObject=null
     tiling=NoTile
     imageSize=<32 32>
     margin=<ltrb 0 0 0 0>
     outlineSettings {
      bUseBrushTransparency=false
      roundingType=HalfHeightRadius
      width=0
      color {colorUseRule=UseColor_Specified, specifiedColor=<rgba 0 0 0 0>}
      cornerRadii=<0 0 0 0>
     }
     tintColor {colorUseRule=UseColor_Specified, specifiedColor=<rgba 0.0625 0.0625 0.0625 1>}
     uVRegion {bIsValid=false, max=<0 0>, min=<0 0>}
    }
   }
   slot {class=OverlaySlot, properties {horizontalAlignment=HAlign_Fill, verticalAlignment=VAlign_Fill, padding=<ltrb -2 0 -2 0>}}
  }
  StepsVerticalBox {
   widgetClassPath=&VerticalBox
   bIsVariable=true
   parent=Overlay_1
   widget=StepsVerticalBox
   slot {class=OverlaySlot, properties {horizontalAlignment=HAlign_Fill, verticalAlignment=VAlign_Fill, padding=<ltrb 0 2 0 0>}}
  }
  StepCell {
   widgetClassPath=&/EWBP_SparksSequencePool_TimelineEventBar
   bIsVariable=true
   parent=StepsVerticalBox
   widget=StepCell
   properties {eventName=Step ??, visibility=Visible, barColor=<rgba 1 1 1 1>}
   slot {class=VerticalBoxSlot, properties {verticalAlignment=VAlign_Top}}
  }
  NoeventsTextEvents {
   widgetClassPath=&TextBlock
   bIsVariable=true
   parent=StepsVerticalBox
   widget=NoeventsTextEvents
   properties {
    justification=Center
    text=No Events Here
    colorAndOpacity {colorUseRule=UseColor_Specified, specifiedColor=<rgba 0.3906 0.3906 0.3906 1>}
    font {
     bForceMonospaced=false
     bMaterialIsStencil=false
     fontMaterial=null
     fontObject=&HeadingNow
     letterSpacing=0
     monospacedWidth=1
     size=8.25
     skewAmount=0
     typefaceFontName=75 Medium
     outlineSettings {
      bApplyOutlineToDropShadows=false
      bMiteredCorners=false
      bSeparateFillAlpha=false
      outlineMaterial=null
      outlineSize=0
      outlineColor=<rgba 0 0 0 1>
     }
    }
   }
   slot {class=VerticalBoxSlot, properties {verticalAlignment=VAlign_Center, padding=<ltrb 0 5 0 5>, size {sizeRule=Fill, value=1}}}
  }
  EventsCellsVerticalBox {
   widgetClassPath=&VerticalBox
   bIsVariable=true
   parent=StepsVerticalBox
   widget=EventsCellsVerticalBox
   slot {class=VerticalBoxSlot, properties {size {sizeRule=Fill, value=1}}}
  }
  EWBP_SparksSequencePool_TimelineEventBar_1 {
   widgetClassPath=&EWBP_SparksSequencePool_TimelineEventBar
   parent=EventsCellsVerticalBox
   widget=EWBP_SparksSequencePool_TimelineEventBar_1
   properties {barColor=<rgba 0.3886 0.9662 1 1>}
   slot {class=VerticalBoxSlot, properties {horizontalAlignment=HAlign_Left}}
  }
  EWBP_SparksSequencePool_TimelineEventBar_2 {
   widgetClassPath=&EWBP_SparksSequencePool_TimelineEventBar
   parent=EventsCellsVerticalBox
   widget=EWBP_SparksSequencePool_TimelineEventBar_2
   properties {barColor=<rgba 0.3886 0.9662 1 1>}
   slot {class=VerticalBoxSlot, properties {horizontalAlignment=HAlign_Left}}
  }
  SizeBox_2 {
   widgetClassPath=&SizeBox
   parent=StepsVerticalBox
   widget=SizeBox_2
   properties {bOverride_MinDesiredWidth=true, minDesiredWidth=128}
   slot {class=VerticalBoxSlot, properties {horizontalAlignment=HAlign_Center, verticalAlignment=VAlign_Center, padding=<ltrb 0 5 0 5>}}
  }
  NewEventButton {
   widgetClassPath=&EditorUtilityButton
   parent=SizeBox_2
   widget=NewEventButton
   properties {widgetStyle {
     clickedSlateSound {resourceObject=null}
     disabled {
      bIsDynamicallyLoaded=false
      drawAs=RoundedBox
      imageType=NoImage
      mirroring=NoMirror
      resourceName=None
      resourceObject=null
      tiling=NoTile
      imageSize=<32 32>
      margin=<ltrb 0.5 0.5 0.5 0.5>
      outlineSettings {
       bUseBrushTransparency=false
       roundingType=FixedRadius
       width=0
       color {colorUseRule=UseColor_ColorTable, specifiedColor=<rgba 0 0 0 0>}
       cornerRadii=<4 4 4 4>
      }
      tintColor {colorUseRule=UseColor_Specified, specifiedColor=<rgba 0.0395 0.0395 0.0395 1>}
      uVRegion {bIsValid=false, max=<0 0>, min=<0 0>}
     }
     disabledForeground {colorUseRule=UseColor_ColorTable, specifiedColor=<rgba 0 0 0 0>}
     hovered {
      bIsDynamicallyLoaded=false
      drawAs=RoundedBox
      imageType=NoImage
      mirroring=NoMirror
      resourceName=None
      resourceObject=null
      tiling=NoTile
      imageSize=<32 32>
      margin=<ltrb 0 0 0 0>
      outlineSettings {
       bUseBrushTransparency=false
       roundingType=FixedRadius
       width=1
       color {colorUseRule=UseColor_ColorTable, specifiedColor=<rgba 0 0 0 0>}
       cornerRadii=<4 4 4 4>
      }
      tintColor {colorUseRule=UseColor_Specified, specifiedColor=<rgba 0.1198 0.4374 1 1>}
      uVRegion {bIsValid=false, max=<0 0>, min=<0 0>}
     }
     hoveredForeground {colorUseRule=UseColor_ColorTable, specifiedColor=<rgba 0 0 0 0>}
     hoveredSlateSound {resourceObject=null}
     normal {
      bIsDynamicallyLoaded=false
      drawAs=RoundedBox
      imageType=NoImage
      mirroring=NoMirror
      resourceName=None
      resourceObject=null
      tiling=NoTile
      imageSize=<32 32>
      margin=<ltrb 0.5 0.5 0.5 0.5>
      outlineSettings {
       bUseBrushTransparency=false
       roundingType=FixedRadius
       width=0
       color {colorUseRule=UseColor_ColorTable, specifiedColor=<rgba 0 0 0 0>}
       cornerRadii=<4 4 4 4>
      }
      tintColor {colorUseRule=UseColor_Specified, specifiedColor=<rgba 0 0.2537 0.7031 1>}
      uVRegion {bIsValid=false, max=<0 0>, min=<0 0>}
     }
     normalForeground {colorUseRule=UseColor_ColorTable, specifiedColor=<rgba 0 0 0 0>}
     normalPadding=<ltrb 12 1.5 12 1.5>
     pressed {
      bIsDynamicallyLoaded=false
      drawAs=RoundedBox
      imageType=NoImage
      mirroring=NoMirror
      resourceName=None
      resourceObject=null
      tiling=NoTile
      imageSize=<32 32>
      margin=<ltrb 0.5 0.5 0.5 0.5>
      outlineSettings {
       bUseBrushTransparency=false
       roundingType=FixedRadius
       width=0
       color {colorUseRule=UseColor_ColorTable, specifiedColor=<rgba 0 0 0 0>}
       cornerRadii=<4 4 4 4>
      }
      tintColor {colorUseRule=UseColor_Specified, specifiedColor=<rgba 0.0591 0.2238 0.5156 1>}
      uVRegion {bIsValid=false, max=<0 0>, min=<0 0>}
     }
     pressedForeground {colorUseRule=UseColor_ColorTable, specifiedColor=<rgba 0 0 0 0>}
     pressedPadding=<ltrb 12 2.5 12 0.5>
     pressedSlateSound {resourceObject=null}
    }}
   slot {class=SizeBoxSlot}
  }
  HorizontalBox_0 {
   widgetClassPath=&HorizontalBox
   parent=NewEventButton
   widget=HorizontalBox_0
   slot {class=ButtonSlot, properties {padding=<ltrb -5 0 -5 0>}}
  }
  Image_1 {
   widgetClassPath=&Image
   parent=HorizontalBox_0
   widget=Image_1
   properties {brush {
     bIsDynamicallyLoaded=false
     drawAs=Image
     imageType=NoImage
     mirroring=NoMirror
     resourceName=None
     resourceObject=/PilgrimQuickplay/StageManager/NeverCook/EditorUtilities/StageManagerEditor/Textures/MI_SparksEditorUI_PlusMinus_NoBorder
     tiling=NoTile
     imageSize=<15 15>
     margin=<ltrb 0 0 0 0>
     outlineSettings {
      bUseBrushTransparency=false
      roundingType=HalfHeightRadius
      width=0
      color {colorUseRule=UseColor_Specified, specifiedColor=<rgba 0 0 0 0>}
      cornerRadii=<0 0 0 0>
     }
     tintColor {colorUseRule=UseColor_Specified, specifiedColor=<rgba 1 1 1 1>}
     uVRegion {bIsValid=false, max=<0 0>, min=<0 0>}
    }}
   slot {class=HorizontalBoxSlot, properties {horizontalAlignment=HAlign_Left, verticalAlignment=VAlign_Center}}
  }
  TextBlock_58 {
   widgetClassPath=&TextBlock
   parent=HorizontalBox_0
   widget=TextBlock_58
   properties {
    text=New Event
    font {
     bForceMonospaced=false
     bMaterialIsStencil=false
     fontMaterial=null
     fontObject=&HeadingNow
     letterSpacing=0
     monospacedWidth=1
     size=8.25
     skewAmount=0
     typefaceFontName=75 Medium
     outlineSettings {
      bApplyOutlineToDropShadows=false
      bMiteredCorners=false
      bSeparateFillAlpha=false
      outlineMaterial=null
      outlineSize=0
      outlineColor=<rgba 0 0 0 1>
     }
    }
   }
   slot {class=HorizontalBoxSlot, properties {horizontalAlignment=HAlign_Center, verticalAlignment=VAlign_Center, padding=<ltrb 4 2 4 2>}}
  }
  Spacer_111 {widgetClassPath=&Spacer, parent=VerticalBox_0, widget=Spacer_111, slot {class=VerticalBoxSlot}}
  LoopingStepsSizeBox {
   widgetClassPath=&SizeBox
   bIsVariable=true
   parent=VerticalBox_0
   widget=LoopingStepsSizeBox
   properties {minDesiredHeight=91.3767}
   slot {class=VerticalBoxSlot, properties {padding=<ltrb 0 0 4 0>}}
  }
  Overlay_3 {widgetClassPath=&Overlay, parent=LoopingStepsSizeBox, widget=Overlay_3, slot {class=SizeBoxSlot}}
  Image {
   widgetClassPath=&Image
   parent=Overlay_3
   widget=Image
   properties {
    visibility=SelfHitTestInvisible
    brush {
     bIsDynamicallyLoaded=false
     drawAs=Image
     imageType=NoImage
     mirroring=NoMirror
     resourceName=None
     resourceObject=null
     tiling=NoTile
     imageSize=<32 32>
     margin=<ltrb 0 0 0 0>
     outlineSettings {
      bUseBrushTransparency=false
      roundingType=HalfHeightRadius
      width=0
      color {colorUseRule=UseColor_Specified, specifiedColor=<rgba 0 0 0 0>}
      cornerRadii=<0 0 0 0>
     }
     tintColor {colorUseRule=UseColor_Specified, specifiedColor=<rgba 0.0625 0.0625 0.0625 1>}
     uVRegion {bIsValid=false, max=<0 0>, min=<0 0>}
    }
   }
   slot {class=OverlaySlot, properties {horizontalAlignment=HAlign_Fill, verticalAlignment=VAlign_Fill, padding=<ltrb -2 0 -2 0>}}
  }
  LoopingStepsVerticalBox {
   widgetClassPath=&VerticalBox
   bIsVariable=true
   parent=Overlay_3
   widget=LoopingStepsVerticalBox
   slot {class=OverlaySlot, properties {horizontalAlignment=HAlign_Fill, verticalAlignment=VAlign_Fill, padding=<ltrb 0 4 0 0>}}
  }
  LoopingEventsCellsVerticalBox {
   widgetClassPath=&VerticalBox
   bIsVariable=true
   parent=LoopingStepsVerticalBox
   widget=LoopingEventsCellsVerticalBox
   slot {class=VerticalBoxSlot, properties {size {sizeRule=Fill, value=1}}}
  }
  EWBP_SparksSequencePool_TimelineEventBar {
   widgetClassPath=&EWBP_SparksSequencePool_TimelineEventBar
   parent=LoopingEventsCellsVerticalBox
   widget=EWBP_SparksSequencePool_TimelineEventBar
   properties {showIcon=true, barColor=<rgba 0.125 1 0.209 1>}
   slot {class=VerticalBoxSlot, properties {horizontalAlignment=HAlign_Left}}
  }
  NoEventsTextLooping {
   widgetClassPath=&TextBlock
   bIsVariable=true
   parent=LoopingStepsVerticalBox
   widget=NoEventsTextLooping
   properties {
    justification=Center
    text=No Events Here
    visibility=Collapsed
    colorAndOpacity {colorUseRule=UseColor_Specified, specifiedColor=<rgba 0.3906 0.3906 0.3906 1>}
    font {
     bForceMonospaced=false
     bMaterialIsStencil=false
     fontMaterial=null
     fontObject=&HeadingNow
     letterSpacing=0
     monospacedWidth=1
     size=8.25
     skewAmount=0
     typefaceFontName=75 Medium
     outlineSettings {
      bApplyOutlineToDropShadows=false
      bMiteredCorners=false
      bSeparateFillAlpha=false
      outlineMaterial=null
      outlineSize=0
      outlineColor=<rgba 0 0 0 1>
     }
    }
   }
   slot {class=VerticalBoxSlot, properties {verticalAlignment=VAlign_Center, padding=<ltrb 0 5 0 2>, size {sizeRule=Fill, value=1}}}
  }
  SizeBox_3 {
   widgetClassPath=&SizeBox
   parent=LoopingStepsVerticalBox
   widget=SizeBox_3
   properties {bOverride_MinDesiredWidth=true, minDesiredWidth=128.2927}
   slot {class=VerticalBoxSlot, properties {horizontalAlignment=HAlign_Center, verticalAlignment=VAlign_Bottom, padding=<ltrb 0 5 0 5>}}
  }
  NewLoopEventButton {
   widgetClassPath=&EditorUtilityButton
   parent=SizeBox_3
   widget=NewLoopEventButton
   properties {widgetStyle {
     clickedSlateSound {resourceObject=null}
     disabled {
      bIsDynamicallyLoaded=false
      drawAs=RoundedBox
      imageType=NoImage
      mirroring=NoMirror
      resourceName=None
      resourceObject=null
      tiling=NoTile
      imageSize=<32 32>
      margin=<ltrb 0.5 0.5 0.5 0.5>
      outlineSettings {
       bUseBrushTransparency=false
       roundingType=FixedRadius
       width=0
       color {colorUseRule=UseColor_ColorTable, specifiedColor=<rgba 0 0 0 0>}
       cornerRadii=<4 4 4 4>
      }
      tintColor {colorUseRule=UseColor_Specified, specifiedColor=<rgba 0.0395 0.0395 0.0395 1>}
      uVRegion {bIsValid=false, max=<0 0>, min=<0 0>}
     }
     disabledForeground {colorUseRule=UseColor_ColorTable, specifiedColor=<rgba 0 0 0 0>}
     hovered {
      bIsDynamicallyLoaded=false
      drawAs=RoundedBox
      imageType=NoImage
      mirroring=NoMirror
      resourceName=None
      resourceObject=null
      tiling=NoTile
      imageSize=<32 32>
      margin=<ltrb 0 0 0 0>
      outlineSettings {
       bUseBrushTransparency=false
       roundingType=FixedRadius
       width=1
       color {colorUseRule=UseColor_ColorTable, specifiedColor=<rgba 0 0 0 0>}
       cornerRadii=<4 4 4 4>
      }
      tintColor {colorUseRule=UseColor_Specified, specifiedColor=<rgba 0.0547 0.875 0.1335 1>}
      uVRegion {bIsValid=false, max=<0 0>, min=<0 0>}
     }
     hoveredForeground {colorUseRule=UseColor_ColorTable, specifiedColor=<rgba 0 0 0 0>}
     hoveredSlateSound {resourceObject=null}
     normal {
      bIsDynamicallyLoaded=false
      drawAs=RoundedBox
      imageType=NoImage
      mirroring=NoMirror
      resourceName=None
      resourceObject=null
      tiling=NoTile
      imageSize=<32 32>
      margin=<ltrb 0.5 0.5 0.5 0.5>
      outlineSettings {
       bUseBrushTransparency=false
       roundingType=FixedRadius
       width=0
       color {colorUseRule=UseColor_ColorTable, specifiedColor=<rgba 0 0 0 0>}
       cornerRadii=<4 4 4 4>
      }
      tintColor {colorUseRule=UseColor_Specified, specifiedColor=<rgba 0.0335 0.5365 0.0818 1>}
      uVRegion {bIsValid=false, max=<0 0>, min=<0 0>}
     }
     normalForeground {colorUseRule=UseColor_ColorTable, specifiedColor=<rgba 0 0 0 0>}
     normalPadding=<ltrb 12 1.5 12 1.5>
     pressed {
      bIsDynamicallyLoaded=false
      drawAs=RoundedBox
      imageType=NoImage
      mirroring=NoMirror
      resourceName=None
      resourceObject=null
      tiling=NoTile
      imageSize=<32 32>
      margin=<ltrb 0.5 0.5 0.5 0.5>
      outlineSettings {
       bUseBrushTransparency=false
       roundingType=FixedRadius
       width=0
       color {colorUseRule=UseColor_ColorTable, specifiedColor=<rgba 0 0 0 0>}
       cornerRadii=<4 4 4 4>
      }
      tintColor {colorUseRule=UseColor_Specified, specifiedColor=<rgba 0.0273 0.4375 0.0667 1>}
      uVRegion {bIsValid=false, max=<0 0>, min=<0 0>}
     }
     pressedForeground {colorUseRule=UseColor_ColorTable, specifiedColor=<rgba 0 0 0 0>}
     pressedPadding=<ltrb 12 2.5 12 0.5>
     pressedSlateSound {resourceObject=null}
    }}
   slot {class=SizeBoxSlot}
  }
  HorizontalBox {
   widgetClassPath=&HorizontalBox
   parent=NewLoopEventButton
   widget=HorizontalBox
   slot {class=ButtonSlot, properties {padding=<ltrb -5 0 -5 0>}}
  }
  Image_2 {
   widgetClassPath=&Image
   parent=HorizontalBox
   widget=Image_2
   properties {brush {
     bIsDynamicallyLoaded=false
     drawAs=Image
     imageType=NoImage
     mirroring=NoMirror
     resourceName=None
     resourceObject=&MI_SparksEditorUI_PlusMinus_NoBorder
     tiling=NoTile
     imageSize=<15 15>
     margin=<ltrb 0 0 0 0>
     outlineSettings {
      bUseBrushTransparency=false
      roundingType=HalfHeightRadius
      width=0
      color {colorUseRule=UseColor_Specified, specifiedColor=<rgba 0 0 0 0>}
      cornerRadii=<0 0 0 0>
     }
     tintColor {colorUseRule=UseColor_Specified, specifiedColor=<rgba 1 1 1 1>}
     uVRegion {bIsValid=false, max=<0 0>, min=<0 0>}
    }}
   slot {class=HorizontalBoxSlot, properties {horizontalAlignment=HAlign_Left, verticalAlignment=VAlign_Center}}
  }
  TextBlock {
   widgetClassPath=&TextBlock
   parent=HorizontalBox
   widget=TextBlock
   properties {
    text=New Loop Event
    font {
     bForceMonospaced=false
     bMaterialIsStencil=false
     fontMaterial=null
     fontObject=&HeadingNow
     letterSpacing=0
     monospacedWidth=1
     size=8.25
     skewAmount=0
     typefaceFontName=75 Medium
     outlineSettings {
      bApplyOutlineToDropShadows=false
      bMiteredCorners=false
      bSeparateFillAlpha=false
      outlineMaterial=null
      outlineSize=0
      outlineColor=<rgba 0 0 0 1>
     }
    }
   }
   slot {class=HorizontalBoxSlot, properties {horizontalAlignment=HAlign_Center, verticalAlignment=VAlign_Center, padding=<ltrb 4 2 4 2>}}
  }
  NewStepArea {
   widgetClassPath=&Overlay
   bIsVariable=true
   parent=HorizontalBox_1
   widget=NewStepArea
   slot {class=HorizontalBoxSlot, properties {padding=<ltrb 0 2 0 0>}}
  }
  Image_3 {
   widgetClassPath=&Image
   parent=NewStepArea
   widget=Image_3
   properties {
    visibility=SelfHitTestInvisible
    brush {
     bIsDynamicallyLoaded=false
     drawAs=Image
     imageType=NoImage
     mirroring=NoMirror
     resourceName=None
     resourceObject=null
     tiling=NoTile
     imageSize=<32 32>
     margin=<ltrb 0 0 0 0>
     outlineSettings {
      bUseBrushTransparency=false
      roundingType=HalfHeightRadius
      width=0
      color {colorUseRule=UseColor_Specified, specifiedColor=<rgba 0 0 0 0>}
      cornerRadii=<0 0 0 0>
     }
     tintColor {colorUseRule=UseColor_Specified, specifiedColor=<rgba 0.0625 0.0625 0.0625 1>}
     uVRegion {bIsValid=false, max=<0 0>, min=<0 0>}
    }
   }
   slot {class=OverlaySlot, properties {horizontalAlignment=HAlign_Fill, verticalAlignment=VAlign_Fill, padding=<ltrb 0 0 -2 0>}}
  }
  NewStepButton {
   widgetClassPath=&EditorUtilityButton
   parent=NewStepArea
   widget=NewStepButton
   properties {widgetStyle {
     clickedSlateSound {resourceObject=null}
     disabled {
      bIsDynamicallyLoaded=false
      drawAs=RoundedBox
      imageType=NoImage
      mirroring=NoMirror
      resourceName=None
      resourceObject=/Game/UI/Foundation/Textures/Buttons/T-Button-Square
      tiling=NoTile
      imageSize=<32 32>
      margin=<ltrb 0.5 0.5 0.5 0.5>
      outlineSettings {
       bUseBrushTransparency=false
       roundingType=FixedRadius
       width=0
       color {colorUseRule=UseColor_ColorTable, specifiedColor=<rgba 0 0 0 0>}
       cornerRadii=<4 4 4 4>
      }
      tintColor {colorUseRule=UseColor_Specified, specifiedColor=<rgba 0.0395 0.0395 0.0395 1>}
      uVRegion {bIsValid=false, max=<0 0>, min=<0 0>}
     }
     disabledForeground {colorUseRule=UseColor_ColorTable, specifiedColor=<rgba 0 0 0 0>}
     hovered {
      bIsDynamicallyLoaded=false
      drawAs=RoundedBox
      imageType=NoImage
      mirroring=NoMirror
      resourceName=None
      resourceObject=&T-Button-Square
      tiling=NoTile
      imageSize=<32 32>
      margin=<ltrb 0.5 0.5 0.5 0.5>
      outlineSettings {
       bUseBrushTransparency=false
       roundingType=FixedRadius
       width=0
       color {colorUseRule=UseColor_ColorTable, specifiedColor=<rgba 0 0 0 0>}
       cornerRadii=<4 4 4 4>
      }
      tintColor {colorUseRule=UseColor_Specified, specifiedColor=<rgba 0.7344 0.4651 0 1>}
      uVRegion {bIsValid=false, max=<0 0>, min=<0 0>}
     }
     hoveredForeground {colorUseRule=UseColor_ColorTable, specifiedColor=<rgba 0 0 0 0>}
     hoveredSlateSound {resourceObject=null}
     normal {
      bIsDynamicallyLoaded=false
      drawAs=RoundedBox
      imageType=NoImage
      mirroring=NoMirror
      resourceName=None
      resourceObject=&T-Button-Square
      tiling=NoTile
      imageSize=<32 32>
      margin=<ltrb 0.5 0.5 0.5 0.5>
      outlineSettings {
       bUseBrushTransparency=false
       roundingType=FixedRadius
       width=0
       color {colorUseRule=UseColor_ColorTable, specifiedColor=<rgba 0 0 0 0>}
       cornerRadii=<4 4 4 4>
      }
      tintColor {colorUseRule=UseColor_Specified, specifiedColor=<rgba 0.7991 0.5583 0.0003 1>}
      uVRegion {bIsValid=false, max=<0 0>, min=<0 0>}
     }
     normalForeground {colorUseRule=UseColor_ColorTable, specifiedColor=<rgba 0 0 0 0>}
     normalPadding=<ltrb 12 1.5 12 1.5>
     pressed {
      bIsDynamicallyLoaded=false
      drawAs=RoundedBox
      imageType=NoImage
      mirroring=NoMirror
      resourceName=None
      resourceObject=&T-Button-Square
      tiling=NoTile
      imageSize=<32 32>
      margin=<ltrb 0.5 0.5 0.5 0.5>
      outlineSettings {
       bUseBrushTransparency=false
       roundingType=FixedRadius
       width=0
       color {colorUseRule=UseColor_ColorTable, specifiedColor=<rgba 0 0 0 0>}
       cornerRadii=<4 4 4 4>
      }
      tintColor {colorUseRule=UseColor_Specified, specifiedColor=<rgba 0.6458 0.409 0 1>}
      uVRegion {bIsValid=false, max=<0 0>, min=<0 0>}
     }
     pressedForeground {colorUseRule=UseColor_ColorTable, specifiedColor=<rgba 0 0 0 0>}
     pressedPadding=<ltrb 12 2.5 12 0.5>
     pressedSlateSound {resourceObject=null}
    }}
   slot {class=OverlaySlot, properties {padding=<ltrb 5 2 5 0>}}
  }
  HorizontalBox_2 {
   widgetClassPath=&HorizontalBox
   parent=NewStepButton
   widget=HorizontalBox_2
   slot {class=ButtonSlot, properties {horizontalAlignment=HAlign_Left}}
  }
  ExternalIcon_1 {
   widgetClassPath=&Image
   parent=HorizontalBox_2
   widget=ExternalIcon_1
   properties {brush {
     bIsDynamicallyLoaded=false
     drawAs=Image
     imageType=NoImage
     mirroring=NoMirror
     resourceName=None
     resourceObject=&MI_SparksEditorUI_PlusMinus_NoBorder
     tiling=NoTile
     imageSize=<15 15>
     margin=<ltrb 0 0 0 0>
     outlineSettings {
      bUseBrushTransparency=false
      roundingType=HalfHeightRadius
      width=0
      color {colorUseRule=UseColor_Specified, specifiedColor=<rgba 0 0 0 0>}
      cornerRadii=<0 0 0 0>
     }
     tintColor {colorUseRule=UseColor_Specified, specifiedColor=<rgba 1 1 1 1>}
     uVRegion {bIsValid=false, max=<0 0>, min=<0 0>}
    }}
   slot {class=HorizontalBoxSlot, properties {horizontalAlignment=HAlign_Center, verticalAlignment=VAlign_Center, padding=<ltrb 0 0 3 0>}}
  }
  TextBlock_1 {
   widgetClassPath=&TextBlock
   parent=HorizontalBox_2
   widget=TextBlock_1
   properties {
    text=New Step
    font {
     bForceMonospaced=false
     bMaterialIsStencil=false
     fontMaterial=null
     fontObject=&HeadingNow
     letterSpacing=0
     monospacedWidth=1
     size=7.5
     skewAmount=0
     typefaceFontName=75 Medium
     outlineSettings {
      bApplyOutlineToDropShadows=false
      bMiteredCorners=false
      bSeparateFillAlpha=false
      outlineMaterial=null
      outlineSize=0
      outlineColor=<rgba 0 0 0 1>
     }
    }
   }
   slot {class=HorizontalBoxSlot, properties {verticalAlignment=VAlign_Center}}
  }
 ]
}
