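# Editor utility widget for one sequence row in the Sparks Stage Manager timeline editor.
# FillWithEvents creates an event group widget per step, measuring desired sizes after forced layout prepasses to equalize event and looping-step heights.
# The eye toggle collapses or restores those groups, and a dropdown opens the sequence submenu.
# An empty sequence offers a button that appends a step to the sequence asset, marks it for undo and refreshes the editor.
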EditorUtilityWidgetBlueprint /PilgrimQuickplay/StageManager/NeverCook/EditorUtilities/StageManagerEditor/EWBP_SparksSequencePoolTimelineCell
graphs [
 EventGraph {nodes [
   n1 = "On Check State Changed (ToggleViewDa)"!() -> (OutputDelegate, then, bIsChecked:bool) {is=title K2Node_ComponentBoundEvent}
   > Set IsOpen(IsOpen:bool=@n1.bIsChecked) -> (Output_Get:bool)
   > n2 = For Each Loop(Array:Array of EWBP Sparks Sequence Pool Timeline Event Group Object References=Get EventsCached() -> (EventsCached:Array of EWBP Sparks Sequence Pool Timeline Event Group Object References)) -> (Array Element:EWBP Sparks Sequence Pool Timeline Event Group Object Reference)
   SetMinDesiredHeight!(
    execute=@n2.Completed
    self:Size Box Object Reference=Get SequenceSizeBox() -> (SequenceSizeBox:Size Box Object Reference)
    InMinDesiredHeight:float=SelectFloat(A=Get LastEventHeightSize() -> (LastEventHeightSize:double), B=50, bPickA=Get IsOpen() -> (IsOpen:bool))
   ) -> (then)
   ChangeEventsVisibility!(
    execute=@n2.LoopBody
    self:EWBP Sparks Sequence Pool Timeline Event Group Object Reference=@n2
    InVisibility:ESlateVisibility Enum=Select(Option 0=Collapsed, Option 1=SelfHitTestInvisible, Index=Get IsOpen())
   ) -> (then)
   > n3 = Sequence()
   > SetMinDesiredHeight(
    self=Get EventsSizeBox!(self:EWBP Sparks Sequence Pool Timeline Event Group Object Reference=@n2) -> (EventsSizeBox:Size Box Object Reference)
    InMinDesiredHeight=SelectFloat(A=Get LastEventHeightSize(), B=0, bPickA=Get IsOpen())
   )
   SetMinDesiredHeight(
    execute=@n3.then_1
    self=Get LoopingStepsSizeBox!(self:EWBP Sparks Sequence Pool Timeline Event Group Object Reference=@n2) -> (LoopingStepsSizeBox:Size Box Object Reference)
    InMinDesiredHeight=SelectFloat(A=Get LastLEventHeightSize() -> (LastLEventHeightSize:double), B=-2, bPickA=Get IsOpen())
   )
   "On Pressed (DropDownButton)"() {is=title K2Node_ComponentBoundEvent}
   > UpdateToggles!(
    execute
    self:Sparks Stage Manager Timeline Editor Object Reference=Get SparksTimelineEditor() -> (SparksTimelineEditor:Sparks Stage Manager Timeline Editor Object Reference)
    Toggle:EWBP Sparks Sequence Pool Timeline Event Bar Object Reference
    ForceCloseAll:bool=true
    DontCallEvent:bool=false
   ) -> (then)
   > n4 = Set IsSubOpen(IsSubOpen:bool=NOT Boolean(A=Get IsSubOpen() -> (IsSubOpen:bool))) -> (Output_Get:bool)
   > UpdateDropdownVisibility!(execute, self, IsOpen:bool=@n4) -> (then)
   "On Deleted (EWBP_SparksTimelineSubmenu)"() {is=title K2Node_ComponentBoundEvent}
   > ForceRefresh(self:Sparks Stage Manager Timeline Editor Object Reference=Get SparksTimelineEditor())
   "On Pressed (NewStepButton)"() {is=title K2Node_ComponentBoundEvent}
   n5 = Get SequenceReference() -> (SequenceReference:Mojo Fx Manager Sequence Data Asset Object Reference)
   > Add(
    TargetArray:Array of Stage Manager Sequence Steps Structures=Get SequenceSteps!(self:Stage Show Sequence Object Reference=@n5) -> (SequenceSteps:Array of Stage Manager Sequence Steps Structures)
    NewItem:"Stage Manager Sequence Steps Structure (by ref)"
   ) {is=title K2Node_CallArrayFunction}
   > TransactObject(Object=@n5)
   > ForceRefresh(self=Get SparksTimelineEditor())
   UpdateSubmenu() {is=title K2Node_CustomEvent}
   > Branch(Condition=Get IsSubOpen())
   > UpdateDropdownVisibility(IsOpen=false)
   > Set IsSubOpen(IsSubOpen=false)
   > ForceInjectTimelineEditorReference!(
    execute
    self:EWBP Sparks Timeline Submenu Object Reference=Get EWBP_SparksTimelineSubmenu() -> (EWBP_SparksTimelineSubmenu:EWBP Sparks Timeline Submenu Object Reference)
    SparksTimelineEditor:"Sparks Stage Manager Timeline Editor Object Reference (by ref)"=Get SparksTimelineEditor()
   ) -> (then)
   n6 = ReturningFromSubmenu!() -> (OutputDelegate, then, IsVisible:bool=~) {is=title K2Node_CustomEvent}
   > n7 = Branch(Condition=@n6.IsVisible)
   SetVisibility(execute=@n7.else, self=Get ArrowUp() -> (ArrowUp:Image Object Reference), InVisibility=Collapsed)
   > SetVisibility(self=Get ArrowDown() -> (ArrowDown:Image Object Reference), InVisibility=Visible)
   > Set IsSubOpen(IsSubOpen=false)
   Event Construct()
   > FillWithEvents()
   > UpdateSubmenuData!(
    execute
    self:EWBP Sparks Timeline Submenu Object Reference=Get EWBP_SparksTimelineSubmenu()
    SequencePoolData:"Stage Show Sequence Pool Object Reference (by ref)"=GetSelectedSequencePool!(self:Sparks Stage Manager Timeline Editor Object Reference=Get SparksTimelineEditor()) -> (Value:Stage Show Sequence Pool Object Reference)
    SequenceReference:"Mojo Fx Manager Sequence Data Asset Object Reference (by ref)"=Get SequenceReference()
    EventReference:"Object Reference (by ref)"
    SparksTimelineEditor:"Sparks Stage Manager Timeline Editor Object Reference (by ref)"=Get SparksTimelineEditor()
    EventBarReference:EWBP Sparks Sequence Pool Timeline Event Bar Object Reference
    SequenceIndex:int=Get SequenceIndex() -> (SequenceIndex:int=0)
    StepIndex:int=0
    EventIndex:int=0
    IsHeader:bool=false
   ) -> (then)
   n8 = Create Event()
   n9 = Get EWBP_SparksTimelineSubmenu()
   > Bind Event to Update Visibility Status(self:EWBP Sparks Timeline Submenu Object Reference=@n9, Delegate=@n8) {is=title K2Node_AddDelegate}
   Event Destruct()
   > Unbind Event from Toggle Updated(self:Sparks Stage Manager Timeline Editor Object Reference=Get SparksTimelineEditor(), Delegate=Create Event()) {is=title K2Node_RemoveDelegate}
   Bind Event to On Deleted(self:EWBP Sparks Timeline Submenu Object Reference=@n9, Delegate=@n8) {is=title K2Node_AddDelegate}
   Bind Event to On Visibility Changed(self:User Widget Object Reference=@n9, Delegate=@n8) {is=title K2Node_AddDelegate}
  ]}
 FillWithEvents {nodes [
   FillWithEvents() {is=title K2Node_FunctionEntry}
   > Clear(TargetArray:Array of EWBP Sparks Sequence Pool Timeline Event Group Object References=Get EventsCached())
   > "SetText (Text)"(
    self=Get CellName() -> (CellName:Text Object Reference)
    InText=Format Text!(Format="LOCTEXT(\"Sequence {index}\")", index:int=Get SequenceIndex() -> (SequenceIndex=0)) -> (Result)
   )
   n1 = "To String (Text)"(InText=Get SequenceName() -> (SequenceName:Text))
   > "SetText (Text)"(
    self=Get SequenceNameText() -> (SequenceNameText:Text Object Reference)
    InText="To Text (String)"(InString=Split(SourceString=@n1, InStr=Pilgrim_StageFX_Sequence_, SearchDir=FromStart).RightS)
   )
   > SetToolTipText!(execute, self:Widget Object Reference=Get SequenceNameText(), InToolTipText:"Text (by ref)"=Get SequenceName()) -> (then)
   > ClearChildren(self=Get EventsHorizontalBox() -> (EventsHorizontalBox:Horizontal Box Object Reference))
   n2 = Get EventSteps() -> (EventSteps:Array of Stage Manager Sequence Steps Structures)
   > n3 = For Each Loop(Array:Array of Stage Manager Sequence Steps Structures=@n2) -> (Array Element:Stage Manager Sequence Steps Structure)
   n4 = Branch(
    execute=@n3.Completed
    Condition=Is Empty!(TargetArray:Array of EWBP Sparks Sequence Pool Timeline Event Group Object References=Get EventsCached()) -> (ReturnValue:bool)
   )
   > SetVisibility(self=Get EventsHorizontalBox(), InVisibility=Collapsed)
   > SetVisibility(
    self=[Get NewStepArea() -> (NewStepArea:Overlay Object Reference), Get NoStepsText() -> (NoStepsText:Text Object Reference)]
    InVisibility=SelfHitTestInvisible
   )
   SetVisibility(execute=@n4.else, self=Get EventsHorizontalBox(), InVisibility=SelfHitTestInvisible)
   > SetVisibility(self=[Get NoStepsText(), Get NewStepArea()], InVisibility=Collapsed)
   > n5 = For Each Loop(Array:Array of EWBP Sparks Sequence Pool Timeline Event Group Object References=Get EventsCached()) -> (Array Element:EWBP Sparks Sequence Pool Timeline Event Group Object Reference)
   > n6 = Sequence()
   > SetMinDesiredHeight(self=Get EventsSizeBox(self=@n5), InMinDesiredHeight=Get LastEventHeightSize())
   SetMinDesiredHeight(execute=@n6.then_1, self=Get LoopingStepsSizeBox(self=@n5), InMinDesiredHeight=Get LastLEventHeightSize())
   n7 = Length(TargetArray:Array of Stage Manager Sequence Steps Structures=@n2)
   n8 = Create Widget!(
    execute:Exec=@n3.LoopBody
    WorldContextObject:Object Reference
    Class:User Widget Class Reference=&/EWBP_SparksSequencePool_TimelineEventGroup.EWBP_SparksSequencePool_TimelineEventGroup_C
    OwningPlayer:Player Controller Object Reference
    StepIndex:int=@n3.Array Index
    StepData:Stage Manager Sequence Steps Structure=@n3
    LastElement:bool="Equal (Integer)"(A=@n3.Array Index, B=int - int(A=@n7))
    SparksTimelineEditor:Sparks Stage Manager Timeline Editor Object Reference=Get SparksTimelineEditor()
    SequenceReference:Mojo Fx Manager Sequence Data Asset Object Reference=Get SequenceReference()
   ) -> (then:Exec, ReturnValue:EWBP Sparks Sequence Pool Timeline Event Group Object Reference)
   > Add Unique(
    TargetArray:Array of EWBP Sparks Sequence Pool Timeline Event Group Object References=Get EventsCached()
    NewItem:"EWBP Sparks Sequence Pool Timeline Event Group Object Reference (by ref)"=@n8
   )
   > AddChild(self=Get EventsHorizontalBox(), Content=@n8)
   > n9 = Sequence()
   n10 = Get StepsVerticalBox!(self:EWBP Sparks Sequence Pool Timeline Event Group Object Reference=@n8) -> (StepsVerticalBox:Vertical Box Object Reference)
   > ForceLayoutPrepass(self:Widget Object Reference=@n10)
   n11 = GetDesiredSize!(self:Widget Object Reference=@n10) -> (ReturnValue_X:double, ReturnValue_Y:double)
   > Branch(Condition="float > float"(A=@n11.ReturnValue_Y, B=Get LastEventHeightSize()))
   > SetMinDesiredHeight(self=Get SequenceSizeBox(), InMinDesiredHeight=@n11.ReturnValue_Y)
   > Set LastEventHeightSize(LastEventHeightSize:double=@n11.ReturnValue_Y) -> (Output_Get:double)
   n12 = Get LoopingStepsVerticalBox!(self:EWBP Sparks Sequence Pool Timeline Event Group Object Reference=@n8) -> (LoopingStepsVerticalBox:Vertical Box Object Reference)
   ForceLayoutPrepass(execute=@n9.then_1, self=@n12)
   n13 = GetDesiredSize(self=@n12)
   > Branch(Condition="float > float"(A=@n13.ReturnValue_Y, B=Get LastLEventHeightSize()))
   > Set LastLEventHeightSize(LastLEventHeightSize:double=@n13.ReturnValue_Y) -> (Output_Get:double)
  ]}
 OpenSubmenu {nodes [OpenSubmenu!() -> (then, Reference:Widget Object Reference) {is=title K2Node_FunctionEntry}]}
 UpdateDropdownVisibility {nodes [
   n1 = UpdateDropdownVisibility!() -> (then, IsOpen:bool=~) {is=title K2Node_FunctionEntry}
   > SetVisibility(self=[Get EWBP_SparksTimelineSubmenu(), Get ArrowUp()], InVisibility=Select(Option 0=Collapsed, Option 1=Visible, Index=@n1))
   > SetVisibility(self=Get ArrowDown(), InVisibility=Select(Option 0=Visible, Option 1=Collapsed, Index=@n1))
  ]}
 UpdateSubmenuSequenceDA {nodes [
   n1 = UpdateSubmenuSequenceDA!() -> (then, EventReference:Object Reference) {is=title K2Node_FunctionEntry}
   > Set EventReference!(execute, EventReference:Object Reference=@n1, self:EWBP Sparks Timeline Submenu Object Reference=Get EWBP_SparksTimelineSubmenu()) -> (then, Output_Get:Object Reference)
  ]}
]
variables [
 EventsCached {type="array<EWBP_SparksSequencePool_TimelineEventGroup_C>", default []}
 EventSteps {type="array<StageManagerSequenceSteps>", default [{stepEvents []}, {stepEvents []}]}
 IsOpen {type=boolean}
 IsSubOpen {type=boolean}
 LastEventHeightSize {type=number}
 LastLEventHeightSize {type=number}
 OpenSubmenu {default=()}
 SequenceIndex {type=integer}
 SequenceName {type=string, default=Stage_Seq_Lights_01_01}
 SequenceReference {type=MojoFxManagerSequence_DataAsset_C, default=null}
 SparksTimelineEditor {type=SparksStageManagerTimelineEditor_C, default=null}
]
widgets {
 info {rootWidgetClass=&Overlay, widgetCount=37}
 widgets [
  Overlay_0 {widgetClassPath=&Overlay, widget=Overlay_0}
  Image_99 {
   widgetClassPath=&Image
   parent=Overlay_0
   widget=Image_99
   properties {brush {
     bIsDynamicallyLoaded=false
     drawAs=Image
     imageType=NoImage
     mirroring=NoMirror
     resourceName=None
     resourceObject=null
     tiling=NoTile
     imageSize=<32 32>
     margin=<ltrb 0 0 0 0>
     outlineSettings {
      bUseBrushTransparency=false
      roundingType=HalfHeightRadius
      width=0
      color {colorUseRule=UseColor_Specified, specifiedColor=<rgba 0 0 0 0>}
      cornerRadii=<0 0 0 0>
     }
     tintColor {colorUseRule=UseColor_Specified, specifiedColor=<rgba 0.101 0.101 0.101 1>}
     uVRegion {bIsValid=false, max=<0 0>, min=<0 0>}
    }}
   slot {class=OverlaySlot, properties {horizontalAlignment=HAlign_Fill, verticalAlignment=VAlign_Fill}}
  }
  HorizontalBox_0 {
   widgetClassPath=&HorizontalBox
   parent=Overlay_0
   widget=HorizontalBox_0
   slot {class=OverlaySlot, properties {horizontalAlignment=HAlign_Fill, verticalAlignment=VAlign_Fill, padding=<ltrb 0 0 0 1>}}
  }
  VerticalBox_0 {
   widgetClassPath=&VerticalBox
   parent=HorizontalBox_0
   widget=VerticalBox_0
   slot {class=HorizontalBoxSlot, properties {horizontalAlignment=HAlign_Left}}
  }
  SequenceSideBar {
   widgetClassPath=&Overlay
   bIsVariable=true
   parent=VerticalBox_0
   widget=SequenceSideBar
   slot {class=VerticalBoxSlot, properties {padding=<ltrb 0 2 4 0>}}
  }
  SequenceSizeBox {
   widgetClassPath=&SizeBox
   bIsVariable=true
   parent=SequenceSideBar
   widget=SequenceSizeBox
   properties {bOverride_MaxDesiredWidth=true, bOverride_MinDesiredWidth=true, maxDesiredWidth=225, minDesiredHeight=44.3379, minDesiredWidth=225}
   slot {class=OverlaySlot, properties {horizontalAlignment=HAlign_Fill, verticalAlignment=VAlign_Fill, padding=<ltrb 5 0 0 0>}}
  }
  Overlay_1 {widgetClassPath=&Overlay, parent=SequenceSizeBox, widget=Overlay_1, properties {clipping=ClipToBounds}, slot {class=SizeBoxSlot}}
  Image_84 {
   widgetClassPath=&Image
   parent=Overlay_1
   widget=Image_84
   properties {brush {
     bIsDynamicallyLoaded=false
     drawAs=Image
     imageType=NoImage
     mirroring=NoMirror
     resourceName=None
     resourceObject=null
     tiling=NoTile
     imageSize=<32 32>
     margin=<ltrb 0 0 0 0>
     outlineSettings {
      bUseBrushTransparency=false
      roundingType=HalfHeightRadius
      width=0
      color {colorUseRule=UseColor_Specified, specifiedColor=<rgba 0 0 0 0>}
      cornerRadii=<0 0 0 0>
     }
     tintColor {colorUseRule=UseColor_Specified, specifiedColor=<rgba 0.0232 0.0232 0.0232 1>}
     uVRegion {bIsValid=false, max=<0 0>, min=<0 0>}
    }}
   slot {class=OverlaySlot, properties {horizontalAlignment=HAlign_Fill, verticalAlignment=VAlign_Fill, padding=<ltrb -22 0 0 0>}}
  }
  VerticalBox_1 {
   widgetClassPath=&VerticalBox
   parent=Overlay_1
   widget=VerticalBox_1
   slot {class=OverlaySlot, properties {horizontalAlignment=HAlign_Fill, padding=<ltrb 10 10 16 10>}}
  }
  HorizontalBox_1 {
   widgetClassPath=&HorizontalBox
   parent=VerticalBox_1
   widget=HorizontalBox_1
   slot {class=VerticalBoxSlot, properties {verticalAlignment=VAlign_Top}}
  }
  SizeBox_0 {
   widgetClassPath=&SizeBox
   parent=HorizontalBox_1
   widget=SizeBox_0
   properties {bOverride_MinDesiredHeight=true, bOverride_MinDesiredWidth=true, minDesiredHeight=22, minDesiredWidth=22}
   slot {class=HorizontalBoxSlot, properties {verticalAlignment=VAlign_Center, padding=<ltrb 0 0 10 0>}}
  }
  ToggleViewDa {
   widgetClassPath=&EditorUtilityCheckBox
   bIsVariable=true
   parent=SizeBox_0
   widget=ToggleViewDa
   properties {
    checkedState=Checked
    widgetStyle {
     checkBoxType=ToggleButton
     backgroundHoveredImage {
      bIsDynamicallyLoaded=false
      drawAs=RoundedBox
      imageType=NoImage
      mirroring=NoMirror
      resourceName=None
      resourceObject=null
      tiling=NoTile
      imageSize=<18 18>
      margin=<ltrb 0 0 0 0>
      outlineSettings {
       bUseBrushTransparency=false
       roundingType=FixedRadius
       width=1
       color {colorUseRule=UseColor_ColorTable, specifiedColor=<rgba 0 0 0 0>}
       cornerRadii=<3 3 3 3>
      }
      tintColor {colorUseRule=UseColor_Foreground, specifiedColor=<rgba 0.0048 0.0048 0.0048 1>}
      uVRegion {bIsValid=false, max=<0 0>, min=<0 0>}
     }
     backgroundImage {
      bIsDynamicallyLoaded=false
      drawAs=RoundedBox
      imageType=NoImage
      mirroring=NoMirror
      resourceName=None
      resourceObject=null
      tiling=NoTile
      imageSize=<18 18>
      margin=<ltrb 0 0 0 0>
      outlineSettings {
       bUseBrushTransparency=false
       roundingType=FixedRadius
       width=1
       color {colorUseRule=UseColor_ColorTable, specifiedColor=<rgba 0 0 0 0>}
       cornerRadii=<3 3 3 3>
      }
      tintColor {colorUseRule=UseColor_Foreground, specifiedColor=<rgba 0.0048 0.0048 0.0048 1>}
      uVRegion {bIsValid=false, max=<0 0>, min=<0 0>}
     }
     backgroundPressedImage {
      bIsDynamicallyLoaded=false
      drawAs=RoundedBox
      imageType=NoImage
      mirroring=NoMirror
      resourceName=None
      resourceObject=null
      tiling=NoTile
      imageSize=<18 18>
      margin=<ltrb 0 0 0 0>
      outlineSettings {
       bUseBrushTransparency=false
       roundingType=FixedRadius
       width=1
       color {colorUseRule=UseColor_ColorTable, specifiedColor=<rgba 0 0 0 0>}
       cornerRadii=<3 3 3 3>
      }
      tintColor {colorUseRule=UseColor_Specified, specifiedColor=<rgba 0.0048 0.0048 0.0048 1>}
      uVRegion {bIsValid=false, max=<0 0>, min=<0 0>}
     }
     borderBackgroundColor {colorUseRule=UseColor_Specified, specifiedColor=<rgba 1 1 1 1>}
     checkedForeground {colorUseRule=UseColor_Specified, specifiedColor=<rgba 1 1 1 1>}
     checkedHoveredForeground {colorUseRule=UseColor_Specified, specifiedColor=<rgba 1 1 1 1>}
     checkedHoveredImage {
      bIsDynamicallyLoaded=false
      drawAs=Image
      imageType=NoImage
      mirroring=NoMirror
      resourceName=None
      resourceObject=/Game/Athena/UI/InGame/Creative/PropertyEditor/Icon_Picker/Textures_V2/64x64/T_UI_IconLibrary_Eye_64
      tiling=NoTile
      imageSize=<64 64>
      margin=<ltrb 0 0 0 0>
      outlineSettings {
       bUseBrushTransparency=false
       roundingType=HalfHeightRadius
       width=0
       color {colorUseRule=UseColor_Specified, specifiedColor=<rgba 0 0 0 0>}
       cornerRadii=<0 0 0 0>
      }
      tintColor {colorUseRule=UseColor_Specified, specifiedColor=<rgba 1 1 1 1>}
      uVRegion {bIsValid=false, max=<0 0>, min=<0 0>}
     }
     checkedImage {
      bIsDynamicallyLoaded=false
      drawAs=Image
      imageType=NoImage
      mirroring=NoMirror
      resourceName=None
      resourceObject=&T_UI_IconLibrary_Eye_64
      tiling=NoTile
      imageSize=<15 15>
      margin=<ltrb 0 0 0 0>
      outlineSettings {
       bUseBrushTransparency=false
       roundingType=HalfHeightRadius
       width=0
       color {colorUseRule=UseColor_Specified, specifiedColor=<rgba 0 0 0 0>}
       cornerRadii=<0 0 0 0>
      }
      tintColor {colorUseRule=UseColor_Specified, specifiedColor=<rgba 1 1 1 1>}
      uVRegion {bIsValid=false, max=<0 0>, min=<0 0>}
     }
     checkedPressedForeground {colorUseRule=UseColor_Specified, specifiedColor=<rgba 1 1 1 1>}
     checkedPressedImage {
      bIsDynamicallyLoaded=false
      drawAs=Image
      imageType=NoImage
      mirroring=NoMirror
      resourceName=None
      resourceObject=&T_UI_IconLibrary_Eye_64
      tiling=NoTile
      imageSize=<15 15>
      margin=<ltrb 0 0 0 0>
      outlineSettings {
       bUseBrushTransparency=false
       roundingType=HalfHeightRadius
       width=0
       color {colorUseRule=UseColor_Specified, specifiedColor=<rgba 0 0 0 0>}
       cornerRadii=<0 0 0 0>
      }
      tintColor {colorUseRule=UseColor_Specified, specifiedColor=<rgba 1 1 1 1>}
      uVRegion {bIsValid=false, max=<0 0>, min=<0 0>}
     }
     checkedSlateSound {resourceObject=null}
     foregroundColor {colorUseRule=UseColor_Specified, specifiedColor=<rgba 1 1 1 1>}
     hoveredForeground {colorUseRule=UseColor_Specified, specifiedColor=<rgba 1 1 1 1>}
     hoveredSlateSound {resourceObject=null}
     padding=<ltrb 2 0 0 0>
     pressedForeground {colorUseRule=UseColor_Specified, specifiedColor=<rgba 1 1 1 1>}
     uncheckedHoveredImage {
      bIsDynamicallyLoaded=false
      drawAs=RoundedBox
      imageType=Vector
      mirroring=NoMirror
      resourceName=../../../Engine/Content/Slate/Starship/CoreWidgets/CheckBox/indeterminate.svg
      resourceObject=null
      tiling=NoTile
      imageSize=<21.6344 15.3298>
      margin=<ltrb 0 0 0 0>
      outlineSettings {
       bUseBrushTransparency=false
       roundingType=HalfHeightRadius
       width=0
       color {colorUseRule=UseColor_Specified, specifiedColor=<rgba 0 0 0 0>}
       cornerRadii=<0 0 0 0>
      }
      tintColor {colorUseRule=UseColor_Specified, specifiedColor=<rgba 0.9378 1 0.6936 1>}
      uVRegion {bIsValid=false, max=<0 0>, min=<0 0>}
     }
     uncheckedImage {
      bIsDynamicallyLoaded=false
      drawAs=RoundedBox
      imageType=Vector
      mirroring=NoMirror
      resourceName=../../../Engine/Content/Slate/Starship/CoreWidgets/CheckBox/indeterminate.svg
      resourceObject=null
      tiling=NoTile
      imageSize=<21.6344 15.3298>
      margin=<ltrb 0 0 0 0>
      outlineSettings {
       bUseBrushTransparency=false
       roundingType=HalfHeightRadius
       width=0
       color {colorUseRule=UseColor_Specified, specifiedColor=<rgba 0 0 0 0>}
       cornerRadii=<0 0 0 0>
      }
      tintColor {colorUseRule=UseColor_Specified, specifiedColor=<rgba 1 1 1 1>}
      uVRegion {bIsValid=false, max=<0 0>, min=<0 0>}
     }
     uncheckedPressedImage {
      bIsDynamicallyLoaded=false
      drawAs=RoundedBox
      imageType=Vector
      mirroring=NoMirror
      resourceName=../../../Engine/Content/Slate/Starship/CoreWidgets/CheckBox/indeterminate.svg
      resourceObject=null
      tiling=NoTile
      imageSize=<21.6344 15.3298>
      margin=<ltrb 0 0 0 0>
      outlineSettings {
       bUseBrushTransparency=false
       roundingType=HalfHeightRadius
       width=0
       color {colorUseRule=UseColor_Specified, specifiedColor=<rgba 0 0 0 0>}
       cornerRadii=<0 0 0 0>
      }
      tintColor {colorUseRule=UseColor_Specified, specifiedColor=<rgba 0.5417 0.5417 0.5417 1>}
      uVRegion {bIsValid=false, max=<0 0>, min=<0 0>}
     }
     uncheckedSlateSound {resourceObject=null}
     undeterminedForeground {colorUseRule=UseColor_Specified, specifiedColor=<rgba 1 1 1 1>}
     undeterminedHoveredImage {
      bIsDynamicallyLoaded=false
      drawAs=Image
      imageType=Vector
      mirroring=NoMirror
      resourceName=../../../Engine/Content/Slate/Starship/CoreWidgets/CheckBox/indeterminate.svg
      resourceObject=null
      tiling=NoTile
      imageSize=<16 16>
      margin=<ltrb 0 0 0 0>
      outlineSettings {
       bUseBrushTransparency=false
       roundingType=HalfHeightRadius
       width=0
       color {colorUseRule=UseColor_Specified, specifiedColor=<rgba 0 0 0 0>}
       cornerRadii=<0 0 0 0>
      }
      tintColor {colorUseRule=UseColor_Specified, specifiedColor=<rgba 0.0044 0.2384 1 1>}
      uVRegion {bIsValid=false, max=<0 0>, min=<0 0>}
     }
     undeterminedImage {
      bIsDynamicallyLoaded=false
      drawAs=Image
      imageType=Vector
      mirroring=NoMirror
      resourceName=../../../Engine/Content/Slate/Starship/CoreWidgets/CheckBox/indeterminate.svg
      resourceObject=null
      tiling=NoTile
      imageSize=<16 16>
      margin=<ltrb 0 0 0 0>
      outlineSettings {
       bUseBrushTransparency=false
       roundingType=HalfHeightRadius
       width=0
       color {colorUseRule=UseColor_Specified, specifiedColor=<rgba 0 0 0 0>}
       cornerRadii=<0 0 0 0>
      }
      tintColor {colorUseRule=UseColor_Specified, specifiedColor=<rgba 0 0.162 0.7454 1>}
      uVRegion {bIsValid=false, max=<0 0>, min=<0 0>}
     }
     undeterminedPressedImage {
      bIsDynamicallyLoaded=false
      drawAs=Image
      imageType=Vector
      mirroring=NoMirror
      resourceName=../../../Engine/Content/Slate/Starship/CoreWidgets/CheckBox/indeterminate.svg
      resourceObject=null
      tiling=NoTile
      imageSize=<16 16>
      margin=<ltrb 0 0 0 0>
      outlineSettings {
       bUseBrushTransparency=false
       roundingType=HalfHeightRadius
       width=0
       color {colorUseRule=UseColor_Specified, specifiedColor=<rgba 0 0 0 0>}
       cornerRadii=<0 0 0 0>
      }
      tintColor {colorUseRule=UseColor_Specified, specifiedColor=<rgba 0 0.0802 0.3515 1>}
      uVRegion {bIsValid=false, max=<0 0>, min=<0 0>}
     }
    }
   }
   slot {class=SizeBoxSlot}
  }
  CellName {
   widgetClassPath=&TextBlock
   bIsVariable=true
   parent=HorizontalBox_1
   widget=CellName
   properties {
    text=Sequence 01
    font {
     bForceMonospaced=false
     bMaterialIsStencil=false
     fontMaterial=null
     fontObject=&HeadingNow
     letterSpacing=0
     monospacedWidth=1
     size=7.5
     skewAmount=0
     typefaceFontName=75 Medium
     outlineSettings {
      bApplyOutlineToDropShadows=false
      bMiteredCorners=false
      bSeparateFillAlpha=false
      outlineMaterial=null
      outlineSize=0
      outlineColor=<rgba 0 0 0 1>
     }
    }
   }
   slot {class=HorizontalBoxSlot, properties {horizontalAlignment=HAlign_Left, verticalAlignment=VAlign_Center, padding=<ltrb 0 0 5 0>}}
  }
  SizeBox_1 {
   widgetClassPath=&SizeBox
   parent=HorizontalBox_1
   widget=SizeBox_1
   properties {bOverride_MaxDesiredWidth=true, maxDesiredHeight=18.3163, maxDesiredWidth=19.3278}
   slot {class=HorizontalBoxSlot, properties {horizontalAlignment=HAlign_Left, verticalAlignment=VAlign_Center}}
  }
  DropDownButton {
   widgetClassPath=&EditorUtilityButton
   parent=SizeBox_1
   widget=DropDownButton
   properties {widgetStyle {
     clickedSlateSound {resourceObject=null}
     disabled {
      bIsDynamicallyLoaded=false
      drawAs=RoundedBox
      imageType=NoImage
      mirroring=NoMirror
      resourceName=None
      resourceObject=/Game/UI/Foundation/Textures/Buttons/T-Button-Square
      tiling=NoTile
      imageSize=<32 32>
      margin=<ltrb 0.5 0.5 0.5 0.5>
      outlineSettings {
       bUseBrushTransparency=false
       roundingType=FixedRadius
       width=0
       color {colorUseRule=UseColor_ColorTable, specifiedColor=<rgba 0 0 0 0>}
       cornerRadii=<4 4 4 4>
      }
      tintColor {colorUseRule=UseColor_Specified, specifiedColor=<rgba 0.0395 0.0395 0.0395 1>}
      uVRegion {bIsValid=false, max=<0 0>, min=<0 0>}
     }
     disabledForeground {colorUseRule=UseColor_ColorTable, specifiedColor=<rgba 0 0 0 0>}
     hovered {
      bIsDynamicallyLoaded=false
      drawAs=RoundedBox
      imageType=NoImage
      mirroring=NoMirror
      resourceName=None
      resourceObject=&T-Button-Square
      tiling=NoTile
      imageSize=<15 15>
      margin=<ltrb 0.5 0.5 0.5 0.5>
      outlineSettings {
       bUseBrushTransparency=false
       roundingType=FixedRadius
       width=0
       color {colorUseRule=UseColor_ColorTable, specifiedColor=<rgba 0 0 0 0>}
       cornerRadii=<4 4 4 4>
      }
      tintColor {colorUseRule=UseColor_Specified, specifiedColor=<rgba 0.7344 0.4651 0 1>}
      uVRegion {bIsValid=false, max=<0 0>, min=<0 0>}
     }
     hoveredForeground {colorUseRule=UseColor_ColorTable, specifiedColor=<rgba 0 0 0 0>}
     hoveredSlateSound {resourceObject=null}
     normal {
      bIsDynamicallyLoaded=false
      drawAs=RoundedBox
      imageType=NoImage
      mirroring=NoMirror
      resourceName=None
      resourceObject=&T-Button-Square
      tiling=NoTile
      imageSize=<15 15>
      margin=<ltrb 0.5 0.5 0.5 0.5>
      outlineSettings {
       bUseBrushTransparency=false
       roundingType=FixedRadius
       width=0
       color {colorUseRule=UseColor_ColorTable, specifiedColor=<rgba 0 0 0 0>}
       cornerRadii=<4 4 4 4>
      }
      tintColor {colorUseRule=UseColor_Specified, specifiedColor=<rgba 1 0.6333 0 1>}
      uVRegion {bIsValid=false, max=<0 0>, min=<0 0>}
     }
     normalForeground {colorUseRule=UseColor_ColorTable, specifiedColor=<rgba 0 0 0 0>}
     normalPadding=<ltrb 2 1.5 2 1.5>
     pressed {
      bIsDynamicallyLoaded=false
      drawAs=RoundedBox
      imageType=NoImage
      mirroring=NoMirror
      resourceName=None
      resourceObject=&T-Button-Square
      tiling=NoTile
      imageSize=<15 15>
      margin=<ltrb 0.5 0.5 0.5 0.5>
      outlineSettings {
       bUseBrushTransparency=false
       roundingType=FixedRadius
       width=0
       color {colorUseRule=UseColor_ColorTable, specifiedColor=<rgba 0 0 0 0>}
       cornerRadii=<4 4 4 4>
      }
      tintColor {colorUseRule=UseColor_Specified, specifiedColor=<rgba 0.6458 0.409 0 1>}
      uVRegion {bIsValid=false, max=<0 0>, min=<0 0>}
     }
     pressedForeground {colorUseRule=UseColor_ColorTable, specifiedColor=<rgba 0 0 0 0>}
     pressedPadding=<ltrb 4 2.5 4 0.5>
     pressedSlateSound {resourceObject=null}
    }}
   slot {class=SizeBoxSlot}
  }
  Overlay_3 {
   widgetClassPath=&Overlay
   parent=DropDownButton
   widget=Overlay_3
   slot {class=ButtonSlot, properties {horizontalAlignment=HAlign_Fill, verticalAlignment=VAlign_Fill, padding=<ltrb 0 0 0 0>}}
  }
  ArrowDown {
   widgetClassPath=&Image
   bIsVariable=true
   parent=Overlay_3
   widget=ArrowDown
   properties {brush {
     bIsDynamicallyLoaded=false
     drawAs=Image
     imageType=NoImage
     mirroring=NoMirror
     resourceName=None
     resourceObject=/PilgrimQuickplay/StageManager/NeverCook/EditorUtilities/StageManagerEditor/Textures/T_EditorDownArrow
     tiling=NoTile
     imageSize=<15 15>
     margin=<ltrb 0 0 0 0>
     outlineSettings {
      bUseBrushTransparency=false
      roundingType=HalfHeightRadius
      width=0
      color {colorUseRule=UseColor_Specified, specifiedColor=<rgba 0 0 0 0>}
      cornerRadii=<0 0 0 0>
     }
     tintColor {colorUseRule=UseColor_Specified, specifiedColor=<rgba 1 1 1 1>}
     uVRegion {bIsValid=false, max=<0 0>, min=<0 0>}
    }}
   slot {class=OverlaySlot}
  }
  ArrowUp {
   widgetClassPath=&Image
   bIsVariable=true
   parent=Overlay_3
   widget=ArrowUp
   properties {
    visibility=Collapsed
    brush {
     bIsDynamicallyLoaded=false
     drawAs=Image
     imageType=NoImage
     mirroring=NoMirror
     resourceName=None
     resourceObject=&T_EditorDownArrow
     tiling=NoTile
     imageSize=<15 15>
     margin=<ltrb 0 0 0 0>
     outlineSettings {
      bUseBrushTransparency=false
      roundingType=HalfHeightRadius
      width=0
      color {colorUseRule=UseColor_Specified, specifiedColor=<rgba 0 0 0 0>}
      cornerRadii=<0 0 0 0>
     }
     tintColor {colorUseRule=UseColor_Specified, specifiedColor=<rgba 1 1 1 1>}
     uVRegion {bIsValid=false, max=<0 0>, min=<0 0>}
    }
    renderTransform {angle=0, scale=<1 -1>, shear=<0 0>, translation=<0 0>}
   }
   slot {class=OverlaySlot}
  }
  Spacer_166 {widgetClassPath=&Spacer, parent=VerticalBox_1, widget=Spacer_166, properties {size=<1 5>}, slot {class=VerticalBoxSlot}}
  SequenceNameText {
   widgetClassPath=&TextBlock
   bIsVariable=true
   parent=VerticalBox_1
   widget=SequenceNameText
   properties {
    text=Stage_Seq_Lights_01_01 LALALALA
    font {
     bForceMonospaced=false
     bMaterialIsStencil=false
     fontMaterial=null
     fontObject=&HeadingNow
     letterSpacing=0
     monospacedWidth=1
     size=7.5
     skewAmount=0
     typefaceFontName=75 Medium
     outlineSettings {
      bApplyOutlineToDropShadows=false
      bMiteredCorners=false
      bSeparateFillAlpha=false
      outlineMaterial=null
      outlineSize=0
      outlineColor=<rgba 0 0 0 1>
     }
    }
   }
   slot {class=VerticalBoxSlot, properties {verticalAlignment=VAlign_Center}}
  }
  Spacer_111 {widgetClassPath=&Spacer, parent=VerticalBox_0, widget=Spacer_111, slot {class=VerticalBoxSlot}}
  LoopingEventsSideBar {
   widgetClassPath=&SizeBox
   bIsVariable=true
   parent=VerticalBox_0
   widget=LoopingEventsSideBar
   properties {bOverride_MinDesiredWidth=true, minDesiredWidth=182}
   slot {class=VerticalBoxSlot, properties {padding=<ltrb 5 0 4 0>, size {sizeRule=Fill, value=1}}}
  }
  Overlay_2 {
   widgetClassPath=&Overlay
   parent=LoopingEventsSideBar
   widget=Overlay_2
   properties {clipping=ClipToBounds}
   slot {class=SizeBoxSlot}
  }
  Image {
   widgetClassPath=&Image
   parent=Overlay_2
   widget=Image
   properties {brush {
     bIsDynamicallyLoaded=false
     drawAs=Image
     imageType=NoImage
     mirroring=NoMirror
     resourceName=None
     resourceObject=null
     tiling=NoTile
     imageSize=<32 32>
     margin=<ltrb 0 0 0 0>
     outlineSettings {
      bUseBrushTransparency=false
      roundingType=HalfHeightRadius
      width=0
      color {colorUseRule=UseColor_Specified, specifiedColor=<rgba 0 0 0 0>}
      cornerRadii=<0 0 0 0>
     }
     tintColor {colorUseRule=UseColor_Specified, specifiedColor=<rgba 0.0232 0.0232 0.0232 1>}
     uVRegion {bIsValid=false, max=<0 0>, min=<0 0>}
    }}
   slot {class=OverlaySlot, properties {horizontalAlignment=HAlign_Fill, verticalAlignment=VAlign_Fill, padding=<ltrb -22 0 0 0>}}
  }
  LoopEventsText {
   widgetClassPath=&TextBlock
   parent=Overlay_2
   widget=LoopEventsText
   properties {
    justification=Center
    text=LoopingEvents
    colorAndOpacity {colorUseRule=UseColor_Specified, specifiedColor=<rgba 1 1 1 0.508>}
    font {
     bForceMonospaced=false
     bMaterialIsStencil=false
     fontMaterial=null
     fontObject=&HeadingNow
     letterSpacing=0
     monospacedWidth=1
     size=7.5
     skewAmount=0
     typefaceFontName=75 Medium
     outlineSettings {
      bApplyOutlineToDropShadows=false
      bMiteredCorners=false
      bSeparateFillAlpha=false
      outlineMaterial=null
      outlineSize=0
      outlineColor=<rgba 0 0 0 1>
     }
    }
   }
   slot {class=OverlaySlot, properties {padding=<ltrb 20 10 16 10>}}
  }
  EventsHorizontalBox {
   widgetClassPath=&HorizontalBox
   bIsVariable=true
   parent=HorizontalBox_0
   widget=EventsHorizontalBox
   slot {class=HorizontalBoxSlot, properties {horizontalAlignment=HAlign_Left, size {sizeRule=Fill, value=1}}}
  }
  EWBP_SparksSequencePool_TimelineEventGroup_1 {
   widgetClassPath=&/EWBP_SparksSequencePool_TimelineEventGroup
   bIsVariable=true
   parent=EventsHorizontalBox
   widget=EWBP_SparksSequencePool_TimelineEventGroup_1
   properties {delayValue=1.4549, eventSize=175.4671}
   slot {class=HorizontalBoxSlot}
  }
  EWBP_SparksSequencePool_TimelineEventGroup {
   widgetClassPath=&EWBP_SparksSequencePool_TimelineEventGroup
   bIsVariable=true
   parent=EventsHorizontalBox
   widget=EWBP_SparksSequencePool_TimelineEventGroup
   slot {class=HorizontalBoxSlot}
  }
  EWBP_SparksSequencePool_TimelineEventGroup_2 {
   widgetClassPath=&EWBP_SparksSequencePool_TimelineEventGroup
   bIsVariable=true
   parent=EventsHorizontalBox
   widget=EWBP_SparksSequencePool_TimelineEventGroup_2
   slot {class=HorizontalBoxSlot}
  }
  NoStepsText {
   widgetClassPath=&TextBlock
   bIsVariable=true
   parent=HorizontalBox_0
   widget=NoStepsText
   properties {
    text=There is no steps on this sequence.
    visibility=Collapsed
    font {
     bForceMonospaced=false
     bMaterialIsStencil=false
     fontMaterial=null
     fontObject=&HeadingNow
     letterSpacing=0
     monospacedWidth=1
     size=7.5
     skewAmount=0
     typefaceFontName=75 Medium
     outlineSettings {
      bApplyOutlineToDropShadows=false
      bMiteredCorners=false
      bSeparateFillAlpha=false
      outlineMaterial=null
      outlineSize=0
      outlineColor=<rgba 0 0 0 1>
     }
    }
   }
   slot {class=HorizontalBoxSlot, properties {verticalAlignment=VAlign_Center, padding=<ltrb 10 0 10 0>}}
  }
  NewStepArea {
   widgetClassPath=&Overlay
   bIsVariable=true
   parent=HorizontalBox_0
   widget=NewStepArea
   properties {visibility=Collapsed}
   slot {class=HorizontalBoxSlot, properties {padding=<ltrb 0 2 0 0>}}
  }
  Image_3 {
   widgetClassPath=&Image
   parent=NewStepArea
   widget=Image_3
   properties {
    visibility=SelfHitTestInvisible
    brush {
     bIsDynamicallyLoaded=false
     drawAs=Image
     imageType=NoImage
     mirroring=NoMirror
     resourceName=None
     resourceObject=null
     tiling=NoTile
     imageSize=<32 32>
     margin=<ltrb 0 0 0 0>
     outlineSettings {
      bUseBrushTransparency=false
      roundingType=HalfHeightRadius
      width=0
      color {colorUseRule=UseColor_Specified, specifiedColor=<rgba 0 0 0 0>}
      cornerRadii=<0 0 0 0>
     }
     tintColor {colorUseRule=UseColor_Specified, specifiedColor=<rgba 0.0625 0.0625 0.0625 1>}
     uVRegion {bIsValid=false, max=<0 0>, min=<0 0>}
    }
   }
   slot {class=OverlaySlot, properties {horizontalAlignment=HAlign_Fill, verticalAlignment=VAlign_Fill, padding=<ltrb 0 0 -2 0>}}
  }
  NewStepButton {
   widgetClassPath=&EditorUtilityButton
   parent=NewStepArea
   widget=NewStepButton
   properties {widgetStyle {
     clickedSlateSound {resourceObject=null}
     disabled {
      bIsDynamicallyLoaded=false
      drawAs=RoundedBox
      imageType=NoImage
      mirroring=NoMirror
      resourceName=None
      resourceObject=&T-Button-Square
      tiling=NoTile
      imageSize=<32 32>
      margin=<ltrb 0.5 0.5 0.5 0.5>
      outlineSettings {
       bUseBrushTransparency=false
       roundingType=FixedRadius
       width=0
       color {colorUseRule=UseColor_ColorTable, specifiedColor=<rgba 0 0 0 0>}
       cornerRadii=<4 4 4 4>
      }
      tintColor {colorUseRule=UseColor_Specified, specifiedColor=<rgba 0.0395 0.0395 0.0395 1>}
      uVRegion {bIsValid=false, max=<0 0>, min=<0 0>}
     }
     disabledForeground {colorUseRule=UseColor_ColorTable, specifiedColor=<rgba 0 0 0 0>}
     hovered {
      bIsDynamicallyLoaded=false
      drawAs=RoundedBox
      imageType=NoImage
      mirroring=NoMirror
      resourceName=None
      resourceObject=&T-Button-Square
      tiling=NoTile
      imageSize=<32 32>
      margin=<ltrb 0.5 0.5 0.5 0.5>
      outlineSettings {
       bUseBrushTransparency=false
       roundingType=FixedRadius
       width=0
       color {colorUseRule=UseColor_ColorTable, specifiedColor=<rgba 0 0 0 0>}
       cornerRadii=<4 4 4 4>
      }
      tintColor {colorUseRule=UseColor_Specified, specifiedColor=<rgba 0.7344 0.4651 0 1>}
      uVRegion {bIsValid=false, max=<0 0>, min=<0 0>}
     }
     hoveredForeground {colorUseRule=UseColor_ColorTable, specifiedColor=<rgba 0 0 0 0>}
     hoveredSlateSound {resourceObject=null}
     normal {
      bIsDynamicallyLoaded=false
      drawAs=RoundedBox
      imageType=NoImage
      mirroring=NoMirror
      resourceName=None
      resourceObject=&T-Button-Square
      tiling=NoTile
      imageSize=<32 32>
      margin=<ltrb 0.5 0.5 0.5 0.5>
      outlineSettings {
       bUseBrushTransparency=false
       roundingType=FixedRadius
       width=0
       color {colorUseRule=UseColor_ColorTable, specifiedColor=<rgba 0 0 0 0>}
       cornerRadii=<4 4 4 4>
      }
      tintColor {colorUseRule=UseColor_Specified, specifiedColor=<rgba 0.7991 0.5583 0.0003 1>}
      uVRegion {bIsValid=false, max=<0 0>, min=<0 0>}
     }
     normalForeground {colorUseRule=UseColor_ColorTable, specifiedColor=<rgba 0 0 0 0>}
     normalPadding=<ltrb 12 1.5 12 1.5>
     pressed {
      bIsDynamicallyLoaded=false
      drawAs=RoundedBox
      imageType=NoImage
      mirroring=NoMirror
      resourceName=None
      resourceObject=&T-Button-Square
      tiling=NoTile
      imageSize=<32 32>
      margin=<ltrb 0.5 0.5 0.5 0.5>
      outlineSettings {
       bUseBrushTransparency=false
       roundingType=FixedRadius
       width=0
       color {colorUseRule=UseColor_ColorTable, specifiedColor=<rgba 0 0 0 0>}
       cornerRadii=<4 4 4 4>
      }
      tintColor {colorUseRule=UseColor_Specified, specifiedColor=<rgba 0.6458 0.409 0 1>}
      uVRegion {bIsValid=false, max=<0 0>, min=<0 0>}
     }
     pressedForeground {colorUseRule=UseColor_ColorTable, specifiedColor=<rgba 0 0 0 0>}
     pressedPadding=<ltrb 12 2.5 12 0.5>
     pressedSlateSound {resourceObject=null}
    }}
   slot {class=OverlaySlot, properties {padding=<ltrb 5 2 5 0>}}
  }
  HorizontalBox_2 {
   widgetClassPath=&HorizontalBox
   parent=NewStepButton
   widget=HorizontalBox_2
   slot {class=ButtonSlot, properties {horizontalAlignment=HAlign_Left}}
  }
  ExternalIcon_1 {
   widgetClassPath=&Image
   parent=HorizontalBox_2
   widget=ExternalIcon_1
   properties {brush {
     bIsDynamicallyLoaded=false
     drawAs=Image
     imageType=NoImage
     mirroring=NoMirror
     resourceName=None
     resourceObject=/PilgrimQuickplay/StageManager/NeverCook/EditorUtilities/StageManagerEditor/Textures/MI_SparksEditorUI_PlusMinus_NoBorder
     tiling=NoTile
     imageSize=<15 15>
     margin=<ltrb 0 0 0 0>
     outlineSettings {
      bUseBrushTransparency=false
      roundingType=HalfHeightRadius
      width=0
      color {colorUseRule=UseColor_Specified, specifiedColor=<rgba 0 0 0 0>}
      cornerRadii=<0 0 0 0>
     }
     tintColor {colorUseRule=UseColor_Specified, specifiedColor=<rgba 1 1 1 1>}
     uVRegion {bIsValid=false, max=<0 0>, min=<0 0>}
    }}
   slot {class=HorizontalBoxSlot, properties {horizontalAlignment=HAlign_Center, verticalAlignment=VAlign_Center, padding=<ltrb 0 0 3 0>}}
  }
  TextBlock_1 {
   widgetClassPath=&TextBlock
   parent=HorizontalBox_2
   widget=TextBlock_1
   properties {
    text=New Step
    font {
     bForceMonospaced=false
     bMaterialIsStencil=false
     fontMaterial=null
     fontObject=&HeadingNow
     letterSpacing=0
     monospacedWidth=1
     size=7.5
     skewAmount=0
     typefaceFontName=75 Medium
     outlineSettings {
      bApplyOutlineToDropShadows=false
      bMiteredCorners=false
      bSeparateFillAlpha=false
      outlineMaterial=null
      outlineSize=0
      outlineColor=<rgba 0 0 0 1>
     }
    }
   }
   slot {class=HorizontalBoxSlot, properties {verticalAlignment=VAlign_Center}}
  }
  EWBP_SparksTimelineSubmenu {
   widgetClassPath=&/EWBP_SparksTimelineSubmenu
   bIsVariable=true
   parent=Overlay_0
   widget=EWBP_SparksTimelineSubmenu
   properties {visibility=Collapsed, renderTransform {angle=0, scale=<1 1>, shear=<0 0>, translation=<37 34>}}
   slot {class=OverlaySlot}
  }
 ]
}
