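# Editor utility widget that edits Quickplay stage-manager show data: moods, MIDI event maps, gameplay-state event maps and their sequence pools.
# Combo boxes pick the mood, MIDI map, event and sequence, while popups add, change, rename or delete events and assets.
# New assets are duplicated from _ReadOnly templates; most map edits use temporary copies written back inside editor transactions.
# FX manager events open in a separate event-data window linked to this editor.
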EditorUtilityWidgetBlueprint /PilgrimQuickplay/StageManager/NeverCook/EditorUtilities/StageManagerEditor/SparksStageManagerTimelineEditor
graphs [
 ChangeAddTextEventPopup {nodes [
   n1 = "On Selection Changed (NewTextEventCombobox)"!() -> (OutputDelegate, then, SelectedItem:String, SelectionType:ESelectInfo Enum) {is=title K2Node_ComponentBoundEvent}
   > Branch(Condition="Not Equal Exactly (String)"(A=@n1.SelectedItem, B=aaaa))
   > Set NewEventSelected(NewEventSelected:String=Find(TargetMap=Get NewEventsEnumToStringMap() -> (NewEventsEnumToStringMap:Map of Strings to Strings), Key=@n1.SelectedItem)) -> (Output_Get:String)
   "On Pressed (NewTextEventSelectAndAdd)"() {is=title K2Node_ComponentBoundEvent}
   > BeginTransaction(
    Context=AddNewEventOnStageManager
    Description="INVTEXT(\"Add new Event on Stage manager\")"
    PrimaryObject=GetCurrentMIDIMap!() -> (SequenceMIDIMap:BP Mojo Stage Map Midi Sequence Data Asset Object Reference, FilteredName:String)
   )
   > n2 = Sequence()
   > n3 = Branch(Condition=Get IsMidiEvent() -> (IsMidiEvent:bool))
   > SetVisibility(self=Get TextEventsChangerGroup() -> (TextEventsChangerGroup:Overlay Object Reference), InVisibility=Collapsed)
   n4 = "To Integer (Byte)"!(InByte:Byte=GetStageManagerEventsMidiTextEnumFromString!(MidiTextEventString:String=Get NewEventSelected() -> (NewEventSelected:String), __WorldContext:Object Reference) -> (MidiTextEventEnum:ESparksStageManager_Events_Midi_Text_Events Enum=NewEnumerator0)) -> (ReturnValue:int=0)
   > Add!(
    execute
    TargetMap:Map of ESparksMidiTextEvents Enums to Stage Show Sequence Pool Object References=Get TempMidiEventMap() -> (TempMidiEventMap:Map of ESparksMidiTextEvents Enums to Stage Show Sequence Pool Object References=None)
    Key:"ESparksMidiTextEvents Enum (by ref)"=Byte to Enum ESparksMidiTextEvents!(Byte:Byte="To Byte (Integer)"!(InInt:int=@n4) -> (ReturnValue:Byte=0)) -> (ReturnValue:ESparksMidiTextEvents Enum) {is=title K2Node_CastByteToEnum}
    Value:"Stage Show Sequence Pool Object Reference (by ref)"
   ) -> (then) {is=title K2Node_CallFunction}
   n5 = Get TempMidiEventMap() -> (TempMidiEventMap=None)
   > Set LastEventIndex(LastEventIndex:int=int - int(A=Length!(TargetMap:Map of ESparksMidiTextEvents Enums to Stage Show Sequence Pool Object References=@n5) -> (ReturnValue:int=0) {is=title K2Node_CallFunction})) -> (Output_Get:int=0)
   SetVisibility(execute=@n3.else, self=Get TextEventsChangerGroup(), InVisibility=Collapsed)
   > Add(
    TargetMap:Map of EPilgrimQuickplayState Enums to Stage Show Sequence Pool Object References=Get TempQuickplayEventMap() -> (TempQuickplayEventMap:Map of EPilgrimQuickplayState Enums to Stage Show Sequence Pool Object References=Invalid)
    Key:"EPilgrimQuickplayState Enum (by ref)"=GetStageManageQPEventsEnumFromString!(QPEventString:String=Get NewEventSelected(), __WorldContext:Object Reference) -> (QPEventEnum:EPilgrimQuickplayState Enum=Invalid)
   )
   n6 = Get TempQuickplayEventMap() -> (TempQuickplayEventMap=Invalid)
   > Set LastEventIndex(LastEventIndex=int - int(A=Length(TargetMap:Map of EPilgrimQuickplayState Enums to Stage Show Sequence Pool Object References=@n6) -> (ReturnValue=0))) -> (Output_Get=0)
   Set EventMap!(
    execute=@n2.then_1
    EventMap:Map of ESparksMidiTextEvents Enums to Stage Show Sequence Pool Object References=Get TempMidiEventMap() -> (TempMidiEventMap=None)
    self:Stage Show MIDIEvent Map Object Reference=GetCurrentMIDIMap()
   ) -> (then, Output_Get:Map of ESparksMidiTextEvents Enums to Stage Show Sequence Pool Object References=None)
   > Set EventMap(
    EventMap:Map of EPilgrimQuickplayState Enums to Stage Show Sequence Pool Object References=Get TempQuickplayEventMap() -> (TempQuickplayEventMap=Invalid)
    self:Stage Show Gameplay Event Map Object Reference=Get GamePlayEventData() -> (GamePlayEventData:BP Mojo Stage Map Gameplay Sequence Data Asset Object Reference)
   ) -> (Output_Get:Map of EPilgrimQuickplayState Enums to Stage Show Sequence Pool Object References=Invalid)
   > TransactObject(Object=GetCurrentMIDIMap())
   > EndTransaction()
   > Set CurrentSequenceIndex(CurrentSequenceIndex:int=0) -> (Output_Get:int=0)
   > ForceRefresh()
   "On Pressed (NewTextEventSelectAndChange)"() {is=title K2Node_ComponentBoundEvent}
   > BeginTransaction(Context=ChangeTextEventOnStageManager, Description="INVTEXT(\"Change Text  Event on Stage manager\")", PrimaryObject=GetCurrentMIDIMap())
   > SetVisibility(self=Get TextEventsChangerGroup(), InVisibility=Collapsed)
   > n7 = Sequence()
   > n8 = Branch(Condition=Get IsMidiEvent())
   n9 = Get TempMidiEventMap() -> (TempMidiEventMap=None)
   > Set LastSequencePool(LastSequencePool:Stage Show Sequence Pool Object Reference=Find(
     TargetMap:Map of ESparksMidiTextEvents Enums to Stage Show Sequence Pool Object References=@n9
     Key:"ESparksMidiTextEvents Enum (by ref)"=Get MidiTextEvents() -> (MidiTextEvents:ESparksMidiTextEvents Enum=None)
    ) -> (Value:Stage Show Sequence Pool Object Reference)) -> (Output_Get:Stage Show Sequence Pool Object Reference)
   > Remove!(
    execute
    TargetMap:Map of ESparksMidiTextEvents Enums to Stage Show Sequence Pool Object References=@n9
    Key:"ESparksMidiTextEvents Enum (by ref)"=Get MidiTextEvents() -> (MidiTextEvents=None)
   ) -> (then, ReturnValue:bool)
   n10 = "To Integer (Byte)"(InByte=GetStageManagerEventsMidiTextEnumFromString(MidiTextEventString=Get NewEventSelected()) -> (MidiTextEventEnum=NewEnumerator0)) -> (ReturnValue=0)
   > Add(
    TargetMap:Map of ESparksMidiTextEvents Enums to Stage Show Sequence Pool Object References=@n9
    Key:"ESparksMidiTextEvents Enum (by ref)"=Byte to Enum ESparksMidiTextEvents(Byte="To Byte (Integer)"(InInt=@n10) -> (ReturnValue=0))
    Value=Get LastSequencePool() -> (LastSequencePool:Stage Show Sequence Pool Object Reference)
   )
   n11 = Get TempQuickplayEventMap() -> (TempQuickplayEventMap=Invalid)
   Set LastSequencePool(
    execute=@n8.else
    LastSequencePool=Find(
     TargetMap:Map of EPilgrimQuickplayState Enums to Stage Show Sequence Pool Object References=@n11
     Key:"EPilgrimQuickplayState Enum (by ref)"=Get QuickPlayEvents() -> (QuickPlayEvents:EPilgrimQuickplayState Enum=Invalid)
    )
   )
   > Remove(
    TargetMap:Map of EPilgrimQuickplayState Enums to Stage Show Sequence Pool Object References=@n11
    Key:"EPilgrimQuickplayState Enum (by ref)"=Get QuickPlayEvents() -> (QuickPlayEvents=Invalid)
   )
   > Add(
    TargetMap:Map of EPilgrimQuickplayState Enums to Stage Show Sequence Pool Object References=@n11
    Key:"EPilgrimQuickplayState Enum (by ref)"=GetStageManageQPEventsEnumFromString(QPEventString=Get NewEventSelected()) -> (QPEventEnum=Invalid)
    Value=Get LastSequencePool()
   )
   Set EventMap(
    execute=@n7.then_1
    EventMap:Map of ESparksMidiTextEvents Enums to Stage Show Sequence Pool Object References=Get TempMidiEventMap() -> (TempMidiEventMap=None)
    self:Stage Show MIDIEvent Map Object Reference=GetCurrentMIDIMap()
   ) -> (Output_Get:Map of ESparksMidiTextEvents Enums to Stage Show Sequence Pool Object References=None)
   > Set EventMap(
    EventMap:Map of EPilgrimQuickplayState Enums to Stage Show Sequence Pool Object References=Get TempQuickplayEventMap() -> (TempQuickplayEventMap=Invalid)
    self:Stage Show Gameplay Event Map Object Reference=Get GamePlayEventData()
   ) -> (Output_Get:Map of EPilgrimQuickplayState Enums to Stage Show Sequence Pool Object References=Invalid)
   > TransactObject(Object=GetCurrentMIDIMap())
   > EndTransaction()
   > Set LastEventIndex(LastEventIndex=FindOptionIndex!(
     self:"Searchable Combo Box (String) Object Reference"=Get KeySlotComboBox() -> (KeySlotComboBox:"Searchable Combo Box (String) Object Reference")
     Option:String=Get CurrentTextEvent() -> (CurrentTextEvent:String)
    ) -> (ReturnValue:int=0)) -> (Output_Get=0)
   > ForceRefresh()
   "On Pressed (CloseNewTextEventWindowButton)"() {is=title K2Node_ComponentBoundEvent}
   > SetVisibility(self=Get TextEventsChangerGroup(), InVisibility=Collapsed)
   "On Pressed (DeleteEventButton)"() {is=title K2Node_ComponentBoundEvent}
   > BeginTransaction(Context=DeleteTextEventOnStageManager, Description="INVTEXT(\"Delete Text  Event on Stage manager\")", PrimaryObject=GetCurrentMIDIMap())
   > n12 = Sequence()
   > n13 = Branch(Condition=Get IsMidiEvent())
   > Remove(
    TargetMap:Map of ESparksMidiTextEvents Enums to Stage Show Sequence Pool Object References=Get TempMidiEventMap() -> (TempMidiEventMap=None)
    Key:"ESparksMidiTextEvents Enum (by ref)"=Get MidiTextEvents() -> (MidiTextEvents=None)
   )
   Remove(
    execute=@n13.else
    TargetMap:Map of EPilgrimQuickplayState Enums to Stage Show Sequence Pool Object References=Get TempQuickplayEventMap() -> (TempQuickplayEventMap=Invalid)
    Key:"EPilgrimQuickplayState Enum (by ref)"=Get QuickPlayEvents() -> (QuickPlayEvents=Invalid)
   )
   Set EventMap(
    execute=@n12.then_1
    EventMap:Map of ESparksMidiTextEvents Enums to Stage Show Sequence Pool Object References=Get TempMidiEventMap() -> (TempMidiEventMap=None)
    self:Stage Show MIDIEvent Map Object Reference=GetCurrentMIDIMap()
   ) -> (Output_Get:Map of ESparksMidiTextEvents Enums to Stage Show Sequence Pool Object References=None)
   > Set EventMap(
    EventMap:Map of EPilgrimQuickplayState Enums to Stage Show Sequence Pool Object References=Get TempQuickplayEventMap() -> (TempQuickplayEventMap=Invalid)
    self:Stage Show Gameplay Event Map Object Reference=Get GamePlayEventData()
   ) -> (Output_Get:Map of EPilgrimQuickplayState Enums to Stage Show Sequence Pool Object References=Invalid)
   > TransactObject(Object=GetCurrentMIDIMap())
   > EndTransaction()
   > Set LastEventIndex(LastEventIndex=0) -> (Output_Get=0)
   > ForceRefresh()
   "On Pressed (RootAdd)"() {is=title K2Node_ComponentBoundEvent}
   > PopulateEventCombobox()
   > SetVisibility(self=Get TextEventsChangerGroup(), InVisibility=Visible)
  ]}
 ChangeCreateDAPopup {nodes [
   n1 = "On Text Changed (NewFxManagerNameTextBox)"!() -> (OutputDelegate, then, Text:"Text (by ref)") {is=title K2Node_ComponentBoundEvent}
   n2 = TextTrimPrecedingAndTrailing!(InText:"Text (by ref)"=@n1.Text) -> (ReturnValue:Text)
   > n3 = Branch(Condition=TextIsEmpty(InText=@n2))
   > n4 = Set OutputFileName(OutputFileName:String) -> (Output_Get:String)
   n5 = Get PropertyName() -> (PropertyName:Name=None)
   n6 = OR Boolean(A="Equal (Name)"(A=@n5, B=MojoStage_MIDISequenceData), B="Equal (Name)"(A=@n5, B=MojoStage_Mood))
   "SetText (Text)"(
    execute=[@n4.then, @n8.then]
    self=Get DANamePreviewText() -> (DANamePreviewText:Text Object Reference)
    InText="To Text (String)"(InString=GrabNewAssetName!(IgnoreFirstIndex:bool=@n6, WarningDuplicated:bool=@n6) -> (OutputName:String))
   )
   n7 = GetCurrentData!() -> (ReturnValue:Object Reference, AssetBaseName:String, GroupName:String)
   n8 = Set OutputFileName(
    execute=@n3.else
    OutputFileName=Append!(A=Split(SourceString=@n7.AssetBaseName, InStr=_, SearchDir=FromEnd), B=_, C:String="To String (Text)"(InText=@n2)) -> (ReturnValue)
   )
   "On Pressed (CreateChangeButton)"() {is=title K2Node_ComponentBoundEvent}
   n9 = GetCurrentData()
   > n10 = DuplicateAsset!(
    execute
    self:Editor Asset Subsystem Object Reference=Get EditorAssetSubsystem() -> (ReturnValue:Editor Asset Subsystem Object Reference) {is=title K2Node_GetEditorSubsystem}
    SourceAssetPath:String=Append!(A=Get AssetsBasePath() -> (AssetsBasePath:String), B=Append(A=@n9.AssetBaseName, B=_ReadOnly)) -> (ReturnValue)
    DestinationAssetPath:String=Append(A=GetAssetDestinationPath!() -> (FinalPath:String), B=GrabNewAssetName(IgnoreFirstIndex=false, WarningDuplicated=false))
   ) -> (then, ReturnValue:Object Reference)
   > n11 = Sequence()
   > n12 = Cast To MojoFxManagerEvent_DataAsset!(execute, Object=@n10) -> (then, CastFailed, AsMojo Fx Manager Event Data Asset:Mojo Fx Manager Event Data Asset Object Reference) {is=title K2Node_DynamicCast}
   > Set LoopingEvent!(execute, LoopingEvent:bool=Get IsLoopingEvent() -> (IsLoopingEvent:bool), self:Stage Show Event Object Reference=@n12) -> (then, Output_Get:bool)
   > TransactObject(Object=@n12)
   UpdateSequencePool!(execute=@n11.then_1, self, ReturnObject:Object Reference=@n10, IsNewAsset:bool=false) -> (then)
   > SetVisibility(self=Get NewSequenceEventWindow() -> (NewSequenceEventWindow:Overlay Object Reference), InVisibility=Collapsed)
   > Call On Asset Changed!(execute:Exec, self:Sparks Stage Manager Timeline Editor Object Reference, ReturnObj:Object Reference=@n10, IsNewFile:bool) -> (then:Exec) {is=title K2Node_CallDelegate}
   > Set CurrentEventSelected(CurrentEventSelected:String) -> (Output_Get:String)
   "On Pressed (AddSequenceDA)"() {is=title K2Node_ComponentBoundEvent}
   > SetVisibility(
    self=Get NewSequenceEventChangeButton() -> (NewSequenceEventChangeButton:Editor Utility Button Object Reference)
    InVisibility=Select(Option 0=Visible, Option 1=Collapsed, Index=Get NoSequencePool() -> (NoSequencePool:bool))
   )
   > Set CurrentEventSelected(CurrentEventSelected=Get CurrentSequence() -> (CurrentSequence:String))
   n13 = FindOptionIndex(
    self=Get SequencesDataListComboBox() -> (SequencesDataListComboBox:"Searchable Combo Box (String) Object Reference")
    Option=Get CurrentSequence()
   ) -> (ReturnValue=0)
   n14 = Get Sequences!(self:Stage Show Sequence Pool Object Reference=Find(
     TargetMap:Map of ESparksMidiTextEvents Enums to Stage Show Sequence Pool Object References=Get TempMidiEventMap() -> (TempMidiEventMap=None)
     Key:"ESparksMidiTextEvents Enum (by ref)"=Get MidiTextEvents() -> (MidiTextEvents=None)
    )) -> (Sequences:Array of Stage Show Sequence Object References)
   n15 = FindOptionIndex(self=Get SequencesDataListComboBox(), Option=Get CurrentSequence()) -> (ReturnValue=0)
   n16 = Get Sequences(self=Find(
     TargetMap:Map of EPilgrimQuickplayState Enums to Stage Show Sequence Pool Object References=Get TempQuickplayEventMap() -> (TempQuickplayEventMap=Invalid)
     Key:"EPilgrimQuickplayState Enum (by ref)"=Get QuickPlayEvents() -> (QuickPlayEvents=Invalid)
    ))
   > Set MojoStage_Sequences PoolData(MojoStage_Sequences PoolData:Stage Show Sequence Object Reference=Select(
     Option 0:Stage Show Sequence Object Reference="Get (a copy)"(Array:Array of Stage Show Sequence Object References=@n16, Dimension 1=@n15) -> (Output:Stage Show Sequence Object Reference)
     Option 1:Stage Show Sequence Object Reference="Get (a copy)"(Array=@n14, Dimension 1=@n13)
     Index=Get IsMidiEvent()
    ) -> (ReturnValue:Stage Show Sequence Object Reference)) -> (Output_Get:Stage Show Sequence Object Reference)
   n17 = Get QuickPlayEvents() -> (QuickPlayEvents=Invalid)
   n18 = Get TempQuickplayEventMap() -> (TempQuickplayEventMap=Invalid)
   n19 = Get MidiTextEvents() -> (MidiTextEvents=None)
   n20 = Get TempMidiEventMap() -> (TempMidiEventMap=None)
   n21 = Get Sequences(self=Select(
     Option 0:Stage Show Sequence Pool Object Reference=Find(TargetMap=@n18, Key=@n17)
     Option 1:Stage Show Sequence Pool Object Reference=Find(
      TargetMap:Map of ESparksMidiTextEvents Enums to Stage Show Sequence Pool Object References=@n20
      Key:"ESparksMidiTextEvents Enum (by ref)"=@n19
     )
     Index=Get IsMidiEvent()
    ) -> (ReturnValue:Stage Show Sequence Pool Object Reference))
   > OpenAddChangeWindow!(
    execute
    self
    WindowType:Name=MojoStage_Sequences PoolData
    IsJustCreate:bool=true
    HideSelectSection:bool=false
    IsLoopingEvent:bool=false
    KeepSelectChange:bool=Select(
     Option 0:bool=true
     Option 1:bool=false
     Index=Is Empty!(TargetArray:Array of Stage Show Sequence Object References=@n21) -> (ReturnValue:bool)
    ) -> (ReturnValue:bool)
   ) -> (then)
   > SetVisibility(self=Get ChooseAndAddEvent() -> (ChooseAndAddEvent:Editor Utility Button Object Reference), InVisibility=Visible)
   > SetVisibility(self=Get CreateChangeButton() -> (CreateChangeButton:Editor Utility Button Object Reference), InVisibility=Visible)
   "On Pressed (DeleteSequenceDA)"() {is=title K2Node_ComponentBoundEvent}
   n22 = Get SequencesDataListComboBox()
   n23 = Get CurrentSequence()
   n24 = Find(TargetMap=Get TempMidiEventMap() -> (TempMidiEventMap=None), Key=Get MidiTextEvents() -> (MidiTextEvents=None))
   > OpenWindow!(
    execute
    self:EWBP Sparks Delete Confirmation Object Reference=Get DeleteConfirmationPopup() -> (DeleteConfirmationPopup:EWBP Sparks Delete Confirmation Object Reference)
    Object:Object Reference="Get (a copy)"(Array=Get Sequences(self=@n24), Dimension 1=FindOptionIndex(self=@n22, Option=@n23) -> (ReturnValue=0))
   ) -> (then)
   "On Pressed (ChooseAndAddEvent)"() {is=title K2Node_ComponentBoundEvent}
   > Set CurrentSequenceIndex(CurrentSequenceIndex=0) -> (Output_Get=0)
   > UpdateSequencePool(ReturnObject=GetCurrentData(), IsNewAsset=true)
   > SetVisibility(self=Get NewSequenceEventWindow(), InVisibility=Collapsed)
   > Call On Asset Changed(ReturnObj=GetCurrentData(), IsNewFile=true)
   n25 = "On Closed (DeleteConfirmationPopup)"!() -> (OutputDelegate, then, IsDeleted:bool, PropertyName:Name=None, ObjToDelete:Object Reference) {is=title K2Node_ComponentBoundEvent}
   > Is Valid(InputObject=@n25.ObjToDelete)
   > n26 = Sequence!(execute) -> (then_0, then_1, then_2:Exec)
   > Branch(Condition="Equal (Name)"(A=@n25.PropertyName, B=MojoStage_Sequences PoolData))
   > n27 = Sequence()
   > BeginTransaction(
    Context=RemovingDAPoolTimelineTool
    Description="INVTEXT(\"Removing Data asset from stage manager sequence pool list.\")"
    PrimaryObject=GetCurrentMIDIMap()
   )
   > n28 = Sequence!(execute) -> (then_0, then_1)
   > n29 = Branch(Condition=Get IsMidiEvent())
   n30 = Get TempPool() -> (TempPool:Array of Stage Show Sequence Pool Object References)
   > Clear(TargetArray:Array of Stage Show Sequence Pool Object References=@n30)
   n31 = Get TempMidiEventMap() -> (TempMidiEventMap=None)
   > Add(
    TargetArray:Array of Stage Show Sequence Pool Object References=@n30
    NewItem:"Stage Show Sequence Pool Object Reference (by ref)"=Find(TargetMap=@n31, Key=Get MidiTextEvents() -> (MidiTextEvents=None))
   ) {is=title K2Node_CallArrayFunction}
   > Remove Index!(
    execute
    TargetArray:Array of Stage Show Sequence Pool Object References=@n30
    IndexToRemove:int=FindOptionIndex(self=Get SequencesDataListComboBox(), Option=Get CurrentSequence()) -> (ReturnValue=0)
   ) -> (then)
   > Remove(
    TargetMap:Map of ESparksMidiTextEvents Enums to Stage Show Sequence Pool Object References=@n31
    Key:"ESparksMidiTextEvents Enum (by ref)"=Get MidiTextEvents() -> (MidiTextEvents=None)
   )
   > Add(
    TargetMap:Map of ESparksMidiTextEvents Enums to Stage Show Sequence Pool Object References=@n31
    Key:"ESparksMidiTextEvents Enum (by ref)"=Get MidiTextEvents() -> (MidiTextEvents=None)
    Value="Get (a copy)"(Array:Array of Stage Show Sequence Pool Object References=Get TempPool()) -> (Output:Stage Show Sequence Pool Object Reference)
   ) {is=title K2Node_CallFunction}
   n32 = Get TempPool()
   Clear(execute=@n29.else, TargetArray=@n32)
   n33 = Get TempQuickplayEventMap() -> (TempQuickplayEventMap=Invalid)
   > Add(
    TargetArray=@n32
    NewItem=Find(
     TargetMap:Map of EPilgrimQuickplayState Enums to Stage Show Sequence Pool Object References=@n33
     Key:"EPilgrimQuickplayState Enum (by ref)"=Get QuickPlayEvents() -> (QuickPlayEvents=Invalid)
    )
   ) {is=title K2Node_CallArrayFunction}
   > Remove Index(TargetArray=@n32, IndexToRemove=FindOptionIndex(self=Get SequencesDataListComboBox(), Option=Get CurrentSequence()) -> (ReturnValue=0))
   > Remove(
    TargetMap:Map of EPilgrimQuickplayState Enums to Stage Show Sequence Pool Object References=@n33
    Key:"EPilgrimQuickplayState Enum (by ref)"=Get QuickPlayEvents() -> (QuickPlayEvents=Invalid)
   )
   > Add(
    TargetMap:Map of EPilgrimQuickplayState Enums to Stage Show Sequence Pool Object References=@n33
    Key:"EPilgrimQuickplayState Enum (by ref)"=Get QuickPlayEvents() -> (QuickPlayEvents=Invalid)
    Value="Get (a copy)"(Array=Get TempPool())
   ) {is=title K2Node_CallFunction}
   Set EventMap(
    execute=@n28.then_1
    EventMap:Map of ESparksMidiTextEvents Enums to Stage Show Sequence Pool Object References=Get TempMidiEventMap() -> (TempMidiEventMap=None)
    self:Stage Show MIDIEvent Map Object Reference=GetCurrentMIDIMap()
   ) -> (Output_Get:Map of ESparksMidiTextEvents Enums to Stage Show Sequence Pool Object References=None)
   > Set EventMap(
    EventMap:Map of EPilgrimQuickplayState Enums to Stage Show Sequence Pool Object References=Get TempQuickplayEventMap() -> (TempQuickplayEventMap=Invalid)
    self:Stage Show Gameplay Event Map Object Reference=Get GamePlayEventData()
   ) -> (Output_Get:Map of EPilgrimQuickplayState Enums to Stage Show Sequence Pool Object References=Invalid)
   > EndTransaction()
   > TransactObject(Object=GetCurrentMIDIMap())
   > Set CurrentSequenceIndex(CurrentSequenceIndex=0) -> (Output_Get=0)
   Branch(execute=[@n34.then_2, @n35.then_0, @n27.then_1], Condition=@n25.IsDeleted)
   > DeleteLoadedAsset!(execute, self:Editor Asset Subsystem Object Reference=Get EditorAssetSubsystem(), AssetToDelete:Object Reference=@n25.ObjToDelete) -> (then, ReturnValue:bool)
   ForceRefresh(execute=@n27.then_2)
   Branch(execute=@n26.then_1, Condition="Equal (Name)"(A=@n25.PropertyName, B=MojoStage_MIDISequenceData))
   > BeginTransaction(
    Context=RemovingMidiTextMap
    Description="INVTEXT(\"Removing MIDI Text Map Data asset from the mood.\")"
    PrimaryObject=Find(
     TargetMap:Map of Strings to BP Mojo Stage Mood Data Asset Object References=Get MoodMap() -> (MoodMap:Map of Strings to BP Mojo Stage Mood Data Asset Object References)
     Key:"String (by ref)"=Get CurrentMood() -> (CurrentMood:String)
    ) -> (Value:BP Mojo Stage Mood Data Asset Object Reference)
   )
   > n34 = Sequence!(execute) -> (then_0, then_1, then_2:Exec, then_3:Exec)
   > Branch(Condition=Get IsMidiEvent())
   > Set SequenceMIDIMap!(
    execute
    SequenceMIDIMap:BP Mojo Stage Map Midi Sequence Data Asset Object Reference=/PilgrimQuickplay/StageManager/Data/Maps/Sequence/MIDISequenceMap_Default
    self:BP Mojo Stage Mood Data Asset Object Reference=Find(TargetMap=Get MoodMap(), Key=Get CurrentMood())
   ) -> (then, Output_Get:BP Mojo Stage Map Midi Sequence Data Asset Object Reference)
   EndTransaction(execute=@n34.then_1)
   > TransactObject(Object=Find(TargetMap=Get MoodMap(), Key=Get CurrentMood()))
   SuperRefresh(execute=@n34.then_3)
   Branch(execute=@n26.then_2, Condition="Equal (Name)"(A=@n25.PropertyName, B=MojoStage_Mood))
   > n35 = Sequence!(execute) -> (then_0, then_1)
   Set CurrentMood(execute=@n35.then_1, CurrentMood:String=Default) -> (Output_Get:String)
   > SuperRefresh()
   "On Pressed (NewSequenceEventChangeButton)"() {is=title K2Node_ComponentBoundEvent}
   > UpdateSequencePool(ReturnObject=GetCurrentData(), IsNewAsset=false)
   > SetVisibility(self=Get NewSequenceEventWindow(), InVisibility=Collapsed)
   > Call On Asset Changed(ReturnObj=GetCurrentData())
   > Set CurrentEventSelected()
   "On Pressed (CloseNewSequenceEventWindowButton)"() {is=title K2Node_ComponentBoundEvent}
   > SetVisibility(self=Get NewSequenceEventWindow(), InVisibility=Collapsed)
   > Set CurrentEventSelected()
   > Call On Popup Closed!(execute:Exec, self:Sparks Stage Manager Timeline Editor Object Reference) -> (then:Exec) {is=title K2Node_CallDelegate}
   "On Property Changed (DaSlotProperty)"!() -> (OutputDelegate, then, PropertyName:Name=None) {is=title K2Node_ComponentBoundEvent}
   > SetObject(self=Get EventDetailView() -> (EventDetailView:Details View Object Reference), NewObject=GetCurrentData())
   "On Pressed (NewSequenceEventAddButton)"() {is=title K2Node_ComponentBoundEvent}
   n36 = GetCurrentData()
   > n37 = DuplicateAsset(
    self=Get EditorAssetSubsystem()
    SourceAssetPath=Append(A=Get AssetsBasePath(), B=Append(A=@n36.AssetBaseName, B=_ReadOnly))
    DestinationAssetPath=Append(A=GetAssetDestinationPath(), B=GrabNewAssetName(IgnoreFirstIndex=false, WarningDuplicated=false))
   )
   > n38 = Sequence()
   > n39 = Cast To MojoFxManagerEvent_DataAsset(Object=@n37)
   > Set LoopingEvent(LoopingEvent=Get IsLoopingEvent(), self=@n39)
   > TransactObject(Object=@n39)
   UpdateSequencePool(execute=@n38.then_1, ReturnObject=@n37, IsNewAsset=true)
   > SetVisibility(self=Get NewSequenceEventWindow(), InVisibility=Collapsed)
   > Call On Asset Changed(ReturnObj=@n37, IsNewFile=true)
   > Set CurrentEventSelected()
   "On Pressed (OpenEditorForEverything)"() {is=title K2Node_ComponentBoundEvent}
   > n40 = Is Valid(InputObject=GetCurrentData())
   SetIsEnabled(
    execute=@n40.Is Not Valid
    self=Get OpenEditorForEverything() -> (OpenEditorForEverything:Editor Utility Button Object Reference)
    bInIsEnabled=false
   )
   n41 = Cast To MojoFxManagerEvent_DataAsset(execute=@n40.Is Valid, Object=GetCurrentData())
   > n42 = SpawnAndRegisterTabAndGetID!(
    execute
    self:Editor Utility Subsystem Object Reference=Get EditorUtilitySubsystem()
    InBlueprint:Editor Utility Widget Blueprint Object Reference=/PilgrimQuickplay/StageManager/NeverCook/EditorUtilities/StageManagerEditor/FxManagerEventWindow/SparksFxManagerEventDataWindow
   ) -> (then, NewTabID:Name=None, ReturnValue:Editor Utility Widget Object Reference)
   > n43 = Cast To SparksFxManagerEventDataWindow!(execute, Object=@n42.ReturnValue) -> (then, CastFailed, AsSparks Fx Manager Event Data Window:Sparks Fx Manager Event Data Window Object Reference) {is=title K2Node_DynamicCast}
   > Init!(
    execute
    self:Sparks Fx Manager Event Data Window Object Reference=@n43
    EventDataAsset:Mojo Fx Manager Event Data Asset Object Reference=@n41
   ) -> (then)
   > n44 = Sequence()
   > n45 = Get!(execute:Exec) -> (then:Exec, else:Exec, CurrentEventBarReference:EWBP Sparks Sequence Pool Timeline Event Bar Object Reference)
   > SetEventBarReference!(
    execute
    self:Sparks Fx Manager Event Data Window Object Reference=@n43
    EventBarReference:"EWBP Sparks Sequence Pool Timeline Event Bar Object Reference (by ref)"=@n45
   ) -> (then)
   SetTimelineEditorReference!(
    execute=@n44.then_1
    self:Sparks Fx Manager Event Data Window Object Reference=@n43
    SparksTimelineEditor:Sparks Stage Manager Timeline Editor Object Reference=Self-Reference()
   ) -> (then)
   OpenEditorForAssets!(
    execute=@n41.CastFailed
    self:Asset Editor Subsystem Object Reference=Get AssetEditorSubsystem()
    Assets:Array of Object References=Make Array("[0]":Object Reference=GetCurrentData()) -> (Array:Array of Object References)
    OpenedMethod:EAssetTypeActivationOpenedMethod Enum=Edit
   ) -> (then, ReturnValue:bool)
   n46 = "On Selection Changed (DAFilteredComboBox)"!() -> (OutputDelegate, then, SelectedItem:String, SelectionType:ESelectInfo Enum) {is=title K2Node_ComponentBoundEvent}
   > Branch(Condition="Not Equal Exactly (String)"(A=@n46.SelectedItem, B=aaaa))
   > n47 = Set CurrentEventSelected(CurrentEventSelected=@n46.SelectedItem)
   n48 = GetCurrentData()
   > n49 = Is Valid(InputObject=@n48)
   SetIsEnabled(
    execute=@n49.Is Not Valid
    self=[
     Get ToggleViewDa() -> (ToggleViewDa:Editor Utility Check Box Object Reference)
     Get OpenEditorForEverything()
     Get ChooseAndAddEvent()
     Get NewSequenceEventChangeButton()
     Get RenameEventButton() -> (RenameEventButton:Editor Utility Button Object Reference)
    ]
    bInIsEnabled=false
   )
   SetObject(execute=@n49.Is Valid, self=Get EventDetailView(), NewObject=@n48)
   > SetToolTipText!(
    execute
    self:Widget Object Reference=Get DAFilteredComboBox() -> (DAFilteredComboBox:"Searchable Combo Box (String) Object Reference")
    InToolTipText:"Text (by ref)"="To Text (String)"(InString=@n47)
   ) -> (then)
   n50 = Split(SourceString=GrabNewAssetName(IgnoreFirstIndex=false, WarningDuplicated=true), InStr=_, SearchDir=FromEnd)
   n51 = Get CurrentEventSelected() -> (CurrentEventSelected:String)
   > "SetText (Text Box)"!(
    execute
    self:Text Box Object Reference=Get RenameAssetTextBox() -> (RenameAssetTextBox:Editor Utility Editable Text Box Object Reference)
    InText:Text="To Text (String)"(InString=Split(SourceString=@n51, InStr=@n50, SearchDir=FromEnd).RightS)
   ) -> (then)
   > Set OutputFileName(OutputFileName=Get CurrentEventSelected())
   > "SetText (Text)"(self=Get DANamePreviewText(), InText="To Text (String)"(InString=GrabNewAssetName(IgnoreFirstIndex=true, WarningDuplicated=false)))
   > SetIsEnabled(
    self=[Get OpenEditorForEverything(), Get ToggleViewDa(), Get ChooseAndAddEvent(), Get NewSequenceEventChangeButton(), Get RenameEventButton()]
    bInIsEnabled=true
   )
   n52 = "On Check State Changed (ToggleViewDa)"!() -> (OutputDelegate, then, bIsChecked:bool) {is=title K2Node_ComponentBoundEvent}
   > SetVisibility(
    self=Get EventPreviewerGroup() -> (EventPreviewerGroup:Size Box Object Reference)
    InVisibility=Select(Option 0:ESlateVisibility Enum=Collapsed, Option 1:ESlateVisibility Enum=Visible, Index=@n52.bIsChecked) -> (ReturnValue:ESlateVisibility Enum)
   )
   "On Check State Changed (SelectAndChangeToggle)"!() -> (OutputDelegate, then, bIsChecked:bool) {is=title K2Node_ComponentBoundEvent}
   > SetCheckedState!(
    execute
    self:Check Box Object Reference=Get AddAndCreateToggle() -> (AddAndCreateToggle:Editor Utility Check Box Object Reference)
    InCheckedState:ECheckBoxState Enum=Unchecked
   ) -> (then)
   > SetVisibility(self=Get SelectChangeGroup() -> (SelectChangeGroup:Vertical Box Object Reference), InVisibility=Visible)
   > SetVisibility(self=Get CreateNewDAGroup() -> (CreateNewDAGroup:Vertical Box Object Reference), InVisibility=Collapsed)
   n53 = GetDisplayName(Object=GetCurrentData())
   n54 = Append(A=Split(SourceString=GrabNewAssetName(IgnoreFirstIndex=false, WarningDuplicated=true), InStr=_, SearchDir=FromEnd), B=_)
   > "SetText (Text Box)"(self=Get RenameAssetTextBox(), InText="To Text (String)"(InString=Split(SourceString=@n53, InStr=@n54, SearchDir=FromEnd).RightS))
   > Set OutputFileName(OutputFileName=@n53)
   > "SetText (Text)"(self=Get DANamePreviewText(), InText="To Text (String)"(InString=GrabNewAssetName(IgnoreFirstIndex=true, WarningDuplicated=true)))
   "On Check State Changed (AddAndCreateToggle)"!() -> (OutputDelegate, then, bIsChecked:bool) {is=title K2Node_ComponentBoundEvent}
   > SetCheckedState(self=Get SelectAndChangeToggle() -> (SelectAndChangeToggle:Editor Utility Check Box Object Reference), InCheckedState=Unchecked)
   > SetVisibility(self=Get SelectChangeGroup(), InVisibility=Collapsed)
   > SetVisibility(self=Get CreateNewDAGroup(), InVisibility=Visible)
   > Set OutputFileName()
   n55 = Get PropertyName() -> (PropertyName=None)
   n56 = OR Boolean(A="Equal (Name)"(A=@n55, B=MojoStage_MIDISequenceData), B="Equal (Name)"(A=@n55, B=MojoStage_Mood))
   > "SetText (Text)"(self=Get DANamePreviewText(), InText="To Text (String)"(InString=GrabNewAssetName(IgnoreFirstIndex=@n56, WarningDuplicated=@n56)))
   n57 = "On Text Changed (RenameAssetTextBox)"!() -> (OutputDelegate, then, Text:"Text (by ref)") {is=title K2Node_ComponentBoundEvent}
   n58 = GetCurrentData()
   n59 = TextTrimPrecedingAndTrailing(InText=@n57.Text)
   n60 = "To String (Text)"(InText=@n59)
   n61 = TextIsEmpty(InText=@n59)
   n62 = Get CurrentTextEvent()
   > Set OutputFileName(OutputFileName=Append!(
     A=Split(SourceString=@n58.AssetBaseName, InStr=_, SearchDir=FromEnd)
     B=_
     C:String=Select(Option 0:String=@n60, Option 1:String=@n62, Index=@n61) -> (ReturnValue:String)
    ) -> (ReturnValue))
   > "SetText (Text)"(self=Get DANamePreviewText(), InText="To Text (String)"(InString=GrabNewAssetName(IgnoreFirstIndex=false, WarningDuplicated=true)))
   "On Pressed (RenameEventButton)"() {is=title K2Node_ComponentBoundEvent}
   n63 = GetCurrentData()
   > RenameAsset!(execute, Asset:Object Reference=@n63, NewName:String=GrabNewAssetName(IgnoreFirstIndex=false, WarningDuplicated=false)) -> (then)
   > TransactObject(Object=@n63)
   > SuperRefresh()
   > SetVisibility(self=Get NewSequenceEventWindow(), InVisibility=Collapsed)
   > Set CurrentEventSelected()
   n64 = Event PreConstruct()
   > Branch(Condition=@n64.IsDesignTime)
   > SetObject(
    self=Get EventDetailView()
    NewObject=Get FxManager_Event StepData() -> (FxManager_Event StepData:Mojo Fx Manager Event Data Asset Object Reference)
   )
   "Equal (Name)"(A=Get PropertyName() -> (PropertyName=None), B=MojoStage_MIDISequenceData)
  ]}
 CheckLoopingEvent {nodes [
   n1 = CheckLoopingEvent!() -> (then, Object:Object Reference) {is=title K2Node_FunctionEntry}
   > n2 = Cast To MojoFxManagerEvent_DataAsset(Object=@n1)
   > Return Node!(
    execute
    IsLoopingEvent:bool=OR Boolean(A=Get LoopingEvent!(self:Stage Show Event Object Reference=@n2) -> (LoopingEvent:bool), B=Get IsLoopingEvent())
   )
   Return Node(execute=@n2.CastFailed, IsLoopingEvent=Get IsLoopingEvent())
  ]}
 EventGraph {nodes [
   "On Pressed (DeleteMIDITextMapBtn)"() {is=title K2Node_ComponentBoundEvent}
   n1 = Get MIDITextMapComboBox() -> (MIDITextMapComboBox:"Searchable Combo Box (String) Object Reference")
   > OpenWindow(
    self=Get DeleteConfirmationPopup()
    Object=Find(
     TargetMap:Map of Strings to BP Mojo Stage Map Midi Sequence Data Asset Object References=Get MIDISequenceMap() -> (MIDISequenceMap:Map of Strings to BP Mojo Stage Map Midi Sequence Data Asset Object References)
     Key=GetSelectedOption!(self:"Searchable Combo Box (String) Object Reference"=@n1) -> (ReturnValue:String)
    ) -> (Value:BP Mojo Stage Map Midi Sequence Data Asset Object Reference)
   )
   "On Clicked (OpenEditorWithThemedGroup)"() {is=title K2Node_ComponentBoundEvent}
   n2 = Get MoodMap()
   n3 = Get CurrentMood()
   > OpenEditorForAssets(
    self=Get AssetEditorSubsystem()
    Assets=Make Array("[0]"=Find(TargetMap:Map of Strings to BP Mojo Stage Mood Data Asset Object References=@n2, Key=@n3) -> (Value:BP Mojo Stage Mood Data Asset Object Reference))
    OpenedMethod=Edit
   )
   "On Pressed (AddMoodButton)"() {is=title K2Node_ComponentBoundEvent}
   > OpenAddChangeWindow(WindowType=MojoStage_Mood, IsJustCreate=true, HideSelectSection=true, IsLoopingEvent=false, KeepSelectChange=true)
   n4 = "On Selection Changed (MIDITextMapComboBox)"!() -> (OutputDelegate, then, SelectedItem:String, SelectionType:ESelectInfo Enum) {is=title K2Node_ComponentBoundEvent}
   n5 = Find(TargetMap:Map of Strings to BP Mojo Stage Map Midi Sequence Data Asset Object References=Get MIDISequenceMap(), Key=@n4.SelectedItem) -> (Value:BP Mojo Stage Map Midi Sequence Data Asset Object Reference)
   > Branch(Condition=AND Boolean(A="Not Equal Exactly (String)"(A=@n4.SelectedItem, B=aaaa), B=@n5.ReturnValue))
   > Set CurrentMIDISequenceSelected(CurrentMIDISequenceSelected:String=@n4.SelectedItem) -> (Output_Get:String)
   > SetIsEnabled(
    self=Get DeleteMIDITextMapBtn() -> (DeleteMIDITextMapBtn:Editor Utility Button Object Reference)
    bInIsEnabled=NOT Boolean(A="Equal Exactly (String)"(A=@n4.SelectedItem, B=Default))
   )
   n6 = Find(TargetMap:Map of Strings to BP Mojo Stage Mood Data Asset Object References=Get MoodMap(), Key=Get CurrentMood()) -> (Value:BP Mojo Stage Mood Data Asset Object Reference)
   > Branch(Condition="Not Equal (Object)"(
     A:Object Reference=@n5
     B:Object Reference=Get SequenceMIDIMap!(self:BP Mojo Stage Mood Data Asset Object Reference=@n6) -> (SequenceMIDIMap:BP Mojo Stage Map Midi Sequence Data Asset Object Reference)
    ) -> (ReturnValue:bool=~) {is=title K2Node_PromotableOperator})
   > Set SequenceMIDIMap(SequenceMIDIMap=@n5, self=@n6)
   > ForceRefresh()
   > TransactObject(Object=@n6)
   "On Pressed (AddMIDITextMapButton)"() {is=title K2Node_ComponentBoundEvent}
   > Set CurrentEventSelected(CurrentEventSelected=GetDisplayName(Object=GetCurrentMIDIMap()))
   > OpenAddChangeWindow(WindowType=MojoStage_MIDISequenceData, IsJustCreate=true, HideSelectSection=true, IsLoopingEvent=false, KeepSelectChange=true)
   n7 = "On Selection Changed (KeySlotComboBox)"!() -> (OutputDelegate, then, SelectedItem:String, SelectionType:ESelectInfo Enum) {is=title K2Node_ComponentBoundEvent}
   > Branch(Condition=AND Boolean(A="Not Equal Exactly (String)"(A=@n7.SelectedItem, B=a), B="Not Equal Exactly (String)"(A=@n7.SelectedItem)))
   > n8 = Set CurrentTextEvent(CurrentTextEvent:String=@n7.SelectedItem) -> (Output_Get:String)
   > Clear!(
    execute
    TargetMap:Map of Strings to Stage Show Sequence Pool Object References=Get CurrentSequenceMap() -> (CurrentSequenceMap:Map of Strings to Stage Show Sequence Pool Object References)
   ) -> (then) {is=title K2Node_CallFunction}
   > ClearOptions(self:"Searchable Combo Box (String) Object Reference"=Get SequencesDataListComboBox())
   > Set LastTextEventIndex(LastTextEventIndex:int=FindOptionIndex(self=Get KeySlotComboBox(), Option=@n8) -> (ReturnValue=0)) -> (Output_Get:int=0)
   > n9 = Branch(Condition=Get IsMidiEvent())
   n10 = Find(TargetMap:Map of Strings to Strings=Get EventsEnumToStringMap() -> (EventsEnumToStringMap:Map of Strings to Strings), Key=@n8) -> (Value:String)
   n11 = "To Integer (Byte)"(InByte=GetStageManagerEventsMidiTextEnumFromString(MidiTextEventString=@n10) -> (MidiTextEventEnum=NewEnumerator0)) -> (ReturnValue=0)
   > Set MidiTextEvents(MidiTextEvents:ESparksMidiTextEvents Enum=Byte to Enum ESparksMidiTextEvents(Byte="To Byte (Integer)"(InInt=@n11) -> (ReturnValue=0))) -> (Output_Get:ESparksMidiTextEvents Enum=None)
   n12 = GetCurrentMidiTextEventMap!() -> (
    MIDIEventMap:Map of ESparksMidiTextEvents Enums to Stage Show Sequence Pool Object References=None
    QPEventMap:Map of EPilgrimQuickplayState Enums to Stage Show Sequence Pool Object References=Invalid
   )
   n13 = Find(
    TargetMap:Map of ESparksMidiTextEvents Enums to Stage Show Sequence Pool Object References=@n12
    Key:"ESparksMidiTextEvents Enum (by ref)"=Get MidiTextEvents() -> (MidiTextEvents=None)
   ) -> (Value:Stage Show Sequence Pool Object Reference)
   > n14 = Branch(Condition=@n13.ReturnValue)
   n15 = GetDisplayName()
   AddOption!(execute=[@n14.then, @n32.then], self:"Searchable Combo Box (String) Object Reference"=Get SequencesDataListComboBox(), Option:String=@n15) -> (then)
   n16 = Find(
    TargetMap:Map of EPilgrimQuickplayState Enums to Stage Show Sequence Pool Object References=@n12.QPEventMap
    Key:"EPilgrimQuickplayState Enum (by ref)"=Get QuickPlayEvents() -> (QuickPlayEvents=Invalid)
   )
   n17 = Select(Option 0:Stage Show Sequence Pool Object Reference=@n16, Option 1:Stage Show Sequence Pool Object Reference=@n13, Index=Get IsMidiEvent()) -> (ReturnValue:Stage Show Sequence Pool Object Reference)
   > Add(TargetMap:Map of Strings to Stage Show Sequence Pool Object References=Get CurrentSequenceMap(), Key:"String (by ref)"=@n15, Value=@n17)
   > AddOption(self=Get SequencesDataListComboBox(), Option=Empty)
   > n18 = Set NoSequencePool(NoSequencePool:bool=Is Empty(TargetArray=Get Sequences(self=@n17))) -> (Output_Get:bool)
   > n19 = Branch(Condition=@n18)
   > Set CurrentSequenceIndex(CurrentSequenceIndex=0) -> (Output_Get=0)
   > Set CurrentEventSelected()
   > AddOption(self=Get SequencesDataListComboBox(), Option=a)
   > SetSelectedOption!(execute, self:"Searchable Combo Box (String) Object Reference"=Get SequencesDataListComboBox(), Option:String=a) -> (then)
   > SetSelectedOption(self=Get SequencesDataListComboBox(), Option=Empty)
   > n20 = RemoveOption!(execute, self:"Searchable Combo Box (String) Object Reference"=Get SequencesDataListComboBox(), Option:String=a) -> (then, ReturnValue:bool)
   SetVisibility(
    execute=[@n20.then, @n31.then]
    self=Get EmptySequencePoolGroup() -> (EmptySequencePoolGroup:Overlay Object Reference)
    InVisibility=Select(Option 0:ESlateVisibility Enum=Collapsed, Option 1:ESlateVisibility Enum=Visible, Index=Get NoSequencePool()) -> (ReturnValue:ESlateVisibility Enum)
   )
   n21 = Get Sequences(self=Find(TargetMap:Map of Strings to Stage Show Sequence Pool Object References=Get CurrentSequenceMap(), Key:"String (by ref)"=Get CurrentSequence()))
   n22 = Is Not Empty!(TargetArray:Array of Stage Show Sequence Object References=@n21) -> (ReturnValue:bool)
   n23 = Get NoSequencePool()
   > SetVisibility(
    self=Get EmptySequenceGroup() -> (EmptySequenceGroup:Overlay Object Reference)
    InVisibility=Select(Option 0=Visible, Option 1=Collapsed, Index=OR Boolean(A=@n23, B=@n22))
   )
   n24 = Is Empty(TargetArray=@n21)
   n25 = Get SequencesDataListComboBox()
   n26 = NOT Boolean(A="integer >= integer"(A=GetOptionCount!(self:"Searchable Combo Box (String) Object Reference"=@n25) -> (ReturnValue:int=0), B=1))
   > SetVisibility(
    self=Get SequencesScrollBoxGroup() -> (SequencesScrollBoxGroup:Editor Utility Scroll Box Object Reference)
    InVisibility=Select(Option 0=Visible, Option 1=Collapsed, Index=OR Boolean!(A=@n23, B=@n24, C:bool=@n26) -> (ReturnValue))
   )
   n27 = Get NoSequencePool()
   n28 = GetOptionCount(self=Get SequencesDataListComboBox()) -> (ReturnValue=0)
   > SetIsEnabled(
    self=[
     Get AddNewSequenceButton() -> (AddNewSequenceButton:Editor Utility Button Object Reference)
     Get DeleteSequenceDA() -> (DeleteSequenceDA:Editor Utility Button Object Reference)
     Get OpenSequenceDAButton() -> (OpenSequenceDAButton:Editor Utility Button Object Reference)
    ]
    bInIsEnabled=AND Boolean(A=NOT Boolean(A=@n27), B="integer >= integer"(A=@n28, B=1))
   )
   > Set IsFromInit(IsFromInit:bool=false) -> (Output_Get:bool)
   RemoveOption(execute=@n19.else, self=Get SequencesDataListComboBox(), Option=Empty)
   > AddOption(self=Get SequencesDataListComboBox(), Option=a)
   > SetSelectedOption(self=Get SequencesDataListComboBox(), Option=a)
   > RemoveOption(self=Get SequencesDataListComboBox(), Option=a)
   n29 = Get SequencesDataListComboBox()
   n30 = GetOptionAtIndex!(
    self:"Searchable Combo Box (String) Object Reference"=@n29
    Index:int=Select(Option 0:int=0, Option 1:int=Get CurrentSequenceIndex() -> (CurrentSequenceIndex:int=0), Index=Get IsFromInit() -> (IsFromInit:bool)) -> (ReturnValue:int)
   ) -> (ReturnValue:String)
   > SetSelectedOption(self=@n29, Option=@n30)
   > n31 = Set CurrentSequence(CurrentSequence:String=@n30) -> (Output_Get:String)
   Set QuickPlayEvents(
    execute=@n9.else
    QuickPlayEvents:EPilgrimQuickplayState Enum=GetStageManageQPEventsEnumFromString(QPEventString=@n10) -> (QPEventEnum=Invalid)
   ) -> (Output_Get:EPilgrimQuickplayState Enum=Invalid)
   > n32 = Branch(Condition=@n16.ReturnValue)
   "On Clicked (OpenSequenceDAButton)"() {is=title K2Node_ComponentBoundEvent}
   n33 = Get CurrentSequence()
   n34 = Get CurrentSequenceMap()
   > OpenEditorForAssets(self=Get AssetEditorSubsystem(), Assets=Make Array("[0]"=Find(TargetMap=@n34, Key=@n33)), OpenedMethod=Edit)
   n35 = "On Selection Changed (SequencesDataListComboBox)"!() -> (OutputDelegate, then, SelectedItem:String, SelectionType:ESelectInfo Enum) {is=title K2Node_ComponentBoundEvent}
   > Branch(Condition=AND Boolean(A="Not Equal Exactly (String)"(A=@n35.SelectedItem, B=a), B="Not Equal Exactly (String)"(A=@n35.SelectedItem, B=Empty)))
   > Set CurrentSequence(CurrentSequence=@n35.SelectedItem)
   > Set CurrentSequenceIndex(CurrentSequenceIndex=FindOptionIndex(self=Get SequencesDataListComboBox(), Option=Get CurrentSequence()) -> (ReturnValue=0)) -> (Output_Get=0)
   > n36 = Branch(Condition=Get IsMidiEvent())
   > PopulateSequencesArray!(execute, self, IsChecked:bool=true) -> (then)
   PopulateSequencesArray(execute=@n36.else, IsChecked=false)
   "On Pressed (DeleteMoodButton)"() {is=title K2Node_ComponentBoundEvent}
   > OpenWindow(
    self=Get DeleteConfirmationPopup()
    Object=Find(TargetMap:Map of Strings to BP Mojo Stage Mood Data Asset Object References=Get MoodMap(), Key=Get CurrentMood()) -> (Value:BP Mojo Stage Mood Data Asset Object Reference)
   )
   n37 = "On Check State Changed (MidiTextEventsToggle)"!() -> (OutputDelegate, then, bIsChecked:bool) {is=title K2Node_ComponentBoundEvent}
   > Branch(Condition=@n37.bIsChecked)
   > n38 = Set IsMidiEvent(IsMidiEvent:bool=true) -> (Output_Get:bool)
   Set LastEventIndex(execute=[@n40.then, @n38.then], LastEventIndex=0) -> (Output_Get=0)
   > UpdateEventComboBox()
   n39 = "On Check State Changed (QuickPlayEventsToggle)"!() -> (OutputDelegate, then, bIsChecked:bool) {is=title K2Node_ComponentBoundEvent}
   > Branch(Condition=@n39.bIsChecked)
   > n40 = Set IsMidiEvent(IsMidiEvent=false)
   n41 = "On Selection Changed (SequenceTypeComboBox)"!() -> (OutputDelegate, then, SelectedItem:String, SelectionType:ESelectInfo Enum) {is=title K2Node_ComponentBoundEvent}
   > Branch(Condition="Not Equal Exactly (String)"(A=@n41.SelectedItem, B=aaaa))
   > n42 = Set CurrentMood(CurrentMood=@n41.SelectedItem)
   > "SetText (Text Box)"(
    self=Get EventGroupTextBox() -> (EventGroupTextBox:Editor Utility Editable Text Box Object Reference)
    InText="To Text (String)"(InString=@n42)
   )
   > SetIsEnabled(
    self=[Get EventGroupTextBox(), Get DeleteMoodButton() -> (DeleteMoodButton:Editor Utility Button Object Reference)]
    bInIsEnabled="Not Equal (Name)"(A:String=Get CurrentMood(), B=Default) -> (ReturnValue=~)
   )
   > n43 = SetSelectedOption(self=Get MIDITextMapComboBox(), Option=GetCurrentMIDIMap().FilteredName)
   UpdateMapValues(execute=[@n43.then, @n50.then, @n47.then_1])
   > UpdateEventComboBox()
   > n44 = Values!(
    execute
    TargetMap:Map of ESparksMidiTextEvents Enums to Stage Show Sequence Pool Object References=Get TempMidiEventMap() -> (TempMidiEventMap=None)
   ) -> (then, Values:Array of Stage Show Sequence Pool Object References)
   > Set NewSequenceButton(NewSequenceButton:bool=false) -> (Output_Get:bool)
   > "SetText (Text)"(
    self=Get MidiMembersNumber() -> (MidiMembersNumber:Text Object Reference)
    InText=Format Text!(
     Format="INVTEXT(\"{number} Events\")"
     number:int=Length(TargetArray:Array of Stage Show Sequence Pool Object References=@n44) {is=title K2Node_CallArrayFunction}
    ) -> (Result)
   )
   > n45 = Values(TargetMap:Map of EPilgrimQuickplayState Enums to Stage Show Sequence Pool Object References=Get TempQuickplayEventMap() -> (TempQuickplayEventMap=Invalid))
   > "SetText (Text)"(
    self=Get QuickPlayMembersNumber() -> (QuickPlayMembersNumber:Text Object Reference)
    InText=Format Text(Format="INVTEXT(\"{number} Events\")", number=Length(TargetArray=@n45))
   )
   > Clear(TargetArray:Array of EWBP Sparks Sequence Pool Timeline Event Bar Object References=Get AllEvents() -> (AllEvents:Array of EWBP Sparks Sequence Pool Timeline Event Bar Object References)) {is=title K2Node_CallArrayFunction}
   > PopulateSequencesArray(IsChecked=Get IsMidiEvent())
   > SetVisibility(
    self=Get NewSequenceEventChangeButton()
    InVisibility=Select(Option 0:ESlateVisibility Enum=Visible, Option 1:ESlateVisibility Enum=Collapsed, Index=Get NoSequencePool()) -> (ReturnValue:ESlateVisibility Enum)
   )
   "On Pressed (AddNewSequenceButton)"() {is=title K2Node_ComponentBoundEvent}
   > Set NewSequenceButton(NewSequenceButton=true)
   > OpenAddChangeWindow(WindowType=MojoStage_SequenceData, IsJustCreate=true, HideSelectSection=false, IsLoopingEvent=false, KeepSelectChange=false)
   n46 = SuperRefresh() {is=title K2Node_CustomEvent}
   UpdateAndStoreMoods(execute=[@n51.then, @n46.then])
   > n47 = Sequence()
   > ClearOptions(self=Get SequenceTypeComboBox() -> (SequenceTypeComboBox:"Searchable Combo Box (String) Object Reference"))
   > n48 = Keys!(execute, TargetMap:Map of Strings to BP Mojo Stage Mood Data Asset Object References=Get MoodMap()) -> (then, Keys:Array of Strings)
   > n49 = For Each Loop(Array=@n48)
   AddOption(execute=@n49.Completed, self=Get SequenceTypeComboBox(), Option=aaaa)
   > SetSelectedOption(self=Get SequenceTypeComboBox(), Option=aaaa)
   > RemoveOption(self=Get SequenceTypeComboBox(), Option=aaaa)
   > SetSelectedOption(self=Get SequenceTypeComboBox(), Option=Get CurrentMood())
   > "SetText (Text Box)"(self=Get EventGroupTextBox(), InText="To Text (String)"(InString=Get CurrentMood()))
   AddOption(execute=@n49.LoopBody, self=Get SequenceTypeComboBox(), Option=@n49)
   n50 = ForceRefresh() {is=title K2Node_CustomEvent}
   Event Construct()
   > SetVisibility(self=Get ErrorMessage() -> (ErrorMessage:Horizontal Box Object Reference), InVisibility=Collapsed)
   > n51 = SetVisibility(self=Get MainContent() -> (MainContent:Vertical Box Object Reference), InVisibility=SelfHitTestInvisible)
  ]}
 GetAssetDestinationPath {nodes [
   GetAssetDestinationPath() {is=title K2Node_FunctionEntry}
   n1 = Append!(A=/PilgrimQuickplay/StageManager/, B=Data/, C=GetCurrentData().GroupName, D:String=/) -> (ReturnValue)
   > PrintString(InString=@n1, bPrintToScreen=false)
   > Return Node!(execute, FinalPath:String=@n1)
  ]}
 GetCurrentData {nodes [
   GetCurrentData() {is=title K2Node_FunctionEntry}
   > n1 = Branch(Condition="Equal (Name)"(A=Get PropertyName() -> (PropertyName=None), B=MojoStage_SequenceData))
   n2 = GetAsset(InAssetData=Find(
     TargetMap:Map of Strings to Asset Data Structures=Get FilteredEventsMap() -> (FilteredEventsMap:Map of Strings to Asset Data Structures)
     Key=Get CurrentEventSelected()
    ) -> (Value:Asset Data Structure))
   > n3 = Cast To MojoFxManagerSequence_DataAsset!(execute, Object=@n2) -> (then, CastFailed, AsMojo Fx Manager Sequence Data Asset:Mojo Fx Manager Sequence Data Asset Object Reference) {is=title K2Node_DynamicCast}
   > n4 = Set MojoStage_SequenceData(MojoStage_SequenceData:Mojo Fx Manager Sequence Data Asset Object Reference=@n3) -> (Output_Get:Mojo Fx Manager Sequence Data Asset Object Reference)
   > Return Node!(execute, ReturnValue:Object Reference=@n4, AssetBaseName:String=Pilgrim_StageFX_Sequence_Default, GroupName:String=Sequences)
   Return Node(execute=@n3.CastFailed, AssetBaseName=Pilgrim_StageFX_Sequence_Default, GroupName=Sequences)
   n5 = Branch(execute=@n1.else, Condition="Equal (Name)"(A=Get PropertyName() -> (PropertyName=None), B=MojoStage_Sequences PoolData))
   > n6 = Cast To StageShowSequence!(execute, Object=@n2) -> (then, CastFailed, AsStage Show Sequence:Stage Show Sequence Object Reference) {is=title K2Node_DynamicCast}
   > n7 = Set MojoStage_Sequences PoolData(MojoStage_Sequences PoolData=@n6)
   > Return Node(ReturnValue=@n7, AssetBaseName=Pilgrim_StageFX_SequencePool_Default, GroupName=SequencePools)
   Return Node(execute=@n6.CastFailed, AssetBaseName=Pilgrim_StageFX_SequencePool_Default, GroupName=SequencePools)
   n8 = Branch(execute=@n5.else, Condition="Equal (Name)"(A=Get PropertyName() -> (PropertyName=None), B=FxManager_Event StepData))
   > n9 = Cast To MojoFxManagerEvent_DataAsset(Object=@n2)
   > n10 = Set FxManager_Event StepData(FxManager_Event StepData:Mojo Fx Manager Event Data Asset Object Reference=@n9) -> (Output_Get:Mojo Fx Manager Event Data Asset Object Reference)
   > Return Node(ReturnValue=@n10, AssetBaseName=Pilgrim_StageFX_Event_Default, GroupName=Events)
   Return Node(execute=@n9.CastFailed, AssetBaseName=Pilgrim_StageFX_Event_Default, GroupName=Events)
   n11 = Branch(execute=@n8.else, Condition="Equal (Name)"(A=Get PropertyName() -> (PropertyName=None), B=MojoStage_MIDISequenceData))
   > n12 = Cast To BP_MojoStageMapMidiSequence_DataAsset!(execute, Object=@n2) -> (then, CastFailed, AsBP Mojo Stage Map Midi Sequence Data Asset:BP Mojo Stage Map Midi Sequence Data Asset Object Reference) {is=title K2Node_DynamicCast}
   > n13 = Set MojoStage_MIDISequenceData(MojoStage_MIDISequenceData:BP Mojo Stage Map Midi Sequence Data Asset Object Reference=@n12) -> (Output_Get:BP Mojo Stage Map Midi Sequence Data Asset Object Reference)
   > Return Node(ReturnValue=@n13, AssetBaseName=MIDISequenceMap_Default, GroupName=Maps/Sequence)
   Return Node(execute=@n12.CastFailed, AssetBaseName=MIDISequenceMap_Default, GroupName=Maps/Sequence)
   Branch(execute=@n11.else, Condition="Equal (Name)"(A=Get PropertyName() -> (PropertyName=None), B=MojoStage_Mood))
   > n14 = Set MojoStage_Mood(MojoStage_Mood:BP Mojo Stage Mood Data Asset Object Reference=Find(TargetMap:Map of Strings to BP Mojo Stage Mood Data Asset Object References=Get MoodMap(), Key=Get CurrentMood()) -> (Value:BP Mojo Stage Mood Data Asset Object Reference)) -> (Output_Get:BP Mojo Stage Mood Data Asset Object Reference)
   > n15 = Is Valid(InputObject=@n14)
   Return Node(execute=@n15.Is Not Valid, AssetBaseName=DA_Sparks_Festival_Mood_Default, GroupName=Moods)
   > Return Node(ReturnValue=@n14, AssetBaseName=DA_Sparks_Festival_Mood_Default, GroupName=Moods)
  ]}
 GetCurrentMIDIMap {nodes [
   GetCurrentMIDIMap() {is=title K2Node_FunctionEntry}
   n1 = Get SequenceMIDIMap(self=Find(TargetMap=Get MoodMap(), Key=Get CurrentMood()))
   > Return Node!(
    execute
    SequenceMIDIMap:BP Mojo Stage Map Midi Sequence Data Asset Object Reference=@n1
    FilteredName:String=Split(SourceString=GetDisplayName(Object=@n1), InStr=_, SearchDir=FromEnd).RightS
   )
  ]}
 GetCurrentMidiTextEventMap {nodes [
   GetCurrentMidiTextEventMap() {is=title K2Node_FunctionEntry}
   n1 = Find(TargetMap=Get MoodMap(), Key=Get CurrentMood())
   > Return Node!(
    execute
    MIDIEventMap:Map of ESparksMidiTextEvents Enums to Stage Show Sequence Pool Object References=Get EventMap!(self:Stage Show MIDIEvent Map Object Reference=Get SequenceMIDIMap(self=@n1)) -> (EventMap:Map of ESparksMidiTextEvents Enums to Stage Show Sequence Pool Object References=None)
    QPEventMap:Map of EPilgrimQuickplayState Enums to Stage Show Sequence Pool Object References=Get EventMap(self:Stage Show Gameplay Event Map Object Reference=Get GamePlayEventData()) -> (EventMap:Map of EPilgrimQuickplayState Enums to Stage Show Sequence Pool Object References=Invalid)
   )
   Get TempQuickplayEventMap() -> (TempQuickplayEventMap=Invalid)
   Get TempMidiEventMap() -> (TempMidiEventMap=None)
  ]}
 GetSelectedSequencePool {nodes [
   GetSelectedSequencePool() {is=title K2Node_FunctionEntry}
   > Return Node!(
    execute
    Value:Stage Show Sequence Pool Object Reference=Find(TargetMap:Map of Strings to Stage Show Sequence Pool Object References=Get CurrentSequenceMap(), Key=Get CurrentSequence()) -> (Value:Stage Show Sequence Pool Object Reference)
   )
  ]}
 GrabNewAssetName {nodes [
   n1 = GrabNewAssetName!() -> (then, IgnoreFirstIndex:bool=~, WarningDuplicated:bool=~) {is=title K2Node_FunctionEntry}
   > Set LocalWarningDuplicated(LocalWarningDuplicated:bool=@n1.WarningDuplicated) -> (Output_Get:bool)
   > Clear(TargetArray:Array of Asset Data Structures=Get AssetsFilteredSecondPass() -> (AssetsFilteredSecondPass:Array of Asset Data Structures))
   > Clear(TargetMap:Map of Strings to Asset Data Structures=Get AssetsFilteredFirstPass() -> (AssetsFilteredFirstPass:Map of Strings to Asset Data Structures)) {is=title K2Node_CallFunction}
   > Set FoundDuplicated(FoundDuplicated:bool=false) -> (Output_Get:bool)
   > n2 = GetAssetsByPath!(
    execute
    self:Asset Registry Interface=GetAssetRegistry()
    PackagePath:Name=/PilgrimQuickplay/StageManager/Data
    bRecursive:bool=true
    bIncludeOnlyOnDiskAssets:bool=false
   ) -> (then, OutAssetData:Array of Asset Data Structures, ReturnValue:bool)
   > n3 = For Each Loop(Array:Array of Asset Data Structures=@n2) -> (Array Element:Asset Data Structure)
   n4 = Keys(execute=@n3.Completed, TargetMap:Map of Strings to Asset Data Structures=Get AssetsFilteredFirstPass())
   > n5 = For Each Loop(Array:Array of Strings=@n4) -> (Array Element:String)
   n6 = Sequence(execute=@n5.Completed)
   > n7 = Branch(Condition=Get FoundDuplicated() -> (FoundDuplicated:bool))
   n8 = Get PropertyName() -> (PropertyName=None)
   n9 = Get AssetsFilteredSecondPass()
   n10 = "To Text (Integer)"(Value=int + int(A=Length(TargetArray:Array of Asset Data Structures=@n9)), MinimumIntegralDigits=2)
   n11 = Get OutputFileName() -> (OutputFileName:String)
   n12 = GetCurrentData() {is=title K2Node_CallFunction}
   n13 = CheckLoopingEvent!(Object:Object Reference=@n12) -> (IsLoopingEvent:bool) {is=title K2Node_CallFunction}
   n14 = Get PropertyName() -> (PropertyName=None)
   n15 = "Equal (Name)"(A=@n14, B=MojoStage_MIDISequenceData)
   n16 = OR Boolean!(A=@n15, B="Equal (Name)"(A=@n14, B=MojoStage_Mood)) -> (ReturnValue)
   n17 = Find(
    TargetMap:Map of Strings to BP Mojo Stage Map Midi Sequence Data Asset Object References=Get MIDISequenceMap()
    Key=Get CurrentMIDISequenceSelected() -> (CurrentMIDISequenceSelected:String)
   ) -> (Value:BP Mojo Stage Map Midi Sequence Data Asset Object Reference)
   n18 = Find(TargetMap:Map of Strings to BP Mojo Stage Mood Data Asset Object References=Get MoodMap(), Key=Get CurrentMood()) -> (Value:BP Mojo Stage Mood Data Asset Object Reference)
   n19 = Select(
    Option 0:String=Split(SourceString=GetDisplayName(Object=@n18), InStr=_, SearchDir=FromEnd).RightS
    Option 1:String=Split(SourceString=GetDisplayName(Object=@n17), InStr=_, SearchDir=FromEnd).RightS
    Index=@n15
   ) -> (ReturnValue:String)
   n20 = Get CurrentTextEvent()
   n21 = Select(
    Option 0=@n11
    Option 1=Append(
     A=Split(SourceString=@n12.AssetBaseName, InStr=_, SearchDir=FromEnd)
     B=_
     C=Select(Option 0=@n20, Option 1=@n19, Index=@n16)
     D=Select(Option 1=_Looping, Index=@n13)
    )
    Index=IsEmpty(InString=@n11)
   )
   n22 = Append!(A=@n21, B=_) -> (ReturnValue)
   n23 = SelectString(
    A=@n21
    B=Append(A=@n22, B="To String (Text)"(InText=@n10))
    bPickA=OR Boolean(A="Equal (Name)"(A=@n8, B=MojoStage_MIDISequenceData), B="Equal (Name)"(A=@n8, B=MojoStage_Mood))
   )
   > Set FinalResult(FinalResult:String=@n23) -> (Output_Get:String)
   > SetVisibility(
    self=Get NameDuplicatedWarningImage() -> (NameDuplicatedWarningImage:Image Object Reference)
    InVisibility=Select(
     Option 0:ESlateVisibility Enum=Collapsed
     Option 1:ESlateVisibility Enum=Visible
     Index=Get LocalWarningDuplicated() -> (LocalWarningDuplicated:bool)
    ) -> (ReturnValue:ESlateVisibility Enum)
   )
   > SetIsEnabled(
    self=[
     Get NewSequenceEventAddButton() -> (NewSequenceEventAddButton:Editor Utility Button Object Reference)
     Get CreateChangeButton()
     Get RenameEventButton()
    ]
    bInIsEnabled=NOT Boolean(A=Get LocalWarningDuplicated())
   )
   n24 = Get PropertyName() -> (PropertyName=None)
   n25 = "Equal (Name)"(A=@n24, B=MojoStage_MIDISequenceData)
   n26 = "Equal (Name)"(A=@n24, B=MojoStage_Mood)
   n27 = Append(A=@n22, B=01)
   Set FinalResult(
    execute=@n7.else
    FinalResult=SelectString(A=@n21, B=Select(Option 0:String=@n27, Option 1:String=@n21, Index=@n1) -> (ReturnValue:String), bPickA=OR Boolean(A=@n25, B=@n26))
   )
   > SetIsEnabled(self=Get RenameEventButton(), bInIsEnabled=IsValid(Object=GetCurrentData()))
   > SetIsEnabled(self=Get NewSequenceEventAddButton(), bInIsEnabled=true)
   > SetVisibility(self=Get NameDuplicatedWarningImage(), InVisibility=Collapsed)
   n28 = Get FinalResult() -> (FinalResult:String)
   Return Node!(execute=@n6.then_1, OutputName:String=Replace(SourceString=Trim!(SourceString:String=@n28) -> (ReturnValue:String), From=" ", To=_))
   n29 = Find(TargetMap:Map of Strings to Asset Data Structures=Get AssetsFilteredFirstPass(), Key=@n23) -> (Value:Asset Data Structure)
   Branch(execute=@n5.LoopBody, Condition=@n29.ReturnValue)
   > Add(TargetArray:Array of Asset Data Structures=Get AssetsFilteredSecondPass(), NewItem:"Asset Data Structure (by ref)"=@n29) {is=title K2Node_CallArrayFunction}
   > Set FoundDuplicated(FoundDuplicated=true)
   n30 = "To String (Name)"(InName=Break Asset Data!(AssetData=@n3) -> (AssetName))
   n31 = Get PropertyName() -> (PropertyName=None)
   n32 = OR Boolean(A="Equal (Name)"(A=@n31, B=MojoStage_MIDISequenceData), B="Equal (Name)"(A=@n31, B=MojoStage_Mood))
   Branch(execute=@n3.LoopBody, Condition=Contains(SearchIn=@n30, Substring=Select(Option 0=@n22, Option 1=@n21, Index=@n32), bSearchFromEnd=true))
   > Add(TargetMap:Map of Strings to Asset Data Structures=Get AssetsFilteredFirstPass(), Key=@n30, Value:"Asset Data Structure (by ref)"=@n3) {is=title K2Node_CallFunction}
   Get CreateChangeButton()
  ]}
 OnAssetChanged {nodes [OnAssetChanged!() -> (then, ReturnObj:Object Reference, IsNewFile:bool=~) {is=title K2Node_FunctionEntry}]}
 OnPopupClosed {nodes [OnPopupClosed() {is=title K2Node_FunctionEntry}]}
 OpenAddChangeWindow {nodes [
   n1 = OpenAddChangeWindow!() -> (then, WindowType:Name, IsJustCreate:bool=~, HideSelectSection:bool=~, IsLoopingEvent:bool=~, KeepSelectChange:bool=true) {is=title K2Node_FunctionEntry}
   > Set PropertyName(PropertyName:Name=@n1) -> (Output_Get:Name=None)
   > Set IsLoopingEvent(IsLoopingEvent:bool=@n1.IsLoopingEvent) -> (Output_Get:bool)
   > SetVisibility(self=Get NewSequenceEventWindow(), InVisibility=Visible)
   > n2 = Sequence!(execute) -> (then_0, then_1, then_2:Exec)
   > SetVisibility(
    self=[Get CreateChangeButton(), Get SelectChangeGroup()]
    InVisibility=Select(Option 0:ESlateVisibility Enum=Visible, Option 1:ESlateVisibility Enum=Collapsed, Index=@n1.IsJustCreate) -> (ReturnValue:ESlateVisibility Enum)
   )
   > SetVisibility(
    self=[Get ChooseAndAddEvent(), Get NewSequenceEventAddButton()]
    InVisibility=Select(Option 0=Collapsed, Option 1=Visible, Index=@n1.IsJustCreate)
   )
   > PopulateDataAssetComboBox!(execute, self, IsLoopingEvent:bool=@n1.IsLoopingEvent) -> (then)
   > UpdateNewSequenceEventWindow()
   n3 = Split(SourceString="To String (Name)"(InName=Get PropertyName() -> (PropertyName=None)), InStr=_, SearchDir=FromEnd)
   > "SetText (Text)"(
    self=Get EventsTitleText() -> (EventsTitleText:Text Object Reference)
    InText=Format Text!(Format="INVTEXT(\"Add or Change {type} Window\")", type:String=Split(SourceString=@n3.RightS, InStr=D, SearchDir=FromEnd)) -> (Result)
   )
   > SetCheckedState(self=Get SelectAndChangeToggle(), InCheckedState=Checked)
   > SetVisibility(self=Get SelectChangeGroup(), InVisibility=Visible)
   > SetVisibility(self=Get CreateNewDAGroup(), InVisibility=Collapsed)
   > SetCheckedState(self=[Get ToggleViewDa(), Get AddAndCreateToggle()], InCheckedState=Unchecked)
   > SetVisibility(self=[Get SelectAndChangeToggle(), Get SelectDataSectionBorder() -> (SelectDataSectionBorder:Border Object Reference)], InVisibility=Visible)
   n4 = Get PropertyName() -> (PropertyName=None)
   > Branch(Condition=NOT Boolean(A="Equal (Name)"(A=@n4, B=MojoStage_MIDISequenceData)))
   > SetSelectedOption(self=Get DAFilteredComboBox(), Option=Select)
   SetVisibility(
    execute=@n2.then_1
    self=Get NewSequenceEventChangeButton()
    InVisibility=Select(Option 0=Collapsed, Option 1=Visible, Index=@n1.KeepSelectChange)
   )
   > Branch(Condition=@n1.IsJustCreate)
   > Set CurrentEventSelected()
   > Set CurrentSequenceIndex(CurrentSequenceIndex=0) -> (Output_Get=0)
   Branch(execute=@n2.then_2, Condition=AND Boolean(A=@n1.IsJustCreate, B=@n1.HideSelectSection))
   > SetCheckedState(self=Get SelectAndChangeToggle(), InCheckedState=Unchecked)
   > SetCheckedState(self=Get AddAndCreateToggle(), InCheckedState=Checked)
   > SetVisibility(self=Get SelectChangeGroup(), InVisibility=Collapsed)
   > SetVisibility(self=Get CreateNewDAGroup(), InVisibility=Visible)
   > Set OutputFileName()
   n5 = Get PropertyName() -> (PropertyName=None)
   n6 = OR Boolean(A="Equal (Name)"(A=@n5, B=MojoStage_MIDISequenceData), B="Equal (Name)"(A=@n5, B=MojoStage_Mood))
   > "SetText (Text)"(
    self=Get DANamePreviewText()
    InText="To Text (String)"(InString=GrabNewAssetName(IgnoreFirstIndex=@n6, WarningDuplicated=@n6) {is=title K2Node_CallFunction})
   )
   > SetVisibility(self=Get SelectDataSectionBorder(), InVisibility=Collapsed)
  ]}
 PopulateDataAssetComboBox {nodes [
   n1 = PopulateDataAssetComboBox!() -> (then, IsLoopingEvent:bool=~) {is=title K2Node_FunctionEntry}
   > Set LocalLoopingEvent(LocalLoopingEvent:bool=@n1) -> (Output_Get:bool)
   > Clear(TargetArray=Get AssetsFiltered() -> (AssetsFiltered:Array of Asset Data Structures)) {is=title K2Node_CallArrayFunction}
   > ClearOptions(self=Get DAFilteredComboBox())
   > Clear(TargetMap=Get FilteredEventsMap()) {is=title K2Node_CallFunction}
   > AddOption(self=Get DAFilteredComboBox(), Option=Select)
   > n2 = GetAssetsByPath(self=GetAssetRegistry(), PackagePath=&Data, bRecursive=true, bIncludeOnlyOnDiskAssets=false)
   > n3 = For Each Loop(Array:Array of Asset Data Structures=@n2) -> (Array Element:Asset Data Structure)
   AddOption(execute=@n3.Completed, self=Get DAFilteredComboBox(), Option=aaaa)
   > SetSelectedOption(self=Get DAFilteredComboBox(), Option=aaaa)
   n4 = Get CurrentEventSelected()
   > SetSelectedOption(
    self=Get DAFilteredComboBox()
    Option=Select(Option 0:String=@n4, Option 1:String=Select, Index=IsEmpty(InString=@n4)) -> (ReturnValue:String)
   )
   > RemoveOption(self=Get DAFilteredComboBox(), Option=aaaa)
   n5 = Cast To MojoFxManagerEvent_DataAsset(execute=@n3.LoopBody, Object=GetAsset(InAssetData=@n3))
   n6 = "To String (Name)"(InName=Break Asset Data(AssetData=@n3))
   n7 = Contains(SearchIn=@n6, Substring=Split(SourceString=GetCurrentData().AssetBaseName, InStr=D, SearchDir=FromEnd), bSearchFromEnd=true)
   > Branch(Condition=@n7)
   > n8 = Branch(Condition=Get LocalLoopingEvent() -> (LocalLoopingEvent:bool))
   n9 = Get LoopingEvent(self=@n5)
   > n10 = Branch(Condition=@n9)
   Add(execute=[@n10.then, @n11.else, @n12.then], TargetMap=Get FilteredEventsMap(), Key=@n6, Value=@n3)
   > AddOption(self=Get DAFilteredComboBox(), Option=@n6)
   n11 = Branch(execute=@n8.else, Condition=@n9)
   n12 = Branch(execute=@n5.CastFailed, Condition=@n7)
  ]}
 PopulateEventCombobox {nodes [
   PopulateEventCombobox() {is=title K2Node_FunctionEntry}
   > SetVisibility(self=Get Notexteventstext() -> (Notexteventstext:Text Object Reference), InVisibility=Collapsed)
   > SetVisibility(self=Get NewEventsInsideGroup() -> (NewEventsInsideGroup:Overlay Object Reference), InVisibility=Visible)
   > ClearOptions(self=Get NewTextEventCombobox() -> (NewTextEventCombobox:"Searchable Combo Box (String) Object Reference"))
   > Clear(TargetMap:Map of Strings to Strings=Get NewEventsEnumToStringMap())
   > n1 = Sequence!(execute) -> (then_0, then_1)
   > n2 = Branch(Condition=Get IsMidiEvent())
   > n3 = For Each Loop(Array:Array of Strings=GetAllStageManagerEventsMidiTextStrings!(__WorldContext:Object Reference) -> (MidiTextStrings:Array of Strings)) -> (Array Element:String)
   n4 = Enum to String!(Enumerator:EPilgrimEventsMidiTextEvents Enum=GetEventsMidiTextEnumFromString!(MidiTextEventString:String=@n3, __WorldContext:Object Reference) -> (MidiTextEventEnum:EPilgrimEventsMidiTextEvents Enum=NewEnumerator0)) -> (ReturnValue:String)
   > Add(TargetMap:Map of Strings to Strings=Get NewEventsEnumToStringMap(), Key=@n4, Value:"String (by ref)"=@n3)
   > n5 = Keys(TargetMap:Map of Strings to Strings=Get EventsEnumToStringMap())
   > n6 = For Each Loop(Array=@n5)
   > Branch(Condition="Equal Exactly (String)"(A=@n4, B=@n6))
   > Remove(TargetMap:Map of Strings to Strings=Get NewEventsEnumToStringMap(), Key:"String (by ref)"=@n6)
   n7 = For Each Loop(Exec=@n2.else, Array=GetAllStageManagerQPEventStrings!(__WorldContext:Object Reference) -> (QuickPlayStrings:Array of Strings))
   n8 = Enum to String(Enumerator=GetEventsMidiTextEnumFromString(MidiTextEventString=@n7) -> (MidiTextEventEnum=NewEnumerator0))
   > Add(TargetMap=Get NewEventsEnumToStringMap(), Key=@n8, Value=@n7)
   > n9 = Keys(TargetMap=Get EventsEnumToStringMap())
   > n10 = For Each Loop(Array=@n9)
   > Branch(Condition="Equal Exactly (String)"(A=@n8, B=@n10))
   > Remove(TargetMap=Get NewEventsEnumToStringMap(), Key=@n10)
   Remove(execute=@n1.then_1, TargetMap=Get NewEventsEnumToStringMap(), Key=None)
   > n11 = Keys(TargetMap=Get NewEventsEnumToStringMap())
   > n12 = Branch(Condition=Is Empty(TargetArray:Array of Strings=@n11))
   > SetVisibility(self=Get NewEventsInsideGroup(), InVisibility=Collapsed)
   > SetVisibility(self=Get Notexteventstext(), InVisibility=Visible)
   n13 = For Each Loop(Exec=@n12.else, Array=@n11)
   AddOption(execute=@n13.Completed, self=Get NewTextEventCombobox(), Option=aaaa)
   > SetSelectedOption(self=Get NewTextEventCombobox(), Option=aaaa)
   > SetSelectedOption(self=Get NewTextEventCombobox(), Option=GetOptionAtIndex(self=Get NewTextEventCombobox(), Index=0))
   > RemoveOption(self=Get NewTextEventCombobox(), Option=aaaa)
   AddOption(execute=@n13.LoopBody, self=Get NewTextEventCombobox(), Option=@n13)
  ]}
 PopulateMIDITextMapComboBox {nodes [
   PopulateMIDITextMapComboBox() {is=title K2Node_FunctionEntry}
   > Clear(TargetMap:Map of Strings to BP Mojo Stage Map Midi Sequence Data Asset Object References=Get MIDISequenceMap())
   > ClearOptions(self=Get MIDITextMapComboBox())
   > n1 = GetAssetsByPath(self=GetAssetRegistry(), PackagePath=/PilgrimQuickplay/StageManager/Data/Maps/Sequence, bRecursive=false, bIncludeOnlyOnDiskAssets=false)
   > n2 = For Each Loop(Array:Array of Asset Data Structures=@n1) -> (Array Element:Asset Data Structure)
   AddOption(execute=@n2.Completed, self=Get MIDITextMapComboBox(), Option=aaaa)
   > SetSelectedOption(self=Get MIDITextMapComboBox(), Option=aaaa)
   > SetSelectedOption(self=Get MIDITextMapComboBox(), Option=GetCurrentMIDIMap() {is=title K2Node_CallFunction}.FilteredName)
   > RemoveOption(self=Get MIDITextMapComboBox(), Option=aaaa)
   n3 = Cast To BP_MojoStageMapMidiSequence_DataAsset(execute=@n2.LoopBody, Object=GetAsset(InAssetData=@n2))
   n4 = Split(SourceString=GetDisplayName(Object=@n3), InStr=_, SearchDir=FromEnd)
   > Add(
    TargetMap:Map of Strings to BP Mojo Stage Map Midi Sequence Data Asset Object References=Get MIDISequenceMap()
    Key=@n4.RightS
    Value:"BP Mojo Stage Map Midi Sequence Data Asset Object Reference (by ref)"=@n3
   )
   > AddOption(self=Get MIDITextMapComboBox(), Option=@n4.RightS)
  ]}
 PopulateSequencesArray {nodes [
   n1 = PopulateSequencesArray!() -> (then, IsChecked:bool=~) {is=title K2Node_FunctionEntry}
   > n2 = Branch(Condition=@n1)
   > Set IsMidiEvent(IsMidiEvent=true)
   > n3 = SetIsChecked!(
    execute
    self:Check Box Object Reference=Get QuickplayEventsToggle() -> (QuickplayEventsToggle:Editor Utility Check Box Object Reference)
    InIsChecked:bool=false
   ) -> (then)
   ClearChildren(execute=[@n10.then, @n3.then], self=Get SequencesVerticalBox() -> (SequencesVerticalBox:Vertical Box Object Reference))
   n4 = Find(TargetMap:Map of Strings to Stage Show Sequence Pool Object References=Get CurrentSequenceMap(), Key=Get CurrentSequence()) -> (Value:Stage Show Sequence Pool Object Reference)
   > Branch(Condition=@n4.ReturnValue)
   n5 = Get Sequences(self=@n4)
   > n6 = For Each Loop(Array:Array of Stage Show Sequence Object References=@n5) -> (Array Element:Stage Show Sequence Object Reference)
   n7 = Is Empty(TargetArray:Array of Stage Show Sequence Object References=@n5)
   SetVisibility(
    execute=@n6.Completed
    self=Get EmptySequenceGroup()
    InVisibility=Select(Option 0:ESlateVisibility Enum=Collapsed, Option 1:ESlateVisibility Enum=Visible, Index=@n7) -> (ReturnValue:ESlateVisibility Enum)
   )
   > SetVisibility(self=Get SequencesScrollBoxGroup(), InVisibility=Select(Option 0=Visible, Option 1=Collapsed, Index=@n7))
   > SetVisibility(self=Get EmptySequenceGroup(), InVisibility=Select(Option 0=Collapsed, Option 1=Visible, Index=@n7))
   n8 = Reroute Node(InputPin:Object Reference) -> (OutputPin:Object Reference)
   n9 = Create Widget!(
    execute:Exec=@n6.LoopBody
    WorldContextObject:Object Reference
    Class:User Widget Class Reference=&/EWBP_SparksSequencePoolTimelineCell.EWBP_SparksSequencePoolTimelineCell_C
    OwningPlayer:Player Controller Object Reference
    SequenceIndex:int=@n6.Array Index
    SequenceName:Text="To Text (String)"(InString=GetDisplayName(Object=@n6))
    EventSteps:Array of Stage Manager Sequence Steps Structures=Get SequenceSteps!(self:Stage Show Sequence Object Reference=@n6) -> (SequenceSteps:Array of Stage Manager Sequence Steps Structures)
    SparksTimelineEditor:Sparks Stage Manager Timeline Editor Object Reference=Self-Reference()
    SequenceReference:Mojo Fx Manager Sequence Data Asset Object Reference=@n8
   ) -> (then:Exec, ReturnValue:EWBP Sparks Sequence Pool Timeline Cell Object Reference)
   > AddChild(self=Get SequencesVerticalBox(), Content=@n9)
   > Bind Event to Toggle Updated(self:Sparks Stage Manager Timeline Editor Object Reference, Delegate=Create Event(self=@n9)) {is=title K2Node_AddDelegate}
   > UpdateSubmenuSequenceDA!(execute, self:EWBP Sparks Sequence Pool Timeline Cell Object Reference=@n9, EventReference:Object Reference=@n8) -> (then)
   Set IsMidiEvent(execute=@n2.else, IsMidiEvent=false)
   > n10 = SetIsChecked(self=Get MidiTextEventsToggle() -> (MidiTextEventsToggle:Editor Utility Check Box Object Reference), InIsChecked=false)
   Get NoSequencePool()
  ]}
 ToggleUpdated {nodes [ToggleUpdated() {is=title K2Node_FunctionEntry}]}
 UpdateAndStoreMoods {nodes [
   UpdateAndStoreMoods() {is=title K2Node_FunctionEntry}
   > Clear(TargetMap:Map of Strings to BP Mojo Stage Mood Data Asset Object References=Get MoodMap())
   > n1 = GetAssetsByPath(self=GetAssetRegistry(), PackagePath=/PilgrimQuickplay/StageManager/Data/Moods, bRecursive=false, bIncludeOnlyOnDiskAssets=false)
   > n2 = For Each Loop(Array:Array of Asset Data Structures=@n1) -> (Array Element:Asset Data Structure)
   PopulateMIDITextMapComboBox(execute=@n2.Completed) {is=title K2Node_CallFunction}
   n3 = Cast To BP_MojoStageMood_DataAsset!(execute=@n2.LoopBody, Object=GetAsset(InAssetData=@n2)) -> (then, CastFailed, AsBP Mojo Stage Mood Data Asset:BP Mojo Stage Mood Data Asset Object Reference) {is=title K2Node_DynamicCast}
   > Add(
    TargetMap:Map of Strings to BP Mojo Stage Mood Data Asset Object References=Get MoodMap()
    Key=Split(SourceString=GetDisplayName(Object=@n3), InStr=_, SearchDir=FromEnd).RightS
    Value:"BP Mojo Stage Mood Data Asset Object Reference (by ref)"=@n3
   )
  ]}
 UpdateEventComboBox {nodes [
   UpdateEventComboBox() {is=title K2Node_FunctionEntry}
   > Set IsFromInit(IsFromInit=true)
   > ClearOptions(self=Get KeySlotComboBox())
   > n1 = Sequence()
   > Clear(TargetMap:Map of Strings to Strings=Get EventsEnumToStringMap())
   > n2 = Branch(Condition=Get IsMidiEvent())
   > SetVisibility(self=Get SequenceMIDITextMapField() -> (SequenceMIDITextMapField:Horizontal Box Object Reference), InVisibility=Visible)
   > n3 = Keys(TargetMap:Map of ESparksMidiTextEvents Enums to Stage Show Sequence Pool Object References=GetCurrentMidiTextEventMap() -> (MIDIEventMap=None, QPEventMap=Invalid) {is=title K2Node_CallFunction}) -> (Keys:Array of ESparksMidiTextEvents Enums)
   > n4 = For Each Loop(Array:Array of ESparksMidiTextEvents Enums=@n3) -> (Array Element:ESparksMidiTextEvents Enum)
   AddOption(execute=@n4.Completed, self=Get KeySlotComboBox(), Option=a)
   > SetSelectedOption(self=Get KeySlotComboBox(), Option=a)
   > SetSelectedIndex!(
    execute
    self:"Searchable Combo Box (String) Object Reference"=Get KeySlotComboBox()
    Index:int=Get LastTextEventIndex() -> (LastTextEventIndex:int=0)
   ) -> (then)
   > RemoveOption(self=Get KeySlotComboBox(), Option=a)
   n5 = Byte to Enum ESparksStageManager_Events_Midi_Text_Events!(Byte:Byte="To Byte (Integer)"(InInt="To Integer (Byte)"(InByte=@n4) -> (ReturnValue=0)) -> (ReturnValue=0)) -> (ReturnValue:ESparksStageManager_Events_Midi_Text_Events Enum) {is=title K2Node_CastByteToEnum}
   n6 = Enum to String(Enumerator:ESparksStageManager_Events_Midi_Text_Events Enum=@n5)
   Add(
    execute=@n4.LoopBody
    TargetMap:Map of Strings to Strings=Get EventsEnumToStringMap()
    Key=@n6
    Value:"String (by ref)"=GetStageManagerEventsMidiTextStringFromEnum!(MidiTextEventEnum:ESparksStageManager_Events_Midi_Text_Events Enum=@n5, __WorldContext:Object Reference) -> (MidiTextEventString:String)
   )
   > AddOption(self=Get KeySlotComboBox(), Option=@n6)
   SetVisibility(execute=@n2.else, self=Get SequenceMIDITextMapField(), InVisibility=Collapsed)
   > n7 = Keys(TargetMap:Map of EPilgrimQuickplayState Enums to Stage Show Sequence Pool Object References=Get EventMap(self=Get GamePlayEventData()) -> (EventMap=Invalid)) -> (Keys:Array of EPilgrimQuickplayState Enums)
   > n8 = For Each Loop(Array:Array of EPilgrimQuickplayState Enums=@n7) -> (Array Element:EPilgrimQuickplayState Enum)
   AddOption(execute=@n8.Completed, self=Get KeySlotComboBox(), Option=a)
   > SetSelectedOption(self=Get KeySlotComboBox(), Option=a)
   > SetSelectedIndex(self=Get KeySlotComboBox(), Index=Get LastTextEventIndex() -> (LastTextEventIndex=0))
   > RemoveOption(self=Get KeySlotComboBox(), Option=a)
   n9 = Enum to String(Enumerator:EPilgrimQuickplayState Enum=@n8)
   Add(
    execute=@n8.LoopBody
    TargetMap=Get EventsEnumToStringMap()
    Key=@n9
    Value=GetStageManagerQPEventstStringFromEnum!(QPEventEnum:EPilgrimQuickplayState Enum=@n8, __WorldContext:Object Reference) -> (QPEventString:String)
   )
   > AddOption(self=Get KeySlotComboBox(), Option=@n9)
   n10 = "integer > integer"(A=GetOptionCount(self=Get KeySlotComboBox()) -> (ReturnValue=0))
   SetIsEnabled(
    execute=@n1.then_1
    self=[
     Get OpenSequenceDAButton()
     Get DeleteEventButton() -> (DeleteEventButton:Editor Utility Button Object Reference)
     Get AddSequenceDA() -> (AddSequenceDA:Editor Utility Button Object Reference)
     Get AddNewSequenceButton()
     Get DeleteSequenceDA()
    ]
    bInIsEnabled=@n10
   )
   n11 = Get NoSequencePool()
   > SetIsEnabled(
    self=[Get AddNewSequenceButton(), Get DeleteSequenceDA(), Get OpenSequenceDAButton()]
    bInIsEnabled=AND Boolean(A=@n10, B=NOT Boolean(A=@n11))
   )
  ]}
 UpdateMapValues {nodes [
   UpdateMapValues() {is=title K2Node_FunctionEntry}
   > n1 = Sequence()
   n2 = GetCurrentMidiTextEventMap() -> (MIDIEventMap=None, QPEventMap=Invalid)
   > Set TempMidiEventMap(TempMidiEventMap:Map of ESparksMidiTextEvents Enums to Stage Show Sequence Pool Object References=@n2) -> (Output_Get:Map of ESparksMidiTextEvents Enums to Stage Show Sequence Pool Object References=None)
   Set TempQuickplayEventMap(execute=@n1.then_1, TempQuickplayEventMap:Map of EPilgrimQuickplayState Enums to Stage Show Sequence Pool Object References=@n2.QPEventMap) -> (Output_Get:Map of EPilgrimQuickplayState Enums to Stage Show Sequence Pool Object References=Invalid)
  ]}
 UpdateNewSequenceEventWindow {nodes [
   UpdateNewSequenceEventWindow() {is=title K2Node_FunctionEntry}
   n1 = Get DaSlotProperty() -> (DaSlotProperty:Single Property View Object Reference)
   > SetPropertyName!(execute, self:Single Property View Object Reference=@n1, NewPropertyName:Name=Get PropertyName() -> (PropertyName=None)) -> (then)
   > SetObject(self=@n1, NewObject=Self-Reference())
   n2 = Split(SourceString="To String (Name)"(InName=Get PropertyName() -> (PropertyName=None)), InStr=_, SearchDir=FromEnd)
   > "SetText (Text)"(
    self=Get NewSequenceEventCellName() -> (NewSequenceEventCellName:Text Object Reference)
    InText=Format Text!(Format="LOCTEXT(\"{name} \")", name:String=Split(SourceString=@n2.RightS, InStr=D, SearchDir=FromEnd)) -> (Result)
   )
   > SetObject(self=Get EventDetailView(), NewObject=GetCurrentData())
   n3 = GetDisplayName(Object=GetCurrentData())
   n4 = Split(SourceString=GrabNewAssetName(IgnoreFirstIndex=false, WarningDuplicated=true), InStr=_, SearchDir=FromEnd)
   > "SetText (Text Box)"(self=Get RenameAssetTextBox(), InText="To Text (String)"(InString=Split(SourceString=@n3, InStr=@n4, SearchDir=FromEnd).RightS))
   > Set OutputFileName(OutputFileName=@n3)
   > "SetText (Text)"(self=Get DANamePreviewText(), InText="To Text (String)"(InString=GrabNewAssetName(IgnoreFirstIndex=true, WarningDuplicated=false)))
   > "SetText (Text Box)"(self=Get NewFxManagerNameTextBox() -> (NewFxManagerNameTextBox:Editor Utility Editable Text Box Object Reference))
  ]}
 UpdateSequencePool {nodes [
   n1 = UpdateSequencePool!() -> (then, ReturnObject:Object Reference, IsNewAsset:bool=~) {is=title K2Node_FunctionEntry}
   > Set IsNewAssetLocal(IsNewAssetLocal:bool=@n1.IsNewAsset) -> (Output_Get:bool)
   > n2 = Cast To BP_MojoStageSequence_DataAsset!(execute, Object=@n1) -> (then, CastFailed, AsBP Mojo Stage Sequence Data Asset:BP Mojo Stage Sequence Data Asset Object Reference) {is=title K2Node_DynamicCast}
   > Set SequencePoolReturned(SequencePoolReturned:BP Mojo Stage Sequence Data Asset Object Reference=@n2) -> (Output_Get:BP Mojo Stage Sequence Data Asset Object Reference)
   > BeginTransaction(
    Context=AddingDAPoolTimelineTool
    Description="INVTEXT(\"Adding  Data asset from stage manager sequence pool list.\")"
    PrimaryObject=GetCurrentMIDIMap()
   )
   > n3 = Sequence()
   > n4 = Branch(Condition=Get IsMidiEvent())
   n5 = Get TempPool()
   > Clear(TargetArray:Array of Stage Show Sequence Pool Object References=@n5) {is=title K2Node_CallArrayFunction}
   n6 = Get TempMidiEventMap() -> (TempMidiEventMap=None)
   > Add(
    TargetArray:Array of Stage Show Sequence Pool Object References=@n5
    NewItem:"Stage Show Sequence Pool Object Reference (by ref)"=Find(
     TargetMap:Map of ESparksMidiTextEvents Enums to Stage Show Sequence Pool Object References=@n6
     Key:"ESparksMidiTextEvents Enum (by ref)"=Get MidiTextEvents() -> (MidiTextEvents=None)
    )
   ) {is=title K2Node_CallArrayFunction}
   > n7 = Branch(Condition=Get IsNewAssetLocal() -> (IsNewAssetLocal:bool))
   > n8 = Add Unique(
    TargetArray:Array of Stage Show Sequence Pool Object References=@n5
    NewItem:"Stage Show Sequence Pool Object Reference (by ref)"=Get SequencePoolReturned() -> (SequencePoolReturned:BP Mojo Stage Sequence Data Asset Object Reference)
   )
   Add(
    execute=[@n8.then, @n13.then]
    TargetMap:Map of ESparksMidiTextEvents Enums to Stage Show Sequence Pool Object References=@n6
    Key:"ESparksMidiTextEvents Enum (by ref)"=Get MidiTextEvents() -> (MidiTextEvents=None)
    Value:"Stage Show Sequence Pool Object Reference (by ref)"="Get (a copy)"(Array=Get TempPool())
   ) {is=title K2Node_CallFunction}
   n9 = Add Unique(execute=[@n16.else, @n7.else], TargetArray=Get TempPool(), NewItem=Get SequencePoolReturned())
   n10 = Length(TargetArray:Array of Stage Show Sequence Pool Object References=Get TempPool())
   > n11 = Branch(Condition="integer > integer"(A=@n10, B=1))
   > Swap Array Elements!(
    execute
    TargetArray:Array of Stage Show Sequence Pool Object References=Get TempPool()
    FirstIndex:int=FindOptionIndex(self=Get SequencesDataListComboBox(), Option=Get CurrentSequence()) -> (ReturnValue=0)
    SecondIndex:int=@n9
   ) -> (then) {is=title K2Node_CallArrayFunction}
   > n12 = Remove Index(TargetArray=Get TempPool(), IndexToRemove=int - int(A=@n10))
   n13 = Branch(execute=[@n12.then, @n11.else], Condition=Get IsMidiEvent())
   n14 = Get TempQuickplayEventMap() -> (TempQuickplayEventMap=Invalid)
   Add(
    execute=[@n17.then, @n13.else]
    TargetMap:Map of EPilgrimQuickplayState Enums to Stage Show Sequence Pool Object References=@n14
    Key:"EPilgrimQuickplayState Enum (by ref)"=Get QuickPlayEvents() -> (QuickPlayEvents=Invalid)
    Value="Get (a copy)"(Array=Get TempPool())
   )
   n15 = Get TempPool()
   Clear(execute=@n4.else, TargetArray=@n15)
   > Add(
    TargetArray=@n15
    NewItem=Find(
     TargetMap:Map of EPilgrimQuickplayState Enums to Stage Show Sequence Pool Object References=@n14
     Key:"EPilgrimQuickplayState Enum (by ref)"=Get QuickPlayEvents() -> (QuickPlayEvents=Invalid)
    )
   ) {is=title K2Node_CallArrayFunction}
   > n16 = Branch(Condition=Get IsNewAssetLocal())
   > n17 = Add Unique(TargetArray=@n15, NewItem=Get SequencePoolReturned())
   Set EventMap(
    execute=@n3.then_1
    EventMap:Map of ESparksMidiTextEvents Enums to Stage Show Sequence Pool Object References=Get TempMidiEventMap() -> (TempMidiEventMap=None)
    self:Stage Show MIDIEvent Map Object Reference=GetCurrentMIDIMap()
   ) -> (Output_Get:Map of ESparksMidiTextEvents Enums to Stage Show Sequence Pool Object References=None)
   > Set EventMap(
    EventMap:Map of EPilgrimQuickplayState Enums to Stage Show Sequence Pool Object References=Get TempQuickplayEventMap() -> (TempQuickplayEventMap=Invalid)
    self:Stage Show Gameplay Event Map Object Reference=Get GamePlayEventData()
   ) -> (Output_Get:Map of EPilgrimQuickplayState Enums to Stage Show Sequence Pool Object References=Invalid)
   > RemoveOption(self=Get SequencesDataListComboBox(), Option=Empty)
   > TransactObject(Object=GetCurrentMIDIMap())
   > EndTransaction()
   n18 = Get SequencesDataListComboBox()
   n19 = Get SequencesDataListComboBox()
   n20 = Get CurrentSequence()
   > Set CurrentSequenceIndex(CurrentSequenceIndex=SelectInt(A=GetOptionCount(self=@n18) -> (ReturnValue=0), B=FindOptionIndex(self=@n19, Option=@n20) -> (ReturnValue=0), bPickA=Get IsNewAssetLocal())) -> (Output_Get=0)
   > ForceRefresh() {is=title K2Node_CallFunction}
   n21 = Cast To MojoFxManagerSequence_DataAsset(execute=@n2.CastFailed, Object=@n1)
   > Set SequenceReturned(SequenceReturned:Mojo Fx Manager Sequence Data Asset Object Reference=@n21) -> (Output_Get:Mojo Fx Manager Sequence Data Asset Object Reference)
   > Branch(Condition=Get NewSequenceButton() -> (NewSequenceButton:bool))
   > BeginTransaction(
    Context=AddingDAPoolTimelineTool
    Description="INVTEXT(\"Adding  Data asset from stage manager sequence pool list.\")"
    PrimaryObject=GetCurrentMIDIMap()
   )
   > Add(
    TargetArray:Array of Stage Show Sequence Object References=Get Sequences(self=GetSelectedSequencePool!() -> (Value:Stage Show Sequence Pool Object Reference) {is=title K2Node_CallFunction})
    NewItem:"Stage Show Sequence Object Reference (by ref)"=Get SequenceReturned() -> (SequenceReturned:Mojo Fx Manager Sequence Data Asset Object Reference)
   )
   > TransactObject(Object=GetCurrentMIDIMap())
   > EndTransaction()
   > ForceRefresh()
   n22 = Get PropertyName() -> (PropertyName=None)
   n23 = Get PropertyName() -> (PropertyName=None)
   Branch(
    execute=@n21.CastFailed
    Condition=OR Boolean(A="Equal (Name)"(A=@n22, B=MojoStage_Mood), B="Equal (Name)"(A=@n23, B=MojoStage_MIDISequenceData))
   )
   > n24 = Cast To BP_MojoStageMapMidiSequence_DataAsset(Object=@n1)
   n25 = Find(TargetMap:Map of Strings to BP Mojo Stage Mood Data Asset Object References=Get MoodMap(), Key:"String (by ref)"=Get CurrentMood()) -> (Value:BP Mojo Stage Mood Data Asset Object Reference)
   > Set SequenceMIDIMap(SequenceMIDIMap=@n24, self=@n25)
   > TransactObject(Object=@n25)
   > SetIsEnabled(self=Get DeleteMIDITextMapBtn(), bInIsEnabled=true)
   > SuperRefresh() {is=title K2Node_CallFunction}
   n26 = Cast To BP_MojoStageMood_DataAsset(execute=@n24.CastFailed, Object=@n1)
   > Set CurrentMood(CurrentMood=Split(SourceString=GetDisplayName(Object=@n26), InStr=_, SearchDir=FromEnd).RightS)
   > SuperRefresh()
   Find(
    TargetMap:Map of EPilgrimQuickplayState Enums to Stage Show Sequence Pool Object References=Get TempQuickplayEventMap() -> (TempQuickplayEventMap=Invalid)
    Key:"EPilgrimQuickplayState Enum (by ref)"=Get QuickPlayEvents() -> (QuickPlayEvents=Invalid)
   ) -> (Value:Stage Show Sequence Pool Object Reference)
   Get MojoStage_SequenceData() -> (MojoStage_SequenceData:Mojo Fx Manager Sequence Data Asset Object Reference)
  ]}
 UpdateToggles {nodes [
   n1 = UpdateToggles!() -> (then, Toggle:EWBP Sparks Sequence Pool Timeline Event Bar Object Reference, ForceCloseAll:bool=~, DontCallEvent:bool=~) {is=title K2Node_FunctionEntry}
   > n2 = For Each Loop(Array:Array of EWBP Sparks Sequence Pool Timeline Event Bar Object References=Get AllEvents()) -> (Array Element:EWBP Sparks Sequence Pool Timeline Event Bar Object Reference)
   n3 = Branch(execute=@n2.Completed, Condition=@n1.DontCallEvent)
   Call Toggle Updated!(execute:Exec=@n3.else, self:Sparks Stage Manager Timeline Editor Object Reference) -> (then:Exec) {is=title K2Node_CallDelegate}
   n4 = Is Valid(exec=@n2.LoopBody, InputObject=@n1)
   Branch(execute=@n4.Is Not Valid, Condition=@n1.ForceCloseAll)
   > UpdateSubmenu!(execute, self:EWBP Sparks Sequence Pool Timeline Event Bar Object Reference=@n2, isEnable:bool=false) -> (then)
   n5 = Branch(execute=@n4.Is Valid, Condition="Equal (Object)"(A=@n1, B=@n2))
   > UpdateSubmenu(self=@n2, isEnable=true)
   UpdateSubmenu(execute=@n5.else, self=@n2, isEnable=false)
  ]}
]
variables [
 AllEvents {type="array<EWBP_SparksSequencePool_TimelineEventBar_C>", default []}
 AssetsBasePath {type=string, default=/PilgrimQuickplay/StageManager/BaseData/}
 CurrentEventBarReference {type=EWBP_SparksSequencePool_TimelineEventBar_C, default=null}
 CurrentEventSelected {type=string, default=""}
 CurrentEventType {type=string, default=Default}
 CurrentMIDISequenceSelected {type=string, default=""}
 CurrentMood {type=string, default=Default}
 CurrentSequence {type=string, default=""}
 CurrentSequenceIndex {type=integer}
 CurrentSequenceMap {type="map<StageShowSequencePool>", default {}}
 CurrentTextEvent {type=string, default=""}
 EventsEnumToStringMap {type="map<string>", default {}}
 FilteredEventsMap {type="map<AssetData>", default {}}
 FxManager_Event StepData {type=MojoFxManagerEvent_DataAsset_C, default=/PilgrimQuickplay/StageManager/Data/Events/Pilgrim_StageFX_Event_Bridge_LightsRingUpper_Pattern}
 GamePlayEventData {type=BP_MojoStageMapGameplaySequence_DataAsset_C, default=/PilgrimQuickplay/StageManager/Data/Maps/Sequence/GameplaySequenceMap_Default}
 IsFromInit {type=boolean, default=true}
 IsLoopingEvent {type=boolean}
 IsMidiEvent {type=boolean, default=true}
 LastEventIndex {type=integer}
 LastSequencePool {type=StageShowSequencePool, default=null}
 LastTextEventIndex {type=integer}
 LevelToEdit {type=BuildingLevelInstance, default=null}
 Local Warning Duplicated {type=boolean}
 MIDISequenceMap {type="map<BP_MojoStageMapMidiSequence_DataAsset_C>", default {}}
 MidiTextEvents {type=ESparksMidiTextEvents, default=ESparksMidiTextEvents_MAX}
 MojoStage_MIDISequenceData {type=BP_MojoStageMapMidiSequence_DataAsset_C, default=null}
 MojoStage_Mood {type=BP_MojoStageMood_DataAsset_C, default=null}
 MojoStage_SequenceData {
  type=MojoFxManagerSequence_DataAsset_C
  default=/PilgrimQuickplay/StageManager/Data/Sequences/Pilgrim_StageFX_Sequence_Generic_LsStageBeamsPs_01a
 }
 MojoStage_Sequences PoolData {type=StageShowSequence, default=null}
 MoodMap {type="map<BP_MojoStageMood_DataAsset_C>", default {}}
 NewEventSelected {type=string, default=""}
 NewEventsEnumToStringMap {type="map<string>", default {}}
 NewSequenceButton {type=boolean}
 NoSequencePool {type=boolean}
 OnAssetChanged {default=()}
 OnPopupClosed {default=()}
 OutputFileName {type=string, default=""}
 PropertyName {type=string, default=MojoStage_SequenceData}
 QuickPlayEvents {type=EPilgrimQuickplayState, default=Loading}
 TempMidiEventMap {default {}}
 TempPool {type="array<StageShowSequencePool>", default []}
 TempQuickplayEventMap {default {}}
 ToggleUpdated {default=()}
]
widgets {
 info {rootWidgetClass=&Overlay, widgetCount=162}
 widgets [
  Overlay_0 {widgetClassPath=&Overlay, widget=Overlay_0}
  MainContent {
   widgetClassPath=&VerticalBox
   bIsVariable=true
   parent=Overlay_0
   widget=MainContent
   slot {class=OverlaySlot, properties {horizontalAlignment=HAlign_Fill, padding=<ltrb 10 15 0 0>}}
  }
  HorizontalBox_0 {widgetClassPath=&HorizontalBox, parent=MainContent, widget=HorizontalBox_0, slot {class=VerticalBoxSlot}}
  EventTypeName {
   widgetClassPath=&TextBlock
   parent=HorizontalBox_0
   widget=EventTypeName
   properties {
    text="Selected Mood: "
    colorAndOpacity {colorUseRule=UseColor_Specified, specifiedColor=<rgba 0.6562 0.6562 0.6562 1>}
    font {
     bForceMonospaced=false
     bMaterialIsStencil=false
     fontMaterial=null
     fontObject=&HeadingNow
     letterSpacing=0
     monospacedWidth=1
     size=7.5
     skewAmount=0
     typefaceFontName=75 Medium
     outlineSettings {
      bApplyOutlineToDropShadows=false
      bMiteredCorners=false
      bSeparateFillAlpha=false
      outlineMaterial=null
      outlineSize=0
      outlineColor=<rgba 0 0 0 1>
     }
    }
   }
   slot {class=HorizontalBoxSlot, properties {verticalAlignment=VAlign_Center, padding=<ltrb 0 0 5 0>}}
  }
  SequenceTypeComboBox {
   widgetClassPath=/Script/FortniteEditor.SearchableStringComboBox
   bIsVariable=true
   parent=HorizontalBox_0
   widget=SequenceTypeComboBox
   properties {
    selectedOption=Sequence Type List
    font {
     bForceMonospaced=false
     bMaterialIsStencil=false
     fontMaterial=null
     fontObject=&HeadingNow
     letterSpacing=0
     monospacedWidth=1
     size=7.5
     skewAmount=0
     typefaceFontName=75 Medium
     outlineSettings {
      bApplyOutlineToDropShadows=false
      bMiteredCorners=false
      bSeparateFillAlpha=false
      outlineMaterial=null
      outlineSize=0
      outlineColor=<rgba 0 0 0 1>
     }
    }
    foregroundColor {colorUseRule=UseColor_Specified, specifiedColor=<rgba 0.5271 0.5271 0.5271 1>}
    itemStyle {
     bUseParentRowBrush=false
     activeBrush {
      bIsDynamicallyLoaded=false
      drawAs=Image
      imageType=NoImage
      mirroring=NoMirror
      resourceName=None
      resourceObject=null
      tiling=NoTile
      imageSize=<0 0>
      margin=<ltrb 0 0 0 0>
      outlineSettings {
       bUseBrushTransparency=false
       roundingType=HalfHeightRadius
       width=0
       color {colorUseRule=UseColor_Specified, specifiedColor=<rgba 0 0 0 0>}
       cornerRadii=<0 0 0 0>
      }
      tintColor {colorUseRule=UseColor_Specified, specifiedColor=<rgba 0 0.162 0.7454 1>}
      uVRegion {bIsValid=false, max=<0 0>, min=<0 0>}
     }
     activeHighlightedBrush {
      bIsDynamicallyLoaded=false
      drawAs=Image
      imageType=NoImage
      mirroring=NoMirror
      resourceName=None
      resourceObject=null
      tiling=NoTile
      imageSize=<0 0>
      margin=<ltrb 0 0 0 0>
      outlineSettings {
       bUseBrushTransparency=false
       roundingType=HalfHeightRadius
       width=0
       color {colorUseRule=UseColor_Specified, specifiedColor=<rgba 0 0 0 0>}
       cornerRadii=<0 0 0 0>
      }
      tintColor {colorUseRule=UseColor_Specified, specifiedColor=<rgba 0.0252 0.0319 0.0423 1>}
      uVRegion {bIsValid=false, max=<0 0>, min=<0 0>}
     }
     activeHoveredBrush {
      bIsDynamicallyLoaded=false
      drawAs=Image
      imageType=NoImage
      mirroring=NoMirror
      resourceName=None
      resourceObject=null
      tiling=NoTile
      imageSize=<0 0>
      margin=<ltrb 0 0 0 0>
      outlineSettings {
       bUseBrushTransparency=false
       roundingType=HalfHeightRadius
       width=0
       color {colorUseRule=UseColor_Specified, specifiedColor=<rgba 0 0 0 0>}
       cornerRadii=<0 0 0 0>
      }
      tintColor {colorUseRule=UseColor_Specified, specifiedColor=<rgba 0 0.162 0.7454 1>}
      uVRegion {bIsValid=false, max=<0 0>, min=<0 0>}
     }
     dropIndicator_Above {
      bIsDynamicallyLoaded=false
      drawAs=Box
      imageType=FullColor
      mirroring=NoMirror
      resourceName=../../../Engine/Content/Slate/Common/DropZoneIndicator_Above.png
      resourceObject=null
      tiling=NoTile
      imageSize=<0 0>
      margin=<ltrb 0.625 0.625 0 0>
      outlineSettings {
       bUseBrushTransparency=false
       roundingType=HalfHeightRadius
       width=0
       color {colorUseRule=UseColor_Specified, specifiedColor=<rgba 0 0 0 0>}
       cornerRadii=<0 0 0 0>
      }
      tintColor {colorUseRule=UseColor_Specified, specifiedColor=<rgba 0.0091 0.3515 0.9647 1>}
      uVRegion {bIsValid=false, max=<0 0>, min=<0 0>}
     }
     dropIndicator_Below {
      bIsDynamicallyLoaded=false
      drawAs=Box
      imageType=FullColor
      mirroring=NoMirror
      resourceName=../../../Engine/Content/Slate/Common/DropZoneIndicator_Below.png
      resourceObject=null
      tiling=NoTile
      imageSize=<0 0>
      margin=<ltrb 0.625 0 0 0.625>
      outlineSettings {
       bUseBrushTransparency=false
       roundingType=HalfHeightRadius
       width=0
       color {colorUseRule=UseColor_Specified, specifiedColor=<rgba 0 0 0 0>}
       cornerRadii=<0 0 0 0>
      }
      tintColor {colorUseRule=UseColor_Specified, specifiedColor=<rgba 0.0091 0.3515 0.9647 1>}
      uVRegion {bIsValid=false, max=<0 0>, min=<0 0>}
     }
     dropIndicator_Onto {
      bIsDynamicallyLoaded=false
      drawAs=Box
      imageType=FullColor
      mirroring=NoMirror
      resourceName=../../../Engine/Content/Slate/Common/DropZoneIndicator_Onto.png
      resourceObject=null
      tiling=NoTile
      imageSize=<0 0>
      margin=<ltrb 0.25 0.25 0.25 0.25>
      outlineSettings {
       bUseBrushTransparency=false
       roundingType=HalfHeightRadius
       width=0
       color {colorUseRule=UseColor_Specified, specifiedColor=<rgba 0 0 0 0>}
       cornerRadii=<0 0 0 0>
      }
      tintColor {colorUseRule=UseColor_Specified, specifiedColor=<rgba 0.0091 0.3515 0.9647 1>}
      uVRegion {bIsValid=false, max=<0 0>, min=<0 0>}
     }
     evenRowBackgroundBrush {
      bIsDynamicallyLoaded=false
      drawAs=Image
      imageType=NoImage
      mirroring=NoMirror
      resourceName=None
      resourceObject=null
      tiling=NoTile
      imageSize=<0 0>
      margin=<ltrb 0 0 0 0>
      outlineSettings {
       bUseBrushTransparency=false
       roundingType=HalfHeightRadius
       width=0
       color {colorUseRule=UseColor_Specified, specifiedColor=<rgba 0 0 0 0>}
       cornerRadii=<0 0 0 0>
      }
      tintColor {colorUseRule=UseColor_Specified, specifiedColor=<rgba 0.0103 0.0103 0.0103 1>}
      uVRegion {bIsValid=false, max=<0 0>, min=<0 0>}
     }
     evenRowBackgroundHoveredBrush {
      bIsDynamicallyLoaded=false
      drawAs=Image
      imageType=NoImage
      mirroring=NoMirror
      resourceName=None
      resourceObject=null
      tiling=NoTile
      imageSize=<0 0>
      margin=<ltrb 0 0 0 0>
      outlineSettings {
       bUseBrushTransparency=false
       roundingType=HalfHeightRadius
       width=0
       color {colorUseRule=UseColor_Specified, specifiedColor=<rgba 0 0 0 0>}
       cornerRadii=<0 0 0 0>
      }
      tintColor {colorUseRule=UseColor_Specified, specifiedColor=<rgba 0.0176 0.0176 0.0176 1>}
      uVRegion {bIsValid=false, max=<0 0>, min=<0 0>}
     }
     inactiveBrush {
      bIsDynamicallyLoaded=false
      drawAs=Image
      imageType=NoImage
      mirroring=NoMirror
      resourceName=None
      resourceObject=null
      tiling=NoTile
      imageSize=<0 0>
      margin=<ltrb 0 0 0 0>
      outlineSettings {
       bUseBrushTransparency=false
       roundingType=HalfHeightRadius
       width=0
       color {colorUseRule=UseColor_Specified, specifiedColor=<rgba 0 0 0 0>}
       cornerRadii=<0 0 0 0>
      }
      tintColor {colorUseRule=UseColor_Specified, specifiedColor=<rgba 0.0513 0.0953 0.159 1>}
      uVRegion {bIsValid=false, max=<0 0>, min=<0 0>}
     }
     inactiveHighlightedBrush {
      bIsDynamicallyLoaded=false
      drawAs=Image
      imageType=NoImage
      mirroring=NoMirror
      resourceName=None
      resourceObject=null
      tiling=NoTile
      imageSize=<0 0>
      margin=<ltrb 0 0 0 0>
      outlineSettings {
       bUseBrushTransparency=false
       roundingType=HalfHeightRadius
       width=0
       color {colorUseRule=UseColor_Specified, specifiedColor=<rgba 0 0 0 0>}
       cornerRadii=<0 0 0 0>
      }
      tintColor {colorUseRule=UseColor_Specified, specifiedColor=<rgba 0.0252 0.0319 0.0423 1>}
      uVRegion {bIsValid=false, max=<0 0>, min=<0 0>}
     }
     inactiveHoveredBrush {
      bIsDynamicallyLoaded=false
      drawAs=Image
      imageType=NoImage
      mirroring=NoMirror
      resourceName=None
      resourceObject=null
      tiling=NoTile
      imageSize=<0 0>
      margin=<ltrb 0 0 0 0>
      outlineSettings {
       bUseBrushTransparency=false
       roundingType=HalfHeightRadius
       width=0
       color {colorUseRule=UseColor_Specified, specifiedColor=<rgba 0 0 0 0>}
       cornerRadii=<0 0 0 0>
      }
      tintColor {colorUseRule=UseColor_Specified, specifiedColor=<rgba 0.0513 0.0953 0.159 1>}
      uVRegion {bIsValid=false, max=<0 0>, min=<0 0>}
     }
     oddRowBackgroundBrush {
      bIsDynamicallyLoaded=false
      drawAs=Image
      imageType=NoImage
      mirroring=NoMirror
      resourceName=None
      resourceObject=null
      tiling=NoTile
      imageSize=<0 0>
      margin=<ltrb 0 0 0 0>
      outlineSettings {
       bUseBrushTransparency=false
       roundingType=HalfHeightRadius
       width=0
       color {colorUseRule=UseColor_Specified, specifiedColor=<rgba 0 0 0 0>}
       cornerRadii=<0 0 0 0>
      }
      tintColor {colorUseRule=UseColor_Specified, specifiedColor=<rgba 0.0103 0.0103 0.0103 1>}
      uVRegion {bIsValid=false, max=<0 0>, min=<0 0>}
     }
     oddRowBackgroundHoveredBrush {
      bIsDynamicallyLoaded=false
      drawAs=Image
      imageType=NoImage
      mirroring=NoMirror
      resourceName=None
      resourceObject=null
      tiling=NoTile
      imageSize=<0 0>
      margin=<ltrb 0 0 0 0>
      outlineSettings {
       bUseBrushTransparency=false
       roundingType=HalfHeightRadius
       width=0
       color {colorUseRule=UseColor_Specified, specifiedColor=<rgba 0 0 0 0>}
       cornerRadii=<0 0 0 0>
      }
      tintColor {colorUseRule=UseColor_Specified, specifiedColor=<rgba 0.0176 0.0176 0.0176 1>}
      uVRegion {bIsValid=false, max=<0 0>, min=<0 0>}
     }
     parentRowBackgroundBrush {
      bIsDynamicallyLoaded=false
      drawAs=Image
      imageType=NoImage
      mirroring=NoMirror
      resourceName=None
      resourceObject=null
      tiling=NoTile
      imageSize=<32 32>
      margin=<ltrb 0 0 0 0>
      outlineSettings {
       bUseBrushTransparency=false
       roundingType=HalfHeightRadius
       width=0
       color {colorUseRule=UseColor_Specified, specifiedColor=<rgba 0 0 0 0>}
       cornerRadii=<0 0 0 0>
      }
      tintColor {colorUseRule=UseColor_Specified, specifiedColor=<rgba 1 1 1 1>}
      uVRegion {bIsValid=false, max=<0 0>, min=<0 0>}
     }
     parentRowBackgroundHoveredBrush {
      bIsDynamicallyLoaded=false
      drawAs=Image
      imageType=NoImage
      mirroring=NoMirror
      resourceName=None
      resourceObject=null
      tiling=NoTile
      imageSize=<32 32>
      margin=<ltrb 0 0 0 0>
      outlineSettings {
       bUseBrushTransparency=false
       roundingType=HalfHeightRadius
       width=0
       color {colorUseRule=UseColor_Specified, specifiedColor=<rgba 0 0 0 0>}
       cornerRadii=<0 0 0 0>
      }
      tintColor {colorUseRule=UseColor_Specified, specifiedColor=<rgba 1 1 1 1>}
      uVRegion {bIsValid=false, max=<0 0>, min=<0 0>}
     }
     selectedTextColor {colorUseRule=UseColor_Specified, specifiedColor=<rgba 1 1 1 1>}
     selectorFocusedBrush {
      bIsDynamicallyLoaded=false
      drawAs=Border
      imageType=FullColor
      mirroring=NoMirror
      resourceName=../../../Engine/Content/Slate/Common/Selector.png
      resourceObject=null
      tiling=Both
      imageSize=<0 0>
      margin=<ltrb 0.25 0.25 0.25 0.25>
      outlineSettings {
       bUseBrushTransparency=false
       roundingType=HalfHeightRadius
       width=0
       color {colorUseRule=UseColor_Specified, specifiedColor=<rgba 0 0 0 0>}
       cornerRadii=<0 0 0 0>
      }
      tintColor {colorUseRule=UseColor_Specified, specifiedColor=<rgba 0.701 0.225 0.003 1>}
      uVRegion {bIsValid=false, max=<0 0>, min=<0 0>}
     }
     textColor {colorUseRule=UseColor_Specified, specifiedColor=<rgba 0.5271 0.5271 0.5271 1>}
    }
    widgetStyle {
     comboButtonStyle {
      downArrowAlign=VAlign_Center
      buttonStyle {
       clickedSlateSound {resourceObject=null}
       disabled {
        bIsDynamicallyLoaded=false
        drawAs=RoundedBox
        imageType=NoImage
        mirroring=NoMirror
        resourceName=None
        resourceObject=null
        tiling=NoTile
        imageSize=<0 0>
        margin=<ltrb 0 0 0 0>
        outlineSettings {
         bUseBrushTransparency=false
         roundingType=FixedRadius
         width=1
         color {colorUseRule=UseColor_ColorTable, specifiedColor=<rgba 0 0 0 0>}
         cornerRadii=<4 4 4 4>
        }
        tintColor {colorUseRule=UseColor_Specified, specifiedColor=<rgba 0.0395 0.0395 0.0395 1>}
        uVRegion {bIsValid=false, max=<0 0>, min=<0 0>}
       }
       disabledForeground {colorUseRule=UseColor_ColorTable, specifiedColor=<rgba 0 0 0 0>}
       hovered {
        bIsDynamicallyLoaded=false
        drawAs=RoundedBox
        imageType=NoImage
        mirroring=NoMirror
        resourceName=None
        resourceObject=null
        tiling=NoTile
        imageSize=<0 0>
        margin=<ltrb 0 0 0 0>
        outlineSettings {
         bUseBrushTransparency=false
         roundingType=FixedRadius
         width=1
         color {colorUseRule=UseColor_ColorTable, specifiedColor=<rgba 0 0 0 0>}
         cornerRadii=<4 4 4 4>
        }
        tintColor {colorUseRule=UseColor_Specified, specifiedColor=<rgba 0.0048 0.0048 0.0048 1>}
        uVRegion {bIsValid=false, max=<0 0>, min=<0 0>}
       }
       hoveredForeground {colorUseRule=UseColor_ColorTable, specifiedColor=<rgba 0 0 0 0>}
       hoveredSlateSound {resourceObject=null}
       normal {
        bIsDynamicallyLoaded=false
        drawAs=RoundedBox
        imageType=NoImage
        mirroring=NoMirror
        resourceName=None
        resourceObject=null
        tiling=NoTile
        imageSize=<0 0>
        margin=<ltrb 0 0 0 0>
        outlineSettings {
         bUseBrushTransparency=false
         roundingType=FixedRadius
         width=1
         color {colorUseRule=UseColor_ColorTable, specifiedColor=<rgba 0 0 0 0>}
         cornerRadii=<4 4 4 4>
        }
        tintColor {colorUseRule=UseColor_Specified, specifiedColor=<rgba 0.0048 0.0048 0.0048 1>}
        uVRegion {bIsValid=false, max=<0 0>, min=<0 0>}
       }
       normalForeground {colorUseRule=UseColor_ColorTable, specifiedColor=<rgba 0 0 0 0>}
       normalPadding=<ltrb 8 1 8 1>
       pressed {
        bIsDynamicallyLoaded=false
        drawAs=RoundedBox
        imageType=NoImage
        mirroring=NoMirror
        resourceName=None
        resourceObject=null
        tiling=NoTile
        imageSize=<0 0>
        margin=<ltrb 0 0 0 0>
        outlineSettings {
         bUseBrushTransparency=false
         roundingType=FixedRadius
         width=1
         color {colorUseRule=UseColor_ColorTable, specifiedColor=<rgba 0 0 0 0>}
         cornerRadii=<4 4 4 4>
        }
        tintColor {colorUseRule=UseColor_Specified, specifiedColor=<rgba 0.0048 0.0048 0.0048 1>}
        uVRegion {bIsValid=false, max=<0 0>, min=<0 0>}
       }
       pressedForeground {colorUseRule=UseColor_ColorTable, specifiedColor=<rgba 0 0 0 0>}
       pressedPadding=<ltrb 8 2 8 0>
       pressedSlateSound {resourceObject=null}
      }
      contentPadding=<ltrb 0 0 0 0>
      downArrowImage {
       bIsDynamicallyLoaded=false
       drawAs=Image
       imageType=Vector
       mirroring=NoMirror
       resourceName=../../../Engine/Content/Slate/Starship/CoreWidgets/ComboBox/wide-chevron-down.svg
       resourceObject=null
       tiling=NoTile
       imageSize=<16 16>
       margin=<ltrb 0 0 0 0>
       outlineSettings {
        bUseBrushTransparency=false
        roundingType=HalfHeightRadius
        width=0
        color {colorUseRule=UseColor_Specified, specifiedColor=<rgba 0 0 0 0>}
        cornerRadii=<0 0 0 0>
       }
       tintColor {colorUseRule=UseColor_Specified, specifiedColor=<rgba 1 1 1 1>}
       uVRegion {bIsValid=false, max=<0 0>, min=<0 0>}
      }
      downArrowPadding=<ltrb 2 2 2 2>
      menuBorderBrush {
       bIsDynamicallyLoaded=false
       drawAs=RoundedBox
       imageType=NoImage
       mirroring=NoMirror
       resourceName=None
       resourceObject=null
       tiling=NoTile
       imageSize=<0 0>
       margin=<ltrb 0 0 0 0>
       outlineSettings {
        bUseBrushTransparency=false
        roundingType=FixedRadius
        width=1
        color {colorUseRule=UseColor_ColorTable, specifiedColor=<rgba 0 0 0 0>}
        cornerRadii=<0 0 0 0>
       }
       tintColor {colorUseRule=UseColor_Specified, specifiedColor=<rgba 0.0395 0.0395 0.0395 1>}
       uVRegion {bIsValid=false, max=<0 0>, min=<0 0>}
      }
      menuBorderPadding=<ltrb 0 0 0 0>
      shadowColorAndOpacity=<rgba 0 0 0 1>
      shadowOffset=<0 0>
     }
     contentPadding=<ltrb 0 0 0 0>
     menuRowPadding=<ltrb 8 3 8 3>
     pressedSlateSound {resourceObject=null}
     selectionChangeSlateSound {resourceObject=null}
    }
   }
   slot {class=HorizontalBoxSlot, properties {horizontalAlignment=HAlign_Left, verticalAlignment=VAlign_Center}}
  }
  OpenEditorWithThemedGroup {
   widgetClassPath=&EditorUtilityButton
   parent=HorizontalBox_0
   widget=OpenEditorWithThemedGroup
   properties {widgetStyle {
     clickedSlateSound {resourceObject=null}
     disabled {
      bIsDynamicallyLoaded=false
      drawAs=RoundedBox
      imageType=NoImage
      mirroring=NoMirror
      resourceName=None
      resourceObject=/Game/UI/Foundation/Textures/Buttons/T-Button-Square
      tiling=NoTile
      imageSize=<32 32>
      margin=<ltrb 0.5 0.5 0.5 0.5>
      outlineSettings {
       bUseBrushTransparency=false
       roundingType=FixedRadius
       width=0
       color {colorUseRule=UseColor_ColorTable, specifiedColor=<rgba 0 0 0 0>}
       cornerRadii=<4 4 4 4>
      }
      tintColor {colorUseRule=UseColor_Specified, specifiedColor=<rgba 0.0395 0.0395 0.0395 1>}
      uVRegion {bIsValid=false, max=<0 0>, min=<0 0>}
     }
     disabledForeground {colorUseRule=UseColor_ColorTable, specifiedColor=<rgba 0 0 0 0>}
     hovered {
      bIsDynamicallyLoaded=false
      drawAs=RoundedBox
      imageType=NoImage
      mirroring=NoMirror
      resourceName=None
      resourceObject=&T-Button-Square
      tiling=NoTile
      imageSize=<32 32>
      margin=<ltrb 0.5 0.5 0.5 0.5>
      outlineSettings {
       bUseBrushTransparency=false
       roundingType=FixedRadius
       width=0
       color {colorUseRule=UseColor_ColorTable, specifiedColor=<rgba 0 0 0 0>}
       cornerRadii=<4 4 4 4>
      }
      tintColor {colorUseRule=UseColor_Specified, specifiedColor=<rgba 0 0.3021 0.0241 1>}
      uVRegion {bIsValid=false, max=<0 0>, min=<0 0>}
     }
     hoveredForeground {colorUseRule=UseColor_ColorTable, specifiedColor=<rgba 0 0 0 0>}
     hoveredSlateSound {resourceObject=null}
     normal {
      bIsDynamicallyLoaded=false
      drawAs=RoundedBox
      imageType=NoImage
      mirroring=NoMirror
      resourceName=None
      resourceObject=&T-Button-Square
      tiling=NoTile
      imageSize=<32 32>
      margin=<ltrb 0.5 0.5 0.5 0.5>
      outlineSettings {
       bUseBrushTransparency=false
       roundingType=FixedRadius
       width=0
       color {colorUseRule=UseColor_ColorTable, specifiedColor=<rgba 0 0 0 0>}
       cornerRadii=<4 4 4 4>
      }
      tintColor {colorUseRule=UseColor_Specified, specifiedColor=<rgba 0.0071 0.7594 0 1>}
      uVRegion {bIsValid=false, max=<0 0>, min=<0 0>}
     }
     normalForeground {colorUseRule=UseColor_ColorTable, specifiedColor=<rgba 0 0 0 0>}
     normalPadding=<ltrb 12 1.5 12 1.5>
     pressed {
      bIsDynamicallyLoaded=false
      drawAs=RoundedBox
      imageType=NoImage
      mirroring=NoMirror
      resourceName=None
      resourceObject=&T-Button-Square
      tiling=NoTile
      imageSize=<32 32>
      margin=<ltrb 0.5 0.5 0.5 0.5>
      outlineSettings {
       bUseBrushTransparency=false
       roundingType=FixedRadius
       width=0
       color {colorUseRule=UseColor_ColorTable, specifiedColor=<rgba 0 0 0 0>}
       cornerRadii=<4 4 4 4>
      }
      tintColor {colorUseRule=UseColor_Specified, specifiedColor=<rgba 0.3119 0.7917 0.2715 1>}
      uVRegion {bIsValid=false, max=<0 0>, min=<0 0>}
     }
     pressedForeground {colorUseRule=UseColor_ColorTable, specifiedColor=<rgba 0 0 0 0>}
     pressedPadding=<ltrb 12 2.5 12 0.5>
     pressedSlateSound {resourceObject=null}
    }}
   slot {class=HorizontalBoxSlot, properties {verticalAlignment=VAlign_Center, padding=<ltrb 5 5 5 5>}}
  }
  ExternalIcon {
   widgetClassPath=&Image
   parent=OpenEditorWithThemedGroup
   widget=ExternalIcon
   properties {brush {
     bIsDynamicallyLoaded=false
     drawAs=Image
     imageType=NoImage
     mirroring=NoMirror
     resourceName=None
     resourceObject=/FortUILibrary/Foundation/Buttons/Materials/M_UI_Icon_ExternalLink
     tiling=NoTile
     imageSize=<20 20>
     margin=<ltrb 0 0 0 0>
     outlineSettings {
      bUseBrushTransparency=false
      roundingType=HalfHeightRadius
      width=0
      color {colorUseRule=UseColor_Specified, specifiedColor=<rgba 0 0 0 0>}
      cornerRadii=<0 0 0 0>
     }
     tintColor {colorUseRule=UseColor_Specified, specifiedColor=<rgba 1 1 1 1>}
     uVRegion {bIsValid=false, max=<0 0>, min=<0 0>}
    }}
   slot {class=ButtonSlot, properties {padding=<ltrb 0 0 0 0>}}
  }
  AddMoodButton {
   widgetClassPath=&EditorUtilityButton
   parent=HorizontalBox_0
   widget=AddMoodButton
   properties {widgetStyle {
     clickedSlateSound {resourceObject=null}
     disabled {
      bIsDynamicallyLoaded=false
      drawAs=RoundedBox
      imageType=NoImage
      mirroring=NoMirror
      resourceName=None
      resourceObject=null
      tiling=NoTile
      imageSize=<32 32>
      margin=<ltrb 0.5 0.5 0.5 0.5>
      outlineSettings {
       bUseBrushTransparency=false
       roundingType=FixedRadius
       width=0
       color {colorUseRule=UseColor_ColorTable, specifiedColor=<rgba 0 0 0 0>}
       cornerRadii=<4 4 4 4>
      }
      tintColor {colorUseRule=UseColor_Specified, specifiedColor=<rgba 0.0395 0.0395 0.0395 1>}
      uVRegion {bIsValid=false, max=<0 0>, min=<0 0>}
     }
     disabledForeground {colorUseRule=UseColor_ColorTable, specifiedColor=<rgba 0 0 0 0>}
     hovered {
      bIsDynamicallyLoaded=false
      drawAs=RoundedBox
      imageType=NoImage
      mirroring=NoMirror
      resourceName=None
      resourceObject=null
      tiling=NoTile
      imageSize=<32 32>
      margin=<ltrb 0 0 0 0>
      outlineSettings {
       bUseBrushTransparency=false
       roundingType=FixedRadius
       width=1
       color {colorUseRule=UseColor_ColorTable, specifiedColor=<rgba 0 0 0 0>}
       cornerRadii=<4 4 4 4>
      }
      tintColor {colorUseRule=UseColor_Specified, specifiedColor=<rgba 0 1 0.0916 1>}
      uVRegion {bIsValid=false, max=<0 0>, min=<0 0>}
     }
     hoveredForeground {colorUseRule=UseColor_ColorTable, specifiedColor=<rgba 0 0 0 0>}
     hoveredSlateSound {resourceObject=null}
     normal {
      bIsDynamicallyLoaded=false
      drawAs=RoundedBox
      imageType=NoImage
      mirroring=NoMirror
      resourceName=None
      resourceObject=null
      tiling=NoTile
      imageSize=<32 32>
      margin=<ltrb 0.5 0.5 0.5 0.5>
      outlineSettings {
       bUseBrushTransparency=false
       roundingType=FixedRadius
       width=0
       color {colorUseRule=UseColor_ColorTable, specifiedColor=<rgba 0 0 0 0>}
       cornerRadii=<4 4 4 4>
      }
      tintColor {colorUseRule=UseColor_Specified, specifiedColor=<rgba 0.0771 0.6167 0.1289 1>}
      uVRegion {bIsValid=false, max=<0 0>, min=<0 0>}
     }
     normalForeground {colorUseRule=UseColor_ColorTable, specifiedColor=<rgba 0 0 0 0>}
     normalPadding=<ltrb 12 1.5 12 1.5>
     pressed {
      bIsDynamicallyLoaded=false
      drawAs=RoundedBox
      imageType=NoImage
      mirroring=NoMirror
      resourceName=None
      resourceObject=null
      tiling=NoTile
      imageSize=<32 32>
      margin=<ltrb 0.5 0.5 0.5 0.5>
      outlineSettings {
       bUseBrushTransparency=false
       roundingType=FixedRadius
       width=0
       color {colorUseRule=UseColor_ColorTable, specifiedColor=<rgba 0 0 0 0>}
       cornerRadii=<4 4 4 4>
      }
      tintColor {colorUseRule=UseColor_Specified, specifiedColor=<rgba 0.1806 0.4583 0.1572 1>}
      uVRegion {bIsValid=false, max=<0 0>, min=<0 0>}
     }
     pressedForeground {colorUseRule=UseColor_ColorTable, specifiedColor=<rgba 0 0 0 0>}
     pressedPadding=<ltrb 12 2.5 12 0.5>
     pressedSlateSound {resourceObject=null}
    }}
   slot {class=HorizontalBoxSlot, properties {horizontalAlignment=HAlign_Center, verticalAlignment=VAlign_Center}}
  }
  Image_16 {
   widgetClassPath=&Image
   parent=AddMoodButton
   widget=Image_16
   properties {brush {
     bIsDynamicallyLoaded=false
     drawAs=Image
     imageType=NoImage
     mirroring=NoMirror
     resourceName=None
     resourceObject=/PilgrimQuickplay/StageManager/NeverCook/EditorUtilities/StageManagerEditor/Textures/MI_SparksEditorUI_PlusMinus_NoBorder
     tiling=NoTile
     imageSize=<20 20>
     margin=<ltrb 0 0 0 0>
     outlineSettings {
      bUseBrushTransparency=false
      roundingType=HalfHeightRadius
      width=0
      color {colorUseRule=UseColor_Specified, specifiedColor=<rgba 0 0 0 0>}
      cornerRadii=<0 0 0 0>
     }
     tintColor {colorUseRule=UseColor_Specified, specifiedColor=<rgba 1 1 1 1>}
     uVRegion {bIsValid=false, max=<0 0>, min=<0 0>}
    }}
   slot {class=ButtonSlot, properties {padding=<ltrb -5 0 -5 0>}}
  }
  DeleteMoodButton {
   widgetClassPath=&EditorUtilityButton
   parent=HorizontalBox_0
   widget=DeleteMoodButton
   properties {widgetStyle {
     clickedSlateSound {resourceObject=null}
     disabled {
      bIsDynamicallyLoaded=false
      drawAs=RoundedBox
      imageType=NoImage
      mirroring=NoMirror
      resourceName=None
      resourceObject=null
      tiling=NoTile
      imageSize=<20 20>
      margin=<ltrb 0 0 0 0>
      outlineSettings {
       bUseBrushTransparency=false
       roundingType=FixedRadius
       width=0
       color {colorUseRule=UseColor_ColorTable, specifiedColor=<rgba 0 0 0 0>}
       cornerRadii=<4 4 4 4>
      }
      tintColor {colorUseRule=UseColor_Specified, specifiedColor=<rgba 0.0395 0.0395 0.0395 1>}
      uVRegion {bIsValid=false, max=<0 0>, min=<0 0>}
     }
     disabledForeground {colorUseRule=UseColor_ColorTable, specifiedColor=<rgba 0 0 0 0>}
     hovered {
      bIsDynamicallyLoaded=false
      drawAs=RoundedBox
      imageType=NoImage
      mirroring=NoMirror
      resourceName=None
      resourceObject=null
      tiling=NoTile
      imageSize=<32 32>
      margin=<ltrb 0 0 0 0>
      outlineSettings {
       bUseBrushTransparency=false
       roundingType=FixedRadius
       width=0
       color {colorUseRule=UseColor_ColorTable, specifiedColor=<rgba 0 0 0 0>}
       cornerRadii=<4 4 4 4>
      }
      tintColor {colorUseRule=UseColor_Specified, specifiedColor=<rgba 1 0.1252 0.253 1>}
      uVRegion {bIsValid=false, max=<0 0>, min=<0 0>}
     }
     hoveredForeground {colorUseRule=UseColor_ColorTable, specifiedColor=<rgba 0 0 0 0>}
     hoveredSlateSound {resourceObject=null}
     normal {
      bIsDynamicallyLoaded=false
      drawAs=RoundedBox
      imageType=NoImage
      mirroring=NoMirror
      resourceName=None
      resourceObject=null
      tiling=NoTile
      imageSize=<20 20>
      margin=<ltrb 0 0 0 0>
      outlineSettings {
       bUseBrushTransparency=false
       roundingType=FixedRadius
       width=0
       color {colorUseRule=UseColor_ColorTable, specifiedColor=<rgba 0 0 0 0>}
       cornerRadii=<4 4 4 4>
      }
      tintColor {colorUseRule=UseColor_Specified, specifiedColor=<rgba 0.775 0 0.0304 1>}
      uVRegion {bIsValid=false, max=<0 0>, min=<0 0>}
     }
     normalForeground {colorUseRule=UseColor_ColorTable, specifiedColor=<rgba 0 0 0 0>}
     normalPadding=<ltrb 12 1.5 12 1.5>
     pressed {
      bIsDynamicallyLoaded=false
      drawAs=RoundedBox
      imageType=NoImage
      mirroring=NoMirror
      resourceName=None
      resourceObject=null
      tiling=NoTile
      imageSize=<20 20>
      margin=<ltrb 0 0 0 0>
      outlineSettings {
       bUseBrushTransparency=false
       roundingType=FixedRadius
       width=0
       color {colorUseRule=UseColor_ColorTable, specifiedColor=<rgba 0 0 0 0>}
       cornerRadii=<4 4 4 4>
      }
      tintColor {colorUseRule=UseColor_Specified, specifiedColor=<rgba 0.2396 0 0 1>}
      uVRegion {bIsValid=false, max=<0 0>, min=<0 0>}
     }
     pressedForeground {colorUseRule=UseColor_ColorTable, specifiedColor=<rgba 0 0 0 0>}
     pressedPadding=<ltrb 12 2.5 12 0.5>
     pressedSlateSound {resourceObject=null}
    }}
   slot {class=HorizontalBoxSlot, properties {horizontalAlignment=HAlign_Left, verticalAlignment=VAlign_Center, padding=<ltrb 10 0 0 0>}}
  }
  Image_20 {
   widgetClassPath=&Image
   parent=DeleteMoodButton
   widget=Image_20
   properties {
    visibility=SelfHitTestInvisible
    brush {
     bIsDynamicallyLoaded=false
     drawAs=Image
     imageType=FullColor
     mirroring=NoMirror
     resourceName=None
     resourceObject=/PilgrimQuickplay/StageManager/NeverCook/EditorUtilities/StageManagerEditor/Textures/T_SparksEditorUI_TrashCan_64
     tiling=NoTile
     imageSize=<20 20>
     margin=<ltrb 0 0 0 0>
     outlineSettings {
      bUseBrushTransparency=false
      roundingType=HalfHeightRadius
      width=0
      color {colorUseRule=UseColor_Specified, specifiedColor=<rgba 0 0 0 0>}
      cornerRadii=<0 0 0 0>
     }
     tintColor {colorUseRule=UseColor_Specified, specifiedColor=<rgba 1 1 1 1>}
     uVRegion {bIsValid=false, max=<0 0>, min=<0 0>}
    }
   }
   slot {class=ButtonSlot, properties {padding=<ltrb -5 0 -5 0>}}
  }
  Spacer {widgetClassPath=&Spacer, parent=HorizontalBox_0, widget=Spacer, properties {size=<20.1542 1>}, slot {class=HorizontalBoxSlot}}
  MidiMembersNumber {
   widgetClassPath=&TextBlock
   bIsVariable=true
   parent=HorizontalBox_0
   widget=MidiMembersNumber
   properties {
    text=10 Events
    colorAndOpacity {colorUseRule=UseColor_Specified, specifiedColor=<rgba 0.6562 0.6562 0.6562 1>}
    font {
     bForceMonospaced=false
     bMaterialIsStencil=false
     fontMaterial=null
     fontObject=&HeadingNow
     letterSpacing=0
     monospacedWidth=1
     size=7.5
     skewAmount=0
     typefaceFontName=75 Medium
     outlineSettings {
      bApplyOutlineToDropShadows=false
      bMiteredCorners=false
      bSeparateFillAlpha=false
      outlineMaterial=null
      outlineSize=0
      outlineColor=<rgba 0 0 0 1>
     }
    }
   }
   slot {class=HorizontalBoxSlot, properties {verticalAlignment=VAlign_Center, padding=<ltrb 0 0 5 0>}}
  }
  MidiTextEventsToggle {
   widgetClassPath=&EditorUtilityCheckBox
   bIsVariable=true
   parent=HorizontalBox_0
   widget=MidiTextEventsToggle
   properties {
    checkedState=Checked
    toolTipText=Shows the MIDI Text Events
    widgetStyle {
     checkBoxType=ToggleButton
     backgroundHoveredImage {
      bIsDynamicallyLoaded=false
      drawAs=NoDrawType
      imageType=NoImage
      mirroring=NoMirror
      resourceName=None
      resourceObject=null
      tiling=NoTile
      imageSize=<18 18>
      margin=<ltrb 0 0 0 0>
      outlineSettings {
       bUseBrushTransparency=false
       roundingType=FixedRadius
       width=1
       color {colorUseRule=UseColor_ColorTable, specifiedColor=<rgba 0 0 0 0>}
       cornerRadii=<3 3 3 3>
      }
      tintColor {colorUseRule=UseColor_Specified, specifiedColor=<rgba 0.0048 0.0048 0.0048 1>}
      uVRegion {bIsValid=false, max=<0 0>, min=<0 0>}
     }
     backgroundImage {
      bIsDynamicallyLoaded=false
      drawAs=NoDrawType
      imageType=NoImage
      mirroring=NoMirror
      resourceName=None
      resourceObject=null
      tiling=NoTile
      imageSize=<18 18>
      margin=<ltrb 0 0 0 0>
      outlineSettings {
       bUseBrushTransparency=false
       roundingType=FixedRadius
       width=1
       color {colorUseRule=UseColor_ColorTable, specifiedColor=<rgba 0 0 0 0>}
       cornerRadii=<3 3 3 3>
      }
      tintColor {colorUseRule=UseColor_Specified, specifiedColor=<rgba 0.0048 0.0048 0.0048 1>}
      uVRegion {bIsValid=false, max=<0 0>, min=<0 0>}
     }
     backgroundPressedImage {
      bIsDynamicallyLoaded=false
      drawAs=NoDrawType
      imageType=NoImage
      mirroring=NoMirror
      resourceName=None
      resourceObject=null
      tiling=NoTile
      imageSize=<18 18>
      margin=<ltrb 0 0 0 0>
      outlineSettings {
       bUseBrushTransparency=false
       roundingType=FixedRadius
       width=1
       color {colorUseRule=UseColor_ColorTable, specifiedColor=<rgba 0 0 0 0>}
       cornerRadii=<3 3 3 3>
      }
      tintColor {colorUseRule=UseColor_Specified, specifiedColor=<rgba 0.0048 0.0048 0.0048 1>}
      uVRegion {bIsValid=false, max=<0 0>, min=<0 0>}
     }
     borderBackgroundColor {colorUseRule=UseColor_Specified, specifiedColor=<rgba 1 1 1 1>}
     checkedForeground {colorUseRule=UseColor_Specified, specifiedColor=<rgba 1 1 1 1>}
     checkedHoveredForeground {colorUseRule=UseColor_Specified, specifiedColor=<rgba 1 1 1 1>}
     checkedHoveredImage {
      bIsDynamicallyLoaded=false
      drawAs=RoundedBox
      imageType=NoImage
      mirroring=NoMirror
      resourceName=None
      resourceObject=null
      tiling=NoTile
      imageSize=<32 32>
      margin=<ltrb 0 0 0 0>
      outlineSettings {
       bUseBrushTransparency=false
       roundingType=FixedRadius
       width=0.5
       color {colorUseRule=UseColor_Specified, specifiedColor=<rgba 0 0 0 0>}
       cornerRadii=<5 5 5 1>
      }
      tintColor {colorUseRule=UseColor_Specified, specifiedColor=<rgba 0.2734 0.4375 0.2816 1>}
      uVRegion {bIsValid=false, max=<0 0>, min=<0 0>}
     }
     checkedImage {
      bIsDynamicallyLoaded=false
      drawAs=RoundedBox
      imageType=NoImage
      mirroring=NoMirror
      resourceName=None
      resourceObject=null
      tiling=NoTile
      imageSize=<32 32>
      margin=<ltrb 0.5 0.5 0.5 0.5>
      outlineSettings {
       bUseBrushTransparency=false
       roundingType=FixedRadius
       width=0.5
       color {colorUseRule=UseColor_Specified, specifiedColor=<rgba 0 0 0 0>}
       cornerRadii=<5 5 5 1>
      }
      tintColor {colorUseRule=UseColor_Specified, specifiedColor=<rgba 0.0355 0.4375 0.0554 1>}
      uVRegion {bIsValid=false, max=<0 0>, min=<0 0>}
     }
     checkedPressedForeground {colorUseRule=UseColor_Specified, specifiedColor=<rgba 1 1 1 1>}
     checkedPressedImage {
      bIsDynamicallyLoaded=false
      drawAs=RoundedBox
      imageType=NoImage
      mirroring=NoMirror
      resourceName=None
      resourceObject=null
      tiling=NoTile
      imageSize=<32 32>
      margin=<ltrb 0 0 0 0>
      outlineSettings {
       bUseBrushTransparency=false
       roundingType=FixedRadius
       width=0.5
       color {colorUseRule=UseColor_Specified, specifiedColor=<rgba 0 0 0 0>}
       cornerRadii=<5 5 5 1>
      }
      tintColor {colorUseRule=UseColor_Specified, specifiedColor=<rgba 0.4248 0.9062 0.4486 1>}
      uVRegion {bIsValid=false, max=<0 0>, min=<0 0>}
     }
     checkedSlateSound {resourceObject=null}
     foregroundColor {colorUseRule=UseColor_Specified, specifiedColor=<rgba 1 1 1 1>}
     hoveredForeground {colorUseRule=UseColor_Specified, specifiedColor=<rgba 1 1 1 1>}
     hoveredSlateSound {resourceObject=null}
     padding=<ltrb 2 0 0 0>
     pressedForeground {colorUseRule=UseColor_Specified, specifiedColor=<rgba 1 1 1 1>}
     uncheckedHoveredImage {
      bIsDynamicallyLoaded=false
      drawAs=RoundedBox
      imageType=NoImage
      mirroring=NoMirror
      resourceName=None
      resourceObject=null
      tiling=NoTile
      imageSize=<32 60.845>
      margin=<ltrb 0 0 0 0>
      outlineSettings {
       bUseBrushTransparency=false
       roundingType=FixedRadius
       width=0.5
       color {colorUseRule=UseColor_Specified, specifiedColor=<rgba 0 0 0 0>}
       cornerRadii=<5 5 5 1>
      }
      tintColor {colorUseRule=UseColor_Specified, specifiedColor=<rgba 0.724 0.724 0.724 1>}
      uVRegion {bIsValid=false, max=<0 0>, min=<0 0>}
     }
     uncheckedImage {
      bIsDynamicallyLoaded=false
      drawAs=RoundedBox
      imageType=NoImage
      mirroring=NoMirror
      resourceName=None
      resourceObject=null
      tiling=NoTile
      imageSize=<32 32>
      margin=<ltrb 0 0 0 0>
      outlineSettings {
       bUseBrushTransparency=false
       roundingType=FixedRadius
       width=0.5
       color {colorUseRule=UseColor_Specified, specifiedColor=<rgba 0 0 0 0>}
       cornerRadii=<5 5 5 1>
      }
      tintColor {colorUseRule=UseColor_Specified, specifiedColor=<rgba 0.2448 0.2448 0.2448 1>}
      uVRegion {bIsValid=false, max=<0 0>, min=<0 0>}
     }
     uncheckedPressedImage {
      bIsDynamicallyLoaded=false
      drawAs=RoundedBox
      imageType=NoImage
      mirroring=NoMirror
      resourceName=None
      resourceObject=null
      tiling=NoTile
      imageSize=<32 32>
      margin=<ltrb 0 0 0 0>
      outlineSettings {
       bUseBrushTransparency=false
       roundingType=FixedRadius
       width=0.5
       color {colorUseRule=UseColor_Specified, specifiedColor=<rgba 0 0 0 0>}
       cornerRadii=<5 5 5 1>
      }
      tintColor {colorUseRule=UseColor_Specified, specifiedColor=<rgba 0.8594 0.8494 0.698 1>}
      uVRegion {bIsValid=false, max=<0 0>, min=<0 0>}
     }
     uncheckedSlateSound {resourceObject=null}
     undeterminedForeground {colorUseRule=UseColor_Specified, specifiedColor=<rgba 1 1 1 1>}
     undeterminedHoveredImage {
      bIsDynamicallyLoaded=false
      drawAs=Image
      imageType=Vector
      mirroring=NoMirror
      resourceName=../../../Engine/Content/Slate/Starship/CoreWidgets/CheckBox/indeterminate.svg
      resourceObject=null
      tiling=NoTile
      imageSize=<16 16>
      margin=<ltrb 0 0 0 0>
      outlineSettings {
       bUseBrushTransparency=false
       roundingType=HalfHeightRadius
       width=0
       color {colorUseRule=UseColor_Specified, specifiedColor=<rgba 0 0 0 0>}
       cornerRadii=<0 0 0 0>
      }
      tintColor {colorUseRule=UseColor_Specified, specifiedColor=<rgba 0.0044 0.2384 1 1>}
      uVRegion {bIsValid=false, max=<0 0>, min=<0 0>}
     }
     undeterminedImage {
      bIsDynamicallyLoaded=false
      drawAs=Image
      imageType=Vector
      mirroring=NoMirror
      resourceName=../../../Engine/Content/Slate/Starship/CoreWidgets/CheckBox/indeterminate.svg
      resourceObject=null
      tiling=NoTile
      imageSize=<16 16>
      margin=<ltrb 0 0 0 0>
      outlineSettings {
       bUseBrushTransparency=false
       roundingType=HalfHeightRadius
       width=0
       color {colorUseRule=UseColor_Specified, specifiedColor=<rgba 0 0 0 0>}
       cornerRadii=<0 0 0 0>
      }
      tintColor {colorUseRule=UseColor_Specified, specifiedColor=<rgba 0 0.162 0.7454 1>}
      uVRegion {bIsValid=false, max=<0 0>, min=<0 0>}
     }
     undeterminedPressedImage {
      bIsDynamicallyLoaded=false
      drawAs=Image
      imageType=Vector
      mirroring=NoMirror
      resourceName=../../../Engine/Content/Slate/Starship/CoreWidgets/CheckBox/indeterminate.svg
      resourceObject=null
      tiling=NoTile
      imageSize=<16 16>
      margin=<ltrb 0 0 0 0>
      outlineSettings {
       bUseBrushTransparency=false
       roundingType=HalfHeightRadius
       width=0
       color {colorUseRule=UseColor_Specified, specifiedColor=<rgba 0 0 0 0>}
       cornerRadii=<0 0 0 0>
      }
      tintColor {colorUseRule=UseColor_Specified, specifiedColor=<rgba 0 0.0802 0.3515 1>}
      uVRegion {bIsValid=false, max=<0 0>, min=<0 0>}
     }
    }
   }
   slot {class=HorizontalBoxSlot, properties {verticalAlignment=VAlign_Center, padding=<ltrb 0 0 10 0>}}
  }
  TextBlock {
   widgetClassPath=&TextBlock
   parent=MidiTextEventsToggle
   widget=TextBlock
   properties {
    justification=Center
    text=Mood MIDI Text Events
    font {
     bForceMonospaced=false
     bMaterialIsStencil=false
     fontMaterial=null
     fontObject=&HeadingNow
     letterSpacing=0
     monospacedWidth=1
     size=8.25
     skewAmount=0
     typefaceFontName=75 Medium
     outlineSettings {
      bApplyOutlineToDropShadows=false
      bMiteredCorners=false
      bSeparateFillAlpha=false
      outlineMaterial=null
      outlineSize=0
      outlineColor=<rgba 0 0 0 1>
     }
    }
    margin=<ltrb 7 7 7 7>
   }
   slot {class=PanelSlot}
  }
  QuickPlayMembersNumber {
   widgetClassPath=&TextBlock
   bIsVariable=true
   parent=HorizontalBox_0
   widget=QuickPlayMembersNumber
   properties {
    text=10 Events
    colorAndOpacity {colorUseRule=UseColor_Specified, specifiedColor=<rgba 0.6562 0.6562 0.6562 1>}
    font {
     bForceMonospaced=false
     bMaterialIsStencil=false
     fontMaterial=null
     fontObject=&HeadingNow
     letterSpacing=0
     monospacedWidth=1
     size=7.5
     skewAmount=0
     typefaceFontName=75 Medium
     outlineSettings {
      bApplyOutlineToDropShadows=false
      bMiteredCorners=false
      bSeparateFillAlpha=false
      outlineMaterial=null
      outlineSize=0
      outlineColor=<rgba 0 0 0 1>
     }
    }
   }
   slot {class=HorizontalBoxSlot, properties {verticalAlignment=VAlign_Center, padding=<ltrb 0 0 5 0>}}
  }
  QuickPlayEventsToggle {
   widgetClassPath=&EditorUtilityCheckBox
   bIsVariable=true
   parent=HorizontalBox_0
   widget=QuickPlayEventsToggle
   properties {
    toolTipText="Show the global quick play events, this element is not associated with the Mood."
    widgetStyle {
     checkBoxType=ToggleButton
     backgroundHoveredImage {
      bIsDynamicallyLoaded=false
      drawAs=NoDrawType
      imageType=NoImage
      mirroring=NoMirror
      resourceName=None
      resourceObject=null
      tiling=NoTile
      imageSize=<18 18>
      margin=<ltrb 0 0 0 0>
      outlineSettings {
       bUseBrushTransparency=false
       roundingType=FixedRadius
       width=1
       color {colorUseRule=UseColor_ColorTable, specifiedColor=<rgba 0 0 0 0>}
       cornerRadii=<3 3 3 3>
      }
      tintColor {colorUseRule=UseColor_Specified, specifiedColor=<rgba 0.0048 0.0048 0.0048 1>}
      uVRegion {bIsValid=false, max=<0 0>, min=<0 0>}
     }
     backgroundImage {
      bIsDynamicallyLoaded=false
      drawAs=NoDrawType
      imageType=NoImage
      mirroring=NoMirror
      resourceName=None
      resourceObject=null
      tiling=NoTile
      imageSize=<18 18>
      margin=<ltrb 0 0 0 0>
      outlineSettings {
       bUseBrushTransparency=false
       roundingType=FixedRadius
       width=1
       color {colorUseRule=UseColor_ColorTable, specifiedColor=<rgba 0 0 0 0>}
       cornerRadii=<3 3 3 3>
      }
      tintColor {colorUseRule=UseColor_Specified, specifiedColor=<rgba 0.0048 0.0048 0.0048 1>}
      uVRegion {bIsValid=false, max=<0 0>, min=<0 0>}
     }
     backgroundPressedImage {
      bIsDynamicallyLoaded=false
      drawAs=NoDrawType
      imageType=NoImage
      mirroring=NoMirror
      resourceName=None
      resourceObject=null
      tiling=NoTile
      imageSize=<18 18>
      margin=<ltrb 0 0 0 0>
      outlineSettings {
       bUseBrushTransparency=false
       roundingType=FixedRadius
       width=1
       color {colorUseRule=UseColor_ColorTable, specifiedColor=<rgba 0 0 0 0>}
       cornerRadii=<3 3 3 3>
      }
      tintColor {colorUseRule=UseColor_Specified, specifiedColor=<rgba 0.0048 0.0048 0.0048 1>}
      uVRegion {bIsValid=false, max=<0 0>, min=<0 0>}
     }
     borderBackgroundColor {colorUseRule=UseColor_Specified, specifiedColor=<rgba 1 1 1 1>}
     checkedForeground {colorUseRule=UseColor_Specified, specifiedColor=<rgba 1 1 1 1>}
     checkedHoveredForeground {colorUseRule=UseColor_Specified, specifiedColor=<rgba 1 1 1 1>}
     checkedHoveredImage {
      bIsDynamicallyLoaded=false
      drawAs=RoundedBox
      imageType=NoImage
      mirroring=NoMirror
      resourceName=None
      resourceObject=null
      tiling=NoTile
      imageSize=<32 32>
      margin=<ltrb 0 0 0 0>
      outlineSettings {
       bUseBrushTransparency=false
       roundingType=FixedRadius
       width=0.5
       color {colorUseRule=UseColor_Specified, specifiedColor=<rgba 0 0 0 0>}
       cornerRadii=<5 5 5 1>
      }
      tintColor {colorUseRule=UseColor_Specified, specifiedColor=<rgba 0.2734 0.4375 0.2816 1>}
      uVRegion {bIsValid=false, max=<0 0>, min=<0 0>}
     }
     checkedImage {
      bIsDynamicallyLoaded=false
      drawAs=RoundedBox
      imageType=NoImage
      mirroring=NoMirror
      resourceName=None
      resourceObject=null
      tiling=NoTile
      imageSize=<32 32>
      margin=<ltrb 0.5 0.5 0.5 0.5>
      outlineSettings {
       bUseBrushTransparency=false
       roundingType=FixedRadius
       width=0.5
       color {colorUseRule=UseColor_Specified, specifiedColor=<rgba 0 0 0 0>}
       cornerRadii=<5 5 5 1>
      }
      tintColor {colorUseRule=UseColor_Specified, specifiedColor=<rgba 0.0355 0.4375 0.0554 1>}
      uVRegion {bIsValid=false, max=<0 0>, min=<0 0>}
     }
     checkedPressedForeground {colorUseRule=UseColor_Specified, specifiedColor=<rgba 1 1 1 1>}
     checkedPressedImage {
      bIsDynamicallyLoaded=false
      drawAs=RoundedBox
      imageType=NoImage
      mirroring=NoMirror
      resourceName=None
      resourceObject=null
      tiling=NoTile
      imageSize=<32 32>
      margin=<ltrb 0 0 0 0>
      outlineSettings {
       bUseBrushTransparency=false
       roundingType=FixedRadius
       width=0.5
       color {colorUseRule=UseColor_Specified, specifiedColor=<rgba 0 0 0 0>}
       cornerRadii=<5 5 5 1>
      }
      tintColor {colorUseRule=UseColor_Specified, specifiedColor=<rgba 0.4248 0.9062 0.4486 1>}
      uVRegion {bIsValid=false, max=<0 0>, min=<0 0>}
     }
     checkedSlateSound {resourceObject=null}
     foregroundColor {colorUseRule=UseColor_Specified, specifiedColor=<rgba 1 1 1 1>}
     hoveredForeground {colorUseRule=UseColor_Specified, specifiedColor=<rgba 1 1 1 1>}
     hoveredSlateSound {resourceObject=null}
     padding=<ltrb 2 0 0 0>
     pressedForeground {colorUseRule=UseColor_Specified, specifiedColor=<rgba 1 1 1 1>}
     uncheckedHoveredImage {
      bIsDynamicallyLoaded=false
      drawAs=RoundedBox
      imageType=NoImage
      mirroring=NoMirror
      resourceName=None
      resourceObject=null
      tiling=NoTile
      imageSize=<32 60.845>
      margin=<ltrb 0 0 0 0>
      outlineSettings {
       bUseBrushTransparency=false
       roundingType=FixedRadius
       width=0.5
       color {colorUseRule=UseColor_Specified, specifiedColor=<rgba 0 0 0 0>}
       cornerRadii=<5 5 5 1>
      }
      tintColor {colorUseRule=UseColor_Specified, specifiedColor=<rgba 0.724 0.724 0.724 1>}
      uVRegion {bIsValid=false, max=<0 0>, min=<0 0>}
     }
     uncheckedImage {
      bIsDynamicallyLoaded=false
      drawAs=RoundedBox
      imageType=NoImage
      mirroring=NoMirror
      resourceName=None
      resourceObject=null
      tiling=NoTile
      imageSize=<32 32>
      margin=<ltrb 0 0 0 0>
      outlineSettings {
       bUseBrushTransparency=false
       roundingType=FixedRadius
       width=0.5
       color {colorUseRule=UseColor_Specified, specifiedColor=<rgba 0 0 0 0>}
       cornerRadii=<5 5 5 1>
      }
      tintColor {colorUseRule=UseColor_Specified, specifiedColor=<rgba 0.2448 0.2448 0.2448 1>}
      uVRegion {bIsValid=false, max=<0 0>, min=<0 0>}
     }
     uncheckedPressedImage {
      bIsDynamicallyLoaded=false
      drawAs=RoundedBox
      imageType=NoImage
      mirroring=NoMirror
      resourceName=None
      resourceObject=null
      tiling=NoTile
      imageSize=<32 32>
      margin=<ltrb 0 0 0 0>
      outlineSettings {
       bUseBrushTransparency=false
       roundingType=FixedRadius
       width=0.5
       color {colorUseRule=UseColor_Specified, specifiedColor=<rgba 0 0 0 0>}
       cornerRadii=<5 5 5 1>
      }
      tintColor {colorUseRule=UseColor_Specified, specifiedColor=<rgba 0.8594 0.8494 0.698 1>}
      uVRegion {bIsValid=false, max=<0 0>, min=<0 0>}
     }
     uncheckedSlateSound {resourceObject=null}
     undeterminedForeground {colorUseRule=UseColor_Specified, specifiedColor=<rgba 1 1 1 1>}
     undeterminedHoveredImage {
      bIsDynamicallyLoaded=false
      drawAs=Image
      imageType=Vector
      mirroring=NoMirror
      resourceName=../../../Engine/Content/Slate/Starship/CoreWidgets/CheckBox/indeterminate.svg
      resourceObject=null
      tiling=NoTile
      imageSize=<16 16>
      margin=<ltrb 0 0 0 0>
      outlineSettings {
       bUseBrushTransparency=false
       roundingType=HalfHeightRadius
       width=0
       color {colorUseRule=UseColor_Specified, specifiedColor=<rgba 0 0 0 0>}
       cornerRadii=<0 0 0 0>
      }
      tintColor {colorUseRule=UseColor_Specified, specifiedColor=<rgba 0.0044 0.2384 1 1>}
      uVRegion {bIsValid=false, max=<0 0>, min=<0 0>}
     }
     undeterminedImage {
      bIsDynamicallyLoaded=false
      drawAs=Image
      imageType=Vector
      mirroring=NoMirror
      resourceName=../../../Engine/Content/Slate/Starship/CoreWidgets/CheckBox/indeterminate.svg
      resourceObject=null
      tiling=NoTile
      imageSize=<16 16>
      margin=<ltrb 0 0 0 0>
      outlineSettings {
       bUseBrushTransparency=false
       roundingType=HalfHeightRadius
       width=0
       color {colorUseRule=UseColor_Specified, specifiedColor=<rgba 0 0 0 0>}
       cornerRadii=<0 0 0 0>
      }
      tintColor {colorUseRule=UseColor_Specified, specifiedColor=<rgba 0 0.162 0.7454 1>}
      uVRegion {bIsValid=false, max=<0 0>, min=<0 0>}
     }
     undeterminedPressedImage {
      bIsDynamicallyLoaded=false
      drawAs=Image
      imageType=Vector
      mirroring=NoMirror
      resourceName=../../../Engine/Content/Slate/Starship/CoreWidgets/CheckBox/indeterminate.svg
      resourceObject=null
      tiling=NoTile
      imageSize=<16 16>
      margin=<ltrb 0 0 0 0>
      outlineSettings {
       bUseBrushTransparency=false
       roundingType=HalfHeightRadius
       width=0
       color {colorUseRule=UseColor_Specified, specifiedColor=<rgba 0 0 0 0>}
       cornerRadii=<0 0 0 0>
      }
      tintColor {colorUseRule=UseColor_Specified, specifiedColor=<rgba 0 0.0802 0.3515 1>}
      uVRegion {bIsValid=false, max=<0 0>, min=<0 0>}
     }
    }
   }
   slot {class=HorizontalBoxSlot, properties {verticalAlignment=VAlign_Center, padding=<ltrb 0 0 10 0>}}
  }
  TextBlock_1 {
   widgetClassPath=&TextBlock
   parent=QuickPlayEventsToggle
   widget=TextBlock_1
   properties {
    justification=Center
    text=Global Quickplay Events
    font {
     bForceMonospaced=false
     bMaterialIsStencil=false
     fontMaterial=null
     fontObject=&HeadingNow
     letterSpacing=0
     monospacedWidth=1
     size=8.25
     skewAmount=0
     typefaceFontName=75 Medium
     outlineSettings {
      bApplyOutlineToDropShadows=false
      bMiteredCorners=false
      bSeparateFillAlpha=false
      outlineMaterial=null
      outlineSize=0
      outlineColor=<rgba 0 0 0 1>
     }
    }
    margin=<ltrb 7 7 7 7>
   }
   slot {class=PanelSlot}
  }
  SequenceMIDITextMapField {
   widgetClassPath=&HorizontalBox
   bIsVariable=true
   parent=MainContent
   widget=SequenceMIDITextMapField
   slot {class=VerticalBoxSlot, properties {verticalAlignment=VAlign_Top, padding=<ltrb 0 10 0 0>, size {sizeRule=Automatic, value=0.5}}}
  }
  NewSequenceEventCellName_1 {
   widgetClassPath=&TextBlock
   parent=SequenceMIDITextMapField
   widget=NewSequenceEventCellName_1
   properties {
    text="Sequence MIDI Text Map: "
    font {
     bForceMonospaced=false
     bMaterialIsStencil=false
     fontMaterial=null
     fontObject=&HeadingNow
     letterSpacing=0
     monospacedWidth=1
     size=7.5
     skewAmount=0
     typefaceFontName=75 Medium
     outlineSettings {
      bApplyOutlineToDropShadows=false
      bMiteredCorners=false
      bSeparateFillAlpha=false
      outlineMaterial=null
      outlineSize=0
      outlineColor=<rgba 0 0 0 1>
     }
    }
   }
   slot {class=HorizontalBoxSlot, properties {horizontalAlignment=HAlign_Left, verticalAlignment=VAlign_Center, padding=<ltrb 0 0 15 0>}}
  }
  MIDITextMapComboBox {
   widgetClassPath=&SearchableStringComboBox
   bIsVariable=true
   parent=SequenceMIDITextMapField
   widget=MIDITextMapComboBox
   properties {
    enableGamepadNavigationMode=false
    selectedOption=Sequence Type List
    font {
     bForceMonospaced=false
     bMaterialIsStencil=false
     fontMaterial=null
     fontObject=&HeadingNow
     letterSpacing=0
     monospacedWidth=1
     size=7.5
     skewAmount=0
     typefaceFontName=75 Medium
     outlineSettings {
      bApplyOutlineToDropShadows=false
      bMiteredCorners=false
      bSeparateFillAlpha=false
      outlineMaterial=null
      outlineSize=0
      outlineColor=<rgba 0 0 0 1>
     }
    }
    foregroundColor {colorUseRule=UseColor_Specified, specifiedColor=<rgba 0.5271 0.5271 0.5271 1>}
    itemStyle {
     bUseParentRowBrush=false
     activeBrush {
      bIsDynamicallyLoaded=false
      drawAs=Image
      imageType=NoImage
      mirroring=NoMirror
      resourceName=None
      resourceObject=null
      tiling=NoTile
      imageSize=<0 0>
      margin=<ltrb 0 0 0 0>
      outlineSettings {
       bUseBrushTransparency=false
       roundingType=HalfHeightRadius
       width=0
       color {colorUseRule=UseColor_Specified, specifiedColor=<rgba 0 0 0 0>}
       cornerRadii=<0 0 0 0>
      }
      tintColor {colorUseRule=UseColor_Specified, specifiedColor=<rgba 0 0.162 0.7454 1>}
      uVRegion {bIsValid=false, max=<0 0>, min=<0 0>}
     }
     activeHighlightedBrush {
      bIsDynamicallyLoaded=false
      drawAs=Image
      imageType=NoImage
      mirroring=NoMirror
      resourceName=None
      resourceObject=null
      tiling=NoTile
      imageSize=<0 0>
      margin=<ltrb 0 0 0 0>
      outlineSettings {
       bUseBrushTransparency=false
       roundingType=HalfHeightRadius
       width=0
       color {colorUseRule=UseColor_Specified, specifiedColor=<rgba 0 0 0 0>}
       cornerRadii=<0 0 0 0>
      }
      tintColor {colorUseRule=UseColor_Specified, specifiedColor=<rgba 0.0252 0.0319 0.0423 1>}
      uVRegion {bIsValid=false, max=<0 0>, min=<0 0>}
     }
     activeHoveredBrush {
      bIsDynamicallyLoaded=false
      drawAs=Image
      imageType=NoImage
      mirroring=NoMirror
      resourceName=None
      resourceObject=null
      tiling=NoTile
      imageSize=<0 0>
      margin=<ltrb 0 0 0 0>
      outlineSettings {
       bUseBrushTransparency=false
       roundingType=HalfHeightRadius
       width=0
       color {colorUseRule=UseColor_Specified, specifiedColor=<rgba 0 0 0 0>}
       cornerRadii=<0 0 0 0>
      }
      tintColor {colorUseRule=UseColor_Specified, specifiedColor=<rgba 0 0.162 0.7454 1>}
      uVRegion {bIsValid=false, max=<0 0>, min=<0 0>}
     }
     dropIndicator_Above {
      bIsDynamicallyLoaded=false
      drawAs=Box
      imageType=FullColor
      mirroring=NoMirror
      resourceName=../../../Engine/Content/Slate/Common/DropZoneIndicator_Above.png
      resourceObject=null
      tiling=NoTile
      imageSize=<0 0>
      margin=<ltrb 0.625 0.625 0 0>
      outlineSettings {
       bUseBrushTransparency=false
       roundingType=HalfHeightRadius
       width=0
       color {colorUseRule=UseColor_Specified, specifiedColor=<rgba 0 0 0 0>}
       cornerRadii=<0 0 0 0>
      }
      tintColor {colorUseRule=UseColor_Specified, specifiedColor=<rgba 0.0091 0.3515 0.9647 1>}
      uVRegion {bIsValid=false, max=<0 0>, min=<0 0>}
     }
     dropIndicator_Below {
      bIsDynamicallyLoaded=false
      drawAs=Box
      imageType=FullColor
      mirroring=NoMirror
      resourceName=../../../Engine/Content/Slate/Common/DropZoneIndicator_Below.png
      resourceObject=null
      tiling=NoTile
      imageSize=<0 0>
      margin=<ltrb 0.625 0 0 0.625>
      outlineSettings {
       bUseBrushTransparency=false
       roundingType=HalfHeightRadius
       width=0
       color {colorUseRule=UseColor_Specified, specifiedColor=<rgba 0 0 0 0>}
       cornerRadii=<0 0 0 0>
      }
      tintColor {colorUseRule=UseColor_Specified, specifiedColor=<rgba 0.0091 0.3515 0.9647 1>}
      uVRegion {bIsValid=false, max=<0 0>, min=<0 0>}
     }
     dropIndicator_Onto {
      bIsDynamicallyLoaded=false
      drawAs=Box
      imageType=FullColor
      mirroring=NoMirror
      resourceName=../../../Engine/Content/Slate/Common/DropZoneIndicator_Onto.png
      resourceObject=null
      tiling=NoTile
      imageSize=<0 0>
      margin=<ltrb 0.25 0.25 0.25 0.25>
      outlineSettings {
       bUseBrushTransparency=false
       roundingType=HalfHeightRadius
       width=0
       color {colorUseRule=UseColor_Specified, specifiedColor=<rgba 0 0 0 0>}
       cornerRadii=<0 0 0 0>
      }
      tintColor {colorUseRule=UseColor_Specified, specifiedColor=<rgba 0.0091 0.3515 0.9647 1>}
      uVRegion {bIsValid=false, max=<0 0>, min=<0 0>}
     }
     evenRowBackgroundBrush {
      bIsDynamicallyLoaded=false
      drawAs=Image
      imageType=NoImage
      mirroring=NoMirror
      resourceName=None
      resourceObject=null
      tiling=NoTile
      imageSize=<0 0>
      margin=<ltrb 0 0 0 0>
      outlineSettings {
       bUseBrushTransparency=false
       roundingType=HalfHeightRadius
       width=0
       color {colorUseRule=UseColor_Specified, specifiedColor=<rgba 0 0 0 0>}
       cornerRadii=<0 0 0 0>
      }
      tintColor {colorUseRule=UseColor_Specified, specifiedColor=<rgba 0.0103 0.0103 0.0103 1>}
      uVRegion {bIsValid=false, max=<0 0>, min=<0 0>}
     }
     evenRowBackgroundHoveredBrush {
      bIsDynamicallyLoaded=false
      drawAs=Image
      imageType=NoImage
      mirroring=NoMirror
      resourceName=None
      resourceObject=null
      tiling=NoTile
      imageSize=<0 0>
      margin=<ltrb 0 0 0 0>
      outlineSettings {
       bUseBrushTransparency=false
       roundingType=HalfHeightRadius
       width=0
       color {colorUseRule=UseColor_Specified, specifiedColor=<rgba 0 0 0 0>}
       cornerRadii=<0 0 0 0>
      }
      tintColor {colorUseRule=UseColor_Specified, specifiedColor=<rgba 0.0176 0.0176 0.0176 1>}
      uVRegion {bIsValid=false, max=<0 0>, min=<0 0>}
     }
     inactiveBrush {
      bIsDynamicallyLoaded=false
      drawAs=Image
      imageType=NoImage
      mirroring=NoMirror
      resourceName=None
      resourceObject=null
      tiling=NoTile
      imageSize=<0 0>
      margin=<ltrb 0 0 0 0>
      outlineSettings {
       bUseBrushTransparency=false
       roundingType=HalfHeightRadius
       width=0
       color {colorUseRule=UseColor_Specified, specifiedColor=<rgba 0 0 0 0>}
       cornerRadii=<0 0 0 0>
      }
      tintColor {colorUseRule=UseColor_Specified, specifiedColor=<rgba 0.0513 0.0953 0.159 1>}
      uVRegion {bIsValid=false, max=<0 0>, min=<0 0>}
     }
     inactiveHighlightedBrush {
      bIsDynamicallyLoaded=false
      drawAs=Image
      imageType=NoImage
      mirroring=NoMirror
      resourceName=None
      resourceObject=null
      tiling=NoTile
      imageSize=<0 0>
      margin=<ltrb 0 0 0 0>
      outlineSettings {
       bUseBrushTransparency=false
       roundingType=HalfHeightRadius
       width=0
       color {colorUseRule=UseColor_Specified, specifiedColor=<rgba 0 0 0 0>}
       cornerRadii=<0 0 0 0>
      }
      tintColor {colorUseRule=UseColor_Specified, specifiedColor=<rgba 0.0252 0.0319 0.0423 1>}
      uVRegion {bIsValid=false, max=<0 0>, min=<0 0>}
     }
     inactiveHoveredBrush {
      bIsDynamicallyLoaded=false
      drawAs=Image
      imageType=NoImage
      mirroring=NoMirror
      resourceName=None
      resourceObject=null
      tiling=NoTile
      imageSize=<0 0>
      margin=<ltrb 0 0 0 0>
      outlineSettings {
       bUseBrushTransparency=false
       roundingType=HalfHeightRadius
       width=0
       color {colorUseRule=UseColor_Specified, specifiedColor=<rgba 0 0 0 0>}
       cornerRadii=<0 0 0 0>
      }
      tintColor {colorUseRule=UseColor_Specified, specifiedColor=<rgba 0.0513 0.0953 0.159 1>}
      uVRegion {bIsValid=false, max=<0 0>, min=<0 0>}
     }
     oddRowBackgroundBrush {
      bIsDynamicallyLoaded=false
      drawAs=Image
      imageType=NoImage
      mirroring=NoMirror
      resourceName=None
      resourceObject=null
      tiling=NoTile
      imageSize=<0 0>
      margin=<ltrb 0 0 0 0>
      outlineSettings {
       bUseBrushTransparency=false
       roundingType=HalfHeightRadius
       width=0
       color {colorUseRule=UseColor_Specified, specifiedColor=<rgba 0 0 0 0>}
       cornerRadii=<0 0 0 0>
      }
      tintColor {colorUseRule=UseColor_Specified, specifiedColor=<rgba 0.0103 0.0103 0.0103 1>}
      uVRegion {bIsValid=false, max=<0 0>, min=<0 0>}
     }
     oddRowBackgroundHoveredBrush {
      bIsDynamicallyLoaded=false
      drawAs=Image
      imageType=NoImage
      mirroring=NoMirror
      resourceName=None
      resourceObject=null
      tiling=NoTile
      imageSize=<0 0>
      margin=<ltrb 0 0 0 0>
      outlineSettings {
       bUseBrushTransparency=false
       roundingType=HalfHeightRadius
       width=0
       color {colorUseRule=UseColor_Specified, specifiedColor=<rgba 0 0 0 0>}
       cornerRadii=<0 0 0 0>
      }
      tintColor {colorUseRule=UseColor_Specified, specifiedColor=<rgba 0.0176 0.0176 0.0176 1>}
      uVRegion {bIsValid=false, max=<0 0>, min=<0 0>}
     }
     parentRowBackgroundBrush {
      bIsDynamicallyLoaded=false
      drawAs=Image
      imageType=NoImage
      mirroring=NoMirror
      resourceName=None
      resourceObject=null
      tiling=NoTile
      imageSize=<32 32>
      margin=<ltrb 0 0 0 0>
      outlineSettings {
       bUseBrushTransparency=false
       roundingType=HalfHeightRadius
       width=0
       color {colorUseRule=UseColor_Specified, specifiedColor=<rgba 0 0 0 0>}
       cornerRadii=<0 0 0 0>
      }
      tintColor {colorUseRule=UseColor_Specified, specifiedColor=<rgba 1 1 1 1>}
      uVRegion {bIsValid=false, max=<0 0>, min=<0 0>}
     }
     parentRowBackgroundHoveredBrush {
      bIsDynamicallyLoaded=false
      drawAs=Image
      imageType=NoImage
      mirroring=NoMirror
      resourceName=None
      resourceObject=null
      tiling=NoTile
      imageSize=<32 32>
      margin=<ltrb 0 0 0 0>
      outlineSettings {
       bUseBrushTransparency=false
       roundingType=HalfHeightRadius
       width=0
       color {colorUseRule=UseColor_Specified, specifiedColor=<rgba 0 0 0 0>}
       cornerRadii=<0 0 0 0>
      }
      tintColor {colorUseRule=UseColor_Specified, specifiedColor=<rgba 1 1 1 1>}
      uVRegion {bIsValid=false, max=<0 0>, min=<0 0>}
     }
     selectedTextColor {colorUseRule=UseColor_Specified, specifiedColor=<rgba 1 1 1 1>}
     selectorFocusedBrush {
      bIsDynamicallyLoaded=false
      drawAs=Border
      imageType=FullColor
      mirroring=NoMirror
      resourceName=../../../Engine/Content/Slate/Common/Selector.png
      resourceObject=null
      tiling=Both
      imageSize=<0 0>
      margin=<ltrb 0.25 0.25 0.25 0.25>
      outlineSettings {
       bUseBrushTransparency=false
       roundingType=HalfHeightRadius
       width=0
       color {colorUseRule=UseColor_Specified, specifiedColor=<rgba 0 0 0 0>}
       cornerRadii=<0 0 0 0>
      }
      tintColor {colorUseRule=UseColor_Specified, specifiedColor=<rgba 0.701 0.225 0.003 1>}
      uVRegion {bIsValid=false, max=<0 0>, min=<0 0>}
     }
     textColor {colorUseRule=UseColor_Specified, specifiedColor=<rgba 0.5271 0.5271 0.5271 1>}
    }
    widgetStyle {
     comboButtonStyle {
      downArrowAlign=VAlign_Center
      buttonStyle {
       clickedSlateSound {resourceObject=null}
       disabled {
        bIsDynamicallyLoaded=false
        drawAs=RoundedBox
        imageType=NoImage
        mirroring=NoMirror
        resourceName=None
        resourceObject=null
        tiling=NoTile
        imageSize=<0 0>
        margin=<ltrb 0 0 0 0>
        outlineSettings {
         bUseBrushTransparency=false
         roundingType=FixedRadius
         width=1
         color {colorUseRule=UseColor_ColorTable, specifiedColor=<rgba 0 0 0 0>}
         cornerRadii=<4 4 4 4>
        }
        tintColor {colorUseRule=UseColor_Specified, specifiedColor=<rgba 0.0395 0.0395 0.0395 1>}
        uVRegion {bIsValid=false, max=<0 0>, min=<0 0>}
       }
       disabledForeground {colorUseRule=UseColor_ColorTable, specifiedColor=<rgba 0 0 0 0>}
       hovered {
        bIsDynamicallyLoaded=false
        drawAs=RoundedBox
        imageType=NoImage
        mirroring=NoMirror
        resourceName=None
        resourceObject=null
        tiling=NoTile
        imageSize=<0 0>
        margin=<ltrb 0 0 0 0>
        outlineSettings {
         bUseBrushTransparency=false
         roundingType=FixedRadius
         width=1
         color {colorUseRule=UseColor_ColorTable, specifiedColor=<rgba 0 0 0 0>}
         cornerRadii=<4 4 4 4>
        }
        tintColor {colorUseRule=UseColor_Specified, specifiedColor=<rgba 0.0048 0.0048 0.0048 1>}
        uVRegion {bIsValid=false, max=<0 0>, min=<0 0>}
       }
       hoveredForeground {colorUseRule=UseColor_ColorTable, specifiedColor=<rgba 0 0 0 0>}
       hoveredSlateSound {resourceObject=null}
       normal {
        bIsDynamicallyLoaded=false
        drawAs=RoundedBox
        imageType=NoImage
        mirroring=NoMirror
        resourceName=None
        resourceObject=null
        tiling=NoTile
        imageSize=<0 0>
        margin=<ltrb 0 0 0 0>
        outlineSettings {
         bUseBrushTransparency=false
         roundingType=FixedRadius
         width=1
         color {colorUseRule=UseColor_ColorTable, specifiedColor=<rgba 0 0 0 0>}
         cornerRadii=<4 4 4 4>
        }
        tintColor {colorUseRule=UseColor_Specified, specifiedColor=<rgba 0.0048 0.0048 0.0048 1>}
        uVRegion {bIsValid=false, max=<0 0>, min=<0 0>}
       }
       normalForeground {colorUseRule=UseColor_ColorTable, specifiedColor=<rgba 0 0 0 0>}
       normalPadding=<ltrb 8 1 8 1>
       pressed {
        bIsDynamicallyLoaded=false
        drawAs=RoundedBox
        imageType=NoImage
        mirroring=NoMirror
        resourceName=None
        resourceObject=null
        tiling=NoTile
        imageSize=<0 0>
        margin=<ltrb 0 0 0 0>
        outlineSettings {
         bUseBrushTransparency=false
         roundingType=FixedRadius
         width=1
         color {colorUseRule=UseColor_ColorTable, specifiedColor=<rgba 0 0 0 0>}
         cornerRadii=<4 4 4 4>
        }
        tintColor {colorUseRule=UseColor_Specified, specifiedColor=<rgba 0.0048 0.0048 0.0048 1>}
        uVRegion {bIsValid=false, max=<0 0>, min=<0 0>}
       }
       pressedForeground {colorUseRule=UseColor_ColorTable, specifiedColor=<rgba 0 0 0 0>}
       pressedPadding=<ltrb 8 2 8 0>
       pressedSlateSound {resourceObject=null}
      }
      contentPadding=<ltrb 0 0 0 0>
      downArrowImage {
       bIsDynamicallyLoaded=false
       drawAs=Image
       imageType=Vector
       mirroring=NoMirror
       resourceName=../../../Engine/Content/Slate/Starship/CoreWidgets/ComboBox/wide-chevron-down.svg
       resourceObject=null
       tiling=NoTile
       imageSize=<16 16>
       margin=<ltrb 0 0 0 0>
       outlineSettings {
        bUseBrushTransparency=false
        roundingType=HalfHeightRadius
        width=0
        color {colorUseRule=UseColor_Specified, specifiedColor=<rgba 0 0 0 0>}
        cornerRadii=<0 0 0 0>
       }
       tintColor {colorUseRule=UseColor_Specified, specifiedColor=<rgba 1 1 1 1>}
       uVRegion {bIsValid=false, max=<0 0>, min=<0 0>}
      }
      downArrowPadding=<ltrb 2 2 2 2>
      menuBorderBrush {
       bIsDynamicallyLoaded=false
       drawAs=RoundedBox
       imageType=NoImage
       mirroring=NoMirror
       resourceName=None
       resourceObject=null
       tiling=NoTile
       imageSize=<0 0>
       margin=<ltrb 0 0 0 0>
       outlineSettings {
        bUseBrushTransparency=false
        roundingType=FixedRadius
        width=1
        color {colorUseRule=UseColor_ColorTable, specifiedColor=<rgba 0 0 0 0>}
        cornerRadii=<0 0 0 0>
       }
       tintColor {colorUseRule=UseColor_Specified, specifiedColor=<rgba 0.0395 0.0395 0.0395 1>}
       uVRegion {bIsValid=false, max=<0 0>, min=<0 0>}
      }
      menuBorderPadding=<ltrb 0 0 0 0>
      shadowColorAndOpacity=<rgba 0 0 0 1>
      shadowOffset=<0 0>
     }
     contentPadding=<ltrb 0 0 0 0>
     menuRowPadding=<ltrb 8 3 8 3>
     pressedSlateSound {resourceObject=null}
     selectionChangeSlateSound {resourceObject=null}
    }
   }
   slot {class=HorizontalBoxSlot, properties {horizontalAlignment=HAlign_Left, verticalAlignment=VAlign_Center}}
  }
  AddMIDITextMapButton {
   widgetClassPath=&EditorUtilityButton
   parent=SequenceMIDITextMapField
   widget=AddMIDITextMapButton
   properties {widgetStyle {
     clickedSlateSound {resourceObject=null}
     disabled {
      bIsDynamicallyLoaded=false
      drawAs=RoundedBox
      imageType=NoImage
      mirroring=NoMirror
      resourceName=None
      resourceObject=null
      tiling=NoTile
      imageSize=<32 32>
      margin=<ltrb 0.5 0.5 0.5 0.5>
      outlineSettings {
       bUseBrushTransparency=false
       roundingType=FixedRadius
       width=0
       color {colorUseRule=UseColor_ColorTable, specifiedColor=<rgba 0 0 0 0>}
       cornerRadii=<4 4 4 4>
      }
      tintColor {colorUseRule=UseColor_Specified, specifiedColor=<rgba 0.0395 0.0395 0.0395 1>}
      uVRegion {bIsValid=false, max=<0 0>, min=<0 0>}
     }
     disabledForeground {colorUseRule=UseColor_ColorTable, specifiedColor=<rgba 0 0 0 0>}
     hovered {
      bIsDynamicallyLoaded=false
      drawAs=RoundedBox
      imageType=NoImage
      mirroring=NoMirror
      resourceName=None
      resourceObject=null
      tiling=NoTile
      imageSize=<32 32>
      margin=<ltrb 0 0 0 0>
      outlineSettings {
       bUseBrushTransparency=false
       roundingType=FixedRadius
       width=1
       color {colorUseRule=UseColor_ColorTable, specifiedColor=<rgba 0 0 0 0>}
       cornerRadii=<4 4 4 4>
      }
      tintColor {colorUseRule=UseColor_Specified, specifiedColor=<rgba 0 1 0.0916 1>}
      uVRegion {bIsValid=false, max=<0 0>, min=<0 0>}
     }
     hoveredForeground {colorUseRule=UseColor_ColorTable, specifiedColor=<rgba 0 0 0 0>}
     hoveredSlateSound {resourceObject=null}
     normal {
      bIsDynamicallyLoaded=false
      drawAs=RoundedBox
      imageType=NoImage
      mirroring=NoMirror
      resourceName=None
      resourceObject=null
      tiling=NoTile
      imageSize=<32 32>
      margin=<ltrb 0.5 0.5 0.5 0.5>
      outlineSettings {
       bUseBrushTransparency=false
       roundingType=FixedRadius
       width=0
       color {colorUseRule=UseColor_ColorTable, specifiedColor=<rgba 0 0 0 0>}
       cornerRadii=<4 4 4 4>
      }
      tintColor {colorUseRule=UseColor_Specified, specifiedColor=<rgba 0.0771 0.6167 0.1289 1>}
      uVRegion {bIsValid=false, max=<0 0>, min=<0 0>}
     }
     normalForeground {colorUseRule=UseColor_ColorTable, specifiedColor=<rgba 0 0 0 0>}
     normalPadding=<ltrb 12 1.5 12 1.5>
     pressed {
      bIsDynamicallyLoaded=false
      drawAs=RoundedBox
      imageType=NoImage
      mirroring=NoMirror
      resourceName=None
      resourceObject=null
      tiling=NoTile
      imageSize=<32 32>
      margin=<ltrb 0.5 0.5 0.5 0.5>
      outlineSettings {
       bUseBrushTransparency=false
       roundingType=FixedRadius
       width=0
       color {colorUseRule=UseColor_ColorTable, specifiedColor=<rgba 0 0 0 0>}
       cornerRadii=<4 4 4 4>
      }
      tintColor {colorUseRule=UseColor_Specified, specifiedColor=<rgba 0.1806 0.4583 0.1572 1>}
      uVRegion {bIsValid=false, max=<0 0>, min=<0 0>}
     }
     pressedForeground {colorUseRule=UseColor_ColorTable, specifiedColor=<rgba 0 0 0 0>}
     pressedPadding=<ltrb 12 2.5 12 0.5>
     pressedSlateSound {resourceObject=null}
    }}
   slot {class=HorizontalBoxSlot, properties {horizontalAlignment=HAlign_Center, verticalAlignment=VAlign_Center, padding=<ltrb 5 0 0 0>}}
  }
  Image_15 {
   widgetClassPath=&Image
   parent=AddMIDITextMapButton
   widget=Image_15
   properties {brush {
     bIsDynamicallyLoaded=false
     drawAs=Image
     imageType=NoImage
     mirroring=NoMirror
     resourceName=None
     resourceObject=&MI_SparksEditorUI_PlusMinus_NoBorder
     tiling=NoTile
     imageSize=<20 20>
     margin=<ltrb 0 0 0 0>
     outlineSettings {
      bUseBrushTransparency=false
      roundingType=HalfHeightRadius
      width=0
      color {colorUseRule=UseColor_Specified, specifiedColor=<rgba 0 0 0 0>}
      cornerRadii=<0 0 0 0>
     }
     tintColor {colorUseRule=UseColor_Specified, specifiedColor=<rgba 1 1 1 1>}
     uVRegion {bIsValid=false, max=<0 0>, min=<0 0>}
    }}
   slot {class=ButtonSlot, properties {padding=<ltrb -5 0 -5 0>}}
  }
  DeleteMIDITextMapBtn {
   widgetClassPath=&EditorUtilityButton
   parent=SequenceMIDITextMapField
   widget=DeleteMIDITextMapBtn
   properties {widgetStyle {
     clickedSlateSound {resourceObject=null}
     disabled {
      bIsDynamicallyLoaded=false
      drawAs=RoundedBox
      imageType=NoImage
      mirroring=NoMirror
      resourceName=None
      resourceObject=null
      tiling=NoTile
      imageSize=<20 20>
      margin=<ltrb 0 0 0 0>
      outlineSettings {
       bUseBrushTransparency=false
       roundingType=FixedRadius
       width=0
       color {colorUseRule=UseColor_ColorTable, specifiedColor=<rgba 0 0 0 0>}
       cornerRadii=<4 4 4 4>
      }
      tintColor {colorUseRule=UseColor_Specified, specifiedColor=<rgba 0.0395 0.0395 0.0395 1>}
      uVRegion {bIsValid=false, max=<0 0>, min=<0 0>}
     }
     disabledForeground {colorUseRule=UseColor_ColorTable, specifiedColor=<rgba 0 0 0 0>}
     hovered {
      bIsDynamicallyLoaded=false
      drawAs=RoundedBox
      imageType=NoImage
      mirroring=NoMirror
      resourceName=None
      resourceObject=null
      tiling=NoTile
      imageSize=<32 32>
      margin=<ltrb 0 0 0 0>
      outlineSettings {
       bUseBrushTransparency=false
       roundingType=FixedRadius
       width=0
       color {colorUseRule=UseColor_ColorTable, specifiedColor=<rgba 0 0 0 0>}
       cornerRadii=<4 4 4 4>
      }
      tintColor {colorUseRule=UseColor_Specified, specifiedColor=<rgba 1 0.1252 0.253 1>}
      uVRegion {bIsValid=false, max=<0 0>, min=<0 0>}
     }
     hoveredForeground {colorUseRule=UseColor_ColorTable, specifiedColor=<rgba 0 0 0 0>}
     hoveredSlateSound {resourceObject=null}
     normal {
      bIsDynamicallyLoaded=false
      drawAs=RoundedBox
      imageType=NoImage
      mirroring=NoMirror
      resourceName=None
      resourceObject=null
      tiling=NoTile
      imageSize=<20 20>
      margin=<ltrb 0 0 0 0>
      outlineSettings {
       bUseBrushTransparency=false
       roundingType=FixedRadius
       width=0
       color {colorUseRule=UseColor_ColorTable, specifiedColor=<rgba 0 0 0 0>}
       cornerRadii=<4 4 4 4>
      }
      tintColor {colorUseRule=UseColor_Specified, specifiedColor=<rgba 0.775 0 0.0304 1>}
      uVRegion {bIsValid=false, max=<0 0>, min=<0 0>}
     }
     normalForeground {colorUseRule=UseColor_ColorTable, specifiedColor=<rgba 0 0 0 0>}
     normalPadding=<ltrb 12 1.5 12 1.5>
     pressed {
      bIsDynamicallyLoaded=false
      drawAs=RoundedBox
      imageType=NoImage
      mirroring=NoMirror
      resourceName=None
      resourceObject=null
      tiling=NoTile
      imageSize=<20 20>
      margin=<ltrb 0 0 0 0>
      outlineSettings {
       bUseBrushTransparency=false
       roundingType=FixedRadius
       width=0
       color {colorUseRule=UseColor_ColorTable, specifiedColor=<rgba 0 0 0 0>}
       cornerRadii=<4 4 4 4>
      }
      tintColor {colorUseRule=UseColor_Specified, specifiedColor=<rgba 0.2396 0 0 1>}
      uVRegion {bIsValid=false, max=<0 0>, min=<0 0>}
     }
     pressedForeground {colorUseRule=UseColor_ColorTable, specifiedColor=<rgba 0 0 0 0>}
     pressedPadding=<ltrb 12 2.5 12 0.5>
     pressedSlateSound {resourceObject=null}
    }}
   slot {class=HorizontalBoxSlot, properties {horizontalAlignment=HAlign_Left, verticalAlignment=VAlign_Center, padding=<ltrb 10 0 0 0>}}
  }
  Image_19 {
   widgetClassPath=&Image
   parent=DeleteMIDITextMapBtn
   widget=Image_19
   properties {
    visibility=SelfHitTestInvisible
    brush {
     bIsDynamicallyLoaded=false
     drawAs=Image
     imageType=FullColor
     mirroring=NoMirror
     resourceName=None
     resourceObject=&T_SparksEditorUI_TrashCan_64
     tiling=NoTile
     imageSize=<20 20>
     margin=<ltrb 0 0 0 0>
     outlineSettings {
      bUseBrushTransparency=false
      roundingType=HalfHeightRadius
      width=0
      color {colorUseRule=UseColor_Specified, specifiedColor=<rgba 0 0 0 0>}
      cornerRadii=<0 0 0 0>
     }
     tintColor {colorUseRule=UseColor_Specified, specifiedColor=<rgba 1 1 1 1>}
     uVRegion {bIsValid=false, max=<0 0>, min=<0 0>}
    }
   }
   slot {class=ButtonSlot, properties {padding=<ltrb -5 0 -5 0>}}
  }
  HorizontalBox_2 {
   widgetClassPath=&HorizontalBox
   parent=MainContent
   widget=HorizontalBox_2
   slot {class=VerticalBoxSlot, properties {horizontalAlignment=HAlign_Left, verticalAlignment=VAlign_Center, padding=<ltrb 0 5 0 5>}}
  }
  HorizontalBox_1 {widgetClassPath=&HorizontalBox, parent=HorizontalBox_2, widget=HorizontalBox_1, slot {class=HorizontalBoxSlot}}
  EventTypeName_3 {
   widgetClassPath=&TextBlock
   parent=HorizontalBox_1
   widget=EventTypeName_3
   properties {
    text="Event List: "
    colorAndOpacity {colorUseRule=UseColor_Specified, specifiedColor=<rgba 0.6562 0.6562 0.6562 1>}
    font {
     bForceMonospaced=false
     bMaterialIsStencil=false
     fontMaterial=null
     fontObject=&HeadingNow
     letterSpacing=0
     monospacedWidth=1
     size=7.5
     skewAmount=0
     typefaceFontName=75 Medium
     outlineSettings {
      bApplyOutlineToDropShadows=false
      bMiteredCorners=false
      bSeparateFillAlpha=false
      outlineMaterial=null
      outlineSize=0
      outlineColor=<rgba 0 0 0 1>
     }
    }
   }
   slot {class=HorizontalBoxSlot, properties {verticalAlignment=VAlign_Center, padding=<ltrb 0 0 5 0>}}
  }
  SizeBox_4 {
   widgetClassPath=&SizeBox
   parent=HorizontalBox_1
   widget=SizeBox_4
   properties {bOverride_MinDesiredWidth=true, minDesiredWidth=100}
   slot {class=HorizontalBoxSlot, properties {horizontalAlignment=HAlign_Left, verticalAlignment=VAlign_Center}}
  }
  KeySlotComboBox {
   widgetClassPath=&SearchableStringComboBox
   bIsVariable=true
   parent=SizeBox_4
   widget=KeySlotComboBox
   properties {
    selectedOption=Sequence Type List
    font {
     bForceMonospaced=false
     bMaterialIsStencil=false
     fontMaterial=null
     fontObject=&HeadingNow
     letterSpacing=0
     monospacedWidth=1
     size=7.5
     skewAmount=0
     typefaceFontName=75 Medium
     outlineSettings {
      bApplyOutlineToDropShadows=false
      bMiteredCorners=false
      bSeparateFillAlpha=false
      outlineMaterial=null
      outlineSize=0
      outlineColor=<rgba 0 0 0 1>
     }
    }
    foregroundColor {colorUseRule=UseColor_Specified, specifiedColor=<rgba 0.5271 0.5271 0.5271 1>}
    itemStyle {
     bUseParentRowBrush=false
     activeBrush {
      bIsDynamicallyLoaded=false
      drawAs=Image
      imageType=NoImage
      mirroring=NoMirror
      resourceName=None
      resourceObject=null
      tiling=NoTile
      imageSize=<0 0>
      margin=<ltrb 0 0 0 0>
      outlineSettings {
       bUseBrushTransparency=false
       roundingType=HalfHeightRadius
       width=0
       color {colorUseRule=UseColor_Specified, specifiedColor=<rgba 0 0 0 0>}
       cornerRadii=<0 0 0 0>
      }
      tintColor {colorUseRule=UseColor_Specified, specifiedColor=<rgba 0 0.162 0.7454 1>}
      uVRegion {bIsValid=false, max=<0 0>, min=<0 0>}
     }
     activeHighlightedBrush {
      bIsDynamicallyLoaded=false
      drawAs=Image
      imageType=NoImage
      mirroring=NoMirror
      resourceName=None
      resourceObject=null
      tiling=NoTile
      imageSize=<0 0>
      margin=<ltrb 0 0 0 0>
      outlineSettings {
       bUseBrushTransparency=false
       roundingType=HalfHeightRadius
       width=0
       color {colorUseRule=UseColor_Specified, specifiedColor=<rgba 0 0 0 0>}
       cornerRadii=<0 0 0 0>
      }
      tintColor {colorUseRule=UseColor_Specified, specifiedColor=<rgba 0.0252 0.0319 0.0423 1>}
      uVRegion {bIsValid=false, max=<0 0>, min=<0 0>}
     }
     activeHoveredBrush {
      bIsDynamicallyLoaded=false
      drawAs=Image
      imageType=NoImage
      mirroring=NoMirror
      resourceName=None
      resourceObject=null
      tiling=NoTile
      imageSize=<0 0>
      margin=<ltrb 0 0 0 0>
      outlineSettings {
       bUseBrushTransparency=false
       roundingType=HalfHeightRadius
       width=0
       color {colorUseRule=UseColor_Specified, specifiedColor=<rgba 0 0 0 0>}
       cornerRadii=<0 0 0 0>
      }
      tintColor {colorUseRule=UseColor_Specified, specifiedColor=<rgba 0 0.162 0.7454 1>}
      uVRegion {bIsValid=false, max=<0 0>, min=<0 0>}
     }
     dropIndicator_Above {
      bIsDynamicallyLoaded=false
      drawAs=Box
      imageType=FullColor
      mirroring=NoMirror
      resourceName=../../../Engine/Content/Slate/Common/DropZoneIndicator_Above.png
      resourceObject=null
      tiling=NoTile
      imageSize=<0 0>
      margin=<ltrb 0.625 0.625 0 0>
      outlineSettings {
       bUseBrushTransparency=false
       roundingType=HalfHeightRadius
       width=0
       color {colorUseRule=UseColor_Specified, specifiedColor=<rgba 0 0 0 0>}
       cornerRadii=<0 0 0 0>
      }
      tintColor {colorUseRule=UseColor_Specified, specifiedColor=<rgba 0.0091 0.3515 0.9647 1>}
      uVRegion {bIsValid=false, max=<0 0>, min=<0 0>}
     }
     dropIndicator_Below {
      bIsDynamicallyLoaded=false
      drawAs=Box
      imageType=FullColor
      mirroring=NoMirror
      resourceName=../../../Engine/Content/Slate/Common/DropZoneIndicator_Below.png
      resourceObject=null
      tiling=NoTile
      imageSize=<0 0>
      margin=<ltrb 0.625 0 0 0.625>
      outlineSettings {
       bUseBrushTransparency=false
       roundingType=HalfHeightRadius
       width=0
       color {colorUseRule=UseColor_Specified, specifiedColor=<rgba 0 0 0 0>}
       cornerRadii=<0 0 0 0>
      }
      tintColor {colorUseRule=UseColor_Specified, specifiedColor=<rgba 0.0091 0.3515 0.9647 1>}
      uVRegion {bIsValid=false, max=<0 0>, min=<0 0>}
     }
     dropIndicator_Onto {
      bIsDynamicallyLoaded=false
      drawAs=Box
      imageType=FullColor
      mirroring=NoMirror
      resourceName=../../../Engine/Content/Slate/Common/DropZoneIndicator_Onto.png
      resourceObject=null
      tiling=NoTile
      imageSize=<0 0>
      margin=<ltrb 0.25 0.25 0.25 0.25>
      outlineSettings {
       bUseBrushTransparency=false
       roundingType=HalfHeightRadius
       width=0
       color {colorUseRule=UseColor_Specified, specifiedColor=<rgba 0 0 0 0>}
       cornerRadii=<0 0 0 0>
      }
      tintColor {colorUseRule=UseColor_Specified, specifiedColor=<rgba 0.0091 0.3515 0.9647 1>}
      uVRegion {bIsValid=false, max=<0 0>, min=<0 0>}
     }
     evenRowBackgroundBrush {
      bIsDynamicallyLoaded=false
      drawAs=Image
      imageType=NoImage
      mirroring=NoMirror
      resourceName=None
      resourceObject=null
      tiling=NoTile
      imageSize=<0 0>
      margin=<ltrb 0 0 0 0>
      outlineSettings {
       bUseBrushTransparency=false
       roundingType=HalfHeightRadius
       width=0
       color {colorUseRule=UseColor_Specified, specifiedColor=<rgba 0 0 0 0>}
       cornerRadii=<0 0 0 0>
      }
      tintColor {colorUseRule=UseColor_Specified, specifiedColor=<rgba 0.0103 0.0103 0.0103 1>}
      uVRegion {bIsValid=false, max=<0 0>, min=<0 0>}
     }
     evenRowBackgroundHoveredBrush {
      bIsDynamicallyLoaded=false
      drawAs=Image
      imageType=NoImage
      mirroring=NoMirror
      resourceName=None
      resourceObject=null
      tiling=NoTile
      imageSize=<0 0>
      margin=<ltrb 0 0 0 0>
      outlineSettings {
       bUseBrushTransparency=false
       roundingType=HalfHeightRadius
       width=0
       color {colorUseRule=UseColor_Specified, specifiedColor=<rgba 0 0 0 0>}
       cornerRadii=<0 0 0 0>
      }
      tintColor {colorUseRule=UseColor_Specified, specifiedColor=<rgba 0.0176 0.0176 0.0176 1>}
      uVRegion {bIsValid=false, max=<0 0>, min=<0 0>}
     }
     inactiveBrush {
      bIsDynamicallyLoaded=false
      drawAs=Image
      imageType=NoImage
      mirroring=NoMirror
      resourceName=None
      resourceObject=null
      tiling=NoTile
      imageSize=<0 0>
      margin=<ltrb 0 0 0 0>
      outlineSettings {
       bUseBrushTransparency=false
       roundingType=HalfHeightRadius
       width=0
       color {colorUseRule=UseColor_Specified, specifiedColor=<rgba 0 0 0 0>}
       cornerRadii=<0 0 0 0>
      }
      tintColor {colorUseRule=UseColor_Specified, specifiedColor=<rgba 0.0513 0.0953 0.159 1>}
      uVRegion {bIsValid=false, max=<0 0>, min=<0 0>}
     }
     inactiveHighlightedBrush {
      bIsDynamicallyLoaded=false
      drawAs=Image
      imageType=NoImage
      mirroring=NoMirror
      resourceName=None
      resourceObject=null
      tiling=NoTile
      imageSize=<0 0>
      margin=<ltrb 0 0 0 0>
      outlineSettings {
       bUseBrushTransparency=false
       roundingType=HalfHeightRadius
       width=0
       color {colorUseRule=UseColor_Specified, specifiedColor=<rgba 0 0 0 0>}
       cornerRadii=<0 0 0 0>
      }
      tintColor {colorUseRule=UseColor_Specified, specifiedColor=<rgba 0.0252 0.0319 0.0423 1>}
      uVRegion {bIsValid=false, max=<0 0>, min=<0 0>}
     }
     inactiveHoveredBrush {
      bIsDynamicallyLoaded=false
      drawAs=Image
      imageType=NoImage
      mirroring=NoMirror
      resourceName=None
      resourceObject=null
      tiling=NoTile
      imageSize=<0 0>
      margin=<ltrb 0 0 0 0>
      outlineSettings {
       bUseBrushTransparency=false
       roundingType=HalfHeightRadius
       width=0
       color {colorUseRule=UseColor_Specified, specifiedColor=<rgba 0 0 0 0>}
       cornerRadii=<0 0 0 0>
      }
      tintColor {colorUseRule=UseColor_Specified, specifiedColor=<rgba 0.0513 0.0953 0.159 1>}
      uVRegion {bIsValid=false, max=<0 0>, min=<0 0>}
     }
     oddRowBackgroundBrush {
      bIsDynamicallyLoaded=false
      drawAs=Image
      imageType=NoImage
      mirroring=NoMirror
      resourceName=None
      resourceObject=null
      tiling=NoTile
      imageSize=<0 0>
      margin=<ltrb 0 0 0 0>
      outlineSettings {
       bUseBrushTransparency=false
       roundingType=HalfHeightRadius
       width=0
       color {colorUseRule=UseColor_Specified, specifiedColor=<rgba 0 0 0 0>}
       cornerRadii=<0 0 0 0>
      }
      tintColor {colorUseRule=UseColor_Specified, specifiedColor=<rgba 0.0103 0.0103 0.0103 1>}
      uVRegion {bIsValid=false, max=<0 0>, min=<0 0>}
     }
     oddRowBackgroundHoveredBrush {
      bIsDynamicallyLoaded=false
      drawAs=Image
      imageType=NoImage
      mirroring=NoMirror
      resourceName=None
      resourceObject=null
      tiling=NoTile
      imageSize=<0 0>
      margin=<ltrb 0 0 0 0>
      outlineSettings {
       bUseBrushTransparency=false
       roundingType=HalfHeightRadius
       width=0
       color {colorUseRule=UseColor_Specified, specifiedColor=<rgba 0 0 0 0>}
       cornerRadii=<0 0 0 0>
      }
      tintColor {colorUseRule=UseColor_Specified, specifiedColor=<rgba 0.0176 0.0176 0.0176 1>}
      uVRegion {bIsValid=false, max=<0 0>, min=<0 0>}
     }
     parentRowBackgroundBrush {
      bIsDynamicallyLoaded=false
      drawAs=Image
      imageType=NoImage
      mirroring=NoMirror
      resourceName=None
      resourceObject=null
      tiling=NoTile
      imageSize=<32 32>
      margin=<ltrb 0 0 0 0>
      outlineSettings {
       bUseBrushTransparency=false
       roundingType=HalfHeightRadius
       width=0
       color {colorUseRule=UseColor_Specified, specifiedColor=<rgba 0 0 0 0>}
       cornerRadii=<0 0 0 0>
      }
      tintColor {colorUseRule=UseColor_Specified, specifiedColor=<rgba 1 1 1 1>}
      uVRegion {bIsValid=false, max=<0 0>, min=<0 0>}
     }
     parentRowBackgroundHoveredBrush {
      bIsDynamicallyLoaded=false
      drawAs=Image
      imageType=NoImage
      mirroring=NoMirror
      resourceName=None
      resourceObject=null
      tiling=NoTile
      imageSize=<32 32>
      margin=<ltrb 0 0 0 0>
      outlineSettings {
       bUseBrushTransparency=false
       roundingType=HalfHeightRadius
       width=0
       color {colorUseRule=UseColor_Specified, specifiedColor=<rgba 0 0 0 0>}
       cornerRadii=<0 0 0 0>
      }
      tintColor {colorUseRule=UseColor_Specified, specifiedColor=<rgba 1 1 1 1>}
      uVRegion {bIsValid=false, max=<0 0>, min=<0 0>}
     }
     selectedTextColor {colorUseRule=UseColor_Specified, specifiedColor=<rgba 1 1 1 1>}
     selectorFocusedBrush {
      bIsDynamicallyLoaded=false
      drawAs=Border
      imageType=FullColor
      mirroring=NoMirror
      resourceName=../../../Engine/Content/Slate/Common/Selector.png
      resourceObject=null
      tiling=Both
      imageSize=<0 0>
      margin=<ltrb 0.25 0.25 0.25 0.25>
      outlineSettings {
       bUseBrushTransparency=false
       roundingType=HalfHeightRadius
       width=0
       color {colorUseRule=UseColor_Specified, specifiedColor=<rgba 0 0 0 0>}
       cornerRadii=<0 0 0 0>
      }
      tintColor {colorUseRule=UseColor_Specified, specifiedColor=<rgba 0.701 0.225 0.003 1>}
      uVRegion {bIsValid=false, max=<0 0>, min=<0 0>}
     }
     textColor {colorUseRule=UseColor_Specified, specifiedColor=<rgba 0.5271 0.5271 0.5271 1>}
    }
    widgetStyle {
     comboButtonStyle {
      downArrowAlign=VAlign_Center
      buttonStyle {
       clickedSlateSound {resourceObject=null}
       disabled {
        bIsDynamicallyLoaded=false
        drawAs=RoundedBox
        imageType=NoImage
        mirroring=NoMirror
        resourceName=None
        resourceObject=null
        tiling=NoTile
        imageSize=<0 0>
        margin=<ltrb 0 0 0 0>
        outlineSettings {
         bUseBrushTransparency=false
         roundingType=FixedRadius
         width=1
         color {colorUseRule=UseColor_ColorTable, specifiedColor=<rgba 0 0 0 0>}
         cornerRadii=<4 4 4 4>
        }
        tintColor {colorUseRule=UseColor_Specified, specifiedColor=<rgba 0.0395 0.0395 0.0395 1>}
        uVRegion {bIsValid=false, max=<0 0>, min=<0 0>}
       }
       disabledForeground {colorUseRule=UseColor_ColorTable, specifiedColor=<rgba 0 0 0 0>}
       hovered {
        bIsDynamicallyLoaded=false
        drawAs=RoundedBox
        imageType=NoImage
        mirroring=NoMirror
        resourceName=None
        resourceObject=null
        tiling=NoTile
        imageSize=<0 0>
        margin=<ltrb 0 0 0 0>
        outlineSettings {
         bUseBrushTransparency=false
         roundingType=FixedRadius
         width=1
         color {colorUseRule=UseColor_ColorTable, specifiedColor=<rgba 0 0 0 0>}
         cornerRadii=<4 4 4 4>
        }
        tintColor {colorUseRule=UseColor_Specified, specifiedColor=<rgba 0.0048 0.0048 0.0048 1>}
        uVRegion {bIsValid=false, max=<0 0>, min=<0 0>}
       }
       hoveredForeground {colorUseRule=UseColor_ColorTable, specifiedColor=<rgba 0 0 0 0>}
       hoveredSlateSound {resourceObject=null}
       normal {
        bIsDynamicallyLoaded=false
        drawAs=RoundedBox
        imageType=NoImage
        mirroring=NoMirror
        resourceName=None
        resourceObject=null
        tiling=NoTile
        imageSize=<0 0>
        margin=<ltrb 0 0 0 0>
        outlineSettings {
         bUseBrushTransparency=false
         roundingType=FixedRadius
         width=1
         color {colorUseRule=UseColor_ColorTable, specifiedColor=<rgba 0 0 0 0>}
         cornerRadii=<4 4 4 4>
        }
        tintColor {colorUseRule=UseColor_Specified, specifiedColor=<rgba 0.0048 0.0048 0.0048 1>}
        uVRegion {bIsValid=false, max=<0 0>, min=<0 0>}
       }
       normalForeground {colorUseRule=UseColor_ColorTable, specifiedColor=<rgba 0 0 0 0>}
       normalPadding=<ltrb 8 1 8 1>
       pressed {
        bIsDynamicallyLoaded=false
        drawAs=RoundedBox
        imageType=NoImage
        mirroring=NoMirror
        resourceName=None
        resourceObject=null
        tiling=NoTile
        imageSize=<0 0>
        margin=<ltrb 0 0 0 0>
        outlineSettings {
         bUseBrushTransparency=false
         roundingType=FixedRadius
         width=1
         color {colorUseRule=UseColor_ColorTable, specifiedColor=<rgba 0 0 0 0>}
         cornerRadii=<4 4 4 4>
        }
        tintColor {colorUseRule=UseColor_Specified, specifiedColor=<rgba 0.0048 0.0048 0.0048 1>}
        uVRegion {bIsValid=false, max=<0 0>, min=<0 0>}
       }
       pressedForeground {colorUseRule=UseColor_ColorTable, specifiedColor=<rgba 0 0 0 0>}
       pressedPadding=<ltrb 8 2 8 0>
       pressedSlateSound {resourceObject=null}
      }
      contentPadding=<ltrb 0 0 0 0>
      downArrowImage {
       bIsDynamicallyLoaded=false
       drawAs=Image
       imageType=Vector
       mirroring=NoMirror
       resourceName=../../../Engine/Content/Slate/Starship/CoreWidgets/ComboBox/wide-chevron-down.svg
       resourceObject=null
       tiling=NoTile
       imageSize=<16 16>
       margin=<ltrb 0 0 0 0>
       outlineSettings {
        bUseBrushTransparency=false
        roundingType=HalfHeightRadius
        width=0
        color {colorUseRule=UseColor_Specified, specifiedColor=<rgba 0 0 0 0>}
        cornerRadii=<0 0 0 0>
       }
       tintColor {colorUseRule=UseColor_Specified, specifiedColor=<rgba 1 1 1 1>}
       uVRegion {bIsValid=false, max=<0 0>, min=<0 0>}
      }
      downArrowPadding=<ltrb 2 2 2 2>
      menuBorderBrush {
       bIsDynamicallyLoaded=false
       drawAs=RoundedBox
       imageType=NoImage
       mirroring=NoMirror
       resourceName=None
       resourceObject=null
       tiling=NoTile
       imageSize=<0 0>
       margin=<ltrb 0 0 0 0>
       outlineSettings {
        bUseBrushTransparency=false
        roundingType=FixedRadius
        width=1
        color {colorUseRule=UseColor_ColorTable, specifiedColor=<rgba 0 0 0 0>}
        cornerRadii=<0 0 0 0>
       }
       tintColor {colorUseRule=UseColor_Specified, specifiedColor=<rgba 0.0395 0.0395 0.0395 1>}
       uVRegion {bIsValid=false, max=<0 0>, min=<0 0>}
      }
      menuBorderPadding=<ltrb 0 0 0 0>
      shadowColorAndOpacity=<rgba 0 0 0 1>
      shadowOffset=<0 0>
     }
     contentPadding=<ltrb 0 0 0 0>
     menuRowPadding=<ltrb 8 3 8 3>
     pressedSlateSound {resourceObject=null}
     selectionChangeSlateSound {resourceObject=null}
    }
   }
   slot {class=SizeBoxSlot}
  }
  RootAdd {
   widgetClassPath=&EditorUtilityButton
   parent=HorizontalBox_1
   widget=RootAdd
   properties {widgetStyle {
     clickedSlateSound {resourceObject=null}
     disabled {
      bIsDynamicallyLoaded=false
      drawAs=RoundedBox
      imageType=NoImage
      mirroring=NoMirror
      resourceName=None
      resourceObject=null
      tiling=NoTile
      imageSize=<32 32>
      margin=<ltrb 0.5 0.5 0.5 0.5>
      outlineSettings {
       bUseBrushTransparency=false
       roundingType=FixedRadius
       width=0
       color {colorUseRule=UseColor_ColorTable, specifiedColor=<rgba 0 0 0 0>}
       cornerRadii=<4 4 4 4>
      }
      tintColor {colorUseRule=UseColor_Specified, specifiedColor=<rgba 0.0395 0.0395 0.0395 1>}
      uVRegion {bIsValid=false, max=<0 0>, min=<0 0>}
     }
     disabledForeground {colorUseRule=UseColor_ColorTable, specifiedColor=<rgba 0 0 0 0>}
     hovered {
      bIsDynamicallyLoaded=false
      drawAs=RoundedBox
      imageType=NoImage
      mirroring=NoMirror
      resourceName=None
      resourceObject=null
      tiling=NoTile
      imageSize=<32 32>
      margin=<ltrb 0 0 0 0>
      outlineSettings {
       bUseBrushTransparency=false
       roundingType=FixedRadius
       width=1
       color {colorUseRule=UseColor_ColorTable, specifiedColor=<rgba 0 0 0 0>}
       cornerRadii=<4 4 4 4>
      }
      tintColor {colorUseRule=UseColor_Specified, specifiedColor=<rgba 0 1 0.0916 1>}
      uVRegion {bIsValid=false, max=<0 0>, min=<0 0>}
     }
     hoveredForeground {colorUseRule=UseColor_ColorTable, specifiedColor=<rgba 0 0 0 0>}
     hoveredSlateSound {resourceObject=null}
     normal {
      bIsDynamicallyLoaded=false
      drawAs=RoundedBox
      imageType=NoImage
      mirroring=NoMirror
      resourceName=None
      resourceObject=null
      tiling=NoTile
      imageSize=<32 32>
      margin=<ltrb 0.5 0.5 0.5 0.5>
      outlineSettings {
       bUseBrushTransparency=false
       roundingType=FixedRadius
       width=0
       color {colorUseRule=UseColor_ColorTable, specifiedColor=<rgba 0 0 0 0>}
       cornerRadii=<4 4 4 4>
      }
      tintColor {colorUseRule=UseColor_Specified, specifiedColor=<rgba 0.0771 0.6167 0.1289 1>}
      uVRegion {bIsValid=false, max=<0 0>, min=<0 0>}
     }
     normalForeground {colorUseRule=UseColor_ColorTable, specifiedColor=<rgba 0 0 0 0>}
     normalPadding=<ltrb 12 1.5 12 1.5>
     pressed {
      bIsDynamicallyLoaded=false
      drawAs=RoundedBox
      imageType=NoImage
      mirroring=NoMirror
      resourceName=None
      resourceObject=null
      tiling=NoTile
      imageSize=<32 32>
      margin=<ltrb 0.5 0.5 0.5 0.5>
      outlineSettings {
       bUseBrushTransparency=false
       roundingType=FixedRadius
       width=0
       color {colorUseRule=UseColor_ColorTable, specifiedColor=<rgba 0 0 0 0>}
       cornerRadii=<4 4 4 4>
      }
      tintColor {colorUseRule=UseColor_Specified, specifiedColor=<rgba 0.1806 0.4583 0.1572 1>}
      uVRegion {bIsValid=false, max=<0 0>, min=<0 0>}
     }
     pressedForeground {colorUseRule=UseColor_ColorTable, specifiedColor=<rgba 0 0 0 0>}
     pressedPadding=<ltrb 12 2.5 12 0.5>
     pressedSlateSound {resourceObject=null}
    }}
   slot {class=HorizontalBoxSlot, properties {horizontalAlignment=HAlign_Center, verticalAlignment=VAlign_Center, padding=<ltrb 10 0 0 0>}}
  }
  Image {
   widgetClassPath=&Image
   parent=RootAdd
   widget=Image
   properties {brush {
     bIsDynamicallyLoaded=false
     drawAs=Image
     imageType=NoImage
     mirroring=NoMirror
     resourceName=None
     resourceObject=&MI_SparksEditorUI_PlusMinus_NoBorder
     tiling=NoTile
     imageSize=<20 20>
     margin=<ltrb 0 0 0 0>
     outlineSettings {
      bUseBrushTransparency=false
      roundingType=HalfHeightRadius
      width=0
      color {colorUseRule=UseColor_Specified, specifiedColor=<rgba 0 0 0 0>}
      cornerRadii=<0 0 0 0>
     }
     tintColor {colorUseRule=UseColor_Specified, specifiedColor=<rgba 1 1 1 1>}
     uVRegion {bIsValid=false, max=<0 0>, min=<0 0>}
    }}
   slot {class=ButtonSlot, properties {padding=<ltrb -5 0 -5 0>}}
  }
  DeleteEventButton {
   widgetClassPath=&EditorUtilityButton
   parent=HorizontalBox_1
   widget=DeleteEventButton
   properties {widgetStyle {
     clickedSlateSound {resourceObject=null}
     disabled {
      bIsDynamicallyLoaded=false
      drawAs=RoundedBox
      imageType=NoImage
      mirroring=NoMirror
      resourceName=None
      resourceObject=null
      tiling=NoTile
      imageSize=<20 20>
      margin=<ltrb 0 0 0 0>
      outlineSettings {
       bUseBrushTransparency=false
       roundingType=FixedRadius
       width=0
       color {colorUseRule=UseColor_ColorTable, specifiedColor=<rgba 0 0 0 0>}
       cornerRadii=<4 4 4 4>
      }
      tintColor {colorUseRule=UseColor_Specified, specifiedColor=<rgba 0.0395 0.0395 0.0395 1>}
      uVRegion {bIsValid=false, max=<0 0>, min=<0 0>}
     }
     disabledForeground {colorUseRule=UseColor_ColorTable, specifiedColor=<rgba 0 0 0 0>}
     hovered {
      bIsDynamicallyLoaded=false
      drawAs=RoundedBox
      imageType=NoImage
      mirroring=NoMirror
      resourceName=None
      resourceObject=null
      tiling=NoTile
      imageSize=<32 32>
      margin=<ltrb 0 0 0 0>
      outlineSettings {
       bUseBrushTransparency=false
       roundingType=FixedRadius
       width=0
       color {colorUseRule=UseColor_ColorTable, specifiedColor=<rgba 0 0 0 0>}
       cornerRadii=<4 4 4 4>
      }
      tintColor {colorUseRule=UseColor_Specified, specifiedColor=<rgba 1 0.1252 0.253 1>}
      uVRegion {bIsValid=false, max=<0 0>, min=<0 0>}
     }
     hoveredForeground {colorUseRule=UseColor_ColorTable, specifiedColor=<rgba 0 0 0 0>}
     hoveredSlateSound {resourceObject=null}
     normal {
      bIsDynamicallyLoaded=false
      drawAs=RoundedBox
      imageType=NoImage
      mirroring=NoMirror
      resourceName=None
      resourceObject=null
      tiling=NoTile
      imageSize=<20 20>
      margin=<ltrb 0 0 0 0>
      outlineSettings {
       bUseBrushTransparency=false
       roundingType=FixedRadius
       width=0
       color {colorUseRule=UseColor_ColorTable, specifiedColor=<rgba 0 0 0 0>}
       cornerRadii=<4 4 4 4>
      }
      tintColor {colorUseRule=UseColor_Specified, specifiedColor=<rgba 0.775 0 0.0304 1>}
      uVRegion {bIsValid=false, max=<0 0>, min=<0 0>}
     }
     normalForeground {colorUseRule=UseColor_ColorTable, specifiedColor=<rgba 0 0 0 0>}
     normalPadding=<ltrb 12 1.5 12 1.5>
     pressed {
      bIsDynamicallyLoaded=false
      drawAs=RoundedBox
      imageType=NoImage
      mirroring=NoMirror
      resourceName=None
      resourceObject=null
      tiling=NoTile
      imageSize=<20 20>
      margin=<ltrb 0 0 0 0>
      outlineSettings {
       bUseBrushTransparency=false
       roundingType=FixedRadius
       width=0
       color {colorUseRule=UseColor_ColorTable, specifiedColor=<rgba 0 0 0 0>}
       cornerRadii=<4 4 4 4>
      }
      tintColor {colorUseRule=UseColor_Specified, specifiedColor=<rgba 0.2396 0 0 1>}
      uVRegion {bIsValid=false, max=<0 0>, min=<0 0>}
     }
     pressedForeground {colorUseRule=UseColor_ColorTable, specifiedColor=<rgba 0 0 0 0>}
     pressedPadding=<ltrb 12 2.5 12 0.5>
     pressedSlateSound {resourceObject=null}
    }}
   slot {class=HorizontalBoxSlot, properties {horizontalAlignment=HAlign_Left, verticalAlignment=VAlign_Center, padding=<ltrb 10 0 10 0>}}
  }
  Image_98 {
   widgetClassPath=&Image
   parent=DeleteEventButton
   widget=Image_98
   properties {brush {
     bIsDynamicallyLoaded=false
     drawAs=Image
     imageType=FullColor
     mirroring=NoMirror
     resourceName=None
     resourceObject=&T_SparksEditorUI_TrashCan_64
     tiling=NoTile
     imageSize=<20 20>
     margin=<ltrb 0 0 0 0>
     outlineSettings {
      bUseBrushTransparency=false
      roundingType=HalfHeightRadius
      width=0
      color {colorUseRule=UseColor_Specified, specifiedColor=<rgba 0 0 0 0>}
      cornerRadii=<0 0 0 0>
     }
     tintColor {colorUseRule=UseColor_Specified, specifiedColor=<rgba 1 1 1 1>}
     uVRegion {bIsValid=false, max=<0 0>, min=<0 0>}
    }}
   slot {class=ButtonSlot, properties {padding=<ltrb -5 0 -5 0>}}
  }
  HorizontalBox {
   widgetClassPath=&HorizontalBox
   parent=HorizontalBox_2
   widget=HorizontalBox
   slot {class=HorizontalBoxSlot, properties {horizontalAlignment=HAlign_Left, verticalAlignment=VAlign_Center}}
  }
  EventTypeName_2 {
   widgetClassPath=&TextBlock
   parent=HorizontalBox
   widget=EventTypeName_2
   properties {
    text="Pool List: "
    colorAndOpacity {colorUseRule=UseColor_Specified, specifiedColor=<rgba 0.6562 0.6562 0.6562 1>}
    font {
     bForceMonospaced=false
     bMaterialIsStencil=false
     fontMaterial=null
     fontObject=&HeadingNow
     letterSpacing=0
     monospacedWidth=1
     size=7.5
     skewAmount=0
     typefaceFontName=75 Medium
     outlineSettings {
      bApplyOutlineToDropShadows=false
      bMiteredCorners=false
      bSeparateFillAlpha=false
      outlineMaterial=null
      outlineSize=0
      outlineColor=<rgba 0 0 0 1>
     }
    }
   }
   slot {class=HorizontalBoxSlot, properties {verticalAlignment=VAlign_Center, padding=<ltrb 0 0 5 0>}}
  }
  SizeBox_6 {
   widgetClassPath=&SizeBox
   parent=HorizontalBox
   widget=SizeBox_6
   properties {bOverride_MinDesiredWidth=true, maxDesiredWidth=300, minDesiredWidth=300}
   slot {class=HorizontalBoxSlot, properties {horizontalAlignment=HAlign_Left, verticalAlignment=VAlign_Center}}
  }
  SequencesDataListComboBox {
   widgetClassPath=&SearchableStringComboBox
   bIsVariable=true
   parent=SizeBox_6
   widget=SequencesDataListComboBox
   properties {
    enableGamepadNavigationMode=false
    selectedOption=Sequence Type List
    font {
     bForceMonospaced=false
     bMaterialIsStencil=false
     fontMaterial=null
     fontObject=&HeadingNow
     letterSpacing=0
     monospacedWidth=1
     size=7.5
     skewAmount=0
     typefaceFontName=75 Medium
     outlineSettings {
      bApplyOutlineToDropShadows=false
      bMiteredCorners=false
      bSeparateFillAlpha=false
      outlineMaterial=null
      outlineSize=0
      outlineColor=<rgba 0 0 0 1>
     }
    }
    foregroundColor {colorUseRule=UseColor_Specified, specifiedColor=<rgba 0.5271 0.5271 0.5271 1>}
    itemStyle {
     bUseParentRowBrush=false
     activeBrush {
      bIsDynamicallyLoaded=false
      drawAs=Image
      imageType=NoImage
      mirroring=NoMirror
      resourceName=None
      resourceObject=null
      tiling=NoTile
      imageSize=<0 0>
      margin=<ltrb 0 0 0 0>
      outlineSettings {
       bUseBrushTransparency=false
       roundingType=HalfHeightRadius
       width=0
       color {colorUseRule=UseColor_Specified, specifiedColor=<rgba 0 0 0 0>}
       cornerRadii=<0 0 0 0>
      }
      tintColor {colorUseRule=UseColor_Specified, specifiedColor=<rgba 0 0.162 0.7454 1>}
      uVRegion {bIsValid=false, max=<0 0>, min=<0 0>}
     }
     activeHighlightedBrush {
      bIsDynamicallyLoaded=false
      drawAs=Image
      imageType=NoImage
      mirroring=NoMirror
      resourceName=None
      resourceObject=null
      tiling=NoTile
      imageSize=<0 0>
      margin=<ltrb 0 0 0 0>
      outlineSettings {
       bUseBrushTransparency=false
       roundingType=HalfHeightRadius
       width=0
       color {colorUseRule=UseColor_Specified, specifiedColor=<rgba 0 0 0 0>}
       cornerRadii=<0 0 0 0>
      }
      tintColor {colorUseRule=UseColor_Specified, specifiedColor=<rgba 0.0252 0.0319 0.0423 1>}
      uVRegion {bIsValid=false, max=<0 0>, min=<0 0>}
     }
     activeHoveredBrush {
      bIsDynamicallyLoaded=false
      drawAs=Image
      imageType=NoImage
      mirroring=NoMirror
      resourceName=None
      resourceObject=null
      tiling=NoTile
      imageSize=<0 0>
      margin=<ltrb 0 0 0 0>
      outlineSettings {
       bUseBrushTransparency=false
       roundingType=HalfHeightRadius
       width=0
       color {colorUseRule=UseColor_Specified, specifiedColor=<rgba 0 0 0 0>}
       cornerRadii=<0 0 0 0>
      }
      tintColor {colorUseRule=UseColor_Specified, specifiedColor=<rgba 0 0.162 0.7454 1>}
      uVRegion {bIsValid=false, max=<0 0>, min=<0 0>}
     }
     dropIndicator_Above {
      bIsDynamicallyLoaded=false
      drawAs=Box
      imageType=FullColor
      mirroring=NoMirror
      resourceName=../../../Engine/Content/Slate/Common/DropZoneIndicator_Above.png
      resourceObject=null
      tiling=NoTile
      imageSize=<0 0>
      margin=<ltrb 0.625 0.625 0 0>
      outlineSettings {
       bUseBrushTransparency=false
       roundingType=HalfHeightRadius
       width=0
       color {colorUseRule=UseColor_Specified, specifiedColor=<rgba 0 0 0 0>}
       cornerRadii=<0 0 0 0>
      }
      tintColor {colorUseRule=UseColor_Specified, specifiedColor=<rgba 0.0091 0.3515 0.9647 1>}
      uVRegion {bIsValid=false, max=<0 0>, min=<0 0>}
     }
     dropIndicator_Below {
      bIsDynamicallyLoaded=false
      drawAs=Box
      imageType=FullColor
      mirroring=NoMirror
      resourceName=../../../Engine/Content/Slate/Common/DropZoneIndicator_Below.png
      resourceObject=null
      tiling=NoTile
      imageSize=<0 0>
      margin=<ltrb 0.625 0 0 0.625>
      outlineSettings {
       bUseBrushTransparency=false
       roundingType=HalfHeightRadius
       width=0
       color {colorUseRule=UseColor_Specified, specifiedColor=<rgba 0 0 0 0>}
       cornerRadii=<0 0 0 0>
      }
      tintColor {colorUseRule=UseColor_Specified, specifiedColor=<rgba 0.0091 0.3515 0.9647 1>}
      uVRegion {bIsValid=false, max=<0 0>, min=<0 0>}
     }
     dropIndicator_Onto {
      bIsDynamicallyLoaded=false
      drawAs=Box
      imageType=FullColor
      mirroring=NoMirror
      resourceName=../../../Engine/Content/Slate/Common/DropZoneIndicator_Onto.png
      resourceObject=null
      tiling=NoTile
      imageSize=<0 0>
      margin=<ltrb 0.25 0.25 0.25 0.25>
      outlineSettings {
       bUseBrushTransparency=false
       roundingType=HalfHeightRadius
       width=0
       color {colorUseRule=UseColor_Specified, specifiedColor=<rgba 0 0 0 0>}
       cornerRadii=<0 0 0 0>
      }
      tintColor {colorUseRule=UseColor_Specified, specifiedColor=<rgba 0.0091 0.3515 0.9647 1>}
      uVRegion {bIsValid=false, max=<0 0>, min=<0 0>}
     }
     evenRowBackgroundBrush {
      bIsDynamicallyLoaded=false
      drawAs=Image
      imageType=NoImage
      mirroring=NoMirror
      resourceName=None
      resourceObject=null
      tiling=NoTile
      imageSize=<0 0>
      margin=<ltrb 0 0 0 0>
      outlineSettings {
       bUseBrushTransparency=false
       roundingType=HalfHeightRadius
       width=0
       color {colorUseRule=UseColor_Specified, specifiedColor=<rgba 0 0 0 0>}
       cornerRadii=<0 0 0 0>
      }
      tintColor {colorUseRule=UseColor_Specified, specifiedColor=<rgba 0.0103 0.0103 0.0103 1>}
      uVRegion {bIsValid=false, max=<0 0>, min=<0 0>}
     }
     evenRowBackgroundHoveredBrush {
      bIsDynamicallyLoaded=false
      drawAs=Image
      imageType=NoImage
      mirroring=NoMirror
      resourceName=None
      resourceObject=null
      tiling=NoTile
      imageSize=<0 0>
      margin=<ltrb 0 0 0 0>
      outlineSettings {
       bUseBrushTransparency=false
       roundingType=HalfHeightRadius
       width=0
       color {colorUseRule=UseColor_Specified, specifiedColor=<rgba 0 0 0 0>}
       cornerRadii=<0 0 0 0>
      }
      tintColor {colorUseRule=UseColor_Specified, specifiedColor=<rgba 0.0176 0.0176 0.0176 1>}
      uVRegion {bIsValid=false, max=<0 0>, min=<0 0>}
     }
     inactiveBrush {
      bIsDynamicallyLoaded=false
      drawAs=Image
      imageType=NoImage
      mirroring=NoMirror
      resourceName=None
      resourceObject=null
      tiling=NoTile
      imageSize=<0 0>
      margin=<ltrb 0 0 0 0>
      outlineSettings {
       bUseBrushTransparency=false
       roundingType=HalfHeightRadius
       width=0
       color {colorUseRule=UseColor_Specified, specifiedColor=<rgba 0 0 0 0>}
       cornerRadii=<0 0 0 0>
      }
      tintColor {colorUseRule=UseColor_Specified, specifiedColor=<rgba 0.0513 0.0953 0.159 1>}
      uVRegion {bIsValid=false, max=<0 0>, min=<0 0>}
     }
     inactiveHighlightedBrush {
      bIsDynamicallyLoaded=false
      drawAs=Image
      imageType=NoImage
      mirroring=NoMirror
      resourceName=None
      resourceObject=null
      tiling=NoTile
      imageSize=<0 0>
      margin=<ltrb 0 0 0 0>
      outlineSettings {
       bUseBrushTransparency=false
       roundingType=HalfHeightRadius
       width=0
       color {colorUseRule=UseColor_Specified, specifiedColor=<rgba 0 0 0 0>}
       cornerRadii=<0 0 0 0>
      }
      tintColor {colorUseRule=UseColor_Specified, specifiedColor=<rgba 0.0252 0.0319 0.0423 1>}
      uVRegion {bIsValid=false, max=<0 0>, min=<0 0>}
     }
     inactiveHoveredBrush {
      bIsDynamicallyLoaded=false
      drawAs=Image
      imageType=NoImage
      mirroring=NoMirror
      resourceName=None
      resourceObject=null
      tiling=NoTile
      imageSize=<0 0>
      margin=<ltrb 0 0 0 0>
      outlineSettings {
       bUseBrushTransparency=false
       roundingType=HalfHeightRadius
       width=0
       color {colorUseRule=UseColor_Specified, specifiedColor=<rgba 0 0 0 0>}
       cornerRadii=<0 0 0 0>
      }
      tintColor {colorUseRule=UseColor_Specified, specifiedColor=<rgba 0.0513 0.0953 0.159 1>}
      uVRegion {bIsValid=false, max=<0 0>, min=<0 0>}
     }
     oddRowBackgroundBrush {
      bIsDynamicallyLoaded=false
      drawAs=Image
      imageType=NoImage
      mirroring=NoMirror
      resourceName=None
      resourceObject=null
      tiling=NoTile
      imageSize=<0 0>
      margin=<ltrb 0 0 0 0>
      outlineSettings {
       bUseBrushTransparency=false
       roundingType=HalfHeightRadius
       width=0
       color {colorUseRule=UseColor_Specified, specifiedColor=<rgba 0 0 0 0>}
       cornerRadii=<0 0 0 0>
      }
      tintColor {colorUseRule=UseColor_Specified, specifiedColor=<rgba 0.0103 0.0103 0.0103 1>}
      uVRegion {bIsValid=false, max=<0 0>, min=<0 0>}
     }
     oddRowBackgroundHoveredBrush {
      bIsDynamicallyLoaded=false
      drawAs=Image
      imageType=NoImage
      mirroring=NoMirror
      resourceName=None
      resourceObject=null
      tiling=NoTile
      imageSize=<0 0>
      margin=<ltrb 0 0 0 0>
      outlineSettings {
       bUseBrushTransparency=false
       roundingType=HalfHeightRadius
       width=0
       color {colorUseRule=UseColor_Specified, specifiedColor=<rgba 0 0 0 0>}
       cornerRadii=<0 0 0 0>
      }
      tintColor {colorUseRule=UseColor_Specified, specifiedColor=<rgba 0.0176 0.0176 0.0176 1>}
      uVRegion {bIsValid=false, max=<0 0>, min=<0 0>}
     }
     parentRowBackgroundBrush {
      bIsDynamicallyLoaded=false
      drawAs=Image
      imageType=NoImage
      mirroring=NoMirror
      resourceName=None
      resourceObject=null
      tiling=NoTile
      imageSize=<32 32>
      margin=<ltrb 0 0 0 0>
      outlineSettings {
       bUseBrushTransparency=false
       roundingType=HalfHeightRadius
       width=0
       color {colorUseRule=UseColor_Specified, specifiedColor=<rgba 0 0 0 0>}
       cornerRadii=<0 0 0 0>
      }
      tintColor {colorUseRule=UseColor_Specified, specifiedColor=<rgba 1 1 1 1>}
      uVRegion {bIsValid=false, max=<0 0>, min=<0 0>}
     }
     parentRowBackgroundHoveredBrush {
      bIsDynamicallyLoaded=false
      drawAs=Image
      imageType=NoImage
      mirroring=NoMirror
      resourceName=None
      resourceObject=null
      tiling=NoTile
      imageSize=<32 32>
      margin=<ltrb 0 0 0 0>
      outlineSettings {
       bUseBrushTransparency=false
       roundingType=HalfHeightRadius
       width=0
       color {colorUseRule=UseColor_Specified, specifiedColor=<rgba 0 0 0 0>}
       cornerRadii=<0 0 0 0>
      }
      tintColor {colorUseRule=UseColor_Specified, specifiedColor=<rgba 1 1 1 1>}
      uVRegion {bIsValid=false, max=<0 0>, min=<0 0>}
     }
     selectedTextColor {colorUseRule=UseColor_Specified, specifiedColor=<rgba 1 1 1 1>}
     selectorFocusedBrush {
      bIsDynamicallyLoaded=false
      drawAs=Border
      imageType=FullColor
      mirroring=NoMirror
      resourceName=../../../Engine/Content/Slate/Common/Selector.png
      resourceObject=null
      tiling=Both
      imageSize=<0 0>
      margin=<ltrb 0.25 0.25 0.25 0.25>
      outlineSettings {
       bUseBrushTransparency=false
       roundingType=HalfHeightRadius
       width=0
       color {colorUseRule=UseColor_Specified, specifiedColor=<rgba 0 0 0 0>}
       cornerRadii=<0 0 0 0>
      }
      tintColor {colorUseRule=UseColor_Specified, specifiedColor=<rgba 0.701 0.225 0.003 1>}
      uVRegion {bIsValid=false, max=<0 0>, min=<0 0>}
     }
     textColor {colorUseRule=UseColor_Specified, specifiedColor=<rgba 0.5271 0.5271 0.5271 1>}
    }
    widgetStyle {
     comboButtonStyle {
      downArrowAlign=VAlign_Center
      buttonStyle {
       clickedSlateSound {resourceObject=null}
       disabled {
        bIsDynamicallyLoaded=false
        drawAs=RoundedBox
        imageType=NoImage
        mirroring=NoMirror
        resourceName=None
        resourceObject=null
        tiling=NoTile
        imageSize=<0 0>
        margin=<ltrb 0 0 0 0>
        outlineSettings {
         bUseBrushTransparency=false
         roundingType=FixedRadius
         width=1
         color {colorUseRule=UseColor_ColorTable, specifiedColor=<rgba 0 0 0 0>}
         cornerRadii=<4 4 4 4>
        }
        tintColor {colorUseRule=UseColor_Specified, specifiedColor=<rgba 0.0395 0.0395 0.0395 1>}
        uVRegion {bIsValid=false, max=<0 0>, min=<0 0>}
       }
       disabledForeground {colorUseRule=UseColor_ColorTable, specifiedColor=<rgba 0 0 0 0>}
       hovered {
        bIsDynamicallyLoaded=false
        drawAs=RoundedBox
        imageType=NoImage
        mirroring=NoMirror
        resourceName=None
        resourceObject=null
        tiling=NoTile
        imageSize=<0 0>
        margin=<ltrb 0 0 0 0>
        outlineSettings {
         bUseBrushTransparency=false
         roundingType=FixedRadius
         width=1
         color {colorUseRule=UseColor_ColorTable, specifiedColor=<rgba 0 0 0 0>}
         cornerRadii=<4 4 4 4>
        }
        tintColor {colorUseRule=UseColor_Specified, specifiedColor=<rgba 0.0048 0.0048 0.0048 1>}
        uVRegion {bIsValid=false, max=<0 0>, min=<0 0>}
       }
       hoveredForeground {colorUseRule=UseColor_ColorTable, specifiedColor=<rgba 0 0 0 0>}
       hoveredSlateSound {resourceObject=null}
       normal {
        bIsDynamicallyLoaded=false
        drawAs=RoundedBox
        imageType=NoImage
        mirroring=NoMirror
        resourceName=None
        resourceObject=null
        tiling=NoTile
        imageSize=<0 0>
        margin=<ltrb 0 0 0 0>
        outlineSettings {
         bUseBrushTransparency=false
         roundingType=FixedRadius
         width=1
         color {colorUseRule=UseColor_ColorTable, specifiedColor=<rgba 0 0 0 0>}
         cornerRadii=<4 4 4 4>
        }
        tintColor {colorUseRule=UseColor_Specified, specifiedColor=<rgba 0.0048 0.0048 0.0048 1>}
        uVRegion {bIsValid=false, max=<0 0>, min=<0 0>}
       }
       normalForeground {colorUseRule=UseColor_ColorTable, specifiedColor=<rgba 0 0 0 0>}
       normalPadding=<ltrb 8 1 8 1>
       pressed {
        bIsDynamicallyLoaded=false
        drawAs=RoundedBox
        imageType=NoImage
        mirroring=NoMirror
        resourceName=None
        resourceObject=null
        tiling=NoTile
        imageSize=<0 0>
        margin=<ltrb 0 0 0 0>
        outlineSettings {
         bUseBrushTransparency=false
         roundingType=FixedRadius
         width=1
         color {colorUseRule=UseColor_ColorTable, specifiedColor=<rgba 0 0 0 0>}
         cornerRadii=<4 4 4 4>
        }
        tintColor {colorUseRule=UseColor_Specified, specifiedColor=<rgba 0.0048 0.0048 0.0048 1>}
        uVRegion {bIsValid=false, max=<0 0>, min=<0 0>}
       }
       pressedForeground {colorUseRule=UseColor_ColorTable, specifiedColor=<rgba 0 0 0 0>}
       pressedPadding=<ltrb 8 2 8 0>
       pressedSlateSound {resourceObject=null}
      }
      contentPadding=<ltrb 0 0 0 0>
      downArrowImage {
       bIsDynamicallyLoaded=false
       drawAs=Image
       imageType=Vector
       mirroring=NoMirror
       resourceName=../../../Engine/Content/Slate/Starship/CoreWidgets/ComboBox/wide-chevron-down.svg
       resourceObject=null
       tiling=NoTile
       imageSize=<16 16>
       margin=<ltrb 0 0 0 0>
       outlineSettings {
        bUseBrushTransparency=false
        roundingType=HalfHeightRadius
        width=0
        color {colorUseRule=UseColor_Specified, specifiedColor=<rgba 0 0 0 0>}
        cornerRadii=<0 0 0 0>
       }
       tintColor {colorUseRule=UseColor_Specified, specifiedColor=<rgba 1 1 1 1>}
       uVRegion {bIsValid=false, max=<0 0>, min=<0 0>}
      }
      downArrowPadding=<ltrb 2 2 2 2>
      menuBorderBrush {
       bIsDynamicallyLoaded=false
       drawAs=RoundedBox
       imageType=NoImage
       mirroring=NoMirror
       resourceName=None
       resourceObject=null
       tiling=NoTile
       imageSize=<0 0>
       margin=<ltrb 0 0 0 0>
       outlineSettings {
        bUseBrushTransparency=false
        roundingType=FixedRadius
        width=1
        color {colorUseRule=UseColor_ColorTable, specifiedColor=<rgba 0 0 0 0>}
        cornerRadii=<0 0 0 0>
       }
       tintColor {colorUseRule=UseColor_Specified, specifiedColor=<rgba 0.0395 0.0395 0.0395 1>}
       uVRegion {bIsValid=false, max=<0 0>, min=<0 0>}
      }
      menuBorderPadding=<ltrb 0 0 0 0>
      shadowColorAndOpacity=<rgba 0 0 0 1>
      shadowOffset=<0 0>
     }
     contentPadding=<ltrb 0 0 0 0>
     menuRowPadding=<ltrb 8 3 8 3>
     pressedSlateSound {resourceObject=null}
     selectionChangeSlateSound {resourceObject=null}
    }
   }
   slot {class=SizeBoxSlot}
  }
  AddSequenceDA {
   widgetClassPath=&EditorUtilityButton
   parent=HorizontalBox
   widget=AddSequenceDA
   properties {widgetStyle {
     clickedSlateSound {resourceObject=null}
     disabled {
      bIsDynamicallyLoaded=false
      drawAs=RoundedBox
      imageType=NoImage
      mirroring=NoMirror
      resourceName=None
      resourceObject=null
      tiling=NoTile
      imageSize=<32 32>
      margin=<ltrb 0.5 0.5 0.5 0.5>
      outlineSettings {
       bUseBrushTransparency=false
       roundingType=FixedRadius
       width=0
       color {colorUseRule=UseColor_ColorTable, specifiedColor=<rgba 0 0 0 0>}
       cornerRadii=<4 4 4 4>
      }
      tintColor {colorUseRule=UseColor_Specified, specifiedColor=<rgba 0.0395 0.0395 0.0395 1>}
      uVRegion {bIsValid=false, max=<0 0>, min=<0 0>}
     }
     disabledForeground {colorUseRule=UseColor_ColorTable, specifiedColor=<rgba 0 0 0 0>}
     hovered {
      bIsDynamicallyLoaded=false
      drawAs=RoundedBox
      imageType=NoImage
      mirroring=NoMirror
      resourceName=None
      resourceObject=null
      tiling=NoTile
      imageSize=<32 32>
      margin=<ltrb 0 0 0 0>
      outlineSettings {
       bUseBrushTransparency=false
       roundingType=FixedRadius
       width=1
       color {colorUseRule=UseColor_ColorTable, specifiedColor=<rgba 0 0 0 0>}
       cornerRadii=<4 4 4 4>
      }
      tintColor {colorUseRule=UseColor_Specified, specifiedColor=<rgba 0 1 0.0916 1>}
      uVRegion {bIsValid=false, max=<0 0>, min=<0 0>}
     }
     hoveredForeground {colorUseRule=UseColor_ColorTable, specifiedColor=<rgba 0 0 0 0>}
     hoveredSlateSound {resourceObject=null}
     normal {
      bIsDynamicallyLoaded=false
      drawAs=RoundedBox
      imageType=NoImage
      mirroring=NoMirror
      resourceName=None
      resourceObject=null
      tiling=NoTile
      imageSize=<32 32>
      margin=<ltrb 0.5 0.5 0.5 0.5>
      outlineSettings {
       bUseBrushTransparency=false
       roundingType=FixedRadius
       width=0
       color {colorUseRule=UseColor_ColorTable, specifiedColor=<rgba 0 0 0 0>}
       cornerRadii=<4 4 4 4>
      }
      tintColor {colorUseRule=UseColor_Specified, specifiedColor=<rgba 0.0771 0.6167 0.1289 1>}
      uVRegion {bIsValid=false, max=<0 0>, min=<0 0>}
     }
     normalForeground {colorUseRule=UseColor_ColorTable, specifiedColor=<rgba 0 0 0 0>}
     normalPadding=<ltrb 12 1.5 12 1.5>
     pressed {
      bIsDynamicallyLoaded=false
      drawAs=RoundedBox
      imageType=NoImage
      mirroring=NoMirror
      resourceName=None
      resourceObject=null
      tiling=NoTile
      imageSize=<32 32>
      margin=<ltrb 0.5 0.5 0.5 0.5>
      outlineSettings {
       bUseBrushTransparency=false
       roundingType=FixedRadius
       width=0
       color {colorUseRule=UseColor_ColorTable, specifiedColor=<rgba 0 0 0 0>}
       cornerRadii=<4 4 4 4>
      }
      tintColor {colorUseRule=UseColor_Specified, specifiedColor=<rgba 0.1806 0.4583 0.1572 1>}
      uVRegion {bIsValid=false, max=<0 0>, min=<0 0>}
     }
     pressedForeground {colorUseRule=UseColor_ColorTable, specifiedColor=<rgba 0 0 0 0>}
     pressedPadding=<ltrb 12 2.5 12 0.5>
     pressedSlateSound {resourceObject=null}
    }}
   slot {class=HorizontalBoxSlot, properties {horizontalAlignment=HAlign_Center, verticalAlignment=VAlign_Center, padding=<ltrb 10 0 0 0>}}
  }
  Image_1 {
   widgetClassPath=&Image
   parent=AddSequenceDA
   widget=Image_1
   properties {brush {
     bIsDynamicallyLoaded=false
     drawAs=Image
     imageType=NoImage
     mirroring=NoMirror
     resourceName=None
     resourceObject=&MI_SparksEditorUI_PlusMinus_NoBorder
     tiling=NoTile
     imageSize=<20 20>
     margin=<ltrb 0 0 0 0>
     outlineSettings {
      bUseBrushTransparency=false
      roundingType=HalfHeightRadius
      width=0
      color {colorUseRule=UseColor_Specified, specifiedColor=<rgba 0 0 0 0>}
      cornerRadii=<0 0 0 0>
     }
     tintColor {colorUseRule=UseColor_Specified, specifiedColor=<rgba 1 1 1 1>}
     uVRegion {bIsValid=false, max=<0 0>, min=<0 0>}
    }}
   slot {class=ButtonSlot, properties {padding=<ltrb -5 0 -5 0>}}
  }
  DeleteSequenceDA {
   widgetClassPath=&EditorUtilityButton
   parent=HorizontalBox
   widget=DeleteSequenceDA
   properties {widgetStyle {
     clickedSlateSound {resourceObject=null}
     disabled {
      bIsDynamicallyLoaded=false
      drawAs=RoundedBox
      imageType=NoImage
      mirroring=NoMirror
      resourceName=None
      resourceObject=null
      tiling=NoTile
      imageSize=<20 20>
      margin=<ltrb 0 0 0 0>
      outlineSettings {
       bUseBrushTransparency=false
       roundingType=FixedRadius
       width=0
       color {colorUseRule=UseColor_ColorTable, specifiedColor=<rgba 0 0 0 0>}
       cornerRadii=<4 4 4 4>
      }
      tintColor {colorUseRule=UseColor_Specified, specifiedColor=<rgba 0.0395 0.0395 0.0395 1>}
      uVRegion {bIsValid=false, max=<0 0>, min=<0 0>}
     }
     disabledForeground {colorUseRule=UseColor_ColorTable, specifiedColor=<rgba 0 0 0 0>}
     hovered {
      bIsDynamicallyLoaded=false
      drawAs=RoundedBox
      imageType=NoImage
      mirroring=NoMirror
      resourceName=None
      resourceObject=null
      tiling=NoTile
      imageSize=<32 32>
      margin=<ltrb 0 0 0 0>
      outlineSettings {
       bUseBrushTransparency=false
       roundingType=FixedRadius
       width=0
       color {colorUseRule=UseColor_ColorTable, specifiedColor=<rgba 0 0 0 0>}
       cornerRadii=<4 4 4 4>
      }
      tintColor {colorUseRule=UseColor_Specified, specifiedColor=<rgba 1 0.1252 0.253 1>}
      uVRegion {bIsValid=false, max=<0 0>, min=<0 0>}
     }
     hoveredForeground {colorUseRule=UseColor_ColorTable, specifiedColor=<rgba 0 0 0 0>}
     hoveredSlateSound {resourceObject=null}
     normal {
      bIsDynamicallyLoaded=false
      drawAs=RoundedBox
      imageType=NoImage
      mirroring=NoMirror
      resourceName=None
      resourceObject=null
      tiling=NoTile
      imageSize=<20 20>
      margin=<ltrb 0 0 0 0>
      outlineSettings {
       bUseBrushTransparency=false
       roundingType=FixedRadius
       width=0
       color {colorUseRule=UseColor_ColorTable, specifiedColor=<rgba 0 0 0 0>}
       cornerRadii=<4 4 4 4>
      }
      tintColor {colorUseRule=UseColor_Specified, specifiedColor=<rgba 0.775 0 0.0304 1>}
      uVRegion {bIsValid=false, max=<0 0>, min=<0 0>}
     }
     normalForeground {colorUseRule=UseColor_ColorTable, specifiedColor=<rgba 0 0 0 0>}
     normalPadding=<ltrb 12 1.5 12 1.5>
     pressed {
      bIsDynamicallyLoaded=false
      drawAs=RoundedBox
      imageType=NoImage
      mirroring=NoMirror
      resourceName=None
      resourceObject=null
      tiling=NoTile
      imageSize=<20 20>
      margin=<ltrb 0 0 0 0>
      outlineSettings {
       bUseBrushTransparency=false
       roundingType=FixedRadius
       width=0
       color {colorUseRule=UseColor_ColorTable, specifiedColor=<rgba 0 0 0 0>}
       cornerRadii=<4 4 4 4>
      }
      tintColor {colorUseRule=UseColor_Specified, specifiedColor=<rgba 0.2396 0 0 1>}
      uVRegion {bIsValid=false, max=<0 0>, min=<0 0>}
     }
     pressedForeground {colorUseRule=UseColor_ColorTable, specifiedColor=<rgba 0 0 0 0>}
     pressedPadding=<ltrb 12 2.5 12 0.5>
     pressedSlateSound {resourceObject=null}
    }}
   slot {class=HorizontalBoxSlot, properties {horizontalAlignment=HAlign_Left, verticalAlignment=VAlign_Center, padding=<ltrb 10 0 0 0>}}
  }
  Image_2 {
   widgetClassPath=&Image
   parent=DeleteSequenceDA
   widget=Image_2
   properties {
    visibility=SelfHitTestInvisible
    brush {
     bIsDynamicallyLoaded=false
     drawAs=Image
     imageType=FullColor
     mirroring=NoMirror
     resourceName=None
     resourceObject=&T_SparksEditorUI_TrashCan_64
     tiling=NoTile
     imageSize=<20 20>
     margin=<ltrb 0 0 0 0>
     outlineSettings {
      bUseBrushTransparency=false
      roundingType=HalfHeightRadius
      width=0
      color {colorUseRule=UseColor_Specified, specifiedColor=<rgba 0 0 0 0>}
      cornerRadii=<0 0 0 0>
     }
     tintColor {colorUseRule=UseColor_Specified, specifiedColor=<rgba 1 1 1 1>}
     uVRegion {bIsValid=false, max=<0 0>, min=<0 0>}
    }
   }
   slot {class=ButtonSlot, properties {padding=<ltrb -5 0 -5 0>}}
  }
  OpenSequenceDAButton {
   widgetClassPath=&EditorUtilityButton
   parent=HorizontalBox
   widget=OpenSequenceDAButton
   properties {widgetStyle {
     clickedSlateSound {resourceObject=null}
     disabled {
      bIsDynamicallyLoaded=false
      drawAs=RoundedBox
      imageType=NoImage
      mirroring=NoMirror
      resourceName=None
      resourceObject=&T-Button-Square
      tiling=NoTile
      imageSize=<32 32>
      margin=<ltrb 0.5 0.5 0.5 0.5>
      outlineSettings {
       bUseBrushTransparency=false
       roundingType=FixedRadius
       width=0
       color {colorUseRule=UseColor_ColorTable, specifiedColor=<rgba 0 0 0 0>}
       cornerRadii=<4 4 4 4>
      }
      tintColor {colorUseRule=UseColor_Specified, specifiedColor=<rgba 0.0395 0.0395 0.0395 1>}
      uVRegion {bIsValid=false, max=<0 0>, min=<0 0>}
     }
     disabledForeground {colorUseRule=UseColor_ColorTable, specifiedColor=<rgba 0 0 0 0>}
     hovered {
      bIsDynamicallyLoaded=false
      drawAs=RoundedBox
      imageType=NoImage
      mirroring=NoMirror
      resourceName=None
      resourceObject=&T-Button-Square
      tiling=NoTile
      imageSize=<32 32>
      margin=<ltrb 0.5 0.5 0.5 0.5>
      outlineSettings {
       bUseBrushTransparency=false
       roundingType=FixedRadius
       width=0
       color {colorUseRule=UseColor_ColorTable, specifiedColor=<rgba 0 0 0 0>}
       cornerRadii=<4 4 4 4>
      }
      tintColor {colorUseRule=UseColor_Specified, specifiedColor=<rgba 0 0.3021 0.0241 1>}
      uVRegion {bIsValid=false, max=<0 0>, min=<0 0>}
     }
     hoveredForeground {colorUseRule=UseColor_ColorTable, specifiedColor=<rgba 0 0 0 0>}
     hoveredSlateSound {resourceObject=null}
     normal {
      bIsDynamicallyLoaded=false
      drawAs=RoundedBox
      imageType=NoImage
      mirroring=NoMirror
      resourceName=None
      resourceObject=&T-Button-Square
      tiling=NoTile
      imageSize=<32 32>
      margin=<ltrb 0.5 0.5 0.5 0.5>
      outlineSettings {
       bUseBrushTransparency=false
       roundingType=FixedRadius
       width=0
       color {colorUseRule=UseColor_ColorTable, specifiedColor=<rgba 0 0 0 0>}
       cornerRadii=<4 4 4 4>
      }
      tintColor {colorUseRule=UseColor_Specified, specifiedColor=<rgba 0.0071 0.7594 0 1>}
      uVRegion {bIsValid=false, max=<0 0>, min=<0 0>}
     }
     normalForeground {colorUseRule=UseColor_ColorTable, specifiedColor=<rgba 0 0 0 0>}
     normalPadding=<ltrb 12 1.5 12 1.5>
     pressed {
      bIsDynamicallyLoaded=false
      drawAs=RoundedBox
      imageType=NoImage
      mirroring=NoMirror
      resourceName=None
      resourceObject=&T-Button-Square
      tiling=NoTile
      imageSize=<32 32>
      margin=<ltrb 0.5 0.5 0.5 0.5>
      outlineSettings {
       bUseBrushTransparency=false
       roundingType=FixedRadius
       width=0
       color {colorUseRule=UseColor_ColorTable, specifiedColor=<rgba 0 0 0 0>}
       cornerRadii=<4 4 4 4>
      }
      tintColor {colorUseRule=UseColor_Specified, specifiedColor=<rgba 0.3119 0.7917 0.2715 1>}
      uVRegion {bIsValid=false, max=<0 0>, min=<0 0>}
     }
     pressedForeground {colorUseRule=UseColor_ColorTable, specifiedColor=<rgba 0 0 0 0>}
     pressedPadding=<ltrb 12 2.5 12 0.5>
     pressedSlateSound {resourceObject=null}
    }}
   slot {class=HorizontalBoxSlot, properties {verticalAlignment=VAlign_Center, padding=<ltrb 5 5 5 5>}}
  }
  ExternalIcon_1 {
   widgetClassPath=&Image
   parent=OpenSequenceDAButton
   widget=ExternalIcon_1
   properties {brush {
     bIsDynamicallyLoaded=false
     drawAs=Image
     imageType=NoImage
     mirroring=NoMirror
     resourceName=None
     resourceObject=&M_UI_Icon_ExternalLink
     tiling=NoTile
     imageSize=<20 20>
     margin=<ltrb 0 0 0 0>
     outlineSettings {
      bUseBrushTransparency=false
      roundingType=HalfHeightRadius
      width=0
      color {colorUseRule=UseColor_Specified, specifiedColor=<rgba 0 0 0 0>}
      cornerRadii=<0 0 0 0>
     }
     tintColor {colorUseRule=UseColor_Specified, specifiedColor=<rgba 1 1 1 1>}
     uVRegion {bIsValid=false, max=<0 0>, min=<0 0>}
    }}
   slot {class=ButtonSlot, properties {padding=<ltrb 0 0 0 0>}}
  }
  AddNewSequenceButton {
   widgetClassPath=&EditorUtilityButton
   parent=HorizontalBox_2
   widget=AddNewSequenceButton
   properties {widgetStyle {
     clickedSlateSound {resourceObject=null}
     disabled {
      bIsDynamicallyLoaded=false
      drawAs=RoundedBox
      imageType=NoImage
      mirroring=NoMirror
      resourceName=None
      resourceObject=null
      tiling=NoTile
      imageSize=<32 32>
      margin=<ltrb 0.5 0.5 0.5 0.5>
      outlineSettings {
       bUseBrushTransparency=false
       roundingType=FixedRadius
       width=0
       color {colorUseRule=UseColor_ColorTable, specifiedColor=<rgba 0 0 0 0>}
       cornerRadii=<4 4 4 4>
      }
      tintColor {colorUseRule=UseColor_Specified, specifiedColor=<rgba 0.0395 0.0395 0.0395 1>}
      uVRegion {bIsValid=false, max=<0 0>, min=<0 0>}
     }
     disabledForeground {colorUseRule=UseColor_ColorTable, specifiedColor=<rgba 0 0 0 0>}
     hovered {
      bIsDynamicallyLoaded=false
      drawAs=RoundedBox
      imageType=NoImage
      mirroring=NoMirror
      resourceName=None
      resourceObject=null
      tiling=NoTile
      imageSize=<32 32>
      margin=<ltrb 0 0 0 0>
      outlineSettings {
       bUseBrushTransparency=false
       roundingType=FixedRadius
       width=0
       color {colorUseRule=UseColor_ColorTable, specifiedColor=<rgba 0 0 0 0>}
       cornerRadii=<4 4 4 4>
      }
      tintColor {colorUseRule=UseColor_Specified, specifiedColor=<rgba 0.1126 0.7031 0 1>}
      uVRegion {bIsValid=false, max=<0 0>, min=<0 0>}
     }
     hoveredForeground {colorUseRule=UseColor_ColorTable, specifiedColor=<rgba 0 0 0 0>}
     hoveredSlateSound {resourceObject=null}
     normal {
      bIsDynamicallyLoaded=false
      drawAs=RoundedBox
      imageType=NoImage
      mirroring=NoMirror
      resourceName=None
      resourceObject=null
      tiling=NoTile
      imageSize=<32 32>
      margin=<ltrb 0.5 0.5 0.5 0.5>
      outlineSettings {
       bUseBrushTransparency=false
       roundingType=FixedRadius
       width=0
       color {colorUseRule=UseColor_ColorTable, specifiedColor=<rgba 0 0 0 0>}
       cornerRadii=<4 4 4 4>
      }
      tintColor {colorUseRule=UseColor_Specified, specifiedColor=<rgba 0.0771 0.6167 0.1289 1>}
      uVRegion {bIsValid=false, max=<0 0>, min=<0 0>}
     }
     normalForeground {colorUseRule=UseColor_ColorTable, specifiedColor=<rgba 0 0 0 0>}
     normalPadding=<ltrb 12 1.5 12 1.5>
     pressed {
      bIsDynamicallyLoaded=false
      drawAs=RoundedBox
      imageType=NoImage
      mirroring=NoMirror
      resourceName=None
      resourceObject=null
      tiling=NoTile
      imageSize=<32 32>
      margin=<ltrb 0.5 0.5 0.5 0.5>
      outlineSettings {
       bUseBrushTransparency=false
       roundingType=FixedRadius
       width=0
       color {colorUseRule=UseColor_ColorTable, specifiedColor=<rgba 0 0 0 0>}
       cornerRadii=<4 4 4 4>
      }
      tintColor {colorUseRule=UseColor_Specified, specifiedColor=<rgba 0.1806 0.4583 0.1572 1>}
      uVRegion {bIsValid=false, max=<0 0>, min=<0 0>}
     }
     pressedForeground {colorUseRule=UseColor_ColorTable, specifiedColor=<rgba 0 0 0 0>}
     pressedPadding=<ltrb 12 2.5 12 0.5>
     pressedSlateSound {resourceObject=null}
    }}
   slot {class=HorizontalBoxSlot, properties {verticalAlignment=VAlign_Center, padding=<ltrb 5 0 0 0>}}
  }
  HorizontalBox_4 {
   widgetClassPath=&HorizontalBox
   parent=AddNewSequenceButton
   widget=HorizontalBox_4
   slot {class=ButtonSlot, properties {horizontalAlignment=HAlign_Left}}
  }
  Image_3 {
   widgetClassPath=&Image
   parent=HorizontalBox_4
   widget=Image_3
   properties {brush {
     bIsDynamicallyLoaded=false
     drawAs=Image
     imageType=NoImage
     mirroring=NoMirror
     resourceName=None
     resourceObject=&MI_SparksEditorUI_PlusMinus_NoBorder
     tiling=NoTile
     imageSize=<15 15>
     margin=<ltrb 0 0 0 0>
     outlineSettings {
      bUseBrushTransparency=false
      roundingType=HalfHeightRadius
      width=0
      color {colorUseRule=UseColor_Specified, specifiedColor=<rgba 0 0 0 0>}
      cornerRadii=<0 0 0 0>
     }
     tintColor {colorUseRule=UseColor_Specified, specifiedColor=<rgba 1 1 1 1>}
     uVRegion {bIsValid=false, max=<0 0>, min=<0 0>}
    }}
   slot {class=HorizontalBoxSlot, properties {horizontalAlignment=HAlign_Center, verticalAlignment=VAlign_Center, padding=<ltrb 0 0 3 0>}}
  }
  TextBlock_25 {
   widgetClassPath=&TextBlock
   parent=HorizontalBox_4
   widget=TextBlock_25
   properties {
    text=Add new sequence
    font {
     bForceMonospaced=false
     bMaterialIsStencil=false
     fontMaterial=null
     fontObject=&HeadingNow
     letterSpacing=0
     monospacedWidth=1
     size=7.5
     skewAmount=0
     typefaceFontName=75 Medium
     outlineSettings {
      bApplyOutlineToDropShadows=false
      bMiteredCorners=false
      bSeparateFillAlpha=false
      outlineMaterial=null
      outlineSize=0
      outlineColor=<rgba 0 0 0 1>
     }
    }
   }
   slot {class=HorizontalBoxSlot, properties {verticalAlignment=VAlign_Center}}
  }
  HorizontalBox_3 {
   widgetClassPath=&HorizontalBox
   parent=MainContent
   widget=HorizontalBox_3
   properties {visibility=Collapsed}
   slot {class=VerticalBoxSlot, properties {verticalAlignment=VAlign_Center, padding=<ltrb 0 5 5 5>}}
  }
  EsquenceName {
   widgetClassPath=&TextBlock
   parent=HorizontalBox_3
   widget=EsquenceName
   properties {
    text="Theme Name: "
    colorAndOpacity {colorUseRule=UseColor_Specified, specifiedColor=<rgba 0.6562 0.6562 0.6562 1>}
    font {
     bForceMonospaced=false
     bMaterialIsStencil=false
     fontMaterial=null
     fontObject=&HeadingNow
     letterSpacing=0
     monospacedWidth=1
     size=7.5
     skewAmount=0
     typefaceFontName=75 Medium
     outlineSettings {
      bApplyOutlineToDropShadows=false
      bMiteredCorners=false
      bSeparateFillAlpha=false
      outlineMaterial=null
      outlineSize=0
      outlineColor=<rgba 0 0 0 1>
     }
    }
   }
   slot {class=HorizontalBoxSlot, properties {verticalAlignment=VAlign_Center}}
  }
  EventGroupTextBox {
   widgetClassPath=&EditorUtilityEditableTextBox
   bIsVariable=true
   parent=HorizontalBox_3
   widget=EventGroupTextBox
   properties {
    hintText=Sequence Name
    widgetStyle {
     backgroundColor {colorUseRule=UseColor_Specified, specifiedColor=<rgba 1 1 1 1>}
     backgroundImageFocused {
      bIsDynamicallyLoaded=false
      drawAs=RoundedBox
      imageType=NoImage
      mirroring=NoMirror
      resourceName=None
      resourceObject=null
      tiling=NoTile
      imageSize=<0 0>
      margin=<ltrb 0 0 0 0>
      outlineSettings {
       bUseBrushTransparency=false
       roundingType=FixedRadius
       width=1
       color {colorUseRule=UseColor_ColorTable, specifiedColor=<rgba 0 0 0 0>}
       cornerRadii=<4 4 4 4>
      }
      tintColor {colorUseRule=UseColor_Specified, specifiedColor=<rgba 0.0048 0.0048 0.0048 1>}
      uVRegion {bIsValid=false, max=<0 0>, min=<0 0>}
     }
     backgroundImageHovered {
      bIsDynamicallyLoaded=false
      drawAs=RoundedBox
      imageType=NoImage
      mirroring=NoMirror
      resourceName=None
      resourceObject=null
      tiling=NoTile
      imageSize=<0 0>
      margin=<ltrb 0 0 0 0>
      outlineSettings {
       bUseBrushTransparency=false
       roundingType=FixedRadius
       width=1
       color {colorUseRule=UseColor_ColorTable, specifiedColor=<rgba 0 0 0 0>}
       cornerRadii=<4 4 4 4>
      }
      tintColor {colorUseRule=UseColor_Specified, specifiedColor=<rgba 0.0048 0.0048 0.0048 1>}
      uVRegion {bIsValid=false, max=<0 0>, min=<0 0>}
     }
     backgroundImageNormal {
      bIsDynamicallyLoaded=false
      drawAs=RoundedBox
      imageType=NoImage
      mirroring=NoMirror
      resourceName=None
      resourceObject=null
      tiling=NoTile
      imageSize=<0 0>
      margin=<ltrb 0 0 0 0>
      outlineSettings {
       bUseBrushTransparency=false
       roundingType=FixedRadius
       width=1
       color {colorUseRule=UseColor_ColorTable, specifiedColor=<rgba 0 0 0 0>}
       cornerRadii=<4 4 4 4>
      }
      tintColor {colorUseRule=UseColor_Specified, specifiedColor=<rgba 0.0048 0.0048 0.0048 1>}
      uVRegion {bIsValid=false, max=<0 0>, min=<0 0>}
     }
     backgroundImageReadOnly {
      bIsDynamicallyLoaded=false
      drawAs=RoundedBox
      imageType=NoImage
      mirroring=NoMirror
      resourceName=None
      resourceObject=null
      tiling=NoTile
      imageSize=<0 0>
      margin=<ltrb 0 0 0 0>
      outlineSettings {
       bUseBrushTransparency=false
       roundingType=FixedRadius
       width=1
       color {colorUseRule=UseColor_ColorTable, specifiedColor=<rgba 0 0 0 0>}
       cornerRadii=<4 4 4 4>
      }
      tintColor {colorUseRule=UseColor_Specified, specifiedColor=<rgba 0.0284 0.0284 0.0284 1>}
      uVRegion {bIsValid=false, max=<0 0>, min=<0 0>}
     }
     focusedForegroundColor {colorUseRule=UseColor_Specified, specifiedColor=<rgba 1 1 1 1>}
     foregroundColor {colorUseRule=UseColor_Specified, specifiedColor=<rgba 0.5271 0.5271 0.5271 1>}
     hScrollBarPadding=<ltrb 0 0 0 0>
     padding=<ltrb 8 3 8 4>
     readOnlyForegroundColor {colorUseRule=UseColor_Foreground, specifiedColor=<rgba 1 0 1 1>}
     scrollBarStyle {
      thickness=8
      draggedThumbImage {
       bIsDynamicallyLoaded=false
       drawAs=RoundedBox
       imageType=NoImage
       mirroring=NoMirror
       resourceName=None
       resourceObject=null
       tiling=NoTile
       imageSize=<0 0>
       margin=<ltrb 0 0 0 0>
       outlineSettings {
        bUseBrushTransparency=false
        roundingType=FixedRadius
        width=0
        color {colorUseRule=UseColor_Specified, specifiedColor=<rgba 0 0 0 0>}
        cornerRadii=<4 4 4 4>
       }
       tintColor {colorUseRule=UseColor_Specified, specifiedColor=<rgba 0.2159 0.2159 0.2159 1>}
       uVRegion {bIsValid=false, max=<0 0>, min=<0 0>}
      }
      horizontalBackgroundImage {
       bIsDynamicallyLoaded=false
       drawAs=NoDrawType
       imageType=NoImage
       mirroring=NoMirror
       resourceName=None
       resourceObject=null
       tiling=NoTile
       imageSize=<0 0>
       margin=<ltrb 0 0 0 0>
       outlineSettings {
        bUseBrushTransparency=false
        roundingType=HalfHeightRadius
        width=0
        color {colorUseRule=UseColor_Specified, specifiedColor=<rgba 0 0 0 0>}
        cornerRadii=<0 0 0 0>
       }
       tintColor {colorUseRule=UseColor_Specified, specifiedColor=<rgba 1 1 1 1>}
       uVRegion {bIsValid=false, max=<0 0>, min=<0 0>}
      }
      horizontalBottomSlotImage {
       bIsDynamicallyLoaded=false
       drawAs=NoDrawType
       imageType=NoImage
       mirroring=NoMirror
       resourceName=None
       resourceObject=null
       tiling=NoTile
       imageSize=<0 0>
       margin=<ltrb 0 0 0 0>
       outlineSettings {
        bUseBrushTransparency=false
        roundingType=HalfHeightRadius
        width=0
        color {colorUseRule=UseColor_Specified, specifiedColor=<rgba 0 0 0 0>}
        cornerRadii=<0 0 0 0>
       }
       tintColor {colorUseRule=UseColor_Specified, specifiedColor=<rgba 1 1 1 1>}
       uVRegion {bIsValid=false, max=<0 0>, min=<0 0>}
      }
      horizontalTopSlotImage {
       bIsDynamicallyLoaded=false
       drawAs=NoDrawType
       imageType=NoImage
       mirroring=NoMirror
       resourceName=None
       resourceObject=null
       tiling=NoTile
       imageSize=<0 0>
       margin=<ltrb 0 0 0 0>
       outlineSettings {
        bUseBrushTransparency=false
        roundingType=HalfHeightRadius
        width=0
        color {colorUseRule=UseColor_Specified, specifiedColor=<rgba 0 0 0 0>}
        cornerRadii=<0 0 0 0>
       }
       tintColor {colorUseRule=UseColor_Specified, specifiedColor=<rgba 1 1 1 1>}
       uVRegion {bIsValid=false, max=<0 0>, min=<0 0>}
      }
      hoveredThumbImage {
       bIsDynamicallyLoaded=false
       drawAs=RoundedBox
       imageType=NoImage
       mirroring=NoMirror
       resourceName=None
       resourceObject=null
       tiling=NoTile
       imageSize=<0 0>
       margin=<ltrb 0 0 0 0>
       outlineSettings {
        bUseBrushTransparency=false
        roundingType=FixedRadius
        width=0
        color {colorUseRule=UseColor_Specified, specifiedColor=<rgba 0 0 0 0>}
        cornerRadii=<4 4 4 4>
       }
       tintColor {colorUseRule=UseColor_Specified, specifiedColor=<rgba 0.2159 0.2159 0.2159 1>}
       uVRegion {bIsValid=false, max=<0 0>, min=<0 0>}
      }
      normalThumbImage {
       bIsDynamicallyLoaded=false
       drawAs=RoundedBox
       imageType=NoImage
       mirroring=NoMirror
       resourceName=None
       resourceObject=null
       tiling=NoTile
       imageSize=<0 0>
       margin=<ltrb 0 0 0 0>
       outlineSettings {
        bUseBrushTransparency=false
        roundingType=FixedRadius
        width=0
        color {colorUseRule=UseColor_Specified, specifiedColor=<rgba 0 0 0 0>}
        cornerRadii=<4 4 4 4>
       }
       tintColor {colorUseRule=UseColor_Specified, specifiedColor=<rgba 0.0953 0.0953 0.0953 1>}
       uVRegion {bIsValid=false, max=<0 0>, min=<0 0>}
      }
      verticalBackgroundImage {
       bIsDynamicallyLoaded=false
       drawAs=NoDrawType
       imageType=NoImage
       mirroring=NoMirror
       resourceName=None
       resourceObject=null
       tiling=NoTile
       imageSize=<0 0>
       margin=<ltrb 0 0 0 0>
       outlineSettings {
        bUseBrushTransparency=false
        roundingType=HalfHeightRadius
        width=0
        color {colorUseRule=UseColor_Specified, specifiedColor=<rgba 0 0 0 0>}
        cornerRadii=<0 0 0 0>
       }
       tintColor {colorUseRule=UseColor_Specified, specifiedColor=<rgba 1 1 1 1>}
       uVRegion {bIsValid=false, max=<0 0>, min=<0 0>}
      }
      verticalBottomSlotImage {
       bIsDynamicallyLoaded=false
       drawAs=NoDrawType
       imageType=NoImage
       mirroring=NoMirror
       resourceName=None
       resourceObject=null
       tiling=NoTile
       imageSize=<0 0>
       margin=<ltrb 0 0 0 0>
       outlineSettings {
        bUseBrushTransparency=false
        roundingType=HalfHeightRadius
        width=0
        color {colorUseRule=UseColor_Specified, specifiedColor=<rgba 0 0 0 0>}
        cornerRadii=<0 0 0 0>
       }
       tintColor {colorUseRule=UseColor_Specified, specifiedColor=<rgba 1 1 1 1>}
       uVRegion {bIsValid=false, max=<0 0>, min=<0 0>}
      }
      verticalTopSlotImage {
       bIsDynamicallyLoaded=false
       drawAs=NoDrawType
       imageType=NoImage
       mirroring=NoMirror
       resourceName=None
       resourceObject=null
       tiling=NoTile
       imageSize=<0 0>
       margin=<ltrb 0 0 0 0>
       outlineSettings {
        bUseBrushTransparency=false
        roundingType=HalfHeightRadius
        width=0
        color {colorUseRule=UseColor_Specified, specifiedColor=<rgba 0 0 0 0>}
        cornerRadii=<0 0 0 0>
       }
       tintColor {colorUseRule=UseColor_Specified, specifiedColor=<rgba 1 1 1 1>}
       uVRegion {bIsValid=false, max=<0 0>, min=<0 0>}
      }
     }
     textStyle {
      overflowPolicy=Clip
      transformPolicy=None
      colorAndOpacity {colorUseRule=UseColor_Foreground, specifiedColor=<rgba 1 0 1 1>}
      font {
       bForceMonospaced=false
       bMaterialIsStencil=false
       fontMaterial=null
       fontObject=&HeadingNow
       letterSpacing=0
       monospacedWidth=1
       size=7.5
       skewAmount=0
       typefaceFontName=75 Medium
       outlineSettings {
        bApplyOutlineToDropShadows=false
        bMiteredCorners=false
        bSeparateFillAlpha=false
        outlineMaterial=null
        outlineSize=0
        outlineColor=<rgba 0 0 0 1>
       }
      }
      highlightColor {colorUseRule=UseColor_ColorTable, specifiedColor=<rgba 0 0 0 0>}
      highlightShape {
       bIsDynamicallyLoaded=false
       drawAs=Image
       imageType=NoImage
       mirroring=NoMirror
       resourceName=None
       resourceObject=null
       tiling=NoTile
       imageSize=<0 0>
       margin=<ltrb 0 0 0 0>
       outlineSettings {
        bUseBrushTransparency=false
        roundingType=HalfHeightRadius
        width=0
        color {colorUseRule=UseColor_Specified, specifiedColor=<rgba 0 0 0 0>}
        cornerRadii=<0 0 0 0>
       }
       tintColor {colorUseRule=UseColor_ColorTable, specifiedColor=<rgba 0 0 0 0>}
       uVRegion {bIsValid=false, max=<0 0>, min=<0 0>}
      }
      selectedBackgroundColor {colorUseRule=UseColor_ColorTable, specifiedColor=<rgba 0 0 0 0>}
      shadowColorAndOpacity=<rgba 0 0 0 1>
      shadowOffset=<0 0>
      strikeBrush {
       bIsDynamicallyLoaded=false
       drawAs=Image
       imageType=NoImage
       mirroring=NoMirror
       resourceName=None
       resourceObject=null
       tiling=NoTile
       imageSize=<32 32>
       margin=<ltrb 0 0 0 0>
       outlineSettings {
        bUseBrushTransparency=false
        roundingType=HalfHeightRadius
        width=0
        color {colorUseRule=UseColor_Specified, specifiedColor=<rgba 0 0 0 0>}
        cornerRadii=<0 0 0 0>
       }
       tintColor {colorUseRule=UseColor_Specified, specifiedColor=<rgba 1 1 1 1>}
       uVRegion {bIsValid=false, max=<0 0>, min=<0 0>}
      }
      underlineBrush {
       bIsDynamicallyLoaded=false
       drawAs=Image
       imageType=NoImage
       mirroring=NoMirror
       resourceName=None
       resourceObject=null
       tiling=NoTile
       imageSize=<32 32>
       margin=<ltrb 0 0 0 0>
       outlineSettings {
        bUseBrushTransparency=false
        roundingType=HalfHeightRadius
        width=0
        color {colorUseRule=UseColor_Specified, specifiedColor=<rgba 0 0 0 0>}
        cornerRadii=<0 0 0 0>
       }
       tintColor {colorUseRule=UseColor_Specified, specifiedColor=<rgba 1 1 1 1>}
       uVRegion {bIsValid=false, max=<0 0>, min=<0 0>}
      }
     }
     vScrollBarPadding=<ltrb 0 0 0 0>
    }
   }
   slot {class=HorizontalBoxSlot, properties {horizontalAlignment=HAlign_Left, verticalAlignment=VAlign_Center}}
  }
  UpdateNameButton {
   widgetClassPath=&EditorUtilityButton
   parent=HorizontalBox_3
   widget=UpdateNameButton
   properties {widgetStyle {
     clickedSlateSound {resourceObject=null}
     disabled {
      bIsDynamicallyLoaded=false
      drawAs=RoundedBox
      imageType=NoImage
      mirroring=NoMirror
      resourceName=None
      resourceObject=null
      tiling=NoTile
      imageSize=<0 0>
      margin=<ltrb 0 0 0 0>
      outlineSettings {
       bUseBrushTransparency=false
       roundingType=FixedRadius
       width=0
       color {colorUseRule=UseColor_ColorTable, specifiedColor=<rgba 0 0 0 0>}
       cornerRadii=<4 4 4 4>
      }
      tintColor {colorUseRule=UseColor_Specified, specifiedColor=<rgba 0.0395 0.0395 0.0395 1>}
      uVRegion {bIsValid=false, max=<0 0>, min=<0 0>}
     }
     disabledForeground {colorUseRule=UseColor_ColorTable, specifiedColor=<rgba 0 0 0 0>}
     hovered {
      bIsDynamicallyLoaded=false
      drawAs=RoundedBox
      imageType=NoImage
      mirroring=NoMirror
      resourceName=None
      resourceObject=null
      tiling=NoTile
      imageSize=<32 32>
      margin=<ltrb 0 0 0 0>
      outlineSettings {
       bUseBrushTransparency=false
       roundingType=FixedRadius
       width=0
       color {colorUseRule=UseColor_ColorTable, specifiedColor=<rgba 0 0 0 0>}
       cornerRadii=<4 4 4 4>
      }
      tintColor {colorUseRule=UseColor_Specified, specifiedColor=<rgba 0.0953 0.0953 0.0953 1>}
      uVRegion {bIsValid=false, max=<0 0>, min=<0 0>}
     }
     hoveredForeground {colorUseRule=UseColor_ColorTable, specifiedColor=<rgba 0 0 0 0>}
     hoveredSlateSound {resourceObject=null}
     normal {
      bIsDynamicallyLoaded=false
      drawAs=RoundedBox
      imageType=NoImage
      mirroring=NoMirror
      resourceName=None
      resourceObject=null
      tiling=NoTile
      imageSize=<32 32>
      margin=<ltrb 0 0 0 0>
      outlineSettings {
       bUseBrushTransparency=false
       roundingType=FixedRadius
       width=0
       color {colorUseRule=UseColor_ColorTable, specifiedColor=<rgba 0 0 0 0>}
       cornerRadii=<4 4 4 4>
      }
      tintColor {colorUseRule=UseColor_Specified, specifiedColor=<rgba 0.3385 0.3385 0.3385 1>}
      uVRegion {bIsValid=false, max=<0 0>, min=<0 0>}
     }
     normalForeground {colorUseRule=UseColor_ColorTable, specifiedColor=<rgba 0 0 0 0>}
     normalPadding=<ltrb 12 1.5 12 1.5>
     pressed {
      bIsDynamicallyLoaded=false
      drawAs=RoundedBox
      imageType=NoImage
      mirroring=NoMirror
      resourceName=None
      resourceObject=null
      tiling=NoTile
      imageSize=<32 32>
      margin=<ltrb 0 0 0 0>
      outlineSettings {
       bUseBrushTransparency=false
       roundingType=FixedRadius
       width=0
       color {colorUseRule=UseColor_ColorTable, specifiedColor=<rgba 0 0 0 0>}
       cornerRadii=<4 4 4 4>
      }
      tintColor {colorUseRule=UseColor_Specified, specifiedColor=<rgba 0.0284 0.0284 0.0284 1>}
      uVRegion {bIsValid=false, max=<0 0>, min=<0 0>}
     }
     pressedForeground {colorUseRule=UseColor_ColorTable, specifiedColor=<rgba 0 0 0 0>}
     pressedPadding=<ltrb 12 2.5 12 0.5>
     pressedSlateSound {resourceObject=null}
    }}
   slot {class=HorizontalBoxSlot, properties {padding=<ltrb 10 0 0 0>}}
  }
  TextBlock_58 {
   widgetClassPath=&TextBlock
   parent=UpdateNameButton
   widget=TextBlock_58
   properties {
    text=Update Name
    font {
     bForceMonospaced=false
     bMaterialIsStencil=false
     fontMaterial=null
     fontObject=&HeadingNow
     letterSpacing=0
     monospacedWidth=1
     size=8.25
     skewAmount=0
     typefaceFontName=75 Medium
     outlineSettings {
      bApplyOutlineToDropShadows=false
      bMiteredCorners=false
      bSeparateFillAlpha=false
      outlineMaterial=null
      outlineSize=0
      outlineColor=<rgba 0 0 0 1>
     }
    }
   }
   slot {class=ButtonSlot}
  }
  SequencesScrollBoxGroup {
   widgetClassPath=&EditorUtilityScrollBox
   bIsVariable=true
   parent=MainContent
   widget=SequencesScrollBoxGroup
   properties {consumeMouseWheel=Never}
   slot {class=VerticalBoxSlot, properties {padding=<ltrb 0 0 0 50>, size {sizeRule=Fill, value=1}}}
  }
  EditorUtilityScrollBox_105 {
   widgetClassPath=&EditorUtilityScrollBox
   parent=SequencesScrollBoxGroup
   widget=EditorUtilityScrollBox_105
   properties {clipping=Inherit, orientation=Orient_Horizontal}
   slot {class=ScrollBoxSlot, properties {padding=<ltrb 0 0 0 40>, size {sizeRule=Fill, value=1}}}
  }
  SequencesVerticalBox {
   widgetClassPath=&VerticalBox
   bIsVariable=true
   parent=EditorUtilityScrollBox_105
   widget=SequencesVerticalBox
   slot {class=ScrollBoxSlot, properties {size {sizeRule=Fill, value=1}}}
  }
  EWBP_SparksSequencePoolTimelineCell {
   widgetClassPath=&/EWBP_SparksSequencePoolTimelineCell
   parent=SequencesVerticalBox
   widget=EWBP_SparksSequencePoolTimelineCell
   slot {class=VerticalBoxSlot}
  }
  EmptySequencePoolGroup {
   widgetClassPath=&Overlay
   bIsVariable=true
   parent=MainContent
   widget=EmptySequencePoolGroup
   properties {visibility=Collapsed}
   slot {class=VerticalBoxSlot, properties {padding=<ltrb 0 10 0 0>, size {sizeRule=Fill, value=1}}}
  }
  Image_99 {
   widgetClassPath=&Image
   parent=EmptySequencePoolGroup
   widget=Image_99
   properties {brush {
     bIsDynamicallyLoaded=false
     drawAs=Image
     imageType=NoImage
     mirroring=NoMirror
     resourceName=None
     resourceObject=null
     tiling=NoTile
     imageSize=<32 32>
     margin=<ltrb 0 0 0 0>
     outlineSettings {
      bUseBrushTransparency=false
      roundingType=HalfHeightRadius
      width=0
      color {colorUseRule=UseColor_Specified, specifiedColor=<rgba 0 0 0 0>}
      cornerRadii=<0 0 0 0>
     }
     tintColor {colorUseRule=UseColor_Specified, specifiedColor=<rgba 0.101 0.101 0.101 1>}
     uVRegion {bIsValid=false, max=<0 0>, min=<0 0>}
    }}
   slot {class=OverlaySlot, properties {horizontalAlignment=HAlign_Fill, verticalAlignment=VAlign_Fill}}
  }
  NoStepsText {
   widgetClassPath=&TextBlock
   bIsVariable=true
   parent=EmptySequencePoolGroup
   widget=NoStepsText
   properties {
    text=There is no Valid sequence pool selected.
    font {
     bForceMonospaced=false
     bMaterialIsStencil=false
     fontMaterial=null
     fontObject=&HeadingNow
     letterSpacing=0
     monospacedWidth=1
     size=7.5
     skewAmount=0
     typefaceFontName=75 Medium
     outlineSettings {
      bApplyOutlineToDropShadows=false
      bMiteredCorners=false
      bSeparateFillAlpha=false
      outlineMaterial=null
      outlineSize=0
      outlineColor=<rgba 0 0 0 1>
     }
    }
   }
   slot {class=OverlaySlot, properties {horizontalAlignment=HAlign_Fill, verticalAlignment=VAlign_Center, padding=<ltrb 10 0 10 0>}}
  }
  EmptySequenceGroup {
   widgetClassPath=&Overlay
   bIsVariable=true
   parent=MainContent
   widget=EmptySequenceGroup
   properties {visibility=Collapsed}
   slot {class=VerticalBoxSlot, properties {padding=<ltrb 0 10 0 0>, size {sizeRule=Fill, value=1}}}
  }
  Image_18 {
   widgetClassPath=&Image
   parent=EmptySequenceGroup
   widget=Image_18
   properties {brush {
     bIsDynamicallyLoaded=false
     drawAs=Image
     imageType=NoImage
     mirroring=NoMirror
     resourceName=None
     resourceObject=null
     tiling=NoTile
     imageSize=<32 32>
     margin=<ltrb 0 0 0 0>
     outlineSettings {
      bUseBrushTransparency=false
      roundingType=HalfHeightRadius
      width=0
      color {colorUseRule=UseColor_Specified, specifiedColor=<rgba 0 0 0 0>}
      cornerRadii=<0 0 0 0>
     }
     tintColor {colorUseRule=UseColor_Specified, specifiedColor=<rgba 0.101 0.101 0.101 1>}
     uVRegion {bIsValid=false, max=<0 0>, min=<0 0>}
    }}
   slot {class=OverlaySlot, properties {horizontalAlignment=HAlign_Fill, verticalAlignment=VAlign_Fill}}
  }
  NoStepsText_1 {
   widgetClassPath=&TextBlock
   bIsVariable=true
   parent=EmptySequenceGroup
   widget=NoStepsText_1
   properties {
    text=There is no sequence in this pool.
    font {
     bForceMonospaced=false
     bMaterialIsStencil=false
     fontMaterial=null
     fontObject=&HeadingNow
     letterSpacing=0
     monospacedWidth=1
     size=7.5
     skewAmount=0
     typefaceFontName=75 Medium
     outlineSettings {
      bApplyOutlineToDropShadows=false
      bMiteredCorners=false
      bSeparateFillAlpha=false
      outlineMaterial=null
      outlineSize=0
      outlineColor=<rgba 0 0 0 1>
     }
    }
   }
   slot {class=OverlaySlot, properties {horizontalAlignment=HAlign_Fill, verticalAlignment=VAlign_Center, padding=<ltrb 10 0 10 0>}}
  }
  ErrorMessage {
   widgetClassPath=&HorizontalBox
   bIsVariable=true
   parent=Overlay_0
   widget=ErrorMessage
   properties {visibility=Collapsed}
   slot {class=OverlaySlot, properties {horizontalAlignment=HAlign_Center, verticalAlignment=VAlign_Center}}
  }
  TextBlock_195 {
   widgetClassPath=&TextBlock
   parent=ErrorMessage
   widget=TextBlock_195
   properties {
    justification=Center
    text=Fail To Find a Stage Manager
    font {
     bForceMonospaced=false
     bMaterialIsStencil=false
     fontMaterial=null
     fontObject=&Roboto
     letterSpacing=0
     monospacedWidth=1
     size=15
     skewAmount=0
     typefaceFontName=Bold
     outlineSettings {
      bApplyOutlineToDropShadows=false
      bMiteredCorners=false
      bSeparateFillAlpha=false
      outlineMaterial=null
      outlineSize=0
      outlineColor=<rgba 0 0 0 1>
     }
    }
   }
   slot {class=HorizontalBoxSlot, properties {horizontalAlignment=HAlign_Center}}
  }
  TryFindStageButton {
   widgetClassPath=&EditorUtilityButton
   parent=ErrorMessage
   widget=TryFindStageButton
   slot {class=HorizontalBoxSlot, properties {horizontalAlignment=HAlign_Left, padding=<ltrb 25 0 25 0>}}
  }
  TextBlock_251 {
   widgetClassPath=&TextBlock
   parent=TryFindStageButton
   widget=TextBlock_251
   properties {
    text=Try to find
    font {
     bForceMonospaced=false
     bMaterialIsStencil=false
     fontMaterial=null
     fontObject=&Roboto
     letterSpacing=0
     monospacedWidth=1
     size=7.5
     skewAmount=0
     typefaceFontName=Bold
     outlineSettings {
      bApplyOutlineToDropShadows=false
      bMiteredCorners=false
      bSeparateFillAlpha=false
      outlineMaterial=null
      outlineSize=0
      outlineColor=<rgba 0 0 0 1>
     }
    }
   }
   slot {class=ButtonSlot}
  }
  TextEventsChangerGroup {
   widgetClassPath=&Overlay
   bIsVariable=true
   parent=Overlay_0
   widget=TextEventsChangerGroup
   properties {visibility=Collapsed}
   slot {class=OverlaySlot, properties {horizontalAlignment=HAlign_Fill, verticalAlignment=VAlign_Fill}}
  }
  Image_12 {
   widgetClassPath=&Image
   parent=TextEventsChangerGroup
   widget=Image_12
   properties {
    brush {
     bIsDynamicallyLoaded=false
     drawAs=Image
     imageType=NoImage
     mirroring=NoMirror
     resourceName=None
     resourceObject=null
     tiling=NoTile
     imageSize=<32 32>
     margin=<ltrb 0 0 0 0>
     outlineSettings {
      bUseBrushTransparency=false
      roundingType=HalfHeightRadius
      width=0
      color {colorUseRule=UseColor_Specified, specifiedColor=<rgba 0 0 0 0>}
      cornerRadii=<0 0 0 0>
     }
     tintColor {colorUseRule=UseColor_Specified, specifiedColor=<rgba 0 0 0 0.359>}
     uVRegion {bIsValid=false, max=<0 0>, min=<0 0>}
    }
    colorAndOpacity=<rgba 0.026 0.026 0.026 1>
    renderTransform {angle=0, scale=<1.005 1.01>, shear=<0 0>, translation=<0 0.5>}
   }
   slot {class=OverlaySlot, properties {horizontalAlignment=HAlign_Fill, verticalAlignment=VAlign_Fill}}
  }
  Overlay_2 {
   widgetClassPath=&Overlay
   parent=TextEventsChangerGroup
   widget=Overlay_2
   slot {class=OverlaySlot, properties {horizontalAlignment=HAlign_Center, verticalAlignment=VAlign_Center, padding=<ltrb 50 50 50 50>}}
  }
  Image_81 {
   widgetClassPath=&Image
   parent=Overlay_2
   widget=Image_81
   properties {
    brush {
     bIsDynamicallyLoaded=false
     drawAs=Image
     imageType=NoImage
     mirroring=NoMirror
     resourceName=None
     resourceObject=null
     tiling=NoTile
     imageSize=<32 32>
     margin=<ltrb 0 0 0 0>
     outlineSettings {
      bUseBrushTransparency=false
      roundingType=HalfHeightRadius
      width=0
      color {colorUseRule=UseColor_Specified, specifiedColor=<rgba 0 0 0 0>}
      cornerRadii=<0 0 0 0>
     }
     tintColor {colorUseRule=UseColor_Specified, specifiedColor=<rgba 0 0 0 0.232>}
     uVRegion {bIsValid=false, max=<0 0>, min=<0 0>}
    }
    colorAndOpacity=<rgba 0.026 0.026 0.026 1>
    renderTransform {angle=0, scale=<1.005 1.01>, shear=<0 0>, translation=<0 0.5>}
   }
   slot {class=OverlaySlot, properties {horizontalAlignment=HAlign_Fill, verticalAlignment=VAlign_Fill}}
  }
  Image_6 {
   widgetClassPath=&Image
   parent=Overlay_2
   widget=Image_6
   properties {colorAndOpacity=<rgba 0.026 0.026 0.026 1>}
   slot {class=OverlaySlot, properties {horizontalAlignment=HAlign_Fill, verticalAlignment=VAlign_Fill}}
  }
  Border_0 {
   widgetClassPath=&Border
   parent=Overlay_2
   widget=Border_0
   properties {horizontalAlignment=HAlign_Center, brushColor=<rgba 0.063 0.1406 0.0994 1>}
   slot {class=OverlaySlot, properties {horizontalAlignment=HAlign_Fill}}
  }
  neweventwindowtitle {
   widgetClassPath=&TextBlock
   parent=Border_0
   widget=neweventwindowtitle
   properties {
    text=Add or Change Text Events Window
    colorAndOpacity {colorUseRule=UseColor_Specified, specifiedColor=<rgba 0.6562 0.6562 0.6562 1>}
    font {
     bForceMonospaced=false
     bMaterialIsStencil=false
     fontMaterial=null
     fontObject=&HeadingNow
     letterSpacing=0
     monospacedWidth=1
     size=11.25
     skewAmount=0
     typefaceFontName=75 Medium
     outlineSettings {
      bApplyOutlineToDropShadows=false
      bMiteredCorners=false
      bSeparateFillAlpha=false
      outlineMaterial=null
      outlineSize=0
      outlineColor=<rgba 0 0 0 1>
     }
    }
   }
   slot {class=BorderSlot, properties {horizontalAlignment=HAlign_Center}}
  }
  NewEventsInsideGroup {
   widgetClassPath=&Overlay
   bIsVariable=true
   parent=Overlay_2
   widget=NewEventsInsideGroup
   properties {visibility=Visible}
   slot {class=OverlaySlot, properties {horizontalAlignment=HAlign_Fill, verticalAlignment=VAlign_Fill, padding=<ltrb 20 45 20 20>}}
  }
  HorizontalBox_5 {
   widgetClassPath=&HorizontalBox
   parent=NewEventsInsideGroup
   widget=HorizontalBox_5
   slot {class=OverlaySlot, properties {horizontalAlignment=HAlign_Center, verticalAlignment=VAlign_Fill}}
  }
  Spacer_1 {
   widgetClassPath=&Spacer
   parent=HorizontalBox_5
   widget=Spacer_1
   properties {visibility=Collapsed, size=<20.1542 1>}
   slot {class=HorizontalBoxSlot}
  }
  EventTypeName_1 {
   widgetClassPath=&TextBlock
   parent=HorizontalBox_5
   widget=EventTypeName_1
   properties {
    text="New Event: "
    colorAndOpacity {colorUseRule=UseColor_Specified, specifiedColor=<rgba 0.6562 0.6562 0.6562 1>}
    font {
     bForceMonospaced=false
     bMaterialIsStencil=false
     fontMaterial=null
     fontObject=&HeadingNow
     letterSpacing=0
     monospacedWidth=1
     size=7.5
     skewAmount=0
     typefaceFontName=75 Medium
     outlineSettings {
      bApplyOutlineToDropShadows=false
      bMiteredCorners=false
      bSeparateFillAlpha=false
      outlineMaterial=null
      outlineSize=0
      outlineColor=<rgba 0 0 0 1>
     }
    }
   }
   slot {class=HorizontalBoxSlot, properties {verticalAlignment=VAlign_Center, padding=<ltrb 0 0 5 0>}}
  }
  NewTextEventCombobox {
   widgetClassPath=&SearchableStringComboBox
   bIsVariable=true
   parent=HorizontalBox_5
   widget=NewTextEventCombobox
   properties {
    selectedOption=Sequence Type List
    font {
     bForceMonospaced=false
     bMaterialIsStencil=false
     fontMaterial=null
     fontObject=&HeadingNow
     letterSpacing=0
     monospacedWidth=1
     size=7.5
     skewAmount=0
     typefaceFontName=75 Medium
     outlineSettings {
      bApplyOutlineToDropShadows=false
      bMiteredCorners=false
      bSeparateFillAlpha=false
      outlineMaterial=null
      outlineSize=0
      outlineColor=<rgba 0 0 0 1>
     }
    }
    foregroundColor {colorUseRule=UseColor_Specified, specifiedColor=<rgba 0.5271 0.5271 0.5271 1>}
    itemStyle {
     bUseParentRowBrush=false
     activeBrush {
      bIsDynamicallyLoaded=false
      drawAs=Image
      imageType=NoImage
      mirroring=NoMirror
      resourceName=None
      resourceObject=null
      tiling=NoTile
      imageSize=<0 0>
      margin=<ltrb 0 0 0 0>
      outlineSettings {
       bUseBrushTransparency=false
       roundingType=HalfHeightRadius
       width=0
       color {colorUseRule=UseColor_Specified, specifiedColor=<rgba 0 0 0 0>}
       cornerRadii=<0 0 0 0>
      }
      tintColor {colorUseRule=UseColor_Specified, specifiedColor=<rgba 0 0.162 0.7454 1>}
      uVRegion {bIsValid=false, max=<0 0>, min=<0 0>}
     }
     activeHighlightedBrush {
      bIsDynamicallyLoaded=false
      drawAs=Image
      imageType=NoImage
      mirroring=NoMirror
      resourceName=None
      resourceObject=null
      tiling=NoTile
      imageSize=<0 0>
      margin=<ltrb 0 0 0 0>
      outlineSettings {
       bUseBrushTransparency=false
       roundingType=HalfHeightRadius
       width=0
       color {colorUseRule=UseColor_Specified, specifiedColor=<rgba 0 0 0 0>}
       cornerRadii=<0 0 0 0>
      }
      tintColor {colorUseRule=UseColor_Specified, specifiedColor=<rgba 0.0252 0.0319 0.0423 1>}
      uVRegion {bIsValid=false, max=<0 0>, min=<0 0>}
     }
     activeHoveredBrush {
      bIsDynamicallyLoaded=false
      drawAs=Image
      imageType=NoImage
      mirroring=NoMirror
      resourceName=None
      resourceObject=null
      tiling=NoTile
      imageSize=<0 0>
      margin=<ltrb 0 0 0 0>
      outlineSettings {
       bUseBrushTransparency=false
       roundingType=HalfHeightRadius
       width=0
       color {colorUseRule=UseColor_Specified, specifiedColor=<rgba 0 0 0 0>}
       cornerRadii=<0 0 0 0>
      }
      tintColor {colorUseRule=UseColor_Specified, specifiedColor=<rgba 0 0.162 0.7454 1>}
      uVRegion {bIsValid=false, max=<0 0>, min=<0 0>}
     }
     dropIndicator_Above {
      bIsDynamicallyLoaded=false
      drawAs=Box
      imageType=FullColor
      mirroring=NoMirror
      resourceName=../../../Engine/Content/Slate/Common/DropZoneIndicator_Above.png
      resourceObject=null
      tiling=NoTile
      imageSize=<0 0>
      margin=<ltrb 0.625 0.625 0 0>
      outlineSettings {
       bUseBrushTransparency=false
       roundingType=HalfHeightRadius
       width=0
       color {colorUseRule=UseColor_Specified, specifiedColor=<rgba 0 0 0 0>}
       cornerRadii=<0 0 0 0>
      }
      tintColor {colorUseRule=UseColor_Specified, specifiedColor=<rgba 0.0091 0.3515 0.9647 1>}
      uVRegion {bIsValid=false, max=<0 0>, min=<0 0>}
     }
     dropIndicator_Below {
      bIsDynamicallyLoaded=false
      drawAs=Box
      imageType=FullColor
      mirroring=NoMirror
      resourceName=../../../Engine/Content/Slate/Common/DropZoneIndicator_Below.png
      resourceObject=null
      tiling=NoTile
      imageSize=<0 0>
      margin=<ltrb 0.625 0 0 0.625>
      outlineSettings {
       bUseBrushTransparency=false
       roundingType=HalfHeightRadius
       width=0
       color {colorUseRule=UseColor_Specified, specifiedColor=<rgba 0 0 0 0>}
       cornerRadii=<0 0 0 0>
      }
      tintColor {colorUseRule=UseColor_Specified, specifiedColor=<rgba 0.0091 0.3515 0.9647 1>}
      uVRegion {bIsValid=false, max=<0 0>, min=<0 0>}
     }
     dropIndicator_Onto {
      bIsDynamicallyLoaded=false
      drawAs=Box
      imageType=FullColor
      mirroring=NoMirror
      resourceName=../../../Engine/Content/Slate/Common/DropZoneIndicator_Onto.png
      resourceObject=null
      tiling=NoTile
      imageSize=<0 0>
      margin=<ltrb 0.25 0.25 0.25 0.25>
      outlineSettings {
       bUseBrushTransparency=false
       roundingType=HalfHeightRadius
       width=0
       color {colorUseRule=UseColor_Specified, specifiedColor=<rgba 0 0 0 0>}
       cornerRadii=<0 0 0 0>
      }
      tintColor {colorUseRule=UseColor_Specified, specifiedColor=<rgba 0.0091 0.3515 0.9647 1>}
      uVRegion {bIsValid=false, max=<0 0>, min=<0 0>}
     }
     evenRowBackgroundBrush {
      bIsDynamicallyLoaded=false
      drawAs=Image
      imageType=NoImage
      mirroring=NoMirror
      resourceName=None
      resourceObject=null
      tiling=NoTile
      imageSize=<0 0>
      margin=<ltrb 0 0 0 0>
      outlineSettings {
       bUseBrushTransparency=false
       roundingType=HalfHeightRadius
       width=0
       color {colorUseRule=UseColor_Specified, specifiedColor=<rgba 0 0 0 0>}
       cornerRadii=<0 0 0 0>
      }
      tintColor {colorUseRule=UseColor_Specified, specifiedColor=<rgba 0.0103 0.0103 0.0103 1>}
      uVRegion {bIsValid=false, max=<0 0>, min=<0 0>}
     }
     evenRowBackgroundHoveredBrush {
      bIsDynamicallyLoaded=false
      drawAs=Image
      imageType=NoImage
      mirroring=NoMirror
      resourceName=None
      resourceObject=null
      tiling=NoTile
      imageSize=<0 0>
      margin=<ltrb 0 0 0 0>
      outlineSettings {
       bUseBrushTransparency=false
       roundingType=HalfHeightRadius
       width=0
       color {colorUseRule=UseColor_Specified, specifiedColor=<rgba 0 0 0 0>}
       cornerRadii=<0 0 0 0>
      }
      tintColor {colorUseRule=UseColor_Specified, specifiedColor=<rgba 0.0176 0.0176 0.0176 1>}
      uVRegion {bIsValid=false, max=<0 0>, min=<0 0>}
     }
     inactiveBrush {
      bIsDynamicallyLoaded=false
      drawAs=Image
      imageType=NoImage
      mirroring=NoMirror
      resourceName=None
      resourceObject=null
      tiling=NoTile
      imageSize=<0 0>
      margin=<ltrb 0 0 0 0>
      outlineSettings {
       bUseBrushTransparency=false
       roundingType=HalfHeightRadius
       width=0
       color {colorUseRule=UseColor_Specified, specifiedColor=<rgba 0 0 0 0>}
       cornerRadii=<0 0 0 0>
      }
      tintColor {colorUseRule=UseColor_Specified, specifiedColor=<rgba 0.0513 0.0953 0.159 1>}
      uVRegion {bIsValid=false, max=<0 0>, min=<0 0>}
     }
     inactiveHighlightedBrush {
      bIsDynamicallyLoaded=false
      drawAs=Image
      imageType=NoImage
      mirroring=NoMirror
      resourceName=None
      resourceObject=null
      tiling=NoTile
      imageSize=<0 0>
      margin=<ltrb 0 0 0 0>
      outlineSettings {
       bUseBrushTransparency=false
       roundingType=HalfHeightRadius
       width=0
       color {colorUseRule=UseColor_Specified, specifiedColor=<rgba 0 0 0 0>}
       cornerRadii=<0 0 0 0>
      }
      tintColor {colorUseRule=UseColor_Specified, specifiedColor=<rgba 0.0252 0.0319 0.0423 1>}
      uVRegion {bIsValid=false, max=<0 0>, min=<0 0>}
     }
     inactiveHoveredBrush {
      bIsDynamicallyLoaded=false
      drawAs=Image
      imageType=NoImage
      mirroring=NoMirror
      resourceName=None
      resourceObject=null
      tiling=NoTile
      imageSize=<0 0>
      margin=<ltrb 0 0 0 0>
      outlineSettings {
       bUseBrushTransparency=false
       roundingType=HalfHeightRadius
       width=0
       color {colorUseRule=UseColor_Specified, specifiedColor=<rgba 0 0 0 0>}
       cornerRadii=<0 0 0 0>
      }
      tintColor {colorUseRule=UseColor_Specified, specifiedColor=<rgba 0.0513 0.0953 0.159 1>}
      uVRegion {bIsValid=false, max=<0 0>, min=<0 0>}
     }
     oddRowBackgroundBrush {
      bIsDynamicallyLoaded=false
      drawAs=Image
      imageType=NoImage
      mirroring=NoMirror
      resourceName=None
      resourceObject=null
      tiling=NoTile
      imageSize=<0 0>
      margin=<ltrb 0 0 0 0>
      outlineSettings {
       bUseBrushTransparency=false
       roundingType=HalfHeightRadius
       width=0
       color {colorUseRule=UseColor_Specified, specifiedColor=<rgba 0 0 0 0>}
       cornerRadii=<0 0 0 0>
      }
      tintColor {colorUseRule=UseColor_Specified, specifiedColor=<rgba 0.0103 0.0103 0.0103 1>}
      uVRegion {bIsValid=false, max=<0 0>, min=<0 0>}
     }
     oddRowBackgroundHoveredBrush {
      bIsDynamicallyLoaded=false
      drawAs=Image
      imageType=NoImage
      mirroring=NoMirror
      resourceName=None
      resourceObject=null
      tiling=NoTile
      imageSize=<0 0>
      margin=<ltrb 0 0 0 0>
      outlineSettings {
       bUseBrushTransparency=false
       roundingType=HalfHeightRadius
       width=0
       color {colorUseRule=UseColor_Specified, specifiedColor=<rgba 0 0 0 0>}
       cornerRadii=<0 0 0 0>
      }
      tintColor {colorUseRule=UseColor_Specified, specifiedColor=<rgba 0.0176 0.0176 0.0176 1>}
      uVRegion {bIsValid=false, max=<0 0>, min=<0 0>}
     }
     parentRowBackgroundBrush {
      bIsDynamicallyLoaded=false
      drawAs=Image
      imageType=NoImage
      mirroring=NoMirror
      resourceName=None
      resourceObject=null
      tiling=NoTile
      imageSize=<32 32>
      margin=<ltrb 0 0 0 0>
      outlineSettings {
       bUseBrushTransparency=false
       roundingType=HalfHeightRadius
       width=0
       color {colorUseRule=UseColor_Specified, specifiedColor=<rgba 0 0 0 0>}
       cornerRadii=<0 0 0 0>
      }
      tintColor {colorUseRule=UseColor_Specified, specifiedColor=<rgba 1 1 1 1>}
      uVRegion {bIsValid=false, max=<0 0>, min=<0 0>}
     }
     parentRowBackgroundHoveredBrush {
      bIsDynamicallyLoaded=false
      drawAs=Image
      imageType=NoImage
      mirroring=NoMirror
      resourceName=None
      resourceObject=null
      tiling=NoTile
      imageSize=<32 32>
      margin=<ltrb 0 0 0 0>
      outlineSettings {
       bUseBrushTransparency=false
       roundingType=HalfHeightRadius
       width=0
       color {colorUseRule=UseColor_Specified, specifiedColor=<rgba 0 0 0 0>}
       cornerRadii=<0 0 0 0>
      }
      tintColor {colorUseRule=UseColor_Specified, specifiedColor=<rgba 1 1 1 1>}
      uVRegion {bIsValid=false, max=<0 0>, min=<0 0>}
     }
     selectedTextColor {colorUseRule=UseColor_Specified, specifiedColor=<rgba 1 1 1 1>}
     selectorFocusedBrush {
      bIsDynamicallyLoaded=false
      drawAs=Border
      imageType=FullColor
      mirroring=NoMirror
      resourceName=../../../Engine/Content/Slate/Common/Selector.png
      resourceObject=null
      tiling=Both
      imageSize=<0 0>
      margin=<ltrb 0.25 0.25 0.25 0.25>
      outlineSettings {
       bUseBrushTransparency=false
       roundingType=HalfHeightRadius
       width=0
       color {colorUseRule=UseColor_Specified, specifiedColor=<rgba 0 0 0 0>}
       cornerRadii=<0 0 0 0>
      }
      tintColor {colorUseRule=UseColor_Specified, specifiedColor=<rgba 0.701 0.225 0.003 1>}
      uVRegion {bIsValid=false, max=<0 0>, min=<0 0>}
     }
     textColor {colorUseRule=UseColor_Specified, specifiedColor=<rgba 0.5271 0.5271 0.5271 1>}
    }
    widgetStyle {
     comboButtonStyle {
      downArrowAlign=VAlign_Center
      buttonStyle {
       clickedSlateSound {resourceObject=null}
       disabled {
        bIsDynamicallyLoaded=false
        drawAs=RoundedBox
        imageType=NoImage
        mirroring=NoMirror
        resourceName=None
        resourceObject=null
        tiling=NoTile
        imageSize=<0 0>
        margin=<ltrb 0 0 0 0>
        outlineSettings {
         bUseBrushTransparency=false
         roundingType=FixedRadius
         width=1
         color {colorUseRule=UseColor_ColorTable, specifiedColor=<rgba 0 0 0 0>}
         cornerRadii=<4 4 4 4>
        }
        tintColor {colorUseRule=UseColor_Specified, specifiedColor=<rgba 0.0395 0.0395 0.0395 1>}
        uVRegion {bIsValid=false, max=<0 0>, min=<0 0>}
       }
       disabledForeground {colorUseRule=UseColor_ColorTable, specifiedColor=<rgba 0 0 0 0>}
       hovered {
        bIsDynamicallyLoaded=false
        drawAs=RoundedBox
        imageType=NoImage
        mirroring=NoMirror
        resourceName=None
        resourceObject=null
        tiling=NoTile
        imageSize=<0 0>
        margin=<ltrb 0 0 0 0>
        outlineSettings {
         bUseBrushTransparency=false
         roundingType=FixedRadius
         width=1
         color {colorUseRule=UseColor_ColorTable, specifiedColor=<rgba 0 0 0 0>}
         cornerRadii=<4 4 4 4>
        }
        tintColor {colorUseRule=UseColor_Specified, specifiedColor=<rgba 0.0048 0.0048 0.0048 1>}
        uVRegion {bIsValid=false, max=<0 0>, min=<0 0>}
       }
       hoveredForeground {colorUseRule=UseColor_ColorTable, specifiedColor=<rgba 0 0 0 0>}
       hoveredSlateSound {resourceObject=null}
       normal {
        bIsDynamicallyLoaded=false
        drawAs=RoundedBox
        imageType=NoImage
        mirroring=NoMirror
        resourceName=None
        resourceObject=null
        tiling=NoTile
        imageSize=<0 0>
        margin=<ltrb 0 0 0 0>
        outlineSettings {
         bUseBrushTransparency=false
         roundingType=FixedRadius
         width=1
         color {colorUseRule=UseColor_ColorTable, specifiedColor=<rgba 0 0 0 0>}
         cornerRadii=<4 4 4 4>
        }
        tintColor {colorUseRule=UseColor_Specified, specifiedColor=<rgba 0.0048 0.0048 0.0048 1>}
        uVRegion {bIsValid=false, max=<0 0>, min=<0 0>}
       }
       normalForeground {colorUseRule=UseColor_ColorTable, specifiedColor=<rgba 0 0 0 0>}
       normalPadding=<ltrb 8 1 8 1>
       pressed {
        bIsDynamicallyLoaded=false
        drawAs=RoundedBox
        imageType=NoImage
        mirroring=NoMirror
        resourceName=None
        resourceObject=null
        tiling=NoTile
        imageSize=<0 0>
        margin=<ltrb 0 0 0 0>
        outlineSettings {
         bUseBrushTransparency=false
         roundingType=FixedRadius
         width=1
         color {colorUseRule=UseColor_ColorTable, specifiedColor=<rgba 0 0 0 0>}
         cornerRadii=<4 4 4 4>
        }
        tintColor {colorUseRule=UseColor_Specified, specifiedColor=<rgba 0.0048 0.0048 0.0048 1>}
        uVRegion {bIsValid=false, max=<0 0>, min=<0 0>}
       }
       pressedForeground {colorUseRule=UseColor_ColorTable, specifiedColor=<rgba 0 0 0 0>}
       pressedPadding=<ltrb 8 2 8 0>
       pressedSlateSound {resourceObject=null}
      }
      contentPadding=<ltrb 0 0 0 0>
      downArrowImage {
       bIsDynamicallyLoaded=false
       drawAs=Image
       imageType=Vector
       mirroring=NoMirror
       resourceName=../../../Engine/Content/Slate/Starship/CoreWidgets/ComboBox/wide-chevron-down.svg
       resourceObject=null
       tiling=NoTile
       imageSize=<16 16>
       margin=<ltrb 0 0 0 0>
       outlineSettings {
        bUseBrushTransparency=false
        roundingType=HalfHeightRadius
        width=0
        color {colorUseRule=UseColor_Specified, specifiedColor=<rgba 0 0 0 0>}
        cornerRadii=<0 0 0 0>
       }
       tintColor {colorUseRule=UseColor_Specified, specifiedColor=<rgba 1 1 1 1>}
       uVRegion {bIsValid=false, max=<0 0>, min=<0 0>}
      }
      downArrowPadding=<ltrb 2 2 2 2>
      menuBorderBrush {
       bIsDynamicallyLoaded=false
       drawAs=RoundedBox
       imageType=NoImage
       mirroring=NoMirror
       resourceName=None
       resourceObject=null
       tiling=NoTile
       imageSize=<0 0>
       margin=<ltrb 0 0 0 0>
       outlineSettings {
        bUseBrushTransparency=false
        roundingType=FixedRadius
        width=1
        color {colorUseRule=UseColor_ColorTable, specifiedColor=<rgba 0 0 0 0>}
        cornerRadii=<0 0 0 0>
       }
       tintColor {colorUseRule=UseColor_Specified, specifiedColor=<rgba 0.0395 0.0395 0.0395 1>}
       uVRegion {bIsValid=false, max=<0 0>, min=<0 0>}
      }
      menuBorderPadding=<ltrb 0 0 0 0>
      shadowColorAndOpacity=<rgba 0 0 0 1>
      shadowOffset=<0 0>
     }
     contentPadding=<ltrb 0 0 0 0>
     menuRowPadding=<ltrb 8 3 8 3>
     pressedSlateSound {resourceObject=null}
     selectionChangeSlateSound {resourceObject=null}
    }
   }
   slot {class=HorizontalBoxSlot, properties {horizontalAlignment=HAlign_Left, verticalAlignment=VAlign_Center}}
  }
  NewTextEventSelectAndAdd {
   widgetClassPath=&EditorUtilityButton
   parent=HorizontalBox_5
   widget=NewTextEventSelectAndAdd
   properties {widgetStyle {
     clickedSlateSound {resourceObject=null}
     disabled {
      bIsDynamicallyLoaded=false
      drawAs=RoundedBox
      imageType=NoImage
      mirroring=NoMirror
      resourceName=None
      resourceObject=null
      tiling=NoTile
      imageSize=<32 32>
      margin=<ltrb 0.5 0.5 0.5 0.5>
      outlineSettings {
       bUseBrushTransparency=false
       roundingType=FixedRadius
       width=0
       color {colorUseRule=UseColor_ColorTable, specifiedColor=<rgba 0 0 0 0>}
       cornerRadii=<4 4 4 4>
      }
      tintColor {colorUseRule=UseColor_Specified, specifiedColor=<rgba 0.0395 0.0395 0.0395 1>}
      uVRegion {bIsValid=false, max=<0 0>, min=<0 0>}
     }
     disabledForeground {colorUseRule=UseColor_ColorTable, specifiedColor=<rgba 0 0 0 0>}
     hovered {
      bIsDynamicallyLoaded=false
      drawAs=RoundedBox
      imageType=NoImage
      mirroring=NoMirror
      resourceName=None
      resourceObject=null
      tiling=NoTile
      imageSize=<32 32>
      margin=<ltrb 0 0 0 0>
      outlineSettings {
       bUseBrushTransparency=false
       roundingType=FixedRadius
       width=0
       color {colorUseRule=UseColor_ColorTable, specifiedColor=<rgba 0 0 0 0>}
       cornerRadii=<4 4 4 4>
      }
      tintColor {colorUseRule=UseColor_Specified, specifiedColor=<rgba 0.1126 0.7031 0 1>}
      uVRegion {bIsValid=false, max=<0 0>, min=<0 0>}
     }
     hoveredForeground {colorUseRule=UseColor_ColorTable, specifiedColor=<rgba 0 0 0 0>}
     hoveredSlateSound {resourceObject=null}
     normal {
      bIsDynamicallyLoaded=false
      drawAs=RoundedBox
      imageType=NoImage
      mirroring=NoMirror
      resourceName=None
      resourceObject=null
      tiling=NoTile
      imageSize=<32 32>
      margin=<ltrb 0.5 0.5 0.5 0.5>
      outlineSettings {
       bUseBrushTransparency=false
       roundingType=FixedRadius
       width=0
       color {colorUseRule=UseColor_ColorTable, specifiedColor=<rgba 0 0 0 0>}
       cornerRadii=<4 4 4 4>
      }
      tintColor {colorUseRule=UseColor_Specified, specifiedColor=<rgba 0.0771 0.6167 0.1289 1>}
      uVRegion {bIsValid=false, max=<0 0>, min=<0 0>}
     }
     normalForeground {colorUseRule=UseColor_ColorTable, specifiedColor=<rgba 0 0 0 0>}
     normalPadding=<ltrb 12 1.5 12 1.5>
     pressed {
      bIsDynamicallyLoaded=false
      drawAs=RoundedBox
      imageType=NoImage
      mirroring=NoMirror
      resourceName=None
      resourceObject=null
      tiling=NoTile
      imageSize=<32 32>
      margin=<ltrb 0.5 0.5 0.5 0.5>
      outlineSettings {
       bUseBrushTransparency=false
       roundingType=FixedRadius
       width=0
       color {colorUseRule=UseColor_ColorTable, specifiedColor=<rgba 0 0 0 0>}
       cornerRadii=<4 4 4 4>
      }
      tintColor {colorUseRule=UseColor_Specified, specifiedColor=<rgba 0.1806 0.4583 0.1572 1>}
      uVRegion {bIsValid=false, max=<0 0>, min=<0 0>}
     }
     pressedForeground {colorUseRule=UseColor_ColorTable, specifiedColor=<rgba 0 0 0 0>}
     pressedPadding=<ltrb 12 2.5 12 0.5>
     pressedSlateSound {resourceObject=null}
    }}
   slot {class=HorizontalBoxSlot, properties {verticalAlignment=VAlign_Center, padding=<ltrb 10 0 0 0>}}
  }
  HorizontalBox_6 {
   widgetClassPath=&HorizontalBox
   parent=NewTextEventSelectAndAdd
   widget=HorizontalBox_6
   slot {class=ButtonSlot, properties {horizontalAlignment=HAlign_Left}}
  }
  Image_4 {
   widgetClassPath=&Image
   parent=HorizontalBox_6
   widget=Image_4
   properties {brush {
     bIsDynamicallyLoaded=false
     drawAs=Image
     imageType=NoImage
     mirroring=NoMirror
     resourceName=None
     resourceObject=&MI_SparksEditorUI_PlusMinus_NoBorder
     tiling=NoTile
     imageSize=<15 15>
     margin=<ltrb 0 0 0 0>
     outlineSettings {
      bUseBrushTransparency=false
      roundingType=HalfHeightRadius
      width=0
      color {colorUseRule=UseColor_Specified, specifiedColor=<rgba 0 0 0 0>}
      cornerRadii=<0 0 0 0>
     }
     tintColor {colorUseRule=UseColor_Specified, specifiedColor=<rgba 1 1 1 1>}
     uVRegion {bIsValid=false, max=<0 0>, min=<0 0>}
    }}
   slot {class=HorizontalBoxSlot, properties {horizontalAlignment=HAlign_Center, verticalAlignment=VAlign_Center, padding=<ltrb 0 0 3 0>}}
  }
  TextBlock_2 {
   widgetClassPath=&TextBlock
   parent=HorizontalBox_6
   widget=TextBlock_2
   properties {
    text=Select and Add
    font {
     bForceMonospaced=false
     bMaterialIsStencil=false
     fontMaterial=null
     fontObject=&HeadingNow
     letterSpacing=0
     monospacedWidth=1
     size=7.5
     skewAmount=0
     typefaceFontName=75 Medium
     outlineSettings {
      bApplyOutlineToDropShadows=false
      bMiteredCorners=false
      bSeparateFillAlpha=false
      outlineMaterial=null
      outlineSize=0
      outlineColor=<rgba 0 0 0 1>
     }
    }
   }
   slot {class=HorizontalBoxSlot, properties {verticalAlignment=VAlign_Center}}
  }
  NewTextEventSelectAndChange {
   widgetClassPath=&EditorUtilityButton
   parent=HorizontalBox_5
   widget=NewTextEventSelectAndChange
   properties {widgetStyle {
     clickedSlateSound {resourceObject=null}
     disabled {
      bIsDynamicallyLoaded=false
      drawAs=RoundedBox
      imageType=NoImage
      mirroring=NoMirror
      resourceName=None
      resourceObject=null
      tiling=NoTile
      imageSize=<32 32>
      margin=<ltrb 0.5 0.5 0.5 0.5>
      outlineSettings {
       bUseBrushTransparency=false
       roundingType=FixedRadius
       width=0
       color {colorUseRule=UseColor_ColorTable, specifiedColor=<rgba 0 0 0 0>}
       cornerRadii=<4 4 4 4>
      }
      tintColor {colorUseRule=UseColor_Specified, specifiedColor=<rgba 0.0395 0.0395 0.0395 1>}
      uVRegion {bIsValid=false, max=<0 0>, min=<0 0>}
     }
     disabledForeground {colorUseRule=UseColor_ColorTable, specifiedColor=<rgba 0 0 0 0>}
     hovered {
      bIsDynamicallyLoaded=false
      drawAs=RoundedBox
      imageType=NoImage
      mirroring=NoMirror
      resourceName=None
      resourceObject=null
      tiling=NoTile
      imageSize=<32 32>
      margin=<ltrb 0 0 0 0>
      outlineSettings {
       bUseBrushTransparency=false
       roundingType=FixedRadius
       width=0
       color {colorUseRule=UseColor_ColorTable, specifiedColor=<rgba 0 0 0 0>}
       cornerRadii=<4 4 4 4>
      }
      tintColor {colorUseRule=UseColor_Specified, specifiedColor=<rgba 0.1126 0.7031 0 1>}
      uVRegion {bIsValid=false, max=<0 0>, min=<0 0>}
     }
     hoveredForeground {colorUseRule=UseColor_ColorTable, specifiedColor=<rgba 0 0 0 0>}
     hoveredSlateSound {resourceObject=null}
     normal {
      bIsDynamicallyLoaded=false
      drawAs=RoundedBox
      imageType=NoImage
      mirroring=NoMirror
      resourceName=None
      resourceObject=null
      tiling=NoTile
      imageSize=<32 32>
      margin=<ltrb 0.5 0.5 0.5 0.5>
      outlineSettings {
       bUseBrushTransparency=false
       roundingType=FixedRadius
       width=0
       color {colorUseRule=UseColor_ColorTable, specifiedColor=<rgba 0 0 0 0>}
       cornerRadii=<4 4 4 4>
      }
      tintColor {colorUseRule=UseColor_Specified, specifiedColor=<rgba 0.0771 0.6167 0.1289 1>}
      uVRegion {bIsValid=false, max=<0 0>, min=<0 0>}
     }
     normalForeground {colorUseRule=UseColor_ColorTable, specifiedColor=<rgba 0 0 0 0>}
     normalPadding=<ltrb 12 1.5 12 1.5>
     pressed {
      bIsDynamicallyLoaded=false
      drawAs=RoundedBox
      imageType=NoImage
      mirroring=NoMirror
      resourceName=None
      resourceObject=null
      tiling=NoTile
      imageSize=<32 32>
      margin=<ltrb 0.5 0.5 0.5 0.5>
      outlineSettings {
       bUseBrushTransparency=false
       roundingType=FixedRadius
       width=0
       color {colorUseRule=UseColor_ColorTable, specifiedColor=<rgba 0 0 0 0>}
       cornerRadii=<4 4 4 4>
      }
      tintColor {colorUseRule=UseColor_Specified, specifiedColor=<rgba 0.1806 0.4583 0.1572 1>}
      uVRegion {bIsValid=false, max=<0 0>, min=<0 0>}
     }
     pressedForeground {colorUseRule=UseColor_ColorTable, specifiedColor=<rgba 0 0 0 0>}
     pressedPadding=<ltrb 12 2.5 12 0.5>
     pressedSlateSound {resourceObject=null}
    }}
   slot {class=HorizontalBoxSlot, properties {verticalAlignment=VAlign_Center, padding=<ltrb 10 0 0 0>}}
  }
  HorizontalBox_7 {
   widgetClassPath=&HorizontalBox
   parent=NewTextEventSelectAndChange
   widget=HorizontalBox_7
   slot {class=ButtonSlot, properties {horizontalAlignment=HAlign_Left}}
  }
  Image_5 {
   widgetClassPath=&Image
   parent=HorizontalBox_7
   widget=Image_5
   properties {brush {
     bIsDynamicallyLoaded=false
     drawAs=Image
     imageType=NoImage
     mirroring=NoMirror
     resourceName=None
     resourceObject=&MI_SparksEditorUI_PlusMinus_NoBorder
     tiling=NoTile
     imageSize=<15 15>
     margin=<ltrb 0 0 0 0>
     outlineSettings {
      bUseBrushTransparency=false
      roundingType=HalfHeightRadius
      width=0
      color {colorUseRule=UseColor_Specified, specifiedColor=<rgba 0 0 0 0>}
      cornerRadii=<0 0 0 0>
     }
     tintColor {colorUseRule=UseColor_Specified, specifiedColor=<rgba 1 1 1 1>}
     uVRegion {bIsValid=false, max=<0 0>, min=<0 0>}
    }}
   slot {class=HorizontalBoxSlot, properties {horizontalAlignment=HAlign_Center, verticalAlignment=VAlign_Center, padding=<ltrb 0 0 3 0>}}
  }
  TextBlock_3 {
   widgetClassPath=&TextBlock
   parent=HorizontalBox_7
   widget=TextBlock_3
   properties {
    text=Select and Change
    font {
     bForceMonospaced=false
     bMaterialIsStencil=false
     fontMaterial=null
     fontObject=&HeadingNow
     letterSpacing=0
     monospacedWidth=1
     size=7.5
     skewAmount=0
     typefaceFontName=75 Medium
     outlineSettings {
      bApplyOutlineToDropShadows=false
      bMiteredCorners=false
      bSeparateFillAlpha=false
      outlineMaterial=null
      outlineSize=0
      outlineColor=<rgba 0 0 0 1>
     }
    }
   }
   slot {class=HorizontalBoxSlot, properties {verticalAlignment=VAlign_Center}}
  }
  Notexteventstext {
   widgetClassPath=&TextBlock
   bIsVariable=true
   parent=Overlay_2
   widget=Notexteventstext
   properties {
    justification=Center
    text=All avaliable events are in use.
    visibility=Collapsed
    font {
     bForceMonospaced=false
     bMaterialIsStencil=false
     fontMaterial=null
     fontObject=&Roboto
     letterSpacing=0
     monospacedWidth=1
     size=15
     skewAmount=0
     typefaceFontName=Bold
     outlineSettings {
      bApplyOutlineToDropShadows=false
      bMiteredCorners=false
      bSeparateFillAlpha=false
      outlineMaterial=null
      outlineSize=0
      outlineColor=<rgba 0 0 0 1>
     }
    }
   }
   slot {class=OverlaySlot, properties {horizontalAlignment=HAlign_Center, verticalAlignment=VAlign_Center}}
  }
  CloseNewTextEventWindowButton {
   widgetClassPath=&EditorUtilityButton
   parent=Overlay_2
   widget=CloseNewTextEventWindowButton
   properties {widgetStyle {
     clickedSlateSound {resourceObject=null}
     disabled {
      bIsDynamicallyLoaded=false
      drawAs=RoundedBox
      imageType=NoImage
      mirroring=NoMirror
      resourceName=None
      resourceObject=null
      tiling=NoTile
      imageSize=<20 20>
      margin=<ltrb 0 0 0 0>
      outlineSettings {
       bUseBrushTransparency=false
       roundingType=FixedRadius
       width=0
       color {colorUseRule=UseColor_ColorTable, specifiedColor=<rgba 0 0 0 0>}
       cornerRadii=<4 4 4 4>
      }
      tintColor {colorUseRule=UseColor_Specified, specifiedColor=<rgba 0.0395 0.0395 0.0395 1>}
      uVRegion {bIsValid=false, max=<0 0>, min=<0 0>}
     }
     disabledForeground {colorUseRule=UseColor_ColorTable, specifiedColor=<rgba 0 0 0 0>}
     hovered {
      bIsDynamicallyLoaded=false
      drawAs=RoundedBox
      imageType=NoImage
      mirroring=NoMirror
      resourceName=None
      resourceObject=null
      tiling=NoTile
      imageSize=<32 32>
      margin=<ltrb 0 0 0 0>
      outlineSettings {
       bUseBrushTransparency=false
       roundingType=FixedRadius
       width=0
       color {colorUseRule=UseColor_ColorTable, specifiedColor=<rgba 0 0 0 0>}
       cornerRadii=<4 4 4 4>
      }
      tintColor {colorUseRule=UseColor_Specified, specifiedColor=<rgba 1 0.1252 0.253 1>}
      uVRegion {bIsValid=false, max=<0 0>, min=<0 0>}
     }
     hoveredForeground {colorUseRule=UseColor_ColorTable, specifiedColor=<rgba 0 0 0 0>}
     hoveredSlateSound {resourceObject=null}
     normal {
      bIsDynamicallyLoaded=false
      drawAs=RoundedBox
      imageType=NoImage
      mirroring=NoMirror
      resourceName=None
      resourceObject=null
      tiling=NoTile
      imageSize=<20 20>
      margin=<ltrb 0 0 0 0>
      outlineSettings {
       bUseBrushTransparency=false
       roundingType=FixedRadius
       width=0
       color {colorUseRule=UseColor_ColorTable, specifiedColor=<rgba 0 0 0 0>}
       cornerRadii=<4 4 4 4>
      }
      tintColor {colorUseRule=UseColor_Specified, specifiedColor=<rgba 0.775 0 0.0304 1>}
      uVRegion {bIsValid=false, max=<0 0>, min=<0 0>}
     }
     normalForeground {colorUseRule=UseColor_ColorTable, specifiedColor=<rgba 0 0 0 0>}
     normalPadding=<ltrb 12 1.5 12 1.5>
     pressed {
      bIsDynamicallyLoaded=false
      drawAs=RoundedBox
      imageType=NoImage
      mirroring=NoMirror
      resourceName=None
      resourceObject=null
      tiling=NoTile
      imageSize=<20 20>
      margin=<ltrb 0 0 0 0>
      outlineSettings {
       bUseBrushTransparency=false
       roundingType=FixedRadius
       width=0
       color {colorUseRule=UseColor_ColorTable, specifiedColor=<rgba 0 0 0 0>}
       cornerRadii=<4 4 4 4>
      }
      tintColor {colorUseRule=UseColor_Specified, specifiedColor=<rgba 0.2396 0 0 1>}
      uVRegion {bIsValid=false, max=<0 0>, min=<0 0>}
     }
     pressedForeground {colorUseRule=UseColor_ColorTable, specifiedColor=<rgba 0 0 0 0>}
     pressedPadding=<ltrb 12 2.5 12 0.5>
     pressedSlateSound {resourceObject=null}
    }}
   slot {class=OverlaySlot, properties {horizontalAlignment=HAlign_Right, padding=<ltrb 10 4 10 0>}}
  }
  TextBlock_4 {
   widgetClassPath=&TextBlock
   parent=CloseNewTextEventWindowButton
   widget=TextBlock_4
   properties {
    text=X
    font {
     bForceMonospaced=false
     bMaterialIsStencil=false
     fontMaterial=null
     fontObject=&HeadingNow
     letterSpacing=0
     monospacedWidth=1
     size=7.5
     skewAmount=0
     typefaceFontName=86 Bold
     outlineSettings {
      bApplyOutlineToDropShadows=false
      bMiteredCorners=false
      bSeparateFillAlpha=false
      outlineMaterial=null
      outlineSize=0
      outlineColor=<rgba 0 0 0 1>
     }
    }
   }
   slot {class=ButtonSlot, properties {horizontalAlignment=HAlign_Fill, padding=<ltrb 0 0 0 0>}}
  }
  NewSequenceEventWindow {
   widgetClassPath=&Overlay
   bIsVariable=true
   parent=Overlay_0
   widget=NewSequenceEventWindow
   properties {visibility=Collapsed}
   slot {class=OverlaySlot, properties {horizontalAlignment=HAlign_Fill, verticalAlignment=VAlign_Fill}}
  }
  Image_11 {
   widgetClassPath=&Image
   parent=NewSequenceEventWindow
   widget=Image_11
   properties {
    brush {
     bIsDynamicallyLoaded=false
     drawAs=Image
     imageType=NoImage
     mirroring=NoMirror
     resourceName=None
     resourceObject=null
     tiling=NoTile
     imageSize=<32 32>
     margin=<ltrb 0 0 0 0>
     outlineSettings {
      bUseBrushTransparency=false
      roundingType=HalfHeightRadius
      width=0
      color {colorUseRule=UseColor_Specified, specifiedColor=<rgba 0 0 0 0>}
      cornerRadii=<0 0 0 0>
     }
     tintColor {colorUseRule=UseColor_Specified, specifiedColor=<rgba 0 0 0 0.715>}
     uVRegion {bIsValid=false, max=<0 0>, min=<0 0>}
    }
    colorAndOpacity=<rgba 0.026 0.026 0.026 1>
    renderTransform {angle=0, scale=<1.005 1.01>, shear=<0 0>, translation=<0 0.5>}
   }
   slot {class=OverlaySlot, properties {horizontalAlignment=HAlign_Fill, verticalAlignment=VAlign_Fill}}
  }
  SizeBox_2 {
   widgetClassPath=&SizeBox
   parent=NewSequenceEventWindow
   widget=SizeBox_2
   properties {bOverride_MinDesiredWidth=true, minDesiredWidth=700}
   slot {class=OverlaySlot, properties {horizontalAlignment=HAlign_Center, padding=<ltrb 50 180 50 50>}}
  }
  Overlay {widgetClassPath=&Overlay, parent=SizeBox_2, widget=Overlay, slot {class=SizeBoxSlot}}
  Image_7 {
   widgetClassPath=&Image
   parent=Overlay
   widget=Image_7
   properties {
    brush {
     bIsDynamicallyLoaded=false
     drawAs=Image
     imageType=NoImage
     mirroring=NoMirror
     resourceName=None
     resourceObject=null
     tiling=NoTile
     imageSize=<32 32>
     margin=<ltrb 0 0 0 0>
     outlineSettings {
      bUseBrushTransparency=false
      roundingType=HalfHeightRadius
      width=0
      color {colorUseRule=UseColor_Specified, specifiedColor=<rgba 0 0 0 0>}
      cornerRadii=<0 0 0 0>
     }
     tintColor {colorUseRule=UseColor_Specified, specifiedColor=<rgba 0 0 0 0.232>}
     uVRegion {bIsValid=false, max=<0 0>, min=<0 0>}
    }
    colorAndOpacity=<rgba 0.026 0.026 0.026 1>
    renderTransform {angle=0, scale=<1.005 1.01>, shear=<0 0>, translation=<0 0.5>}
   }
   slot {class=OverlaySlot, properties {horizontalAlignment=HAlign_Fill, verticalAlignment=VAlign_Fill}}
  }
  Image_8 {
   widgetClassPath=&Image
   parent=Overlay
   widget=Image_8
   properties {colorAndOpacity=<rgba 0.026 0.026 0.026 1>}
   slot {class=OverlaySlot, properties {horizontalAlignment=HAlign_Fill, verticalAlignment=VAlign_Fill}}
  }
  Border {
   widgetClassPath=&Border
   parent=Overlay
   widget=Border
   properties {clipping=ClipToBounds, horizontalAlignment=HAlign_Left, brushColor=<rgba 0.063 0.1406 0.0994 1>, padding=<ltrb 10 2 4 2>}
   slot {class=OverlaySlot, properties {horizontalAlignment=HAlign_Fill}}
  }
  EventsTitleText {
   widgetClassPath=&TextBlock
   bIsVariable=true
   parent=Border
   widget=EventsTitleText
   properties {
    text="Add or Change Sequence Pool, Sequences, Events Window"
    colorAndOpacity {colorUseRule=UseColor_Specified, specifiedColor=<rgba 0.6562 0.6562 0.6562 1>}
    font {
     bForceMonospaced=false
     bMaterialIsStencil=false
     fontMaterial=null
     fontObject=&HeadingNow
     letterSpacing=0
     monospacedWidth=1
     size=11.25
     skewAmount=0
     typefaceFontName=75 Medium
     outlineSettings {
      bApplyOutlineToDropShadows=false
      bMiteredCorners=false
      bSeparateFillAlpha=false
      outlineMaterial=null
      outlineSize=0
      outlineColor=<rgba 0 0 0 1>
     }
    }
   }
   slot {class=BorderSlot, properties {horizontalAlignment=HAlign_Left, padding=<ltrb 10 2 4 2>}}
  }
  SizeBox_5 {
   widgetClassPath=&SizeBox
   parent=Overlay
   widget=SizeBox_5
   properties {bOverride_MinDesiredWidth=true, minDesiredWidth=670}
   slot {class=OverlaySlot, properties {horizontalAlignment=HAlign_Fill, verticalAlignment=VAlign_Fill}}
  }
  NewSequenceEventsInsideGroup {
   widgetClassPath=&Overlay
   bIsVariable=true
   parent=SizeBox_5
   widget=NewSequenceEventsInsideGroup
   properties {clipping=ClipToBounds, visibility=Visible}
   slot {class=SizeBoxSlot}
  }
  VerticalBox_0 {
   widgetClassPath=&VerticalBox
   parent=NewSequenceEventsInsideGroup
   widget=VerticalBox_0
   slot {class=OverlaySlot, properties {horizontalAlignment=HAlign_Fill, verticalAlignment=VAlign_Fill, padding=<ltrb 10 35 10 25>}}
  }
  Border_2 {
   widgetClassPath=&Border
   parent=VerticalBox_0
   widget=Border_2
   properties {
    background {
     bIsDynamicallyLoaded=false
     drawAs=Image
     imageType=NoImage
     mirroring=NoMirror
     resourceName=None
     resourceObject=null
     tiling=NoTile
     imageSize=<32 32>
     margin=<ltrb 0 0 0 0>
     outlineSettings {
      bUseBrushTransparency=false
      roundingType=HalfHeightRadius
      width=0
      color {colorUseRule=UseColor_Specified, specifiedColor=<rgba 0 0 0 0>}
      cornerRadii=<0 0 0 0>
     }
     tintColor {colorUseRule=UseColor_Specified, specifiedColor=<rgba 0.0729 0.0729 0.0729 1>}
     uVRegion {bIsValid=false, max=<0 0>, min=<0 0>}
    }
    padding=<ltrb 4 5 4 5>
   }
   slot {class=VerticalBoxSlot, properties {verticalAlignment=VAlign_Center, padding=<ltrb -10 0 -10 2>}}
  }
  HorizontalBox_10 {
   widgetClassPath=&HorizontalBox
   parent=Border_2
   widget=HorizontalBox_10
   slot {class=BorderSlot, properties {padding=<ltrb 4 5 4 5>}}
  }
  SelectAndChangeToggle {
   widgetClassPath=&EditorUtilityCheckBox
   bIsVariable=true
   parent=HorizontalBox_10
   widget=SelectAndChangeToggle
   properties {
    checkedState=Checked
    toolTipText=Shows the MIDI Text Events
    widgetStyle {
     checkBoxType=ToggleButton
     backgroundHoveredImage {
      bIsDynamicallyLoaded=false
      drawAs=NoDrawType
      imageType=NoImage
      mirroring=NoMirror
      resourceName=None
      resourceObject=null
      tiling=NoTile
      imageSize=<18 18>
      margin=<ltrb 0 0 0 0>
      outlineSettings {
       bUseBrushTransparency=false
       roundingType=FixedRadius
       width=1
       color {colorUseRule=UseColor_ColorTable, specifiedColor=<rgba 0 0 0 0>}
       cornerRadii=<3 3 3 3>
      }
      tintColor {colorUseRule=UseColor_Specified, specifiedColor=<rgba 0.0048 0.0048 0.0048 1>}
      uVRegion {bIsValid=false, max=<0 0>, min=<0 0>}
     }
     backgroundImage {
      bIsDynamicallyLoaded=false
      drawAs=NoDrawType
      imageType=NoImage
      mirroring=NoMirror
      resourceName=None
      resourceObject=null
      tiling=NoTile
      imageSize=<18 18>
      margin=<ltrb 0 0 0 0>
      outlineSettings {
       bUseBrushTransparency=false
       roundingType=FixedRadius
       width=1
       color {colorUseRule=UseColor_ColorTable, specifiedColor=<rgba 0 0 0 0>}
       cornerRadii=<3 3 3 3>
      }
      tintColor {colorUseRule=UseColor_Specified, specifiedColor=<rgba 0.0048 0.0048 0.0048 1>}
      uVRegion {bIsValid=false, max=<0 0>, min=<0 0>}
     }
     backgroundPressedImage {
      bIsDynamicallyLoaded=false
      drawAs=NoDrawType
      imageType=NoImage
      mirroring=NoMirror
      resourceName=None
      resourceObject=null
      tiling=NoTile
      imageSize=<18 18>
      margin=<ltrb 0 0 0 0>
      outlineSettings {
       bUseBrushTransparency=false
       roundingType=FixedRadius
       width=1
       color {colorUseRule=UseColor_ColorTable, specifiedColor=<rgba 0 0 0 0>}
       cornerRadii=<3 3 3 3>
      }
      tintColor {colorUseRule=UseColor_Specified, specifiedColor=<rgba 0.0048 0.0048 0.0048 1>}
      uVRegion {bIsValid=false, max=<0 0>, min=<0 0>}
     }
     borderBackgroundColor {colorUseRule=UseColor_Specified, specifiedColor=<rgba 1 1 1 1>}
     checkedForeground {colorUseRule=UseColor_Specified, specifiedColor=<rgba 1 1 1 1>}
     checkedHoveredForeground {colorUseRule=UseColor_Specified, specifiedColor=<rgba 1 1 1 1>}
     checkedHoveredImage {
      bIsDynamicallyLoaded=false
      drawAs=RoundedBox
      imageType=NoImage
      mirroring=NoMirror
      resourceName=None
      resourceObject=null
      tiling=NoTile
      imageSize=<32 32>
      margin=<ltrb 0 0 0 0>
      outlineSettings {
       bUseBrushTransparency=false
       roundingType=FixedRadius
       width=0.5
       color {colorUseRule=UseColor_Specified, specifiedColor=<rgba 0 0 0 0>}
       cornerRadii=<5 5 5 1>
      }
      tintColor {colorUseRule=UseColor_Specified, specifiedColor=<rgba 0.2734 0.4375 0.2816 1>}
      uVRegion {bIsValid=false, max=<0 0>, min=<0 0>}
     }
     checkedImage {
      bIsDynamicallyLoaded=false
      drawAs=RoundedBox
      imageType=NoImage
      mirroring=NoMirror
      resourceName=None
      resourceObject=null
      tiling=NoTile
      imageSize=<32 32>
      margin=<ltrb 0.5 0.5 0.5 0.5>
      outlineSettings {
       bUseBrushTransparency=false
       roundingType=FixedRadius
       width=0.5
       color {colorUseRule=UseColor_Specified, specifiedColor=<rgba 0 0 0 0>}
       cornerRadii=<5 5 5 1>
      }
      tintColor {colorUseRule=UseColor_Specified, specifiedColor=<rgba 0.0355 0.4375 0.0554 1>}
      uVRegion {bIsValid=false, max=<0 0>, min=<0 0>}
     }
     checkedPressedForeground {colorUseRule=UseColor_Specified, specifiedColor=<rgba 1 1 1 1>}
     checkedPressedImage {
      bIsDynamicallyLoaded=false
      drawAs=RoundedBox
      imageType=NoImage
      mirroring=NoMirror
      resourceName=None
      resourceObject=null
      tiling=NoTile
      imageSize=<32 32>
      margin=<ltrb 0 0 0 0>
      outlineSettings {
       bUseBrushTransparency=false
       roundingType=FixedRadius
       width=0.5
       color {colorUseRule=UseColor_Specified, specifiedColor=<rgba 0 0 0 0>}
       cornerRadii=<5 5 5 1>
      }
      tintColor {colorUseRule=UseColor_Specified, specifiedColor=<rgba 0.4248 0.9062 0.4486 1>}
      uVRegion {bIsValid=false, max=<0 0>, min=<0 0>}
     }
     checkedSlateSound {resourceObject=null}
     foregroundColor {colorUseRule=UseColor_Specified, specifiedColor=<rgba 1 1 1 1>}
     hoveredForeground {colorUseRule=UseColor_Specified, specifiedColor=<rgba 1 1 1 1>}
     hoveredSlateSound {resourceObject=null}
     padding=<ltrb 2 0 0 0>
     pressedForeground {colorUseRule=UseColor_Specified, specifiedColor=<rgba 1 1 1 1>}
     uncheckedHoveredImage {
      bIsDynamicallyLoaded=false
      drawAs=RoundedBox
      imageType=NoImage
      mirroring=NoMirror
      resourceName=None
      resourceObject=null
      tiling=NoTile
      imageSize=<32 60.845>
      margin=<ltrb 0 0 0 0>
      outlineSettings {
       bUseBrushTransparency=false
       roundingType=FixedRadius
       width=0.5
       color {colorUseRule=UseColor_Specified, specifiedColor=<rgba 0 0 0 0>}
       cornerRadii=<5 5 5 1>
      }
      tintColor {colorUseRule=UseColor_Specified, specifiedColor=<rgba 0.724 0.724 0.724 1>}
      uVRegion {bIsValid=false, max=<0 0>, min=<0 0>}
     }
     uncheckedImage {
      bIsDynamicallyLoaded=false
      drawAs=RoundedBox
      imageType=NoImage
      mirroring=NoMirror
      resourceName=None
      resourceObject=null
      tiling=NoTile
      imageSize=<32 32>
      margin=<ltrb 0 0 0 0>
      outlineSettings {
       bUseBrushTransparency=false
       roundingType=FixedRadius
       width=0.5
       color {colorUseRule=UseColor_Specified, specifiedColor=<rgba 0 0 0 0>}
       cornerRadii=<5 5 5 1>
      }
      tintColor {colorUseRule=UseColor_Specified, specifiedColor=<rgba 0.2448 0.2448 0.2448 1>}
      uVRegion {bIsValid=false, max=<0 0>, min=<0 0>}
     }
     uncheckedPressedImage {
      bIsDynamicallyLoaded=false
      drawAs=RoundedBox
      imageType=NoImage
      mirroring=NoMirror
      resourceName=None
      resourceObject=null
      tiling=NoTile
      imageSize=<32 32>
      margin=<ltrb 0 0 0 0>
      outlineSettings {
       bUseBrushTransparency=false
       roundingType=FixedRadius
       width=0.5
       color {colorUseRule=UseColor_Specified, specifiedColor=<rgba 0 0 0 0>}
       cornerRadii=<5 5 5 1>
      }
      tintColor {colorUseRule=UseColor_Specified, specifiedColor=<rgba 0.8594 0.8494 0.698 1>}
      uVRegion {bIsValid=false, max=<0 0>, min=<0 0>}
     }
     uncheckedSlateSound {resourceObject=null}
     undeterminedForeground {colorUseRule=UseColor_Specified, specifiedColor=<rgba 1 1 1 1>}
     undeterminedHoveredImage {
      bIsDynamicallyLoaded=false
      drawAs=Image
      imageType=Vector
      mirroring=NoMirror
      resourceName=../../../Engine/Content/Slate/Starship/CoreWidgets/CheckBox/indeterminate.svg
      resourceObject=null
      tiling=NoTile
      imageSize=<16 16>
      margin=<ltrb 0 0 0 0>
      outlineSettings {
       bUseBrushTransparency=false
       roundingType=HalfHeightRadius
       width=0
       color {colorUseRule=UseColor_Specified, specifiedColor=<rgba 0 0 0 0>}
       cornerRadii=<0 0 0 0>
      }
      tintColor {colorUseRule=UseColor_Specified, specifiedColor=<rgba 0.0044 0.2384 1 1>}
      uVRegion {bIsValid=false, max=<0 0>, min=<0 0>}
     }
     undeterminedImage {
      bIsDynamicallyLoaded=false
      drawAs=Image
      imageType=Vector
      mirroring=NoMirror
      resourceName=../../../Engine/Content/Slate/Starship/CoreWidgets/CheckBox/indeterminate.svg
      resourceObject=null
      tiling=NoTile
      imageSize=<16 16>
      margin=<ltrb 0 0 0 0>
      outlineSettings {
       bUseBrushTransparency=false
       roundingType=HalfHeightRadius
       width=0
       color {colorUseRule=UseColor_Specified, specifiedColor=<rgba 0 0 0 0>}
       cornerRadii=<0 0 0 0>
      }
      tintColor {colorUseRule=UseColor_Specified, specifiedColor=<rgba 0 0.162 0.7454 1>}
      uVRegion {bIsValid=false, max=<0 0>, min=<0 0>}
     }
     undeterminedPressedImage {
      bIsDynamicallyLoaded=false
      drawAs=Image
      imageType=Vector
      mirroring=NoMirror
      resourceName=../../../Engine/Content/Slate/Starship/CoreWidgets/CheckBox/indeterminate.svg
      resourceObject=null
      tiling=NoTile
      imageSize=<16 16>
      margin=<ltrb 0 0 0 0>
      outlineSettings {
       bUseBrushTransparency=false
       roundingType=HalfHeightRadius
       width=0
       color {colorUseRule=UseColor_Specified, specifiedColor=<rgba 0 0 0 0>}
       cornerRadii=<0 0 0 0>
      }
      tintColor {colorUseRule=UseColor_Specified, specifiedColor=<rgba 0 0.0802 0.3515 1>}
      uVRegion {bIsValid=false, max=<0 0>, min=<0 0>}
     }
    }
   }
   slot {class=HorizontalBoxSlot, properties {padding=<ltrb 0 0 2 0>}}
  }
  TextBlock_6 {
   widgetClassPath=&TextBlock
   parent=SelectAndChangeToggle
   widget=TextBlock_6
   properties {
    justification=Center
    text=Select Existent Data Asset
    font {
     bForceMonospaced=false
     bMaterialIsStencil=false
     fontMaterial=null
     fontObject=&HeadingNow
     letterSpacing=0
     monospacedWidth=1
     size=8.25
     skewAmount=0
     typefaceFontName=75 Medium
     outlineSettings {
      bApplyOutlineToDropShadows=false
      bMiteredCorners=false
      bSeparateFillAlpha=false
      outlineMaterial=null
      outlineSize=0
      outlineColor=<rgba 0 0 0 1>
     }
    }
    margin=<ltrb 7 7 7 7>
   }
   slot {class=PanelSlot}
  }
  AddAndCreateToggle {
   widgetClassPath=&EditorUtilityCheckBox
   bIsVariable=true
   parent=HorizontalBox_10
   widget=AddAndCreateToggle
   properties {
    toolTipText=Shows the MIDI Text Events
    widgetStyle {
     checkBoxType=ToggleButton
     backgroundHoveredImage {
      bIsDynamicallyLoaded=false
      drawAs=NoDrawType
      imageType=NoImage
      mirroring=NoMirror
      resourceName=None
      resourceObject=null
      tiling=NoTile
      imageSize=<18 18>
      margin=<ltrb 0 0 0 0>
      outlineSettings {
       bUseBrushTransparency=false
       roundingType=FixedRadius
       width=1
       color {colorUseRule=UseColor_ColorTable, specifiedColor=<rgba 0 0 0 0>}
       cornerRadii=<3 3 3 3>
      }
      tintColor {colorUseRule=UseColor_Specified, specifiedColor=<rgba 0.0048 0.0048 0.0048 1>}
      uVRegion {bIsValid=false, max=<0 0>, min=<0 0>}
     }
     backgroundImage {
      bIsDynamicallyLoaded=false
      drawAs=NoDrawType
      imageType=NoImage
      mirroring=NoMirror
      resourceName=None
      resourceObject=null
      tiling=NoTile
      imageSize=<18 18>
      margin=<ltrb 0 0 0 0>
      outlineSettings {
       bUseBrushTransparency=false
       roundingType=FixedRadius
       width=1
       color {colorUseRule=UseColor_ColorTable, specifiedColor=<rgba 0 0 0 0>}
       cornerRadii=<3 3 3 3>
      }
      tintColor {colorUseRule=UseColor_Specified, specifiedColor=<rgba 0.0048 0.0048 0.0048 1>}
      uVRegion {bIsValid=false, max=<0 0>, min=<0 0>}
     }
     backgroundPressedImage {
      bIsDynamicallyLoaded=false
      drawAs=NoDrawType
      imageType=NoImage
      mirroring=NoMirror
      resourceName=None
      resourceObject=null
      tiling=NoTile
      imageSize=<18 18>
      margin=<ltrb 0 0 0 0>
      outlineSettings {
       bUseBrushTransparency=false
       roundingType=FixedRadius
       width=1
       color {colorUseRule=UseColor_ColorTable, specifiedColor=<rgba 0 0 0 0>}
       cornerRadii=<3 3 3 3>
      }
      tintColor {colorUseRule=UseColor_Specified, specifiedColor=<rgba 0.0048 0.0048 0.0048 1>}
      uVRegion {bIsValid=false, max=<0 0>, min=<0 0>}
     }
     borderBackgroundColor {colorUseRule=UseColor_Specified, specifiedColor=<rgba 1 1 1 1>}
     checkedForeground {colorUseRule=UseColor_Specified, specifiedColor=<rgba 1 1 1 1>}
     checkedHoveredForeground {colorUseRule=UseColor_Specified, specifiedColor=<rgba 1 1 1 1>}
     checkedHoveredImage {
      bIsDynamicallyLoaded=false
      drawAs=RoundedBox
      imageType=NoImage
      mirroring=NoMirror
      resourceName=None
      resourceObject=null
      tiling=NoTile
      imageSize=<32 32>
      margin=<ltrb 0 0 0 0>
      outlineSettings {
       bUseBrushTransparency=false
       roundingType=FixedRadius
       width=0.5
       color {colorUseRule=UseColor_Specified, specifiedColor=<rgba 0 0 0 0>}
       cornerRadii=<5 5 5 1>
      }
      tintColor {colorUseRule=UseColor_Specified, specifiedColor=<rgba 0.2734 0.4375 0.2816 1>}
      uVRegion {bIsValid=false, max=<0 0>, min=<0 0>}
     }
     checkedImage {
      bIsDynamicallyLoaded=false
      drawAs=RoundedBox
      imageType=NoImage
      mirroring=NoMirror
      resourceName=None
      resourceObject=null
      tiling=NoTile
      imageSize=<32 32>
      margin=<ltrb 0.5 0.5 0.5 0.5>
      outlineSettings {
       bUseBrushTransparency=false
       roundingType=FixedRadius
       width=0.5
       color {colorUseRule=UseColor_Specified, specifiedColor=<rgba 0 0 0 0>}
       cornerRadii=<5 5 5 1>
      }
      tintColor {colorUseRule=UseColor_Specified, specifiedColor=<rgba 0.0355 0.4375 0.0554 1>}
      uVRegion {bIsValid=false, max=<0 0>, min=<0 0>}
     }
     checkedPressedForeground {colorUseRule=UseColor_Specified, specifiedColor=<rgba 1 1 1 1>}
     checkedPressedImage {
      bIsDynamicallyLoaded=false
      drawAs=RoundedBox
      imageType=NoImage
      mirroring=NoMirror
      resourceName=None
      resourceObject=null
      tiling=NoTile
      imageSize=<32 32>
      margin=<ltrb 0 0 0 0>
      outlineSettings {
       bUseBrushTransparency=false
       roundingType=FixedRadius
       width=0.5
       color {colorUseRule=UseColor_Specified, specifiedColor=<rgba 0 0 0 0>}
       cornerRadii=<5 5 5 1>
      }
      tintColor {colorUseRule=UseColor_Specified, specifiedColor=<rgba 0.4248 0.9062 0.4486 1>}
      uVRegion {bIsValid=false, max=<0 0>, min=<0 0>}
     }
     checkedSlateSound {resourceObject=null}
     foregroundColor {colorUseRule=UseColor_Specified, specifiedColor=<rgba 1 1 1 1>}
     hoveredForeground {colorUseRule=UseColor_Specified, specifiedColor=<rgba 1 1 1 1>}
     hoveredSlateSound {resourceObject=null}
     padding=<ltrb 2 0 0 0>
     pressedForeground {colorUseRule=UseColor_Specified, specifiedColor=<rgba 1 1 1 1>}
     uncheckedHoveredImage {
      bIsDynamicallyLoaded=false
      drawAs=RoundedBox
      imageType=NoImage
      mirroring=NoMirror
      resourceName=None
      resourceObject=null
      tiling=NoTile
      imageSize=<32 60.845>
      margin=<ltrb 0 0 0 0>
      outlineSettings {
       bUseBrushTransparency=false
       roundingType=FixedRadius
       width=0.5
       color {colorUseRule=UseColor_Specified, specifiedColor=<rgba 0 0 0 0>}
       cornerRadii=<5 5 5 1>
      }
      tintColor {colorUseRule=UseColor_Specified, specifiedColor=<rgba 0.724 0.724 0.724 1>}
      uVRegion {bIsValid=false, max=<0 0>, min=<0 0>}
     }
     uncheckedImage {
      bIsDynamicallyLoaded=false
      drawAs=RoundedBox
      imageType=NoImage
      mirroring=NoMirror
      resourceName=None
      resourceObject=null
      tiling=NoTile
      imageSize=<32 32>
      margin=<ltrb 0 0 0 0>
      outlineSettings {
       bUseBrushTransparency=false
       roundingType=FixedRadius
       width=0.5
       color {colorUseRule=UseColor_Specified, specifiedColor=<rgba 0 0 0 0>}
       cornerRadii=<5 5 5 1>
      }
      tintColor {colorUseRule=UseColor_Specified, specifiedColor=<rgba 0.2448 0.2448 0.2448 1>}
      uVRegion {bIsValid=false, max=<0 0>, min=<0 0>}
     }
     uncheckedPressedImage {
      bIsDynamicallyLoaded=false
      drawAs=RoundedBox
      imageType=NoImage
      mirroring=NoMirror
      resourceName=None
      resourceObject=null
      tiling=NoTile
      imageSize=<32 32>
      margin=<ltrb 0 0 0 0>
      outlineSettings {
       bUseBrushTransparency=false
       roundingType=FixedRadius
       width=0.5
       color {colorUseRule=UseColor_Specified, specifiedColor=<rgba 0 0 0 0>}
       cornerRadii=<5 5 5 1>
      }
      tintColor {colorUseRule=UseColor_Specified, specifiedColor=<rgba 0.8594 0.8494 0.698 1>}
      uVRegion {bIsValid=false, max=<0 0>, min=<0 0>}
     }
     uncheckedSlateSound {resourceObject=null}
     undeterminedForeground {colorUseRule=UseColor_Specified, specifiedColor=<rgba 1 1 1 1>}
     undeterminedHoveredImage {
      bIsDynamicallyLoaded=false
      drawAs=Image
      imageType=Vector
      mirroring=NoMirror
      resourceName=../../../Engine/Content/Slate/Starship/CoreWidgets/CheckBox/indeterminate.svg
      resourceObject=null
      tiling=NoTile
      imageSize=<16 16>
      margin=<ltrb 0 0 0 0>
      outlineSettings {
       bUseBrushTransparency=false
       roundingType=HalfHeightRadius
       width=0
       color {colorUseRule=UseColor_Specified, specifiedColor=<rgba 0 0 0 0>}
       cornerRadii=<0 0 0 0>
      }
      tintColor {colorUseRule=UseColor_Specified, specifiedColor=<rgba 0.0044 0.2384 1 1>}
      uVRegion {bIsValid=false, max=<0 0>, min=<0 0>}
     }
     undeterminedImage {
      bIsDynamicallyLoaded=false
      drawAs=Image
      imageType=Vector
      mirroring=NoMirror
      resourceName=../../../Engine/Content/Slate/Starship/CoreWidgets/CheckBox/indeterminate.svg
      resourceObject=null
      tiling=NoTile
      imageSize=<16 16>
      margin=<ltrb 0 0 0 0>
      outlineSettings {
       bUseBrushTransparency=false
       roundingType=HalfHeightRadius
       width=0
       color {colorUseRule=UseColor_Specified, specifiedColor=<rgba 0 0 0 0>}
       cornerRadii=<0 0 0 0>
      }
      tintColor {colorUseRule=UseColor_Specified, specifiedColor=<rgba 0 0.162 0.7454 1>}
      uVRegion {bIsValid=false, max=<0 0>, min=<0 0>}
     }
     undeterminedPressedImage {
      bIsDynamicallyLoaded=false
      drawAs=Image
      imageType=Vector
      mirroring=NoMirror
      resourceName=../../../Engine/Content/Slate/Starship/CoreWidgets/CheckBox/indeterminate.svg
      resourceObject=null
      tiling=NoTile
      imageSize=<16 16>
      margin=<ltrb 0 0 0 0>
      outlineSettings {
       bUseBrushTransparency=false
       roundingType=HalfHeightRadius
       width=0
       color {colorUseRule=UseColor_Specified, specifiedColor=<rgba 0 0 0 0>}
       cornerRadii=<0 0 0 0>
      }
      tintColor {colorUseRule=UseColor_Specified, specifiedColor=<rgba 0 0.0802 0.3515 1>}
      uVRegion {bIsValid=false, max=<0 0>, min=<0 0>}
     }
    }
   }
   slot {class=HorizontalBoxSlot}
  }
  TextBlock_10 {
   widgetClassPath=&TextBlock
   parent=AddAndCreateToggle
   widget=TextBlock_10
   properties {
    justification=Center
    text=Create New Data Asset
    font {
     bForceMonospaced=false
     bMaterialIsStencil=false
     fontMaterial=null
     fontObject=&HeadingNow
     letterSpacing=0
     monospacedWidth=1
     size=8.25
     skewAmount=0
     typefaceFontName=75 Medium
     outlineSettings {
      bApplyOutlineToDropShadows=false
      bMiteredCorners=false
      bSeparateFillAlpha=false
      outlineMaterial=null
      outlineSize=0
      outlineColor=<rgba 0 0 0 1>
     }
    }
    margin=<ltrb 7 7 7 7>
   }
   slot {class=PanelSlot}
  }
  EditorUtilityScrollBox_2 {
   widgetClassPath=&EditorUtilityScrollBox
   parent=VerticalBox_0
   widget=EditorUtilityScrollBox_2
   slot {class=VerticalBoxSlot, properties {padding=<ltrb -10 0 -10 0>, size {sizeRule=Fill, value=0.5}}}
  }
  SelectChangeGroup {
   widgetClassPath=&VerticalBox
   bIsVariable=true
   parent=EditorUtilityScrollBox_2
   widget=SelectChangeGroup
   slot {class=ScrollBoxSlot, properties {padding=<ltrb 10 0 10 0>}}
  }
  SelectDataSectionBorder {
   widgetClassPath=&Border
   bIsVariable=true
   parent=SelectChangeGroup
   widget=SelectDataSectionBorder
   properties {background {
     bIsDynamicallyLoaded=false
     drawAs=Image
     imageType=NoImage
     mirroring=NoMirror
     resourceName=None
     resourceObject=null
     tiling=NoTile
     imageSize=<32 32>
     margin=<ltrb 0 0 0 0>
     outlineSettings {
      bUseBrushTransparency=false
      roundingType=HalfHeightRadius
      width=0
      color {colorUseRule=UseColor_Specified, specifiedColor=<rgba 0 0 0 0>}
      cornerRadii=<0 0 0 0>
     }
     tintColor {colorUseRule=UseColor_Specified, specifiedColor=<rgba 0.0729 0.0729 0.0729 1>}
     uVRegion {bIsValid=false, max=<0 0>, min=<0 0>}
    }}
   slot {class=VerticalBoxSlot, properties {padding=<ltrb -10 0 -10 0>}}
  }
  HorizontalBox_11 {widgetClassPath=&HorizontalBox, parent=SelectDataSectionBorder, widget=HorizontalBox_11, slot {class=BorderSlot}}
  SizeBox_0 {
   widgetClassPath=&SizeBox
   parent=HorizontalBox_11
   widget=SizeBox_0
   properties {bOverride_MinDesiredHeight=true, bOverride_MinDesiredWidth=true, maxDesiredWidth=14.4662, minDesiredHeight=20, minDesiredWidth=20}
   slot {class=HorizontalBoxSlot, properties {verticalAlignment=VAlign_Center, padding=<ltrb 0 0 5 0>}}
  }
  ToggleViewDa {
   widgetClassPath=&EditorUtilityCheckBox
   bIsVariable=true
   parent=SizeBox_0
   widget=ToggleViewDa
   properties {
    toolTipText=Preview Data
    widgetStyle {
     checkBoxType=ToggleButton
     backgroundHoveredImage {
      bIsDynamicallyLoaded=false
      drawAs=RoundedBox
      imageType=NoImage
      mirroring=NoMirror
      resourceName=None
      resourceObject=null
      tiling=NoTile
      imageSize=<18 18>
      margin=<ltrb 0 0 0 0>
      outlineSettings {
       bUseBrushTransparency=false
       roundingType=FixedRadius
       width=1
       color {colorUseRule=UseColor_ColorTable, specifiedColor=<rgba 0 0 0 0>}
       cornerRadii=<3 3 3 3>
      }
      tintColor {colorUseRule=UseColor_Foreground, specifiedColor=<rgba 0.0048 0.0048 0.0048 1>}
      uVRegion {bIsValid=false, max=<0 0>, min=<0 0>}
     }
     backgroundImage {
      bIsDynamicallyLoaded=false
      drawAs=RoundedBox
      imageType=NoImage
      mirroring=NoMirror
      resourceName=None
      resourceObject=null
      tiling=NoTile
      imageSize=<18 18>
      margin=<ltrb 0 0 0 0>
      outlineSettings {
       bUseBrushTransparency=false
       roundingType=FixedRadius
       width=1
       color {colorUseRule=UseColor_ColorTable, specifiedColor=<rgba 0 0 0 0>}
       cornerRadii=<3 3 3 3>
      }
      tintColor {colorUseRule=UseColor_Foreground, specifiedColor=<rgba 0.0048 0.0048 0.0048 1>}
      uVRegion {bIsValid=false, max=<0 0>, min=<0 0>}
     }
     backgroundPressedImage {
      bIsDynamicallyLoaded=false
      drawAs=RoundedBox
      imageType=NoImage
      mirroring=NoMirror
      resourceName=None
      resourceObject=null
      tiling=NoTile
      imageSize=<18 18>
      margin=<ltrb 0 0 0 0>
      outlineSettings {
       bUseBrushTransparency=false
       roundingType=FixedRadius
       width=1
       color {colorUseRule=UseColor_ColorTable, specifiedColor=<rgba 0 0 0 0>}
       cornerRadii=<3 3 3 3>
      }
      tintColor {colorUseRule=UseColor_Specified, specifiedColor=<rgba 0.0048 0.0048 0.0048 1>}
      uVRegion {bIsValid=false, max=<0 0>, min=<0 0>}
     }
     borderBackgroundColor {colorUseRule=UseColor_Specified, specifiedColor=<rgba 1 1 1 1>}
     checkedForeground {colorUseRule=UseColor_Specified, specifiedColor=<rgba 1 1 1 1>}
     checkedHoveredForeground {colorUseRule=UseColor_Specified, specifiedColor=<rgba 1 1 1 1>}
     checkedHoveredImage {
      bIsDynamicallyLoaded=false
      drawAs=Image
      imageType=NoImage
      mirroring=NoMirror
      resourceName=None
      resourceObject=/Game/Athena/UI/InGame/Creative/PropertyEditor/Icon_Picker/Textures_V2/64x64/T_UI_IconLibrary_Eye_64
      tiling=NoTile
      imageSize=<64 64>
      margin=<ltrb 0 0 0 0>
      outlineSettings {
       bUseBrushTransparency=false
       roundingType=HalfHeightRadius
       width=0
       color {colorUseRule=UseColor_Specified, specifiedColor=<rgba 0 0 0 0>}
       cornerRadii=<0 0 0 0>
      }
      tintColor {colorUseRule=UseColor_Specified, specifiedColor=<rgba 1 1 1 1>}
      uVRegion {bIsValid=false, max=<0 0>, min=<0 0>}
     }
     checkedImage {
      bIsDynamicallyLoaded=false
      drawAs=Image
      imageType=NoImage
      mirroring=NoMirror
      resourceName=None
      resourceObject=&T_UI_IconLibrary_Eye_64
      tiling=NoTile
      imageSize=<15 15>
      margin=<ltrb 0 0 0 0>
      outlineSettings {
       bUseBrushTransparency=false
       roundingType=HalfHeightRadius
       width=0
       color {colorUseRule=UseColor_Specified, specifiedColor=<rgba 0 0 0 0>}
       cornerRadii=<0 0 0 0>
      }
      tintColor {colorUseRule=UseColor_Specified, specifiedColor=<rgba 1 1 1 1>}
      uVRegion {bIsValid=false, max=<0 0>, min=<0 0>}
     }
     checkedPressedForeground {colorUseRule=UseColor_Specified, specifiedColor=<rgba 1 1 1 1>}
     checkedPressedImage {
      bIsDynamicallyLoaded=false
      drawAs=Image
      imageType=NoImage
      mirroring=NoMirror
      resourceName=None
      resourceObject=&T_UI_IconLibrary_Eye_64
      tiling=NoTile
      imageSize=<15 15>
      margin=<ltrb 0 0 0 0>
      outlineSettings {
       bUseBrushTransparency=false
       roundingType=HalfHeightRadius
       width=0
       color {colorUseRule=UseColor_Specified, specifiedColor=<rgba 0 0 0 0>}
       cornerRadii=<0 0 0 0>
      }
      tintColor {colorUseRule=UseColor_Specified, specifiedColor=<rgba 1 1 1 1>}
      uVRegion {bIsValid=false, max=<0 0>, min=<0 0>}
     }
     checkedSlateSound {resourceObject=null}
     foregroundColor {colorUseRule=UseColor_Specified, specifiedColor=<rgba 1 1 1 1>}
     hoveredForeground {colorUseRule=UseColor_Specified, specifiedColor=<rgba 1 1 1 1>}
     hoveredSlateSound {resourceObject=null}
     padding=<ltrb 2 0 0 0>
     pressedForeground {colorUseRule=UseColor_Specified, specifiedColor=<rgba 1 1 1 1>}
     uncheckedHoveredImage {
      bIsDynamicallyLoaded=false
      drawAs=RoundedBox
      imageType=Vector
      mirroring=NoMirror
      resourceName=../../../Engine/Content/Slate/Starship/CoreWidgets/CheckBox/indeterminate.svg
      resourceObject=null
      tiling=NoTile
      imageSize=<21.6344 15.3298>
      margin=<ltrb 0 0 0 0>
      outlineSettings {
       bUseBrushTransparency=false
       roundingType=HalfHeightRadius
       width=0
       color {colorUseRule=UseColor_Specified, specifiedColor=<rgba 0 0 0 0>}
       cornerRadii=<0 0 0 0>
      }
      tintColor {colorUseRule=UseColor_Specified, specifiedColor=<rgba 0.9378 1 0.6936 1>}
      uVRegion {bIsValid=false, max=<0 0>, min=<0 0>}
     }
     uncheckedImage {
      bIsDynamicallyLoaded=false
      drawAs=RoundedBox
      imageType=Vector
      mirroring=NoMirror
      resourceName=../../../Engine/Content/Slate/Starship/CoreWidgets/CheckBox/indeterminate.svg
      resourceObject=null
      tiling=NoTile
      imageSize=<21.6344 15.3298>
      margin=<ltrb 0 0 0 0>
      outlineSettings {
       bUseBrushTransparency=false
       roundingType=HalfHeightRadius
       width=0
       color {colorUseRule=UseColor_Specified, specifiedColor=<rgba 0 0 0 0>}
       cornerRadii=<0 0 0 0>
      }
      tintColor {colorUseRule=UseColor_Specified, specifiedColor=<rgba 1 1 1 1>}
      uVRegion {bIsValid=false, max=<0 0>, min=<0 0>}
     }
     uncheckedPressedImage {
      bIsDynamicallyLoaded=false
      drawAs=RoundedBox
      imageType=Vector
      mirroring=NoMirror
      resourceName=../../../Engine/Content/Slate/Starship/CoreWidgets/CheckBox/indeterminate.svg
      resourceObject=null
      tiling=NoTile
      imageSize=<21.6344 15.3298>
      margin=<ltrb 0 0 0 0>
      outlineSettings {
       bUseBrushTransparency=false
       roundingType=HalfHeightRadius
       width=0
       color {colorUseRule=UseColor_Specified, specifiedColor=<rgba 0 0 0 0>}
       cornerRadii=<0 0 0 0>
      }
      tintColor {colorUseRule=UseColor_Specified, specifiedColor=<rgba 0.5417 0.5417 0.5417 1>}
      uVRegion {bIsValid=false, max=<0 0>, min=<0 0>}
     }
     uncheckedSlateSound {resourceObject=null}
     undeterminedForeground {colorUseRule=UseColor_Specified, specifiedColor=<rgba 1 1 1 1>}
     undeterminedHoveredImage {
      bIsDynamicallyLoaded=false
      drawAs=Image
      imageType=Vector
      mirroring=NoMirror
      resourceName=../../../Engine/Content/Slate/Starship/CoreWidgets/CheckBox/indeterminate.svg
      resourceObject=null
      tiling=NoTile
      imageSize=<16 16>
      margin=<ltrb 0 0 0 0>
      outlineSettings {
       bUseBrushTransparency=false
       roundingType=HalfHeightRadius
       width=0
       color {colorUseRule=UseColor_Specified, specifiedColor=<rgba 0 0 0 0>}
       cornerRadii=<0 0 0 0>
      }
      tintColor {colorUseRule=UseColor_Specified, specifiedColor=<rgba 0.0044 0.2384 1 1>}
      uVRegion {bIsValid=false, max=<0 0>, min=<0 0>}
     }
     undeterminedImage {
      bIsDynamicallyLoaded=false
      drawAs=Image
      imageType=Vector
      mirroring=NoMirror
      resourceName=../../../Engine/Content/Slate/Starship/CoreWidgets/CheckBox/indeterminate.svg
      resourceObject=null
      tiling=NoTile
      imageSize=<16 16>
      margin=<ltrb 0 0 0 0>
      outlineSettings {
       bUseBrushTransparency=false
       roundingType=HalfHeightRadius
       width=0
       color {colorUseRule=UseColor_Specified, specifiedColor=<rgba 0 0 0 0>}
       cornerRadii=<0 0 0 0>
      }
      tintColor {colorUseRule=UseColor_Specified, specifiedColor=<rgba 0 0.162 0.7454 1>}
      uVRegion {bIsValid=false, max=<0 0>, min=<0 0>}
     }
     undeterminedPressedImage {
      bIsDynamicallyLoaded=false
      drawAs=Image
      imageType=Vector
      mirroring=NoMirror
      resourceName=../../../Engine/Content/Slate/Starship/CoreWidgets/CheckBox/indeterminate.svg
      resourceObject=null
      tiling=NoTile
      imageSize=<16 16>
      margin=<ltrb 0 0 0 0>
      outlineSettings {
       bUseBrushTransparency=false
       roundingType=HalfHeightRadius
       width=0
       color {colorUseRule=UseColor_Specified, specifiedColor=<rgba 0 0 0 0>}
       cornerRadii=<0 0 0 0>
      }
      tintColor {colorUseRule=UseColor_Specified, specifiedColor=<rgba 0 0.0802 0.3515 1>}
      uVRegion {bIsValid=false, max=<0 0>, min=<0 0>}
     }
    }
   }
   slot {class=SizeBoxSlot}
  }
  NewSequenceEventCellName {
   widgetClassPath=&TextBlock
   bIsVariable=true
   parent=HorizontalBox_11
   widget=NewSequenceEventCellName
   properties {
    text="Index [0]"
    font {
     bForceMonospaced=false
     bMaterialIsStencil=false
     fontMaterial=null
     fontObject=&HeadingNow
     letterSpacing=0
     monospacedWidth=1
     size=7.5
     skewAmount=0
     typefaceFontName=75 Medium
     outlineSettings {
      bApplyOutlineToDropShadows=false
      bMiteredCorners=false
      bSeparateFillAlpha=false
      outlineMaterial=null
      outlineSize=0
      outlineColor=<rgba 0 0 0 1>
     }
    }
   }
   slot {class=HorizontalBoxSlot, properties {horizontalAlignment=HAlign_Left, verticalAlignment=VAlign_Center, padding=<ltrb 0 0 5 0>}}
  }
  DaSlotProperty {
   widgetClassPath=&SinglePropertyView
   parent=HorizontalBox_11
   widget=DaSlotProperty
   properties {nameOverride=" ", propertyName=DaSlot, visibility=Collapsed}
   slot {class=HorizontalBoxSlot, properties {verticalAlignment=VAlign_Center}}
  }
  SizeBox_1 {
   widgetClassPath=&SizeBox
   parent=HorizontalBox_11
   widget=SizeBox_1
   properties {bOverride_MaxDesiredWidth=true, bOverride_MinDesiredWidth=true, maxDesiredWidth=350, minDesiredWidth=350}
   slot {class=HorizontalBoxSlot, properties {horizontalAlignment=HAlign_Left, verticalAlignment=VAlign_Center}}
  }
  DAFilteredComboBox {
   widgetClassPath=&SearchableStringComboBox
   bIsVariable=true
   parent=SizeBox_1
   widget=DAFilteredComboBox
   properties {
    enableGamepadNavigationMode=false
    selectedOption=aaaaaaaaaaaaaaaaaaaaaaaaa
    defaultOptions [aaaaaaaaaaaaaaaaaaaaaaaaa]
    font {
     bForceMonospaced=false
     bMaterialIsStencil=false
     fontMaterial=null
     fontObject=&HeadingNow
     letterSpacing=0
     monospacedWidth=1
     size=7.5
     skewAmount=0
     typefaceFontName=75 Medium
     outlineSettings {
      bApplyOutlineToDropShadows=false
      bMiteredCorners=false
      bSeparateFillAlpha=false
      outlineMaterial=null
      outlineSize=0
      outlineColor=<rgba 0 0 0 1>
     }
    }
    foregroundColor {colorUseRule=UseColor_Specified, specifiedColor=<rgba 0.5271 0.5271 0.5271 1>}
    itemStyle {
     bUseParentRowBrush=false
     activeBrush {
      bIsDynamicallyLoaded=false
      drawAs=Image
      imageType=NoImage
      mirroring=NoMirror
      resourceName=None
      resourceObject=null
      tiling=NoTile
      imageSize=<0 0>
      margin=<ltrb 0 0 0 0>
      outlineSettings {
       bUseBrushTransparency=false
       roundingType=HalfHeightRadius
       width=0
       color {colorUseRule=UseColor_Specified, specifiedColor=<rgba 0 0 0 0>}
       cornerRadii=<0 0 0 0>
      }
      tintColor {colorUseRule=UseColor_Specified, specifiedColor=<rgba 0 0.162 0.7454 1>}
      uVRegion {bIsValid=false, max=<0 0>, min=<0 0>}
     }
     activeHighlightedBrush {
      bIsDynamicallyLoaded=false
      drawAs=Image
      imageType=NoImage
      mirroring=NoMirror
      resourceName=None
      resourceObject=null
      tiling=NoTile
      imageSize=<0 0>
      margin=<ltrb 0 0 0 0>
      outlineSettings {
       bUseBrushTransparency=false
       roundingType=HalfHeightRadius
       width=0
       color {colorUseRule=UseColor_Specified, specifiedColor=<rgba 0 0 0 0>}
       cornerRadii=<0 0 0 0>
      }
      tintColor {colorUseRule=UseColor_Specified, specifiedColor=<rgba 0.0252 0.0319 0.0423 1>}
      uVRegion {bIsValid=false, max=<0 0>, min=<0 0>}
     }
     activeHoveredBrush {
      bIsDynamicallyLoaded=false
      drawAs=Image
      imageType=NoImage
      mirroring=NoMirror
      resourceName=None
      resourceObject=null
      tiling=NoTile
      imageSize=<0 0>
      margin=<ltrb 0 0 0 0>
      outlineSettings {
       bUseBrushTransparency=false
       roundingType=HalfHeightRadius
       width=0
       color {colorUseRule=UseColor_Specified, specifiedColor=<rgba 0 0 0 0>}
       cornerRadii=<0 0 0 0>
      }
      tintColor {colorUseRule=UseColor_Specified, specifiedColor=<rgba 0 0.162 0.7454 1>}
      uVRegion {bIsValid=false, max=<0 0>, min=<0 0>}
     }
     dropIndicator_Above {
      bIsDynamicallyLoaded=false
      drawAs=Box
      imageType=FullColor
      mirroring=NoMirror
      resourceName=../../../Engine/Content/Slate/Common/DropZoneIndicator_Above.png
      resourceObject=null
      tiling=NoTile
      imageSize=<0 0>
      margin=<ltrb 0.625 0.625 0 0>
      outlineSettings {
       bUseBrushTransparency=false
       roundingType=HalfHeightRadius
       width=0
       color {colorUseRule=UseColor_Specified, specifiedColor=<rgba 0 0 0 0>}
       cornerRadii=<0 0 0 0>
      }
      tintColor {colorUseRule=UseColor_Specified, specifiedColor=<rgba 0.0091 0.3515 0.9647 1>}
      uVRegion {bIsValid=false, max=<0 0>, min=<0 0>}
     }
     dropIndicator_Below {
      bIsDynamicallyLoaded=false
      drawAs=Box
      imageType=FullColor
      mirroring=NoMirror
      resourceName=../../../Engine/Content/Slate/Common/DropZoneIndicator_Below.png
      resourceObject=null
      tiling=NoTile
      imageSize=<0 0>
      margin=<ltrb 0.625 0 0 0.625>
      outlineSettings {
       bUseBrushTransparency=false
       roundingType=HalfHeightRadius
       width=0
       color {colorUseRule=UseColor_Specified, specifiedColor=<rgba 0 0 0 0>}
       cornerRadii=<0 0 0 0>
      }
      tintColor {colorUseRule=UseColor_Specified, specifiedColor=<rgba 0.0091 0.3515 0.9647 1>}
      uVRegion {bIsValid=false, max=<0 0>, min=<0 0>}
     }
     dropIndicator_Onto {
      bIsDynamicallyLoaded=false
      drawAs=Box
      imageType=FullColor
      mirroring=NoMirror
      resourceName=../../../Engine/Content/Slate/Common/DropZoneIndicator_Onto.png
      resourceObject=null
      tiling=NoTile
      imageSize=<0 0>
      margin=<ltrb 0.25 0.25 0.25 0.25>
      outlineSettings {
       bUseBrushTransparency=false
       roundingType=HalfHeightRadius
       width=0
       color {colorUseRule=UseColor_Specified, specifiedColor=<rgba 0 0 0 0>}
       cornerRadii=<0 0 0 0>
      }
      tintColor {colorUseRule=UseColor_Specified, specifiedColor=<rgba 0.0091 0.3515 0.9647 1>}
      uVRegion {bIsValid=false, max=<0 0>, min=<0 0>}
     }
     evenRowBackgroundBrush {
      bIsDynamicallyLoaded=false
      drawAs=Image
      imageType=NoImage
      mirroring=NoMirror
      resourceName=None
      resourceObject=null
      tiling=NoTile
      imageSize=<0 0>
      margin=<ltrb 0 0 0 0>
      outlineSettings {
       bUseBrushTransparency=false
       roundingType=HalfHeightRadius
       width=0
       color {colorUseRule=UseColor_Specified, specifiedColor=<rgba 0 0 0 0>}
       cornerRadii=<0 0 0 0>
      }
      tintColor {colorUseRule=UseColor_Specified, specifiedColor=<rgba 0.0103 0.0103 0.0103 1>}
      uVRegion {bIsValid=false, max=<0 0>, min=<0 0>}
     }
     evenRowBackgroundHoveredBrush {
      bIsDynamicallyLoaded=false
      drawAs=Image
      imageType=NoImage
      mirroring=NoMirror
      resourceName=None
      resourceObject=null
      tiling=NoTile
      imageSize=<0 0>
      margin=<ltrb 0 0 0 0>
      outlineSettings {
       bUseBrushTransparency=false
       roundingType=HalfHeightRadius
       width=0
       color {colorUseRule=UseColor_Specified, specifiedColor=<rgba 0 0 0 0>}
       cornerRadii=<0 0 0 0>
      }
      tintColor {colorUseRule=UseColor_Specified, specifiedColor=<rgba 0.0176 0.0176 0.0176 1>}
      uVRegion {bIsValid=false, max=<0 0>, min=<0 0>}
     }
     inactiveBrush {
      bIsDynamicallyLoaded=false
      drawAs=Image
      imageType=NoImage
      mirroring=NoMirror
      resourceName=None
      resourceObject=null
      tiling=NoTile
      imageSize=<0 0>
      margin=<ltrb 0 0 0 0>
      outlineSettings {
       bUseBrushTransparency=false
       roundingType=HalfHeightRadius
       width=0
       color {colorUseRule=UseColor_Specified, specifiedColor=<rgba 0 0 0 0>}
       cornerRadii=<0 0 0 0>
      }
      tintColor {colorUseRule=UseColor_Specified, specifiedColor=<rgba 0.0513 0.0953 0.159 1>}
      uVRegion {bIsValid=false, max=<0 0>, min=<0 0>}
     }
     inactiveHighlightedBrush {
      bIsDynamicallyLoaded=false
      drawAs=Image
      imageType=NoImage
      mirroring=NoMirror
      resourceName=None
      resourceObject=null
      tiling=NoTile
      imageSize=<0 0>
      margin=<ltrb 0 0 0 0>
      outlineSettings {
       bUseBrushTransparency=false
       roundingType=HalfHeightRadius
       width=0
       color {colorUseRule=UseColor_Specified, specifiedColor=<rgba 0 0 0 0>}
       cornerRadii=<0 0 0 0>
      }
      tintColor {colorUseRule=UseColor_Specified, specifiedColor=<rgba 0.0252 0.0319 0.0423 1>}
      uVRegion {bIsValid=false, max=<0 0>, min=<0 0>}
     }
     inactiveHoveredBrush {
      bIsDynamicallyLoaded=false
      drawAs=Image
      imageType=NoImage
      mirroring=NoMirror
      resourceName=None
      resourceObject=null
      tiling=NoTile
      imageSize=<0 0>
      margin=<ltrb 0 0 0 0>
      outlineSettings {
       bUseBrushTransparency=false
       roundingType=HalfHeightRadius
       width=0
       color {colorUseRule=UseColor_Specified, specifiedColor=<rgba 0 0 0 0>}
       cornerRadii=<0 0 0 0>
      }
      tintColor {colorUseRule=UseColor_Specified, specifiedColor=<rgba 0.0513 0.0953 0.159 1>}
      uVRegion {bIsValid=false, max=<0 0>, min=<0 0>}
     }
     oddRowBackgroundBrush {
      bIsDynamicallyLoaded=false
      drawAs=Image
      imageType=NoImage
      mirroring=NoMirror
      resourceName=None
      resourceObject=null
      tiling=NoTile
      imageSize=<0 0>
      margin=<ltrb 0 0 0 0>
      outlineSettings {
       bUseBrushTransparency=false
       roundingType=HalfHeightRadius
       width=0
       color {colorUseRule=UseColor_Specified, specifiedColor=<rgba 0 0 0 0>}
       cornerRadii=<0 0 0 0>
      }
      tintColor {colorUseRule=UseColor_Specified, specifiedColor=<rgba 0.0103 0.0103 0.0103 1>}
      uVRegion {bIsValid=false, max=<0 0>, min=<0 0>}
     }
     oddRowBackgroundHoveredBrush {
      bIsDynamicallyLoaded=false
      drawAs=Image
      imageType=NoImage
      mirroring=NoMirror
      resourceName=None
      resourceObject=null
      tiling=NoTile
      imageSize=<0 0>
      margin=<ltrb 0 0 0 0>
      outlineSettings {
       bUseBrushTransparency=false
       roundingType=HalfHeightRadius
       width=0
       color {colorUseRule=UseColor_Specified, specifiedColor=<rgba 0 0 0 0>}
       cornerRadii=<0 0 0 0>
      }
      tintColor {colorUseRule=UseColor_Specified, specifiedColor=<rgba 0.0176 0.0176 0.0176 1>}
      uVRegion {bIsValid=false, max=<0 0>, min=<0 0>}
     }
     parentRowBackgroundBrush {
      bIsDynamicallyLoaded=false
      drawAs=Image
      imageType=NoImage
      mirroring=NoMirror
      resourceName=None
      resourceObject=null
      tiling=NoTile
      imageSize=<32 32>
      margin=<ltrb 0 0 0 0>
      outlineSettings {
       bUseBrushTransparency=false
       roundingType=HalfHeightRadius
       width=0
       color {colorUseRule=UseColor_Specified, specifiedColor=<rgba 0 0 0 0>}
       cornerRadii=<0 0 0 0>
      }
      tintColor {colorUseRule=UseColor_Specified, specifiedColor=<rgba 1 1 1 1>}
      uVRegion {bIsValid=false, max=<0 0>, min=<0 0>}
     }
     parentRowBackgroundHoveredBrush {
      bIsDynamicallyLoaded=false
      drawAs=Image
      imageType=NoImage
      mirroring=NoMirror
      resourceName=None
      resourceObject=null
      tiling=NoTile
      imageSize=<32 32>
      margin=<ltrb 0 0 0 0>
      outlineSettings {
       bUseBrushTransparency=false
       roundingType=HalfHeightRadius
       width=0
       color {colorUseRule=UseColor_Specified, specifiedColor=<rgba 0 0 0 0>}
       cornerRadii=<0 0 0 0>
      }
      tintColor {colorUseRule=UseColor_Specified, specifiedColor=<rgba 1 1 1 1>}
      uVRegion {bIsValid=false, max=<0 0>, min=<0 0>}
     }
     selectedTextColor {colorUseRule=UseColor_Specified, specifiedColor=<rgba 1 1 1 1>}
     selectorFocusedBrush {
      bIsDynamicallyLoaded=false
      drawAs=Border
      imageType=FullColor
      mirroring=NoMirror
      resourceName=../../../Engine/Content/Slate/Common/Selector.png
      resourceObject=null
      tiling=Both
      imageSize=<0 0>
      margin=<ltrb 0.25 0.25 0.25 0.25>
      outlineSettings {
       bUseBrushTransparency=false
       roundingType=HalfHeightRadius
       width=0
       color {colorUseRule=UseColor_Specified, specifiedColor=<rgba 0 0 0 0>}
       cornerRadii=<0 0 0 0>
      }
      tintColor {colorUseRule=UseColor_Specified, specifiedColor=<rgba 0.701 0.225 0.003 1>}
      uVRegion {bIsValid=false, max=<0 0>, min=<0 0>}
     }
     textColor {colorUseRule=UseColor_Specified, specifiedColor=<rgba 0.5271 0.5271 0.5271 1>}
    }
    widgetStyle {
     comboButtonStyle {
      downArrowAlign=VAlign_Center
      buttonStyle {
       clickedSlateSound {resourceObject=null}
       disabled {
        bIsDynamicallyLoaded=false
        drawAs=RoundedBox
        imageType=NoImage
        mirroring=NoMirror
        resourceName=None
        resourceObject=null
        tiling=NoTile
        imageSize=<0 0>
        margin=<ltrb 0 0 0 0>
        outlineSettings {
         bUseBrushTransparency=false
         roundingType=FixedRadius
         width=1
         color {colorUseRule=UseColor_ColorTable, specifiedColor=<rgba 0 0 0 0>}
         cornerRadii=<4 4 4 4>
        }
        tintColor {colorUseRule=UseColor_Specified, specifiedColor=<rgba 0.0395 0.0395 0.0395 1>}
        uVRegion {bIsValid=false, max=<0 0>, min=<0 0>}
       }
       disabledForeground {colorUseRule=UseColor_ColorTable, specifiedColor=<rgba 0 0 0 0>}
       hovered {
        bIsDynamicallyLoaded=false
        drawAs=RoundedBox
        imageType=NoImage
        mirroring=NoMirror
        resourceName=None
        resourceObject=null
        tiling=NoTile
        imageSize=<0 0>
        margin=<ltrb 0 0 0 0>
        outlineSettings {
         bUseBrushTransparency=false
         roundingType=FixedRadius
         width=1
         color {colorUseRule=UseColor_ColorTable, specifiedColor=<rgba 0 0 0 0>}
         cornerRadii=<4 4 4 4>
        }
        tintColor {colorUseRule=UseColor_Specified, specifiedColor=<rgba 0.0048 0.0048 0.0048 1>}
        uVRegion {bIsValid=false, max=<0 0>, min=<0 0>}
       }
       hoveredForeground {colorUseRule=UseColor_ColorTable, specifiedColor=<rgba 0 0 0 0>}
       hoveredSlateSound {resourceObject=null}
       normal {
        bIsDynamicallyLoaded=false
        drawAs=RoundedBox
        imageType=NoImage
        mirroring=NoMirror
        resourceName=None
        resourceObject=null
        tiling=NoTile
        imageSize=<0 0>
        margin=<ltrb 0 0 0 0>
        outlineSettings {
         bUseBrushTransparency=false
         roundingType=FixedRadius
         width=1
         color {colorUseRule=UseColor_ColorTable, specifiedColor=<rgba 0 0 0 0>}
         cornerRadii=<4 4 4 4>
        }
        tintColor {colorUseRule=UseColor_Specified, specifiedColor=<rgba 0.0048 0.0048 0.0048 1>}
        uVRegion {bIsValid=false, max=<0 0>, min=<0 0>}
       }
       normalForeground {colorUseRule=UseColor_ColorTable, specifiedColor=<rgba 0 0 0 0>}
       normalPadding=<ltrb 8 1 8 1>
       pressed {
        bIsDynamicallyLoaded=false
        drawAs=RoundedBox
        imageType=NoImage
        mirroring=NoMirror
        resourceName=None
        resourceObject=null
        tiling=NoTile
        imageSize=<0 0>
        margin=<ltrb 0 0 0 0>
        outlineSettings {
         bUseBrushTransparency=false
         roundingType=FixedRadius
         width=1
         color {colorUseRule=UseColor_ColorTable, specifiedColor=<rgba 0 0 0 0>}
         cornerRadii=<4 4 4 4>
        }
        tintColor {colorUseRule=UseColor_Specified, specifiedColor=<rgba 0.0048 0.0048 0.0048 1>}
        uVRegion {bIsValid=false, max=<0 0>, min=<0 0>}
       }
       pressedForeground {colorUseRule=UseColor_ColorTable, specifiedColor=<rgba 0 0 0 0>}
       pressedPadding=<ltrb 8 2 8 0>
       pressedSlateSound {resourceObject=null}
      }
      contentPadding=<ltrb 0 0 0 0>
      downArrowImage {
       bIsDynamicallyLoaded=false
       drawAs=Image
       imageType=Vector
       mirroring=NoMirror
       resourceName=../../../Engine/Content/Slate/Starship/CoreWidgets/ComboBox/wide-chevron-down.svg
       resourceObject=null
       tiling=NoTile
       imageSize=<16 16>
       margin=<ltrb 0 0 0 0>
       outlineSettings {
        bUseBrushTransparency=false
        roundingType=HalfHeightRadius
        width=0
        color {colorUseRule=UseColor_Specified, specifiedColor=<rgba 0 0 0 0>}
        cornerRadii=<0 0 0 0>
       }
       tintColor {colorUseRule=UseColor_Specified, specifiedColor=<rgba 1 1 1 1>}
       uVRegion {bIsValid=false, max=<0 0>, min=<0 0>}
      }
      downArrowPadding=<ltrb 2 2 2 2>
      menuBorderBrush {
       bIsDynamicallyLoaded=false
       drawAs=RoundedBox
       imageType=NoImage
       mirroring=NoMirror
       resourceName=None
       resourceObject=null
       tiling=NoTile
       imageSize=<0 0>
       margin=<ltrb 0 0 0 0>
       outlineSettings {
        bUseBrushTransparency=false
        roundingType=FixedRadius
        width=1
        color {colorUseRule=UseColor_ColorTable, specifiedColor=<rgba 0 0 0 0>}
        cornerRadii=<0 0 0 0>
       }
       tintColor {colorUseRule=UseColor_Specified, specifiedColor=<rgba 0.0395 0.0395 0.0395 1>}
       uVRegion {bIsValid=false, max=<0 0>, min=<0 0>}
      }
      menuBorderPadding=<ltrb 0 0 0 0>
      shadowColorAndOpacity=<rgba 0 0 0 1>
      shadowOffset=<0 0>
     }
     contentPadding=<ltrb 0 0 0 0>
     menuRowPadding=<ltrb 8 3 8 3>
     pressedSlateSound {resourceObject=null}
     selectionChangeSlateSound {resourceObject=null}
    }
   }
   slot {class=SizeBoxSlot}
  }
  OpenEditorForEverything {
   widgetClassPath=&EditorUtilityButton
   parent=HorizontalBox_11
   widget=OpenEditorForEverything
   properties {
    toolTipText=Check
    widgetStyle {
     clickedSlateSound {resourceObject=null}
     disabled {
      bIsDynamicallyLoaded=false
      drawAs=RoundedBox
      imageType=NoImage
      mirroring=NoMirror
      resourceName=None
      resourceObject=&T-Button-Square
      tiling=NoTile
      imageSize=<32 32>
      margin=<ltrb 0.5 0.5 0.5 0.5>
      outlineSettings {
       bUseBrushTransparency=false
       roundingType=FixedRadius
       width=0
       color {colorUseRule=UseColor_ColorTable, specifiedColor=<rgba 0 0 0 0>}
       cornerRadii=<4 4 4 4>
      }
      tintColor {colorUseRule=UseColor_Specified, specifiedColor=<rgba 0.0395 0.0395 0.0395 1>}
      uVRegion {bIsValid=false, max=<0 0>, min=<0 0>}
     }
     disabledForeground {colorUseRule=UseColor_ColorTable, specifiedColor=<rgba 0 0 0 0>}
     hovered {
      bIsDynamicallyLoaded=false
      drawAs=RoundedBox
      imageType=NoImage
      mirroring=NoMirror
      resourceName=None
      resourceObject=&T-Button-Square
      tiling=NoTile
      imageSize=<32 32>
      margin=<ltrb 0.5 0.5 0.5 0.5>
      outlineSettings {
       bUseBrushTransparency=false
       roundingType=FixedRadius
       width=0
       color {colorUseRule=UseColor_ColorTable, specifiedColor=<rgba 0 0 0 0>}
       cornerRadii=<4 4 4 4>
      }
      tintColor {colorUseRule=UseColor_Specified, specifiedColor=<rgba 0 0.3021 0.0241 1>}
      uVRegion {bIsValid=false, max=<0 0>, min=<0 0>}
     }
     hoveredForeground {colorUseRule=UseColor_ColorTable, specifiedColor=<rgba 0 0 0 0>}
     hoveredSlateSound {resourceObject=null}
     normal {
      bIsDynamicallyLoaded=false
      drawAs=RoundedBox
      imageType=NoImage
      mirroring=NoMirror
      resourceName=None
      resourceObject=&T-Button-Square
      tiling=NoTile
      imageSize=<32 32>
      margin=<ltrb 0.5 0.5 0.5 0.5>
      outlineSettings {
       bUseBrushTransparency=false
       roundingType=FixedRadius
       width=0
       color {colorUseRule=UseColor_ColorTable, specifiedColor=<rgba 0 0 0 0>}
       cornerRadii=<4 4 4 4>
      }
      tintColor {colorUseRule=UseColor_Specified, specifiedColor=<rgba 0.0071 0.7594 0 1>}
      uVRegion {bIsValid=false, max=<0 0>, min=<0 0>}
     }
     normalForeground {colorUseRule=UseColor_ColorTable, specifiedColor=<rgba 0 0 0 0>}
     normalPadding=<ltrb 12 1.5 12 1.5>
     pressed {
      bIsDynamicallyLoaded=false
      drawAs=RoundedBox
      imageType=NoImage
      mirroring=NoMirror
      resourceName=None
      resourceObject=&T-Button-Square
      tiling=NoTile
      imageSize=<32 32>
      margin=<ltrb 0.5 0.5 0.5 0.5>
      outlineSettings {
       bUseBrushTransparency=false
       roundingType=FixedRadius
       width=0
       color {colorUseRule=UseColor_ColorTable, specifiedColor=<rgba 0 0 0 0>}
       cornerRadii=<4 4 4 4>
      }
      tintColor {colorUseRule=UseColor_Specified, specifiedColor=<rgba 0.3119 0.7917 0.2715 1>}
      uVRegion {bIsValid=false, max=<0 0>, min=<0 0>}
     }
     pressedForeground {colorUseRule=UseColor_ColorTable, specifiedColor=<rgba 0 0 0 0>}
     pressedPadding=<ltrb 12 2.5 12 0.5>
     pressedSlateSound {resourceObject=null}
    }
   }
   slot {class=HorizontalBoxSlot, properties {verticalAlignment=VAlign_Center, padding=<ltrb 5 5 5 5>}}
  }
  ExternalIcon_3 {
   widgetClassPath=&Image
   parent=OpenEditorForEverything
   widget=ExternalIcon_3
   properties {brush {
     bIsDynamicallyLoaded=false
     drawAs=Image
     imageType=NoImage
     mirroring=NoMirror
     resourceName=None
     resourceObject=&M_UI_Icon_ExternalLink
     tiling=NoTile
     imageSize=<20 20>
     margin=<ltrb 0 0 0 0>
     outlineSettings {
      bUseBrushTransparency=false
      roundingType=HalfHeightRadius
      width=0
      color {colorUseRule=UseColor_Specified, specifiedColor=<rgba 0 0 0 0>}
      cornerRadii=<0 0 0 0>
     }
     tintColor {colorUseRule=UseColor_Specified, specifiedColor=<rgba 1 1 1 1>}
     uVRegion {bIsValid=false, max=<0 0>, min=<0 0>}
    }}
   slot {class=ButtonSlot, properties {padding=<ltrb 0 0 0 0>}}
  }
  NewSequenceEventChangeButton {
   widgetClassPath=&EditorUtilityButton
   parent=HorizontalBox_11
   widget=NewSequenceEventChangeButton
   properties {widgetStyle {
     clickedSlateSound {resourceObject=null}
     disabled {
      bIsDynamicallyLoaded=false
      drawAs=RoundedBox
      imageType=NoImage
      mirroring=NoMirror
      resourceName=None
      resourceObject=null
      tiling=NoTile
      imageSize=<32 32>
      margin=<ltrb 0.5 0.5 0.5 0.5>
      outlineSettings {
       bUseBrushTransparency=false
       roundingType=FixedRadius
       width=0
       color {colorUseRule=UseColor_ColorTable, specifiedColor=<rgba 0 0 0 0>}
       cornerRadii=<4 4 4 4>
      }
      tintColor {colorUseRule=UseColor_Specified, specifiedColor=<rgba 0.0395 0.0395 0.0395 1>}
      uVRegion {bIsValid=false, max=<0 0>, min=<0 0>}
     }
     disabledForeground {colorUseRule=UseColor_ColorTable, specifiedColor=<rgba 0 0 0 0>}
     hovered {
      bIsDynamicallyLoaded=false
      drawAs=RoundedBox
      imageType=NoImage
      mirroring=NoMirror
      resourceName=None
      resourceObject=null
      tiling=NoTile
      imageSize=<32 32>
      margin=<ltrb 0 0 0 0>
      outlineSettings {
       bUseBrushTransparency=false
       roundingType=FixedRadius
       width=0
       color {colorUseRule=UseColor_ColorTable, specifiedColor=<rgba 0 0 0 0>}
       cornerRadii=<4 4 4 4>
      }
      tintColor {colorUseRule=UseColor_Specified, specifiedColor=<rgba 0.1126 0.7031 0 1>}
      uVRegion {bIsValid=false, max=<0 0>, min=<0 0>}
     }
     hoveredForeground {colorUseRule=UseColor_ColorTable, specifiedColor=<rgba 0 0 0 0>}
     hoveredSlateSound {resourceObject=null}
     normal {
      bIsDynamicallyLoaded=false
      drawAs=RoundedBox
      imageType=NoImage
      mirroring=NoMirror
      resourceName=None
      resourceObject=null
      tiling=NoTile
      imageSize=<32 32>
      margin=<ltrb 0.5 0.5 0.5 0.5>
      outlineSettings {
       bUseBrushTransparency=false
       roundingType=FixedRadius
       width=0
       color {colorUseRule=UseColor_ColorTable, specifiedColor=<rgba 0 0 0 0>}
       cornerRadii=<4 4 4 4>
      }
      tintColor {colorUseRule=UseColor_Specified, specifiedColor=<rgba 0.0771 0.6167 0.1289 1>}
      uVRegion {bIsValid=false, max=<0 0>, min=<0 0>}
     }
     normalForeground {colorUseRule=UseColor_ColorTable, specifiedColor=<rgba 0 0 0 0>}
     normalPadding=<ltrb 12 1.5 12 1.5>
     pressed {
      bIsDynamicallyLoaded=false
      drawAs=RoundedBox
      imageType=NoImage
      mirroring=NoMirror
      resourceName=None
      resourceObject=null
      tiling=NoTile
      imageSize=<32 32>
      margin=<ltrb 0.5 0.5 0.5 0.5>
      outlineSettings {
       bUseBrushTransparency=false
       roundingType=FixedRadius
       width=0
       color {colorUseRule=UseColor_ColorTable, specifiedColor=<rgba 0 0 0 0>}
       cornerRadii=<4 4 4 4>
      }
      tintColor {colorUseRule=UseColor_Specified, specifiedColor=<rgba 0.1806 0.4583 0.1572 1>}
      uVRegion {bIsValid=false, max=<0 0>, min=<0 0>}
     }
     pressedForeground {colorUseRule=UseColor_ColorTable, specifiedColor=<rgba 0 0 0 0>}
     pressedPadding=<ltrb 12 2.5 12 0.5>
     pressedSlateSound {resourceObject=null}
    }}
   slot {class=HorizontalBoxSlot, properties {verticalAlignment=VAlign_Center, padding=<ltrb 5 0 0 0>}}
  }
  HorizontalBox_12 {
   widgetClassPath=&HorizontalBox
   parent=NewSequenceEventChangeButton
   widget=HorizontalBox_12
   slot {class=ButtonSlot, properties {horizontalAlignment=HAlign_Left}}
  }
  Image_13 {
   widgetClassPath=&Image
   parent=HorizontalBox_12
   widget=Image_13
   properties {brush {
     bIsDynamicallyLoaded=false
     drawAs=Image
     imageType=NoImage
     mirroring=NoMirror
     resourceName=None
     resourceObject=&MI_SparksEditorUI_PlusMinus_NoBorder
     tiling=NoTile
     imageSize=<15 15>
     margin=<ltrb 0 0 0 0>
     outlineSettings {
      bUseBrushTransparency=false
      roundingType=HalfHeightRadius
      width=0
      color {colorUseRule=UseColor_Specified, specifiedColor=<rgba 0 0 0 0>}
      cornerRadii=<0 0 0 0>
     }
     tintColor {colorUseRule=UseColor_Specified, specifiedColor=<rgba 1 1 1 1>}
     uVRegion {bIsValid=false, max=<0 0>, min=<0 0>}
    }}
   slot {class=HorizontalBoxSlot, properties {horizontalAlignment=HAlign_Center, verticalAlignment=VAlign_Center, padding=<ltrb 0 0 3 0>}}
  }
  TextBlock_8 {
   widgetClassPath=&TextBlock
   parent=HorizontalBox_12
   widget=TextBlock_8
   properties {
    text=Replace to Selected
    font {
     bForceMonospaced=false
     bMaterialIsStencil=false
     fontMaterial=null
     fontObject=&HeadingNow
     letterSpacing=0
     monospacedWidth=1
     size=7.5
     skewAmount=0
     typefaceFontName=75 Medium
     outlineSettings {
      bApplyOutlineToDropShadows=false
      bMiteredCorners=false
      bSeparateFillAlpha=false
      outlineMaterial=null
      outlineSize=0
      outlineColor=<rgba 0 0 0 1>
     }
    }
   }
   slot {class=HorizontalBoxSlot, properties {verticalAlignment=VAlign_Center}}
  }
  ChooseAndAddEvent {
   widgetClassPath=&EditorUtilityButton
   parent=HorizontalBox_11
   widget=ChooseAndAddEvent
   properties {
    visibility=Collapsed
    widgetStyle {
     clickedSlateSound {resourceObject=null}
     disabled {
      bIsDynamicallyLoaded=false
      drawAs=RoundedBox
      imageType=NoImage
      mirroring=NoMirror
      resourceName=None
      resourceObject=null
      tiling=NoTile
      imageSize=<32 32>
      margin=<ltrb 0.5 0.5 0.5 0.5>
      outlineSettings {
       bUseBrushTransparency=false
       roundingType=FixedRadius
       width=0
       color {colorUseRule=UseColor_ColorTable, specifiedColor=<rgba 0 0 0 0>}
       cornerRadii=<4 4 4 4>
      }
      tintColor {colorUseRule=UseColor_Specified, specifiedColor=<rgba 0.0395 0.0395 0.0395 1>}
      uVRegion {bIsValid=false, max=<0 0>, min=<0 0>}
     }
     disabledForeground {colorUseRule=UseColor_ColorTable, specifiedColor=<rgba 0 0 0 0>}
     hovered {
      bIsDynamicallyLoaded=false
      drawAs=RoundedBox
      imageType=NoImage
      mirroring=NoMirror
      resourceName=None
      resourceObject=null
      tiling=NoTile
      imageSize=<32 32>
      margin=<ltrb 0 0 0 0>
      outlineSettings {
       bUseBrushTransparency=false
       roundingType=FixedRadius
       width=0
       color {colorUseRule=UseColor_ColorTable, specifiedColor=<rgba 0 0 0 0>}
       cornerRadii=<4 4 4 4>
      }
      tintColor {colorUseRule=UseColor_Specified, specifiedColor=<rgba 0.1126 0.7031 0 1>}
      uVRegion {bIsValid=false, max=<0 0>, min=<0 0>}
     }
     hoveredForeground {colorUseRule=UseColor_ColorTable, specifiedColor=<rgba 0 0 0 0>}
     hoveredSlateSound {resourceObject=null}
     normal {
      bIsDynamicallyLoaded=false
      drawAs=RoundedBox
      imageType=NoImage
      mirroring=NoMirror
      resourceName=None
      resourceObject=null
      tiling=NoTile
      imageSize=<32 32>
      margin=<ltrb 0.5 0.5 0.5 0.5>
      outlineSettings {
       bUseBrushTransparency=false
       roundingType=FixedRadius
       width=0
       color {colorUseRule=UseColor_ColorTable, specifiedColor=<rgba 0 0 0 0>}
       cornerRadii=<4 4 4 4>
      }
      tintColor {colorUseRule=UseColor_Specified, specifiedColor=<rgba 0.0771 0.6167 0.1289 1>}
      uVRegion {bIsValid=false, max=<0 0>, min=<0 0>}
     }
     normalForeground {colorUseRule=UseColor_ColorTable, specifiedColor=<rgba 0 0 0 0>}
     normalPadding=<ltrb 12 1.5 12 1.5>
     pressed {
      bIsDynamicallyLoaded=false
      drawAs=RoundedBox
      imageType=NoImage
      mirroring=NoMirror
      resourceName=None
      resourceObject=null
      tiling=NoTile
      imageSize=<32 32>
      margin=<ltrb 0.5 0.5 0.5 0.5>
      outlineSettings {
       bUseBrushTransparency=false
       roundingType=FixedRadius
       width=0
       color {colorUseRule=UseColor_ColorTable, specifiedColor=<rgba 0 0 0 0>}
       cornerRadii=<4 4 4 4>
      }
      tintColor {colorUseRule=UseColor_Specified, specifiedColor=<rgba 0.1806 0.4583 0.1572 1>}
      uVRegion {bIsValid=false, max=<0 0>, min=<0 0>}
     }
     pressedForeground {colorUseRule=UseColor_ColorTable, specifiedColor=<rgba 0 0 0 0>}
     pressedPadding=<ltrb 12 2.5 12 0.5>
     pressedSlateSound {resourceObject=null}
    }
   }
   slot {class=HorizontalBoxSlot, properties {verticalAlignment=VAlign_Center, padding=<ltrb 5 0 5 0>}}
  }
  HorizontalBox_17 {
   widgetClassPath=&HorizontalBox
   parent=ChooseAndAddEvent
   widget=HorizontalBox_17
   slot {class=ButtonSlot, properties {horizontalAlignment=HAlign_Left}}
  }
  Image_17 {
   widgetClassPath=&Image
   parent=HorizontalBox_17
   widget=Image_17
   properties {brush {
     bIsDynamicallyLoaded=false
     drawAs=Image
     imageType=NoImage
     mirroring=NoMirror
     resourceName=None
     resourceObject=&MI_SparksEditorUI_PlusMinus_NoBorder
     tiling=NoTile
     imageSize=<15 15>
     margin=<ltrb 0 0 0 0>
     outlineSettings {
      bUseBrushTransparency=false
      roundingType=HalfHeightRadius
      width=0
      color {colorUseRule=UseColor_Specified, specifiedColor=<rgba 0 0 0 0>}
      cornerRadii=<0 0 0 0>
     }
     tintColor {colorUseRule=UseColor_Specified, specifiedColor=<rgba 1 1 1 1>}
     uVRegion {bIsValid=false, max=<0 0>, min=<0 0>}
    }}
   slot {class=HorizontalBoxSlot, properties {horizontalAlignment=HAlign_Center, verticalAlignment=VAlign_Center, padding=<ltrb 0 0 3 0>}}
  }
  TextBlock_12 {
   widgetClassPath=&TextBlock
   parent=HorizontalBox_17
   widget=TextBlock_12
   properties {
    text=Choose and Add
    font {
     bForceMonospaced=false
     bMaterialIsStencil=false
     fontMaterial=null
     fontObject=&HeadingNow
     letterSpacing=0
     monospacedWidth=1
     size=7.5
     skewAmount=0
     typefaceFontName=75 Medium
     outlineSettings {
      bApplyOutlineToDropShadows=false
      bMiteredCorners=false
      bSeparateFillAlpha=false
      outlineMaterial=null
      outlineSize=0
      outlineColor=<rgba 0 0 0 1>
     }
    }
   }
   slot {class=HorizontalBoxSlot, properties {verticalAlignment=VAlign_Center}}
  }
  EventPreviewerGroup {
   widgetClassPath=&SizeBox
   bIsVariable=true
   parent=SelectChangeGroup
   widget=EventPreviewerGroup
   properties {bOverride_MaxDesiredHeight=true, maxDesiredHeight=300, visibility=Collapsed}
   slot {class=VerticalBoxSlot, properties {verticalAlignment=VAlign_Top, padding=<ltrb 0 10 0 5>, size {sizeRule=Fill, value=1}}}
  }
  Border_1 {
   widgetClassPath=&Border
   parent=EventPreviewerGroup
   widget=Border_1
   properties {brushColor=<rgba 0.0267 0.0312 0.0267 1>, contentColorAndOpacity=<rgba 0.6417 1 0.8525 1>, padding=<ltrb 4 2 4 3>}
   slot {class=SizeBoxSlot}
  }
  EditorUtilityScrollBox_0 {
   widgetClassPath=&EditorUtilityScrollBox
   parent=Border_1
   widget=EditorUtilityScrollBox_0
   slot {class=BorderSlot, properties {padding=<ltrb 4 2 4 3>}}
  }
  EventDetailView {
   widgetClassPath=&DetailsView
   parent=EditorUtilityScrollBox_0
   widget=EventDetailView
   properties {bIsEnabled=false, bShowScrollBar=false}
   slot {class=ScrollBoxSlot}
  }
  Border_6 {
   widgetClassPath=&Border
   parent=SelectChangeGroup
   widget=Border_6
   properties {
    background {
     bIsDynamicallyLoaded=false
     drawAs=Image
     imageType=NoImage
     mirroring=NoMirror
     resourceName=None
     resourceObject=null
     tiling=NoTile
     imageSize=<32 32>
     margin=<ltrb 0 0 0 0>
     outlineSettings {
      bUseBrushTransparency=false
      roundingType=HalfHeightRadius
      width=0
      color {colorUseRule=UseColor_Specified, specifiedColor=<rgba 0 0 0 0>}
      cornerRadii=<0 0 0 0>
     }
     tintColor {colorUseRule=UseColor_Specified, specifiedColor=<rgba 0.0729 0.0729 0.0729 1>}
     uVRegion {bIsValid=false, max=<0 0>, min=<0 0>}
    }
    padding=<ltrb 10 2 4 2>
   }
   slot {class=VerticalBoxSlot, properties {padding=<ltrb -10 0 -10 0>}}
  }
  HorizontalBox_16 {
   widgetClassPath=&HorizontalBox
   parent=Border_6
   widget=HorizontalBox_16
   slot {class=BorderSlot, properties {padding=<ltrb 10 2 4 2>}}
  }
  SizeBox {
   widgetClassPath=&SizeBox
   parent=HorizontalBox_16
   widget=SizeBox
   properties {bOverride_MaxDesiredWidth=true, bOverride_MinDesiredWidth=true, maxDesiredWidth=450, minDesiredWidth=450}
   slot {class=HorizontalBoxSlot}
  }
  RenameAssetTextBox {
   widgetClassPath=&EditorUtilityEditableTextBox
   bIsVariable=true
   parent=SizeBox
   widget=RenameAssetTextBox
   properties {
    hintText=Choose the new name.
    toolTipText=Choose a new asset name.
    widgetStyle {
     backgroundColor {colorUseRule=UseColor_Specified, specifiedColor=<rgba 1 1 1 1>}
     backgroundImageFocused {
      bIsDynamicallyLoaded=false
      drawAs=RoundedBox
      imageType=NoImage
      mirroring=NoMirror
      resourceName=None
      resourceObject=null
      tiling=NoTile
      imageSize=<0 0>
      margin=<ltrb 0 0 0 0>
      outlineSettings {
       bUseBrushTransparency=false
       roundingType=FixedRadius
       width=1
       color {colorUseRule=UseColor_ColorTable, specifiedColor=<rgba 0 0 0 0>}
       cornerRadii=<4 4 4 4>
      }
      tintColor {colorUseRule=UseColor_Specified, specifiedColor=<rgba 0.0048 0.0048 0.0048 1>}
      uVRegion {bIsValid=false, max=<0 0>, min=<0 0>}
     }
     backgroundImageHovered {
      bIsDynamicallyLoaded=false
      drawAs=RoundedBox
      imageType=NoImage
      mirroring=NoMirror
      resourceName=None
      resourceObject=null
      tiling=NoTile
      imageSize=<0 0>
      margin=<ltrb 0 0 0 0>
      outlineSettings {
       bUseBrushTransparency=false
       roundingType=FixedRadius
       width=1
       color {colorUseRule=UseColor_ColorTable, specifiedColor=<rgba 0 0 0 0>}
       cornerRadii=<4 4 4 4>
      }
      tintColor {colorUseRule=UseColor_Specified, specifiedColor=<rgba 0.0048 0.0048 0.0048 1>}
      uVRegion {bIsValid=false, max=<0 0>, min=<0 0>}
     }
     backgroundImageNormal {
      bIsDynamicallyLoaded=false
      drawAs=RoundedBox
      imageType=NoImage
      mirroring=NoMirror
      resourceName=None
      resourceObject=null
      tiling=NoTile
      imageSize=<0 0>
      margin=<ltrb 0 0 0 0>
      outlineSettings {
       bUseBrushTransparency=false
       roundingType=FixedRadius
       width=1
       color {colorUseRule=UseColor_ColorTable, specifiedColor=<rgba 0 0 0 0>}
       cornerRadii=<4 4 4 4>
      }
      tintColor {colorUseRule=UseColor_Specified, specifiedColor=<rgba 0.0048 0.0048 0.0048 1>}
      uVRegion {bIsValid=false, max=<0 0>, min=<0 0>}
     }
     backgroundImageReadOnly {
      bIsDynamicallyLoaded=false
      drawAs=RoundedBox
      imageType=NoImage
      mirroring=NoMirror
      resourceName=None
      resourceObject=null
      tiling=NoTile
      imageSize=<0 0>
      margin=<ltrb 0 0 0 0>
      outlineSettings {
       bUseBrushTransparency=false
       roundingType=FixedRadius
       width=1
       color {colorUseRule=UseColor_ColorTable, specifiedColor=<rgba 0 0 0 0>}
       cornerRadii=<4 4 4 4>
      }
      tintColor {colorUseRule=UseColor_Specified, specifiedColor=<rgba 0.0284 0.0284 0.0284 1>}
      uVRegion {bIsValid=false, max=<0 0>, min=<0 0>}
     }
     focusedForegroundColor {colorUseRule=UseColor_Specified, specifiedColor=<rgba 1 1 1 1>}
     foregroundColor {colorUseRule=UseColor_Specified, specifiedColor=<rgba 0.5271 0.5271 0.5271 1>}
     hScrollBarPadding=<ltrb 0 0 0 0>
     padding=<ltrb 8 3 8 4>
     readOnlyForegroundColor {colorUseRule=UseColor_Foreground, specifiedColor=<rgba 1 0 1 1>}
     scrollBarStyle {
      thickness=8
      draggedThumbImage {
       bIsDynamicallyLoaded=false
       drawAs=RoundedBox
       imageType=NoImage
       mirroring=NoMirror
       resourceName=None
       resourceObject=null
       tiling=NoTile
       imageSize=<0 0>
       margin=<ltrb 0 0 0 0>
       outlineSettings {
        bUseBrushTransparency=false
        roundingType=FixedRadius
        width=0
        color {colorUseRule=UseColor_Specified, specifiedColor=<rgba 0 0 0 0>}
        cornerRadii=<4 4 4 4>
       }
       tintColor {colorUseRule=UseColor_Specified, specifiedColor=<rgba 0.2159 0.2159 0.2159 1>}
       uVRegion {bIsValid=false, max=<0 0>, min=<0 0>}
      }
      horizontalBackgroundImage {
       bIsDynamicallyLoaded=false
       drawAs=NoDrawType
       imageType=NoImage
       mirroring=NoMirror
       resourceName=None
       resourceObject=null
       tiling=NoTile
       imageSize=<0 0>
       margin=<ltrb 0 0 0 0>
       outlineSettings {
        bUseBrushTransparency=false
        roundingType=HalfHeightRadius
        width=0
        color {colorUseRule=UseColor_Specified, specifiedColor=<rgba 0 0 0 0>}
        cornerRadii=<0 0 0 0>
       }
       tintColor {colorUseRule=UseColor_Specified, specifiedColor=<rgba 1 1 1 1>}
       uVRegion {bIsValid=false, max=<0 0>, min=<0 0>}
      }
      horizontalBottomSlotImage {
       bIsDynamicallyLoaded=false
       drawAs=NoDrawType
       imageType=NoImage
       mirroring=NoMirror
       resourceName=None
       resourceObject=null
       tiling=NoTile
       imageSize=<0 0>
       margin=<ltrb 0 0 0 0>
       outlineSettings {
        bUseBrushTransparency=false
        roundingType=HalfHeightRadius
        width=0
        color {colorUseRule=UseColor_Specified, specifiedColor=<rgba 0 0 0 0>}
        cornerRadii=<0 0 0 0>
       }
       tintColor {colorUseRule=UseColor_Specified, specifiedColor=<rgba 1 1 1 1>}
       uVRegion {bIsValid=false, max=<0 0>, min=<0 0>}
      }
      horizontalTopSlotImage {
       bIsDynamicallyLoaded=false
       drawAs=NoDrawType
       imageType=NoImage
       mirroring=NoMirror
       resourceName=None
       resourceObject=null
       tiling=NoTile
       imageSize=<0 0>
       margin=<ltrb 0 0 0 0>
       outlineSettings {
        bUseBrushTransparency=false
        roundingType=HalfHeightRadius
        width=0
        color {colorUseRule=UseColor_Specified, specifiedColor=<rgba 0 0 0 0>}
        cornerRadii=<0 0 0 0>
       }
       tintColor {colorUseRule=UseColor_Specified, specifiedColor=<rgba 1 1 1 1>}
       uVRegion {bIsValid=false, max=<0 0>, min=<0 0>}
      }
      hoveredThumbImage {
       bIsDynamicallyLoaded=false
       drawAs=RoundedBox
       imageType=NoImage
       mirroring=NoMirror
       resourceName=None
       resourceObject=null
       tiling=NoTile
       imageSize=<0 0>
       margin=<ltrb 0 0 0 0>
       outlineSettings {
        bUseBrushTransparency=false
        roundingType=FixedRadius
        width=0
        color {colorUseRule=UseColor_Specified, specifiedColor=<rgba 0 0 0 0>}
        cornerRadii=<4 4 4 4>
       }
       tintColor {colorUseRule=UseColor_Specified, specifiedColor=<rgba 0.2159 0.2159 0.2159 1>}
       uVRegion {bIsValid=false, max=<0 0>, min=<0 0>}
      }
      normalThumbImage {
       bIsDynamicallyLoaded=false
       drawAs=RoundedBox
       imageType=NoImage
       mirroring=NoMirror
       resourceName=None
       resourceObject=null
       tiling=NoTile
       imageSize=<0 0>
       margin=<ltrb 0 0 0 0>
       outlineSettings {
        bUseBrushTransparency=false
        roundingType=FixedRadius
        width=0
        color {colorUseRule=UseColor_Specified, specifiedColor=<rgba 0 0 0 0>}
        cornerRadii=<4 4 4 4>
       }
       tintColor {colorUseRule=UseColor_Specified, specifiedColor=<rgba 0.0953 0.0953 0.0953 1>}
       uVRegion {bIsValid=false, max=<0 0>, min=<0 0>}
      }
      verticalBackgroundImage {
       bIsDynamicallyLoaded=false
       drawAs=NoDrawType
       imageType=NoImage
       mirroring=NoMirror
       resourceName=None
       resourceObject=null
       tiling=NoTile
       imageSize=<0 0>
       margin=<ltrb 0 0 0 0>
       outlineSettings {
        bUseBrushTransparency=false
        roundingType=HalfHeightRadius
        width=0
        color {colorUseRule=UseColor_Specified, specifiedColor=<rgba 0 0 0 0>}
        cornerRadii=<0 0 0 0>
       }
       tintColor {colorUseRule=UseColor_Specified, specifiedColor=<rgba 1 1 1 1>}
       uVRegion {bIsValid=false, max=<0 0>, min=<0 0>}
      }
      verticalBottomSlotImage {
       bIsDynamicallyLoaded=false
       drawAs=NoDrawType
       imageType=NoImage
       mirroring=NoMirror
       resourceName=None
       resourceObject=null
       tiling=NoTile
       imageSize=<0 0>
       margin=<ltrb 0 0 0 0>
       outlineSettings {
        bUseBrushTransparency=false
        roundingType=HalfHeightRadius
        width=0
        color {colorUseRule=UseColor_Specified, specifiedColor=<rgba 0 0 0 0>}
        cornerRadii=<0 0 0 0>
       }
       tintColor {colorUseRule=UseColor_Specified, specifiedColor=<rgba 1 1 1 1>}
       uVRegion {bIsValid=false, max=<0 0>, min=<0 0>}
      }
      verticalTopSlotImage {
       bIsDynamicallyLoaded=false
       drawAs=NoDrawType
       imageType=NoImage
       mirroring=NoMirror
       resourceName=None
       resourceObject=null
       tiling=NoTile
       imageSize=<0 0>
       margin=<ltrb 0 0 0 0>
       outlineSettings {
        bUseBrushTransparency=false
        roundingType=HalfHeightRadius
        width=0
        color {colorUseRule=UseColor_Specified, specifiedColor=<rgba 0 0 0 0>}
        cornerRadii=<0 0 0 0>
       }
       tintColor {colorUseRule=UseColor_Specified, specifiedColor=<rgba 1 1 1 1>}
       uVRegion {bIsValid=false, max=<0 0>, min=<0 0>}
      }
     }
     textStyle {
      overflowPolicy=Clip
      transformPolicy=None
      colorAndOpacity {colorUseRule=UseColor_Foreground, specifiedColor=<rgba 1 0 1 1>}
      font {
       bForceMonospaced=false
       bMaterialIsStencil=false
       fontMaterial=null
       fontObject=&HeadingNow
       letterSpacing=0
       monospacedWidth=1
       size=7.5
       skewAmount=0
       typefaceFontName=75 Medium
       outlineSettings {
        bApplyOutlineToDropShadows=false
        bMiteredCorners=false
        bSeparateFillAlpha=false
        outlineMaterial=null
        outlineSize=0
        outlineColor=<rgba 0 0 0 1>
       }
      }
      highlightColor {colorUseRule=UseColor_ColorTable, specifiedColor=<rgba 0 0 0 0>}
      highlightShape {
       bIsDynamicallyLoaded=false
       drawAs=Image
       imageType=NoImage
       mirroring=NoMirror
       resourceName=None
       resourceObject=null
       tiling=NoTile
       imageSize=<0 0>
       margin=<ltrb 0 0 0 0>
       outlineSettings {
        bUseBrushTransparency=false
        roundingType=HalfHeightRadius
        width=0
        color {colorUseRule=UseColor_Specified, specifiedColor=<rgba 0 0 0 0>}
        cornerRadii=<0 0 0 0>
       }
       tintColor {colorUseRule=UseColor_ColorTable, specifiedColor=<rgba 0 0 0 0>}
       uVRegion {bIsValid=false, max=<0 0>, min=<0 0>}
      }
      selectedBackgroundColor {colorUseRule=UseColor_ColorTable, specifiedColor=<rgba 0 0 0 0>}
      shadowColorAndOpacity=<rgba 0 0 0 1>
      shadowOffset=<0 0>
      strikeBrush {
       bIsDynamicallyLoaded=false
       drawAs=Image
       imageType=NoImage
       mirroring=NoMirror
       resourceName=None
       resourceObject=null
       tiling=NoTile
       imageSize=<32 32>
       margin=<ltrb 0 0 0 0>
       outlineSettings {
        bUseBrushTransparency=false
        roundingType=HalfHeightRadius
        width=0
        color {colorUseRule=UseColor_Specified, specifiedColor=<rgba 0 0 0 0>}
        cornerRadii=<0 0 0 0>
       }
       tintColor {colorUseRule=UseColor_Specified, specifiedColor=<rgba 1 1 1 1>}
       uVRegion {bIsValid=false, max=<0 0>, min=<0 0>}
      }
      underlineBrush {
       bIsDynamicallyLoaded=false
       drawAs=Image
       imageType=NoImage
       mirroring=NoMirror
       resourceName=None
       resourceObject=null
       tiling=NoTile
       imageSize=<32 32>
       margin=<ltrb 0 0 0 0>
       outlineSettings {
        bUseBrushTransparency=false
        roundingType=HalfHeightRadius
        width=0
        color {colorUseRule=UseColor_Specified, specifiedColor=<rgba 0 0 0 0>}
        cornerRadii=<0 0 0 0>
       }
       tintColor {colorUseRule=UseColor_Specified, specifiedColor=<rgba 1 1 1 1>}
       uVRegion {bIsValid=false, max=<0 0>, min=<0 0>}
      }
     }
     vScrollBarPadding=<ltrb 0 0 0 0>
    }
   }
   slot {class=SizeBoxSlot}
  }
  RenameEventButton {
   widgetClassPath=&EditorUtilityButton
   parent=HorizontalBox_16
   widget=RenameEventButton
   properties {widgetStyle {
     clickedSlateSound {resourceObject=null}
     disabled {
      bIsDynamicallyLoaded=false
      drawAs=RoundedBox
      imageType=NoImage
      mirroring=NoMirror
      resourceName=None
      resourceObject=null
      tiling=NoTile
      imageSize=<32 32>
      margin=<ltrb 0.5 0.5 0.5 0.5>
      outlineSettings {
       bUseBrushTransparency=false
       roundingType=FixedRadius
       width=0
       color {colorUseRule=UseColor_ColorTable, specifiedColor=<rgba 0 0 0 0>}
       cornerRadii=<4 4 4 4>
      }
      tintColor {colorUseRule=UseColor_Specified, specifiedColor=<rgba 0.0395 0.0395 0.0395 1>}
      uVRegion {bIsValid=false, max=<0 0>, min=<0 0>}
     }
     disabledForeground {colorUseRule=UseColor_ColorTable, specifiedColor=<rgba 0 0 0 0>}
     hovered {
      bIsDynamicallyLoaded=false
      drawAs=RoundedBox
      imageType=NoImage
      mirroring=NoMirror
      resourceName=None
      resourceObject=null
      tiling=NoTile
      imageSize=<32 32>
      margin=<ltrb 0 0 0 0>
      outlineSettings {
       bUseBrushTransparency=false
       roundingType=FixedRadius
       width=0
       color {colorUseRule=UseColor_ColorTable, specifiedColor=<rgba 0 0 0 0>}
       cornerRadii=<4 4 4 4>
      }
      tintColor {colorUseRule=UseColor_Specified, specifiedColor=<rgba 0.8542 0.7778 0.0801 1>}
      uVRegion {bIsValid=false, max=<0 0>, min=<0 0>}
     }
     hoveredForeground {colorUseRule=UseColor_ColorTable, specifiedColor=<rgba 0 0 0 0>}
     hoveredSlateSound {resourceObject=null}
     normal {
      bIsDynamicallyLoaded=false
      drawAs=RoundedBox
      imageType=NoImage
      mirroring=NoMirror
      resourceName=None
      resourceObject=null
      tiling=NoTile
      imageSize=<32 32>
      margin=<ltrb 0.5 0.5 0.5 0.5>
      outlineSettings {
       bUseBrushTransparency=false
       roundingType=FixedRadius
       width=0
       color {colorUseRule=UseColor_ColorTable, specifiedColor=<rgba 0 0 0 0>}
       cornerRadii=<4 4 4 4>
      }
      tintColor {colorUseRule=UseColor_Specified, specifiedColor=<rgba 0.7031 0.6338 0 1>}
      uVRegion {bIsValid=false, max=<0 0>, min=<0 0>}
     }
     normalForeground {colorUseRule=UseColor_ColorTable, specifiedColor=<rgba 0 0 0 0>}
     normalPadding=<ltrb 12 1.5 12 1.5>
     pressed {
      bIsDynamicallyLoaded=false
      drawAs=RoundedBox
      imageType=NoImage
      mirroring=NoMirror
      resourceName=None
      resourceObject=null
      tiling=NoTile
      imageSize=<32 32>
      margin=<ltrb 0.5 0.5 0.5 0.5>
      outlineSettings {
       bUseBrushTransparency=false
       roundingType=FixedRadius
       width=0
       color {colorUseRule=UseColor_ColorTable, specifiedColor=<rgba 0 0 0 0>}
       cornerRadii=<4 4 4 4>
      }
      tintColor {colorUseRule=UseColor_Specified, specifiedColor=<rgba 1 0.2833 0 1>}
      uVRegion {bIsValid=false, max=<0 0>, min=<0 0>}
     }
     pressedForeground {colorUseRule=UseColor_ColorTable, specifiedColor=<rgba 0 0 0 0>}
     pressedPadding=<ltrb 12 2.5 12 0.5>
     pressedSlateSound {resourceObject=null}
    }}
   slot {class=HorizontalBoxSlot, properties {verticalAlignment=VAlign_Center, padding=<ltrb 5 0 5 0>}}
  }
  HorizontalBox_15 {
   widgetClassPath=&HorizontalBox
   parent=RenameEventButton
   widget=HorizontalBox_15
   slot {class=ButtonSlot, properties {horizontalAlignment=HAlign_Left}}
  }
  Image_10 {
   widgetClassPath=&Image
   parent=HorizontalBox_15
   widget=Image_10
   properties {
    visibility=Collapsed
    brush {
     bIsDynamicallyLoaded=false
     drawAs=Image
     imageType=NoImage
     mirroring=NoMirror
     resourceName=None
     resourceObject=&MI_SparksEditorUI_PlusMinus_NoBorder
     tiling=NoTile
     imageSize=<15 15>
     margin=<ltrb 0 0 0 0>
     outlineSettings {
      bUseBrushTransparency=false
      roundingType=HalfHeightRadius
      width=0
      color {colorUseRule=UseColor_Specified, specifiedColor=<rgba 0 0 0 0>}
      cornerRadii=<0 0 0 0>
     }
     tintColor {colorUseRule=UseColor_Specified, specifiedColor=<rgba 1 1 1 1>}
     uVRegion {bIsValid=false, max=<0 0>, min=<0 0>}
    }
   }
   slot {class=HorizontalBoxSlot, properties {horizontalAlignment=HAlign_Center, verticalAlignment=VAlign_Center, padding=<ltrb 0 0 3 0>}}
  }
  TextBlock_11 {
   widgetClassPath=&TextBlock
   parent=HorizontalBox_15
   widget=TextBlock_11
   properties {
    text=Rename
    font {
     bForceMonospaced=false
     bMaterialIsStencil=false
     fontMaterial=null
     fontObject=&HeadingNow
     letterSpacing=0
     monospacedWidth=1
     size=7.5
     skewAmount=0
     typefaceFontName=75 Medium
     outlineSettings {
      bApplyOutlineToDropShadows=false
      bMiteredCorners=false
      bSeparateFillAlpha=false
      outlineMaterial=null
      outlineSize=0
      outlineColor=<rgba 0 0 0 1>
     }
    }
   }
   slot {class=HorizontalBoxSlot, properties {verticalAlignment=VAlign_Center}}
  }
  CreateNewDAGroup {
   widgetClassPath=&VerticalBox
   bIsVariable=true
   parent=EditorUtilityScrollBox_2
   widget=CreateNewDAGroup
   properties {visibility=Collapsed}
   slot {class=ScrollBoxSlot, properties {padding=<ltrb 10 0 10 0>}}
  }
  Border_4 {
   widgetClassPath=&Border
   parent=CreateNewDAGroup
   widget=Border_4
   properties {
    background {
     bIsDynamicallyLoaded=false
     drawAs=Image
     imageType=NoImage
     mirroring=NoMirror
     resourceName=None
     resourceObject=null
     tiling=NoTile
     imageSize=<32 32>
     margin=<ltrb 0 0 0 0>
     outlineSettings {
      bUseBrushTransparency=false
      roundingType=HalfHeightRadius
      width=0
      color {colorUseRule=UseColor_Specified, specifiedColor=<rgba 0 0 0 0>}
      cornerRadii=<0 0 0 0>
     }
     tintColor {colorUseRule=UseColor_Specified, specifiedColor=<rgba 0.0729 0.0729 0.0729 1>}
     uVRegion {bIsValid=false, max=<0 0>, min=<0 0>}
    }
    padding=<ltrb 10 2 4 2>
   }
   slot {class=VerticalBoxSlot, properties {padding=<ltrb -10 0 -10 0>}}
  }
  HorizontalBox_13 {
   widgetClassPath=&HorizontalBox
   parent=Border_4
   widget=HorizontalBox_13
   slot {class=BorderSlot, properties {padding=<ltrb 10 2 4 2>}}
  }
  SizeBox_3 {
   widgetClassPath=&SizeBox
   parent=HorizontalBox_13
   widget=SizeBox_3
   properties {bOverride_MaxDesiredWidth=true, bOverride_MinDesiredWidth=true, maxDesiredWidth=240, minDesiredWidth=240}
   slot {class=HorizontalBoxSlot}
  }
  NewFxManagerNameTextBox {
   widgetClassPath=&EditorUtilityEditableTextBox
   bIsVariable=true
   parent=SizeBox_3
   widget=NewFxManagerNameTextBox
   properties {
    hintText=Insert Type if you want
    toolTipText=You can choose a different type than the default preview one if you want.
    widgetStyle {
     backgroundColor {colorUseRule=UseColor_Specified, specifiedColor=<rgba 1 1 1 1>}
     backgroundImageFocused {
      bIsDynamicallyLoaded=false
      drawAs=RoundedBox
      imageType=NoImage
      mirroring=NoMirror
      resourceName=None
      resourceObject=null
      tiling=NoTile
      imageSize=<0 0>
      margin=<ltrb 0 0 0 0>
      outlineSettings {
       bUseBrushTransparency=false
       roundingType=FixedRadius
       width=1
       color {colorUseRule=UseColor_ColorTable, specifiedColor=<rgba 0 0 0 0>}
       cornerRadii=<4 4 4 4>
      }
      tintColor {colorUseRule=UseColor_Specified, specifiedColor=<rgba 0.0048 0.0048 0.0048 1>}
      uVRegion {bIsValid=false, max=<0 0>, min=<0 0>}
     }
     backgroundImageHovered {
      bIsDynamicallyLoaded=false
      drawAs=RoundedBox
      imageType=NoImage
      mirroring=NoMirror
      resourceName=None
      resourceObject=null
      tiling=NoTile
      imageSize=<0 0>
      margin=<ltrb 0 0 0 0>
      outlineSettings {
       bUseBrushTransparency=false
       roundingType=FixedRadius
       width=1
       color {colorUseRule=UseColor_ColorTable, specifiedColor=<rgba 0 0 0 0>}
       cornerRadii=<4 4 4 4>
      }
      tintColor {colorUseRule=UseColor_Specified, specifiedColor=<rgba 0.0048 0.0048 0.0048 1>}
      uVRegion {bIsValid=false, max=<0 0>, min=<0 0>}
     }
     backgroundImageNormal {
      bIsDynamicallyLoaded=false
      drawAs=RoundedBox
      imageType=NoImage
      mirroring=NoMirror
      resourceName=None
      resourceObject=null
      tiling=NoTile
      imageSize=<0 0>
      margin=<ltrb 0 0 0 0>
      outlineSettings {
       bUseBrushTransparency=false
       roundingType=FixedRadius
       width=1
       color {colorUseRule=UseColor_ColorTable, specifiedColor=<rgba 0 0 0 0>}
       cornerRadii=<4 4 4 4>
      }
      tintColor {colorUseRule=UseColor_Specified, specifiedColor=<rgba 0.0048 0.0048 0.0048 1>}
      uVRegion {bIsValid=false, max=<0 0>, min=<0 0>}
     }
     backgroundImageReadOnly {
      bIsDynamicallyLoaded=false
      drawAs=RoundedBox
      imageType=NoImage
      mirroring=NoMirror
      resourceName=None
      resourceObject=null
      tiling=NoTile
      imageSize=<0 0>
      margin=<ltrb 0 0 0 0>
      outlineSettings {
       bUseBrushTransparency=false
       roundingType=FixedRadius
       width=1
       color {colorUseRule=UseColor_ColorTable, specifiedColor=<rgba 0 0 0 0>}
       cornerRadii=<4 4 4 4>
      }
      tintColor {colorUseRule=UseColor_Specified, specifiedColor=<rgba 0.0284 0.0284 0.0284 1>}
      uVRegion {bIsValid=false, max=<0 0>, min=<0 0>}
     }
     focusedForegroundColor {colorUseRule=UseColor_Specified, specifiedColor=<rgba 1 1 1 1>}
     foregroundColor {colorUseRule=UseColor_Specified, specifiedColor=<rgba 0.5271 0.5271 0.5271 1>}
     hScrollBarPadding=<ltrb 0 0 0 0>
     padding=<ltrb 8 3 8 4>
     readOnlyForegroundColor {colorUseRule=UseColor_Foreground, specifiedColor=<rgba 1 0 1 1>}
     scrollBarStyle {
      thickness=8
      draggedThumbImage {
       bIsDynamicallyLoaded=false
       drawAs=RoundedBox
       imageType=NoImage
       mirroring=NoMirror
       resourceName=None
       resourceObject=null
       tiling=NoTile
       imageSize=<0 0>
       margin=<ltrb 0 0 0 0>
       outlineSettings {
        bUseBrushTransparency=false
        roundingType=FixedRadius
        width=0
        color {colorUseRule=UseColor_Specified, specifiedColor=<rgba 0 0 0 0>}
        cornerRadii=<4 4 4 4>
       }
       tintColor {colorUseRule=UseColor_Specified, specifiedColor=<rgba 0.2159 0.2159 0.2159 1>}
       uVRegion {bIsValid=false, max=<0 0>, min=<0 0>}
      }
      horizontalBackgroundImage {
       bIsDynamicallyLoaded=false
       drawAs=NoDrawType
       imageType=NoImage
       mirroring=NoMirror
       resourceName=None
       resourceObject=null
       tiling=NoTile
       imageSize=<0 0>
       margin=<ltrb 0 0 0 0>
       outlineSettings {
        bUseBrushTransparency=false
        roundingType=HalfHeightRadius
        width=0
        color {colorUseRule=UseColor_Specified, specifiedColor=<rgba 0 0 0 0>}
        cornerRadii=<0 0 0 0>
       }
       tintColor {colorUseRule=UseColor_Specified, specifiedColor=<rgba 1 1 1 1>}
       uVRegion {bIsValid=false, max=<0 0>, min=<0 0>}
      }
      horizontalBottomSlotImage {
       bIsDynamicallyLoaded=false
       drawAs=NoDrawType
       imageType=NoImage
       mirroring=NoMirror
       resourceName=None
       resourceObject=null
       tiling=NoTile
       imageSize=<0 0>
       margin=<ltrb 0 0 0 0>
       outlineSettings {
        bUseBrushTransparency=false
        roundingType=HalfHeightRadius
        width=0
        color {colorUseRule=UseColor_Specified, specifiedColor=<rgba 0 0 0 0>}
        cornerRadii=<0 0 0 0>
       }
       tintColor {colorUseRule=UseColor_Specified, specifiedColor=<rgba 1 1 1 1>}
       uVRegion {bIsValid=false, max=<0 0>, min=<0 0>}
      }
      horizontalTopSlotImage {
       bIsDynamicallyLoaded=false
       drawAs=NoDrawType
       imageType=NoImage
       mirroring=NoMirror
       resourceName=None
       resourceObject=null
       tiling=NoTile
       imageSize=<0 0>
       margin=<ltrb 0 0 0 0>
       outlineSettings {
        bUseBrushTransparency=false
        roundingType=HalfHeightRadius
        width=0
        color {colorUseRule=UseColor_Specified, specifiedColor=<rgba 0 0 0 0>}
        cornerRadii=<0 0 0 0>
       }
       tintColor {colorUseRule=UseColor_Specified, specifiedColor=<rgba 1 1 1 1>}
       uVRegion {bIsValid=false, max=<0 0>, min=<0 0>}
      }
      hoveredThumbImage {
       bIsDynamicallyLoaded=false
       drawAs=RoundedBox
       imageType=NoImage
       mirroring=NoMirror
       resourceName=None
       resourceObject=null
       tiling=NoTile
       imageSize=<0 0>
       margin=<ltrb 0 0 0 0>
       outlineSettings {
        bUseBrushTransparency=false
        roundingType=FixedRadius
        width=0
        color {colorUseRule=UseColor_Specified, specifiedColor=<rgba 0 0 0 0>}
        cornerRadii=<4 4 4 4>
       }
       tintColor {colorUseRule=UseColor_Specified, specifiedColor=<rgba 0.2159 0.2159 0.2159 1>}
       uVRegion {bIsValid=false, max=<0 0>, min=<0 0>}
      }
      normalThumbImage {
       bIsDynamicallyLoaded=false
       drawAs=RoundedBox
       imageType=NoImage
       mirroring=NoMirror
       resourceName=None
       resourceObject=null
       tiling=NoTile
       imageSize=<0 0>
       margin=<ltrb 0 0 0 0>
       outlineSettings {
        bUseBrushTransparency=false
        roundingType=FixedRadius
        width=0
        color {colorUseRule=UseColor_Specified, specifiedColor=<rgba 0 0 0 0>}
        cornerRadii=<4 4 4 4>
       }
       tintColor {colorUseRule=UseColor_Specified, specifiedColor=<rgba 0.0953 0.0953 0.0953 1>}
       uVRegion {bIsValid=false, max=<0 0>, min=<0 0>}
      }
      verticalBackgroundImage {
       bIsDynamicallyLoaded=false
       drawAs=NoDrawType
       imageType=NoImage
       mirroring=NoMirror
       resourceName=None
       resourceObject=null
       tiling=NoTile
       imageSize=<0 0>
       margin=<ltrb 0 0 0 0>
       outlineSettings {
        bUseBrushTransparency=false
        roundingType=HalfHeightRadius
        width=0
        color {colorUseRule=UseColor_Specified, specifiedColor=<rgba 0 0 0 0>}
        cornerRadii=<0 0 0 0>
       }
       tintColor {colorUseRule=UseColor_Specified, specifiedColor=<rgba 1 1 1 1>}
       uVRegion {bIsValid=false, max=<0 0>, min=<0 0>}
      }
      verticalBottomSlotImage {
       bIsDynamicallyLoaded=false
       drawAs=NoDrawType
       imageType=NoImage
       mirroring=NoMirror
       resourceName=None
       resourceObject=null
       tiling=NoTile
       imageSize=<0 0>
       margin=<ltrb 0 0 0 0>
       outlineSettings {
        bUseBrushTransparency=false
        roundingType=HalfHeightRadius
        width=0
        color {colorUseRule=UseColor_Specified, specifiedColor=<rgba 0 0 0 0>}
        cornerRadii=<0 0 0 0>
       }
       tintColor {colorUseRule=UseColor_Specified, specifiedColor=<rgba 1 1 1 1>}
       uVRegion {bIsValid=false, max=<0 0>, min=<0 0>}
      }
      verticalTopSlotImage {
       bIsDynamicallyLoaded=false
       drawAs=NoDrawType
       imageType=NoImage
       mirroring=NoMirror
       resourceName=None
       resourceObject=null
       tiling=NoTile
       imageSize=<0 0>
       margin=<ltrb 0 0 0 0>
       outlineSettings {
        bUseBrushTransparency=false
        roundingType=HalfHeightRadius
        width=0
        color {colorUseRule=UseColor_Specified, specifiedColor=<rgba 0 0 0 0>}
        cornerRadii=<0 0 0 0>
       }
       tintColor {colorUseRule=UseColor_Specified, specifiedColor=<rgba 1 1 1 1>}
       uVRegion {bIsValid=false, max=<0 0>, min=<0 0>}
      }
     }
     textStyle {
      overflowPolicy=Clip
      transformPolicy=None
      colorAndOpacity {colorUseRule=UseColor_Foreground, specifiedColor=<rgba 1 0 1 1>}
      font {
       bForceMonospaced=false
       bMaterialIsStencil=false
       fontMaterial=null
       fontObject=&HeadingNow
       letterSpacing=0
       monospacedWidth=1
       size=11.25
       skewAmount=0
       typefaceFontName=76 Bold
       outlineSettings {
        bApplyOutlineToDropShadows=false
        bMiteredCorners=false
        bSeparateFillAlpha=false
        outlineMaterial=null
        outlineSize=0
        outlineColor=<rgba 0 0 0 1>
       }
      }
      highlightColor {colorUseRule=UseColor_ColorTable, specifiedColor=<rgba 0 0 0 0>}
      highlightShape {
       bIsDynamicallyLoaded=false
       drawAs=Image
       imageType=NoImage
       mirroring=NoMirror
       resourceName=None
       resourceObject=null
       tiling=NoTile
       imageSize=<0 0>
       margin=<ltrb 0 0 0 0>
       outlineSettings {
        bUseBrushTransparency=false
        roundingType=HalfHeightRadius
        width=0
        color {colorUseRule=UseColor_Specified, specifiedColor=<rgba 0 0 0 0>}
        cornerRadii=<0 0 0 0>
       }
       tintColor {colorUseRule=UseColor_ColorTable, specifiedColor=<rgba 0 0 0 0>}
       uVRegion {bIsValid=false, max=<0 0>, min=<0 0>}
      }
      selectedBackgroundColor {colorUseRule=UseColor_ColorTable, specifiedColor=<rgba 0 0 0 0>}
      shadowColorAndOpacity=<rgba 0 0 0 1>
      shadowOffset=<0 0>
      strikeBrush {
       bIsDynamicallyLoaded=false
       drawAs=Image
       imageType=NoImage
       mirroring=NoMirror
       resourceName=None
       resourceObject=null
       tiling=NoTile
       imageSize=<32 32>
       margin=<ltrb 0 0 0 0>
       outlineSettings {
        bUseBrushTransparency=false
        roundingType=HalfHeightRadius
        width=0
        color {colorUseRule=UseColor_Specified, specifiedColor=<rgba 0 0 0 0>}
        cornerRadii=<0 0 0 0>
       }
       tintColor {colorUseRule=UseColor_Specified, specifiedColor=<rgba 1 1 1 1>}
       uVRegion {bIsValid=false, max=<0 0>, min=<0 0>}
      }
      underlineBrush {
       bIsDynamicallyLoaded=false
       drawAs=Image
       imageType=NoImage
       mirroring=NoMirror
       resourceName=None
       resourceObject=null
       tiling=NoTile
       imageSize=<32 32>
       margin=<ltrb 0 0 0 0>
       outlineSettings {
        bUseBrushTransparency=false
        roundingType=HalfHeightRadius
        width=0
        color {colorUseRule=UseColor_Specified, specifiedColor=<rgba 0 0 0 0>}
        cornerRadii=<0 0 0 0>
       }
       tintColor {colorUseRule=UseColor_Specified, specifiedColor=<rgba 1 1 1 1>}
       uVRegion {bIsValid=false, max=<0 0>, min=<0 0>}
      }
     }
     vScrollBarPadding=<ltrb 0 0 0 0>
    }
   }
   slot {class=SizeBoxSlot}
  }
  NewSequenceEventAddButton {
   widgetClassPath=&EditorUtilityButton
   parent=HorizontalBox_13
   widget=NewSequenceEventAddButton
   properties {widgetStyle {
     clickedSlateSound {resourceObject=null}
     disabled {
      bIsDynamicallyLoaded=false
      drawAs=RoundedBox
      imageType=NoImage
      mirroring=NoMirror
      resourceName=None
      resourceObject=null
      tiling=NoTile
      imageSize=<32 32>
      margin=<ltrb 0.5 0.5 0.5 0.5>
      outlineSettings {
       bUseBrushTransparency=false
       roundingType=FixedRadius
       width=0
       color {colorUseRule=UseColor_ColorTable, specifiedColor=<rgba 0 0 0 0>}
       cornerRadii=<4 4 4 4>
      }
      tintColor {colorUseRule=UseColor_Specified, specifiedColor=<rgba 0.0395 0.0395 0.0395 1>}
      uVRegion {bIsValid=false, max=<0 0>, min=<0 0>}
     }
     disabledForeground {colorUseRule=UseColor_ColorTable, specifiedColor=<rgba 0 0 0 0>}
     hovered {
      bIsDynamicallyLoaded=false
      drawAs=RoundedBox
      imageType=NoImage
      mirroring=NoMirror
      resourceName=None
      resourceObject=null
      tiling=NoTile
      imageSize=<32 32>
      margin=<ltrb 0 0 0 0>
      outlineSettings {
       bUseBrushTransparency=false
       roundingType=FixedRadius
       width=0
       color {colorUseRule=UseColor_ColorTable, specifiedColor=<rgba 0 0 0 0>}
       cornerRadii=<4 4 4 4>
      }
      tintColor {colorUseRule=UseColor_Specified, specifiedColor=<rgba 0.1126 0.7031 0 1>}
      uVRegion {bIsValid=false, max=<0 0>, min=<0 0>}
     }
     hoveredForeground {colorUseRule=UseColor_ColorTable, specifiedColor=<rgba 0 0 0 0>}
     hoveredSlateSound {resourceObject=null}
     normal {
      bIsDynamicallyLoaded=false
      drawAs=RoundedBox
      imageType=NoImage
      mirroring=NoMirror
      resourceName=None
      resourceObject=null
      tiling=NoTile
      imageSize=<32 32>
      margin=<ltrb 0.5 0.5 0.5 0.5>
      outlineSettings {
       bUseBrushTransparency=false
       roundingType=FixedRadius
       width=0
       color {colorUseRule=UseColor_ColorTable, specifiedColor=<rgba 0 0 0 0>}
       cornerRadii=<4 4 4 4>
      }
      tintColor {colorUseRule=UseColor_Specified, specifiedColor=<rgba 0.0771 0.6167 0.1289 1>}
      uVRegion {bIsValid=false, max=<0 0>, min=<0 0>}
     }
     normalForeground {colorUseRule=UseColor_ColorTable, specifiedColor=<rgba 0 0 0 0>}
     normalPadding=<ltrb 12 1.5 12 1.5>
     pressed {
      bIsDynamicallyLoaded=false
      drawAs=RoundedBox
      imageType=NoImage
      mirroring=NoMirror
      resourceName=None
      resourceObject=null
      tiling=NoTile
      imageSize=<32 32>
      margin=<ltrb 0.5 0.5 0.5 0.5>
      outlineSettings {
       bUseBrushTransparency=false
       roundingType=FixedRadius
       width=0
       color {colorUseRule=UseColor_ColorTable, specifiedColor=<rgba 0 0 0 0>}
       cornerRadii=<4 4 4 4>
      }
      tintColor {colorUseRule=UseColor_Specified, specifiedColor=<rgba 0.1806 0.4583 0.1572 1>}
      uVRegion {bIsValid=false, max=<0 0>, min=<0 0>}
     }
     pressedForeground {colorUseRule=UseColor_ColorTable, specifiedColor=<rgba 0 0 0 0>}
     pressedPadding=<ltrb 12 2.5 12 0.5>
     pressedSlateSound {resourceObject=null}
    }}
   slot {class=HorizontalBoxSlot, properties {verticalAlignment=VAlign_Center, padding=<ltrb 10 0 0 0>}}
  }
  HorizontalBox_14 {
   widgetClassPath=&HorizontalBox
   parent=NewSequenceEventAddButton
   widget=HorizontalBox_14
   slot {class=ButtonSlot, properties {horizontalAlignment=HAlign_Left}}
  }
  Image_14 {
   widgetClassPath=&Image
   parent=HorizontalBox_14
   widget=Image_14
   properties {brush {
     bIsDynamicallyLoaded=false
     drawAs=Image
     imageType=NoImage
     mirroring=NoMirror
     resourceName=None
     resourceObject=&MI_SparksEditorUI_PlusMinus_NoBorder
     tiling=NoTile
     imageSize=<15 15>
     margin=<ltrb 0 0 0 0>
     outlineSettings {
      bUseBrushTransparency=false
      roundingType=HalfHeightRadius
      width=0
      color {colorUseRule=UseColor_Specified, specifiedColor=<rgba 0 0 0 0>}
      cornerRadii=<0 0 0 0>
     }
     tintColor {colorUseRule=UseColor_Specified, specifiedColor=<rgba 1 1 1 1>}
     uVRegion {bIsValid=false, max=<0 0>, min=<0 0>}
    }}
   slot {class=HorizontalBoxSlot, properties {horizontalAlignment=HAlign_Center, verticalAlignment=VAlign_Center, padding=<ltrb 0 0 3 0>}}
  }
  TextBlock_9 {
   widgetClassPath=&TextBlock
   parent=HorizontalBox_14
   widget=TextBlock_9
   properties {
    text=Create and Add
    font {
     bForceMonospaced=false
     bMaterialIsStencil=false
     fontMaterial=null
     fontObject=&HeadingNow
     letterSpacing=0
     monospacedWidth=1
     size=7.5
     skewAmount=0
     typefaceFontName=75 Medium
     outlineSettings {
      bApplyOutlineToDropShadows=false
      bMiteredCorners=false
      bSeparateFillAlpha=false
      outlineMaterial=null
      outlineSize=0
      outlineColor=<rgba 0 0 0 1>
     }
    }
   }
   slot {class=HorizontalBoxSlot, properties {verticalAlignment=VAlign_Center}}
  }
  CreateChangeButton {
   widgetClassPath=&EditorUtilityButton
   parent=HorizontalBox_13
   widget=CreateChangeButton
   properties {widgetStyle {
     clickedSlateSound {resourceObject=null}
     disabled {
      bIsDynamicallyLoaded=false
      drawAs=RoundedBox
      imageType=NoImage
      mirroring=NoMirror
      resourceName=None
      resourceObject=null
      tiling=NoTile
      imageSize=<32 32>
      margin=<ltrb 0.5 0.5 0.5 0.5>
      outlineSettings {
       bUseBrushTransparency=false
       roundingType=FixedRadius
       width=0
       color {colorUseRule=UseColor_ColorTable, specifiedColor=<rgba 0 0 0 0>}
       cornerRadii=<4 4 4 4>
      }
      tintColor {colorUseRule=UseColor_Specified, specifiedColor=<rgba 0.0395 0.0395 0.0395 1>}
      uVRegion {bIsValid=false, max=<0 0>, min=<0 0>}
     }
     disabledForeground {colorUseRule=UseColor_ColorTable, specifiedColor=<rgba 0 0 0 0>}
     hovered {
      bIsDynamicallyLoaded=false
      drawAs=RoundedBox
      imageType=NoImage
      mirroring=NoMirror
      resourceName=None
      resourceObject=null
      tiling=NoTile
      imageSize=<32 32>
      margin=<ltrb 0 0 0 0>
      outlineSettings {
       bUseBrushTransparency=false
       roundingType=FixedRadius
       width=0
       color {colorUseRule=UseColor_ColorTable, specifiedColor=<rgba 0 0 0 0>}
       cornerRadii=<4 4 4 4>
      }
      tintColor {colorUseRule=UseColor_Specified, specifiedColor=<rgba 0.5437 0.7344 0.1989 1>}
      uVRegion {bIsValid=false, max=<0 0>, min=<0 0>}
     }
     hoveredForeground {colorUseRule=UseColor_ColorTable, specifiedColor=<rgba 0 0 0 0>}
     hoveredSlateSound {resourceObject=null}
     normal {
      bIsDynamicallyLoaded=false
      drawAs=RoundedBox
      imageType=NoImage
      mirroring=NoMirror
      resourceName=None
      resourceObject=null
      tiling=NoTile
      imageSize=<32 32>
      margin=<ltrb 0.5 0.5 0.5 0.5>
      outlineSettings {
       bUseBrushTransparency=false
       roundingType=FixedRadius
       width=0
       color {colorUseRule=UseColor_ColorTable, specifiedColor=<rgba 0 0 0 0>}
       cornerRadii=<4 4 4 4>
      }
      tintColor {colorUseRule=UseColor_Specified, specifiedColor=<rgba 0.3885 0.5521 0.0926 1>}
      uVRegion {bIsValid=false, max=<0 0>, min=<0 0>}
     }
     normalForeground {colorUseRule=UseColor_ColorTable, specifiedColor=<rgba 0 0 0 0>}
     normalPadding=<ltrb 12 1.5 12 1.5>
     pressed {
      bIsDynamicallyLoaded=false
      drawAs=RoundedBox
      imageType=NoImage
      mirroring=NoMirror
      resourceName=None
      resourceObject=null
      tiling=NoTile
      imageSize=<32 32>
      margin=<ltrb 0.5 0.5 0.5 0.5>
      outlineSettings {
       bUseBrushTransparency=false
       roundingType=FixedRadius
       width=0
       color {colorUseRule=UseColor_ColorTable, specifiedColor=<rgba 0 0 0 0>}
       cornerRadii=<4 4 4 4>
      }
      tintColor {colorUseRule=UseColor_Specified, specifiedColor=<rgba 0.1806 0.4583 0.1572 1>}
      uVRegion {bIsValid=false, max=<0 0>, min=<0 0>}
     }
     pressedForeground {colorUseRule=UseColor_ColorTable, specifiedColor=<rgba 0 0 0 0>}
     pressedPadding=<ltrb 12 2.5 12 0.5>
     pressedSlateSound {resourceObject=null}
    }}
   slot {class=HorizontalBoxSlot, properties {verticalAlignment=VAlign_Center, padding=<ltrb 10 0 0 0>}}
  }
  HorizontalBox_9 {
   widgetClassPath=&HorizontalBox
   parent=CreateChangeButton
   widget=HorizontalBox_9
   slot {class=ButtonSlot, properties {horizontalAlignment=HAlign_Left}}
  }
  Image_9 {
   widgetClassPath=&Image
   parent=HorizontalBox_9
   widget=Image_9
   properties {brush {
     bIsDynamicallyLoaded=false
     drawAs=Image
     imageType=NoImage
     mirroring=NoMirror
     resourceName=None
     resourceObject=&MI_SparksEditorUI_PlusMinus_NoBorder
     tiling=NoTile
     imageSize=<15 15>
     margin=<ltrb 0 0 0 0>
     outlineSettings {
      bUseBrushTransparency=false
      roundingType=HalfHeightRadius
      width=0
      color {colorUseRule=UseColor_Specified, specifiedColor=<rgba 0 0 0 0>}
      cornerRadii=<0 0 0 0>
     }
     tintColor {colorUseRule=UseColor_Specified, specifiedColor=<rgba 1 1 1 1>}
     uVRegion {bIsValid=false, max=<0 0>, min=<0 0>}
    }}
   slot {class=HorizontalBoxSlot, properties {horizontalAlignment=HAlign_Center, verticalAlignment=VAlign_Center, padding=<ltrb 0 0 3 0>}}
  }
  TextBlock_5 {
   widgetClassPath=&TextBlock
   parent=HorizontalBox_9
   widget=TextBlock_5
   properties {
    text=Create and Change
    font {
     bForceMonospaced=false
     bMaterialIsStencil=false
     fontMaterial=null
     fontObject=&HeadingNow
     letterSpacing=0
     monospacedWidth=1
     size=7.5
     skewAmount=0
     typefaceFontName=75 Medium
     outlineSettings {
      bApplyOutlineToDropShadows=false
      bMiteredCorners=false
      bSeparateFillAlpha=false
      outlineMaterial=null
      outlineSize=0
      outlineColor=<rgba 0 0 0 1>
     }
    }
   }
   slot {class=HorizontalBoxSlot, properties {verticalAlignment=VAlign_Center}}
  }
  Border_5 {
   widgetClassPath=&Border
   parent=EditorUtilityScrollBox_2
   widget=Border_5
   properties {
    background {
     bIsDynamicallyLoaded=false
     drawAs=Image
     imageType=NoImage
     mirroring=NoMirror
     resourceName=None
     resourceObject=null
     tiling=NoTile
     imageSize=<32 32>
     margin=<ltrb 0 0 0 0>
     outlineSettings {
      bUseBrushTransparency=false
      roundingType=HalfHeightRadius
      width=0
      color {colorUseRule=UseColor_Specified, specifiedColor=<rgba 0 0 0 0>}
      cornerRadii=<0 0 0 0>
     }
     tintColor {colorUseRule=UseColor_Specified, specifiedColor=<rgba 0.0729 0.0729 0.0729 1>}
     uVRegion {bIsValid=false, max=<0 0>, min=<0 0>}
    }
    padding=<ltrb 10 2 4 2>
   }
   slot {class=ScrollBoxSlot, properties {padding=<ltrb 0 4 0 5>}}
  }
  HorizontalBox_8 {
   widgetClassPath=&HorizontalBox
   parent=Border_5
   widget=HorizontalBox_8
   slot {class=BorderSlot, properties {padding=<ltrb 10 2 4 2>}}
  }
  DANamePreviewTitle {
   widgetClassPath=&TextBlock
   parent=HorizontalBox_8
   widget=DANamePreviewTitle
   properties {
    text="Data Asset Name Preview: "
    font {
     bForceMonospaced=false
     bMaterialIsStencil=false
     fontMaterial=null
     fontObject=&HeadingNow
     letterSpacing=0
     monospacedWidth=1
     size=9.75
     skewAmount=0
     typefaceFontName=75 Medium
     outlineSettings {
      bApplyOutlineToDropShadows=false
      bMiteredCorners=false
      bSeparateFillAlpha=false
      outlineMaterial=null
      outlineSize=0
      outlineColor=<rgba 0 0 0 1>
     }
    }
   }
   slot {class=HorizontalBoxSlot}
  }
  DANamePreviewText {
   widgetClassPath=&TextBlock
   bIsVariable=true
   parent=HorizontalBox_8
   widget=DANamePreviewText
   properties {
    colorAndOpacity {colorUseRule=UseColor_Specified, specifiedColor=<rgba 0.4062 0.4062 0.4062 1>}
    font {
     bForceMonospaced=false
     bMaterialIsStencil=false
     fontMaterial=null
     fontObject=&HeadingNow
     letterSpacing=0
     monospacedWidth=1
     size=9.75
     skewAmount=0
     typefaceFontName=74 Regular
     outlineSettings {
      bApplyOutlineToDropShadows=false
      bMiteredCorners=false
      bSeparateFillAlpha=false
      outlineMaterial=null
      outlineSize=0
      outlineColor=<rgba 0 0 0 1>
     }
    }
   }
   slot {class=HorizontalBoxSlot}
  }
  NameDuplicatedWarningImage {
   widgetClassPath=&Image
   bIsVariable=true
   parent=HorizontalBox_8
   widget=NameDuplicatedWarningImage
   properties {
    toolTipText=Asset already exist please choose other name.
    visibility=Collapsed
    brush {
     bIsDynamicallyLoaded=false
     drawAs=Image
     imageType=FullColor
     mirroring=NoMirror
     resourceName=None
     resourceObject=/PilgrimQuickplay/StageManager/NeverCook/EditorUtilities/StageManagerEditor/Textures/T_SparksEditorUI_IconLibrary_Warning_64
     tiling=NoTile
     imageSize=<16 16>
     margin=<ltrb 0 0 0 0>
     outlineSettings {
      bUseBrushTransparency=false
      roundingType=HalfHeightRadius
      width=0
      color {colorUseRule=UseColor_Specified, specifiedColor=<rgba 0 0 0 0>}
      cornerRadii=<0 0 0 0>
     }
     tintColor {colorUseRule=UseColor_Specified, specifiedColor=<rgba 1 1 1 1>}
     uVRegion {bIsValid=false, max=<0 0>, min=<0 0>}
    }
    colorAndOpacity=<rgba 0.9531 0.8592 0 1>
   }
   slot {class=HorizontalBoxSlot, properties {verticalAlignment=VAlign_Center, padding=<ltrb 10 0 0 2>}}
  }
  bg {
   widgetClassPath=&Image
   parent=NewSequenceEventsInsideGroup
   widget=bg
   properties {visibility=Collapsed}
   slot {class=OverlaySlot, properties {horizontalAlignment=HAlign_Fill, verticalAlignment=VAlign_Fill}}
  }
  Nosequenceeventstext {
   widgetClassPath=&TextBlock
   bIsVariable=true
   parent=Overlay
   widget=Nosequenceeventstext
   properties {
    justification=Center
    text=All avaliable events are in use.
    visibility=Collapsed
    font {
     bForceMonospaced=false
     bMaterialIsStencil=false
     fontMaterial=null
     fontObject=&Roboto
     letterSpacing=0
     monospacedWidth=1
     size=15
     skewAmount=0
     typefaceFontName=Bold
     outlineSettings {
      bApplyOutlineToDropShadows=false
      bMiteredCorners=false
      bSeparateFillAlpha=false
      outlineMaterial=null
      outlineSize=0
      outlineColor=<rgba 0 0 0 1>
     }
    }
   }
   slot {class=OverlaySlot, properties {horizontalAlignment=HAlign_Center, verticalAlignment=VAlign_Center}}
  }
  CloseNewSequenceEventWindowButton {
   widgetClassPath=&EditorUtilityButton
   parent=Overlay
   widget=CloseNewSequenceEventWindowButton
   properties {
    clipping=ClipToBounds
    widgetStyle {
     clickedSlateSound {resourceObject=null}
     disabled {
      bIsDynamicallyLoaded=false
      drawAs=RoundedBox
      imageType=NoImage
      mirroring=NoMirror
      resourceName=None
      resourceObject=null
      tiling=NoTile
      imageSize=<20 20>
      margin=<ltrb 0 0 0 0>
      outlineSettings {
       bUseBrushTransparency=false
       roundingType=FixedRadius
       width=0
       color {colorUseRule=UseColor_ColorTable, specifiedColor=<rgba 0 0 0 0>}
       cornerRadii=<4 4 4 4>
      }
      tintColor {colorUseRule=UseColor_Specified, specifiedColor=<rgba 0.0395 0.0395 0.0395 1>}
      uVRegion {bIsValid=false, max=<0 0>, min=<0 0>}
     }
     disabledForeground {colorUseRule=UseColor_ColorTable, specifiedColor=<rgba 0 0 0 0>}
     hovered {
      bIsDynamicallyLoaded=false
      drawAs=RoundedBox
      imageType=NoImage
      mirroring=NoMirror
      resourceName=None
      resourceObject=null
      tiling=NoTile
      imageSize=<32 32>
      margin=<ltrb 0 0 0 0>
      outlineSettings {
       bUseBrushTransparency=false
       roundingType=FixedRadius
       width=0
       color {colorUseRule=UseColor_ColorTable, specifiedColor=<rgba 0 0 0 0>}
       cornerRadii=<4 4 4 4>
      }
      tintColor {colorUseRule=UseColor_Specified, specifiedColor=<rgba 1 0.1252 0.253 1>}
      uVRegion {bIsValid=false, max=<0 0>, min=<0 0>}
     }
     hoveredForeground {colorUseRule=UseColor_ColorTable, specifiedColor=<rgba 0 0 0 0>}
     hoveredSlateSound {resourceObject=null}
     normal {
      bIsDynamicallyLoaded=false
      drawAs=RoundedBox
      imageType=NoImage
      mirroring=NoMirror
      resourceName=None
      resourceObject=null
      tiling=NoTile
      imageSize=<20 20>
      margin=<ltrb 0 0 0 0>
      outlineSettings {
       bUseBrushTransparency=false
       roundingType=FixedRadius
       width=0
       color {colorUseRule=UseColor_ColorTable, specifiedColor=<rgba 0 0 0 0>}
       cornerRadii=<4 4 4 4>
      }
      tintColor {colorUseRule=UseColor_Specified, specifiedColor=<rgba 0.775 0 0.0304 1>}
      uVRegion {bIsValid=false, max=<0 0>, min=<0 0>}
     }
     normalForeground {colorUseRule=UseColor_ColorTable, specifiedColor=<rgba 0 0 0 0>}
     normalPadding=<ltrb 12 1.5 12 1.5>
     pressed {
      bIsDynamicallyLoaded=false
      drawAs=RoundedBox
      imageType=NoImage
      mirroring=NoMirror
      resourceName=None
      resourceObject=null
      tiling=NoTile
      imageSize=<20 20>
      margin=<ltrb 0 0 0 0>
      outlineSettings {
       bUseBrushTransparency=false
       roundingType=FixedRadius
       width=0
       color {colorUseRule=UseColor_ColorTable, specifiedColor=<rgba 0 0 0 0>}
       cornerRadii=<4 4 4 4>
      }
      tintColor {colorUseRule=UseColor_Specified, specifiedColor=<rgba 0.2396 0 0 1>}
      uVRegion {bIsValid=false, max=<0 0>, min=<0 0>}
     }
     pressedForeground {colorUseRule=UseColor_ColorTable, specifiedColor=<rgba 0 0 0 0>}
     pressedPadding=<ltrb 12 2.5 12 0.5>
     pressedSlateSound {resourceObject=null}
    }
   }
   slot {class=OverlaySlot, properties {horizontalAlignment=HAlign_Right, padding=<ltrb 10 3.5 10 0>}}
  }
  TextBlock_7 {
   widgetClassPath=&TextBlock
   parent=CloseNewSequenceEventWindowButton
   widget=TextBlock_7
   properties {
    text=X
    font {
     bForceMonospaced=false
     bMaterialIsStencil=false
     fontMaterial=null
     fontObject=&HeadingNow
     letterSpacing=0
     monospacedWidth=1
     size=7.5
     skewAmount=0
     typefaceFontName=86 Bold
     outlineSettings {
      bApplyOutlineToDropShadows=false
      bMiteredCorners=false
      bSeparateFillAlpha=false
      outlineMaterial=null
      outlineSize=0
      outlineColor=<rgba 0 0 0 1>
     }
    }
   }
   slot {class=ButtonSlot, properties {horizontalAlignment=HAlign_Fill, padding=<ltrb 0 0 0 0>}}
  }
  DeleteConfirmationPopup {
   widgetClassPath=&/EWBP_SparksDeleteConfirmation
   bIsVariable=true
   parent=Overlay_0
   widget=DeleteConfirmationPopup
   properties {visibility=Collapsed}
   slot {class=OverlaySlot, properties {horizontalAlignment=HAlign_Center, padding=<ltrb 0 180 0 0>}}
  }
 ]
}
