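# Translucent UI material that draws a procedural plus or minus icon for the Stage Manager editor utility.
# Two MF_UI_Box distance fields, the second with swizzled size, are unioned with Min to form a plus; isMinus fades out the vertical bar.
# The combined field feeds an anti-aliasing function with a threshold parameter and a soft SmoothStep falloff used in a Hovered-driven blend.
# An IconColor parameter is also exposed.
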Material /PilgrimQuickplay/StageManager/NeverCook/EditorUtilities/StageManagerEditor/Textures/M_SparksEditorUI_PlusMinus
graph {
 properties {blendMode=BLEND_Translucent, materialDomain=MD_UI, pixelDepthOffsetMode=PDOM_Legacy, usedShadingModels=Default Lit}
 nodes [
  n1 = TextureCoordinate()
  n2 = Constant2Vector() {g=0.2, r=0.85}
  n3 = MaterialFunctionCall(XY=@n2) {materialFunction=&Swizzle}
  n4 = Min(
   A=MaterialFunctionCall(UVs=@n1, Size=@n2) {materialFunction=&MF_UI_Box}.SDF
   B=LinearInterpolate(
    A=MaterialFunctionCall(UVs=@n1, Size=@n3.YX) {materialFunction=&MF_UI_Box}.SDF
    Alpha=ScalarParameter() {parameterName=isMinus}
   )
  )
  n5 = OneMinus(None=SmoothStep(Value=Subtract(A=@n4) {constB=0.1}) {constMax=0.2, constMin=-0.2})
  LinearInterpolate(B=Multiply(A=Saturate(None=@n5), B=NamedRerouteUsage()), Alpha=ScalarParameter() {parameterName=Hovered})
  MaterialFunctionCall(DistanceField=Add(A=@n4) {constB=0.5}, Threshold=ScalarParameter() {defaultValue=0.5, parameterName=SDF_Treshold}) {materialFunction=&MF_UI_AntiAliasing}
  VectorParameter() {parameterName=IconColor, defaultValue=<rgba 1 1 1 1>}
 ]
}
